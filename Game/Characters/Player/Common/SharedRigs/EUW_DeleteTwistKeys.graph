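# Editor utility widget that deletes float keys on "_twist_" Control Rig channels in Sequencer, either all or within a frame range.
# On construct it caches the focused Level Sequence's track matching Primary Control Rig, which this graph never sets.
# Delete gathers twist channels from all track sections, then collects and removes keys, filtering by display-rate frame.
# It begins an undo transaction without ending it and never clears its key or channel arrays.
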EditorUtilityWidgetBlueprint /Game/Characters/Player/Common/SharedRigs/EUW_DeleteTwistKeys
graphs [EventGraph {nodes [
   n1 = "On Check State Changed (SpecifiedRangeCheckBox)"!() -> (OutputDelegate, then, bIsChecked:bool) {is=title K2Node_ComponentBoundEvent}
   n2 = Branch(
    execute=[@n23.then, @n1.then]
    Condition=IsChecked(self=Get SpecifiedRangeCheckBox() -> (SpecifiedRangeCheckBox:Editor Utility Check Box Object Reference))
   )
   n3 = Get CustomRangeDetail() -> (CustomRangeDetail:Details View Object Reference)
   > SetVisibility(self=@n3, InVisibility=Visible)
   SetVisibility(execute=@n2.else, self=@n3, InVisibility=Collapsed)
   "On Clicked (Delete_Button)"() {is=title K2Node_ComponentBoundEvent}
   > BeginTransaction(Context=DeleteTwistKeys, Description="LOCTEXT(\"Match Button clicked\")")
   > n4 = GetSections!(execute, Track:Movie Scene Track Object Reference=Get Track() -> (Track:Movie Scene Control Rig Parameter Track Object Reference)) -> (then, ReturnValue:Array of Movie Scene Section Object References)
   > n5 = For Each Loop(Array:Array of Movie Scene Section Object References=@n4) -> (Array Element:Movie Scene Section Object Reference)
   n6 = For Each Loop(
    Exec=@n5.Completed
    Array:Array of Movie Scene Scripting Channel Object References=Get Twist Control Channels() -> (Twist Control Channels:Array of Movie Scene Scripting Channel Object References)
   ) -> (Array Element:Movie Scene Scripting Channel Object Reference)
   > n7 = Cast To MovieSceneScriptingFloatChannel!(execute, Object=@n6) -> (then, CastFailed, AsMovie Scene Scripting Float Channel:Movie Scene Scripting Float Channel Object Reference) {is=title K2Node_DynamicCast}
   > n8 = For Each Loop(Array:Array of Movie Scene Scripting Key Object References="Get Keys (Float)"!(self:Movie Scene Scripting Float Channel Object Reference=@n7) -> (ReturnValue:Array of Movie Scene Scripting Key Object References) {is=title K2Node_CallFunction}) -> (Array Element:Movie Scene Scripting Key Object Reference)
   n9 = For Each Loop(Exec=@n8.Completed, Array=Get Keys To Delete() -> (Keys To Delete:Array of Movie Scene Scripting Key Object References))
   > "Remove Key (Float)"!(execute, self:Movie Scene Scripting Float Channel Object Reference=@n7, Key:Movie Scene Scripting Key Object Reference=@n9) -> (then)
   n10 = Cast To MovieSceneScriptingFloatKey!(execute=@n8.LoopBody, Object=@n8) -> (then, CastFailed, AsMovie Scene Scripting Float Key:Movie Scene Scripting Float Key Object Reference) {is=title K2Node_DynamicCast}
   > n11 = Branch(Condition=IsChecked(self=Get AllCheckBox() -> (AllCheckBox:Editor Utility Check Box Object Reference)))
   > Add(
    TargetArray:Array of Movie Scene Scripting Key Object References=Get Keys To Delete()
    NewItem:"Movie Scene Scripting Key Object Reference (by ref)"=@n10
   ) {is=title K2Node_CallArrayFunction}
   Branch(execute=@n11.else, Condition=IsChecked(self=Get SpecifiedRangeCheckBox()))
   n12 = "Get Time (Float)"!(self:Movie Scene Scripting Float Key Object Reference=@n10, TimeUnit:EMovieSceneTimeUnit Enum=DisplayRate) -> (ReturnValue_FrameNumber_Value:int, ReturnValue_SubFrame:float=~) {is=title K2Node_CallFunction}
   n13 = Get Start Range() -> (Start Range:int=0)
   n14 = Get End Range() -> (End Range:int=0)
   > Branch(Condition=AND Boolean(A="integer >= integer"(A=@n12, B=@n13), B="integer <= integer"(A=@n12, B=@n14)))
   > Add(TargetArray=Get Keys To Delete(), NewItem=@n10)
   n15 = GetAllChannels!(execute=@n5.LoopBody, Section:Movie Scene Section Object Reference=@n5) -> (then, ReturnValue:Array of Movie Scene Scripting Channel Object References)
   > n16 = For Each Loop(Array:Array of Movie Scene Scripting Channel Object References=@n15) -> (Array Element:Movie Scene Scripting Channel Object Reference)
   n17 = GetObjectName(Object=@n16)
   > Branch(Condition=Contains(SearchIn=Split(SourceString=@n17, InStr=., SearchDir=FromStart), Substring=_twist_, bUseCase=true))
   > Add Unique(
    TargetArray:Array of Movie Scene Scripting Channel Object References=Get Twist Control Channels()
    NewItem:"Movie Scene Scripting Channel Object Reference (by ref)"=@n16
   )
   Event Construct()
   > n18 = GetFocusedLevelSequence!(execute) -> (then, ReturnValue:Level Sequence Object Reference)
   > Set Frame Rate(Frame Rate:Frame Rate Structure=MakeFrameRate!(Numerator:int=30, Denominator:int=1) -> (ReturnValue:Frame Rate Structure)) -> (Output_Get:Frame Rate Structure)
   > n19 = GetControlRigs!(execute, LevelSequence:Level Sequence Object Reference=@n18) -> (then, ReturnValue:Array of Control Rig Sequencer Binding Proxy Structures)
   > n20 = For Each Loop(Array:Array of Control Rig Sequencer Binding Proxy Structures=@n19) -> (Array Element:Control Rig Sequencer Binding Proxy Structure)
   n21 = Break Control Rig Sequencer Binding Proxy!(ControlRigSequencerBindingProxy:"Control Rig Sequencer Binding Proxy Structure (by ref)"=@n20) -> (
    Proxy:Movie Scene Binding Proxy Structure
    ControlRig:Control Rig Object Reference
    Track:Movie Scene Control Rig Parameter Track Object Reference
   ) {is=title K2Node_BreakStruct}
   n22 = Get Primary Control Rig() -> (Primary Control Rig:Control Rig Object Reference)
   > Branch(Condition="Equal (Object)"(A=@n21.ControlRig, B=@n22))
   > Set Track(Track:Movie Scene Control Rig Parameter Track Object Reference=@n21.Track) -> (Output_Get:Movie Scene Control Rig Parameter Track Object Reference)
   > Set Rig Hierarchy(Rig Hierarchy:Rig Hierarchy Object Reference=GetHierarchy!(self:Control Rig Object Reference=@n22) -> (ReturnValue:Rig Hierarchy Object Reference)) -> (Output_Get:Rig Hierarchy Object Reference)
   Event PreConstruct()
   > n23 = SetObject(self=Get CustomRangeDetail(), NewObject=Self-Reference())
  ]}]
variables [
 End Range {type=integer, default=150}
 Frame Rate {type=FrameRate, default {denominator=0, numerator=0}}
 Keys To Delete {type="array<MovieSceneScriptingKey>", default []}
 Playback Range {
  type=SequencerScriptingRange
  default {bHasEnd=false, bHasStart=false, exclusiveEnd=0, inclusiveStart=0, internalRate {denominator=1, numerator=60000}}
 }
 Primary Control Rig {type=ControlRig, default=null}
 Range {
  type=SequencerScriptingRange
  default {bHasEnd=false, bHasStart=false, exclusiveEnd=0, inclusiveStart=0, internalRate {denominator=1, numerator=60000}}
 }
 Rig Hierarchy {type=RigHierarchy, default=null}
 Start Range {type=integer}
 Track {type=MovieSceneControlRigParameterTrack, default=null}
 Twist Control Channels {type="array<MovieSceneScriptingChannel>", default []}
]
widgets {
 info {namedSlotCount=2, rootWidgetClass=&VerticalBox, widgetCount=14}
 widgets [
  VerticalBox_18 {widgetClassPath=&VerticalBox, widget=VerticalBox_18}
  FrameRate_EditorUtilityExpandableArea {
   widgetClassPath=/Script/Blutility.EditorUtilityExpandableArea
   bIsVariable=true
   parent=VerticalBox_18
   widget=FrameRate_EditorUtilityExpandableArea
   properties {
    bIsExpanded=true
    borderBrush {
     bIsDynamicallyLoaded=false
     drawAs=RoundedBox
     imageType=NoImage
     mirroring=NoMirror
     resourceName=None
     resourceObject=null
     tiling=NoTile
     imageSize=<0 0>
     margin=<ltrb 0 0 0 0>
     outlineSettings {
      bUseBrushTransparency=false
      roundingType=FixedRadius
      width=0
      color {colorUseRule=UseColor_Specified, specifiedColor=<rgba 0 0 0 0>}
      cornerRadii=<4 4 4 4>
     }
     tintColor {colorUseRule=UseColor_Specified, specifiedColor=<rgba 0 0 0 1>}
     uVRegion {bIsValid=false, max=<0 0>, min=<0 0>}
    }
   }
   slot {class=VerticalBoxSlot}
  }
  TextBlock {
   widgetClassPath=&TextBlock
   namedSlotHost=FrameRate_EditorUtilityExpandableArea
   widget=TextBlock
   properties {
    text=Frame Range
    font {
     bForceMonospaced=false
     bMaterialIsStencil=false
     fontMaterial=null
     fontObject=&Roboto
     letterSpacing=0
     monospacedWidth=1
     size=15
     skewAmount=0
     typefaceFontName=Italic
     outlineSettings {
      bApplyOutlineToDropShadows=false
      bMiteredCorners=false
      bSeparateFillAlpha=false
      outlineMaterial=null
      outlineSize=0
      outlineColor=<rgba 0 0 0 1>
     }
    }
   }
  }
  VerticalBox_137 {widgetClassPath=&VerticalBox, namedSlotHost=FrameRate_EditorUtilityExpandableArea, widget=VerticalBox_137}
  HorizontalBox_2 {
   widgetClassPath=&HorizontalBox
   parent=VerticalBox_137
   widget=HorizontalBox_2
   slot {class=VerticalBoxSlot, properties {horizontalAlignment=HAlign_Center, size {sizeRule=Fill, value=1}}}
  }
  AllCheckBox {
   widgetClassPath=&EditorUtilityCheckBox
   bIsVariable=true
   parent=HorizontalBox_2
   widget=AllCheckBox
   properties {checkedState=Checked}
   slot {class=HorizontalBoxSlot, properties {padding=<ltrb 5 5 5 5>}}
  }
  CurrentFrame_TextBlock {
   widgetClassPath=&TextBlock
   parent=AllCheckBox
   widget=CurrentFrame_TextBlock
   properties {
    text=All
    font {
     bForceMonospaced=false
     bMaterialIsStencil=false
     fontMaterial=null
     fontObject=&Roboto
     letterSpacing=0
     monospacedWidth=1
     size=13.5
     skewAmount=0
     typefaceFontName=Regular
     outlineSettings {
      bApplyOutlineToDropShadows=false
      bMiteredCorners=false
      bSeparateFillAlpha=false
      outlineMaterial=null
      outlineSize=0
      outlineColor=<rgba 0 0 0 1>
     }
    }
   }
   slot {class=PanelSlot}
  }
  SpecifiedRangeCheckBox {
   widgetClassPath=&EditorUtilityCheckBox
   bIsVariable=true
   parent=HorizontalBox_2
   widget=SpecifiedRangeCheckBox
   properties {horizontalAlignment=HAlign_Center}
   slot {class=HorizontalBoxSlot, properties {horizontalAlignment=HAlign_Center, padding=<ltrb 8 8 8 8>}}
  }
  LegR_TextBlock {
   widgetClassPath=&TextBlock
   parent=SpecifiedRangeCheckBox
   widget=LegR_TextBlock
   properties {
    text=Specified Range
    font {
     bForceMonospaced=false
     bMaterialIsStencil=false
     fontMaterial=null
     fontObject=&Roboto
     letterSpacing=0
     monospacedWidth=1
     size=13.5
     skewAmount=0
     typefaceFontName=Regular
     outlineSettings {
      bApplyOutlineToDropShadows=false
      bMiteredCorners=false
      bSeparateFillAlpha=false
      outlineMaterial=null
      outlineSize=0
      outlineColor=<rgba 0 0 0 1>
     }
    }
   }
   slot {class=PanelSlot}
  }
  CustomRangeDetail {
   widgetClassPath=&DetailsView
   parent=VerticalBox_137
   widget=CustomRangeDetail
   properties {categoriesToShow [Custom Range]}
   slot {class=VerticalBoxSlot}
  }
  HorizontalBox_43 {widgetClassPath=&HorizontalBox, parent=VerticalBox_18, widget=HorizontalBox_43, slot {class=VerticalBoxSlot}}
  Delete_Button {
   widgetClassPath=&EditorUtilityButton
   parent=HorizontalBox_43
   widget=Delete_Button
   properties {widgetStyle {
     clickedSlateSound {resourceObject=null}
     disabled {
      bIsDynamicallyLoaded=false
      drawAs=RoundedBox
      imageType=NoImage
      mirroring=NoMirror
      resourceName=None
      resourceObject=null
      tiling=NoTile
      imageSize=<0 0>
      margin=<ltrb 0 0 0 0>
      outlineSettings {
       bUseBrushTransparency=false
       roundingType=FixedRadius
       width=1
       color {colorUseRule=UseColor_ColorTable, specifiedColor=<rgba 0 0 0 0>}
       cornerRadii=<4 4 4 4>
      }
      tintColor {colorUseRule=UseColor_Specified, specifiedColor=<rgba 0.0395 0.0395 0.0395 1>}
      uVRegion {bIsValid=false, max=<0 0>, min=<0 0>}
     }
     disabledForeground {colorUseRule=UseColor_ColorTable, specifiedColor=<rgba 0 0 0 0>}
     hovered {
      bIsDynamicallyLoaded=false
      drawAs=RoundedBox
      imageType=NoImage
      mirroring=NoMirror
      resourceName=None
      resourceObject=null
      tiling=NoTile
      imageSize=<32 32>
      margin=<ltrb 0 0 0 0>
      outlineSettings {
       bUseBrushTransparency=false
       roundingType=FixedRadius
       width=1
       color {colorUseRule=UseColor_ColorTable, specifiedColor=<rgba 0 0 0 0>}
       cornerRadii=<4 4 4 4>
      }
      tintColor {colorUseRule=UseColor_Specified, specifiedColor=<rgba 0.0953 0.0953 0.0953 1>}
      uVRegion {bIsValid=false, max=<0 0>, min=<0 0>}
     }
     hoveredForeground {colorUseRule=UseColor_ColorTable, specifiedColor=<rgba 0 0 0 0>}
     hoveredSlateSound {resourceObject=null}
     normal {
      bIsDynamicallyLoaded=false
      drawAs=RoundedBox
      imageType=NoImage
      mirroring=NoMirror
      resourceName=None
      resourceObject=null
      tiling=NoTile
      imageSize=<32 32>
      margin=<ltrb 0 0 0 0>
      outlineSettings {
       bUseBrushTransparency=false
       roundingType=FixedRadius
       width=1
       color {colorUseRule=UseColor_ColorTable, specifiedColor=<rgba 0 0 0 0>}
       cornerRadii=<4 4 4 4>
      }
      tintColor {colorUseRule=UseColor_Specified, specifiedColor=<rgba 0.0833 0.0017 0.0017 1>}
      uVRegion {bIsValid=false, max=<0 0>, min=<0 0>}
     }
     normalForeground {colorUseRule=UseColor_ColorTable, specifiedColor=<rgba 0 0 0 0>}
     normalPadding=<ltrb 12 1.5 12 1.5>
     pressed {
      bIsDynamicallyLoaded=false
      drawAs=RoundedBox
      imageType=NoImage
      mirroring=NoMirror
      resourceName=None
      resourceObject=null
      tiling=NoTile
      imageSize=<32 32>
      margin=<ltrb 0 0 0 0>
      outlineSettings {
       bUseBrushTransparency=false
       roundingType=FixedRadius
       width=1
       color {colorUseRule=UseColor_ColorTable, specifiedColor=<rgba 0 0 0 0>}
       cornerRadii=<4 4 4 4>
      }
      tintColor {colorUseRule=UseColor_Specified, specifiedColor=<rgba 0.0284 0.0284 0.0284 1>}
      uVRegion {bIsValid=false, max=<0 0>, min=<0 0>}
     }
     pressedForeground {colorUseRule=UseColor_ColorTable, specifiedColor=<rgba 0 0 0 0>}
     pressedPadding=<ltrb 12 2.5 12 0.5>
     pressedSlateSound {resourceObject=null}
    }}
   slot {class=HorizontalBoxSlot, properties {size {sizeRule=Fill, value=1}}}
  }
  TextBlock_53 {
   widgetClassPath=&TextBlock
   parent=Delete_Button
   widget=TextBlock_53
   properties {
    text=Delete
    font {
     bForceMonospaced=false
     bMaterialIsStencil=false
     fontMaterial=null
     fontObject=&Roboto
     letterSpacing=0
     monospacedWidth=1
     size=15
     skewAmount=0
     typefaceFontName=Bold
     outlineSettings {
      bApplyOutlineToDropShadows=false
      bMiteredCorners=false
      bSeparateFillAlpha=false
      outlineMaterial=null
      outlineSize=0
      outlineColor=<rgba 0 0 0 1>
     }
    }
   }
   slot {class=ButtonSlot}
  }
  HorizontalBox {widgetClassPath=&HorizontalBox, parent=VerticalBox_18, widget=HorizontalBox, slot {class=VerticalBoxSlot}}
 ]
}
