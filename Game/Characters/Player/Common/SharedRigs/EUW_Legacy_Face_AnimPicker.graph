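# Editor picker widget for animating Fortnite's legacy face Control Rig in Sequencer, with sliders and buttons over a face image.
# On load, it finds the legacy face rig sharing the passed binding's bound object.
# Each picker child binds to the control matching its widget name, and sliders are cached.
# On load and mouse enter, sliders refresh values and selection from the rig at the current frame, falling back to the passed rig.
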EditorUtilityWidgetBlueprint /Game/Characters/Player/Common/SharedRigs/EUW_Legacy_Face_AnimPicker
graphs [
 EventGraph {nodes [
   n1 = Event Widget has loaded!() -> (
    OutputDelegate
    then
    Control Rig Proxy Binding:Control Rig Sequencer Binding Proxy Structure
    Animatable Creature:Animatable Creature Object Reference
    Anim Picker EUW:Animation Picker Object Reference
   ) {is=title K2Node_Event}
   > Parent: Widget has loaded!(
    execute
    Control Rig Proxy Binding:Control Rig Sequencer Binding Proxy Structure=@n1.Control Rig Proxy Binding
    Animatable Creature:Animatable Creature Object Reference=@n1.Animatable Creature
    Anim Picker EUW:Animation Picker Object Reference=@n1.Anim Picker EUW
   ) -> (then) {is=title K2Node_CallParentFunction}
   > n2 = GetFocusedLevelSequence!(execute) -> (then, ReturnValue:Level Sequence Object Reference)
   > n3 = Set LS(LS:Level Sequence Object Reference=@n2) -> (Output_Get:Level Sequence Object Reference)
   > n4 = GetControlRigs!(execute, LevelSequence:Level Sequence Object Reference=@n3) -> (then, ReturnValue:Array of Control Rig Sequencer Binding Proxy Structures)
   > n5 = For Each Loop(Array:Array of Control Rig Sequencer Binding Proxy Structures=@n4) -> (Array Element:Control Rig Sequencer Binding Proxy Structure)
   n6 = For Each Loop(
    Exec=@n5.Completed
    Array:Array of Widget Object References=GetAllChildren(self=Get CanvasPanel_0() -> (CanvasPanel_0:Canvas Panel Object Reference))
   ) -> (Array Element:Widget Object Reference)
   Update Sliders based on CR(execute=@n6.Completed)
   Branch(
    execute=@n6.LoopBody
    Condition=Does Object Implement Interface!(
     TestObject:Object Reference=@n6
     Interface:Interface Interface=/AnimInEdTools/AnimationPicker/PickerElements/BPI_SelectablePickerElements.BPI_SelectablePickerElements_C
    ) -> (ReturnValue:bool)
   )
   n7 = Make Array("[0]":Name=String To Name(InString=GetDisplayName(Object=@n6))) -> (Array:Array of Names)
   > Set Initial Vars!(
    execute:Exec
    self:Object Reference=@n6
    Control Rig:Control Rig Object Reference=Get Face CR() -> (Face CR:Control Rig Object Reference)
    Level Sequence:Level Sequence Object Reference=Get LS() -> (LS:Level Sequence Object Reference)
    Control Bindings:Map of Control Rig Object References to Animation Picker String Array Struct Structures=Make Map!(
     Key 0:Control Rig Object Reference=Get Face CR()
     Value 0:Animation Picker String Array Struct Structure=Make Animation Picker String Array Struct!(ControlNames_6_#1:Array of Names=@n7) -> (Animation_Picker_StringArray_Struct:Animation Picker String Array Struct Structure) {is=title K2Node_MakeStruct}
    ) -> (Map:Map of Control Rig Object References to Animation Picker String Array Struct Structures) {is=title K2Node_MakeMap}
    Anim Picker EUW:Animation Picker Object Reference=Get Anim Picker EUW() -> (Anim Picker EUW:Animation Picker Object Reference)
    Level Actors:Array of Movie Scene Binding Proxy Structures
   ) -> (then:Exec) {is=title K2Node_Message}
   > n8 = Cast To EUW_Picker_1DSlider!(execute, Object=@n6) -> (then, CastFailed, AsEUW Picker 1DSlider:EUW Picker 1DSlider Object Reference) {is=title K2Node_DynamicCast}
   > Add(
    TargetArray:Array of EUW Picker 1DSlider Object References=Get All 1DSlider() -> (All 1DSlider:Array of EUW Picker 1DSlider Object References)
    NewItem:"EUW Picker 1DSlider Object Reference (by ref)"=@n8
   ) {is=title K2Node_CallArrayFunction}
   n9 = Cast To EUW_Picker_2DSlider!(execute=@n8.CastFailed, Object=@n6) -> (then, CastFailed, AsEUW Picker 2DSlider:EUW Picker 2DSlider Object Reference) {is=title K2Node_DynamicCast}
   > Add(
    TargetArray:Array of EUW Picker 2DSlider Object References=Get All 2DSlider() -> (All 2DSlider:Array of EUW Picker 2DSlider Object References)
    NewItem:"EUW Picker 2DSlider Object Reference (by ref)"=@n9
   )
   n10 = Break Control Rig Sequencer Binding Proxy!(ControlRigSequencerBindingProxy:"Control Rig Sequencer Binding Proxy Structure (by ref)"=@n5) -> (
    Proxy:Movie Scene Binding Proxy Structure
    ControlRig:Control Rig Object Reference
    Track:Movie Scene Control Rig Parameter Track Object Reference
   ) {is=title K2Node_BreakStruct}
   Branch(
    execute=@n5.LoopBody
    Condition="Equal (Class)"(
     A=Get Class(Object=@n10.ControlRig) -> (ReturnValue:Control Rig Class Reference)
     B=/Game/Characters/Player/Common/Fortnite_Base_Head/Facials/ControlRigs/Fortnite_Legacy_Face_CtrlRig.Fortnite_Legacy_Face_CtrlRig_C
    )
   )
   > n11 = GetBindingID!(
    execute
    Sequence:Movie Scene Sequence Object Reference=Get LS()
    InBinding:"Movie Scene Binding Proxy Structure (by ref)"=Break Control Rig Sequencer Binding Proxy(ControlRigSequencerBindingProxy=Get Control Rig Proxy Binding() -> (Control Rig Proxy Binding:Control Rig Sequencer Binding Proxy Structure))
   ) -> (then, ReturnValue:Movie Scene Object Binding ID Structure)
   > n12 = GetBindingID(Sequence=Get LS(), InBinding=@n10)
   n13 = GetBoundObjects!(ObjectBinding:Movie Scene Object Binding ID Structure=@n12) -> (ReturnValue:Array of Object References)
   n14 = GetBoundObjects(ObjectBinding=@n11)
   > Branch(Condition="Equal (Object)"(A="Get (a copy)"(Array:Array of Object References=@n13) -> (Output:Object Reference), B="Get (a copy)"(Array=@n14)))
   > Set Face CR(Face CR:Control Rig Object Reference=@n10.ControlRig) -> (Output_Get:Control Rig Object Reference)
   n15 = Event Mouse Enter!() -> (OutputDelegate, then, Mouse Event:Pointer Event Structure) {is=title K2Node_Event}
   > Parent: Mouse Enter!(execute, Mouse Event:Pointer Event Structure=@n15.Mouse Event) -> (then) {is=title K2Node_CallParentFunction}
   > Update Sliders based on CR()
  ]}
 Get Primary Panel {nodes [
   Get Primary Panel() {is=title K2Node_FunctionEntry}
   > Return Node!(execute, Panel:Panel Widget Object Reference=Get CanvasPanel_0())
  ]}
 Set Bool Var {nodes [
   n1 = Set Bool Var!() -> (then, Variable Name:Name, Value:bool=~) {is=title K2Node_FunctionEntry}
   > SetVariableFromString!(
    execute
    self:Rig VMHost Object Reference=Break Control Rig Sequencer Binding Proxy(ControlRigSequencerBindingProxy=Get Control Rig Proxy Binding()).ControlRig
    InVariableName:"Name (by ref)"=@n1
    InValue:String="To String (Boolean)"(InBool=@n1.Value)
   ) -> (then, ReturnValue:bool)
  ]}
 Update Sliders based on CR {nodes [
   Update Sliders based on CR() {is=title K2Node_FunctionEntry}
   > n1 = GetFocusedLevelSequence()
   > Set LS(LS=@n1)
   > n2 = Get Current Local Time!(execute, TimeUnit:EMovieSceneTimeUnit Enum=DisplayRate) -> (then, ReturnValue:Movie Scene Sequence Playback Params Structure) {is=title K2Node_CallFunction}
   > Set Current Frame Number(Current Frame Number:Frame Number Structure=Make Frame Number!(Value:int=Break Movie Scene Sequence Playback Params!(MovieSceneSequencePlaybackParams:"Movie Scene Sequence Playback Params Structure (by ref)"=@n2) -> (
      Frame_FrameNumber_Value:int
      Frame_SubFrame:float=~
      Time:float=~
      MarkedFrame:String
      Timecode:Timecode Structure
      PositionType:EMovieScenePositionType Enum
      UpdateMethod:EUpdatePositionMethod Enum
      bHasJumped:bool=~
     ) {is=title K2Node_BreakStruct}) -> (FrameNumber:Frame Number Structure) {is=title K2Node_MakeStruct}) -> (Output_Get:Frame Number Structure)
   > n3 = Get!(execute:Exec) -> (then:Exec, else:Exec, Face CR:Control Rig Object Reference)
   > n4 = Set CR(CR:Control Rig Object Reference=@n3) -> (Output_Get:Control Rig Object Reference)
   n5 = For Each Loop(Exec=[@n11.then, @n4.then], Array:Array of EUW Picker 1DSlider Object References=Get All 1DSlider()) -> (Array Element:EUW Picker 1DSlider Object Reference)
   n6 = For Each Loop(Exec=@n5.Completed, Array:Array of EUW Picker 2DSlider Object References=Get All 2DSlider()) -> (Array Element:EUW Picker 2DSlider Object Reference)
   Return Node!(execute=@n6.Completed)
   n7 = String To Name(InString=GetObjectName(Object=@n6))
   > n8 = GetLocalControlRigVector2D!(
    execute
    LevelSequence:Level Sequence Object Reference=Get LS()
    ControlRig:Control Rig Object Reference=Get CR() -> (CR:Control Rig Object Reference)
    ControlName:Name=@n7
    Frame:Frame Number Structure=Get Current Frame Number() -> (Current Frame Number:Frame Number Structure)
    TimeUnit:EMovieSceneTimeUnit Enum=DisplayRate
   ) -> (then, ReturnValue:Vector 2D Structure)
   > SetValue!(
    execute
    self:Synth 2DSlider Object Reference=Get Slider!(self:EUW Picker 2DSlider Object Reference=@n6) -> (Slider:Synth 2DSlider Object Reference)
    InValue:Vector 2D Structure=Convert to Slider 2D Relative Value!(self:EUW Picker 2DSlider Object Reference=@n6, Value:Vector 2D Structure=@n8) -> (ReturnValue:Vector 2D Structure)
   ) -> (then)
   > Update Selection Display!(
    execute:Exec
    self:Object Reference=@n6
    Selected:bool=IsControlSelected!(self:Control Rig Object Reference=Get CR(), InControlName:"Name (by ref)"=@n7) -> (ReturnValue:bool)
   ) -> (then:Exec, Out Selected:bool) {is=title K2Node_Message}
   n9 = String To Name(InString=GetObjectName(Object=@n5))
   n10 = GetLocalControlRigFloat!(
    execute=@n5.LoopBody
    LevelSequence:Level Sequence Object Reference=Get LS()
    ControlRig:Control Rig Object Reference=Get CR()
    ControlName:Name=@n9
    Frame:Frame Number Structure=Get Current Frame Number()
    TimeUnit:EMovieSceneTimeUnit Enum=DisplayRate
   ) -> (then, ReturnValue:float)
   > SetValue(
    self:Slider Object Reference=Get Slider(self:EUW Picker 1DSlider Object Reference=@n5) -> (Slider:Editor Utility Slider Object Reference)
    InValue:float=@n10
   )
   > Update Selection Display(self=@n5, Selected=IsControlSelected(self=Get CR(), InControlName=@n9))
   n11 = Set CR(execute=@n3.else, CR=Break Control Rig Sequencer Binding Proxy(ControlRigSequencerBindingProxy=Get Control Rig Proxy Binding()).ControlRig)
  ]}
]
variables [
 All 1DSlider {type="array<EUW_Picker_1DSlider_C>", default []}
 All 2DSlider {type="array<EUW_Picker_2DSlider_C>", default []}
 Face CR {type=ControlRig, default=null}
 Left Arm FK {type="array<EUW_Picker_Button_C>", default []}
 Left Arm IK {type="array<EUW_Picker_Button_C>", default []}
 Left Leg FK {type="array<EUW_Picker_Button_C>", default []}
 Left Leg IK {type="array<EUW_Picker_Button_C>", default []}
 LS {type=LevelSequence, default=null}
 Neck FK {type="array<EUW_Picker_Button_C>", default []}
 Neck IK {type="array<EUW_Picker_Button_C>", default []}
 Right Arm FK {type="array<EUW_Picker_Button_C>", default []}
 Right Arm IK {type="array<EUW_Picker_Button_C>", default []}
 Right Leg FK {type="array<EUW_Picker_Button_C>", default []}
 Right Leg IK {type="array<EUW_Picker_Button_C>", default []}
 Spine FK {type="array<EUW_Picker_Button_C>", default []}
 Spine IK {type="array<EUW_Picker_Button_C>", default []}
]
widgets {
 info {
  parentClass=/AnimInEdTools/AnimationPicker/EUW_Asset_Picker_Base.EUW_Asset_Picker_Base_C
  rootWidgetClass=&CanvasPanel
  widgetCount=32
 }
 widgets [
  CanvasPanel_0 {widgetClassPath=&CanvasPanel, bIsVariable=true, widget=CanvasPanel_0, properties {renderTransformPivot=<0 0>}}
  R_Eyelid {
   widgetClassPath=/AnimInEdTools/AnimationPicker/PickerElements/EUW_Picker_2DSlider
   bIsVariable=true
   parent=CanvasPanel_0
   widget=R_Eyelid
   properties {bOverride_Cursor=true, max Value Y=-1, min Value Y=1, colorAndOpacity=<rgba 1 0.0521 0.0739 1>}
   slot {
    class=CanvasPanelSlot
    properties {layoutData {alignment=<0.5 0.5>, anchors {maximum=<0.5 0.5>, minimum=<0.5 0.5>}, offsets=<ltrb -313.4423 -156.422 55.0373 56.237>}}
   }
  }
  L_Eyelid {
   widgetClassPath=&EUW_Picker_2DSlider
   bIsVariable=true
   parent=CanvasPanel_0
   widget=L_Eyelid
   properties {
    bOverride_Cursor=true
    max Value Y=-1
    min Value Y=1
    colorAndOpacity=<rgba 0.2613 1 0.1894 1>
    renderTransform {angle=0, scale=<-1 1>, shear=<0 0>, translation=<0 0>}
   }
   slot {
    class=CanvasPanelSlot
    properties {layoutData {alignment=<0.5 0.5>, anchors {maximum=<0.5 0.5>, minimum=<0.5 0.5>}, offsets=<ltrb -53.4423 -152.422 55.0373 56.237>}}
   }
  }
  Look {
   widgetClassPath=&EUW_Picker_2DSlider
   bIsVariable=true
   parent=CanvasPanel_0
   widget=Look
   properties {
    bOverride_Cursor=true
    max Value Y=-1
    min Value Y=1
    colorAndOpacity=<rgba 1 0.8354 0 1>
    renderTransform {angle=0, scale=<-1 1>, shear=<0 0>, translation=<0 0>}
   }
   slot {
    class=CanvasPanelSlot
    properties {layoutData {alignment=<0.5 0.5>, anchors {maximum=<0.5 0.5>, minimum=<0.5 0.5>}, offsets=<ltrb -185.4423 -156.422 70 70>}}
   }
  }
  Jaw {
   widgetClassPath=&EUW_Picker_2DSlider
   bIsVariable=true
   parent=CanvasPanel_0
   widget=Jaw
   properties {
    bOverride_Cursor=true
    max Value Y=-1
    min Value Y=1
    colorAndOpacity=<rgba 1 0.8564 0 1>
    renderTransform {angle=0, scale=<-1 1>, shear=<0 0>, translation=<0 0>}
   }
   slot {
    class=CanvasPanelSlot
    properties {layoutData {alignment=<0.5 0.5>, anchors {maximum=<0.5 0.5>, minimum=<0.5 0.5>}, offsets=<ltrb -184.9609 243.4595 120 120>}}
   }
  }
  Jaw_FwdBk {
   widgetClassPath=&EUW_Picker_2DSlider
   bIsVariable=true
   parent=CanvasPanel_0
   widget=Jaw_FwdBk
   properties {bOverride_Cursor=true, max Value X=0, max Value Y=-1, min Value X=0, min Value Y=1, colorAndOpacity=<rgba 1 0.8564 0 1>}
   slot {
    class=CanvasPanelSlot
    properties {layoutData {alignment=<0.5 0.5>, anchors {maximum=<0.5 0.5>, minimum=<0.5 0.5>}, offsets=<ltrb -353.16 243.4595 20 120>}}
   }
  }
  Tongue {
   widgetClassPath=&EUW_Picker_2DSlider
   bIsVariable=true
   parent=CanvasPanel_0
   widget=Tongue
   properties {bOverride_Cursor=true, max Value Y=-1, min Value Y=1, colorAndOpacity=<rgba 1 0.8564 0 1>}
   slot {
    class=CanvasPanelSlot
    properties {layoutData {alignment=<0.5 0.5>, anchors {maximum=<0.5 0.5>, minimum=<0.5 0.5>}, offsets=<ltrb -37.2743 261.1135 79.3733 75.308>}}
   }
  }
  L_Lipcorner {
   widgetClassPath=&EUW_Picker_2DSlider
   bIsVariable=true
   parent=CanvasPanel_0
   widget=L_Lipcorner
   properties {
    bOverride_Cursor=true
    max Value Y=-1
    min Value Y=1
    colorAndOpacity=<rgba 0.2613 1 0.1894 1>
    renderTransform {angle=0, scale=<-1 1>, shear=<0 0>, translation=<236.6654 47.19>}
   }
   slot {
    class=CanvasPanelSlot
    properties {layoutData {alignment=<0.5 0.5>, anchors {maximum=<0.5 0.5>, minimum=<0.5 0.5>}, offsets=<ltrb -301.2743 33.1135 79.3733 75.308>}}
   }
  }
  L_Cheek {
   widgetClassPath=&EUW_Picker_2DSlider
   bIsVariable=true
   parent=CanvasPanel_0
   widget=L_Cheek
   properties {
    bOverride_Cursor=true
    max Value Y=-1
    min Value Y=1
    colorAndOpacity=<rgba 0.2613 1 0.1894 1>
    renderTransform {angle=0, scale=<-1 1>, shear=<0 0>, translation=<236.6654 47.19>}
   }
   slot {
    class=CanvasPanelSlot
    properties {layoutData {alignment=<0.5 0.5>, anchors {maximum=<0.5 0.5>, minimum=<0.5 0.5>}, offsets=<ltrb -248 -102.8865 79.3733 75.308>}}
   }
  }
  Mouth {
   widgetClassPath=&EUW_Picker_2DSlider
   bIsVariable=true
   parent=CanvasPanel_0
   widget=Mouth
   properties {
    bOverride_Cursor=true
    max Value X=0
    max Value Y=-1
    min Value X=0
    min Value Y=1
    colorAndOpacity=<rgba 1 0.8354 0 1>
    renderTransform {angle=0, scale=<1 1>, shear=<0 0>, translation=<236.6654 47.19>}
   }
   slot {
    class=CanvasPanelSlot
    properties {layoutData {alignment=<0.5 0.5>, anchors {maximum=<0.5 0.5>, minimum=<0.5 0.5>}, offsets=<ltrb -423.8088 22.3527 20 53.7865>}}
   }
  }
  Lip_Upper {
   widgetClassPath=&EUW_Picker_2DSlider
   bIsVariable=true
   parent=CanvasPanel_0
   widget=Lip_Upper
   properties {
    bOverride_Cursor=true
    max Value X=0
    max Value Y=-1
    min Value X=0
    min Value Y=1
    colorAndOpacity=<rgba 1 0.8354 0 1>
    renderTransform {angle=0, scale=<1 1>, shear=<0 0>, translation=<236.6654 47.19>}
   }
   slot {
    class=CanvasPanelSlot
    properties {layoutData {alignment=<0.5 0.5>, anchors {maximum=<0.5 0.5>, minimum=<0.5 0.5>}, offsets=<ltrb -423.8088 -45.6473 20 53.7865>}}
   }
  }
  Lip_Lower {
   widgetClassPath=&EUW_Picker_2DSlider
   bIsVariable=true
   parent=CanvasPanel_0
   widget=Lip_Lower
   properties {
    bOverride_Cursor=true
    max Value X=0
    max Value Y=-1
    min Value X=0
    min Value Y=1
    colorAndOpacity=<rgba 1 0.8354 0 1>
    renderTransform {angle=0, scale=<1 -1>, shear=<0 0>, translation=<236.6654 47.19>}
   }
   slot {
    class=CanvasPanelSlot
    properties {layoutData {alignment=<0.5 0.5>, anchors {maximum=<0.5 0.5>, minimum=<0.5 0.5>}, offsets=<ltrb -423.8088 94.3527 20 53.7865>}}
   }
  }
  L_Lip_Lower {
   widgetClassPath=&EUW_Picker_2DSlider
   bIsVariable=true
   parent=CanvasPanel_0
   widget=L_Lip_Lower
   properties {
    bOverride_Cursor=true
    max Value X=0
    max Value Y=-1
    min Value X=0
    min Value Y=1
    colorAndOpacity=<rgba 0.2613 1 0.1894 1>
    renderTransform {angle=0, scale=<1 -1>, shear=<0 0>, translation=<236.6654 47.19>}
   }
   slot {
    class=CanvasPanelSlot
    properties {layoutData {alignment=<0.5 0.5>, anchors {maximum=<0.5 0.5>, minimum=<0.5 0.5>}, offsets=<ltrb -367.8088 74.3527 20 53.7865>}}
   }
  }
  R_Lip_Lower {
   widgetClassPath=&EUW_Picker_2DSlider
   bIsVariable=true
   parent=CanvasPanel_0
   widget=R_Lip_Lower
   properties {
    bOverride_Cursor=true
    max Value X=0
    max Value Y=-1
    min Value X=0
    min Value Y=1
    colorAndOpacity=<rgba 1 0.0521 0.0739 1>
    renderTransform {angle=0, scale=<1 -1>, shear=<0 0>, translation=<236.6654 47.19>}
   }
   slot {
    class=CanvasPanelSlot
    properties {layoutData {alignment=<0.5 0.5>, anchors {maximum=<0.5 0.5>, minimum=<0.5 0.5>}, offsets=<ltrb -487.8088 74.3527 20 53.7865>}}
   }
  }
  L_Lip_Upper {
   widgetClassPath=&EUW_Picker_2DSlider
   bIsVariable=true
   parent=CanvasPanel_0
   widget=L_Lip_Upper
   properties {
    bOverride_Cursor=true
    max Value X=0
    max Value Y=-1
    min Value X=0
    min Value Y=1
    colorAndOpacity=<rgba 0.2613 1 0.1894 1>
    renderTransform {angle=0, scale=<1 1>, shear=<0 0>, translation=<236.6654 47.19>}
   }
   slot {
    class=CanvasPanelSlot
    properties {layoutData {alignment=<0.5 0.5>, anchors {maximum=<0.5 0.5>, minimum=<0.5 0.5>}, offsets=<ltrb -363.8088 -9.6473 20 53.7865>}}
   }
  }
  R_Lip_Upper {
   widgetClassPath=&EUW_Picker_2DSlider
   bIsVariable=true
   parent=CanvasPanel_0
   widget=R_Lip_Upper
   properties {
    bOverride_Cursor=true
    max Value X=0
    max Value Y=-1
    min Value X=0
    min Value Y=1
    colorAndOpacity=<rgba 1 0.0521 0.0739 1>
    renderTransform {angle=0, scale=<1 1>, shear=<0 0>, translation=<236.6654 47.19>}
   }
   slot {
    class=CanvasPanelSlot
    properties {layoutData {alignment=<0.5 0.5>, anchors {maximum=<0.5 0.5>, minimum=<0.5 0.5>}, offsets=<ltrb -495.8088 -9.6473 20 53.7865>}}
   }
  }
  R_Lipcorner {
   widgetClassPath=&EUW_Picker_2DSlider
   bIsVariable=true
   parent=CanvasPanel_0
   widget=R_Lipcorner
   properties {
    bOverride_Cursor=true
    max Value Y=-1
    min Value Y=1
    colorAndOpacity=<rgba 1 0.0521 0.0739 1>
    renderTransform {angle=0, scale=<1 1>, shear=<0 0>, translation=<-334.9066 58.201>}
   }
   slot {
    class=CanvasPanelSlot
    properties {layoutData {alignment=<0.5 0.5>, anchors {maximum=<0.5 0.5>, minimum=<0.5 0.5>}, offsets=<ltrb 14.7257 25.1135 79.3733 75.308>}}
   }
  }
  R_Cheek {
   widgetClassPath=&EUW_Picker_2DSlider
   bIsVariable=true
   parent=CanvasPanel_0
   widget=R_Cheek
   properties {
    bOverride_Cursor=true
    max Value Y=-1
    min Value Y=1
    colorAndOpacity=<rgba 1 0.0521 0.0739 1>
    renderTransform {angle=0, scale=<1 1>, shear=<0 0>, translation=<-334.9066 58.201>}
   }
   slot {
    class=CanvasPanelSlot
    properties {layoutData {alignment=<0.5 0.5>, anchors {maximum=<0.5 0.5>, minimum=<0.5 0.5>}, offsets=<ltrb -21.2743 -108 79.3733 75.308>}}
   }
  }
  midBrow {
   widgetClassPath=&EUW_Picker_2DSlider
   bIsVariable=true
   parent=CanvasPanel_0
   widget=midBrow
   properties {
    bOverride_Cursor=true
    max Value X=0
    max Value Y=-1
    min Value X=0
    min Value Y=1
    colorAndOpacity=<rgba 1 0.8354 0 1>
    renderTransform {angle=0, scale=<1 1>, shear=<0 0>, translation=<-334.9066 58.201>}
   }
   slot {
    class=CanvasPanelSlot
    properties {layoutData {alignment=<0.5 0.5>, anchors {maximum=<0.5 0.5>, minimum=<0.5 0.5>}, offsets=<ltrb 149.0391 -360 20 75.308>}}
   }
  }
  L_Brow {
   widgetClassPath=&EUW_Picker_2DSlider
   bIsVariable=true
   parent=CanvasPanel_0
   widget=L_Brow
   properties {
    bOverride_Cursor=true
    max Value Y=-1
    min Value Y=1
    colorAndOpacity=<rgba 0.2613 1 0.1894 1>
    renderTransform {angle=0, scale=<-1 1>, shear=<0 0>, translation=<-334.9066 58.201>}
   }
   slot {
    class=CanvasPanelSlot
    properties {layoutData {alignment=<0.5 0.5>, anchors {maximum=<0.5 0.5>, minimum=<0.5 0.5>}, offsets=<ltrb 244.0391 -360 75 75.308>}}
   }
  }
  Nose {
   widgetClassPath=&EUW_Picker_2DSlider
   bIsVariable=true
   parent=CanvasPanel_0
   widget=Nose
   properties {
    bOverride_Cursor=true
    max Value X=0
    max Value Y=-1
    min Value X=0
    min Value Y=1
    colorAndOpacity=<rgba 1 0.8354 0 1>
    renderTransform {angle=0, scale=<1 1>, shear=<0 0>, translation=<-334.9066 58.201>}
   }
   slot {
    class=CanvasPanelSlot
    properties {layoutData {alignment=<0.5 0.5>, anchors {maximum=<0.5 0.5>, minimum=<0.5 0.5>}, offsets=<ltrb 146.5391 -131.521 20 74.039>}}
   }
  }
  R_Brow {
   widgetClassPath=&EUW_Picker_2DSlider
   bIsVariable=true
   parent=CanvasPanel_0
   widget=R_Brow
   properties {
    bOverride_Cursor=true
    max Value Y=-1
    min Value Y=1
    colorAndOpacity=<rgba 1 0.0521 0.0739 1>
    renderTransform {angle=0, scale=<1 1>, shear=<0 0>, translation=<-334.9066 58.201>}
   }
   slot {
    class=CanvasPanelSlot
    properties {layoutData {alignment=<0.5 0.5>, anchors {maximum=<0.5 0.5>, minimum=<0.5 0.5>}, offsets=<ltrb 60.0391 -360 75 75.308>}}
   }
  }
  Graphics {
   widgetClassPath=&CanvasPanel
   parent=CanvasPanel_0
   widget=Graphics
   properties {renderTransform {angle=0, scale=<1 1>, shear=<0 0>, translation=<7.0547 -1.4777>}}
   slot {
    class=CanvasPanelSlot
    properties {zOrder=-1, layoutData {alignment=<0.5 0.5>, anchors {maximum=<1 1>, minimum=<0 0>}, offsets=<ltrb 0 0 0 0>}}
   }
  }
  Face_Image {
   widgetClassPath=&Image
   bIsVariable=true
   parent=Graphics
   widget=Face_Image
   properties {
    brush {
     bIsDynamicallyLoaded=false
     drawAs=Image
     imageType=FullColor
     mirroring=NoMirror
     resourceName=None
     resourceObject=/Game/Characters/Player/Common/SharedRigs/Graphics/Picker_Biped_Jonesy_Face
     tiling=NoTile
     imageSize=<412 611>
     margin=<ltrb 0 0 0 0>
     outlineSettings {
      bUseBrushTransparency=false
      roundingType=HalfHeightRadius
      width=0
      color {colorUseRule=UseColor_Specified, specifiedColor=<rgba 0 0 0 0>}
      cornerRadii=<0 0 0 0>
     }
     tintColor {colorUseRule=UseColor_Specified, specifiedColor=<rgba 1 1 1 1>}
     uVRegion {bIsValid=false, max=<0 0>, min=<0 0>}
    }
    colorAndOpacity=<rgba 1 1 1 0.4>
    renderTransform {angle=0, scale=<1 1>, shear=<0 0>, translation=<-3.575 -3.0507>}
   }
   slot {
    class=CanvasPanelSlot
    properties {layoutData {alignment=<0.5 0.5>, anchors {maximum=<0.5 0.5>, minimum=<0.5 0.5>}, offsets=<ltrb -196.2385 -60.9804 681.4449 839.1203>}}
   }
  }
  TextBlock_1 {
   widgetClassPath=&TextBlock
   parent=Graphics
   widget=TextBlock_1
   properties {
    text=Tongue
    colorAndOpacity {colorUseRule=UseColor_Specified, specifiedColor=<rgba 1 1 1 0.5>}
    font {
     bForceMonospaced=false
     bMaterialIsStencil=false
     fontMaterial=null
     fontObject=&Roboto
     letterSpacing=0
     monospacedWidth=1
     size=18.75
     skewAmount=0
     typefaceFontName=Bold
     outlineSettings {
      bApplyOutlineToDropShadows=false
      bMiteredCorners=false
      bSeparateFillAlpha=false
      outlineMaterial=null
      outlineSize=0
      outlineColor=<rgba 0 0 0 1>
     }
    }
   }
   slot {
    class=CanvasPanelSlot
    properties {bAutoSize=true, layoutData {alignment=<0 0>, anchors {maximum=<0.5 0.5>, minimum=<0.5 0.5>}, offsets=<ltrb -84.9609 311.4595 100 30>}}
   }
  }
  TextBlock {
   widgetClassPath=&TextBlock
   parent=Graphics
   widget=TextBlock
   properties {
    text=Jaw
    colorAndOpacity {colorUseRule=UseColor_Specified, specifiedColor=<rgba 1 1 1 0.5>}
    font {
     bForceMonospaced=false
     bMaterialIsStencil=false
     fontMaterial=null
     fontObject=&Roboto
     letterSpacing=0
     monospacedWidth=1
     size=18.75
     skewAmount=0
     typefaceFontName=Bold
     outlineSettings {
      bApplyOutlineToDropShadows=false
      bMiteredCorners=false
      bSeparateFillAlpha=false
      outlineMaterial=null
      outlineSize=0
      outlineColor=<rgba 0 0 0 1>
     }
    }
   }
   slot {
    class=CanvasPanelSlot
    properties {bAutoSize=true, layoutData {alignment=<0 0>, anchors {maximum=<0.5 0.5>, minimum=<0.5 0.5>}, offsets=<ltrb -212.9609 311.4595 100 30>}}
   }
  }
  TextBlock_2 {
   widgetClassPath=&TextBlock
   parent=Graphics
   widget=TextBlock_2
   properties {
    text=Face Mstr Ctrl
    colorAndOpacity {colorUseRule=UseColor_Specified, specifiedColor=<rgba 1 1 1 0.5>}
    font {
     bForceMonospaced=false
     bMaterialIsStencil=false
     fontMaterial=null
     fontObject=&Roboto
     letterSpacing=0
     monospacedWidth=1
     size=18.75
     skewAmount=0
     typefaceFontName=Bold
     outlineSettings {
      bApplyOutlineToDropShadows=false
      bMiteredCorners=false
      bSeparateFillAlpha=false
      outlineMaterial=null
      outlineSize=0
      outlineColor=<rgba 0 0 0 1>
     }
    }
   }
   slot {
    class=CanvasPanelSlot
    properties {bAutoSize=true, layoutData {alignment=<0 0>, anchors {maximum=<0.5 0.5>, minimum=<0.5 0.5>}, offsets=<ltrb -272.9609 415.4595 100 30>}}
   }
  }
  TextBlock_5 {
   widgetClassPath=&TextBlock
   parent=Graphics
   widget=TextBlock_5
   properties {
    text=Jaw FwdBk
    colorAndOpacity {colorUseRule=UseColor_Specified, specifiedColor=<rgba 1 1 1 0.5>}
    font {
     bForceMonospaced=false
     bMaterialIsStencil=false
     fontMaterial=null
     fontObject=&Roboto
     letterSpacing=0
     monospacedWidth=1
     size=18.75
     skewAmount=0
     typefaceFontName=Bold
     outlineSettings {
      bApplyOutlineToDropShadows=false
      bMiteredCorners=false
      bSeparateFillAlpha=false
      outlineMaterial=null
      outlineSize=0
      outlineColor=<rgba 0 0 0 1>
     }
    }
   }
   slot {
    class=CanvasPanelSlot
    properties {bAutoSize=true, layoutData {alignment=<0 0>, anchors {maximum=<0.5 0.5>, minimum=<0.5 0.5>}, offsets=<ltrb -424.9609 315.4595 100 30>}}
   }
  }
  Face_Ctrl_Master {
   widgetClassPath=/AnimInEdTools/AnimationPicker/EUW_Picker_Button
   bIsVariable=true
   parent=CanvasPanel_0
   widget=Face_Ctrl_Master
   properties {draw Style=Square, label Font Size=8, normal Tint {colorUseRule=UseColor_Specified, specifiedColor=<rgba 0 0.0108 0.4156 1>}}
   slot {
    class=CanvasPanelSlot
    properties {layoutData {alignment=<0.5 0.5>, anchors {maximum=<0.5 0.5>, minimum=<0.5 0.5>}, offsets=<ltrb -186.9609 393.4595 100 20>}}
   }
  }
  C_Eye_Aim {
   widgetClassPath=&EUW_Picker_Button
   bIsVariable=true
   parent=CanvasPanel_0
   widget=C_Eye_Aim
   properties {
    label Text=Aim
    colorAndOpacity=<rgba 1 0.9695 0 1>
    label Color {colorUseRule=UseColor_Specified, specifiedColor=<rgba 0 0 0 1>}
    normal Tint {colorUseRule=UseColor_Specified, specifiedColor=<rgba 0.6771 0.6771 0.6771 1>}
   }
   slot {
    class=CanvasPanelSlot
    properties {layoutData {alignment=<0.5 0.5>, anchors {maximum=<0.5 0.5>, minimum=<0.5 0.5>}, offsets=<ltrb -182.7069 -222.1633 76.5081 36.7544>}}
   }
  }
  L_Eye_Aim {
   widgetClassPath=&EUW_Picker_Button
   bIsVariable=true
   parent=CanvasPanel_0
   widget=L_Eye_Aim
   properties {
    draw Style=Circle
    label Font Size=6
    colorAndOpacity=<rgba 0.2613 1 0.1894 1>
    normal Tint {colorUseRule=UseColor_Specified, specifiedColor=<rgba 0.6094 0.6094 0.6094 1>}
   }
   slot {
    class=CanvasPanelSlot
    properties {zOrder=2, layoutData {alignment=<0.5 0.5>, anchors {maximum=<0.5 0.5>, minimum=<0.5 0.5>}, offsets=<ltrb -157.4609 -221.0405 15 15>}}
   }
  }
  R_Eye_Aim {
   widgetClassPath=&EUW_Picker_Button
   bIsVariable=true
   parent=CanvasPanel_0
   widget=R_Eye_Aim
   properties {
    draw Style=Circle
    label Font Size=6
    colorAndOpacity=<rgba 1 0.0521 0.0739 1>
    normal Tint {colorUseRule=UseColor_Specified, specifiedColor=<rgba 0.5625 0.5625 0.5625 1>}
   }
   slot {
    class=CanvasPanelSlot
    properties {zOrder=2, layoutData {alignment=<0.5 0.5>, anchors {maximum=<0.5 0.5>, minimum=<0.5 0.5>}, offsets=<ltrb -205.4609 -221.0405 15 15>}}
   }
  }
 ]
}
