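# Editor utility picker widget for selecting Fortnite on-face Control Rig controls while animating in Sequencer.
# Construct groups canvas buttons by face region for region, upper/lower and select-all selection.
# On load, it finds the face rig sharing the picker binding's bound object and initializes buttons by widget name.
# Scene updates sync listed sliders from current-frame rig values; Tick shows eye-aim controls only when lookAtSwitch is 1.
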EditorUtilityWidgetBlueprint /Game/Characters/Player/Common/SharedRigs/EUW_MainCharacte_OnFace_AnimPicker
graphs [
 EventGraph {nodes [
   "On Clicked (Select_All_Face)"() {is=title K2Node_ComponentBoundEvent}
   > BeginTransaction(Context=AnimationPicker, Description="LOCTEXT(\"\\\"Select All\\\" button pressed\")")
   n1 = Get Face CR() -> (Face CR:Control Rig Object Reference)
   > n2 = For Each Loop!(
    Exec
    Array:Array of Rig Element Key Structures=Get Controls!(
     self:Rig Hierarchy Object Reference=GetHierarchy!(self:Control Rig Object Reference=@n1) -> (ReturnValue:Rig Hierarchy Object Reference)
     bTraverse:bool=true
    ) -> (ReturnValue:Array of Rig Element Key Structures) {is=title K2Node_CallFunction}
   ) -> (LoopBody, Array Element_Type:ERigElementType Enum, Array Element_Name:Name, Array Index, Completed)
   EndTransaction(execute=@n2.Completed)
   SelectControl!(
    execute=@n2.LoopBody
    self:Control Rig Object Reference=@n1
    InControlName:"Name (by ref)"=@n2.Array Element_Name
    bSelect:bool=true
    bSetupUndo:bool=false
   ) -> (then)
   "On Clicked (Region_L_Brow_Btn)"() {is=title K2Node_ComponentBoundEvent}
   > Select Multiple Buttons Deprecated!(
    execute
    self:EUW Asset Picker Base Object Reference
    Buttons:Array of EUW Picker Button Object References=Get L Brow Region Buttons() -> (L Brow Region Buttons:Array of EUW Picker Button Object References)
    Control Rig:Control Rig Object Reference=Get Face CR()
   ) -> (then)
   "On Clicked (Select_Lower_Face)"() {is=title K2Node_ComponentBoundEvent}
   > Select Multiple Buttons Deprecated(
    Buttons=Get Lower Face Canvas Buttons() -> (Lower Face Canvas Buttons:Array of EUW Picker Button Object References)
    Control Rig=Get Face CR()
   )
   "On Clicked (Region_C_Jaw_Btn)"() {is=title K2Node_ComponentBoundEvent}
   > Select Multiple Buttons Deprecated(
    Buttons=Get Jaw And Mouth Canvas Buttons() -> (Jaw And Mouth Canvas Buttons:Array of EUW Picker Button Object References)
    Control Rig=Get Face CR()
   )
   "On Clicked (Region_L_Eye_Btn)"() {is=title K2Node_ComponentBoundEvent}
   > Select Multiple Buttons Deprecated(Buttons=Get L Eye Region Buttons() -> (L Eye Region Buttons:Array of EUW Picker Button Object References), Control Rig=Get Face CR())
   "On Clicked (Region_R_Eye_Btn)"() {is=title K2Node_ComponentBoundEvent}
   > Select Multiple Buttons Deprecated(Buttons=Get R Eye Region Buttons() -> (R Eye Region Buttons:Array of EUW Picker Button Object References), Control Rig=Get Face CR())
   "On Clicked (Region_R_Brow_Btn)"() {is=title K2Node_ComponentBoundEvent}
   > Select Multiple Buttons Deprecated(Buttons=Get R Brow Region Buttons() -> (R Brow Region Buttons:Array of EUW Picker Button Object References), Control Rig=Get Face CR())
   "On Clicked (Region_C_Mouth_Btn)"() {is=title K2Node_ComponentBoundEvent}
   > Select Multiple Buttons Deprecated(Buttons=Get Mouth Region Buttons() -> (Mouth Region Buttons:Array of EUW Picker Button Object References), Control Rig=Get Face CR())
   "On Clicked (Select_Upper_Face)"() {is=title K2Node_ComponentBoundEvent}
   > Select Multiple Buttons Deprecated(
    Buttons=Get Upper Face Canvas Buttons() -> (Upper Face Canvas Buttons:Array of EUW Picker Button Object References)
    Control Rig=Get Face CR()
   )
   Event Construct()
   > n3 = Sequence!(execute) -> (then_0, then_1, then_2:Exec, then_3:Exec, then_4:Exec, then_5:Exec, then_6:Exec, then_7:Exec)
   > Set Main Canvas Buttons(Main Canvas Buttons:Array of EUW Picker Button Object References=Get Picker Buttons In Canvas!(Canvas:Canvas Panel Object Reference=Get CanvasPanel_0() -> (CanvasPanel_0:Canvas Panel Object Reference)) -> (Picker Buttons:Array of EUW Picker Button Object References) {is=title K2Node_CallFunction}) -> (Output_Get:Array of EUW Picker Button Object References)
   Set Mouth Region Buttons(
    execute=@n3.then_1
    Mouth Region Buttons:Array of EUW Picker Button Object References=Get Picker Buttons In Canvas(Canvas=Get Region_C_Mouth() -> (Region_C_Mouth:Canvas Panel Object Reference))
   ) -> (Output_Get:Array of EUW Picker Button Object References)
   > Set Jaw Region Buttons(Jaw Region Buttons:Array of EUW Picker Button Object References=Get Picker Buttons In Canvas(Canvas=Get Region_C_Jaw() -> (Region_C_Jaw:Canvas Panel Object Reference))) -> (Output_Get:Array of EUW Picker Button Object References)
   Set L Eye Region Buttons(
    execute=@n3.then_2
    L Eye Region Buttons:Array of EUW Picker Button Object References=Get Picker Buttons In Canvas(Canvas=Get Region_L_Eye() -> (Region_L_Eye:Canvas Panel Object Reference))
   ) -> (Output_Get:Array of EUW Picker Button Object References)
   > Set R Eye Region Buttons(R Eye Region Buttons:Array of EUW Picker Button Object References=Get Picker Buttons In Canvas(Canvas=Get Region_R_Eye() -> (Region_R_Eye:Canvas Panel Object Reference))) -> (Output_Get:Array of EUW Picker Button Object References)
   Set L Brow Region Buttons(
    execute=@n3.then_3
    L Brow Region Buttons:Array of EUW Picker Button Object References=Get Picker Buttons In Canvas(Canvas=Get Region_L_Brow() -> (Region_L_Brow:Canvas Panel Object Reference))
   ) -> (Output_Get:Array of EUW Picker Button Object References)
   > Set R Brow Region Buttons(R Brow Region Buttons:Array of EUW Picker Button Object References=Get Picker Buttons In Canvas(Canvas=Get Region_R_Brow() -> (Region_R_Brow:Canvas Panel Object Reference))) -> (Output_Get:Array of EUW Picker Button Object References)
   n4 = Sequence!(execute=@n3.then_4) -> (then_0, then_1, then_2, then_3)
   n5 = Get All Canvas Buttons() -> (All Canvas Buttons:Array of EUW Picker Button Object References)
   > Append Array!(
    execute
    TargetArray:Array of EUW Picker Button Object References=@n5
    SourceArray:Array of EUW Picker Button Object References=Get Main Canvas Buttons() -> (Main Canvas Buttons:Array of EUW Picker Button Object References)
   ) -> (then)
   > Append Array(TargetArray=@n5, SourceArray=Get R Brow Region Buttons())
   > Append Array(TargetArray=@n5, SourceArray=Get L Brow Region Buttons())
   n6 = Get All Canvas Buttons()
   Append Array(execute=@n4.then_1, TargetArray=@n6, SourceArray=Get R Eye Region Buttons())
   > Append Array(TargetArray=@n6, SourceArray=Get L Eye Region Buttons())
   n7 = Get Jaw Region Buttons() -> (Jaw Region Buttons:Array of EUW Picker Button Object References)
   n8 = Get All Canvas Buttons()
   Append Array(execute=@n4.then_2, TargetArray=@n8, SourceArray=@n7)
   n9 = Get Mouth Region Buttons()
   > Append Array(TargetArray=@n8, SourceArray=@n9)
   n10 = Get Jaw And Mouth Canvas Buttons()
   Append Array(execute=@n3.then_5, TargetArray=@n10, SourceArray=@n7)
   > Append Array(TargetArray=@n10, SourceArray=@n9)
   n11 = Get Upper Face Canvas Buttons()
   Append Array(execute=@n3.then_6, TargetArray=@n11, SourceArray=Get L Eye Region Buttons())
   > Append Array(TargetArray=@n11, SourceArray=Get R Eye Region Buttons())
   > Append Array(TargetArray=@n11, SourceArray=Get L Brow Region Buttons())
   > Append Array(TargetArray=@n11, SourceArray=Get R Brow Region Buttons())
   n12 = Get Lower Face Canvas Buttons()
   Append Array(execute=@n3.then_7, TargetArray=@n12, SourceArray=Get Jaw And Mouth Canvas Buttons())
   > Append Array(
    TargetArray=@n12
    SourceArray=Make Array!(
     "[0]":EUW Picker Button Object Reference=Get CTRL_R_nose() -> (CTRL_R_nose:EUW Picker Button Object Reference)
     "[1]":EUW Picker Button Object Reference=Get CTRL_L_nose() -> (CTRL_L_nose:EUW Picker Button Object Reference)
     "[2]":EUW Picker Button Object Reference=Get CTRL_C_nose() -> (CTRL_C_nose:EUW Picker Button Object Reference)
     "[3]":EUW Picker Button Object Reference=Get CTRL_R_eye_cheekRaise() -> (CTRL_R_eye_cheekRaise:EUW Picker Button Object Reference)
     "[4]":EUW Picker Button Object Reference=Get CTRL_L_eye_cheekRaise() -> (CTRL_L_eye_cheekRaise:EUW Picker Button Object Reference)
    ) -> (Array:Array of EUW Picker Button Object References)
   )
   Event Tick()
   > n13 = For Each Loop!(
    Exec
    Array:Array of EUW Picker Button Object References=Make Array!(
     "[0]"=Get CTRL_L_eyeAim() -> (CTRL_L_eyeAim:EUW Picker Button Object Reference)
     "[1]"=Get CTRL_C_eyesAim() -> (CTRL_C_eyesAim:EUW Picker Button Object Reference)
     "[2]"=Get CTRL_R_eyeAim() -> (CTRL_R_eyeAim:EUW Picker Button Object Reference)
     "[3]"
    ) -> (Array)
   ) -> (LoopBody, Array Element:EUW Picker Button Object Reference, Array Index, Completed)
   n14 = Get Slider!(self:EUW Picker 1DSlider Object Reference=Get lookAtSwitch() -> (lookAtSwitch:EUW Picker 1DSlider Object Reference)) -> (Slider:Editor Utility Slider Object Reference)
   n15 = Select(Option 0=Hidden, Option 1=Visible, Index="Equal (Float)"(A=Get Value!(self:Slider Object Reference=@n14) -> (Value:float), B=1))
   > SetVisibility(self=@n13, InVisibility=@n15)
   > SetVisibility(self=Get CTRL_convergenceSwitch() -> (CTRL_convergenceSwitch:EUW Picker 1DSlider Object Reference), InVisibility=@n15)
   n16 = Event Widget has loaded!() -> (
    OutputDelegate
    then
    Control Rig Proxy Binding:Control Rig Sequencer Binding Proxy Structure
    Animatable Creature:Animatable Creature Object Reference
    Anim Picker EUW:Animation Picker Object Reference
   ) {is=title K2Node_Event}
   > Parent: Widget has loaded!(
    execute
    Control Rig Proxy Binding:Control Rig Sequencer Binding Proxy Structure=@n16.Control Rig Proxy Binding
    Animatable Creature:Animatable Creature Object Reference=@n16.Animatable Creature
    Anim Picker EUW:Animation Picker Object Reference=@n16.Anim Picker EUW
   ) -> (then) {is=title K2Node_CallParentFunction}
   > n17 = GetFocusedLevelSequence!(execute) -> (then, ReturnValue:Level Sequence Object Reference)
   > n18 = Set LS(LS:Level Sequence Object Reference=@n17) -> (Output_Get:Level Sequence Object Reference)
   > n19 = GetControlRigs!(execute, LevelSequence:Level Sequence Object Reference=@n18) -> (then, ReturnValue:Array of Control Rig Sequencer Binding Proxy Structures)
   > n20 = For Each Loop(Array:Array of Control Rig Sequencer Binding Proxy Structures=@n19) -> (Array Element:Control Rig Sequencer Binding Proxy Structure)
   n21 = For Each Loop(Exec=@n20.Completed, Array:Array of Widget Object References=Get All Canvas Buttons()) -> (Array Element:Widget Object Reference)
   Update Sliders based on CR(execute=@n21.Completed)
   Branch(
    execute=@n21.LoopBody
    Condition=Does Object Implement Interface!(
     TestObject:Object Reference=@n21
     Interface:Interface Interface=/AnimInEdTools/AnimationPicker/PickerElements/BPI_SelectablePickerElements.BPI_SelectablePickerElements_C
    ) -> (ReturnValue:bool)
   )
   n22 = Make Array!("[0]":Name=String To Name(InString=GetDisplayName(Object=@n21))) -> (Array:Array of Names)
   > Set Initial Vars!(
    execute:Exec
    self:Object Reference=@n21
    Control Rig:Control Rig Object Reference=Get Face CR()
    Level Sequence:Level Sequence Object Reference=Get LS() -> (LS:Level Sequence Object Reference)
    Control Bindings:Map of Control Rig Object References to Animation Picker String Array Struct Structures=Make Map!(
     Key 0:Control Rig Object Reference=Get Face CR()
     Value 0:Animation Picker String Array Struct Structure=Make Animation Picker String Array Struct!(ControlNames_6_#1:Array of Names=@n22) -> (Animation_Picker_StringArray_Struct:Animation Picker String Array Struct Structure) {is=title K2Node_MakeStruct}
    ) -> (Map:Map of Control Rig Object References to Animation Picker String Array Struct Structures) {is=title K2Node_MakeMap}
    Anim Picker EUW:Animation Picker Object Reference=Get Anim Picker EUW() -> (Anim Picker EUW:Animation Picker Object Reference)
    Level Actors:Array of Movie Scene Binding Proxy Structures
   ) -> (then:Exec) {is=title K2Node_Message}
   > n23 = Cast To EUW_Picker_1DSlider!(execute, Object=@n21) -> (then, CastFailed, AsEUW Picker 1DSlider:EUW Picker 1DSlider Object Reference) {is=title K2Node_DynamicCast}
   > Add(
    TargetArray:Array of EUW Picker 1DSlider Object References=Get All 1DSlider() -> (All 1DSlider:Array of EUW Picker 1DSlider Object References)
    NewItem:"EUW Picker 1DSlider Object Reference (by ref)"=@n23
   ) {is=title K2Node_CallArrayFunction}
   n24 = Cast To EUW_Picker_2DSlider!(execute=@n23.CastFailed, Object=@n21) -> (then, CastFailed, AsEUW Picker 2DSlider:EUW Picker 2DSlider Object Reference) {is=title K2Node_DynamicCast}
   > Add(
    TargetArray:Array of EUW Picker 2DSlider Object References=Get All 2DSlider() -> (All 2DSlider:Array of EUW Picker 2DSlider Object References)
    NewItem:"EUW Picker 2DSlider Object Reference (by ref)"=@n24
   )
   n25 = Break Control Rig Sequencer Binding Proxy!(ControlRigSequencerBindingProxy:"Control Rig Sequencer Binding Proxy Structure (by ref)"=@n20) -> (
    Proxy:Movie Scene Binding Proxy Structure
    ControlRig:Control Rig Object Reference
    Track:Movie Scene Control Rig Parameter Track Object Reference
   ) {is=title K2Node_BreakStruct}
   Branch(
    execute=@n20.LoopBody
    Condition="Equal (Class)"(
     A=Get Class(Object=@n25.ControlRig) -> (ReturnValue:Control Rig Class Reference)
     B=/Game/Characters/Player/Common/Fortnite_Base_Head/Facials/ControlRigs/Fortnite_3L_OnFace_CtrlRig.Fortnite_3L_OnFace_CtrlRig_C
    )
   )
   > n26 = GetBindingID!(
    execute
    Sequence:Movie Scene Sequence Object Reference=Get LS()
    InBinding:"Movie Scene Binding Proxy Structure (by ref)"=Break Control Rig Sequencer Binding Proxy(ControlRigSequencerBindingProxy=Get Control Rig Proxy Binding() -> (Control Rig Proxy Binding:Control Rig Sequencer Binding Proxy Structure))
   ) -> (then, ReturnValue:Movie Scene Object Binding ID Structure)
   > n27 = GetBindingID(Sequence=Get LS(), InBinding=@n25)
   n28 = GetBoundObjects!(ObjectBinding:Movie Scene Object Binding ID Structure=@n27) -> (ReturnValue:Array of Object References)
   n29 = GetBoundObjects(ObjectBinding=@n26)
   > Branch(Condition="Equal (Object)"(A="Get (a copy)"(Array:Array of Object References=@n28) -> (Output:Object Reference), B="Get (a copy)"(Array=@n29)))
   > Set Face CR(Face CR:Control Rig Object Reference=@n25.ControlRig) -> (Output_Get:Control Rig Object Reference)
   Event Update From Scene() {is=title K2Node_Event}
   > Parent: Update From Scene() {is=title K2Node_CallParentFunction}
   > Update Sliders based on CR()
  ]}
 Get Picker Buttons In Canvas {nodes [
   n1 = Get Picker Buttons In Canvas!() -> (then, Canvas:Canvas Panel Object Reference) {is=title K2Node_FunctionEntry}
   > n2 = For Each Loop(Array=GetAllChildren(self=@n1))
   Return Node!(
    execute=@n2.Completed
    Picker Buttons:Array of EUW Picker Button Object References=Get Buttons() -> (Buttons:Array of EUW Picker Button Object References)
   )
   > n3 = Cast To EUW_Picker_Button!(execute, Object=@n2) -> (then, CastFailed, AsEUW Picker Button:EUW Picker Button Object Reference) {is=title K2Node_DynamicCast}
   > Add(TargetArray:Array of EUW Picker Button Object References=Get Buttons(), NewItem:"EUW Picker Button Object Reference (by ref)"=@n3)
  ]}
 Get Primary Panel {nodes [
   Get Primary Panel() {is=title K2Node_FunctionEntry}
   > Return Node!(execute, Panel:Panel Widget Object Reference=Get CanvasPanel_0())
  ]}
 Update Sliders based on CR {nodes [
   Update Sliders based on CR() {is=title K2Node_FunctionEntry}
   > n1 = GetFocusedLevelSequence()
   > Set LS(LS=@n1)
   > n2 = Get Current Local Time!(execute, TimeUnit:EMovieSceneTimeUnit Enum=DisplayRate) -> (then, ReturnValue:Movie Scene Sequence Playback Params Structure) {is=title K2Node_CallFunction}
   > Set Current Frame Number(Current Frame Number:Frame Number Structure=Make Frame Number!(Value:int=Break Movie Scene Sequence Playback Params!(MovieSceneSequencePlaybackParams:"Movie Scene Sequence Playback Params Structure (by ref)"=@n2) -> (
      Frame_FrameNumber_Value:int
      Frame_SubFrame:float=~
      Time:float=~
      MarkedFrame:String
      Timecode:Timecode Structure
      PositionType:EMovieScenePositionType Enum
      UpdateMethod:EUpdatePositionMethod Enum
      bHasJumped:bool=~
     ) {is=title K2Node_BreakStruct}) -> (FrameNumber:Frame Number Structure) {is=title K2Node_MakeStruct}) -> (Output_Get:Frame Number Structure)
   > n3 = Get!(execute:Exec) -> (then:Exec, else:Exec, Face CR:Control Rig Object Reference)
   > n4 = Set CR(CR:Control Rig Object Reference=@n3) -> (Output_Get:Control Rig Object Reference)
   n5 = For Each Loop(Exec=[@n11.then, @n4.then], Array:Array of EUW Picker 1DSlider Object References=Get All 1DSlider()) -> (Array Element:EUW Picker 1DSlider Object Reference)
   n6 = For Each Loop(Exec=@n5.Completed, Array:Array of EUW Picker 2DSlider Object References=Get All 2DSlider()) -> (Array Element:EUW Picker 2DSlider Object Reference)
   Return Node!(execute=@n6.Completed)
   n7 = String To Name(InString=GetObjectName(Object=@n6))
   > n8 = GetLocalControlRigVector2D!(
    execute
    LevelSequence:Level Sequence Object Reference=Get LS()
    ControlRig:Control Rig Object Reference=Get CR() -> (CR:Control Rig Object Reference)
    ControlName:Name=@n7
    Frame:Frame Number Structure=Get Current Frame Number() -> (Current Frame Number:Frame Number Structure)
    TimeUnit:EMovieSceneTimeUnit Enum=DisplayRate
   ) -> (then, ReturnValue:Vector 2D Structure)
   > SetValue!(
    execute
    self:Synth 2DSlider Object Reference=Get Slider(self:EUW Picker 2DSlider Object Reference=@n6) -> (Slider:Synth 2DSlider Object Reference)
    InValue:Vector 2D Structure=Convert to Slider 2D Relative Value!(self:EUW Picker 2DSlider Object Reference=@n6, Value:Vector 2D Structure=@n8) -> (ReturnValue:Vector 2D Structure)
   ) -> (then)
   > Update Selection Display!(
    execute:Exec
    self:Object Reference=@n6
    Selected:bool=IsControlSelected!(self:Control Rig Object Reference=Get CR(), InControlName:"Name (by ref)"=@n7) -> (ReturnValue:bool)
   ) -> (then:Exec, Out Selected:bool) {is=title K2Node_Message}
   n9 = String To Name(InString=GetObjectName(Object=@n5))
   n10 = GetLocalControlRigFloat!(
    execute=@n5.LoopBody
    LevelSequence:Level Sequence Object Reference=Get LS()
    ControlRig:Control Rig Object Reference=Get CR()
    ControlName:Name=@n9
    Frame:Frame Number Structure=Get Current Frame Number()
    TimeUnit:EMovieSceneTimeUnit Enum=DisplayRate
   ) -> (then, ReturnValue:float)
   > SetValue(
    self:Slider Object Reference=Get Slider(self:EUW Picker 1DSlider Object Reference=@n5) -> (Slider:Editor Utility Slider Object Reference)
    InValue:float=@n10
   )
   > Update Selection Display(self=@n5, Selected=IsControlSelected(self=Get CR(), InControlName=@n9))
   n11 = Set CR(execute=@n3.else, CR=Break Control Rig Sequencer Binding Proxy(ControlRigSequencerBindingProxy=Get Control Rig Proxy Binding()).ControlRig)
  ]}
]
variables [
 All 1DSlider {type="array<EUW_Picker_1DSlider_C>", default []}
 All 2DSlider {type="array<EUW_Picker_2DSlider_C>", default []}
 All Canvas Buttons {type="array<EUW_Picker_Button_C>", default []}
 Face CR {type=ControlRig, default=null}
 Jaw And Mouth Canvas Buttons {type="array<EUW_Picker_Button_C>", default []}
 Jaw Region Buttons {type="array<EUW_Picker_Button_C>", default []}
 L Brow Region Buttons {type="array<EUW_Picker_Button_C>", default []}
 L Eye Region Buttons {type="array<EUW_Picker_Button_C>", default []}
 Lower Face Canvas Buttons {type="array<EUW_Picker_Button_C>", default []}
 LS {type=LevelSequence, default=null}
 Main Canvas Buttons {type="array<EUW_Picker_Button_C>", default []}
 Mouth Region Buttons {type="array<EUW_Picker_Button_C>", default []}
 R Brow Region Buttons {type="array<EUW_Picker_Button_C>", default []}
 R Eye Region Buttons {type="array<EUW_Picker_Button_C>", default []}
 Upper Face Canvas Buttons {type="array<EUW_Picker_Button_C>", default []}
]
widgets {
 info {
  parentClass=/AnimInEdTools/AnimationPicker/EUW_Asset_Picker_Base.EUW_Asset_Picker_Base_C
  rootWidgetClass=&CanvasPanel
  widgetCount=75
 }
 widgets [
  CanvasPanel_0 {
   widgetClassPath=&CanvasPanel
   bIsVariable=true
   widget=CanvasPanel_0
   properties {renderTransform {angle=0, scale=<1 1>, shear=<0 0>, translation=<120 0>}, renderTransformPivot=<0 0>}
  }
  CTRL_convergenceSwitch {
   widgetClassPath=/AnimInEdTools/AnimationPicker/PickerElements/EUW_Picker_1DSlider
   bIsVariable=true
   parent=CanvasPanel_0
   widget=CTRL_convergenceSwitch
   properties {colorAndOpacity=<rgba 1 0.9695 0 1>}
   slot {
    class=CanvasPanelSlot
    properties {layoutData {alignment=<0.5 0.5>, anchors {maximum=<0.5 0.5>, minimum=<0.5 0.5>}, offsets=<ltrb -353.7455 -200.3483 77.5916 10>}}
   }
  }
  lookAtSwitch {
   widgetClassPath=&EUW_Picker_1DSlider
   bIsVariable=true
   parent=CanvasPanel_0
   widget=lookAtSwitch
   properties {bOverride_Cursor=true}
   slot {
    class=CanvasPanelSlot
    properties {layoutData {alignment=<0.5 0.5>, anchors {maximum=<0.5 0.5>, minimum=<0.5 0.5>}, offsets=<ltrb -301.2692 -405.3066 115.6156 15.6006>}}
   }
  }
  gui_follow_head {
   widgetClassPath=&EUW_Picker_1DSlider
   bIsVariable=true
   parent=CanvasPanel_0
   widget=gui_follow_head
   properties {bOverride_Cursor=true}
   slot {
    class=CanvasPanelSlot
    properties {layoutData {alignment=<0.5 0.5>, anchors {maximum=<0.5 0.5>, minimum=<0.5 0.5>}, offsets=<ltrb -288.7994 392.3209 115.6156 15.6006>}}
   }
  }
  follow_head {
   widgetClassPath=&EUW_Picker_1DSlider
   bIsVariable=true
   parent=CanvasPanel_0
   widget=follow_head
   properties {bOverride_Cursor=true}
   slot {
    class=CanvasPanelSlot
    properties {layoutData {alignment=<0.5 0.5>, anchors {maximum=<0.5 0.5>, minimum=<0.5 0.5>}, offsets=<ltrb -301.2692 -361.0692 115.6156 15.6006>}}
   }
  }
  CTRL_C_eye {
   widgetClassPath=/AnimInEdTools/AnimationPicker/EUW_Picker_Button
   bIsVariable=true
   parent=CanvasPanel_0
   widget=CTRL_C_eye
   properties {
    draw Style=RoundedSquareLargeRadius
    label Font Size=14
    label Text=Eyes
    colorAndOpacity=<rgba 1 0.9695 0 1>
    label Color {colorUseRule=UseColor_Specified, specifiedColor=<rgba 0 0 0 1>}
    normal Tint {colorUseRule=UseColor_Specified, specifiedColor=<rgba 1 1 1 1>}
   }
   slot {
    class=CanvasPanelSlot
    properties {layoutData {alignment=<0.5 0.5>, anchors {maximum=<0.5 0.5>, minimum=<0.5 0.5>}, offsets=<ltrb -355.1589 -141.96 67.604 53.161>}}
   }
  }
  CTRL_C_nose {
   widgetClassPath=&EUW_Picker_Button
   bIsVariable=true
   parent=CanvasPanel_0
   widget=CTRL_C_nose
   properties {label Font Size=6, colorAndOpacity=<rgba 1 0.9695 0 1>, normal Tint {colorUseRule=UseColor_Specified, specifiedColor=<rgba 1 1 1 1>}}
   slot {
    class=CanvasPanelSlot
    properties {layoutData {alignment=<0.5 0.5>, anchors {maximum=<0.5 0.5>, minimum=<0.5 0.5>}, offsets=<ltrb -354.4242 -82.2223 37.0735 28.6365>}}
   }
  }
  Face_Ctrl_Master {
   widgetClassPath=&EUW_Picker_Button
   bIsVariable=true
   parent=CanvasPanel_0
   widget=Face_Ctrl_Master
   properties {
    draw Style=Circle
    label Font Size=6
    colorAndOpacity=<rgba 0 0 1 1>
    normal Tint {colorUseRule=UseColor_Specified, specifiedColor=<rgba 1 1 1 1>}
   }
   slot {
    class=CanvasPanelSlot
    properties {layoutData {alignment=<0.5 0.5>, anchors {maximum=<0.5 0.5>, minimum=<0.5 0.5>}, offsets=<ltrb -353.5972 302.7013 35 35>}}
   }
  }
  CTRL_C_tongue {
   widgetClassPath=&EUW_Picker_Button
   bIsVariable=true
   parent=CanvasPanel_0
   widget=CTRL_C_tongue
   properties {draw Style=Circle, label Font Size=6, normal Tint {colorUseRule=UseColor_Specified, specifiedColor=<rgba 0.7135 0.3419 0.3419 1>}}
   slot {
    class=CanvasPanelSlot
    properties {layoutData {alignment=<0.5 0.5>, anchors {maximum=<0.5 0.5>, minimum=<0.5 0.5>}, offsets=<ltrb -92.9876 164.6475 35 35>}}
   }
  }
  CTRL_L_nose {
   widgetClassPath=&EUW_Picker_Button
   bIsVariable=true
   parent=CanvasPanel_0
   widget=CTRL_L_nose
   properties {
    label Font Size=6
    colorAndOpacity=<rgba 0.2613 1 0.1894 1>
    normal Tint {colorUseRule=UseColor_Specified, specifiedColor=<rgba 0.6094 0.6094 0.6094 1>}
   }
   slot {
    class=CanvasPanelSlot
    properties {layoutData {alignment=<0.5 0.5>, anchors {maximum=<0.5 0.5>, minimum=<0.5 0.5>}, offsets=<ltrb -306.3069 -31.3125 25 25>}}
   }
  }
  CTRL_R_nose {
   widgetClassPath=&EUW_Picker_Button
   bIsVariable=true
   parent=CanvasPanel_0
   widget=CTRL_R_nose
   properties {
    label Font Size=6
    colorAndOpacity=<rgba 1 0.0521 0.0739 1>
    normal Tint {colorUseRule=UseColor_Specified, specifiedColor=<rgba 0.5625 0.5625 0.5625 1>}
   }
   slot {
    class=CanvasPanelSlot
    properties {layoutData {alignment=<0.5 0.5>, anchors {maximum=<0.5 0.5>, minimum=<0.5 0.5>}, offsets=<ltrb -395.9609 -31.5405 25 25>}}
   }
  }
  CTRL_L_eye_cheekRaise {
   widgetClassPath=&EUW_Picker_Button
   bIsVariable=true
   parent=CanvasPanel_0
   widget=CTRL_L_eye_cheekRaise
   properties {
    label Font Size=6
    colorAndOpacity=<rgba 0.2613 1 0.1894 1>
    normal Tint {colorUseRule=UseColor_Specified, specifiedColor=<rgba 0.6094 0.6094 0.6094 1>}
   }
   slot {
    class=CanvasPanelSlot
    properties {layoutData {alignment=<0.5 0.5>, anchors {maximum=<0.5 0.5>, minimum=<0.5 0.5>}, offsets=<ltrb -190.0169 -62.0375 45.888 21.006>}}
   }
  }
  CTRL_R_eye_cheekRaise {
   widgetClassPath=&EUW_Picker_Button
   bIsVariable=true
   parent=CanvasPanel_0
   widget=CTRL_R_eye_cheekRaise
   properties {
    label Font Size=6
    colorAndOpacity=<rgba 1 0.0521 0.0739 1>
    normal Tint {colorUseRule=UseColor_Specified, specifiedColor=<rgba 0.6094 0.6094 0.6094 1>}
   }
   slot {
    class=CanvasPanelSlot
    properties {layoutData {alignment=<0.5 0.5>, anchors {maximum=<0.5 0.5>, minimum=<0.5 0.5>}, offsets=<ltrb -522.0169 -62.0375 45.888 21.006>}}
   }
  }
  Region_L_Eye {
   widgetClassPath=&CanvasPanel
   bIsVariable=true
   parent=CanvasPanel_0
   widget=Region_L_Eye
   slot {
    class=CanvasPanelSlot
    properties {
     zOrder=3
     layoutData {alignment=<0.5 0.5>, anchors {maximum=<0.5 0.5>, minimum=<0.5 0.5>}, offsets=<ltrb -223.982 -140.3251 162.1121 116.4348>}
    }
   }
  }
  CTRL_L_eye_eyelidU {
   widgetClassPath=&EUW_Picker_Button
   bIsVariable=true
   parent=Region_L_Eye
   widget=CTRL_L_eye_eyelidU
   properties {
    label Font Size=6
    colorAndOpacity=<rgba 0.2613 1 0.1894 1>
    normal Tint {colorUseRule=UseColor_Specified, specifiedColor=<rgba 0.6094 0.6094 0.6094 1>}
   }
   slot {
    class=CanvasPanelSlot
    properties {zOrder=1, layoutData {alignment=<0 0>, anchors {maximum=<0 0>, minimum=<0 0>}, offsets=<ltrb 36 4 50 20>}}
   }
  }
  CTRL_L_eye {
   widgetClassPath=&EUW_Picker_Button
   bIsVariable=true
   parent=Region_L_Eye
   widget=CTRL_L_eye
   properties {
    draw Style=Circle
    label Font Size=6
    colorAndOpacity=<rgba 0.2613 1 0.1894 1>
    normal Tint {colorUseRule=UseColor_Specified, specifiedColor=<rgba 0.6094 0.6094 0.6094 1>}
   }
   slot {
    class=CanvasPanelSlot
    properties {zOrder=1, layoutData {alignment=<0 0>, anchors {maximum=<0 0>, minimum=<0 0>}, offsets=<ltrb 36 32 50 50>}}
   }
  }
  CTRL_L_eye_blink {
   widgetClassPath=&EUW_Picker_Button
   bIsVariable=true
   parent=Region_L_Eye
   widget=CTRL_L_eye_blink
   properties {
    label Font Size=6
    colorAndOpacity=<rgba 0.2613 1 0.1894 1>
    normal Tint {colorUseRule=UseColor_Specified, specifiedColor=<rgba 0.6094 0.6094 0.6094 1>}
   }
   slot {
    class=CanvasPanelSlot
    properties {zOrder=1, layoutData {alignment=<0 0>, anchors {maximum=<0 0>, minimum=<0 0>}, offsets=<ltrb 100 52 50 20>}}
   }
  }
  CTRL_L_eye_eyelidD {
   widgetClassPath=&EUW_Picker_Button
   bIsVariable=true
   parent=Region_L_Eye
   widget=CTRL_L_eye_eyelidD
   properties {
    label Font Size=6
    colorAndOpacity=<rgba 0.2613 1 0.1894 1>
    normal Tint {colorUseRule=UseColor_Specified, specifiedColor=<rgba 0.6094 0.6094 0.6094 1>}
   }
   slot {
    class=CanvasPanelSlot
    properties {zOrder=1, layoutData {alignment=<0 0>, anchors {maximum=<0 0>, minimum=<0 0>}, offsets=<ltrb 36 88 50 20>}}
   }
  }
  CTRL_L_eye_squintInner {
   widgetClassPath=&EUW_Picker_Button
   bIsVariable=true
   parent=Region_L_Eye
   widget=CTRL_L_eye_squintInner
   properties {
    label Font Size=6
    colorAndOpacity=<rgba 0.2613 1 0.1894 1>
    normal Tint {colorUseRule=UseColor_Specified, specifiedColor=<rgba 0.6094 0.6094 0.6094 1>}
   }
   slot {
    class=CanvasPanelSlot
    properties {zOrder=1, layoutData {alignment=<0 0>, anchors {maximum=<0 0>, minimum=<0 0>}, offsets=<ltrb 8 96 20 14>}}
   }
  }
  Region_L_Eye_Btn {
   widgetClassPath=&EditorUtilityButton
   parent=Region_L_Eye
   widget=Region_L_Eye_Btn
   properties {
    backgroundColor=<rgba 1 1 1 0.08>
    colorAndOpacity=<rgba 0.099 0.0951 0.0918 0.4348>
    widgetStyle {
     clickedSlateSound {resourceObject=null}
     disabled {
      bIsDynamicallyLoaded=false
      drawAs=RoundedBox
      imageType=NoImage
      mirroring=NoMirror
      resourceName=None
      resourceObject=null
      tiling=NoTile
      imageSize=<0 0>
      margin=<ltrb 0 0 0 0>
      outlineSettings {
       bUseBrushTransparency=false
       roundingType=FixedRadius
       width=1
       color {colorUseRule=UseColor_ColorTable, specifiedColor=<rgba 0 0 0 0>}
       cornerRadii=<4 4 4 4>
      }
      tintColor {colorUseRule=UseColor_Specified, specifiedColor=<rgba 0.0395 0.0395 0.0395 1>}
      uVRegion {bIsValid=false, max=<0 0>, min=<0 0>}
     }
     disabledForeground {colorUseRule=UseColor_ColorTable, specifiedColor=<rgba 0 0 0 0>}
     hovered {
      bIsDynamicallyLoaded=false
      drawAs=RoundedBox
      imageType=NoImage
      mirroring=NoMirror
      resourceName=None
      resourceObject=null
      tiling=NoTile
      imageSize=<32 32>
      margin=<ltrb 0 0 0 0>
      outlineSettings {
       bUseBrushTransparency=false
       roundingType=FixedRadius
       width=1
       color {colorUseRule=UseColor_ColorTable, specifiedColor=<rgba 0 0 0 0>}
       cornerRadii=<4 4 4 4>
      }
      tintColor {colorUseRule=UseColor_Specified, specifiedColor=<rgba 0.0953 0.0953 0.0953 1>}
      uVRegion {bIsValid=false, max=<0 0>, min=<0 0>}
     }
     hoveredForeground {colorUseRule=UseColor_ColorTable, specifiedColor=<rgba 0 0 0 0>}
     hoveredSlateSound {resourceObject=null}
     normal {
      bIsDynamicallyLoaded=false
      drawAs=RoundedBox
      imageType=NoImage
      mirroring=NoMirror
      resourceName=None
      resourceObject=null
      tiling=NoTile
      imageSize=<32 32>
      margin=<ltrb 0 0 0 0>
      outlineSettings {
       bUseBrushTransparency=false
       roundingType=FixedRadius
       width=0
       color {colorUseRule=UseColor_Specified, specifiedColor=<rgba 0 0 0 1>}
       cornerRadii=<4 4 4 4>
      }
      tintColor {colorUseRule=UseColor_Specified, specifiedColor=<rgba 1 1 1 1>}
      uVRegion {bIsValid=false, max=<0 0>, min=<0 0>}
     }
     normalForeground {colorUseRule=UseColor_ColorTable, specifiedColor=<rgba 0 0 0 0>}
     normalPadding=<ltrb 12 1.5 12 1.5>
     pressed {
      bIsDynamicallyLoaded=false
      drawAs=RoundedBox
      imageType=NoImage
      mirroring=NoMirror
      resourceName=None
      resourceObject=null
      tiling=NoTile
      imageSize=<32 32>
      margin=<ltrb 0 0 0 0>
      outlineSettings {
       bUseBrushTransparency=false
       roundingType=FixedRadius
       width=1
       color {colorUseRule=UseColor_ColorTable, specifiedColor=<rgba 0 0 0 0>}
       cornerRadii=<4 4 4 4>
      }
      tintColor {colorUseRule=UseColor_Specified, specifiedColor=<rgba 0.0284 0.0284 0.0284 1>}
      uVRegion {bIsValid=false, max=<0 0>, min=<0 0>}
     }
     pressedForeground {colorUseRule=UseColor_ColorTable, specifiedColor=<rgba 0 0 0 0>}
     pressedPadding=<ltrb 12 2.5 12 0.5>
     pressedSlateSound {resourceObject=null}
    }
   }
   slot {
    class=CanvasPanelSlot
    properties {layoutData {alignment=<0 0>, anchors {maximum=<0.5 0.5>, minimum=<0.5 0.5>}, offsets=<ltrb -81.0561 -58.2174 163.0631 119.5796>}}
   }
  }
  Region_R_Eye {
   widgetClassPath=&CanvasPanel
   bIsVariable=true
   parent=CanvasPanel_0
   widget=Region_R_Eye
   slot {
    class=CanvasPanelSlot
    properties {
     zOrder=3
     layoutData {alignment=<0.5 0.5>, anchors {maximum=<0.5 0.5>, minimum=<0.5 0.5>}, offsets=<ltrb -483.4805 -141.7432 164.1141 117.5976>}
    }
   }
  }
  CTRL_R_eye_eyelidU {
   widgetClassPath=&EUW_Picker_Button
   bIsVariable=true
   parent=Region_R_Eye
   widget=CTRL_R_eye_eyelidU
   properties {
    label Font Size=6
    colorAndOpacity=<rgba 1 0.0521 0.0739 1>
    normal Tint {colorUseRule=UseColor_Specified, specifiedColor=<rgba 0.6094 0.6094 0.6094 1>}
   }
   slot {
    class=CanvasPanelSlot
    properties {zOrder=1, layoutData {alignment=<0 0>, anchors {maximum=<0 0>, minimum=<0 0>}, offsets=<ltrb 76 4 50 20>}}
   }
  }
  CTRL_R_eye_eyelidD {
   widgetClassPath=&EUW_Picker_Button
   bIsVariable=true
   parent=Region_R_Eye
   widget=CTRL_R_eye_eyelidD
   properties {
    label Font Size=6
    colorAndOpacity=<rgba 1 0.0521 0.0739 1>
    normal Tint {colorUseRule=UseColor_Specified, specifiedColor=<rgba 0.6094 0.6094 0.6094 1>}
   }
   slot {
    class=CanvasPanelSlot
    properties {zOrder=1, layoutData {alignment=<0 0>, anchors {maximum=<0 0>, minimum=<0 0>}, offsets=<ltrb 76 84 50 20>}}
   }
  }
  CTRL_R_eye {
   widgetClassPath=&EUW_Picker_Button
   bIsVariable=true
   parent=Region_R_Eye
   widget=CTRL_R_eye
   properties {
    draw Style=Circle
    label Font Size=6
    colorAndOpacity=<rgba 1 0.0521 0.0739 1>
    normal Tint {colorUseRule=UseColor_Specified, specifiedColor=<rgba 0.5625 0.5625 0.5625 1>}
   }
   slot {
    class=CanvasPanelSlot
    properties {zOrder=1, layoutData {alignment=<0 0>, anchors {maximum=<0 0>, minimum=<0 0>}, offsets=<ltrb 76 28 50 50>}}
   }
  }
  CTRL_R_eye_blink {
   widgetClassPath=&EUW_Picker_Button
   bIsVariable=true
   parent=Region_R_Eye
   widget=CTRL_R_eye_blink
   properties {
    label Font Size=6
    colorAndOpacity=<rgba 1 0.0521 0.0739 1>
    normal Tint {colorUseRule=UseColor_Specified, specifiedColor=<rgba 0.6094 0.6094 0.6094 1>}
   }
   slot {
    class=CanvasPanelSlot
    properties {zOrder=1, layoutData {alignment=<0 0>, anchors {maximum=<0 0>, minimum=<0 0>}, offsets=<ltrb 16 48 50 20>}}
   }
  }
  CTRL_R_eye_squintInner {
   widgetClassPath=&EUW_Picker_Button
   bIsVariable=true
   parent=Region_R_Eye
   widget=CTRL_R_eye_squintInner
   properties {
    label Font Size=6
    colorAndOpacity=<rgba 1 0.0521 0.0739 1>
    normal Tint {colorUseRule=UseColor_Specified, specifiedColor=<rgba 0.5625 0.5625 0.5625 1>}
   }
   slot {
    class=CanvasPanelSlot
    properties {zOrder=1, layoutData {alignment=<0 0>, anchors {maximum=<0 0>, minimum=<0 0>}, offsets=<ltrb 136 96 20 14>}}
   }
  }
  Region_R_Eye_Btn {
   widgetClassPath=&EditorUtilityButton
   parent=Region_R_Eye
   widget=Region_R_Eye_Btn
   properties {
    backgroundColor=<rgba 1 1 1 0.08>
    colorAndOpacity=<rgba 0.099 0.0951 0.0918 0.4348>
    widgetStyle {
     clickedSlateSound {resourceObject=null}
     disabled {
      bIsDynamicallyLoaded=false
      drawAs=RoundedBox
      imageType=NoImage
      mirroring=NoMirror
      resourceName=None
      resourceObject=null
      tiling=NoTile
      imageSize=<0 0>
      margin=<ltrb 0 0 0 0>
      outlineSettings {
       bUseBrushTransparency=false
       roundingType=FixedRadius
       width=1
       color {colorUseRule=UseColor_ColorTable, specifiedColor=<rgba 0 0 0 0>}
       cornerRadii=<4 4 4 4>
      }
      tintColor {colorUseRule=UseColor_Specified, specifiedColor=<rgba 0.0395 0.0395 0.0395 1>}
      uVRegion {bIsValid=false, max=<0 0>, min=<0 0>}
     }
     disabledForeground {colorUseRule=UseColor_ColorTable, specifiedColor=<rgba 0 0 0 0>}
     hovered {
      bIsDynamicallyLoaded=false
      drawAs=RoundedBox
      imageType=NoImage
      mirroring=NoMirror
      resourceName=None
      resourceObject=null
      tiling=NoTile
      imageSize=<32 32>
      margin=<ltrb 0 0 0 0>
      outlineSettings {
       bUseBrushTransparency=false
       roundingType=FixedRadius
       width=1
       color {colorUseRule=UseColor_ColorTable, specifiedColor=<rgba 0 0 0 0>}
       cornerRadii=<4 4 4 4>
      }
      tintColor {colorUseRule=UseColor_Specified, specifiedColor=<rgba 0.0953 0.0953 0.0953 1>}
      uVRegion {bIsValid=false, max=<0 0>, min=<0 0>}
     }
     hoveredForeground {colorUseRule=UseColor_ColorTable, specifiedColor=<rgba 0 0 0 0>}
     hoveredSlateSound {resourceObject=null}
     normal {
      bIsDynamicallyLoaded=false
      drawAs=RoundedBox
      imageType=NoImage
      mirroring=NoMirror
      resourceName=None
      resourceObject=null
      tiling=NoTile
      imageSize=<32 32>
      margin=<ltrb 0 0 0 0>
      outlineSettings {
       bUseBrushTransparency=false
       roundingType=FixedRadius
       width=0
       color {colorUseRule=UseColor_Specified, specifiedColor=<rgba 0 0 0 1>}
       cornerRadii=<4 4 4 4>
      }
      tintColor {colorUseRule=UseColor_Specified, specifiedColor=<rgba 1 1 1 1>}
      uVRegion {bIsValid=false, max=<0 0>, min=<0 0>}
     }
     normalForeground {colorUseRule=UseColor_ColorTable, specifiedColor=<rgba 0 0 0 0>}
     normalPadding=<ltrb 12 1.5 12 1.5>
     pressed {
      bIsDynamicallyLoaded=false
      drawAs=RoundedBox
      imageType=NoImage
      mirroring=NoMirror
      resourceName=None
      resourceObject=null
      tiling=NoTile
      imageSize=<32 32>
      margin=<ltrb 0 0 0 0>
      outlineSettings {
       bUseBrushTransparency=false
       roundingType=FixedRadius
       width=1
       color {colorUseRule=UseColor_ColorTable, specifiedColor=<rgba 0 0 0 0>}
       cornerRadii=<4 4 4 4>
      }
      tintColor {colorUseRule=UseColor_Specified, specifiedColor=<rgba 0.0284 0.0284 0.0284 1>}
      uVRegion {bIsValid=false, max=<0 0>, min=<0 0>}
     }
     pressedForeground {colorUseRule=UseColor_ColorTable, specifiedColor=<rgba 0 0 0 0>}
     pressedPadding=<ltrb 12 2.5 12 0.5>
     pressedSlateSound {resourceObject=null}
    }
   }
   slot {
    class=CanvasPanelSlot
    properties {layoutData {alignment=<0 0>, anchors {maximum=<0.5 0.5>, minimum=<0.5 0.5>}, offsets=<ltrb -82.0571 -58.7988 163.0631 119.5796>}}
   }
  }
  Region_L_Brow {
   widgetClassPath=&CanvasPanel
   bIsVariable=true
   parent=CanvasPanel_0
   widget=Region_L_Brow
   slot {
    class=CanvasPanelSlot
    properties {zOrder=3, layoutData {alignment=<0.5 0.5>, anchors {maximum=<0.5 0.5>, minimum=<0.5 0.5>}, offsets=<ltrb -224.1689 -233.096 159.59 52.895>}}
   }
  }
  Region_L_Brow_Btn {
   widgetClassPath=&EditorUtilityButton
   parent=Region_L_Brow
   widget=Region_L_Brow_Btn
   properties {
    backgroundColor=<rgba 1 1 1 0.08>
    colorAndOpacity=<rgba 0.099 0.0951 0.0918 0.4348>
    widgetStyle {
     clickedSlateSound {resourceObject=null}
     disabled {
      bIsDynamicallyLoaded=false
      drawAs=RoundedBox
      imageType=NoImage
      mirroring=NoMirror
      resourceName=None
      resourceObject=null
      tiling=NoTile
      imageSize=<0 0>
      margin=<ltrb 0 0 0 0>
      outlineSettings {
       bUseBrushTransparency=false
       roundingType=FixedRadius
       width=1
       color {colorUseRule=UseColor_ColorTable, specifiedColor=<rgba 0 0 0 0>}
       cornerRadii=<4 4 4 4>
      }
      tintColor {colorUseRule=UseColor_Specified, specifiedColor=<rgba 0.0395 0.0395 0.0395 1>}
      uVRegion {bIsValid=false, max=<0 0>, min=<0 0>}
     }
     disabledForeground {colorUseRule=UseColor_ColorTable, specifiedColor=<rgba 0 0 0 0>}
     hovered {
      bIsDynamicallyLoaded=false
      drawAs=RoundedBox
      imageType=NoImage
      mirroring=NoMirror
      resourceName=None
      resourceObject=null
      tiling=NoTile
      imageSize=<32 32>
      margin=<ltrb 0 0 0 0>
      outlineSettings {
       bUseBrushTransparency=false
       roundingType=FixedRadius
       width=1
       color {colorUseRule=UseColor_ColorTable, specifiedColor=<rgba 0 0 0 0>}
       cornerRadii=<4 4 4 4>
      }
      tintColor {colorUseRule=UseColor_Specified, specifiedColor=<rgba 0.0953 0.0953 0.0953 1>}
      uVRegion {bIsValid=false, max=<0 0>, min=<0 0>}
     }
     hoveredForeground {colorUseRule=UseColor_ColorTable, specifiedColor=<rgba 0 0 0 0>}
     hoveredSlateSound {resourceObject=null}
     normal {
      bIsDynamicallyLoaded=false
      drawAs=RoundedBox
      imageType=NoImage
      mirroring=NoMirror
      resourceName=None
      resourceObject=null
      tiling=NoTile
      imageSize=<32 32>
      margin=<ltrb 0 0 0 0>
      outlineSettings {
       bUseBrushTransparency=false
       roundingType=FixedRadius
       width=0
       color {colorUseRule=UseColor_Specified, specifiedColor=<rgba 0 0 0 1>}
       cornerRadii=<4 4 4 4>
      }
      tintColor {colorUseRule=UseColor_Specified, specifiedColor=<rgba 1 1 1 1>}
      uVRegion {bIsValid=false, max=<0 0>, min=<0 0>}
     }
     normalForeground {colorUseRule=UseColor_ColorTable, specifiedColor=<rgba 0 0 0 0>}
     normalPadding=<ltrb 12 1.5 12 1.5>
     pressed {
      bIsDynamicallyLoaded=false
      drawAs=RoundedBox
      imageType=NoImage
      mirroring=NoMirror
      resourceName=None
      resourceObject=null
      tiling=NoTile
      imageSize=<32 32>
      margin=<ltrb 0 0 0 0>
      outlineSettings {
       bUseBrushTransparency=false
       roundingType=FixedRadius
       width=1
       color {colorUseRule=UseColor_ColorTable, specifiedColor=<rgba 0 0 0 0>}
       cornerRadii=<4 4 4 4>
      }
      tintColor {colorUseRule=UseColor_Specified, specifiedColor=<rgba 0.0284 0.0284 0.0284 1>}
      uVRegion {bIsValid=false, max=<0 0>, min=<0 0>}
     }
     pressedForeground {colorUseRule=UseColor_ColorTable, specifiedColor=<rgba 0 0 0 0>}
     pressedPadding=<ltrb 12 2.5 12 0.5>
     pressedSlateSound {resourceObject=null}
    }
   }
   slot {
    class=CanvasPanelSlot
    properties {layoutData {alignment=<0 0>, anchors {maximum=<0.5 0.5>, minimum=<0.5 0.5>}, offsets=<ltrb -79.795 -30.4515 160 55>}}
   }
  }
  CTRL_L_brow_inner {
   widgetClassPath=&EUW_Picker_Button
   bIsVariable=true
   parent=Region_L_Brow
   widget=CTRL_L_brow_inner
   properties {
    label Font Size=6
    colorAndOpacity=<rgba 0.2613 1 0.1894 1>
    normal Tint {colorUseRule=UseColor_Specified, specifiedColor=<rgba 0.6094 0.6094 0.6094 1>}
   }
   slot {
    class=CanvasPanelSlot
    properties {zOrder=1, layoutData {alignment=<0 0>, anchors {maximum=<0 0>, minimum=<0 0>}, offsets=<ltrb 4 12 25 30>}}
   }
  }
  CTRL_L_brow {
   widgetClassPath=&EUW_Picker_Button
   bIsVariable=true
   parent=Region_L_Brow
   widget=CTRL_L_brow
   properties {
    label Font Size=6
    colorAndOpacity=<rgba 0.2613 1 0.1894 1>
    normal Tint {colorUseRule=UseColor_Specified, specifiedColor=<rgba 0.6094 0.6094 0.6094 1>}
   }
   slot {
    class=CanvasPanelSlot
    properties {zOrder=1, layoutData {alignment=<0 0>, anchors {maximum=<0 0>, minimum=<0 0>}, offsets=<ltrb 31.0031 12 80 30>}}
   }
  }
  CTRL_L_brow_outer {
   widgetClassPath=&EUW_Picker_Button
   bIsVariable=true
   parent=Region_L_Brow
   widget=CTRL_L_brow_outer
   properties {
    label Font Size=6
    colorAndOpacity=<rgba 0.2613 1 0.1894 1>
    normal Tint {colorUseRule=UseColor_Specified, specifiedColor=<rgba 0.6094 0.6094 0.6094 1>}
   }
   slot {
    class=CanvasPanelSlot
    properties {zOrder=1, layoutData {alignment=<0 0>, anchors {maximum=<0 0>, minimum=<0 0>}, offsets=<ltrb 113.0011 12 40 30>}}
   }
  }
  Region_R_Brow {
   widgetClassPath=&CanvasPanel
   bIsVariable=true
   parent=CanvasPanel_0
   widget=Region_R_Brow
   slot {
    class=CanvasPanelSlot
    properties {
     zOrder=3
     layoutData {alignment=<0.5 0.5>, anchors {maximum=<0.5 0.5>, minimum=<0.5 0.5>}, offsets=<ltrb -481.3304 -234.7564 159.2611 51.5682>}
    }
   }
  }
  CTRL_R_brow_outer {
   widgetClassPath=&EUW_Picker_Button
   bIsVariable=true
   parent=Region_R_Brow
   widget=CTRL_R_brow_outer
   properties {
    label Font Size=6
    colorAndOpacity=<rgba 1 0.0521 0.0739 1>
    normal Tint {colorUseRule=UseColor_Specified, specifiedColor=<rgba 0.5625 0.5625 0.5625 1>}
   }
   slot {
    class=CanvasPanelSlot
    properties {zOrder=1, layoutData {alignment=<0 0>, anchors {maximum=<0 0>, minimum=<0 0>}, offsets=<ltrb 6.6392 12 40 30>}}
   }
  }
  CTRL_R_brow {
   widgetClassPath=&EUW_Picker_Button
   bIsVariable=true
   parent=Region_R_Brow
   widget=CTRL_R_brow
   properties {
    label Font Size=6
    colorAndOpacity=<rgba 1 0.0521 0.0739 1>
    normal Tint {colorUseRule=UseColor_Specified, specifiedColor=<rgba 0.5625 0.5625 0.5625 1>}
   }
   slot {
    class=CanvasPanelSlot
    properties {zOrder=1, layoutData {alignment=<0 0>, anchors {maximum=<0 0>, minimum=<0 0>}, offsets=<ltrb 48.4941 12 80 30>}}
   }
  }
  CTRL_R_brow_inner {
   widgetClassPath=&EUW_Picker_Button
   bIsVariable=true
   parent=Region_R_Brow
   widget=CTRL_R_brow_inner
   properties {
    label Font Size=6
    colorAndOpacity=<rgba 1 0.0521 0.0739 1>
    normal Tint {colorUseRule=UseColor_Specified, specifiedColor=<rgba 0.5625 0.5625 0.5625 1>}
   }
   slot {
    class=CanvasPanelSlot
    properties {zOrder=1, layoutData {alignment=<0 0>, anchors {maximum=<0 0>, minimum=<0 0>}, offsets=<ltrb 130.6391 12 25 30>}}
   }
  }
  Region_R_Brow_Btn {
   widgetClassPath=&EditorUtilityButton
   parent=Region_R_Brow
   widget=Region_R_Brow_Btn
   properties {
    backgroundColor=<rgba 1 1 1 0.08>
    colorAndOpacity=<rgba 0.099 0.0951 0.0918 0.4348>
    widgetStyle {
     clickedSlateSound {resourceObject=null}
     disabled {
      bIsDynamicallyLoaded=false
      drawAs=RoundedBox
      imageType=NoImage
      mirroring=NoMirror
      resourceName=None
      resourceObject=null
      tiling=NoTile
      imageSize=<0 0>
      margin=<ltrb 0 0 0 0>
      outlineSettings {
       bUseBrushTransparency=false
       roundingType=FixedRadius
       width=1
       color {colorUseRule=UseColor_ColorTable, specifiedColor=<rgba 0 0 0 0>}
       cornerRadii=<4 4 4 4>
      }
      tintColor {colorUseRule=UseColor_Specified, specifiedColor=<rgba 0.0395 0.0395 0.0395 1>}
      uVRegion {bIsValid=false, max=<0 0>, min=<0 0>}
     }
     disabledForeground {colorUseRule=UseColor_ColorTable, specifiedColor=<rgba 0 0 0 0>}
     hovered {
      bIsDynamicallyLoaded=false
      drawAs=RoundedBox
      imageType=NoImage
      mirroring=NoMirror
      resourceName=None
      resourceObject=null
      tiling=NoTile
      imageSize=<32 32>
      margin=<ltrb 0 0 0 0>
      outlineSettings {
       bUseBrushTransparency=false
       roundingType=FixedRadius
       width=1
       color {colorUseRule=UseColor_ColorTable, specifiedColor=<rgba 0 0 0 0>}
       cornerRadii=<4 4 4 4>
      }
      tintColor {colorUseRule=UseColor_Specified, specifiedColor=<rgba 0.0953 0.0953 0.0953 1>}
      uVRegion {bIsValid=false, max=<0 0>, min=<0 0>}
     }
     hoveredForeground {colorUseRule=UseColor_ColorTable, specifiedColor=<rgba 0 0 0 0>}
     hoveredSlateSound {resourceObject=null}
     normal {
      bIsDynamicallyLoaded=false
      drawAs=RoundedBox
      imageType=NoImage
      mirroring=NoMirror
      resourceName=None
      resourceObject=null
      tiling=NoTile
      imageSize=<32 32>
      margin=<ltrb 0 0 0 0>
      outlineSettings {
       bUseBrushTransparency=false
       roundingType=FixedRadius
       width=0
       color {colorUseRule=UseColor_Specified, specifiedColor=<rgba 0 0 0 1>}
       cornerRadii=<4 4 4 4>
      }
      tintColor {colorUseRule=UseColor_Specified, specifiedColor=<rgba 1 1 1 1>}
      uVRegion {bIsValid=false, max=<0 0>, min=<0 0>}
     }
     normalForeground {colorUseRule=UseColor_ColorTable, specifiedColor=<rgba 0 0 0 0>}
     normalPadding=<ltrb 12 1.5 12 1.5>
     pressed {
      bIsDynamicallyLoaded=false
      drawAs=RoundedBox
      imageType=NoImage
      mirroring=NoMirror
      resourceName=None
      resourceObject=null
      tiling=NoTile
      imageSize=<32 32>
      margin=<ltrb 0 0 0 0>
      outlineSettings {
       bUseBrushTransparency=false
       roundingType=FixedRadius
       width=1
       color {colorUseRule=UseColor_ColorTable, specifiedColor=<rgba 0 0 0 0>}
       cornerRadii=<4 4 4 4>
      }
      tintColor {colorUseRule=UseColor_Specified, specifiedColor=<rgba 0.0284 0.0284 0.0284 1>}
      uVRegion {bIsValid=false, max=<0 0>, min=<0 0>}
     }
     pressedForeground {colorUseRule=UseColor_ColorTable, specifiedColor=<rgba 0 0 0 0>}
     pressedPadding=<ltrb 12 2.5 12 0.5>
     pressedSlateSound {resourceObject=null}
    }
   }
   slot {
    class=CanvasPanelSlot
    properties {layoutData {alignment=<0 0>, anchors {maximum=<0 0.0212>, minimum=<0 0.0212>}, offsets=<ltrb 0 -4.62 160 55>}}
   }
  }
  CTRL_L_eyeAim {
   widgetClassPath=&EUW_Picker_Button
   bIsVariable=true
   parent=CanvasPanel_0
   widget=CTRL_L_eyeAim
   properties {
    draw Style=Circle
    label Font Size=6
    colorAndOpacity=<rgba 0.2613 1 0.1894 1>
    normal Tint {colorUseRule=UseColor_Specified, specifiedColor=<rgba 0.6094 0.6094 0.6094 1>}
   }
   slot {
    class=CanvasPanelSlot
    properties {zOrder=2, layoutData {alignment=<0.5 0.5>, anchors {maximum=<0.5 0.5>, minimum=<0.5 0.5>}, offsets=<ltrb -329.4876 -235.8525 15 15>}}
   }
  }
  CTRL_R_eyeAim {
   widgetClassPath=&EUW_Picker_Button
   bIsVariable=true
   parent=CanvasPanel_0
   widget=CTRL_R_eyeAim
   properties {
    draw Style=Circle
    label Font Size=6
    colorAndOpacity=<rgba 1 0.0521 0.0739 1>
    normal Tint {colorUseRule=UseColor_Specified, specifiedColor=<rgba 0.5625 0.5625 0.5625 1>}
   }
   slot {
    class=CanvasPanelSlot
    properties {zOrder=2, layoutData {alignment=<0.5 0.5>, anchors {maximum=<0.5 0.5>, minimum=<0.5 0.5>}, offsets=<ltrb -376.5271 -236.0104 15 15>}}
   }
  }
  Region_C_Mouth {
   widgetClassPath=&CanvasPanel
   bIsVariable=true
   parent=CanvasPanel_0
   widget=Region_C_Mouth
   slot {
    class=CanvasPanelSlot
    properties {
     zOrder=5
     layoutData {alignment=<0.5 0.5>, anchors {maximum=<0.5 0.5>, minimum=<0.5 0.5>}, offsets=<ltrb -350.041 80.6005 285.6323 114.7341>}
    }
   }
  }
  CTRL_L_mouth_corner {
   widgetClassPath=&EUW_Picker_Button
   bIsVariable=true
   parent=Region_C_Mouth
   widget=CTRL_L_mouth_corner
   properties {
    label Font Size=6
    colorAndOpacity=<rgba 0.2613 1 0.1894 1>
    normal Tint {colorUseRule=UseColor_Specified, specifiedColor=<rgba 0.6094 0.6094 0.6094 1>}
   }
   slot {
    class=CanvasPanelSlot
    properties {zOrder=1, layoutData {alignment=<0 0>, anchors {maximum=<0 0>, minimum=<0 0>}, offsets=<ltrb 256 44 20 20>}}
   }
  }
  CTRL_R_mouth_lipU {
   widgetClassPath=&EUW_Picker_Button
   bIsVariable=true
   parent=Region_C_Mouth
   widget=CTRL_R_mouth_lipU
   properties {
    label Font Size=6
    colorAndOpacity=<rgba 1 0.0521 0.0739 1>
    normal Tint {colorUseRule=UseColor_Specified, specifiedColor=<rgba 0.5625 0.5625 0.5625 1>}
   }
   slot {
    class=CanvasPanelSlot
    properties {zOrder=1, layoutData {alignment=<0 0>, anchors {maximum=<0 0>, minimum=<0 0>}, offsets=<ltrb 56 8 80 20>}}
   }
  }
  CTRL_L_mouth_lipD {
   widgetClassPath=&EUW_Picker_Button
   bIsVariable=true
   parent=Region_C_Mouth
   widget=CTRL_L_mouth_lipD
   properties {
    label Font Size=6
    colorAndOpacity=<rgba 0.2613 1 0.1894 1>
    normal Tint {colorUseRule=UseColor_Specified, specifiedColor=<rgba 0.6094 0.6094 0.6094 1>}
   }
   slot {
    class=CanvasPanelSlot
    properties {zOrder=1, layoutData {alignment=<0 0>, anchors {maximum=<0 0>, minimum=<0 0>}, offsets=<ltrb 144 84 80 20>}}
   }
  }
  CTRL_R_mouth_corner {
   widgetClassPath=&EUW_Picker_Button
   bIsVariable=true
   parent=Region_C_Mouth
   widget=CTRL_R_mouth_corner
   properties {
    label Font Size=6
    colorAndOpacity=<rgba 1 0.0521 0.0739 1>
    normal Tint {colorUseRule=UseColor_Specified, specifiedColor=<rgba 0.5625 0.5625 0.5625 1>}
   }
   slot {
    class=CanvasPanelSlot
    properties {zOrder=1, layoutData {alignment=<0 0>, anchors {maximum=<0 0>, minimum=<0 0>}, offsets=<ltrb 8 44 20 20>}}
   }
  }
  CTRL_R_mouth_lipD {
   widgetClassPath=&EUW_Picker_Button
   bIsVariable=true
   parent=Region_C_Mouth
   widget=CTRL_R_mouth_lipD
   properties {
    label Font Size=6
    colorAndOpacity=<rgba 1 0.0521 0.0739 1>
    normal Tint {colorUseRule=UseColor_Specified, specifiedColor=<rgba 0.5625 0.5625 0.5625 1>}
   }
   slot {
    class=CanvasPanelSlot
    properties {zOrder=1, layoutData {alignment=<0 0>, anchors {maximum=<0 0>, minimum=<0 0>}, offsets=<ltrb 56 84 80 20>}}
   }
  }
  CTRL_L_mouth_smileDepress {
   widgetClassPath=&EUW_Picker_Button
   bIsVariable=true
   parent=Region_C_Mouth
   widget=CTRL_L_mouth_smileDepress
   properties {
    label Font Size=6
    colorAndOpacity=<rgba 0.2613 1 0.1894 1>
    normal Tint {colorUseRule=UseColor_Specified, specifiedColor=<rgba 0.6094 0.6094 0.6094 1>}
   }
   slot {
    class=CanvasPanelSlot
    properties {zOrder=1, layoutData {alignment=<0 0>, anchors {maximum=<0 0>, minimum=<0 0>}, offsets=<ltrb 220 40 30 30>}}
   }
  }
  CTRL_R_mouth_smileDepress {
   widgetClassPath=&EUW_Picker_Button
   bIsVariable=true
   parent=Region_C_Mouth
   widget=CTRL_R_mouth_smileDepress
   properties {
    label Font Size=6
    colorAndOpacity=<rgba 1 0.0521 0.0739 1>
    normal Tint {colorUseRule=UseColor_Specified, specifiedColor=<rgba 0.5625 0.5625 0.5625 1>}
   }
   slot {
    class=CanvasPanelSlot
    properties {zOrder=1, layoutData {alignment=<0 0>, anchors {maximum=<0 0>, minimum=<0 0>}, offsets=<ltrb 32 40 30 30>}}
   }
  }
  CTRL_C_mouth {
   widgetClassPath=&EUW_Picker_Button
   bIsVariable=true
   parent=Region_C_Mouth
   widget=CTRL_C_mouth
   properties {
    draw Style=Square
    label Font Size=15
    label Text=Mouth
    colorAndOpacity=<rgba 1 0.9695 0 1>
    label Color {colorUseRule=UseColor_Specified, specifiedColor=<rgba 0 0 0 1>}
    normal Tint {colorUseRule=UseColor_Specified, specifiedColor=<rgba 1 1 1 1>}
   }
   slot {
    class=CanvasPanelSlot
    properties {zOrder=1, layoutData {alignment=<0 0>, anchors {maximum=<0 0>, minimum=<0 0>}, offsets=<ltrb 92 40 100 30>}}
   }
  }
  CTRL_L_mouth_lipU {
   widgetClassPath=&EUW_Picker_Button
   bIsVariable=true
   parent=Region_C_Mouth
   widget=CTRL_L_mouth_lipU
   properties {
    label Font Size=6
    colorAndOpacity=<rgba 0.2613 1 0.1894 1>
    normal Tint {colorUseRule=UseColor_Specified, specifiedColor=<rgba 0.6094 0.6094 0.6094 1>}
   }
   slot {
    class=CanvasPanelSlot
    properties {zOrder=1, layoutData {alignment=<0 0>, anchors {maximum=<0 0>, minimum=<0 0>}, offsets=<ltrb 144 8 80 20>}}
   }
  }
  Region_C_Mouth_Btn {
   widgetClassPath=&EditorUtilityButton
   parent=Region_C_Mouth
   widget=Region_C_Mouth_Btn
   properties {
    backgroundColor=<rgba 1 1 1 0.08>
    colorAndOpacity=<rgba 0.099 0.0951 0.0918 0.4348>
    widgetStyle {
     clickedSlateSound {resourceObject=null}
     disabled {
      bIsDynamicallyLoaded=false
      drawAs=RoundedBox
      imageType=NoImage
      mirroring=NoMirror
      resourceName=None
      resourceObject=null
      tiling=NoTile
      imageSize=<0 0>
      margin=<ltrb 0 0 0 0>
      outlineSettings {
       bUseBrushTransparency=false
       roundingType=FixedRadius
       width=1
       color {colorUseRule=UseColor_ColorTable, specifiedColor=<rgba 0 0 0 0>}
       cornerRadii=<4 4 4 4>
      }
      tintColor {colorUseRule=UseColor_Specified, specifiedColor=<rgba 0.0395 0.0395 0.0395 1>}
      uVRegion {bIsValid=false, max=<0 0>, min=<0 0>}
     }
     disabledForeground {colorUseRule=UseColor_ColorTable, specifiedColor=<rgba 0 0 0 0>}
     hovered {
      bIsDynamicallyLoaded=false
      drawAs=RoundedBox
      imageType=NoImage
      mirroring=NoMirror
      resourceName=None
      resourceObject=null
      tiling=NoTile
      imageSize=<32 32>
      margin=<ltrb 0 0 0 0>
      outlineSettings {
       bUseBrushTransparency=false
       roundingType=FixedRadius
       width=1
       color {colorUseRule=UseColor_ColorTable, specifiedColor=<rgba 0 0 0 0>}
       cornerRadii=<4 4 4 4>
      }
      tintColor {colorUseRule=UseColor_Specified, specifiedColor=<rgba 0.0953 0.0953 0.0953 1>}
      uVRegion {bIsValid=false, max=<0 0>, min=<0 0>}
     }
     hoveredForeground {colorUseRule=UseColor_ColorTable, specifiedColor=<rgba 0 0 0 0>}
     hoveredSlateSound {resourceObject=null}
     normal {
      bIsDynamicallyLoaded=false
      drawAs=RoundedBox
      imageType=NoImage
      mirroring=NoMirror
      resourceName=None
      resourceObject=null
      tiling=NoTile
      imageSize=<32 32>
      margin=<ltrb 0 0 0 0>
      outlineSettings {
       bUseBrushTransparency=false
       roundingType=FixedRadius
       width=0
       color {colorUseRule=UseColor_Specified, specifiedColor=<rgba 0 0 0 1>}
       cornerRadii=<8 8 8 8>
      }
      tintColor {colorUseRule=UseColor_Specified, specifiedColor=<rgba 1 1 1 1>}
      uVRegion {bIsValid=false, max=<0 0>, min=<0 0>}
     }
     normalForeground {colorUseRule=UseColor_ColorTable, specifiedColor=<rgba 0 0 0 0>}
     normalPadding=<ltrb 12 1.5 12 1.5>
     pressed {
      bIsDynamicallyLoaded=false
      drawAs=RoundedBox
      imageType=NoImage
      mirroring=NoMirror
      resourceName=None
      resourceObject=null
      tiling=NoTile
      imageSize=<32 32>
      margin=<ltrb 0 0 0 0>
      outlineSettings {
       bUseBrushTransparency=false
       roundingType=FixedRadius
       width=1
       color {colorUseRule=UseColor_ColorTable, specifiedColor=<rgba 0 0 0 0>}
       cornerRadii=<4 4 4 4>
      }
      tintColor {colorUseRule=UseColor_Specified, specifiedColor=<rgba 0.0284 0.0284 0.0284 1>}
      uVRegion {bIsValid=false, max=<0 0>, min=<0 0>}
     }
     pressedForeground {colorUseRule=UseColor_ColorTable, specifiedColor=<rgba 0 0 0 0>}
     pressedPadding=<ltrb 12 2.5 12 0.5>
     pressedSlateSound {resourceObject=null}
    }
   }
   slot {
    class=CanvasPanelSlot
    properties {layoutData {alignment=<0 0>, anchors {maximum=<0.5 0.5>, minimum=<0.5 0.5>}, offsets=<ltrb -142.8161 -57.3671 286.0193 114.5746>}}
   }
  }
  Region_C_Jaw {
   widgetClassPath=&CanvasPanel
   bIsVariable=true
   parent=CanvasPanel_0
   widget=Region_C_Jaw
   slot {
    class=CanvasPanelSlot
    properties {zOrder=3, layoutData {alignment=<0.5 0.5>, anchors {maximum=<0.5 0.5>, minimum=<0.5 0.5>}, offsets=<ltrb -349.9407 122.5512 383.2296 262.7>}}
   }
  }
  CTRL_L_mouth_suckBlow {
   widgetClassPath=&EUW_Picker_Button
   bIsVariable=true
   parent=Region_C_Jaw
   widget=CTRL_L_mouth_suckBlow
   properties {
    label Font Size=6
    colorAndOpacity=<rgba 0.2613 1 0.1894 1>
    normal Tint {colorUseRule=UseColor_Specified, specifiedColor=<rgba 0.6094 0.6094 0.6094 1>}
   }
   slot {
    class=CanvasPanelSlot
    properties {zOrder=1, layoutData {alignment=<0 0>, anchors {maximum=<0 0>, minimum=<0 0>}, offsets=<ltrb 340 12 30 30>}}
   }
  }
  CTRL_R_mouth_suckBlow {
   widgetClassPath=&EUW_Picker_Button
   bIsVariable=true
   parent=Region_C_Jaw
   widget=CTRL_R_mouth_suckBlow
   properties {
    label Font Size=6
    colorAndOpacity=<rgba 1 0.0521 0.0739 1>
    normal Tint {colorUseRule=UseColor_Specified, specifiedColor=<rgba 0.5625 0.5625 0.5625 1>}
   }
   slot {
    class=CanvasPanelSlot
    properties {zOrder=1, layoutData {alignment=<0 0>, anchors {maximum=<0 0>, minimum=<0 0>}, offsets=<ltrb 12 12 30 30>}}
   }
  }
  CTRL_C_jaw {
   widgetClassPath=&EUW_Picker_Button
   bIsVariable=true
   parent=Region_C_Jaw
   widget=CTRL_C_jaw
   properties {
    draw Style=Square
    label Font Size=18
    label Text=Jaw
    colorAndOpacity=<rgba 1 0.9695 0 1>
    label Color {colorUseRule=UseColor_Specified, specifiedColor=<rgba 0 0 0 1>}
    normal Tint {colorUseRule=UseColor_Specified, specifiedColor=<rgba 1 1 1 1>}
   }
   slot {
    class=CanvasPanelSlot
    properties {zOrder=1, layoutData {alignment=<0 0>, anchors {maximum=<0 0>, minimum=<0 0>}, offsets=<ltrb 120 200 129.2793 44.2643>}}
   }
  }
  Region_C_Jaw_Btn {
   widgetClassPath=&EditorUtilityButton
   parent=Region_C_Jaw
   widget=Region_C_Jaw_Btn
   properties {
    backgroundColor=<rgba 1 1 1 0.08>
    colorAndOpacity=<rgba 0.099 0.0951 0.0918 0.4348>
    widgetStyle {
     clickedSlateSound {resourceObject=null}
     disabled {
      bIsDynamicallyLoaded=false
      drawAs=RoundedBox
      imageType=NoImage
      mirroring=NoMirror
      resourceName=None
      resourceObject=null
      tiling=NoTile
      imageSize=<0 0>
      margin=<ltrb 0 0 0 0>
      outlineSettings {
       bUseBrushTransparency=false
       roundingType=FixedRadius
       width=1
       color {colorUseRule=UseColor_ColorTable, specifiedColor=<rgba 0 0 0 0>}
       cornerRadii=<4 4 4 4>
      }
      tintColor {colorUseRule=UseColor_Specified, specifiedColor=<rgba 0.0395 0.0395 0.0395 1>}
      uVRegion {bIsValid=false, max=<0 0>, min=<0 0>}
     }
     disabledForeground {colorUseRule=UseColor_ColorTable, specifiedColor=<rgba 0 0 0 0>}
     hovered {
      bIsDynamicallyLoaded=false
      drawAs=RoundedBox
      imageType=NoImage
      mirroring=NoMirror
      resourceName=None
      resourceObject=null
      tiling=NoTile
      imageSize=<32 32>
      margin=<ltrb 0 0 0 0>
      outlineSettings {
       bUseBrushTransparency=false
       roundingType=FixedRadius
       width=1
       color {colorUseRule=UseColor_ColorTable, specifiedColor=<rgba 0 0 0 0>}
       cornerRadii=<4 4 4 4>
      }
      tintColor {colorUseRule=UseColor_Specified, specifiedColor=<rgba 0.0953 0.0953 0.0953 1>}
      uVRegion {bIsValid=false, max=<0 0>, min=<0 0>}
     }
     hoveredForeground {colorUseRule=UseColor_ColorTable, specifiedColor=<rgba 0 0 0 0>}
     hoveredSlateSound {resourceObject=null}
     normal {
      bIsDynamicallyLoaded=false
      drawAs=RoundedBox
      imageType=NoImage
      mirroring=NoMirror
      resourceName=None
      resourceObject=null
      tiling=NoTile
      imageSize=<32 32>
      margin=<ltrb 0 0 0 0>
      outlineSettings {
       bUseBrushTransparency=false
       roundingType=FixedRadius
       width=0
       color {colorUseRule=UseColor_Specified, specifiedColor=<rgba 0 0 0 1>}
       cornerRadii=<8 8 8 8>
      }
      tintColor {colorUseRule=UseColor_Specified, specifiedColor=<rgba 1 1 1 1>}
      uVRegion {bIsValid=false, max=<0 0>, min=<0 0>}
     }
     normalForeground {colorUseRule=UseColor_ColorTable, specifiedColor=<rgba 0 0 0 0>}
     normalPadding=<ltrb 12 1.5 12 1.5>
     pressed {
      bIsDynamicallyLoaded=false
      drawAs=RoundedBox
      imageType=NoImage
      mirroring=NoMirror
      resourceName=None
      resourceObject=null
      tiling=NoTile
      imageSize=<32 32>
      margin=<ltrb 0 0 0 0>
      outlineSettings {
       bUseBrushTransparency=false
       roundingType=FixedRadius
       width=1
       color {colorUseRule=UseColor_ColorTable, specifiedColor=<rgba 0 0 0 0>}
       cornerRadii=<4 4 4 4>
      }
      tintColor {colorUseRule=UseColor_Specified, specifiedColor=<rgba 0.0284 0.0284 0.0284 1>}
      uVRegion {bIsValid=false, max=<0 0>, min=<0 0>}
     }
     pressedForeground {colorUseRule=UseColor_ColorTable, specifiedColor=<rgba 0 0 0 0>}
     pressedPadding=<ltrb 12 2.5 12 0.5>
     pressedSlateSound {resourceObject=null}
    }
   }
   slot {
    class=CanvasPanelSlot
    properties {layoutData {alignment=<0 0>, anchors {maximum=<0.5 0.5>, minimum=<0.5 0.5>}, offsets=<ltrb -191.6148 -131.35 380.7808 260.7207>}}
   }
  }
  CTRL_C_teethD {
   widgetClassPath=&EUW_Picker_Button
   bIsVariable=true
   parent=CanvasPanel_0
   widget=CTRL_C_teethD
   properties {label Font Size=6, normal Tint {colorUseRule=UseColor_Specified, specifiedColor=<rgba 1 1 1 1>}}
   slot {
    class=CanvasPanelSlot
    properties {layoutData {alignment=<0.5 0.5>, anchors {maximum=<0.5 0.5>, minimum=<0.5 0.5>}, offsets=<ltrb -91.6189 123.6475 69.7375 25>}}
   }
  }
  CTRL_C_teethU {
   widgetClassPath=&EUW_Picker_Button
   bIsVariable=true
   parent=CanvasPanel_0
   widget=CTRL_C_teethU
   properties {label Font Size=6, normal Tint {colorUseRule=UseColor_Specified, specifiedColor=<rgba 1 1 1 1>}}
   slot {
    class=CanvasPanelSlot
    properties {layoutData {alignment=<0.5 0.5>, anchors {maximum=<0.5 0.5>, minimum=<0.5 0.5>}, offsets=<ltrb -91.6189 207.6475 74.2112 25>}}
   }
  }
  CTRL_C_eyesAim {
   widgetClassPath=&EUW_Picker_Button
   bIsVariable=true
   parent=CanvasPanel_0
   widget=CTRL_C_eyesAim
   properties {
    label Text=Aim
    colorAndOpacity=<rgba 1 0.9695 0 1>
    label Color {colorUseRule=UseColor_Specified, specifiedColor=<rgba 0 0 0 1>}
    normal Tint {colorUseRule=UseColor_Specified, specifiedColor=<rgba 0.6771 0.6771 0.6771 1>}
   }
   slot {
    class=CanvasPanelSlot
    properties {layoutData {alignment=<0.5 0.5>, anchors {maximum=<0.5 0.5>, minimum=<0.5 0.5>}, offsets=<ltrb -352.7728 -236.9267 76.5081 36.7544>}}
   }
  }
  Graphics {
   widgetClassPath=&CanvasPanel
   parent=CanvasPanel_0
   widget=Graphics
   slot {class=CanvasPanelSlot, properties {layoutData {alignment=<0.5 0.5>, anchors {maximum=<1 1>, minimum=<0 0>}, offsets=<ltrb 0 0 0 0>}}}
  }
  Image_42 {
   widgetClassPath=&Image
   bIsVariable=true
   parent=Graphics
   widget=Image_42
   properties {
    brush {
     bIsDynamicallyLoaded=false
     drawAs=Image
     imageType=FullColor
     mirroring=NoMirror
     resourceName=None
     resourceObject=/Game/Characters/Player/Common/SharedRigs/Graphics/JONESY_HEAD
     tiling=NoTile
     imageSize=<666 847>
     margin=<ltrb 0 0 0 0>
     outlineSettings {
      bUseBrushTransparency=false
      roundingType=HalfHeightRadius
      width=0
      color {colorUseRule=UseColor_Specified, specifiedColor=<rgba 0 0 0 0>}
      cornerRadii=<0 0 0 0>
     }
     tintColor {colorUseRule=UseColor_Specified, specifiedColor=<rgba 1 1 1 1>}
     uVRegion {bIsValid=false, max=<0 0>, min=<0 0>}
    }
    colorAndOpacity=<rgba 1 1 1 0.4>
   }
   slot {
    class=CanvasPanelSlot
    properties {layoutData {alignment=<0.5 0.5>, anchors {maximum=<0.5 0.5>, minimum=<0.5 0.5>}, offsets=<ltrb -365.6156 -52.8803 681.4449 839.1203>}}
   }
  }
  Image_11 {
   widgetClassPath=&Image
   bIsVariable=true
   parent=Graphics
   widget=Image_11
   properties {
    brush {
     bIsDynamicallyLoaded=false
     drawAs=Image
     imageType=FullColor
     mirroring=NoMirror
     resourceName=None
     resourceObject=/Game/Characters/Player/Common/SharedRigs/Graphics/MI_BorderWTCH
     tiling=NoTile
     imageSize=<99 92>
     margin=<ltrb 0 0 0 0>
     outlineSettings {
      bUseBrushTransparency=false
      roundingType=HalfHeightRadius
      width=0
      color {colorUseRule=UseColor_Specified, specifiedColor=<rgba 0 0 0 0>}
      cornerRadii=<0 0 0 0>
     }
     tintColor {colorUseRule=UseColor_Specified, specifiedColor=<rgba 1 1 1 1>}
     uVRegion {bIsValid=false, max=<0 0>, min=<0 0>}
    }
    colorAndOpacity=<rgba 1 1 1 0.4>
   }
   slot {
    class=CanvasPanelSlot
    properties {layoutData {alignment=<0.5 0.5>, anchors {maximum=<0.5 0.5>, minimum=<0.5 0.5>}, offsets=<ltrb -358.9153 -382.2201 256.2196 92.0574>}}
   }
  }
  Image_12 {
   widgetClassPath=&Image
   bIsVariable=true
   parent=Graphics
   widget=Image_12
   properties {
    brush {
     bIsDynamicallyLoaded=false
     drawAs=Image
     imageType=FullColor
     mirroring=NoMirror
     resourceName=None
     resourceObject=&MI_BorderWTCH
     tiling=NoTile
     imageSize=<99 92>
     margin=<ltrb 0 0 0 0>
     outlineSettings {
      bUseBrushTransparency=false
      roundingType=HalfHeightRadius
      width=0
      color {colorUseRule=UseColor_Specified, specifiedColor=<rgba 0 0 0 0>}
      cornerRadii=<0 0 0 0>
     }
     tintColor {colorUseRule=UseColor_Specified, specifiedColor=<rgba 1 1 1 1>}
     uVRegion {bIsValid=false, max=<0 0>, min=<0 0>}
    }
    colorAndOpacity=<rgba 1 1 1 0.4>
   }
   slot {
    class=CanvasPanelSlot
    properties {layoutData {alignment=<0.5 0.5>, anchors {maximum=<0.5 0.5>, minimum=<0.5 0.5>}, offsets=<ltrb -346.7258 393.9238 270.205 57.1957>}}
   }
  }
  Image_14 {
   widgetClassPath=&Image
   bIsVariable=true
   parent=Graphics
   widget=Image_14
   properties {
    brush {
     bIsDynamicallyLoaded=false
     drawAs=Image
     imageType=FullColor
     mirroring=NoMirror
     resourceName=None
     resourceObject=/Game/Characters/Player/Common/SharedRigs/Graphics/MI_Border_thin
     tiling=NoTile
     imageSize=<99 92>
     margin=<ltrb 0 0 0 0>
     outlineSettings {
      bUseBrushTransparency=false
      roundingType=HalfHeightRadius
      width=0
      color {colorUseRule=UseColor_Specified, specifiedColor=<rgba 0 0 0 0>}
      cornerRadii=<0 0 0 0>
     }
     tintColor {colorUseRule=UseColor_Specified, specifiedColor=<rgba 1 1 1 1>}
     uVRegion {bIsValid=false, max=<0 0>, min=<0 0>}
    }
    colorAndOpacity=<rgba 1 1 1 0.4>
   }
   slot {
    class=CanvasPanelSlot
    properties {layoutData {alignment=<0.5 0.5>, anchors {maximum=<0.5 0.5>, minimum=<0.5 0.5>}, offsets=<ltrb -353.3933 88.3609 963.2313 1213.6868>}}
   }
  }
  TextBlock_1 {
   widgetClassPath=&TextBlock
   parent=Graphics
   widget=TextBlock_1
   properties {
    text=tongue
    colorAndOpacity {colorUseRule=UseColor_Specified, specifiedColor=<rgba 1 1 1 0.5>}
    font {
     bForceMonospaced=false
     bMaterialIsStencil=false
     fontMaterial=null
     fontObject=&Roboto
     letterSpacing=0
     monospacedWidth=1
     size=11.25
     skewAmount=0
     typefaceFontName=Bold
     outlineSettings {
      bApplyOutlineToDropShadows=false
      bMiteredCorners=false
      bSeparateFillAlpha=false
      outlineMaterial=null
      outlineSize=0
      outlineColor=<rgba 0 0 0 1>
     }
    }
   }
   slot {
    class=CanvasPanelSlot
    properties {bAutoSize=true, layoutData {alignment=<0 0>, anchors {maximum=<0.5 0.5>, minimum=<0.5 0.5>}, offsets=<ltrb -58.4876 155.1475 100 30>}}
   }
  }
  TextBlock {
   widgetClassPath=&TextBlock
   parent=Graphics
   widget=TextBlock
   properties {
    text=teeth
    colorAndOpacity {colorUseRule=UseColor_Specified, specifiedColor=<rgba 1 1 1 0.5>}
    font {
     bForceMonospaced=false
     bMaterialIsStencil=false
     fontMaterial=null
     fontObject=&Roboto
     letterSpacing=0
     monospacedWidth=1
     size=11.25
     skewAmount=0
     typefaceFontName=Bold
     outlineSettings {
      bApplyOutlineToDropShadows=false
      bMiteredCorners=false
      bSeparateFillAlpha=false
      outlineMaterial=null
      outlineSize=0
      outlineColor=<rgba 0 0 0 1>
     }
    }
   }
   slot {
    class=CanvasPanelSlot
    properties {bAutoSize=true, layoutData {alignment=<0 0>, anchors {maximum=<0.5 0.5>, minimum=<0.5 0.5>}, offsets=<ltrb -108.9609 87.4595 100 30>}}
   }
  }
  TextBlock_16 {
   widgetClassPath=&TextBlock
   parent=Graphics
   widget=TextBlock_16
   properties {
    text=Eye Aim
    colorAndOpacity {colorUseRule=UseColor_Specified, specifiedColor=<rgba 1 1 1 0.5>}
    font {
     bForceMonospaced=false
     bMaterialIsStencil=false
     fontMaterial=null
     fontObject=&Roboto
     letterSpacing=0
     monospacedWidth=1
     size=15
     skewAmount=0
     typefaceFontName=Bold
     outlineSettings {
      bApplyOutlineToDropShadows=false
      bMiteredCorners=false
      bSeparateFillAlpha=false
      outlineMaterial=null
      outlineSize=0
      outlineColor=<rgba 0 0 0 1>
     }
    }
   }
   slot {
    class=CanvasPanelSlot
    properties {bAutoSize=true, layoutData {alignment=<0 0>, anchors {maximum=<0.5 0.5>, minimum=<0.5 0.5>}, offsets=<ltrb -399.5289 -460.249 100 30>}}
   }
  }
  TextBlock_17 {
   widgetClassPath=&TextBlock
   parent=Graphics
   widget=TextBlock_17
   properties {
    text=face GUI follow
    colorAndOpacity {colorUseRule=UseColor_Specified, specifiedColor=<rgba 1 1 1 0.5>}
    font {
     bForceMonospaced=false
     bMaterialIsStencil=false
     fontMaterial=null
     fontObject=&Roboto
     letterSpacing=0
     monospacedWidth=1
     size=11.25
     skewAmount=0
     typefaceFontName=Bold
     outlineSettings {
      bApplyOutlineToDropShadows=false
      bMiteredCorners=false
      bSeparateFillAlpha=false
      outlineMaterial=null
      outlineSize=0
      outlineColor=<rgba 0 0 0 1>
     }
    }
   }
   slot {
    class=CanvasPanelSlot
    properties {bAutoSize=true, layoutData {alignment=<0 0>, anchors {maximum=<0.5 0.5>, minimum=<0.5 0.5>}, offsets=<ltrb -465.8283 385.3259 100 7.7778>}}
   }
  }
  TextBlock_18 {
   widgetClassPath=&TextBlock
   parent=Graphics
   widget=TextBlock_18
   properties {
    text=eyes aim follow
    colorAndOpacity {colorUseRule=UseColor_Specified, specifiedColor=<rgba 1 1 1 0.5>}
    font {
     bForceMonospaced=false
     bMaterialIsStencil=false
     fontMaterial=null
     fontObject=&Roboto
     letterSpacing=0
     monospacedWidth=1
     size=11.25
     skewAmount=0
     typefaceFontName=Bold
     outlineSettings {
      bApplyOutlineToDropShadows=false
      bMiteredCorners=false
      bSeparateFillAlpha=false
      outlineMaterial=null
      outlineSize=0
      outlineColor=<rgba 0 0 0 1>
     }
    }
   }
   slot {
    class=CanvasPanelSlot
    properties {bAutoSize=true, layoutData {alignment=<0 0>, anchors {maximum=<0.5 0.5>, minimum=<0.5 0.5>}, offsets=<ltrb -475.8277 -368.9993 100 7.7778>}}
   }
  }
  TextBlock_2 {
   widgetClassPath=&TextBlock
   parent=Graphics
   widget=TextBlock_2
   properties {
    text=eyes aim switch
    colorAndOpacity {colorUseRule=UseColor_Specified, specifiedColor=<rgba 1 1 1 0.5>}
    font {
     bForceMonospaced=false
     bMaterialIsStencil=false
     fontMaterial=null
     fontObject=&Roboto
     letterSpacing=0
     monospacedWidth=1
     size=11.25
     skewAmount=0
     typefaceFontName=Bold
     outlineSettings {
      bApplyOutlineToDropShadows=false
      bMiteredCorners=false
      bSeparateFillAlpha=false
      outlineMaterial=null
      outlineSize=0
      outlineColor=<rgba 0 0 0 1>
     }
    }
   }
   slot {
    class=CanvasPanelSlot
    properties {bAutoSize=true, layoutData {alignment=<0 0>, anchors {maximum=<0.5 0.5>, minimum=<0.5 0.5>}, offsets=<ltrb -477.0586 -413.404 100 7.7778>}}
   }
  }
  Select_All_Face {
   widgetClassPath=&EditorUtilityButton
   parent=CanvasPanel_0
   widget=Select_All_Face
   properties {widgetStyle {
     clickedSlateSound {resourceObject=null}
     disabled {
      bIsDynamicallyLoaded=false
      drawAs=RoundedBox
      imageType=NoImage
      mirroring=NoMirror
      resourceName=None
      resourceObject=null
      tiling=NoTile
      imageSize=<0 0>
      margin=<ltrb 0 0 0 0>
      outlineSettings {
       bUseBrushTransparency=false
       roundingType=FixedRadius
       width=1
       color {colorUseRule=UseColor_ColorTable, specifiedColor=<rgba 0 0 0 0>}
       cornerRadii=<4 4 4 4>
      }
      tintColor {colorUseRule=UseColor_Specified, specifiedColor=<rgba 0.0395 0.0395 0.0395 1>}
      uVRegion {bIsValid=false, max=<0 0>, min=<0 0>}
     }
     disabledForeground {colorUseRule=UseColor_ColorTable, specifiedColor=<rgba 0 0 0 0>}
     hovered {
      bIsDynamicallyLoaded=false
      drawAs=RoundedBox
      imageType=NoImage
      mirroring=NoMirror
      resourceName=None
      resourceObject=null
      tiling=NoTile
      imageSize=<32 32>
      margin=<ltrb 0 0 0 0>
      outlineSettings {
       bUseBrushTransparency=false
       roundingType=FixedRadius
       width=1
       color {colorUseRule=UseColor_ColorTable, specifiedColor=<rgba 0 0 0 0>}
       cornerRadii=<4 4 4 4>
      }
      tintColor {colorUseRule=UseColor_Specified, specifiedColor=<rgba 0.0953 0.0953 0.0953 1>}
      uVRegion {bIsValid=false, max=<0 0>, min=<0 0>}
     }
     hoveredForeground {colorUseRule=UseColor_ColorTable, specifiedColor=<rgba 0 0 0 0>}
     hoveredSlateSound {resourceObject=null}
     normal {
      bIsDynamicallyLoaded=false
      drawAs=RoundedBox
      imageType=NoImage
      mirroring=NoMirror
      resourceName=None
      resourceObject=null
      tiling=NoTile
      imageSize=<32 32>
      margin=<ltrb 0 0 0 0>
      outlineSettings {
       bUseBrushTransparency=false
       roundingType=FixedRadius
       width=1
       color {colorUseRule=UseColor_ColorTable, specifiedColor=<rgba 0 0 0 0>}
       cornerRadii=<4 4 4 4>
      }
      tintColor {colorUseRule=UseColor_Specified, specifiedColor=<rgba 1 0.2115 0 1>}
      uVRegion {bIsValid=false, max=<0 0>, min=<0 0>}
     }
     normalForeground {colorUseRule=UseColor_ColorTable, specifiedColor=<rgba 0 0 0 0>}
     normalPadding=<ltrb 12 1.5 12 1.5>
     pressed {
      bIsDynamicallyLoaded=false
      drawAs=RoundedBox
      imageType=NoImage
      mirroring=NoMirror
      resourceName=None
      resourceObject=null
      tiling=NoTile
      imageSize=<32 32>
      margin=<ltrb 0 0 0 0>
      outlineSettings {
       bUseBrushTransparency=false
       roundingType=FixedRadius
       width=1
       color {colorUseRule=UseColor_ColorTable, specifiedColor=<rgba 0 0 0 0>}
       cornerRadii=<4 4 4 4>
      }
      tintColor {colorUseRule=UseColor_Specified, specifiedColor=<rgba 0.0284 0.0284 0.0284 1>}
      uVRegion {bIsValid=false, max=<0 0>, min=<0 0>}
     }
     pressedForeground {colorUseRule=UseColor_ColorTable, specifiedColor=<rgba 0 0 0 0>}
     pressedPadding=<ltrb 12 2.5 12 0.5>
     pressedSlateSound {resourceObject=null}
    }}
   slot {
    class=CanvasPanelSlot
    properties {layoutData {alignment=<0 0>, anchors {maximum=<0.5 0.5>, minimum=<0.5 0.5>}, offsets=<ltrb -24.9609 -228.5405 163.0631 47.5075>}}
   }
  }
  TextBlock_150 {widgetClassPath=&TextBlock, parent=Select_All_Face, widget=TextBlock_150, properties {text=All Face}, slot {class=ButtonSlot}}
  Select_Upper_Face {
   widgetClassPath=&EditorUtilityButton
   parent=CanvasPanel_0
   widget=Select_Upper_Face
   properties {widgetStyle {
     clickedSlateSound {resourceObject=null}
     disabled {
      bIsDynamicallyLoaded=false
      drawAs=RoundedBox
      imageType=NoImage
      mirroring=NoMirror
      resourceName=None
      resourceObject=null
      tiling=NoTile
      imageSize=<0 0>
      margin=<ltrb 0 0 0 0>
      outlineSettings {
       bUseBrushTransparency=false
       roundingType=FixedRadius
       width=1
       color {colorUseRule=UseColor_ColorTable, specifiedColor=<rgba 0 0 0 0>}
       cornerRadii=<4 4 4 4>
      }
      tintColor {colorUseRule=UseColor_Specified, specifiedColor=<rgba 0.0395 0.0395 0.0395 1>}
      uVRegion {bIsValid=false, max=<0 0>, min=<0 0>}
     }
     disabledForeground {colorUseRule=UseColor_ColorTable, specifiedColor=<rgba 0 0 0 0>}
     hovered {
      bIsDynamicallyLoaded=false
      drawAs=RoundedBox
      imageType=NoImage
      mirroring=NoMirror
      resourceName=None
      resourceObject=null
      tiling=NoTile
      imageSize=<32 32>
      margin=<ltrb 0 0 0 0>
      outlineSettings {
       bUseBrushTransparency=false
       roundingType=FixedRadius
       width=1
       color {colorUseRule=UseColor_ColorTable, specifiedColor=<rgba 0 0 0 0>}
       cornerRadii=<4 4 4 4>
      }
      tintColor {colorUseRule=UseColor_Specified, specifiedColor=<rgba 0.0953 0.0953 0.0953 1>}
      uVRegion {bIsValid=false, max=<0 0>, min=<0 0>}
     }
     hoveredForeground {colorUseRule=UseColor_ColorTable, specifiedColor=<rgba 0 0 0 0>}
     hoveredSlateSound {resourceObject=null}
     normal {
      bIsDynamicallyLoaded=false
      drawAs=RoundedBox
      imageType=NoImage
      mirroring=NoMirror
      resourceName=None
      resourceObject=null
      tiling=NoTile
      imageSize=<32 32>
      margin=<ltrb 0 0 0 0>
      outlineSettings {
       bUseBrushTransparency=false
       roundingType=FixedRadius
       width=1
       color {colorUseRule=UseColor_ColorTable, specifiedColor=<rgba 0 0 0 0>}
       cornerRadii=<4 4 4 4>
      }
      tintColor {colorUseRule=UseColor_Specified, specifiedColor=<rgba 1 0.2115 0 1>}
      uVRegion {bIsValid=false, max=<0 0>, min=<0 0>}
     }
     normalForeground {colorUseRule=UseColor_ColorTable, specifiedColor=<rgba 0 0 0 0>}
     normalPadding=<ltrb 12 1.5 12 1.5>
     pressed {
      bIsDynamicallyLoaded=false
      drawAs=RoundedBox
      imageType=NoImage
      mirroring=NoMirror
      resourceName=None
      resourceObject=null
      tiling=NoTile
      imageSize=<32 32>
      margin=<ltrb 0 0 0 0>
      outlineSettings {
       bUseBrushTransparency=false
       roundingType=FixedRadius
       width=1
       color {colorUseRule=UseColor_ColorTable, specifiedColor=<rgba 0 0 0 0>}
       cornerRadii=<4 4 4 4>
      }
      tintColor {colorUseRule=UseColor_Specified, specifiedColor=<rgba 0.0284 0.0284 0.0284 1>}
      uVRegion {bIsValid=false, max=<0 0>, min=<0 0>}
     }
     pressedForeground {colorUseRule=UseColor_ColorTable, specifiedColor=<rgba 0 0 0 0>}
     pressedPadding=<ltrb 12 2.5 12 0.5>
     pressedSlateSound {resourceObject=null}
    }}
   slot {
    class=CanvasPanelSlot
    properties {layoutData {alignment=<0 0>, anchors {maximum=<0.5 0.5>, minimum=<0.5 0.5>}, offsets=<ltrb -24.9609 -168.5405 163.0631 47.5075>}}
   }
  }
  TextBlock_3 {
   widgetClassPath=&TextBlock
   parent=Select_Upper_Face
   widget=TextBlock_3
   properties {
    text=Upper Face
    font {
     bForceMonospaced=false
     bMaterialIsStencil=false
     fontMaterial=null
     fontObject=&Roboto
     letterSpacing=0
     monospacedWidth=1
     size=18
     skewAmount=0
     typefaceFontName=Bold
     outlineSettings {
      bApplyOutlineToDropShadows=false
      bMiteredCorners=false
      bSeparateFillAlpha=false
      outlineMaterial=null
      outlineSize=0
      outlineColor=<rgba 0 0 0 1>
     }
    }
   }
   slot {class=ButtonSlot}
  }
  Select_Lower_Face {
   widgetClassPath=&EditorUtilityButton
   parent=CanvasPanel_0
   widget=Select_Lower_Face
   properties {widgetStyle {
     clickedSlateSound {resourceObject=null}
     disabled {
      bIsDynamicallyLoaded=false
      drawAs=RoundedBox
      imageType=NoImage
      mirroring=NoMirror
      resourceName=None
      resourceObject=null
      tiling=NoTile
      imageSize=<0 0>
      margin=<ltrb 0 0 0 0>
      outlineSettings {
       bUseBrushTransparency=false
       roundingType=FixedRadius
       width=1
       color {colorUseRule=UseColor_ColorTable, specifiedColor=<rgba 0 0 0 0>}
       cornerRadii=<4 4 4 4>
      }
      tintColor {colorUseRule=UseColor_Specified, specifiedColor=<rgba 0.0395 0.0395 0.0395 1>}
      uVRegion {bIsValid=false, max=<0 0>, min=<0 0>}
     }
     disabledForeground {colorUseRule=UseColor_ColorTable, specifiedColor=<rgba 0 0 0 0>}
     hovered {
      bIsDynamicallyLoaded=false
      drawAs=RoundedBox
      imageType=NoImage
      mirroring=NoMirror
      resourceName=None
      resourceObject=null
      tiling=NoTile
      imageSize=<32 32>
      margin=<ltrb 0 0 0 0>
      outlineSettings {
       bUseBrushTransparency=false
       roundingType=FixedRadius
       width=1
       color {colorUseRule=UseColor_ColorTable, specifiedColor=<rgba 0 0 0 0>}
       cornerRadii=<4 4 4 4>
      }
      tintColor {colorUseRule=UseColor_Specified, specifiedColor=<rgba 0.0953 0.0953 0.0953 1>}
      uVRegion {bIsValid=false, max=<0 0>, min=<0 0>}
     }
     hoveredForeground {colorUseRule=UseColor_ColorTable, specifiedColor=<rgba 0 0 0 0>}
     hoveredSlateSound {resourceObject=null}
     normal {
      bIsDynamicallyLoaded=false
      drawAs=RoundedBox
      imageType=NoImage
      mirroring=NoMirror
      resourceName=None
      resourceObject=null
      tiling=NoTile
      imageSize=<32 32>
      margin=<ltrb 0 0 0 0>
      outlineSettings {
       bUseBrushTransparency=false
       roundingType=FixedRadius
       width=1
       color {colorUseRule=UseColor_ColorTable, specifiedColor=<rgba 0 0 0 0>}
       cornerRadii=<4 4 4 4>
      }
      tintColor {colorUseRule=UseColor_Specified, specifiedColor=<rgba 1 0.2115 0 1>}
      uVRegion {bIsValid=false, max=<0 0>, min=<0 0>}
     }
     normalForeground {colorUseRule=UseColor_ColorTable, specifiedColor=<rgba 0 0 0 0>}
     normalPadding=<ltrb 12 1.5 12 1.5>
     pressed {
      bIsDynamicallyLoaded=false
      drawAs=RoundedBox
      imageType=NoImage
      mirroring=NoMirror
      resourceName=None
      resourceObject=null
      tiling=NoTile
      imageSize=<32 32>
      margin=<ltrb 0 0 0 0>
      outlineSettings {
       bUseBrushTransparency=false
       roundingType=FixedRadius
       width=1
       color {colorUseRule=UseColor_ColorTable, specifiedColor=<rgba 0 0 0 0>}
       cornerRadii=<4 4 4 4>
      }
      tintColor {colorUseRule=UseColor_Specified, specifiedColor=<rgba 0.0284 0.0284 0.0284 1>}
      uVRegion {bIsValid=false, max=<0 0>, min=<0 0>}
     }
     pressedForeground {colorUseRule=UseColor_ColorTable, specifiedColor=<rgba 0 0 0 0>}
     pressedPadding=<ltrb 12 2.5 12 0.5>
     pressedSlateSound {resourceObject=null}
    }}
   slot {
    class=CanvasPanelSlot
    properties {layoutData {alignment=<0 0>, anchors {maximum=<0.5 0.5>, minimum=<0.5 0.5>}, offsets=<ltrb -24.9609 -104.5405 163.0631 47.5075>}}
   }
  }
  TextBlock_4 {
   widgetClassPath=&TextBlock
   parent=Select_Lower_Face
   widget=TextBlock_4
   properties {
    text=Lower Face
    font {
     bForceMonospaced=false
     bMaterialIsStencil=false
     fontMaterial=null
     fontObject=&Roboto
     letterSpacing=0
     monospacedWidth=1
     size=18
     skewAmount=0
     typefaceFontName=Bold
     outlineSettings {
      bApplyOutlineToDropShadows=false
      bMiteredCorners=false
      bSeparateFillAlpha=false
      outlineMaterial=null
      outlineSize=0
      outlineColor=<rgba 0 0 0 1>
     }
    }
   }
   slot {class=ButtonSlot}
  }
 ]
}
