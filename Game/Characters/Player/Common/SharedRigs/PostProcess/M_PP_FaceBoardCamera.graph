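# Post-process material for a face board camera, which blacks out everything outside a depth band near the face board.
# It reads the face board distance from the MPC_FaceboardCamera parameter collection and compares it with scene depth.
# Two chained If nodes keep scene color between that distance minus a Thickness parameter and the distance itself, outputting zero elsewhere.
# The pass blends at the scene color before DOF stage.
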Material /Game/Characters/Player/Common/SharedRigs/PostProcess/M_PP_FaceBoardCamera
graph {
 properties {bAutomaticallySetUsageInEditor=false, blendableLocation=BL_SceneColorBeforeDOF, materialDomain=MD_PostProcess, usedShadingModels=Default Lit}
 nodes [
  n1 = CollectionParameter() {collection=&/MPC_FaceboardCamera, parameterName=Distance to faceboard}
  n2 = SceneTexture() {sceneTextureId=PPI_SceneDepth}
  n3 = SceneTexture() {sceneTextureId=PPI_PostProcessInput0}
  n4 = If(A=@n2.Color, B=@n1, "A > B"=Constant(), "A == B"=@n3.Color, "A < B"=@n3.Color) {constB=0.95}
  If(A=@n2.Color, B=Subtract(A=@n1, B=ScalarParameter() {defaultValue=1, parameterName=Thickness}), "A > B"=@n4, "A == B"=@n4, "A < B"=Constant()) {constB=0.95}
 ]
}
