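# Editor Utility Widget that serves as the animation picker for the main character's Control Rig in Sequencer.
# Init Sets Buttons groups control buttons by limb and wires multi-select buttons.
# FK/IK checkboxes write rig bool channels and recolor their button sets; a twist checkbox toggles control visibility.
# Other buttons load body or face pickers, open IK/FK matching and twist-key deletion tools, zero controls at the current frame, and spawn a faceboard camera.
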EditorUtilityWidgetBlueprint /Game/Characters/Player/Common/SharedRigs/EUW_MainCharacter_AnimPicker
graphs [
 EventGraph {nodes [
   n1 = "On Check State Changed (leg_l_fk_ik_switch)"!() -> (OutputDelegate, then, bIsChecked:bool) {is=title K2Node_ComponentBoundEvent}
   n2 = Get leg_l_fk_ik_switch() -> (leg_l_fk_ik_switch:Editor Utility Check Box Object Reference)
   > Set Buttons Color From Checkbox State!(
    execute
    self:EUW Asset Picker Base Object Reference
    CheckBox:Editor Utility Check Box Object Reference=@n2
    IK Buttons:Array of EUW Picker Button Object References=Get Left Leg IK() -> (Left Leg IK:Array of EUW Picker Button Object References)
    FK Buttons:Array of EUW Picker Button Object References=Get Left Leg FK() -> (Left Leg FK:Array of EUW Picker Button Object References)
    Color Mult:double=0.1
    Opacity:double=0.5
   ) -> (then) {is=title K2Node_CallFunction}
   > Set Animation Channel Bool!(
    execute
    self:EUW Asset Picker Base Object Reference
    ControlRig:Control Rig Object Reference=Get!() -> (
     Control Rig Proxy Binding_Proxy:Movie Scene Binding Proxy Structure
     Control Rig Proxy Binding_ControlRig:Control Rig Object Reference
     Control Rig Proxy Binding_Track:Movie Scene Control Rig Parameter Track Object Reference
    ).Control Rig Proxy Binding_ControlRig
    Channel Name:Name=String To Name(InString=GetDisplayName(Object=@n2))
    Value:bool=@n1.bIsChecked
   ) -> (then) {is=title K2Node_CallFunction}
   n3 = "On Check State Changed (leg_r_fk_ik_switch)"!() -> (OutputDelegate, then, bIsChecked:bool) {is=title K2Node_ComponentBoundEvent}
   n4 = Get leg_r_fk_ik_switch() -> (leg_r_fk_ik_switch:Editor Utility Check Box Object Reference)
   > Set Buttons Color From Checkbox State(
    CheckBox=@n4
    IK Buttons=Get Right Leg IK() -> (Right Leg IK:Array of EUW Picker Button Object References)
    FK Buttons=Get Right Leg FK() -> (Right Leg FK:Array of EUW Picker Button Object References)
    Color Mult=0.1
    Opacity=0.5
   )
   > Set Animation Channel Bool(
    ControlRig=Get().Control Rig Proxy Binding_ControlRig
    Channel Name=String To Name(InString=GetDisplayName(Object=@n4))
    Value=@n3.bIsChecked
   )
   n5 = "On Check State Changed (TwistVisibility)"!() -> (OutputDelegate, then, bIsChecked:bool) {is=title K2Node_ComponentBoundEvent}
   n6 = Get Twists() -> (Twists:Array of EUW Picker Button Object References)
   > Set Visibility for Array!(
    execute
    self:EUW Asset Picker Base Object Reference
    " Array":Array of EUW Picker Button Object References=@n6
    bVisibility:bool=@n5.bIsChecked
   ) -> (then) {is=title K2Node_CallFunction}
   > n7 = For Each Loop(Array:Array of EUW Picker Button Object References=@n6) -> (Array Element:EUW Picker Button Object Reference)
   > SetControlVisibility!(
    execute
    self:Rig Hierarchy Object Reference=GetHierarchy!(self:Control Rig Object Reference=Get().Control Rig Proxy Binding_ControlRig) -> (ReturnValue:Rig Hierarchy Object Reference)
    InKey_Type:ERigElementType Enum=Control
    InKey_Name:Name=String To Name(InString=GetDisplayName(Object=@n7))
    bVisibility:bool=@n5.bIsChecked
   ) -> (then)
   n8 = "On Check State Changed (neck_fk_ik_switch)"!() -> (OutputDelegate, then, bIsChecked:bool) {is=title K2Node_ComponentBoundEvent}
   n9 = Get neck_fk_ik_switch() -> (neck_fk_ik_switch:Editor Utility Check Box Object Reference)
   > Set Buttons Color From Checkbox State(
    CheckBox=@n9
    IK Buttons=Get Neck IK() -> (Neck IK:Array of EUW Picker Button Object References)
    FK Buttons=Get Neck FK() -> (Neck FK:Array of EUW Picker Button Object References)
    Color Mult=0.1
    Opacity=0.5
   )
   > Set Animation Channel Bool(
    ControlRig=Get().Control Rig Proxy Binding_ControlRig
    Channel Name=String To Name(InString=GetDisplayName(Object=@n9))
    Value=@n8.bIsChecked
   )
   n10 = "On Check State Changed (arm_l_fk_ik_switch)"!() -> (OutputDelegate, then, bIsChecked:bool) {is=title K2Node_ComponentBoundEvent}
   n11 = Get arm_l_fk_ik_switch() -> (arm_l_fk_ik_switch:Editor Utility Check Box Object Reference)
   > Set Buttons Color From Checkbox State(
    CheckBox=@n11
    IK Buttons=Get Left Arm IK() -> (Left Arm IK:Array of EUW Picker Button Object References)
    FK Buttons=Get Left Arm FK() -> (Left Arm FK:Array of EUW Picker Button Object References)
    Color Mult=0.1
    Opacity=0.5
   )
   > Set Animation Channel Bool(
    ControlRig=Get().Control Rig Proxy Binding_ControlRig
    Channel Name=String To Name(InString=GetDisplayName(Object=@n11))
    Value=@n10.bIsChecked
   )
   "Custom on Clicked Event (BodyPickerButton)"() {is=title K2Node_ComponentBoundEvent}
   > n12 = Fill Slot!(
    execute
    self
    Slot:Vertical Box Object Reference=Get NewSlot() -> (NewSlot:Vertical Box Object Reference)
    EUW Class:User Widget Class Reference
   ) -> (then, ReturnValue:User Widget Object Reference)
   > Widget has loaded!(
    execute:Exec
    self:Object Reference=@n12
    Control Rig Proxy Binding:Control Rig Sequencer Binding Proxy Structure=Get Control Rig Proxy Binding() -> (Control Rig Proxy Binding:Control Rig Sequencer Binding Proxy Structure)
    Animatable Creature:Animatable Creature Object Reference=Get Animatable Creature() -> (Animatable Creature:Animatable Creature Object Reference)
    Anim Picker EUW:Animation Picker Object Reference=Get Anim Picker EUW() -> (Anim Picker EUW:Animation Picker Object Reference)
   ) -> (then:Exec) {is=title K2Node_Message}
   "Custom on Clicked Event (3LFacePickerButton)"() {is=title K2Node_ComponentBoundEvent}
   > n13 = Fill Slot(Slot=Get NewSlot(), EUW Class=&/EUW_MainCharacte_Face_AnimPicker.EUW_MainCharacte_Face_AnimPicker_C)
   > n14 = Set Face Picker(Face Picker:User Widget Object Reference=@n13) -> (Output_Get:User Widget Object Reference)
   > Widget has loaded(
    self=@n14
    Control Rig Proxy Binding=Get Control Rig Proxy Binding()
    Animatable Creature=Get Animatable Creature()
    Anim Picker EUW=Get Anim Picker EUW()
   )
   "Custom on Clicked Event (3LOnFacePickerButton)"() {is=title K2Node_ComponentBoundEvent}
   > n15 = Fill Slot(
    Slot=Get NewSlot()
    EUW Class=&/EUW_MainCharacte_OnFace_AnimPicker.EUW_MainCharacte_OnFace_AnimPicker_C
   )
   > n16 = Set Face Picker(Face Picker=@n15)
   > Widget has loaded(
    self=@n16
    Control Rig Proxy Binding=Get Control Rig Proxy Binding()
    Animatable Creature=Get Animatable Creature()
    Anim Picker EUW=Get Anim Picker EUW()
   )
   "Custom on Clicked Event (LegacyFacePickerButton)"() {is=title K2Node_ComponentBoundEvent}
   > n17 = Fill Slot(Slot=Get NewSlot(), EUW Class=&/EUW_Legacy_Face_AnimPicker.EUW_Legacy_Face_AnimPicker_C)
   > n18 = Set Face Picker(Face Picker=@n17)
   > Widget has loaded(
    self=@n18
    Control Rig Proxy Binding=Get Control Rig Proxy Binding()
    Animatable Creature=Get Animatable Creature()
    Anim Picker EUW=Get Anim Picker EUW()
   )
   "Custom on Clicked Event (BakeIKFKButton)"() {is=title K2Node_ComponentBoundEvent}
   n19 = Get ExtraOptionSlot() -> (ExtraOptionSlot:Vertical Box Object Reference)
   > n20 = Branch(Condition="integer > integer"(A=GetChildrenCount!(self:Panel Widget Object Reference=@n19) -> (ReturnValue:int=0)))
   > n21 = Cast To EUW_IkFkMatching!(execute, Object="Get (a copy)"(Array:Array of Widget Object References=GetAllChildren(self=@n19)) -> (Output:Widget Object Reference)) -> (then, CastFailed, AsEUW Ik Fk Matching:EUW Ik Fk Matching Object Reference) {is=title K2Node_DynamicCast}
   > ClearChildren(self=Get ExtraOptionSlot())
   ClearChildren(execute=[@n21.CastFailed, @n20.else], self=Get ExtraOptionSlot())
   > n22 = Create Widget!(
    execute:Exec
    WorldContextObject:Object Reference
    Class:User Widget Class Reference=/AnimInEdTools/IkFkMatchingTool/EUW_IkFkMatching.EUW_IkFkMatching_C
    OwningPlayer:Player Controller Object Reference
    Primary Control Rig:Control Rig Object Reference=Break Control Rig Sequencer Binding Proxy!(ControlRigSequencerBindingProxy:"Control Rig Sequencer Binding Proxy Structure (by ref)"=Get Control Rig Proxy Binding()) -> (
     Proxy:Movie Scene Binding Proxy Structure
     ControlRig:Control Rig Object Reference
     Track:Movie Scene Control Rig Parameter Track Object Reference
    ) {is=title K2Node_BreakStruct}.ControlRig
    Animatable Creature:Animatable Creature Object Reference=Get Animatable Creature()
   ) -> (then:Exec, ReturnValue:EUW Ik Fk Matching Object Reference)
   > AddChild(self=Get ExtraOptionSlot(), Content=@n22)
   "Custom on Clicked Event (DeleteTwistKeysButton)"() {is=title K2Node_ComponentBoundEvent}
   n23 = Get ExtraOptionSlot()
   > n24 = Branch(Condition="integer > integer"(A=GetChildrenCount(self=@n23) -> (ReturnValue=0)))
   > n25 = Cast To EUW_DeleteTwistKeys!(execute, Object="Get (a copy)"(Array=GetAllChildren(self=@n23))) -> (then, CastFailed, AsEUW Delete Twist Keys:EUW Delete Twist Keys Object Reference) {is=title K2Node_DynamicCast}
   > ClearChildren(self=Get ExtraOptionSlot())
   ClearChildren(execute=[@n25.CastFailed, @n24.else], self=Get ExtraOptionSlot())
   > n26 = Create Widget!(
    execute
    WorldContextObject
    Class=&/EUW_DeleteTwistKeys.EUW_DeleteTwistKeys_C
    OwningPlayer
    Primary Control Rig=Break Control Rig Sequencer Binding Proxy(ControlRigSequencerBindingProxy=Get Control Rig Proxy Binding()).ControlRig
   ) -> (then, ReturnValue:EUW Delete Twist Keys Object Reference)
   > AddChild(self=Get ExtraOptionSlot(), Content=@n26)
   n27 = "On Check State Changed (arm_r_fk_ik_switch)"!() -> (OutputDelegate, then, bIsChecked:bool) {is=title K2Node_ComponentBoundEvent}
   n28 = Get arm_r_fk_ik_switch() -> (arm_r_fk_ik_switch:Editor Utility Check Box Object Reference)
   > Set Buttons Color From Checkbox State(
    CheckBox=@n28
    IK Buttons=Get Right Arm IK() -> (Right Arm IK:Array of EUW Picker Button Object References)
    FK Buttons=Get Right Arm FK() -> (Right Arm FK:Array of EUW Picker Button Object References)
    Color Mult=0.1
    Opacity=0.5
   )
   > Set Animation Channel Bool(
    ControlRig=Get().Control Rig Proxy Binding_ControlRig
    Channel Name=String To Name(InString=GetDisplayName(Object=@n28))
    Value=@n27.bIsChecked
   )
   "Custom on Clicked Event (ChangeBGbutton)"() {is=title K2Node_ComponentBoundEvent}
   > Switch BG()
   "Custom on Clicked Event (Zero)"() {is=title K2Node_ComponentBoundEvent}
   > n29 = Get Current Local Time!(execute, TimeUnit:EMovieSceneTimeUnit Enum=DisplayRate) -> (then, ReturnValue:Movie Scene Sequence Playback Params Structure) {is=title K2Node_CallFunction}
   > n30 = GetFocusedLevelSequence!(execute) -> (then, ReturnValue:Level Sequence Object Reference)
   > n31 = For Each Loop(Array:Array of Widget Object References=GetAllChildren(self=Get CanvasPanel_0() -> (CanvasPanel_0:Canvas Panel Object Reference))) -> (Array Element:Widget Object Reference)
   > n32 = Cast To EUW_Picker_Button!(execute, Object=@n31) -> (then, CastFailed, AsEUW Picker Button:EUW Picker Button Object Reference) {is=title K2Node_DynamicCast}
   > SetLocalControlRigEulerTransform!(
    execute
    LevelSequence:Level Sequence Object Reference=@n30
    ControlRig:Control Rig Object Reference=Break Control Rig Sequencer Binding Proxy(ControlRigSequencerBindingProxy=Get Control Rig Proxy Binding()).ControlRig
    ControlName:Name=String To Name(InString=GetObjectName(Object=@n32))
    Frame:Frame Number Structure=Make Frame Number!(Value:int=Break Movie Scene Sequence Playback Params!(MovieSceneSequencePlaybackParams:"Movie Scene Sequence Playback Params Structure (by ref)"=@n29) -> (
      Frame_FrameNumber_Value:int
      Frame_SubFrame:float=~
      Time:float=~
      MarkedFrame:String
      Timecode:Timecode Structure
      PositionType:EMovieScenePositionType Enum
      UpdateMethod:EUpdatePositionMethod Enum
      bHasJumped:bool=~
     ) {is=title K2Node_BreakStruct}) -> (FrameNumber:Frame Number Structure) {is=title K2Node_MakeStruct}
    Value_Location:Vector="0,0,0"
    Value_Rotation:Rotator="0,0,0"
    Value_Scale:Vector="1,1,1"
    TimeUnit:EMovieSceneTimeUnit Enum=DisplayRate
    bSetKey:bool=false
   ) -> (then)
   n33 = "On Check State Changed (spine_fk_ik_switch)"!() -> (OutputDelegate, then, bIsChecked:bool) {is=title K2Node_ComponentBoundEvent}
   n34 = Get spine_fk_ik_switch() -> (spine_fk_ik_switch:Editor Utility Check Box Object Reference)
   > Set Buttons Color From Checkbox State(
    CheckBox=@n34
    IK Buttons=Get Spine IK() -> (Spine IK:Array of EUW Picker Button Object References)
    FK Buttons=Get Spine FK() -> (Spine FK:Array of EUW Picker Button Object References)
    Color Mult=0.1
    Opacity=0.5
   )
   > Set Animation Channel Bool(
    ControlRig=Get().Control Rig Proxy Binding_ControlRig
    Channel Name=String To Name(InString=GetDisplayName(Object=@n34))
    Value=@n33.bIsChecked
   )
   "Custom on Clicked Event (FaceboardcameraButton)"() {is=title K2Node_ComponentBoundEvent}
   > Spawn Faceboard camera()
   n35 = Event Construct()
   Update All IK FK Checkboxes(execute=[@n40.then, @n35.then])
   > Init Sets Buttons()
   > Set Visibility for Array(" Array"=Get Twists(), bVisibility=IsChecked(self=Get TwistVisibility() -> (TwistVisibility:Editor Utility Check Box Object Reference)))
   > Update All IK FK Checkboxes()
   > Update All IK FK Control Colors()
   > n36 = For Each Loop(Array:Array of EUW Picker Button Object References=Get Twists()) -> (Array Element:EUW Picker Button Object Reference)
   > SetControlVisibility(
    self=GetHierarchy(self=Get().Control Rig Proxy Binding_ControlRig)
    InKey_Type=Control
    InKey_Name=String To Name(InString=GetDisplayName(Object=@n36))
    bVisibility=IsChecked(self=Get TwistVisibility())
   )
   n37 = Event OnMouseEnter()
   > Parent: OnMouseEnter!(execute, MyGeometry:Geometry Structure=@n37.MyGeometry, MouseEvent:"Pointer Event Structure (by ref)"=@n37.MouseEvent) -> (then) {is=title K2Node_CallParentFunction}
   > Update All IK FK Checkboxes()
   Event Update From Scene() {is=title K2Node_Event}
   > Parent: Update From Scene() {is=title K2Node_CallParentFunction}
   > n38 = Get!(execute:Exec) -> (then:Exec, else:Exec, Face Picker:User Widget Object Reference)
   > Update From Scene!(execute:Exec, self:Object Reference=@n38) -> (then:Exec) {is=title K2Node_Message}
   n39 = Event Widget has loaded!() -> (
    OutputDelegate
    then
    Control Rig Proxy Binding:Control Rig Sequencer Binding Proxy Structure
    Animatable Creature:Animatable Creature Object Reference
    Anim Picker EUW:Animation Picker Object Reference
   ) {is=title K2Node_Event}
   > n40 = Parent: Widget has loaded!(
    execute
    Control Rig Proxy Binding:Control Rig Sequencer Binding Proxy Structure=@n39.Control Rig Proxy Binding
    Animatable Creature:Animatable Creature Object Reference=@n39.Animatable Creature
    Anim Picker EUW:Animation Picker Object Reference=@n39.Anim Picker EUW
   ) -> (then) {is=title K2Node_CallParentFunction}
  ]}
 Fill Slot {nodes [
   n1 = Fill Slot!() -> (then, Slot:Vertical Box Object Reference, EUW Class:User Widget Class Reference) {is=title K2Node_FunctionEntry}
   > ClearChildren(self=@n1)
   > n2 = Create Widget!(execute, WorldContextObject, Class=@n1.EUW Class, OwningPlayer) -> (then, ReturnValue:User Widget Object Reference)
   > AddChild(self=@n1, Content=@n2)
   > Return Node!(execute, ReturnValue:User Widget Object Reference=@n2)
  ]}
 Get Primary Panel {nodes [
   Get Primary Panel() {is=title K2Node_FunctionEntry}
   > Return Node!(execute, Panel:Panel Widget Object Reference=Get CanvasPanel_0())
  ]}
 Init Sets Buttons {nodes [
   Init Sets Buttons() {is=title K2Node_FunctionEntry}
   > n1 = Sequence!(execute) -> (then_0, then_1, then_2:Exec, then_3:Exec, then_4:Exec)
   > Set Left Arm FK(Left Arm FK:Array of EUW Picker Button Object References=Make Array!(
     "[0]":EUW Picker Button Object Reference=Get lowerarm_l_fk_ctrl() -> (lowerarm_l_fk_ctrl:EUW Picker Button Object Reference)
     "[1]":EUW Picker Button Object Reference=Get upperarm_l_fk_ctrl() -> (upperarm_l_fk_ctrl:EUW Picker Button Object Reference)
     "[2]":EUW Picker Button Object Reference=Get hand_l_fk_ctrl() -> (hand_l_fk_ctrl:EUW Picker Button Object Reference)
    ) -> (Array:Array of EUW Picker Button Object References)) -> (Output_Get:Array of EUW Picker Button Object References)
   > Set Right Arm FK(Right Arm FK:Array of EUW Picker Button Object References=Make Array(
     "[0]"=Get lowerarm_r_fk_ctrl() -> (lowerarm_r_fk_ctrl:EUW Picker Button Object Reference)
     "[1]"=Get upperarm_r_fk_ctrl() -> (upperarm_r_fk_ctrl:EUW Picker Button Object Reference)
     "[2]"=Get hand_r_fk_ctrl() -> (hand_r_fk_ctrl:EUW Picker Button Object Reference)
    )) -> (Output_Get:Array of EUW Picker Button Object References)
   > Set Left Leg IK(Left Leg IK:Array of EUW Picker Button Object References=Make Array!(
     "[0]"=Get leg_l_pv_ik_ctrl() -> (leg_l_pv_ik_ctrl:EUW Picker Button Object Reference)
     "[1]"=Get foot_roll_l_ctrl() -> (foot_roll_l_ctrl:EUW Picker Button Object Reference)
     "[2]"=Get foot_l_ik_ctrl() -> (foot_l_ik_ctrl:EUW Picker Button Object Reference)
     "[3]":EUW Picker Button Object Reference=Get ball_l_ik_ctrl() -> (ball_l_ik_ctrl:EUW Picker Button Object Reference)
     "[4]":EUW Picker Button Object Reference=Get tip_l_ctrl() -> (tip_l_ctrl:EUW Picker Button Object Reference)
     "[5]":EUW Picker Button Object Reference=Get foot_bk1_l_ctrl() -> (foot_bk1_l_ctrl:EUW Picker Button Object Reference)
    ) -> (Array)) -> (Output_Get:Array of EUW Picker Button Object References)
   > Set Right Leg IK(Right Leg IK:Array of EUW Picker Button Object References=Make Array(
     "[0]"=Get foot_roll_r_ctrl() -> (foot_roll_r_ctrl:EUW Picker Button Object Reference)
     "[1]"=Get foot_r_ik_ctrl() -> (foot_r_ik_ctrl:EUW Picker Button Object Reference)
     "[2]"=Get ball_r_ik_ctrl() -> (ball_r_ik_ctrl:EUW Picker Button Object Reference)
     "[3]"=Get tip_r_ctrl() -> (tip_r_ctrl:EUW Picker Button Object Reference)
     "[4]"=Get foot_bk1_r_ctrl() -> (foot_bk1_r_ctrl:EUW Picker Button Object Reference)
     "[5]"=Get leg_r_pv_ik_ctrl() -> (leg_r_pv_ik_ctrl:EUW Picker Button Object Reference)
    )) -> (Output_Get:Array of EUW Picker Button Object References)
   > Set Neck FK(Neck FK:Array of EUW Picker Button Object References=Make Array!(
     "[0]"=Get neck_01_ctrl() -> (neck_01_ctrl:EUW Picker Button Object Reference)
     "[1]"=Get neck_02_ctrl() -> (neck_02_ctrl:EUW Picker Button Object Reference)
     "[2]"=Get head_ctrl() -> (head_ctrl:EUW Picker Button Object Reference)
    ) -> (Array)) -> (Output_Get:Array of EUW Picker Button Object References)
   > Set Neck IK(Neck IK:Array of EUW Picker Button Object References=Make Array!("[0]"=Get head_ik_ctrl() -> (head_ik_ctrl:EUW Picker Button Object Reference)) -> (Array)) -> (Output_Get:Array of EUW Picker Button Object References)
   > Set Spine FK(Spine FK:Array of EUW Picker Button Object References=Make Array!(
     "[0]"=Get spine_01_ctrl() -> (spine_01_ctrl:EUW Picker Button Object Reference)
     "[1]":EUW Picker Button Object Reference=Get spine_02_ctrl() -> (spine_02_ctrl:EUW Picker Button Object Reference)
     "[2]":EUW Picker Button Object Reference=Get spine_03_ctrl() -> (spine_03_ctrl:EUW Picker Button Object Reference)
    ) -> (Array)) -> (Output_Get:Array of EUW Picker Button Object References)
   > Set Spine IK(Spine IK:Array of EUW Picker Button Object References=Make Array(
     "[0]"=Get chest_ctrl() -> (chest_ctrl:EUW Picker Button Object Reference)
     "[1]"=Get chest_tan_ctrl() -> (chest_tan_ctrl:EUW Picker Button Object Reference)
     "[2]"=Get hips_tan_ctrl() -> (hips_tan_ctrl:EUW Picker Button Object Reference)
    )) -> (Output_Get:Array of EUW Picker Button Object References)
   Set Left Arm IK(
    execute=@n1.then_1
    Left Arm IK:Array of EUW Picker Button Object References=Make Array!(
     "[0]"=Get hand_l_ik_ctrl() -> (hand_l_ik_ctrl:EUW Picker Button Object Reference)
     "[1]"=Get arm_l_pv_ik_ctrl() -> (arm_l_pv_ik_ctrl:EUW Picker Button Object Reference)
    ) -> (Array)
   ) -> (Output_Get:Array of EUW Picker Button Object References)
   > Set Right Arm IK(Right Arm IK:Array of EUW Picker Button Object References=Make Array(
     "[0]"=Get hand_r_ik_ctrl() -> (hand_r_ik_ctrl:EUW Picker Button Object Reference)
     "[1]"=Get arm_r_pv_ik_ctrl() -> (arm_r_pv_ik_ctrl:EUW Picker Button Object Reference)
    )) -> (Output_Get:Array of EUW Picker Button Object References)
   > Set Left Leg FK(Left Leg FK:Array of EUW Picker Button Object References=Make Array!(
     "[0]"=Get thigh_l_fk_ctrl() -> (thigh_l_fk_ctrl:EUW Picker Button Object Reference)
     "[1]"=Get calf_l_fk_ctrl() -> (calf_l_fk_ctrl:EUW Picker Button Object Reference)
     "[2]":EUW Picker Button Object Reference=Get foot_l_fk_ctrl() -> (foot_l_fk_ctrl:EUW Picker Button Object Reference)
     "[3]":EUW Picker Button Object Reference=Get ball_l_fk_ctrl() -> (ball_l_fk_ctrl:EUW Picker Button Object Reference)
    ) -> (Array)) -> (Output_Get:Array of EUW Picker Button Object References)
   > Set Right Leg FK(Right Leg FK:Array of EUW Picker Button Object References=Make Array(
     "[0]"=Get thigh_r_fk_ctrl() -> (thigh_r_fk_ctrl:EUW Picker Button Object Reference)
     "[1]"=Get calf_r_fk_ctrl() -> (calf_r_fk_ctrl:EUW Picker Button Object Reference)
     "[2]"=Get foot_r_fk_ctrl() -> (foot_r_fk_ctrl:EUW Picker Button Object Reference)
     "[3]"=Get ball_r_fk_ctrl() -> (ball_r_fk_ctrl:EUW Picker Button Object Reference)
    )) -> (Output_Get:Array of EUW Picker Button Object References)
   > Clear!(
    execute
    TargetMap:Map of Control Rig Object References to Animation Picker String Array Struct Structures=Get Control Bindings!(self:EUW Picker Element Object Reference=Get Zero() -> (Zero:EUW Picker Button Object Reference)) -> (Control Bindings:Map of Control Rig Object References to Animation Picker String Array Struct Structures)
   ) -> (then) {is=title K2Node_CallFunction}
   n2 = Sequence!(execute=@n1.then_3) -> (then_0, then_1, then_2)
   Select Multiple Buttons!(
    execute=@n2.then_1
    self:EUW Asset Picker Base Object Reference
    Button:EUW Picker Button Object Reference=Get fingers_joint_2_r_all() -> (fingers_joint_2_r_all:EUW Picker Button Object Reference)
    Buttons to Select:Array of EUW Picker Button Object References=Make Array(
     "[0]"=Get index_02_r_ctrl() -> (index_02_r_ctrl:EUW Picker Button Object Reference)
     "[1]"=Get ring_02_r_ctrl() -> (ring_02_r_ctrl:EUW Picker Button Object Reference)
     "[2]"=Get middle_02_r_ctrl() -> (middle_02_r_ctrl:EUW Picker Button Object Reference)
     "[3]"=Get pinky_02_r_ctrl() -> (pinky_02_r_ctrl:EUW Picker Button Object Reference)
    )
   ) -> (then)
   > Select Multiple Buttons(
    Button=Get fingers_joint_2_l_all() -> (fingers_joint_2_l_all:EUW Picker Button Object Reference)
    Buttons to Select=Make Array(
     "[0]"=Get middle_02_l_ctrl() -> (middle_02_l_ctrl:EUW Picker Button Object Reference)
     "[1]"=Get ring_02_l_ctrl() -> (ring_02_l_ctrl:EUW Picker Button Object Reference)
     "[2]"=Get index_02_l_ctrl() -> (index_02_l_ctrl:EUW Picker Button Object Reference)
     "[3]"=Get pinky_02_l_ctrl() -> (pinky_02_l_ctrl:EUW Picker Button Object Reference)
    )
   )
   > Select Multiple Buttons(
    Button=Get fingers_joint_3_r_all() -> (fingers_joint_3_r_all:EUW Picker Button Object Reference)
    Buttons to Select=Make Array(
     "[0]"=Get index_03_r_ctrl() -> (index_03_r_ctrl:EUW Picker Button Object Reference)
     "[1]"=Get ring_03_r_ctrl() -> (ring_03_r_ctrl:EUW Picker Button Object Reference)
     "[2]"=Get middle_03_r_ctrl() -> (middle_03_r_ctrl:EUW Picker Button Object Reference)
     "[3]"=Get pinky_03_r_ctrl() -> (pinky_03_r_ctrl:EUW Picker Button Object Reference)
    )
   )
   > Select Multiple Buttons(
    Button=Get fingers_joint_3_l_all() -> (fingers_joint_3_l_all:EUW Picker Button Object Reference)
    Buttons to Select=Make Array(
     "[0]"=Get middle_03_l_ctrl() -> (middle_03_l_ctrl:EUW Picker Button Object Reference)
     "[1]"=Get ring_03_l_ctrl() -> (ring_03_l_ctrl:EUW Picker Button Object Reference)
     "[2]"=Get index_03_l_ctrl() -> (index_03_l_ctrl:EUW Picker Button Object Reference)
     "[3]"=Get pinky_03_l_ctrl() -> (pinky_03_l_ctrl:EUW Picker Button Object Reference)
    )
   )
   Select Multiple Buttons(
    execute=@n2.then_2
    Button=Get WeapBones() -> (WeapBones:EUW Picker Button Object Reference)
    Buttons to Select=Make Array(
     "[0]"=Get weapon_l_ctrl() -> (weapon_l_ctrl:EUW Picker Button Object Reference)
     "[1]"=Get weapon_r_ctrl() -> (weapon_r_ctrl:EUW Picker Button Object Reference)
     "[2]"=Get hand_attach_l_ctrl() -> (hand_attach_l_ctrl:EUW Picker Button Object Reference)
     "[3]"=Get hand_attach_r_ctrl() -> (hand_attach_r_ctrl:EUW Picker Button Object Reference)
    )
   )
   Select Multiple Buttons(
    execute=@n2.then_0
    Button=Get fingers_metacarpal_r_all() -> (fingers_metacarpal_r_all:EUW Picker Button Object Reference)
    Buttons to Select=Make Array(
     "[0]"=Get index_metacarpal_r_ctrl() -> (index_metacarpal_r_ctrl:EUW Picker Button Object Reference)
     "[1]"=Get ring_metacarpal_r_ctrl() -> (ring_metacarpal_r_ctrl:EUW Picker Button Object Reference)
     "[2]"=Get middle_metacarpal_r_ctrl() -> (middle_metacarpal_r_ctrl:EUW Picker Button Object Reference)
     "[3]"=Get pinky_metacarpal_r_ctrl() -> (pinky_metacarpal_r_ctrl:EUW Picker Button Object Reference)
    )
   )
   > Select Multiple Buttons(
    Button=Get fingers_metacarpal_l_all() -> (fingers_metacarpal_l_all:EUW Picker Button Object Reference)
    Buttons to Select=Make Array(
     "[0]"=Get middle_metacarpal_l_ctrl() -> (middle_metacarpal_l_ctrl:EUW Picker Button Object Reference)
     "[1]"=Get ring_metacarpal_l_ctrl() -> (ring_metacarpal_l_ctrl:EUW Picker Button Object Reference)
     "[2]"=Get index_metacarpal_l_ctrl() -> (index_metacarpal_l_ctrl:EUW Picker Button Object Reference)
     "[3]"=Get pinky_metacarpal_l_ctrl() -> (pinky_metacarpal_l_ctrl:EUW Picker Button Object Reference)
    )
   )
   > Select Multiple Buttons(
    Button=Get fingers_joint_1_r_all() -> (fingers_joint_1_r_all:EUW Picker Button Object Reference)
    Buttons to Select=Make Array(
     "[0]"=Get index_01_r_ctrl() -> (index_01_r_ctrl:EUW Picker Button Object Reference)
     "[1]"=Get ring_01_r_ctrl() -> (ring_01_r_ctrl:EUW Picker Button Object Reference)
     "[2]"=Get middle_01_r_ctrl() -> (middle_01_r_ctrl:EUW Picker Button Object Reference)
     "[3]"=Get pinky_01_r_ctrl() -> (pinky_01_r_ctrl:EUW Picker Button Object Reference)
    )
   )
   > Select Multiple Buttons(
    Button=Get fingers_joint_1_l_all() -> (fingers_joint_1_l_all:EUW Picker Button Object Reference)
    Buttons to Select=Make Array(
     "[0]"=Get middle_01_l_ctrl() -> (middle_01_l_ctrl:EUW Picker Button Object Reference)
     "[1]"=Get ring_01_l_ctrl() -> (ring_01_l_ctrl:EUW Picker Button Object Reference)
     "[2]"=Get index_01_l_ctrl() -> (index_01_l_ctrl:EUW Picker Button Object Reference)
     "[3]"=Get pinky_01_l_ctrl() -> (pinky_01_l_ctrl:EUW Picker Button Object Reference)
    )
   )
   Return Node!(execute=@n1.then_4)
   n3 = Set IK Move Buttons(
    execute=@n1.then_2
    IK Move Buttons:Array of EUW Picker Button Object References=Make Array!(
     "[0]"=Get body_ctrl() -> (body_ctrl:EUW Picker Button Object Reference)
     "[1]"=Get hand_r_ik_ctrl()
     "[2]"=Get hand_l_ik_ctrl()
     "[3]"=Get arm_r_pv_ik_ctrl()
     "[4]":EUW Picker Button Object Reference=Get arm_l_pv_ik_ctrl()
     "[5]":EUW Picker Button Object Reference=Get leg_r_pv_ik_ctrl()
     "[6]":EUW Picker Button Object Reference=Get leg_l_pv_ik_ctrl()
     "[7]":EUW Picker Button Object Reference=Get foot_l_ik_ctrl()
     "[8]":EUW Picker Button Object Reference=Get foot_r_ik_ctrl()
    ) -> (Array)
   ) -> (Output_Get:Array of EUW Picker Button Object References)
   > Multi Select Button!(
    execute
    self
    Button:EUW Picker Button Object Reference=Get IKMove() -> (IKMove:EUW Picker Button Object Reference)
    Buttons to Select:Array of EUW Picker Button Object References=@n3
   ) -> (then)
   > n4 = Set FK Hands(FK Hands:Array of EUW Picker Button Object References=Make Array!("[0]"=Get hand_r_fk_ctrl(), "[1]"=Get hand_l_fk_ctrl()) -> (Array)) -> (Output_Get:Array of EUW Picker Button Object References)
   > Multi Select Button(Button=Get FKHands() -> (FKHands:EUW Picker Button Object Reference), Buttons to Select=@n4)
   > n5 = "Set 4 Fingers Right "("4 Fingers Right ":Array of EUW Picker Button Object References=Make Array!(
     "[0]"=Get index_01_r_ctrl()
     "[1]"=Get index_02_r_ctrl()
     "[2]":EUW Picker Button Object Reference=Get index_03_r_ctrl()
     "[3]":EUW Picker Button Object Reference=Get ring_01_r_ctrl()
     "[4]":EUW Picker Button Object Reference=Get ring_02_r_ctrl()
     "[5]":EUW Picker Button Object Reference=Get ring_03_r_ctrl()
     "[6]":EUW Picker Button Object Reference=Get middle_01_r_ctrl()
     "[7]":EUW Picker Button Object Reference=Get middle_02_r_ctrl()
     "[8]":EUW Picker Button Object Reference=Get middle_03_r_ctrl()
     "[9]":EUW Picker Button Object Reference=Get pinky_01_r_ctrl()
     "[10]":EUW Picker Button Object Reference=Get pinky_02_r_ctrl()
     "[11]":EUW Picker Button Object Reference=Get pinky_03_r_ctrl()
    ) -> (Array)) -> (Output_Get:Array of EUW Picker Button Object References)
   > Multi Select Button(Button=Get 4Fingers_r() -> (4Fingers_r:EUW Picker Button Object Reference), Buttons to Select=@n5)
   > n6 = Set 4 Fingers Left(4 Fingers Left:Array of EUW Picker Button Object References=Make Array(
     "[0]"=Get middle_01_l_ctrl()
     "[1]"=Get middle_02_l_ctrl()
     "[2]"=Get middle_03_l_ctrl()
     "[3]"=Get ring_01_l_ctrl()
     "[4]"=Get ring_02_l_ctrl()
     "[5]"=Get ring_03_l_ctrl()
     "[6]"=Get index_01_l_ctrl()
     "[7]"=Get index_02_l_ctrl()
     "[8]"=Get index_03_l_ctrl()
     "[9]"=Get pinky_01_l_ctrl()
     "[10]"=Get pinky_02_l_ctrl()
     "[11]"=Get pinky_03_l_ctrl()
    )) -> (Output_Get:Array of EUW Picker Button Object References)
   > Multi Select Button(Button=Get 4Fingers_l() -> (4Fingers_l:EUW Picker Button Object Reference), Buttons to Select=@n6)
   > n7 = Set Twists(Twists:Array of EUW Picker Button Object References=Make Array!(
     "[0]"=Get upperarm_twist_01_l_ctrl() -> (upperarm_twist_01_l_ctrl:EUW Picker Button Object Reference)
     "[1]"=Get upperarm_twist_02_l_ctrl() -> (upperarm_twist_02_l_ctrl:EUW Picker Button Object Reference)
     "[2]"=Get upperarm_twist_01_r_ctrl() -> (upperarm_twist_01_r_ctrl:EUW Picker Button Object Reference)
     "[3]"=Get upperarm_twist_02_r_ctrl() -> (upperarm_twist_02_r_ctrl:EUW Picker Button Object Reference)
     "[4]"=Get lowerarm_twist_01_r_ctrl() -> (lowerarm_twist_01_r_ctrl:EUW Picker Button Object Reference)
     "[5]"=Get lowerarm_twist_02_r_ctrl() -> (lowerarm_twist_02_r_ctrl:EUW Picker Button Object Reference)
     "[6]"=Get lowerarm_twist_01_l_ctrl() -> (lowerarm_twist_01_l_ctrl:EUW Picker Button Object Reference)
     "[7]"=Get lowerarm_twist_02_l_ctrl() -> (lowerarm_twist_02_l_ctrl:EUW Picker Button Object Reference)
     "[8]"=Get thigh_twist_01_r_ctrl() -> (thigh_twist_01_r_ctrl:EUW Picker Button Object Reference)
     "[9]"=Get thigh_twist_01_l_ctrl() -> (thigh_twist_01_l_ctrl:EUW Picker Button Object Reference)
     "[10]"=Get calf_twist_01_r_ctrl() -> (calf_twist_01_r_ctrl:EUW Picker Button Object Reference)
     "[11]"=Get calf_twist_02_r_ctrl() -> (calf_twist_02_r_ctrl:EUW Picker Button Object Reference)
     "[12]":EUW Picker Button Object Reference=Get calf_twist_01_l_ctrl() -> (calf_twist_01_l_ctrl:EUW Picker Button Object Reference)
     "[13]":EUW Picker Button Object Reference=Get calf_twist_02_l_ctrl() -> (calf_twist_02_l_ctrl:EUW Picker Button Object Reference)
    ) -> (Array)) -> (Output_Get:Array of EUW Picker Button Object References)
   > Multi Select Button(Button=Get AllTwist() -> (AllTwist:EUW Picker Button Object Reference), Buttons to Select=@n7)
   > n8 = Sequence()
   > n9 = Set All IK Leg Extras Right(All IK Leg Extras Right:Array of EUW Picker Button Object References=Make Array!("[0]"=Get tip_r_ctrl(), "[1]"=Get foot_bk1_r_ctrl(), "[2]"=Get foot_roll_r_ctrl()) -> (Array)) -> (Output_Get:Array of EUW Picker Button Object References)
   > Multi Select Button(Button=Get AllIKLegExtras_r() -> (AllIKLegExtras_r:EUW Picker Button Object Reference), Buttons to Select=@n9)
   > n10 = Set All IK Leg Right(All IK Leg Right:Array of EUW Picker Button Object References=Make Array("[0]"=Get foot_r_ik_ctrl(), "[1]"=Get leg_r_pv_ik_ctrl(), "[2]"=Get ball_r_ik_ctrl())) -> (Output_Get:Array of EUW Picker Button Object References)
   > Multi Select Button(Button=Get AllIKLeg_r() -> (AllIKLeg_r:EUW Picker Button Object Reference), Buttons to Select=@n10)
   > n11 = Set All FK Leg Right(All FK Leg Right:Array of EUW Picker Button Object References=Make Array!(
     "[0]"=Get thigh_r_fk_ctrl()
     "[1]"=Get calf_r_fk_ctrl()
     "[2]"=Get foot_r_fk_ctrl()
     "[3]":EUW Picker Button Object Reference=Get ball_r_fk_ctrl()
    ) -> (Array)) -> (Output_Get:Array of EUW Picker Button Object References)
   > Multi Select Button(Button=Get AllFKLeg_r() -> (AllFKLeg_r:EUW Picker Button Object Reference), Buttons to Select=@n11)
   > Set Thumb Right(Thumb Right:Array of EUW Picker Button Object References=Make Array!(
     "[0]"=Get thumb_02_r_ctrl() -> (thumb_02_r_ctrl:EUW Picker Button Object Reference)
     "[1]"=Get thumb_01_r_ctrl() -> (thumb_01_r_ctrl:EUW Picker Button Object Reference)
    ) -> (Array)) -> (Output_Get:Array of EUW Picker Button Object References)
   > n12 = Set All fingers Right(All fingers Right:Array of EUW Picker Button Object References="Get 4 Fingers Right "() -> ("4 Fingers Right ":Array of EUW Picker Button Object References)) -> (Output_Get:Array of EUW Picker Button Object References)
   > Append Array!(
    execute
    TargetArray:Array of EUW Picker Button Object References=@n12
    SourceArray:Array of EUW Picker Button Object References=Get Thumb Right() -> (Thumb Right:Array of EUW Picker Button Object References)
   ) -> (then)
   > Append Array(
    TargetArray=@n12
    SourceArray=Make Array!(
     "[0]"=Get pinky_metacarpal_r_ctrl()
     "[1]"=Get ring_metacarpal_r_ctrl()
     "[2]":EUW Picker Button Object Reference=Get middle_metacarpal_r_ctrl()
     "[3]":EUW Picker Button Object Reference=Get index_metacarpal_r_ctrl()
     "[4]":EUW Picker Button Object Reference=Get thumb_03_r_ctrl() -> (thumb_03_r_ctrl:EUW Picker Button Object Reference)
    ) -> (Array)
   )
   > Multi Select Button(
    Button=Get allFingers_r() -> (allFingers_r:EUW Picker Button Object Reference)
    Buttons to Select=Get All fingers Right() -> (All fingers Right:Array of EUW Picker Button Object References)
   )
   > n13 = Set All Arm Right(All Arm Right:Array of EUW Picker Button Object References=Make Array!(
     "[0]"=Get upperarm_r_fk_ctrl()
     "[1]"=Get lowerarm_r_fk_ctrl()
     "[2]"=Get hand_r_fk_ctrl()
     "[3]"=Get hand_r_ik_ctrl()
     "[4]"=Get arm_r_pv_ik_ctrl()
     "[5]":EUW Picker Button Object Reference=Get clavicle_r_ctrl() -> (clavicle_r_ctrl:EUW Picker Button Object Reference)
    ) -> (Array)) -> (Output_Get:Array of EUW Picker Button Object References)
   > Multi Select Button(Button=Get AllArm_r() -> (AllArm_r:EUW Picker Button Object Reference), Buttons to Select=@n13)
   > n14 = "Set All Arm Fingers Right "("All Arm Fingers Right ":Array of EUW Picker Button Object References=Get All Arm Right() -> (All Arm Right:Array of EUW Picker Button Object References)) -> (Output_Get:Array of EUW Picker Button Object References)
   > Append Array(TargetArray=@n14, SourceArray=Get All fingers Right())
   > Multi Select Button(
    Button=Get ArmsFingers_r() -> (ArmsFingers_r:EUW Picker Button Object Reference)
    Buttons to Select="Get All Arm Fingers Right "() -> ("All Arm Fingers Right ":Array of EUW Picker Button Object References)
   )
   n15 = Set All IK Leg Extras Left(
    execute=@n8.then_1
    All IK Leg Extras Left:Array of EUW Picker Button Object References=Make Array!("[0]"=Get tip_l_ctrl(), "[1]"=Get foot_bk1_l_ctrl(), "[2]"=Get foot_roll_l_ctrl()) -> (Array)
   ) -> (Output_Get:Array of EUW Picker Button Object References)
   > Multi Select Button(Button=Get AllIKLegExtra_l() -> (AllIKLegExtra_l:EUW Picker Button Object Reference), Buttons to Select=@n15)
   > n16 = Set All IK Leg Left(All IK Leg Left:Array of EUW Picker Button Object References=Make Array("[0]"=Get foot_l_ik_ctrl(), "[1]"=Get leg_l_pv_ik_ctrl(), "[2]"=Get ball_l_ik_ctrl())) -> (Output_Get:Array of EUW Picker Button Object References)
   > Multi Select Button(Button=Get AllIKLeg_l() -> (AllIKLeg_l:EUW Picker Button Object Reference), Buttons to Select=@n16)
   n17 = Make Array!(
    "[0]"=Get thigh_l_fk_ctrl()
    "[1]"=Get calf_l_fk_ctrl()
    "[2]"=Get foot_l_fk_ctrl()
    "[3]":EUW Picker Button Object Reference=Get ball_l_fk_ctrl()
   ) -> (Array)
   > Set All FK Leg Left(All FK Leg Left:Array of EUW Picker Button Object References=@n17) -> (Output_Get:Array of EUW Picker Button Object References)
   > Multi Select Button(Button=Get AllFKLeg_l() -> (AllFKLeg_l:EUW Picker Button Object Reference), Buttons to Select=@n17)
   > Set Thumb Left(Thumb Left:Array of EUW Picker Button Object References=Make Array!(
     "[0]"=Get thumb_02_l_ctrl() -> (thumb_02_l_ctrl:EUW Picker Button Object Reference)
     "[1]"=Get thumb_01_l_ctrl() -> (thumb_01_l_ctrl:EUW Picker Button Object Reference)
    ) -> (Array)) -> (Output_Get:Array of EUW Picker Button Object References)
   > n18 = Set All Fingers Left(All Fingers Left:Array of EUW Picker Button Object References=Get 4 Fingers Left() -> (4 Fingers Left:Array of EUW Picker Button Object References)) -> (Output_Get:Array of EUW Picker Button Object References)
   > Append Array(TargetArray=@n18, SourceArray=Get Thumb Left() -> (Thumb Left:Array of EUW Picker Button Object References))
   > Append Array(
    TargetArray=@n18
    SourceArray=Make Array!(
     "[0]"=Get pinky_metacarpal_l_ctrl()
     "[1]"=Get ring_metacarpal_l_ctrl()
     "[2]":EUW Picker Button Object Reference=Get middle_metacarpal_l_ctrl()
     "[3]":EUW Picker Button Object Reference=Get index_metacarpal_l_ctrl()
     "[4]":EUW Picker Button Object Reference=Get thumb_03_l_ctrl() -> (thumb_03_l_ctrl:EUW Picker Button Object Reference)
    ) -> (Array)
   )
   > Multi Select Button(
    Button=Get AllFingers_l() -> (AllFingers_l:EUW Picker Button Object Reference)
    Buttons to Select=Get All Fingers Left() -> (All Fingers Left:Array of EUW Picker Button Object References)
   )
   > n19 = Set All Arm Left(All Arm Left:Array of EUW Picker Button Object References=Make Array!(
     "[0]"=Get upperarm_l_fk_ctrl()
     "[1]"=Get lowerarm_l_fk_ctrl()
     "[2]"=Get hand_l_fk_ctrl()
     "[3]"=Get hand_l_ik_ctrl()
     "[4]"=Get arm_l_pv_ik_ctrl()
     "[5]":EUW Picker Button Object Reference=Get clavicle_l_ctrl() -> (clavicle_l_ctrl:EUW Picker Button Object Reference)
    ) -> (Array)) -> (Output_Get:Array of EUW Picker Button Object References)
   > Multi Select Button(Button=Get AllArm_l() -> (AllArm_l:EUW Picker Button Object Reference), Buttons to Select=@n19)
   > n20 = Set All Arm Fingers Left(All Arm Fingers Left:Array of EUW Picker Button Object References=Get All Arm Left() -> (All Arm Left:Array of EUW Picker Button Object References)) -> (Output_Get:Array of EUW Picker Button Object References)
   > Append Array(TargetArray=@n20, SourceArray=Get All Fingers Left())
   > Multi Select Button(
    Button=Get ArmsFingers_l() -> (ArmsFingers_l:EUW Picker Button Object Reference)
    Buttons to Select=Get All Arm Fingers Left() -> (All Arm Fingers Left:Array of EUW Picker Button Object References)
   )
   Multi Select Button(
    execute=@n8.then_2
    Button=Get AllClavicles() -> (AllClavicles:EUW Picker Button Object Reference)
    Buttons to Select=Make Array!("[0]"=Get clavicle_r_ctrl(), "[1]"=Get clavicle_l_ctrl()) -> (Array)
   )
   > n21 = Set All Arms(All Arms:Array of EUW Picker Button Object References=Get All Arm Left()) -> (Output_Get:Array of EUW Picker Button Object References)
   > Append Array(TargetArray=@n21, SourceArray=Get All Arm Right())
   > Multi Select Button(
    Button=Get AllArms() -> (AllArms:EUW Picker Button Object Reference)
    Buttons to Select=Get All Arms() -> (All Arms:Array of EUW Picker Button Object References)
   )
   > n22 = Set All Fingers(All Fingers:Array of EUW Picker Button Object References=Get All Fingers Left()) -> (Output_Get:Array of EUW Picker Button Object References)
   > Append Array(TargetArray=@n22, SourceArray=Get All fingers Right())
   > Multi Select Button(
    Button=Get AllFingers() -> (AllFingers:EUW Picker Button Object Reference)
    Buttons to Select=Get All Fingers() -> (All Fingers:Array of EUW Picker Button Object References)
   )
   > n23 = Set All FK Arms(All FK Arms:Array of EUW Picker Button Object References=Make Array!(
     "[0]"=Get upperarm_r_fk_ctrl()
     "[1]"=Get lowerarm_r_fk_ctrl()
     "[2]":EUW Picker Button Object Reference=Get hand_r_fk_ctrl()
     "[3]":EUW Picker Button Object Reference=Get clavicle_r_ctrl()
     "[4]":EUW Picker Button Object Reference=Get upperarm_l_fk_ctrl()
     "[5]":EUW Picker Button Object Reference=Get lowerarm_l_fk_ctrl()
     "[6]":EUW Picker Button Object Reference=Get hand_l_fk_ctrl()
     "[7]":EUW Picker Button Object Reference=Get clavicle_l_ctrl()
    ) -> (Array)) -> (Output_Get:Array of EUW Picker Button Object References)
   > Multi Select Button(Button=Get FKArms() -> (FKArms:EUW Picker Button Object Reference), Buttons to Select=@n23)
   > Multi Select Button(
    Button=Get FKElbows() -> (FKElbows:EUW Picker Button Object Reference)
    Buttons to Select=Make Array!("[0]"=Get lowerarm_r_fk_ctrl(), "[1]"=Get lowerarm_l_fk_ctrl()) -> (Array)
   )
   > Multi Select Button(
    Button=Get FKBiceps() -> (FKBiceps:EUW Picker Button Object Reference)
    Buttons to Select=Make Array("[0]"=Get upperarm_r_fk_ctrl(), "[1]"=Get upperarm_l_fk_ctrl())
   )
   > Multi Select Button(Button=Get FKHands(), Buttons to Select=Make Array("[0]"=Get hand_r_fk_ctrl(), "[1]"=Get hand_l_fk_ctrl()))
   > Multi Select Button(
    Button=Get IKArms() -> (IKArms:EUW Picker Button Object Reference)
    Buttons to Select=Make Array!(
     "[0]"=Get hand_r_ik_ctrl()
     "[1]"=Get arm_r_pv_ik_ctrl()
     "[2]":EUW Picker Button Object Reference=Get clavicle_r_ctrl()
     "[3]":EUW Picker Button Object Reference=Get hand_l_ik_ctrl()
     "[4]":EUW Picker Button Object Reference=Get arm_l_pv_ik_ctrl()
     "[5]":EUW Picker Button Object Reference=Get clavicle_l_ctrl()
    ) -> (Array)
   )
   > Multi Select Button(
    Button=Get IKElbows() -> (IKElbows:EUW Picker Button Object Reference)
    Buttons to Select=Make Array!("[0]"=Get arm_r_pv_ik_ctrl(), "[1]"=Get arm_l_pv_ik_ctrl()) -> (Array)
   )
   > Multi Select Button(
    Button=Get IKHands() -> (IKHands:EUW Picker Button Object Reference)
    Buttons to Select=Make Array("[0]"=Get hand_r_ik_ctrl(), "[1]"=Get hand_l_ik_ctrl())
   )
   > n24 = Set All Legs(All Legs:Array of EUW Picker Button Object References=Get All IK Leg Left() -> (All IK Leg Left:Array of EUW Picker Button Object References)) -> (Output_Get:Array of EUW Picker Button Object References)
   > Append Array(TargetArray=@n24, SourceArray=Get All IK Leg Right() -> (All IK Leg Right:Array of EUW Picker Button Object References))
   > Append Array(TargetArray=@n24, SourceArray=Get All IK Leg Extras Left() -> (All IK Leg Extras Left:Array of EUW Picker Button Object References))
   > Append Array(TargetArray=@n24, SourceArray=Get All IK Leg Extras Right() -> (All IK Leg Extras Right:Array of EUW Picker Button Object References))
   > Append Array(TargetArray=@n24, SourceArray=Get All FK Leg Left() -> (All FK Leg Left:Array of EUW Picker Button Object References))
   > Append Array(TargetArray=@n24, SourceArray=Get All FK Leg Right() -> (All FK Leg Right:Array of EUW Picker Button Object References))
   > Multi Select Button(
    Button=Get AllLegs() -> (AllLegs:EUW Picker Button Object Reference)
    Buttons to Select=Get All Legs() -> (All Legs:Array of EUW Picker Button Object References)
   )
   > n25 = Set All FK Legs(All FK Legs:Array of EUW Picker Button Object References=Get All FK Leg Left()) -> (Output_Get:Array of EUW Picker Button Object References)
   > Append Array(TargetArray=@n25, SourceArray=Get All FK Leg Right())
   > Multi Select Button(
    Button=Get AllFKLegs() -> (AllFKLegs:EUW Picker Button Object Reference)
    Buttons to Select=Get All FK Legs() -> (All FK Legs:Array of EUW Picker Button Object References)
   )
   > n26 = Set All IK Legs(All IK Legs:Array of EUW Picker Button Object References=Get All IK Leg Left()) -> (Output_Get:Array of EUW Picker Button Object References)
   > Append Array(TargetArray=@n26, SourceArray=Get All IK Leg Right())
   > Append Array(TargetArray=@n26, SourceArray=Get All IK Leg Extras Left())
   > Append Array(TargetArray=@n26, SourceArray=Get All IK Leg Extras Right())
   > Multi Select Button(
    Button=Get AllIKLegs() -> (AllIKLegs:EUW Picker Button Object Reference)
    Buttons to Select=Get All IK Legs() -> (All IK Legs:Array of EUW Picker Button Object References)
   )
   > Multi Select Button(
    Button=Get FKThighs() -> (FKThighs:EUW Picker Button Object Reference)
    Buttons to Select=Make Array("[0]"=Get thigh_r_fk_ctrl(), "[1]"=Get thigh_l_fk_ctrl())
   )
   > Multi Select Button(
    Button=Get FKKnees() -> (FKKnees:EUW Picker Button Object Reference)
    Buttons to Select=Make Array("[0]"=Get calf_l_fk_ctrl(), "[1]"=Get calf_r_fk_ctrl())
   )
   > Multi Select Button(
    Button=Get FKFeets() -> (FKFeets:EUW Picker Button Object Reference)
    Buttons to Select=Make Array("[0]"=Get foot_r_fk_ctrl(), "[1]"=Get foot_l_fk_ctrl())
   )
   > Multi Select Button(
    Button=Get FKToes() -> (FKToes:EUW Picker Button Object Reference)
    Buttons to Select=Make Array("[0]"=Get ball_r_fk_ctrl(), "[1]"=Get ball_l_fk_ctrl())
   )
   > Multi Select Button(
    Button=Get IKKnees() -> (IKKnees:EUW Picker Button Object Reference)
    Buttons to Select=Make Array("[0]"=Get leg_r_pv_ik_ctrl(), "[1]"=Get leg_l_pv_ik_ctrl())
   )
   > Multi Select Button(
    Button=Get IKFeet() -> (IKFeet:EUW Picker Button Object Reference)
    Buttons to Select=Make Array("[0]"=Get foot_l_ik_ctrl(), "[1]"=Get foot_r_ik_ctrl())
   )
   > Multi Select Button(
    Button=Get IKToes() -> (IKToes:EUW Picker Button Object Reference)
    Buttons to Select=Make Array("[0]"=Get ball_l_ik_ctrl(), "[1]"=Get ball_r_ik_ctrl())
   )
   > n27 = Set All IK Leg Extras(All IK Leg Extras:Array of EUW Picker Button Object References=Get All IK Leg Extras Left()) -> (Output_Get:Array of EUW Picker Button Object References)
   > Append Array(TargetArray=@n27, SourceArray=Get All IK Leg Extras Right())
   > Multi Select Button(Button=Get IKFeetExtras() -> (IKFeetExtras:EUW Picker Button Object Reference), Buttons to Select=@n27)
  ]}
 Multi Select Button {nodes [
   n1 = Multi Select Button!() -> (then, Button:EUW Picker Button Object Reference, Buttons to Select:Array of EUW Picker Button Object References) {is=title K2Node_FunctionEntry}
   > Select Multiple Buttons(Button=@n1, Buttons to Select=@n1.Buttons to Select)
  ]}
 Spawn Faceboard camera {nodes [
   n1 = GetBindingID!(
    execute
    Sequence:Movie Scene Sequence Object Reference=Get Current LS() -> (Current LS:Level Sequence Object Reference)
    InBinding:"Movie Scene Binding Proxy Structure (by ref)"=Get New Camera Binding() -> (New Camera Binding:Movie Scene Binding Proxy Structure)
   ) -> (then, ReturnValue:Movie Scene Object Binding ID Structure)
   > n2 = Cast To CameraActor!(
    execute
    Object="Get (a copy)"(Array:Array of Object References=GetBoundObjects!(ObjectBinding:Movie Scene Object Binding ID Structure=@n1) -> (ReturnValue:Array of Object References)) -> (Output:Object Reference)
   ) -> (then, CastFailed, AsCamera Actor:Camera Actor Object Reference) {is=title K2Node_DynamicCast}
   > Set Tags!(
    execute
    Tags:Array of Names=Make Array!("[0]":Name=Get Camera Tag() -> (Camera Tag:Name=None)) -> (Array:Array of Names)
    self:Actor Object Reference=@n2
   ) -> (then, Output_Get:Array of Names)
   > SetActorLabel(self=@n2, NewActorLabel=FaceboardCamera, bMarkDirty=true)
   > Set Actor Relative Transform!(
    execute
    self:Actor Object Reference=@n2
    NewRelativeTransform_Location:Vector="0,250,-45"
    NewRelativeTransform_Rotation:Rotator="0,-90,90"
    NewRelativeTransform_Scale:Vector="1,1,1"
    bSweep:bool=false
    bTeleport:bool=true
   ) -> (then, SweepHitResult:Hit Result Structure) {is=title K2Node_CallFunction}
   > n3 = Set Faceboard camera component(Faceboard camera component:Camera Component Object Reference=Get CameraComponent!(self:Camera Actor Object Reference=@n2) -> (CameraComponent:Camera Component Object Reference)) -> (Output_Get:Camera Component Object Reference)
   > SetProjectionMode!(execute, self:Camera Component Object Reference=@n3, InProjectionMode:ECameraProjectionMode Enum=Perspective) -> (then)
   n4 = Get Faceboard camera component() -> (Faceboard camera component:Camera Component Object Reference)
   > SetOrthoWidth!(execute, self:Camera Component Object Reference=@n4, InOrthoWidth:float=90) -> (then)
   > SetFieldOfView!(execute, self:Camera Component Object Reference=@n4, InFieldOfView:float=15) -> (then)
   > SetConstraintAspectRatio!(execute, self:Camera Component Object Reference=Get Faceboard camera component(), bInConstrainAspectRatio:bool=false) -> (then)
   > Set bCameraMeshHiddenInGame!(execute, bCameraMeshHiddenInGame:bool=true, self:Camera Component Object Reference=Get Faceboard camera component()) -> (then, Output_Get:bool)
   n5 = Make Weighted Blendable!(Weight:float=1, Object:Object Reference=/Game/Characters/Player/Common/SharedRigs/PostProcess/M_PP_FaceBoardCamera) -> (WeightedBlendable:Weighted Blendable Structure) {is=title K2Node_MakeStruct}
   > n6 = Set members in Post Process Settings!(
    execute:Exec
    StructRef:"Post Process Settings Structure (by ref)"=Get PostProcessSettings!(self:Camera Component Object Reference=Get Faceboard camera component()) -> (PostProcessSettings:Post Process Settings Structure)
    WeightedBlendables:Weighted Blendables Structure=Make Weighted Blendables!(Array:Array of Weighted Blendable Structures=Make Array("[0]":Weighted Blendable Structure=@n5) -> (Array:Array of Weighted Blendable Structures)) -> (WeightedBlendables:Weighted Blendables Structure) {is=title K2Node_MakeStruct}
   ) -> (then:Exec, StructOut:"Post Process Settings Structure (by ref)") {is=title K2Node_SetFieldsInStruct}
   > Set PostProcessSettings!(execute, PostProcessSettings:Post Process Settings Structure=@n6, self:Camera Component Object Reference=Get Faceboard camera component()) -> (then, Output_Get:Post Process Settings Structure)
   n7 = Sequence!(execute) -> (then_0, then_1)
   > n8 = AddTrack!(
    execute
    InBinding:"Movie Scene Binding Proxy Structure (by ref)"=Get New Camera Binding()
    TrackType:Movie Scene Track Class Reference=/Script/MovieSceneTracks.MovieScene3DAttachTrack
   ) -> (then, ReturnValue:Movie Scene Track Object Reference)
   > n9 = AddSection!(execute, Track:Movie Scene Track Object Reference=@n8) -> (then, ReturnValue:Movie Scene Section Object Reference)
   > n10 = GetPlaybackRange!(execute, Sequence:Movie Scene Sequence Object Reference=Get Current LS()) -> (then, ReturnValue_bHasStart:bool=~, ReturnValue_bHasEnd:bool=~, ReturnValue_InclusiveStart:int, ReturnValue_ExclusiveEnd:int)
   > SetRange!(
    execute
    Section:Movie Scene Section Object Reference=@n9
    StartFrame:int=@n10.ReturnValue_InclusiveStart
    EndFrame:int=@n10.ReturnValue_ExclusiveEnd
   ) -> (then)
   > n11 = Cast To MovieScene3DAttachSection!(execute, Object=@n9) -> (then, CastFailed, AsMovie Scene 3DAttach Section:Movie Scene 3DAttach Section Object Reference) {is=title K2Node_DynamicCast}
   > SetConstraintBindingID!(
    execute
    self:Movie Scene 3DConstraint Section Object Reference=@n11
    InConstraintBindingID:"Movie Scene Object Binding ID Structure (by ref)"=Get BP Anim Binding ID() -> (BP Anim Binding ID:Movie Scene Object Binding ID Structure)
   ) -> (then)
   > Set AttachSocketName!(execute, AttachSocketName:Name=head, self:Movie Scene 3DAttach Section Object Reference=@n11) -> (then, Output_Get:Name=None)
   n12 = GetTracks!(execute=@n7.then_1, InBinding:"Movie Scene Binding Proxy Structure (by ref)"=Get New Camera Binding()) -> (then, ReturnValue:Array of Movie Scene Track Object References)
   > n13 = For Each Loop with Break!(Exec:Exec, Array:Array of Movie Scene Track Object References=@n12, Break:Exec) -> (LoopBody:Exec, Array Element:Movie Scene Track Object Reference, Array Index:int, Completed:Exec)
   > n14 = Cast To MovieScene3DTransformTrack!(execute, Object=@n13) -> (then, CastFailed, AsMovie Scene 3DTransform Track:Movie Scene 3DTransform Track Object Reference) {is=title K2Node_DynamicCast}
   > n15 = GetSections!(execute, Track:Movie Scene Track Object Reference=@n14) -> (then, ReturnValue:Array of Movie Scene Section Object References)
   > n16 = For Each Loop with Break(Array:Array of Movie Scene Section Object References=@n15) -> (Array Element:Movie Scene Section Object Reference)
   > n17 = Cast To MovieScene3DTransformSection!(execute, Object=@n16) -> (then, CastFailed, AsMovie Scene 3DTransform Section:Movie Scene 3DTransform Section Object Reference) {is=title K2Node_DynamicCast}
   > n18 = GetChannelsByType!(
    execute
    Section:Movie Scene Section Object Reference=@n17
    ChannelType:Movie Scene Scripting Channel Class Reference=/Script/SequencerScripting.MovieSceneScriptingDoubleChannel
   ) -> (then, ReturnValue:Array of Movie Scene Scripting Channel Object References)
   > n19 = For Each Loop(Array:Array of Movie Scene Scripting Channel Object References=@n18) -> (Array Element:Movie Scene Scripting Channel Object Reference)
   > n20 = Cast To MovieSceneScriptingDoubleChannel!(execute, Object=@n19) -> (then, CastFailed, AsMovie Scene Scripting Double Channel:Movie Scene Scripting Double Channel Object Reference) {is=title K2Node_DynamicCast}
   > PrintString(InString=GetDisplayName(Object=@n20))
   > n21 = "For Each Loop (Map)"!(
    execute:Exec
    MapPin:"Map of Strings to Float (double-precision)s"=Get TransformChannels() -> (TransformChannels:"Map of Strings to Float (double-precision)s")
    BreakPin:Exec
   ) -> (then:Exec, KeyPin:String, ValuePin:double, CompletedPin:Exec)
   > Branch(Condition=Contains(SearchIn=GetDisplayName(Object=@n20), Substring=@n21, bUseCase=true))
   > "Set Default (Double)"!(execute, self:Movie Scene Scripting Double Channel Object Reference=@n20, InDefaultValue:double=@n21.ValuePin) -> (then) {is=title K2Node_CallFunction}
   Spawn Faceboard camera() {is=title K2Node_FunctionEntry}
   > n22 = GetFocusedLevelSequence()
   > Set Current LS(Current LS:Level Sequence Object Reference=@n22) -> (Output_Get:Level Sequence Object Reference)
   > n23 = GetSpawnables!(execute, Sequence:Movie Scene Sequence Object Reference=Get Current LS()) -> (then, ReturnValue:Array of Movie Scene Binding Proxy Structures)
   > n24 = For Each Loop(Array:Array of Movie Scene Binding Proxy Structures=@n23) -> (Array Element:Movie Scene Binding Proxy Structure)
   n25 = AddSpawnableFromClass!(
    execute=@n24.Completed
    self:Level Sequence Editor Subsystem Object Reference=Get LevelSequenceEditorSubsystem() -> (ReturnValue:Level Sequence Editor Subsystem Object Reference) {is=title K2Node_GetEditorSubsystem}
    Sequence:Movie Scene Sequence Object Reference=Get Current LS()
    ClassToSpawn:Object Class Reference=/Game/Characters/Player/Common/SharedRigs/PostProcess/BP_FaceboardCamera.BP_FaceboardCamera_C
   ) -> (then, ReturnValue:Movie Scene Binding Proxy Structure)
   > n26 = Set New Camera Binding(New Camera Binding:Movie Scene Binding Proxy Structure=@n25) -> (Output_Get:Movie Scene Binding Proxy Structure)
   > n27 = GetBindingID(Sequence=Get Current LS(), InBinding=@n26)
   > Set Faceboard Camera Actor(Faceboard Camera Actor:Object Reference="Get (a copy)"(Array=GetBoundObjects(ObjectBinding=@n27))) -> (Output_Get:Object Reference)
   > n28 = Sequence!(execute) -> (then_0, then_1, then_2:Exec, then_3:Exec, then_4:Exec)
   > n29 = GetControlRigs!(execute, LevelSequence:Level Sequence Object Reference=Get Current LS()) -> (then, ReturnValue:Array of Control Rig Sequencer Binding Proxy Structures)
   > n30 = For Each Loop(Array:Array of Control Rig Sequencer Binding Proxy Structures=@n29) -> (Array Element:Control Rig Sequencer Binding Proxy Structure)
   n31 = Break Control Rig Sequencer Binding Proxy(ControlRigSequencerBindingProxy=@n30)
   > Branch(Condition="Equal (Class)"(
     A=Get Class(Object=@n31.ControlRig) -> (ReturnValue:Control Rig Class Reference)
     B=/Game/Characters/Player/Common/Fortnite_Base_Head/Facials/ControlRigs/Fortnite_3L_Face_CtrlRig.Fortnite_3L_Face_CtrlRig_C
    ))
   n32 = Get Current LS()
   > n33 = GetBindingID(Sequence=@n32, InBinding=Break Control Rig Sequencer Binding Proxy(ControlRigSequencerBindingProxy=Get Control Rig Proxy Binding()))
   > n34 = GetBindingID(Sequence=@n32, InBinding=@n31)
   n35 = GetBoundObjects(ObjectBinding=@n34)
   n36 = GetBoundObjects(ObjectBinding=@n33)
   > Branch(Condition="Equal (Object)"(A="Get (a copy)"(Array=@n35), B="Get (a copy)"(Array=@n36)))
   > Set Face CR(Face CR:Control Rig Object Reference=@n31.ControlRig) -> (Output_Get:Control Rig Object Reference)
   n37 = GetPossessables!(execute=@n28.then_1, Sequence:Movie Scene Sequence Object Reference=Get Current LS()) -> (then, ReturnValue:Array of Movie Scene Binding Proxy Structures)
   > n38 = For Each Loop(Array:Array of Movie Scene Binding Proxy Structures=@n37) -> (Array Element:Movie Scene Binding Proxy Structure)
   > n39 = GetBindingID(Sequence=Get Current LS(), InBinding=@n38)
   > Cast To CameraComponent!(execute, Object="Get (a copy)"(Array=GetBoundObjects(ObjectBinding=@n39))) -> (then, CastFailed, AsCamera Component:Camera Component Object Reference) {is=title K2Node_DynamicCast}
   > Remove!(execute, InBinding:"Movie Scene Binding Proxy Structure (by ref)"=@n38) -> (then)
   n40 = GetEditorWorld(execute=@n28.then_2, self=Get UnrealEditorSubsystem())
   > n41 = CreateTransformableHandle!(
    execute
    InWorld:World Object Reference=@n40
    InObject:Object Reference=Get Faceboard Camera Actor() -> (Faceboard Camera Actor:Object Reference)
    InAttachmentName:"Name (by ref)"=MakeLiteralName()
   ) -> (then, ReturnValue:Transformable Handle Object Reference)
   > n42 = CreateFromType!(execute, InWorld:World Object Reference=@n40, InType:ETransformConstraintType Enum=Parent) -> (then, ReturnValue:Tickable Transform Constraint Object Reference)
   > AddConstraint!(
    execute
    InWorld:World Object Reference=@n40
    InParentHandle:Transformable Handle Object Reference=CreateTransformableControlHandle!(
     self:Control Rig Object Reference=Get Face CR() -> (Face CR:Control Rig Object Reference)
     ControlName:"Name (by ref)"=MakeLiteralName(Value=CTRL_Master)
    ) -> (ReturnValue:Transformable Control Handle Object Reference)
    InChildHandle:Transformable Handle Object Reference=@n41
    InConstraint:Tickable Transform Constraint Object Reference=@n42
    bMaintainOffset:bool=false
   ) -> (then, ReturnValue:bool)
   n43 = GetBindingID(execute=@n24.LoopBody, Sequence=Get Current LS(), InBinding=@n24)
   n44 = "Get (a copy)"(Array=GetBoundObjects(ObjectBinding=@n43))
   > n45 = Cast To SkeletalMeshActor!(execute, Object=@n44) -> (then, CastFailed, AsSkeletal Mesh Actor:Skeletal Mesh Actor Object Reference) {is=title K2Node_DynamicCast}
   > Branch(Condition=Contains(SearchIn=GetDisplayName(Object=@n45), Substring=Anim))
   > n46 = GetBindingID(Sequence=Get Current LS(), InBinding=@n24)
   > Set BP Anim Binding ID(BP Anim Binding ID:Movie Scene Object Binding ID Structure=@n46) -> (Output_Get:Movie Scene Object Binding ID Structure)
   n47 = Cast To CameraActor(execute=@n45.CastFailed, Object=@n44)
   > Branch(Condition=ActorHasTag!(self:Actor Object Reference=@n47, Tag:Name=Get Camera Tag() -> (Camera Tag=None)) -> (ReturnValue:bool))
   > Remove(InBinding=@n24)
  ]}
 Switch BG {nodes [
   Switch BG() {is=title K2Node_FunctionEntry}
   > Increment Int(Value=Get BG Index() -> (BG Index:int=0))
   n1 = Get BG Images_0() -> (BG Images_0:Array of Texture 2D Object References)
   n2 = Length(TargetArray:Array of Texture 2D Object References=@n1)
   > n3 = Branch(Condition="integer > integer"(A=Get BG Index() -> (BG Index=0), B=int - int(A=@n2)))
   > n4 = Set BG Index(BG Index:int=0) -> (Output_Get:int=0)
   n5 = "Get (a copy)"(Array:Array of Texture 2D Object References=@n1, Dimension 1=Get BG Index() -> (BG Index=0)) -> (Output:Texture 2D Object Reference)
   Is Valid(exec=[@n3.else, @n4.then], InputObject=@n5)
   > SetBrush(
    self=Get BG() -> (BG:Image Object Reference)
    InBrush=Make Slate Brush!(
     TintColor:Slate Color Structure=Make Slate Color(SpecifiedColor=(R=1,G=1,B=1,A=1))
     DrawAs:ESlateBrushDrawType Enum=Image
     Tiling:ESlateBrushTileType Enum=NoTile
     Mirroring:ESlateBrushMirrorType Enum=NoMirror
     ImageSize:Vector 2D Structure=MakeVector2D(X=1906, Y=1129)
     Margin:Margin Structure=(Left=0,Top=0,Right=0,Bottom=0)
     ResourceObject:Object Reference=@n5
     OutlineSettings:Slate Brush Outline Settings Structure=(CornerRadii=(X=0,Y=0,Z=0,W=0),Color=(SpecifiedColor=(R=0,G=0,B=0,A=0),ColorUseRule=UseColor_Specified),Width=0,RoundingType=HalfHeightRadius,bUseBrushTransparency=False)
    ) -> (SlateBrush:Slate Brush Structure) {is=title K2Node_MakeStruct}
   )
  ]}
 Update All IK FK Checkboxes {nodes [
   Update All IK FK Checkboxes() {is=title K2Node_FunctionEntry}
   > n1 = For Each Loop(Array:Array of Editor Utility Check Box Object References=Make Array!(
     "[0]":Editor Utility Check Box Object Reference=Get neck_fk_ik_switch()
     "[1]":Editor Utility Check Box Object Reference=Get arm_l_fk_ik_switch()
     "[2]":Editor Utility Check Box Object Reference=Get arm_r_fk_ik_switch()
     "[3]":Editor Utility Check Box Object Reference=Get leg_l_fk_ik_switch()
     "[4]":Editor Utility Check Box Object Reference=Get leg_r_fk_ik_switch()
     "[5]":Editor Utility Check Box Object Reference=Get spine_fk_ik_switch()
    ) -> (Array:Array of Editor Utility Check Box Object References)) -> (Array Element:Editor Utility Check Box Object Reference)
   Return Node(execute=@n1.Completed)
   > n2 = GetFocusedLevelSequence()
   > n3 = Get Current Local Time(TimeUnit=DisplayRate)
   > n4 = GetLocalControlRigBool!(
    execute
    LevelSequence:Level Sequence Object Reference=@n2
    ControlRig:Control Rig Object Reference=Break Control Rig Sequencer Binding Proxy(ControlRigSequencerBindingProxy=Get Control Rig Proxy Binding()).ControlRig
    ControlName:Name=String To Name(InString=GetObjectName(Object=@n1))
    Frame:Frame Number Structure=Make Frame Number(Value=Break Movie Scene Sequence Playback Params(MovieSceneSequencePlaybackParams=@n3) -> (Frame_SubFrame=~, Time=~, bHasJumped=~))
    TimeUnit:EMovieSceneTimeUnit Enum=DisplayRate
   ) -> (then, ReturnValue:bool)
   > SetIsChecked!(execute, self:Check Box Object Reference=@n1, InIsChecked:bool=@n4) -> (then)
  ]}
 Update All IK FK Control Colors {nodes [
   Update All IK FK Control Colors() {is=title K2Node_FunctionEntry}
   > Set Buttons Color From Checkbox State(CheckBox=Get leg_l_fk_ik_switch(), IK Buttons=Get Left Leg IK(), FK Buttons=Get Left Leg FK(), Color Mult=0.1, Opacity=0.5)
   > Set Buttons Color From Checkbox State(CheckBox=Get leg_r_fk_ik_switch(), IK Buttons=Get Right Leg IK(), FK Buttons=Get Right Leg FK(), Color Mult=0.1, Opacity=0.5)
   > Set Buttons Color From Checkbox State(CheckBox=Get arm_l_fk_ik_switch(), IK Buttons=Get Left Arm IK(), FK Buttons=Get Left Arm FK(), Color Mult=0.1, Opacity=0.5)
   > Set Buttons Color From Checkbox State(CheckBox=Get arm_r_fk_ik_switch(), IK Buttons=Get Right Arm IK(), FK Buttons=Get Right Arm FK(), Color Mult=0.1, Opacity=0.5)
   > Set Buttons Color From Checkbox State(CheckBox=Get spine_fk_ik_switch(), IK Buttons=Get Spine IK(), FK Buttons=Get Spine FK(), Color Mult=0.1, Opacity=0.5)
   > Set Buttons Color From Checkbox State(CheckBox=Get neck_fk_ik_switch(), IK Buttons=Get Neck IK(), FK Buttons=Get Neck FK(), Color Mult=0.1, Opacity=0.5)
  ]}
]
variables [
 4 Fingers Left {type="array<EUW_Picker_Button_C>", default []}
 "4 Fingers Right " {type="array<EUW_Picker_Button_C>", default []}
 All Arm Fingers Left {type="array<EUW_Picker_Button_C>", default []}
 "All Arm Fingers Right " {type="array<EUW_Picker_Button_C>", default []}
 All Arm Left {type="array<EUW_Picker_Button_C>", default []}
 All Arm Right {type="array<EUW_Picker_Button_C>", default []}
 All Arms {type="array<EUW_Picker_Button_C>", default []}
 All Fingers {type="array<EUW_Picker_Button_C>", default []}
 All Fingers Left {type="array<EUW_Picker_Button_C>", default []}
 All fingers Right {type="array<EUW_Picker_Button_C>", default []}
 All FK Arms {type="array<EUW_Picker_Button_C>", default []}
 All FK Leg Left {type="array<EUW_Picker_Button_C>", default []}
 All FK Leg Right {type="array<EUW_Picker_Button_C>", default []}
 All FK Legs {type="array<EUW_Picker_Button_C>", default []}
 All IK Leg Extras {type="array<EUW_Picker_Button_C>", default []}
 All IK Leg Extras Left {type="array<EUW_Picker_Button_C>", default []}
 All IK Leg Extras Right {type="array<EUW_Picker_Button_C>", default []}
 All IK Leg Left {type="array<EUW_Picker_Button_C>", default []}
 All IK Leg Right {type="array<EUW_Picker_Button_C>", default []}
 All IK Legs {type="array<EUW_Picker_Button_C>", default []}
 All Legs {type="array<EUW_Picker_Button_C>", default []}
 BG Images_0 {
  type="array<Texture2D>"
  default [
   /Game/Characters/Player/Common/SharedRigs/Graphics/Picker_Biped_Jonesy_Body_BW_Dark
   /Game/Characters/Player/Common/SharedRigs/Graphics/Picker_Biped_Jonesy_Body_BW_Light
   /Game/Characters/Player/Common/SharedRigs/Graphics/Picker_Biped_Jonesy_Body_BW_Mid
   /Game/Characters/Player/Common/SharedRigs/Graphics/Picker_Biped_Jonesy_Body_Color
   /Game/Characters/Player/Common/SharedRigs/Graphics/Picker_Biped_Jonesy_Body_Color_Dark
   /Game/Characters/Player/Common/SharedRigs/Graphics/Picker_Biped_Jonesy_Body_Color_Light
   /Game/Characters/Player/Common/SharedRigs/Graphics/Picker_Biped_Jonesy_Body_Color_Mid
  ]
 }
 BG Index {type=integer}
 Face Picker {type=UserWidget, default=null}
 FK Hands {type="array<EUW_Picker_Button_C>", default []}
 IK Move Buttons {type="array<EUW_Picker_Button_C>", default []}
 Left Arm FK {type="array<EUW_Picker_Button_C>", default []}
 Left Arm IK {type="array<EUW_Picker_Button_C>", default []}
 Left Leg FK {type="array<EUW_Picker_Button_C>", default []}
 Left Leg IK {type="array<EUW_Picker_Button_C>", default []}
 Neck FK {type="array<EUW_Picker_Button_C>", default []}
 Neck IK {type="array<EUW_Picker_Button_C>", default []}
 Right Arm FK {type="array<EUW_Picker_Button_C>", default []}
 Right Arm IK {type="array<EUW_Picker_Button_C>", default []}
 Right Leg FK {type="array<EUW_Picker_Button_C>", default []}
 Right Leg IK {type="array<EUW_Picker_Button_C>", default []}
 Spine FK {type="array<EUW_Picker_Button_C>", default []}
 Spine IK {type="array<EUW_Picker_Button_C>", default []}
 Thumb Left {type="array<EUW_Picker_Button_C>", default []}
 Thumb Right {type="array<EUW_Picker_Button_C>", default []}
 Twists {type="array<EUW_Picker_Button_C>", default []}
]
widgets {
 info {
  parentClass=/AnimInEdTools/AnimationPicker/EUW_Asset_Picker_Base.EUW_Asset_Picker_Base_C
  rootWidgetClass=&CanvasPanel
  widgetCount=198
 }
 widgets [
  CanvasPanel_0 {widgetClassPath=&CanvasPanel, bIsVariable=true, widget=CanvasPanel_0}
  CanvasPanel_1 {
   widgetClassPath=&CanvasPanel
   parent=CanvasPanel_0
   widget=CanvasPanel_1
   properties {visibility=HitTestInvisible}
   slot {
    class=CanvasPanelSlot
    properties {layoutData {alignment=<0.5 0.5>, anchors {maximum=<0.5 0.5>, minimum=<0.5 0.5>}, offsets=<ltrb 0 0 931.9431 920.8809>}}
   }
  }
  BG {
   widgetClassPath=&Image
   bIsVariable=true
   parent=CanvasPanel_1
   widget=BG
   properties {brush {
     bIsDynamicallyLoaded=false
     drawAs=Image
     imageType=FullColor
     mirroring=NoMirror
     resourceName=None
     resourceObject=&Picker_Biped_Jonesy_Body_Color
     tiling=NoTile
     imageSize=<1906 1129>
     margin=<ltrb 0 0 0 0>
     outlineSettings {
      bUseBrushTransparency=false
      roundingType=HalfHeightRadius
      width=0
      color {colorUseRule=UseColor_Specified, specifiedColor=<rgba 0 0 0 0>}
      cornerRadii=<0 0 0 0>
     }
     tintColor {colorUseRule=UseColor_Foreground, specifiedColor=<rgba 1 1 1 0>}
     uVRegion {bIsValid=false, max=<0 0>, min=<0 0>}
    }}
   slot {
    class=CanvasPanelSlot
    properties {layoutData {alignment=<0.5 0.5>, anchors {maximum=<0.5 0.5>, minimum=<0.5 0.5>}, offsets=<ltrb 0 -104.532 690 736>}}
   }
  }
  Border_78 {
   widgetClassPath=&Border
   parent=CanvasPanel_1
   widget=Border_78
   properties {brushColor=<rgba 0.0452 0.0452 0.0452 1>}
   slot {
    class=CanvasPanelSlot
    properties {
     bAutoSize=true
     zOrder=-1
     layoutData {alignment=<0 0>, anchors {maximum=<1 1>, minimum=<0 0>}, offsets=<ltrb -881.6817 -504.5045 -2261.6426 -1125.1611>}
    }
   }
  }
  TextBlock_5 {
   widgetClassPath=&TextBlock
   parent=CanvasPanel_0
   widget=TextBlock_5
   properties {text=Biped, colorAndOpacity {colorUseRule=UseColor_Specified, specifiedColor=<rgba 1 1 1 0.2>}}
   slot {
    class=CanvasPanelSlot
    properties {bAutoSize=true, layoutData {alignment=<0 0>, anchors {maximum=<0.5 0.5>, minimum=<0.5 0.5>}, offsets=<ltrb -448.9609 -448.5405 100 30>}}
   }
  }
  CanvasPanel {
   widgetClassPath=&CanvasPanel
   parent=CanvasPanel_0
   widget=CanvasPanel
   slot {class=CanvasPanelSlot, properties {layoutData {alignment=<0.5 0.5>, anchors {maximum=<1 1>, minimum=<0 0>}, offsets=<ltrb 0 4 0 -4>}}}
  }
  Zero {
   widgetClassPath=/AnimInEdTools/AnimationPicker/EUW_Picker_Button
   bIsVariable=true
   parent=CanvasPanel
   widget=Zero
   properties {label Font Size=20, label Text=0, normal Tint {colorUseRule=UseColor_Specified, specifiedColor=<rgba 0.1615 0.0278 0.0028 1>}}
   slot {
    class=CanvasPanelSlot
    properties {layoutData {alignment=<0 0>, anchors {maximum=<0.5 0.5>, minimum=<0.5 0.5>}, offsets=<ltrb -338.963 271.4595 61.234 58.4427>}}
   }
  }
  Holder {
   widgetClassPath=&CanvasPanel
   bIsVariable=true
   parent=CanvasPanel
   widget=Holder
   slot {
    class=CanvasPanelSlot
    properties {layoutData {alignment=<0 0>, anchors {maximum=<0.5 0.5>, minimum=<0.5 0.5>}, offsets=<ltrb -270 320 604.5045 306.2762>}}
   }
  }
  ExtraOptionSlot {
   widgetClassPath=&VerticalBox
   bIsVariable=true
   parent=Holder
   widget=ExtraOptionSlot
   slot {
    class=CanvasPanelSlot
    properties {bAutoSize=true, layoutData {alignment=<0 0>, anchors {maximum=<1 1>, minimum=<0 0>}, offsets=<ltrb 0 0 0 0>}}
   }
  }
  root_ctrl {
   widgetClassPath=&EUW_Picker_Button
   bIsVariable=true
   parent=CanvasPanel_0
   widget=root_ctrl
   properties {
    draw Style=Square
    label Font Size=15
    label Text=R
    label Color {colorUseRule=UseColor_Specified, specifiedColor=<rgba 0 0 0 1>}
    normal Tint {colorUseRule=UseColor_Specified, specifiedColor=<rgba 0.9911 0.4852 0.0452 1>}
   }
   slot {
    class=CanvasPanelSlot
    properties {layoutData {alignment=<0.5 0.5>, anchors {maximum=<0.5 0.5>, minimum=<0.5 0.5>}, offsets=<ltrb -213 -440 30 30>}}
   }
  }
  attach_ctrl {
   widgetClassPath=&EUW_Picker_Button
   bIsVariable=true
   parent=CanvasPanel_0
   widget=attach_ctrl
   properties {
    label Text=attach ctrl
    label Color {colorUseRule=UseColor_Specified, specifiedColor=<rgba 0 0 0 1>}
    normal Tint {colorUseRule=UseColor_Specified, specifiedColor=<rgba 0.9911 0.4852 0.0452 1>}
   }
   slot {
    class=CanvasPanelSlot
    properties {layoutData {alignment=<0.5 0.5>, anchors {maximum=<0.5 0.5>, minimum=<0.5 0.5>}, offsets=<ltrb -385 213.4595 70 20>}}
   }
  }
  body_offset_ctrl {
   widgetClassPath=&EUW_Picker_Button
   bIsVariable=true
   parent=CanvasPanel_0
   widget=body_offset_ctrl
   properties {
    draw Style=Square
    label Font Size=15
    label Text=O
    label Color {colorUseRule=UseColor_Specified, specifiedColor=<rgba 0 0 0 1>}
    normal Tint {colorUseRule=UseColor_Specified, specifiedColor=<rgba 0.9911 0.4852 0.0452 1>}
   }
   slot {
    class=CanvasPanelSlot
    properties {layoutData {alignment=<0.5 0.5>, anchors {maximum=<0.5 0.5>, minimum=<0.5 0.5>}, offsets=<ltrb -245 -440 30 30>}}
   }
  }
  global_ctrl {
   widgetClassPath=&EUW_Picker_Button
   bIsVariable=true
   parent=CanvasPanel_0
   widget=global_ctrl
   properties {
    draw Style=Square
    label Font Size=15
    label Text=G
    label Color {colorUseRule=UseColor_Specified, specifiedColor=<rgba 0 0 0 1>}
    normal Tint {colorUseRule=UseColor_Specified, specifiedColor=<rgba 0.9911 0.4852 0.0452 1>}
   }
   slot {
    class=CanvasPanelSlot
    properties {layoutData {alignment=<0.5 0.5>, anchors {maximum=<0.5 0.5>, minimum=<0.5 0.5>}, offsets=<ltrb -277 -440 30 30>}}
   }
  }
  head_ik_ctrl {
   widgetClassPath=&EUW_Picker_Button
   bIsVariable=true
   parent=CanvasPanel_0
   widget=head_ik_ctrl
   properties {draw Style=Square, label Text=IK, normal Tint {colorUseRule=UseColor_Specified, specifiedColor=<rgba 0.8 0.1263 0 1>}}
   slot {
    class=CanvasPanelSlot
    properties {layoutData {alignment=<0.5 0.5>, anchors {maximum=<0.5 0.5>, minimum=<0.5 0.5>}, offsets=<ltrb -24 -393 40 40>}}
   }
  }
  head_ctrl {
   widgetClassPath=&EUW_Picker_Button
   bIsVariable=true
   parent=CanvasPanel_0
   widget=head_ctrl
   properties {draw Style=Square, label Text=FK, normal Tint {colorUseRule=UseColor_Specified, specifiedColor=<rgba 1 0.2736 0 1>}}
   slot {
    class=CanvasPanelSlot
    properties {layoutData {alignment=<0.5 0.5>, anchors {maximum=<0.5 0.5>, minimum=<0.5 0.5>}, offsets=<ltrb 22 -393 40 40>}}
   }
  }
  neck_01_ctrl {
   widgetClassPath=&EUW_Picker_Button
   bIsVariable=true
   parent=CanvasPanel_0
   widget=neck_01_ctrl
   properties {draw Style=Square, normal Tint {colorUseRule=UseColor_Specified, specifiedColor=<rgba 1 0.4976 0.08 1>}}
   slot {
    class=CanvasPanelSlot
    properties {layoutData {alignment=<0.5 0.5>, anchors {maximum=<0.5 0.5>, minimum=<0.5 0.5>}, offsets=<ltrb 0 -315 40 15>}}
   }
  }
  neck_02_ctrl {
   widgetClassPath=&EUW_Picker_Button
   bIsVariable=true
   parent=CanvasPanel_0
   widget=neck_02_ctrl
   properties {draw Style=Square, normal Tint {colorUseRule=UseColor_Specified, specifiedColor=<rgba 1 0.4976 0.08 1>}}
   slot {
    class=CanvasPanelSlot
    properties {layoutData {alignment=<0.5 0.5>, anchors {maximum=<0.5 0.5>, minimum=<0.5 0.5>}, offsets=<ltrb 0 -335 50 15>}}
   }
  }
  spine_04_ctrl {
   widgetClassPath=&EUW_Picker_Button
   bIsVariable=true
   parent=CanvasPanel_0
   widget=spine_04_ctrl
   properties {draw Style=Square, normal Tint {colorUseRule=UseColor_Specified, specifiedColor=<rgba 1 0.4969 0.0802 1>}}
   slot {
    class=CanvasPanelSlot
    properties {layoutData {alignment=<0.5 0.5>, anchors {maximum=<0.5 0.5>, minimum=<0.5 0.5>}, offsets=<ltrb 0 -265 80 20>}}
   }
  }
  spine_03_ctrl {
   widgetClassPath=&EUW_Picker_Button
   bIsVariable=true
   parent=CanvasPanel_0
   widget=spine_03_ctrl
   properties {
    draw Style=Square
    label Text=FK
    label Color {colorUseRule=UseColor_Specified, specifiedColor=<rgba 0 0 0 1>}
    normal Tint {colorUseRule=UseColor_Specified, specifiedColor=<rgba 1 0.6654 0 1>}
   }
   slot {
    class=CanvasPanelSlot
    properties {layoutData {alignment=<0.5 0.5>, anchors {maximum=<0.5 0.5>, minimum=<0.5 0.5>}, offsets=<ltrb 23.0391 -233.0405 40 15>}}
   }
  }
  spine_02_ctrl {
   widgetClassPath=&EUW_Picker_Button
   bIsVariable=true
   parent=CanvasPanel_0
   widget=spine_02_ctrl
   properties {draw Style=Square, normal Tint {colorUseRule=UseColor_Specified, specifiedColor=<rgba 1 0.6654 0 1>}}
   slot {
    class=CanvasPanelSlot
    properties {layoutData {alignment=<0.5 0.5>, anchors {maximum=<0.5 0.5>, minimum=<0.5 0.5>}, offsets=<ltrb 23.0391 -213.0405 40 15>}}
   }
  }
  spine_01_ctrl {
   widgetClassPath=&EUW_Picker_Button
   bIsVariable=true
   parent=CanvasPanel_0
   widget=spine_01_ctrl
   properties {draw Style=Square, normal Tint {colorUseRule=UseColor_Specified, specifiedColor=<rgba 1 0.6654 0 1>}}
   slot {
    class=CanvasPanelSlot
    properties {layoutData {alignment=<0.5 0.5>, anchors {maximum=<0.5 0.5>, minimum=<0.5 0.5>}, offsets=<ltrb 23.0391 -193.0405 40 15>}}
   }
  }
  chest_ctrl {
   widgetClassPath=&EUW_Picker_Button
   bIsVariable=true
   parent=CanvasPanel_0
   widget=chest_ctrl
   properties {
    draw Style=Square
    label Text=IK
    label Color {colorUseRule=UseColor_Specified, specifiedColor=<rgba 0 0 0 1>}
    normal Tint {colorUseRule=UseColor_Specified, specifiedColor=<rgba 0.93 0.2937 0 1>}
   }
   slot {
    class=CanvasPanelSlot
    properties {layoutData {alignment=<0.5 0.5>, anchors {maximum=<0.5 0.5>, minimum=<0.5 0.5>}, offsets=<ltrb -24.9609 -233.0405 40 15>}}
   }
  }
  chest_tan_ctrl {
   widgetClassPath=&EUW_Picker_Button
   bIsVariable=true
   parent=CanvasPanel_0
   widget=chest_tan_ctrl
   properties {draw Style=Square, label Font Size=5, normal Tint {colorUseRule=UseColor_Specified, specifiedColor=<rgba 0.93 0.2937 0 1>}}
   slot {
    class=CanvasPanelSlot
    properties {layoutData {alignment=<0.5 0.5>, anchors {maximum=<0.5 0.5>, minimum=<0.5 0.5>}, offsets=<ltrb -24.9609 -213.0405 40 15>}}
   }
  }
  hips_tan_ctrl {
   widgetClassPath=&EUW_Picker_Button
   bIsVariable=true
   parent=CanvasPanel_0
   widget=hips_tan_ctrl
   properties {draw Style=Square, label Font Size=5, normal Tint {colorUseRule=UseColor_Specified, specifiedColor=<rgba 0.93 0.2937 0 1>}}
   slot {
    class=CanvasPanelSlot
    properties {layoutData {alignment=<0.5 0.5>, anchors {maximum=<0.5 0.5>, minimum=<0.5 0.5>}, offsets=<ltrb -24.9609 -193.0405 40 15>}}
   }
  }
  clavicle_r_ctrl {
   widgetClassPath=&EUW_Picker_Button
   bIsVariable=true
   parent=CanvasPanel_0
   widget=clavicle_r_ctrl
   properties {draw Style=Square, label Font Size=5, normal Tint {colorUseRule=UseColor_Specified, specifiedColor=<rgba 1 0.1913 0.2232 1>}}
   slot {
    class=CanvasPanelSlot
    properties {layoutData {alignment=<0.5 0.5>, anchors {maximum=<0.5 0.5>, minimum=<0.5 0.5>}, offsets=<ltrb -65 -325 60 25>}}
   }
  }
  hand_r_ik_ctrl {
   widgetClassPath=&EUW_Picker_Button
   bIsVariable=true
   parent=CanvasPanel_0
   widget=hand_r_ik_ctrl
   properties {
    draw Style=Square
    label Font Size=12
    label Text=IK
    label Color {colorUseRule=UseColor_Specified, specifiedColor=<rgba 0 0 0 1>}
    normal Tint {colorUseRule=UseColor_Specified, specifiedColor=<rgba 1 0.1913 0.2232 1>}
   }
   slot {
    class=CanvasPanelSlot
    properties {layoutData {alignment=<0.5 0.5>, anchors {maximum=<0.5 0.5>, minimum=<0.5 0.5>}, offsets=<ltrb -250 -325 25 25>}}
   }
  }
  arm_r_pv_ik_ctrl {
   widgetClassPath=&EUW_Picker_Button
   bIsVariable=true
   parent=CanvasPanel_0
   widget=arm_r_pv_ik_ctrl
   properties {draw Style=Square, label Font Size=5, normal Tint {colorUseRule=UseColor_Specified, specifiedColor=<rgba 1 0.1913 0.2232 1>}}
   slot {
    class=CanvasPanelSlot
    properties {layoutData {alignment=<0.5 0.5>, anchors {maximum=<0.5 0.5>, minimum=<0.5 0.5>}, offsets=<ltrb -150 -325 25 25>}}
   }
  }
  upperarm_twist_02_r_ctrl {
   widgetClassPath=&EUW_Picker_Button
   bIsVariable=true
   parent=CanvasPanel_0
   widget=upperarm_twist_02_r_ctrl
   properties {draw Style=Circle, label Font Size=5, normal Tint {colorUseRule=UseColor_Specified, specifiedColor=<rgba 0.3499 0.1416 0.6783 1>}}
   slot {
    class=CanvasPanelSlot
    properties {layoutData {alignment=<0.5 0.5>, anchors {maximum=<0.5 0.5>, minimum=<0.5 0.5>}, offsets=<ltrb -85 -270 15 15>}}
   }
  }
  upperarm_twist_01_r_ctrl {
   widgetClassPath=&EUW_Picker_Button
   bIsVariable=true
   parent=CanvasPanel_0
   widget=upperarm_twist_01_r_ctrl
   properties {draw Style=Circle, label Font Size=5, normal Tint {colorUseRule=UseColor_Specified, specifiedColor=<rgba 0.3499 0.1416 0.6783 1>}}
   slot {
    class=CanvasPanelSlot
    properties {layoutData {alignment=<0.5 0.5>, anchors {maximum=<0.5 0.5>, minimum=<0.5 0.5>}, offsets=<ltrb -110 -270 15 15>}}
   }
  }
  lowerarm_twist_01_r_ctrl {
   widgetClassPath=&EUW_Picker_Button
   bIsVariable=true
   parent=CanvasPanel_0
   widget=lowerarm_twist_01_r_ctrl
   properties {draw Style=Circle, label Font Size=5, normal Tint {colorUseRule=UseColor_Specified, specifiedColor=<rgba 0.3499 0.1416 0.6783 1>}}
   slot {
    class=CanvasPanelSlot
    properties {layoutData {alignment=<0.5 0.5>, anchors {maximum=<0.5 0.5>, minimum=<0.5 0.5>}, offsets=<ltrb -195 -270 15 15>}}
   }
  }
  lowerarm_twist_02_r_ctrl {
   widgetClassPath=&EUW_Picker_Button
   bIsVariable=true
   parent=CanvasPanel_0
   widget=lowerarm_twist_02_r_ctrl
   properties {draw Style=Circle, label Font Size=5, normal Tint {colorUseRule=UseColor_Specified, specifiedColor=<rgba 0.3499 0.1416 0.6783 1>}}
   slot {
    class=CanvasPanelSlot
    properties {layoutData {alignment=<0.5 0.5>, anchors {maximum=<0.5 0.5>, minimum=<0.5 0.5>}, offsets=<ltrb -170 -270 15 15>}}
   }
  }
  neck_fk_ik_switch {
   widgetClassPath=&EditorUtilityCheckBox
   bIsVariable=true
   parent=CanvasPanel_0
   widget=neck_fk_ik_switch
   properties {renderTransform {angle=0, scale=<0.8 0.8>, shear=<0 0>, translation=<0 0>}}
   slot {
    class=CanvasPanelSlot
    properties {bAutoSize=true, layoutData {alignment=<0.5 0.5>, anchors {maximum=<0.5 0.5>, minimum=<0.5 0.5>}, offsets=<ltrb 0 -435 100 30>}}
   }
  }
  TextBlock_4 {
   widgetClassPath=&TextBlock
   parent=neck_fk_ik_switch
   widget=TextBlock_4
   properties {
    text=Neck IK Switch
    font {
     bForceMonospaced=false
     bMaterialIsStencil=false
     fontMaterial=null
     fontObject=&Roboto
     letterSpacing=0
     monospacedWidth=1
     size=11.25
     skewAmount=0
     typefaceFontName=Bold
     outlineSettings {
      bApplyOutlineToDropShadows=false
      bMiteredCorners=false
      bSeparateFillAlpha=false
      outlineMaterial=null
      outlineSize=0
      outlineColor=<rgba 0 0 0 1>
     }
    }
   }
   slot {class=PanelSlot}
  }
  leg_r_fk_ik_switch {
   widgetClassPath=&EditorUtilityCheckBox
   bIsVariable=true
   parent=CanvasPanel_0
   widget=leg_r_fk_ik_switch
   properties {checkedState=Checked}
   slot {
    class=CanvasPanelSlot
    properties {bAutoSize=true, layoutData {alignment=<0 0>, anchors {maximum=<0.5 0.5>, minimum=<0.5 0.5>}, offsets=<ltrb -246 95 100 30>}}
   }
  }
  TextBlock {
   widgetClassPath=&TextBlock
   parent=leg_r_fk_ik_switch
   widget=TextBlock
   properties {
    text=Right leg IK switch
    font {
     bForceMonospaced=false
     bMaterialIsStencil=false
     fontMaterial=null
     fontObject=&Roboto
     letterSpacing=0
     monospacedWidth=1
     size=11.25
     skewAmount=0
     typefaceFontName=Bold
     outlineSettings {
      bApplyOutlineToDropShadows=false
      bMiteredCorners=false
      bSeparateFillAlpha=false
      outlineMaterial=null
      outlineSize=0
      outlineColor=<rgba 0 0 0 1>
     }
    }
   }
   slot {class=PanelSlot}
  }
  arm_r_fk_ik_switch {
   widgetClassPath=&EditorUtilityCheckBox
   bIsVariable=true
   parent=CanvasPanel_0
   widget=arm_r_fk_ik_switch
   properties {checkedState=Checked}
   slot {
    class=CanvasPanelSlot
    properties {bAutoSize=true, layoutData {alignment=<0.5 0.5>, anchors {maximum=<0.5 0.5>, minimum=<0.5 0.5>}, offsets=<ltrb -129.7487 -358 100 30>}}
   }
  }
  TextBlock_2 {
   widgetClassPath=&TextBlock
   parent=arm_r_fk_ik_switch
   widget=TextBlock_2
   properties {
    text=Right arm IK switch
    font {
     bForceMonospaced=false
     bMaterialIsStencil=false
     fontMaterial=null
     fontObject=&Roboto
     letterSpacing=0
     monospacedWidth=1
     size=11.25
     skewAmount=0
     typefaceFontName=Bold
     outlineSettings {
      bApplyOutlineToDropShadows=false
      bMiteredCorners=false
      bSeparateFillAlpha=false
      outlineMaterial=null
      outlineSize=0
      outlineColor=<rgba 0 0 0 1>
     }
    }
   }
   slot {class=PanelSlot}
  }
  arm_l_fk_ik_switch {
   widgetClassPath=&EditorUtilityCheckBox
   bIsVariable=true
   parent=CanvasPanel_0
   widget=arm_l_fk_ik_switch
   properties {checkedState=Checked}
   slot {
    class=CanvasPanelSlot
    properties {bAutoSize=true, layoutData {alignment=<0.5 0.5>, anchors {maximum=<0.5 0.5>, minimum=<0.5 0.5>}, offsets=<ltrb 130 -358 100 30>}}
   }
  }
  TextBlock_1 {
   widgetClassPath=&TextBlock
   parent=arm_l_fk_ik_switch
   widget=TextBlock_1
   properties {
    text=Left arm IK switch
    font {
     bForceMonospaced=false
     bMaterialIsStencil=false
     fontMaterial=null
     fontObject=&Roboto
     letterSpacing=0
     monospacedWidth=1
     size=11.25
     skewAmount=0
     typefaceFontName=Bold
     outlineSettings {
      bApplyOutlineToDropShadows=false
      bMiteredCorners=false
      bSeparateFillAlpha=false
      outlineMaterial=null
      outlineSize=0
      outlineColor=<rgba 0 0 0 1>
     }
    }
   }
   slot {class=PanelSlot}
  }
  spine_fk_ik_switch {
   widgetClassPath=&EditorUtilityCheckBox
   bIsVariable=true
   parent=CanvasPanel_0
   widget=spine_fk_ik_switch
   properties {checkedState=Checked}
   slot {
    class=CanvasPanelSlot
    properties {bAutoSize=true, layoutData {alignment=<0 0>, anchors {maximum=<0.5 0.5>, minimum=<0.5 0.5>}, offsets=<ltrb -208.9609 -216.5405 100 30>}}
   }
  }
  TextBlock_3 {
   widgetClassPath=&TextBlock
   parent=spine_fk_ik_switch
   widget=TextBlock_3
   properties {
    text=Spine IK switch
    font {
     bForceMonospaced=false
     bMaterialIsStencil=false
     fontMaterial=null
     fontObject=&Roboto
     letterSpacing=0
     monospacedWidth=1
     size=11.25
     skewAmount=0
     typefaceFontName=Bold
     outlineSettings {
      bApplyOutlineToDropShadows=false
      bMiteredCorners=false
      bSeparateFillAlpha=false
      outlineMaterial=null
      outlineSize=0
      outlineColor=<rgba 0 0 0 1>
     }
    }
   }
   slot {class=PanelSlot}
  }
  leg_l_fk_ik_switch {
   widgetClassPath=&EditorUtilityCheckBox
   bIsVariable=true
   parent=CanvasPanel_0
   widget=leg_l_fk_ik_switch
   properties {checkedState=Checked}
   slot {
    class=CanvasPanelSlot
    properties {bAutoSize=true, layoutData {alignment=<0 0>, anchors {maximum=<0.5 0.5>, minimum=<0.5 0.5>}, offsets=<ltrb 105 95 100 30>}}
   }
  }
  TextBlock_26 {
   widgetClassPath=&TextBlock
   parent=leg_l_fk_ik_switch
   widget=TextBlock_26
   properties {
    text=Left leg IK switch
    font {
     bForceMonospaced=false
     bMaterialIsStencil=false
     fontMaterial=null
     fontObject=&Roboto
     letterSpacing=0
     monospacedWidth=1
     size=11.25
     skewAmount=0
     typefaceFontName=Bold
     outlineSettings {
      bApplyOutlineToDropShadows=false
      bMiteredCorners=false
      bSeparateFillAlpha=false
      outlineMaterial=null
      outlineSize=0
      outlineColor=<rgba 0 0 0 1>
     }
    }
   }
   slot {class=PanelSlot}
  }
  hand_r_fk_ctrl {
   widgetClassPath=&EUW_Picker_Button
   bIsVariable=true
   parent=CanvasPanel_0
   widget=hand_r_fk_ctrl
   properties {
    draw Style=Square
    label Font Size=12
    label Text=FK
    label Color {colorUseRule=UseColor_Specified, specifiedColor=<rgba 0 0 0 1>}
    normal Tint {colorUseRule=UseColor_Specified, specifiedColor=<rgba 1 0 0 1>}
   }
   slot {
    class=CanvasPanelSlot
    properties {layoutData {alignment=<0.5 0.5>, anchors {maximum=<0.5 0.5>, minimum=<0.5 0.5>}, offsets=<ltrb -250 -295 25 25>}}
   }
  }
  lowerarm_r_fk_ctrl {
   widgetClassPath=&EUW_Picker_Button
   bIsVariable=true
   parent=CanvasPanel_0
   widget=lowerarm_r_fk_ctrl
   properties {draw Style=Square, label Font Size=5, normal Tint {colorUseRule=UseColor_Specified, specifiedColor=<rgba 1 0 0 1>}}
   slot {
    class=CanvasPanelSlot
    properties {layoutData {alignment=<0.5 0.5>, anchors {maximum=<0.5 0.5>, minimum=<0.5 0.5>}, offsets=<ltrb -190 -295 75 27>}}
   }
  }
  upperarm_r_fk_ctrl {
   widgetClassPath=&EUW_Picker_Button
   bIsVariable=true
   parent=CanvasPanel_0
   widget=upperarm_r_fk_ctrl
   properties {draw Style=Square, label Font Size=5, normal Tint {colorUseRule=UseColor_Specified, specifiedColor=<rgba 1 0 0 1>}}
   slot {
    class=CanvasPanelSlot
    properties {layoutData {alignment=<0.5 0.5>, anchors {maximum=<0.5 0.5>, minimum=<0.5 0.5>}, offsets=<ltrb -110 -295 75 27>}}
   }
  }
  attach_backpack_ctrl {
   widgetClassPath=&EUW_Picker_Button
   bIsVariable=true
   parent=CanvasPanel_0
   widget=attach_backpack_ctrl
   properties {
    label Text=backpack
    label Color {colorUseRule=UseColor_Specified, specifiedColor=<rgba 0 0 0 1>}
    normal Tint {colorUseRule=UseColor_Specified, specifiedColor=<rgba 0.9911 0.4852 0.0452 1>}
   }
   slot {
    class=CanvasPanelSlot
    properties {layoutData {alignment=<0.5 0.5>, anchors {maximum=<0.5 0.5>, minimum=<0.5 0.5>}, offsets=<ltrb -385 189.4595 70 20>}}
   }
  }
  attach_cape_ctrl {
   widgetClassPath=&EUW_Picker_Button
   bIsVariable=true
   parent=CanvasPanel_0
   widget=attach_cape_ctrl
   properties {
    label Text=cape
    label Color {colorUseRule=UseColor_Specified, specifiedColor=<rgba 0 0 0 1>}
    normal Tint {colorUseRule=UseColor_Specified, specifiedColor=<rgba 0.9911 0.4852 0.0452 1>}
   }
   slot {
    class=CanvasPanelSlot
    properties {layoutData {alignment=<0.5 0.5>, anchors {maximum=<0.5 0.5>, minimum=<0.5 0.5>}, offsets=<ltrb -385 165.4595 70 20>}}
   }
  }
  body_ctrl {
   widgetClassPath=&EUW_Picker_Button
   bIsVariable=true
   parent=CanvasPanel_0
   widget=body_ctrl
   properties {draw Style=Square, label Font Size=8, normal Tint {colorUseRule=UseColor_Specified, specifiedColor=<rgba 0.54 0.1492 0 1>}}
   slot {
    class=CanvasPanelSlot
    properties {layoutData {alignment=<0.5 0.5>, anchors {maximum=<0.5 0.5>, minimum=<0.5 0.5>}, offsets=<ltrb 0 -145 100 20>}}
   }
  }
  hips_ctrl {
   widgetClassPath=&EUW_Picker_Button
   bIsVariable=true
   parent=CanvasPanel_0
   widget=hips_ctrl
   properties {draw Style=Square, label Font Size=6, normal Tint {colorUseRule=UseColor_Specified, specifiedColor=<rgba 1 0.4969 0.0802 1>}}
   slot {
    class=CanvasPanelSlot
    properties {layoutData {alignment=<0.5 0.5>, anchors {maximum=<0.5 0.5>, minimum=<0.5 0.5>}, offsets=<ltrb 0 -120 80 15>}}
   }
  }
  ball_l_fk_ctrl {
   widgetClassPath=&EUW_Picker_Button
   bIsVariable=true
   parent=CanvasPanel_0
   widget=ball_l_fk_ctrl
   properties {draw Style=Square, label Font Size=5, normal Tint {colorUseRule=UseColor_Specified, specifiedColor=<rgba 0.0137 0 1 1>}}
   slot {
    class=CanvasPanelSlot
    properties {layoutData {alignment=<0.5 0.5>, anchors {maximum=<0.5 0.5>, minimum=<0.5 0.5>}, offsets=<ltrb 115 202 25 25>}}
   }
  }
  foot_l_fk_ctrl {
   widgetClassPath=&EUW_Picker_Button
   bIsVariable=true
   parent=CanvasPanel_0
   widget=foot_l_fk_ctrl
   properties {draw Style=Square, label Font Size=5, normal Tint {colorUseRule=UseColor_Specified, specifiedColor=<rgba 0.0137 0 1 1>}}
   slot {
    class=CanvasPanelSlot
    properties {layoutData {alignment=<0.5 0.5>, anchors {maximum=<0.5 0.5>, minimum=<0.5 0.5>}, offsets=<ltrb 65 202 65 25>}}
   }
  }
  calf_l_fk_ctrl {
   widgetClassPath=&EUW_Picker_Button
   bIsVariable=true
   parent=CanvasPanel_0
   widget=calf_l_fk_ctrl
   properties {draw Style=Square, label Font Size=5, normal Tint {colorUseRule=UseColor_Specified, specifiedColor=<rgba 0.0137 0 1 1>}}
   slot {
    class=CanvasPanelSlot
    properties {layoutData {alignment=<0.5 0.5>, anchors {maximum=<0.5 0.5>, minimum=<0.5 0.5>}, offsets=<ltrb 45 113 25 138>}}
   }
  }
  thigh_l_fk_ctrl {
   widgetClassPath=&EUW_Picker_Button
   bIsVariable=true
   parent=CanvasPanel_0
   widget=thigh_l_fk_ctrl
   properties {draw Style=Square, label Font Size=5, normal Tint {colorUseRule=UseColor_Specified, specifiedColor=<rgba 0.0137 0 1 1>}}
   slot {
    class=CanvasPanelSlot
    properties {layoutData {alignment=<0.5 0.5>, anchors {maximum=<0.5 0.5>, minimum=<0.5 0.5>}, offsets=<ltrb 45 -32 25 138>}}
   }
  }
  thigh_twist_01_l_ctrl {
   widgetClassPath=&EUW_Picker_Button
   bIsVariable=true
   parent=CanvasPanel_0
   widget=thigh_twist_01_l_ctrl
   properties {draw Style=Circle, label Font Size=5, normal Tint {colorUseRule=UseColor_Specified, specifiedColor=<rgba 0.3499 0.1416 0.6783 1>}}
   slot {
    class=CanvasPanelSlot
    properties {layoutData {alignment=<0.5 0.5>, anchors {maximum=<0.5 0.5>, minimum=<0.5 0.5>}, offsets=<ltrb 80 -10 15 15>}}
   }
  }
  leg_l_pv_ik_ctrl {
   widgetClassPath=&EUW_Picker_Button
   bIsVariable=true
   parent=CanvasPanel_0
   widget=leg_l_pv_ik_ctrl
   properties {draw Style=Square, label Font Size=5, normal Tint {colorUseRule=UseColor_Specified, specifiedColor=<rgba 0 0.4969 0.8228 1>}}
   slot {
    class=CanvasPanelSlot
    properties {layoutData {alignment=<0.5 0.5>, anchors {maximum=<0.5 0.5>, minimum=<0.5 0.5>}, offsets=<ltrb 80 40 30 30>}}
   }
  }
  foot_roll_l_ctrl {
   widgetClassPath=&EUW_Picker_Button
   bIsVariable=true
   parent=CanvasPanel_0
   widget=foot_roll_l_ctrl
   properties {
    draw Style=Square
    label Font Size=5
    label Color {colorUseRule=UseColor_Specified, specifiedColor=<rgba 0 0 0 1>}
    normal Tint {colorUseRule=UseColor_Specified, specifiedColor=<rgba 0.2195 0.5089 1 1>}
   }
   slot {
    class=CanvasPanelSlot
    properties {layoutData {alignment=<0.5 0.5>, anchors {maximum=<0.5 0.5>, minimum=<0.5 0.5>}, offsets=<ltrb 75 172 20 20>}}
   }
  }
  foot_l_ik_ctrl {
   widgetClassPath=&EUW_Picker_Button
   bIsVariable=true
   parent=CanvasPanel_0
   widget=foot_l_ik_ctrl
   properties {draw Style=Square, label Font Size=8, normal Tint {colorUseRule=UseColor_Specified, specifiedColor=<rgba 0 0.4969 0.8228 1>}}
   slot {
    class=CanvasPanelSlot
    properties {layoutData {alignment=<0.5 0.5>, anchors {maximum=<0.5 0.5>, minimum=<0.5 0.5>}, offsets=<ltrb 122 170 65 25>}}
   }
  }
  ball_l_ik_ctrl {
   widgetClassPath=&EUW_Picker_Button
   bIsVariable=true
   parent=CanvasPanel_0
   widget=ball_l_ik_ctrl
   properties {draw Style=Square, label Font Size=5, normal Tint {colorUseRule=UseColor_Specified, specifiedColor=<rgba 0 0.4969 0.8228 1>}}
   slot {
    class=CanvasPanelSlot
    properties {layoutData {alignment=<0.5 0.5>, anchors {maximum=<0.5 0.5>, minimum=<0.5 0.5>}, offsets=<ltrb 172 170 25 25>}}
   }
  }
  tip_l_ctrl {
   widgetClassPath=&EUW_Picker_Button
   bIsVariable=true
   parent=CanvasPanel_0
   widget=tip_l_ctrl
   properties {
    draw Style=Square
    label Font Size=5
    label Color {colorUseRule=UseColor_Specified, specifiedColor=<rgba 0 0 0 1>}
    normal Tint {colorUseRule=UseColor_Specified, specifiedColor=<rgba 0.2195 0.5089 1 1>}
   }
   slot {
    class=CanvasPanelSlot
    properties {layoutData {alignment=<0.5 0.5>, anchors {maximum=<0.5 0.5>, minimum=<0.5 0.5>}, offsets=<ltrb 199 172 20 20>}}
   }
  }
  foot_bk1_l_ctrl {
   widgetClassPath=&EUW_Picker_Button
   bIsVariable=true
   parent=CanvasPanel_0
   widget=foot_bk1_l_ctrl
   properties {
    draw Style=Square
    label Font Size=5
    label Color {colorUseRule=UseColor_Specified, specifiedColor=<rgba 0 0 0 1>}
    normal Tint {colorUseRule=UseColor_Specified, specifiedColor=<rgba 0.2195 0.5089 1 1>}
   }
   slot {
    class=CanvasPanelSlot
    properties {layoutData {alignment=<0.5 0.5>, anchors {maximum=<0.5 0.5>, minimum=<0.5 0.5>}, offsets=<ltrb 144 143 20 20>}}
   }
  }
  calf_twist_02_l_ctrl {
   widgetClassPath=&EUW_Picker_Button
   bIsVariable=true
   parent=CanvasPanel_0
   widget=calf_twist_02_l_ctrl
   properties {draw Style=Circle, label Font Size=5, normal Tint {colorUseRule=UseColor_Specified, specifiedColor=<rgba 0.3499 0.1416 0.6783 1>}}
   slot {
    class=CanvasPanelSlot
    properties {layoutData {alignment=<0.5 0.5>, anchors {maximum=<0.5 0.5>, minimum=<0.5 0.5>}, offsets=<ltrb 80 80 15 15>}}
   }
  }
  calf_twist_01_l_ctrl {
   widgetClassPath=&EUW_Picker_Button
   bIsVariable=true
   parent=CanvasPanel_0
   widget=calf_twist_01_l_ctrl
   properties {draw Style=Circle, label Font Size=5, normal Tint {colorUseRule=UseColor_Specified, specifiedColor=<rgba 0.3499 0.1416 0.6783 1>}}
   slot {
    class=CanvasPanelSlot
    properties {layoutData {alignment=<0.5 0.5>, anchors {maximum=<0.5 0.5>, minimum=<0.5 0.5>}, offsets=<ltrb 80 105 15 15>}}
   }
  }
  thigh_r_fk_ctrl {
   widgetClassPath=&EUW_Picker_Button
   bIsVariable=true
   parent=CanvasPanel_0
   widget=thigh_r_fk_ctrl
   properties {draw Style=Square, label Font Size=5, normal Tint {colorUseRule=UseColor_Specified, specifiedColor=<rgba 1 0 0 1>}}
   slot {
    class=CanvasPanelSlot
    properties {layoutData {alignment=<0.5 0.5>, anchors {maximum=<0.5 0.5>, minimum=<0.5 0.5>}, offsets=<ltrb -45 -32 25 138>}}
   }
  }
  calf_r_fk_ctrl {
   widgetClassPath=&EUW_Picker_Button
   bIsVariable=true
   parent=CanvasPanel_0
   widget=calf_r_fk_ctrl
   properties {draw Style=Square, label Font Size=5, normal Tint {colorUseRule=UseColor_Specified, specifiedColor=<rgba 1 0 0 1>}}
   slot {
    class=CanvasPanelSlot
    properties {layoutData {alignment=<0.5 0.5>, anchors {maximum=<0.5 0.5>, minimum=<0.5 0.5>}, offsets=<ltrb -45 113 25 138>}}
   }
  }
  foot_r_fk_ctrl {
   widgetClassPath=&EUW_Picker_Button
   bIsVariable=true
   parent=CanvasPanel_0
   widget=foot_r_fk_ctrl
   properties {draw Style=Square, label Font Size=5, normal Tint {colorUseRule=UseColor_Specified, specifiedColor=<rgba 1 0 0 1>}}
   slot {
    class=CanvasPanelSlot
    properties {layoutData {alignment=<0.5 0.5>, anchors {maximum=<0.5 0.5>, minimum=<0.5 0.5>}, offsets=<ltrb -65 202 65 25>}}
   }
  }
  ball_r_fk_ctrl {
   widgetClassPath=&EUW_Picker_Button
   bIsVariable=true
   parent=CanvasPanel_0
   widget=ball_r_fk_ctrl
   properties {draw Style=Square, label Font Size=5, normal Tint {colorUseRule=UseColor_Specified, specifiedColor=<rgba 1 0 0 1>}}
   slot {
    class=CanvasPanelSlot
    properties {layoutData {alignment=<0.5 0.5>, anchors {maximum=<0.5 0.5>, minimum=<0.5 0.5>}, offsets=<ltrb -115 202 25 25>}}
   }
  }
  thigh_twist_01_r_ctrl {
   widgetClassPath=&EUW_Picker_Button
   bIsVariable=true
   parent=CanvasPanel_0
   widget=thigh_twist_01_r_ctrl
   properties {draw Style=Circle, label Font Size=5, normal Tint {colorUseRule=UseColor_Specified, specifiedColor=<rgba 0.3499 0.1416 0.6783 1>}}
   slot {
    class=CanvasPanelSlot
    properties {layoutData {alignment=<0.5 0.5>, anchors {maximum=<0.5 0.5>, minimum=<0.5 0.5>}, offsets=<ltrb -80 -10 15 15>}}
   }
  }
  foot_r_ik_ctrl {
   widgetClassPath=&EUW_Picker_Button
   bIsVariable=true
   parent=CanvasPanel_0
   widget=foot_r_ik_ctrl
   properties {draw Style=Square, label Font Size=8, normal Tint {colorUseRule=UseColor_Specified, specifiedColor=<rgba 1 0.1913 0.2232 1>}}
   slot {
    class=CanvasPanelSlot
    properties {layoutData {alignment=<0.5 0.5>, anchors {maximum=<0.5 0.5>, minimum=<0.5 0.5>}, offsets=<ltrb -122 170 65 25>}}
   }
  }
  foot_roll_r_ctrl {
   widgetClassPath=&EUW_Picker_Button
   bIsVariable=true
   parent=CanvasPanel_0
   widget=foot_roll_r_ctrl
   properties {
    draw Style=Square
    label Font Size=5
    label Color {colorUseRule=UseColor_Specified, specifiedColor=<rgba 0 0 0 1>}
    normal Tint {colorUseRule=UseColor_Specified, specifiedColor=<rgba 1 0.33 0.5548 1>}
   }
   slot {
    class=CanvasPanelSlot
    properties {layoutData {alignment=<0.5 0.5>, anchors {maximum=<0.5 0.5>, minimum=<0.5 0.5>}, offsets=<ltrb -75 172 20 20>}}
   }
  }
  ball_r_ik_ctrl {
   widgetClassPath=&EUW_Picker_Button
   bIsVariable=true
   parent=CanvasPanel_0
   widget=ball_r_ik_ctrl
   properties {draw Style=Square, label Font Size=5, normal Tint {colorUseRule=UseColor_Specified, specifiedColor=<rgba 1 0.1913 0.2232 1>}}
   slot {
    class=CanvasPanelSlot
    properties {layoutData {alignment=<0.5 0.5>, anchors {maximum=<0.5 0.5>, minimum=<0.5 0.5>}, offsets=<ltrb -172 170 25 25>}}
   }
  }
  tip_r_ctrl {
   widgetClassPath=&EUW_Picker_Button
   bIsVariable=true
   parent=CanvasPanel_0
   widget=tip_r_ctrl
   properties {
    draw Style=Square
    label Font Size=5
    label Color {colorUseRule=UseColor_Specified, specifiedColor=<rgba 0 0 0 1>}
    normal Tint {colorUseRule=UseColor_Specified, specifiedColor=<rgba 1 0.33 0.5548 1>}
   }
   slot {
    class=CanvasPanelSlot
    properties {layoutData {alignment=<0.5 0.5>, anchors {maximum=<0.5 0.5>, minimum=<0.5 0.5>}, offsets=<ltrb -199 172 20 20>}}
   }
  }
  foot_bk1_r_ctrl {
   widgetClassPath=&EUW_Picker_Button
   bIsVariable=true
   parent=CanvasPanel_0
   widget=foot_bk1_r_ctrl
   properties {
    draw Style=Square
    label Font Size=5
    label Color {colorUseRule=UseColor_Specified, specifiedColor=<rgba 0 0 0 1>}
    normal Tint {colorUseRule=UseColor_Specified, specifiedColor=<rgba 1 0.33 0.5548 1>}
   }
   slot {
    class=CanvasPanelSlot
    properties {layoutData {alignment=<0.5 0.5>, anchors {maximum=<0.5 0.5>, minimum=<0.5 0.5>}, offsets=<ltrb -144 143 20 20>}}
   }
  }
  leg_r_pv_ik_ctrl {
   widgetClassPath=&EUW_Picker_Button
   bIsVariable=true
   parent=CanvasPanel_0
   widget=leg_r_pv_ik_ctrl
   properties {draw Style=Square, label Font Size=5, normal Tint {colorUseRule=UseColor_Specified, specifiedColor=<rgba 1 0.1913 0.2232 1>}}
   slot {
    class=CanvasPanelSlot
    properties {layoutData {alignment=<0.5 0.5>, anchors {maximum=<0.5 0.5>, minimum=<0.5 0.5>}, offsets=<ltrb -80 40 30 30>}}
   }
  }
  calf_twist_01_r_ctrl {
   widgetClassPath=&EUW_Picker_Button
   bIsVariable=true
   parent=CanvasPanel_0
   widget=calf_twist_01_r_ctrl
   properties {draw Style=Circle, label Font Size=5, normal Tint {colorUseRule=UseColor_Specified, specifiedColor=<rgba 0.3499 0.1416 0.6783 1>}}
   slot {
    class=CanvasPanelSlot
    properties {layoutData {alignment=<0.5 0.5>, anchors {maximum=<0.5 0.5>, minimum=<0.5 0.5>}, offsets=<ltrb -80 105 15 15>}}
   }
  }
  calf_twist_02_r_ctrl {
   widgetClassPath=&EUW_Picker_Button
   bIsVariable=true
   parent=CanvasPanel_0
   widget=calf_twist_02_r_ctrl
   properties {draw Style=Circle, label Font Size=5, normal Tint {colorUseRule=UseColor_Specified, specifiedColor=<rgba 0.3499 0.1416 0.6783 1>}}
   slot {
    class=CanvasPanelSlot
    properties {layoutData {alignment=<0.5 0.5>, anchors {maximum=<0.5 0.5>, minimum=<0.5 0.5>}, offsets=<ltrb -80 80 15 15>}}
   }
  }
  clavicle_l_ctrl {
   widgetClassPath=&EUW_Picker_Button
   bIsVariable=true
   parent=CanvasPanel_0
   widget=clavicle_l_ctrl
   properties {draw Style=Square, label Font Size=5, normal Tint {colorUseRule=UseColor_Specified, specifiedColor=<rgba 0 0.4969 0.8228 1>}}
   slot {
    class=CanvasPanelSlot
    properties {layoutData {alignment=<0.5 0.5>, anchors {maximum=<0.5 0.5>, minimum=<0.5 0.5>}, offsets=<ltrb 65 -325 60 25>}}
   }
  }
  upperarm_l_fk_ctrl {
   widgetClassPath=&EUW_Picker_Button
   bIsVariable=true
   parent=CanvasPanel_0
   widget=upperarm_l_fk_ctrl
   properties {draw Style=Square, label Font Size=5, normal Tint {colorUseRule=UseColor_Specified, specifiedColor=<rgba 0.0137 0 1 1>}}
   slot {
    class=CanvasPanelSlot
    properties {layoutData {alignment=<0.5 0.5>, anchors {maximum=<0.5 0.5>, minimum=<0.5 0.5>}, offsets=<ltrb 110 -295 75 27>}}
   }
  }
  lowerarm_l_fk_ctrl {
   widgetClassPath=&EUW_Picker_Button
   bIsVariable=true
   parent=CanvasPanel_0
   widget=lowerarm_l_fk_ctrl
   properties {draw Style=Square, label Font Size=5, normal Tint {colorUseRule=UseColor_Specified, specifiedColor=<rgba 0.0137 0 1 1>}}
   slot {
    class=CanvasPanelSlot
    properties {layoutData {alignment=<0.5 0.5>, anchors {maximum=<0.5 0.5>, minimum=<0.5 0.5>}, offsets=<ltrb 190 -295 75 27>}}
   }
  }
  hand_l_fk_ctrl {
   widgetClassPath=&EUW_Picker_Button
   bIsVariable=true
   parent=CanvasPanel_0
   widget=hand_l_fk_ctrl
   properties {
    draw Style=Square
    label Font Size=12
    label Text=FK
    label Color {colorUseRule=UseColor_Specified, specifiedColor=<rgba 0 0 0 1>}
    normal Tint {colorUseRule=UseColor_Specified, specifiedColor=<rgba 0.0137 0 1 1>}
   }
   slot {
    class=CanvasPanelSlot
    properties {layoutData {alignment=<0.5 0.5>, anchors {maximum=<0.5 0.5>, minimum=<0.5 0.5>}, offsets=<ltrb 250 -295 25 25>}}
   }
  }
  hand_l_ik_ctrl {
   widgetClassPath=&EUW_Picker_Button
   bIsVariable=true
   parent=CanvasPanel_0
   widget=hand_l_ik_ctrl
   properties {
    draw Style=Square
    label Font Size=12
    label Text=IK
    label Color {colorUseRule=UseColor_Specified, specifiedColor=<rgba 0 0 0 1>}
    normal Tint {colorUseRule=UseColor_Specified, specifiedColor=<rgba 0 0.4969 0.8228 1>}
   }
   slot {
    class=CanvasPanelSlot
    properties {layoutData {alignment=<0.5 0.5>, anchors {maximum=<0.5 0.5>, minimum=<0.5 0.5>}, offsets=<ltrb 250 -325 25 25>}}
   }
  }
  arm_l_pv_ik_ctrl {
   widgetClassPath=&EUW_Picker_Button
   bIsVariable=true
   parent=CanvasPanel_0
   widget=arm_l_pv_ik_ctrl
   properties {draw Style=Square, label Font Size=5, normal Tint {colorUseRule=UseColor_Specified, specifiedColor=<rgba 0 0.4969 0.8228 1>}}
   slot {
    class=CanvasPanelSlot
    properties {layoutData {alignment=<0.5 0.5>, anchors {maximum=<0.5 0.5>, minimum=<0.5 0.5>}, offsets=<ltrb 150 -325 25 25>}}
   }
  }
  upperarm_twist_01_l_ctrl {
   widgetClassPath=&EUW_Picker_Button
   bIsVariable=true
   parent=CanvasPanel_0
   widget=upperarm_twist_01_l_ctrl
   properties {draw Style=Circle, label Font Size=5, normal Tint {colorUseRule=UseColor_Specified, specifiedColor=<rgba 0.3499 0.1416 0.6783 1>}}
   slot {
    class=CanvasPanelSlot
    properties {layoutData {alignment=<0.5 0.5>, anchors {maximum=<0.5 0.5>, minimum=<0.5 0.5>}, offsets=<ltrb 85 -270 15 15>}}
   }
  }
  upperarm_twist_02_l_ctrl {
   widgetClassPath=&EUW_Picker_Button
   bIsVariable=true
   parent=CanvasPanel_0
   widget=upperarm_twist_02_l_ctrl
   properties {draw Style=Circle, label Font Size=5, normal Tint {colorUseRule=UseColor_Specified, specifiedColor=<rgba 0.3499 0.1416 0.6783 1>}}
   slot {
    class=CanvasPanelSlot
    properties {layoutData {alignment=<0.5 0.5>, anchors {maximum=<0.5 0.5>, minimum=<0.5 0.5>}, offsets=<ltrb 110 -270 15 15>}}
   }
  }
  lowerarm_twist_02_l_ctrl {
   widgetClassPath=&EUW_Picker_Button
   bIsVariable=true
   parent=CanvasPanel_0
   widget=lowerarm_twist_02_l_ctrl
   properties {draw Style=Circle, label Font Size=5, normal Tint {colorUseRule=UseColor_Specified, specifiedColor=<rgba 0.3499 0.1416 0.6783 1>}}
   slot {
    class=CanvasPanelSlot
    properties {layoutData {alignment=<0.5 0.5>, anchors {maximum=<0.5 0.5>, minimum=<0.5 0.5>}, offsets=<ltrb 170 -270 15 15>}}
   }
  }
  lowerarm_twist_01_l_ctrl {
   widgetClassPath=&EUW_Picker_Button
   bIsVariable=true
   parent=CanvasPanel_0
   widget=lowerarm_twist_01_l_ctrl
   properties {draw Style=Circle, label Font Size=5, normal Tint {colorUseRule=UseColor_Specified, specifiedColor=<rgba 0.3499 0.1416 0.6783 1>}}
   slot {
    class=CanvasPanelSlot
    properties {layoutData {alignment=<0.5 0.5>, anchors {maximum=<0.5 0.5>, minimum=<0.5 0.5>}, offsets=<ltrb 195 -270 15 15>}}
   }
  }
  hand_attach_l_ctrl {
   widgetClassPath=&EUW_Picker_Button
   bIsVariable=true
   parent=CanvasPanel_0
   widget=hand_attach_l_ctrl
   properties {
    draw Style=Square
    label Font Size=8
    label Text=W2
    label Color {colorUseRule=UseColor_Specified, specifiedColor=<rgba 0 0 0 1>}
    normal Tint {colorUseRule=UseColor_Specified, specifiedColor=<rgba 0.22 0.5074 1 1>}
   }
   slot {
    class=CanvasPanelSlot
    properties {layoutData {alignment=<0.5 0.5>, anchors {maximum=<0.5 0.5>, minimum=<0.5 0.5>}, offsets=<ltrb 263 -358 20 20>}}
   }
  }
  weapon_l_ctrl {
   widgetClassPath=&EUW_Picker_Button
   bIsVariable=true
   parent=CanvasPanel_0
   widget=weapon_l_ctrl
   properties {
    draw Style=Square
    label Font Size=8
    label Text=W1
    label Color {colorUseRule=UseColor_Specified, specifiedColor=<rgba 0 0 0 1>}
    normal Tint {colorUseRule=UseColor_Specified, specifiedColor=<rgba 0.22 0.5074 1 1>}
   }
   slot {
    class=CanvasPanelSlot
    properties {layoutData {alignment=<0.5 0.5>, anchors {maximum=<0.5 0.5>, minimum=<0.5 0.5>}, offsets=<ltrb 237 -358 20 20>}}
   }
  }
  thumb_curl_l_ctrl {
   widgetClassPath=&EUW_Picker_Button
   bIsVariable=true
   parent=CanvasPanel_0
   widget=thumb_curl_l_ctrl
   properties {draw Style=Square, label Font Size=5, normal Tint {colorUseRule=UseColor_Specified, specifiedColor=<rgba 0 1 0.2854 1>}}
   slot {
    class=CanvasPanelSlot
    properties {layoutData {alignment=<0.5 0.5>, anchors {maximum=<0.5 0.5>, minimum=<0.5 0.5>}, offsets=<ltrb 175 -113 15 15>}}
   }
  }
  thumb_01_l_ctrl {
   widgetClassPath=&EUW_Picker_Button
   bIsVariable=true
   parent=CanvasPanel_0
   widget=thumb_01_l_ctrl
   properties {draw Style=Square, label Font Size=5, normal Tint {colorUseRule=UseColor_Specified, specifiedColor=<rgba 0.0134 0 1 1>}}
   slot {
    class=CanvasPanelSlot
    properties {layoutData {alignment=<0.5 0.5>, anchors {maximum=<0.5 0.5>, minimum=<0.5 0.5>}, offsets=<ltrb 170 -93 15 15>}}
   }
  }
  thumb_02_l_ctrl {
   widgetClassPath=&EUW_Picker_Button
   bIsVariable=true
   parent=CanvasPanel_0
   widget=thumb_02_l_ctrl
   properties {draw Style=Square, label Font Size=5, normal Tint {colorUseRule=UseColor_Specified, specifiedColor=<rgba 0.0134 0 1 1>}}
   slot {
    class=CanvasPanelSlot
    properties {layoutData {alignment=<0.5 0.5>, anchors {maximum=<0.5 0.5>, minimum=<0.5 0.5>}, offsets=<ltrb 166 -73 15 15>}}
   }
  }
  thumb_03_l_ctrl {
   widgetClassPath=&EUW_Picker_Button
   bIsVariable=true
   parent=CanvasPanel_0
   widget=thumb_03_l_ctrl
   properties {draw Style=Square, label Font Size=5, normal Tint {colorUseRule=UseColor_Specified, specifiedColor=<rgba 0.0134 0 1 1>}}
   slot {
    class=CanvasPanelSlot
    properties {layoutData {alignment=<0.5 0.5>, anchors {maximum=<0.5 0.5>, minimum=<0.5 0.5>}, offsets=<ltrb 162 -53 15 15>}}
   }
  }
  index_curl_l_ctrl {
   widgetClassPath=&EUW_Picker_Button
   bIsVariable=true
   parent=CanvasPanel_0
   widget=index_curl_l_ctrl
   properties {draw Style=Square, label Font Size=5, normal Tint {colorUseRule=UseColor_Specified, specifiedColor=<rgba 0 1 0.2854 1>}}
   slot {
    class=CanvasPanelSlot
    properties {layoutData {alignment=<0.5 0.5>, anchors {maximum=<0.5 0.5>, minimum=<0.5 0.5>}, offsets=<ltrb 196 -105 15 15>}}
   }
  }
  index_metacarpal_l_ctrl {
   widgetClassPath=&EUW_Picker_Button
   bIsVariable=true
   parent=CanvasPanel_0
   widget=index_metacarpal_l_ctrl
   properties {draw Style=Square, label Font Size=5, normal Tint {colorUseRule=UseColor_Specified, specifiedColor=<rgba 0.0134 0 1 1>}}
   slot {
    class=CanvasPanelSlot
    properties {layoutData {alignment=<0.5 0.5>, anchors {maximum=<0.5 0.5>, minimum=<0.5 0.5>}, offsets=<ltrb 196 -85 15 15>}}
   }
  }
  index_01_l_ctrl {
   widgetClassPath=&EUW_Picker_Button
   bIsVariable=true
   parent=CanvasPanel_0
   widget=index_01_l_ctrl
   properties {draw Style=Square, label Font Size=5, normal Tint {colorUseRule=UseColor_Specified, specifiedColor=<rgba 0.0134 0 1 1>}}
   slot {
    class=CanvasPanelSlot
    properties {layoutData {alignment=<0.5 0.5>, anchors {maximum=<0.5 0.5>, minimum=<0.5 0.5>}, offsets=<ltrb 196 -65 15 15>}}
   }
  }
  index_02_l_ctrl {
   widgetClassPath=&EUW_Picker_Button
   bIsVariable=true
   parent=CanvasPanel_0
   widget=index_02_l_ctrl
   properties {draw Style=Square, label Font Size=5, normal Tint {colorUseRule=UseColor_Specified, specifiedColor=<rgba 0.0134 0 1 1>}}
   slot {
    class=CanvasPanelSlot
    properties {layoutData {alignment=<0.5 0.5>, anchors {maximum=<0.5 0.5>, minimum=<0.5 0.5>}, offsets=<ltrb 196 -45 15 15>}}
   }
  }
  index_03_l_ctrl {
   widgetClassPath=&EUW_Picker_Button
   bIsVariable=true
   parent=CanvasPanel_0
   widget=index_03_l_ctrl
   properties {draw Style=Square, label Font Size=5, normal Tint {colorUseRule=UseColor_Specified, specifiedColor=<rgba 0.0134 0 1 1>}}
   slot {
    class=CanvasPanelSlot
    properties {layoutData {alignment=<0.5 0.5>, anchors {maximum=<0.5 0.5>, minimum=<0.5 0.5>}, offsets=<ltrb 196 -25 15 15>}}
   }
  }
  middle_curl_l_ctrl {
   widgetClassPath=&EUW_Picker_Button
   bIsVariable=true
   parent=CanvasPanel_0
   widget=middle_curl_l_ctrl
   properties {draw Style=Square, label Font Size=5, normal Tint {colorUseRule=UseColor_Specified, specifiedColor=<rgba 0 1 0.2854 1>}}
   slot {
    class=CanvasPanelSlot
    properties {layoutData {alignment=<0.5 0.5>, anchors {maximum=<0.5 0.5>, minimum=<0.5 0.5>}, offsets=<ltrb 218 -105 15 15>}}
   }
  }
  middle_metacarpal_l_ctrl {
   widgetClassPath=&EUW_Picker_Button
   bIsVariable=true
   parent=CanvasPanel_0
   widget=middle_metacarpal_l_ctrl
   properties {draw Style=Square, label Font Size=5, normal Tint {colorUseRule=UseColor_Specified, specifiedColor=<rgba 0.0134 0 1 1>}}
   slot {
    class=CanvasPanelSlot
    properties {layoutData {alignment=<0.5 0.5>, anchors {maximum=<0.5 0.5>, minimum=<0.5 0.5>}, offsets=<ltrb 218 -85 15 15>}}
   }
  }
  middle_01_l_ctrl {
   widgetClassPath=&EUW_Picker_Button
   bIsVariable=true
   parent=CanvasPanel_0
   widget=middle_01_l_ctrl
   properties {draw Style=Square, label Font Size=5, normal Tint {colorUseRule=UseColor_Specified, specifiedColor=<rgba 0.0134 0 1 1>}}
   slot {
    class=CanvasPanelSlot
    properties {layoutData {alignment=<0.5 0.5>, anchors {maximum=<0.5 0.5>, minimum=<0.5 0.5>}, offsets=<ltrb 218 -65 15 15>}}
   }
  }
  middle_02_l_ctrl {
   widgetClassPath=&EUW_Picker_Button
   bIsVariable=true
   parent=CanvasPanel_0
   widget=middle_02_l_ctrl
   properties {draw Style=Square, label Font Size=5, normal Tint {colorUseRule=UseColor_Specified, specifiedColor=<rgba 0.0134 0 1 1>}}
   slot {
    class=CanvasPanelSlot
    properties {layoutData {alignment=<0.5 0.5>, anchors {maximum=<0.5 0.5>, minimum=<0.5 0.5>}, offsets=<ltrb 218 -45 15 15>}}
   }
  }
  middle_03_l_ctrl {
   widgetClassPath=&EUW_Picker_Button
   bIsVariable=true
   parent=CanvasPanel_0
   widget=middle_03_l_ctrl
   properties {draw Style=Square, label Font Size=5, normal Tint {colorUseRule=UseColor_Specified, specifiedColor=<rgba 0.0134 0 1 1>}}
   slot {
    class=CanvasPanelSlot
    properties {layoutData {alignment=<0.5 0.5>, anchors {maximum=<0.5 0.5>, minimum=<0.5 0.5>}, offsets=<ltrb 218 -25 15 15>}}
   }
  }
  ring_curl_l_ctrl {
   widgetClassPath=&EUW_Picker_Button
   bIsVariable=true
   parent=CanvasPanel_0
   widget=ring_curl_l_ctrl
   properties {draw Style=Square, label Font Size=5, normal Tint {colorUseRule=UseColor_Specified, specifiedColor=<rgba 0 1 0.2854 1>}}
   slot {
    class=CanvasPanelSlot
    properties {layoutData {alignment=<0.5 0.5>, anchors {maximum=<0.5 0.5>, minimum=<0.5 0.5>}, offsets=<ltrb 240 -105 15 15>}}
   }
  }
  ring_metacarpal_l_ctrl {
   widgetClassPath=&EUW_Picker_Button
   bIsVariable=true
   parent=CanvasPanel_0
   widget=ring_metacarpal_l_ctrl
   properties {draw Style=Square, label Font Size=5, normal Tint {colorUseRule=UseColor_Specified, specifiedColor=<rgba 0.0134 0 1 1>}}
   slot {
    class=CanvasPanelSlot
    properties {layoutData {alignment=<0.5 0.5>, anchors {maximum=<0.5 0.5>, minimum=<0.5 0.5>}, offsets=<ltrb 240 -85 15 15>}}
   }
  }
  ring_01_l_ctrl {
   widgetClassPath=&EUW_Picker_Button
   bIsVariable=true
   parent=CanvasPanel_0
   widget=ring_01_l_ctrl
   properties {draw Style=Square, label Font Size=5, normal Tint {colorUseRule=UseColor_Specified, specifiedColor=<rgba 0.0134 0 1 1>}}
   slot {
    class=CanvasPanelSlot
    properties {layoutData {alignment=<0.5 0.5>, anchors {maximum=<0.5 0.5>, minimum=<0.5 0.5>}, offsets=<ltrb 240 -65 15 15>}}
   }
  }
  ring_02_l_ctrl {
   widgetClassPath=&EUW_Picker_Button
   bIsVariable=true
   parent=CanvasPanel_0
   widget=ring_02_l_ctrl
   properties {draw Style=Square, label Font Size=5, normal Tint {colorUseRule=UseColor_Specified, specifiedColor=<rgba 0.0134 0 1 1>}}
   slot {
    class=CanvasPanelSlot
    properties {layoutData {alignment=<0.5 0.5>, anchors {maximum=<0.5 0.5>, minimum=<0.5 0.5>}, offsets=<ltrb 240 -45 15 15>}}
   }
  }
  ring_03_l_ctrl {
   widgetClassPath=&EUW_Picker_Button
   bIsVariable=true
   parent=CanvasPanel_0
   widget=ring_03_l_ctrl
   properties {draw Style=Square, label Font Size=5, normal Tint {colorUseRule=UseColor_Specified, specifiedColor=<rgba 0.0134 0 1 1>}}
   slot {
    class=CanvasPanelSlot
    properties {layoutData {alignment=<0.5 0.5>, anchors {maximum=<0.5 0.5>, minimum=<0.5 0.5>}, offsets=<ltrb 240 -25 15 15>}}
   }
  }
  pinky_curl_l_ctrl {
   widgetClassPath=&EUW_Picker_Button
   bIsVariable=true
   parent=CanvasPanel_0
   widget=pinky_curl_l_ctrl
   properties {draw Style=Square, label Font Size=5, normal Tint {colorUseRule=UseColor_Specified, specifiedColor=<rgba 0 1 0.2854 1>}}
   slot {
    class=CanvasPanelSlot
    properties {layoutData {alignment=<0.5 0.5>, anchors {maximum=<0.5 0.5>, minimum=<0.5 0.5>}, offsets=<ltrb 262 -105 15 15>}}
   }
  }
  pinky_metacarpal_l_ctrl {
   widgetClassPath=&EUW_Picker_Button
   bIsVariable=true
   parent=CanvasPanel_0
   widget=pinky_metacarpal_l_ctrl
   properties {draw Style=Square, label Font Size=5, normal Tint {colorUseRule=UseColor_Specified, specifiedColor=<rgba 0.0134 0 1 1>}}
   slot {
    class=CanvasPanelSlot
    properties {layoutData {alignment=<0.5 0.5>, anchors {maximum=<0.5 0.5>, minimum=<0.5 0.5>}, offsets=<ltrb 262 -85 15 15>}}
   }
  }
  pinky_01_l_ctrl {
   widgetClassPath=&EUW_Picker_Button
   bIsVariable=true
   parent=CanvasPanel_0
   widget=pinky_01_l_ctrl
   properties {draw Style=Square, label Font Size=5, normal Tint {colorUseRule=UseColor_Specified, specifiedColor=<rgba 0.0134 0 1 1>}}
   slot {
    class=CanvasPanelSlot
    properties {layoutData {alignment=<0.5 0.5>, anchors {maximum=<0.5 0.5>, minimum=<0.5 0.5>}, offsets=<ltrb 262 -65 15 15>}}
   }
  }
  pinky_02_l_ctrl {
   widgetClassPath=&EUW_Picker_Button
   bIsVariable=true
   parent=CanvasPanel_0
   widget=pinky_02_l_ctrl
   properties {draw Style=Square, label Font Size=5, normal Tint {colorUseRule=UseColor_Specified, specifiedColor=<rgba 0.0134 0 1 1>}}
   slot {
    class=CanvasPanelSlot
    properties {layoutData {alignment=<0.5 0.5>, anchors {maximum=<0.5 0.5>, minimum=<0.5 0.5>}, offsets=<ltrb 262 -45 15 15>}}
   }
  }
  pinky_03_l_ctrl {
   widgetClassPath=&EUW_Picker_Button
   bIsVariable=true
   parent=CanvasPanel_0
   widget=pinky_03_l_ctrl
   properties {draw Style=Square, label Font Size=5, normal Tint {colorUseRule=UseColor_Specified, specifiedColor=<rgba 0.0134 0 1 1>}}
   slot {
    class=CanvasPanelSlot
    properties {layoutData {alignment=<0.5 0.5>, anchors {maximum=<0.5 0.5>, minimum=<0.5 0.5>}, offsets=<ltrb 262 -25 15 15>}}
   }
  }
  pinky_03_r_ctrl {
   widgetClassPath=&EUW_Picker_Button
   bIsVariable=true
   parent=CanvasPanel_0
   widget=pinky_03_r_ctrl
   properties {draw Style=Square, label Font Size=5, normal Tint {colorUseRule=UseColor_Specified, specifiedColor=<rgba 1 0 0 1>}}
   slot {
    class=CanvasPanelSlot
    properties {layoutData {alignment=<0.5 0.5>, anchors {maximum=<0.5 0.5>, minimum=<0.5 0.5>}, offsets=<ltrb -257 -25 15 15>}}
   }
  }
  pinky_02_r_ctrl {
   widgetClassPath=&EUW_Picker_Button
   bIsVariable=true
   parent=CanvasPanel_0
   widget=pinky_02_r_ctrl
   properties {draw Style=Square, label Font Size=5, normal Tint {colorUseRule=UseColor_Specified, specifiedColor=<rgba 1 0 0 1>}}
   slot {
    class=CanvasPanelSlot
    properties {layoutData {alignment=<0.5 0.5>, anchors {maximum=<0.5 0.5>, minimum=<0.5 0.5>}, offsets=<ltrb -257 -45 15 15>}}
   }
  }
  pinky_01_r_ctrl {
   widgetClassPath=&EUW_Picker_Button
   bIsVariable=true
   parent=CanvasPanel_0
   widget=pinky_01_r_ctrl
   properties {draw Style=Square, label Font Size=5, normal Tint {colorUseRule=UseColor_Specified, specifiedColor=<rgba 1 0 0 1>}}
   slot {
    class=CanvasPanelSlot
    properties {layoutData {alignment=<0.5 0.5>, anchors {maximum=<0.5 0.5>, minimum=<0.5 0.5>}, offsets=<ltrb -257 -65 15 15>}}
   }
  }
  pinky_metacarpal_r_ctrl {
   widgetClassPath=&EUW_Picker_Button
   bIsVariable=true
   parent=CanvasPanel_0
   widget=pinky_metacarpal_r_ctrl
   properties {draw Style=Square, label Font Size=5, normal Tint {colorUseRule=UseColor_Specified, specifiedColor=<rgba 1 0 0 1>}}
   slot {
    class=CanvasPanelSlot
    properties {layoutData {alignment=<0.5 0.5>, anchors {maximum=<0.5 0.5>, minimum=<0.5 0.5>}, offsets=<ltrb -257 -85 15 15>}}
   }
  }
  pinky_curl_r_ctrl {
   widgetClassPath=&EUW_Picker_Button
   bIsVariable=true
   parent=CanvasPanel_0
   widget=pinky_curl_r_ctrl
   properties {draw Style=Square, label Font Size=5, normal Tint {colorUseRule=UseColor_Specified, specifiedColor=<rgba 0 1 0.2854 1>}}
   slot {
    class=CanvasPanelSlot
    properties {layoutData {alignment=<0.5 0.5>, anchors {maximum=<0.5 0.5>, minimum=<0.5 0.5>}, offsets=<ltrb -257 -105 15 15>}}
   }
  }
  ring_03_r_ctrl {
   widgetClassPath=&EUW_Picker_Button
   bIsVariable=true
   parent=CanvasPanel_0
   widget=ring_03_r_ctrl
   properties {draw Style=Square, label Font Size=5, normal Tint {colorUseRule=UseColor_Specified, specifiedColor=<rgba 1 0 0 1>}}
   slot {
    class=CanvasPanelSlot
    properties {layoutData {alignment=<0.5 0.5>, anchors {maximum=<0.5 0.5>, minimum=<0.5 0.5>}, offsets=<ltrb -235 -25 15 15>}}
   }
  }
  ring_02_r_ctrl {
   widgetClassPath=&EUW_Picker_Button
   bIsVariable=true
   parent=CanvasPanel_0
   widget=ring_02_r_ctrl
   properties {draw Style=Square, label Font Size=5, normal Tint {colorUseRule=UseColor_Specified, specifiedColor=<rgba 1 0 0 1>}}
   slot {
    class=CanvasPanelSlot
    properties {layoutData {alignment=<0.5 0.5>, anchors {maximum=<0.5 0.5>, minimum=<0.5 0.5>}, offsets=<ltrb -235 -45 15 15>}}
   }
  }
  ring_01_r_ctrl {
   widgetClassPath=&EUW_Picker_Button
   bIsVariable=true
   parent=CanvasPanel_0
   widget=ring_01_r_ctrl
   properties {draw Style=Square, label Font Size=5, normal Tint {colorUseRule=UseColor_Specified, specifiedColor=<rgba 1 0 0 1>}}
   slot {
    class=CanvasPanelSlot
    properties {layoutData {alignment=<0.5 0.5>, anchors {maximum=<0.5 0.5>, minimum=<0.5 0.5>}, offsets=<ltrb -235 -65 15 15>}}
   }
  }
  ring_metacarpal_r_ctrl {
   widgetClassPath=&EUW_Picker_Button
   bIsVariable=true
   parent=CanvasPanel_0
   widget=ring_metacarpal_r_ctrl
   properties {draw Style=Square, label Font Size=5, normal Tint {colorUseRule=UseColor_Specified, specifiedColor=<rgba 1 0 0 1>}}
   slot {
    class=CanvasPanelSlot
    properties {layoutData {alignment=<0.5 0.5>, anchors {maximum=<0.5 0.5>, minimum=<0.5 0.5>}, offsets=<ltrb -235 -85 15 15>}}
   }
  }
  ring_curl_r_ctrl {
   widgetClassPath=&EUW_Picker_Button
   bIsVariable=true
   parent=CanvasPanel_0
   widget=ring_curl_r_ctrl
   properties {draw Style=Square, label Font Size=5, normal Tint {colorUseRule=UseColor_Specified, specifiedColor=<rgba 0 1 0.2854 1>}}
   slot {
    class=CanvasPanelSlot
    properties {layoutData {alignment=<0.5 0.5>, anchors {maximum=<0.5 0.5>, minimum=<0.5 0.5>}, offsets=<ltrb -235 -105 15 15>}}
   }
  }
  middle_03_r_ctrl {
   widgetClassPath=&EUW_Picker_Button
   bIsVariable=true
   parent=CanvasPanel_0
   widget=middle_03_r_ctrl
   properties {draw Style=Square, label Font Size=5, normal Tint {colorUseRule=UseColor_Specified, specifiedColor=<rgba 1 0 0 1>}}
   slot {
    class=CanvasPanelSlot
    properties {layoutData {alignment=<0.5 0.5>, anchors {maximum=<0.5 0.5>, minimum=<0.5 0.5>}, offsets=<ltrb -213 -25 15 15>}}
   }
  }
  middle_02_r_ctrl {
   widgetClassPath=&EUW_Picker_Button
   bIsVariable=true
   parent=CanvasPanel_0
   widget=middle_02_r_ctrl
   properties {draw Style=Square, label Font Size=5, normal Tint {colorUseRule=UseColor_Specified, specifiedColor=<rgba 1 0 0 1>}}
   slot {
    class=CanvasPanelSlot
    properties {layoutData {alignment=<0.5 0.5>, anchors {maximum=<0.5 0.5>, minimum=<0.5 0.5>}, offsets=<ltrb -213 -45 15 15>}}
   }
  }
  middle_01_r_ctrl {
   widgetClassPath=&EUW_Picker_Button
   bIsVariable=true
   parent=CanvasPanel_0
   widget=middle_01_r_ctrl
   properties {draw Style=Square, label Font Size=5, normal Tint {colorUseRule=UseColor_Specified, specifiedColor=<rgba 1 0 0 1>}}
   slot {
    class=CanvasPanelSlot
    properties {layoutData {alignment=<0.5 0.5>, anchors {maximum=<0.5 0.5>, minimum=<0.5 0.5>}, offsets=<ltrb -213 -65 15 15>}}
   }
  }
  middle_metacarpal_r_ctrl {
   widgetClassPath=&EUW_Picker_Button
   bIsVariable=true
   parent=CanvasPanel_0
   widget=middle_metacarpal_r_ctrl
   properties {draw Style=Square, label Font Size=5, normal Tint {colorUseRule=UseColor_Specified, specifiedColor=<rgba 1 0 0 1>}}
   slot {
    class=CanvasPanelSlot
    properties {layoutData {alignment=<0.5 0.5>, anchors {maximum=<0.5 0.5>, minimum=<0.5 0.5>}, offsets=<ltrb -213 -85 15 15>}}
   }
  }
  middle_curl_r_ctrl {
   widgetClassPath=&EUW_Picker_Button
   bIsVariable=true
   parent=CanvasPanel_0
   widget=middle_curl_r_ctrl
   properties {draw Style=Square, label Font Size=5, normal Tint {colorUseRule=UseColor_Specified, specifiedColor=<rgba 0 1 0.2854 1>}}
   slot {
    class=CanvasPanelSlot
    properties {layoutData {alignment=<0.5 0.5>, anchors {maximum=<0.5 0.5>, minimum=<0.5 0.5>}, offsets=<ltrb -213 -105 15 15>}}
   }
  }
  index_03_r_ctrl {
   widgetClassPath=&EUW_Picker_Button
   bIsVariable=true
   parent=CanvasPanel_0
   widget=index_03_r_ctrl
   properties {draw Style=Square, label Font Size=5, normal Tint {colorUseRule=UseColor_Specified, specifiedColor=<rgba 1 0 0 1>}}
   slot {
    class=CanvasPanelSlot
    properties {layoutData {alignment=<0.5 0.5>, anchors {maximum=<0.5 0.5>, minimum=<0.5 0.5>}, offsets=<ltrb -191 -25 15 15>}}
   }
  }
  index_02_r_ctrl {
   widgetClassPath=&EUW_Picker_Button
   bIsVariable=true
   parent=CanvasPanel_0
   widget=index_02_r_ctrl
   properties {draw Style=Square, label Font Size=5, normal Tint {colorUseRule=UseColor_Specified, specifiedColor=<rgba 1 0 0 1>}}
   slot {
    class=CanvasPanelSlot
    properties {layoutData {alignment=<0.5 0.5>, anchors {maximum=<0.5 0.5>, minimum=<0.5 0.5>}, offsets=<ltrb -191 -45 15 15>}}
   }
  }
  index_01_r_ctrl {
   widgetClassPath=&EUW_Picker_Button
   bIsVariable=true
   parent=CanvasPanel_0
   widget=index_01_r_ctrl
   properties {draw Style=Square, label Font Size=5, normal Tint {colorUseRule=UseColor_Specified, specifiedColor=<rgba 1 0 0 1>}}
   slot {
    class=CanvasPanelSlot
    properties {layoutData {alignment=<0.5 0.5>, anchors {maximum=<0.5 0.5>, minimum=<0.5 0.5>}, offsets=<ltrb -191 -65 15 15>}}
   }
  }
  index_metacarpal_r_ctrl {
   widgetClassPath=&EUW_Picker_Button
   bIsVariable=true
   parent=CanvasPanel_0
   widget=index_metacarpal_r_ctrl
   properties {draw Style=Square, label Font Size=5, normal Tint {colorUseRule=UseColor_Specified, specifiedColor=<rgba 1 0 0 1>}}
   slot {
    class=CanvasPanelSlot
    properties {layoutData {alignment=<0.5 0.5>, anchors {maximum=<0.5 0.5>, minimum=<0.5 0.5>}, offsets=<ltrb -191 -85 15 15>}}
   }
  }
  index_curl_r_ctrl {
   widgetClassPath=&EUW_Picker_Button
   bIsVariable=true
   parent=CanvasPanel_0
   widget=index_curl_r_ctrl
   properties {draw Style=Square, label Font Size=5, normal Tint {colorUseRule=UseColor_Specified, specifiedColor=<rgba 0 1 0.2854 1>}}
   slot {
    class=CanvasPanelSlot
    properties {layoutData {alignment=<0.5 0.5>, anchors {maximum=<0.5 0.5>, minimum=<0.5 0.5>}, offsets=<ltrb -191 -105 15 15>}}
   }
  }
  thumb_03_r_ctrl {
   widgetClassPath=&EUW_Picker_Button
   bIsVariable=true
   parent=CanvasPanel_0
   widget=thumb_03_r_ctrl
   properties {draw Style=Square, label Font Size=5, normal Tint {colorUseRule=UseColor_Specified, specifiedColor=<rgba 1 0 0 1>}}
   slot {
    class=CanvasPanelSlot
    properties {layoutData {alignment=<0.5 0.5>, anchors {maximum=<0.5 0.5>, minimum=<0.5 0.5>}, offsets=<ltrb -153 -53 15 15>}}
   }
  }
  thumb_02_r_ctrl {
   widgetClassPath=&EUW_Picker_Button
   bIsVariable=true
   parent=CanvasPanel_0
   widget=thumb_02_r_ctrl
   properties {draw Style=Square, label Font Size=5, normal Tint {colorUseRule=UseColor_Specified, specifiedColor=<rgba 1 0 0 1>}}
   slot {
    class=CanvasPanelSlot
    properties {layoutData {alignment=<0.5 0.5>, anchors {maximum=<0.5 0.5>, minimum=<0.5 0.5>}, offsets=<ltrb -157 -73 15 15>}}
   }
  }
  thumb_01_r_ctrl {
   widgetClassPath=&EUW_Picker_Button
   bIsVariable=true
   parent=CanvasPanel_0
   widget=thumb_01_r_ctrl
   properties {draw Style=Square, label Font Size=5, normal Tint {colorUseRule=UseColor_Specified, specifiedColor=<rgba 1 0 0 1>}}
   slot {
    class=CanvasPanelSlot
    properties {layoutData {alignment=<0.5 0.5>, anchors {maximum=<0.5 0.5>, minimum=<0.5 0.5>}, offsets=<ltrb -161 -93 15 15>}}
   }
  }
  thumb_curl_r_ctrl {
   widgetClassPath=&EUW_Picker_Button
   bIsVariable=true
   parent=CanvasPanel_0
   widget=thumb_curl_r_ctrl
   properties {draw Style=Square, label Font Size=5, normal Tint {colorUseRule=UseColor_Specified, specifiedColor=<rgba 0 1 0.2854 1>}}
   slot {
    class=CanvasPanelSlot
    properties {layoutData {alignment=<0.5 0.5>, anchors {maximum=<0.5 0.5>, minimum=<0.5 0.5>}, offsets=<ltrb -169 -113 15 15>}}
   }
  }
  fingers_metacarpal_l_all {
   widgetClassPath=&EUW_Picker_Button
   bIsVariable=true
   parent=CanvasPanel_0
   widget=fingers_metacarpal_l_all
   properties {draw Style=Square, normal Tint {colorUseRule=UseColor_Specified, specifiedColor=<rgba 0 1 0.2874 1>}}
   slot {
    class=CanvasPanelSlot
    properties {layoutData {alignment=<0.5 0.5>, anchors {maximum=<0.5 0.5>, minimum=<0.5 0.5>}, offsets=<ltrb 284 -85 15 15>}}
   }
  }
  fingers_metacarpal_r_all {
   widgetClassPath=&EUW_Picker_Button
   bIsVariable=true
   parent=CanvasPanel_0
   widget=fingers_metacarpal_r_all
   properties {draw Style=Square, normal Tint {colorUseRule=UseColor_Specified, specifiedColor=<rgba 0 1 0.2874 1>}}
   slot {
    class=CanvasPanelSlot
    properties {layoutData {alignment=<0.5 0.5>, anchors {maximum=<0.5 0.5>, minimum=<0.5 0.5>}, offsets=<ltrb -277 -85 15 15>}}
   }
  }
  fingers_joint_1_l_all {
   widgetClassPath=&EUW_Picker_Button
   bIsVariable=true
   parent=CanvasPanel_0
   widget=fingers_joint_1_l_all
   properties {draw Style=Square, normal Tint {colorUseRule=UseColor_Specified, specifiedColor=<rgba 0 1 0.2874 1>}}
   slot {
    class=CanvasPanelSlot
    properties {layoutData {alignment=<0.5 0.5>, anchors {maximum=<0.5 0.5>, minimum=<0.5 0.5>}, offsets=<ltrb 284 -65 15 15>}}
   }
  }
  fingers_joint_1_r_all {
   widgetClassPath=&EUW_Picker_Button
   bIsVariable=true
   parent=CanvasPanel_0
   widget=fingers_joint_1_r_all
   properties {draw Style=Square, normal Tint {colorUseRule=UseColor_Specified, specifiedColor=<rgba 0 1 0.2874 1>}}
   slot {
    class=CanvasPanelSlot
    properties {layoutData {alignment=<0.5 0.5>, anchors {maximum=<0.5 0.5>, minimum=<0.5 0.5>}, offsets=<ltrb -277 -65 15 15>}}
   }
  }
  fingers_joint_2_l_all {
   widgetClassPath=&EUW_Picker_Button
   bIsVariable=true
   parent=CanvasPanel_0
   widget=fingers_joint_2_l_all
   properties {draw Style=Square, normal Tint {colorUseRule=UseColor_Specified, specifiedColor=<rgba 0 1 0.2874 1>}}
   slot {
    class=CanvasPanelSlot
    properties {layoutData {alignment=<0.5 0.5>, anchors {maximum=<0.5 0.5>, minimum=<0.5 0.5>}, offsets=<ltrb 284 -45 15 15>}}
   }
  }
  fingers_joint_2_r_all {
   widgetClassPath=&EUW_Picker_Button
   bIsVariable=true
   parent=CanvasPanel_0
   widget=fingers_joint_2_r_all
   properties {draw Style=Square, normal Tint {colorUseRule=UseColor_Specified, specifiedColor=<rgba 0 1 0.2874 1>}}
   slot {
    class=CanvasPanelSlot
    properties {layoutData {alignment=<0.5 0.5>, anchors {maximum=<0.5 0.5>, minimum=<0.5 0.5>}, offsets=<ltrb -277 -45 15 15>}}
   }
  }
  fingers_joint_3_l_all {
   widgetClassPath=&EUW_Picker_Button
   bIsVariable=true
   parent=CanvasPanel_0
   widget=fingers_joint_3_l_all
   properties {draw Style=Square, normal Tint {colorUseRule=UseColor_Specified, specifiedColor=<rgba 0 1 0.2874 1>}}
   slot {
    class=CanvasPanelSlot
    properties {layoutData {alignment=<0.5 0.5>, anchors {maximum=<0.5 0.5>, minimum=<0.5 0.5>}, offsets=<ltrb 284 -25 15 15>}}
   }
  }
  fingers_joint_3_r_all {
   widgetClassPath=&EUW_Picker_Button
   bIsVariable=true
   parent=CanvasPanel_0
   widget=fingers_joint_3_r_all
   properties {draw Style=Square, normal Tint {colorUseRule=UseColor_Specified, specifiedColor=<rgba 0 1 0.2874 1>}}
   slot {
    class=CanvasPanelSlot
    properties {layoutData {alignment=<0.5 0.5>, anchors {maximum=<0.5 0.5>, minimum=<0.5 0.5>}, offsets=<ltrb -277 -25 15 15>}}
   }
  }
  weapon_r_ctrl {
   widgetClassPath=&EUW_Picker_Button
   bIsVariable=true
   parent=CanvasPanel_0
   widget=weapon_r_ctrl
   properties {
    draw Style=Square
    label Font Size=8
    label Text=W1
    label Color {colorUseRule=UseColor_Specified, specifiedColor=<rgba 0 0 0 1>}
    normal Tint {colorUseRule=UseColor_Specified, specifiedColor=<rgba 1 0.33 0.5548 1>}
   }
   slot {
    class=CanvasPanelSlot
    properties {layoutData {alignment=<0.5 0.5>, anchors {maximum=<0.5 0.5>, minimum=<0.5 0.5>}, offsets=<ltrb -263 -358 20 20>}}
   }
  }
  WeapBones {
   widgetClassPath=&EUW_Picker_Button
   bIsVariable=true
   parent=CanvasPanel_0
   widget=WeapBones
   properties {
    draw Style=Square
    label Font Size=7
    label Text=All Wep Ctrls
    label Color {colorUseRule=UseColor_Specified, specifiedColor=<rgba 0 0 0 1>}
    normal Tint {colorUseRule=UseColor_Specified, specifiedColor=<rgba 0.3613 0.0296 1 1>}
   }
   slot {
    class=CanvasPanelSlot
    properties {layoutData {alignment=<0.5 0.5>, anchors {maximum=<0.5 0.5>, minimum=<0.5 0.5>}, offsets=<ltrb -249.0185 -403.5485 55.8848 17.984>}}
   }
  }
  hand_attach_r_ctrl {
   widgetClassPath=&EUW_Picker_Button
   bIsVariable=true
   parent=CanvasPanel_0
   widget=hand_attach_r_ctrl
   properties {
    draw Style=Square
    label Font Size=8
    label Text=W2
    label Color {colorUseRule=UseColor_Specified, specifiedColor=<rgba 0 0 0 1>}
    normal Tint {colorUseRule=UseColor_Specified, specifiedColor=<rgba 1 0.33 0.5548 1>}
   }
   slot {
    class=CanvasPanelSlot
    properties {layoutData {alignment=<0.5 0.5>, anchors {maximum=<0.5 0.5>, minimum=<0.5 0.5>}, offsets=<ltrb -237 -358 20 20>}}
   }
  }
  SizeBox_41 {
   widgetClassPath=&SizeBox
   bIsVariable=true
   parent=CanvasPanel_0
   widget=SizeBox_41
   slot {
    class=CanvasPanelSlot
    properties {layoutData {alignment=<0 0>, anchors {maximum=<0.5 0.5>, minimum=<0.5 0.5>}, offsets=<ltrb 759.0391 -536.5405 1758.8145 1059.8187>}}
   }
  }
  Border {
   widgetClassPath=&Border
   parent=SizeBox_41
   widget=Border
   properties {brushColor=<rgba 0.0452 0.0452 0.0452 1>, padding=<ltrb 0 0 0 0>}
   slot {class=SizeBoxSlot}
  }
  Holder_1 {
   widgetClassPath=&CanvasPanel
   bIsVariable=true
   parent=Border
   widget=Holder_1
   slot {class=BorderSlot, properties {padding=<ltrb 0 0 0 0>}}
  }
  NewSlot {
   widgetClassPath=&VerticalBox
   bIsVariable=true
   parent=Holder_1
   widget=NewSlot
   properties {flowDirectionPreference=Culture}
   slot {
    class=CanvasPanelSlot
    properties {bAutoSize=true, layoutData {alignment=<0.5 0.5>, anchors {maximum=<0.5 0.5>, minimum=<0.5 0.5>}, offsets=<ltrb -46.0967 -49.6695 100 30>}}
   }
  }
  TwistVisibility {
   widgetClassPath=&EditorUtilityCheckBox
   bIsVariable=true
   parent=CanvasPanel_0
   widget=TwistVisibility
   slot {
    class=CanvasPanelSlot
    properties {bAutoSize=true, layoutData {alignment=<0 0>, anchors {maximum=<0.5 0.5>, minimum=<0.5 0.5>}, offsets=<ltrb 215.0391 -464.5405 121.3547 40>}}
   }
  }
  TextBlock_20 {
   widgetClassPath=&TextBlock
   parent=TwistVisibility
   widget=TextBlock_20
   properties {
    text=Twist Visibility
    font {
     bForceMonospaced=false
     bMaterialIsStencil=false
     fontMaterial=null
     fontObject=&Roboto
     letterSpacing=0
     monospacedWidth=1
     size=11.25
     skewAmount=0
     typefaceFontName=Bold
     outlineSettings {
      bApplyOutlineToDropShadows=false
      bMiteredCorners=false
      bSeparateFillAlpha=false
      outlineMaterial=null
      outlineSize=0
      outlineColor=<rgba 0 0 0 1>
     }
    }
   }
   slot {class=PanelSlot}
  }
  AllFingers_l {
   widgetClassPath=&EUW_Picker_Button
   bIsVariable=true
   parent=CanvasPanel_0
   widget=AllFingers_l
   properties {
    label Font Size=7
    label Text=All L Fingers
    label Color {colorUseRule=UseColor_Specified, specifiedColor=<rgba 0 0 0 1>}
    normal Tint {colorUseRule=UseColor_Specified, specifiedColor=<rgba 0 0.4061 1 1>}
   }
   slot {
    class=CanvasPanelSlot
    properties {layoutData {alignment=<0 0>, anchors {maximum=<0.5 0.5>, minimum=<0.5 0.5>}, offsets=<ltrb 300 -135 60 25>}}
   }
  }
  AllFKLeg_l {
   widgetClassPath=&EUW_Picker_Button
   bIsVariable=true
   parent=CanvasPanel_0
   widget=AllFKLeg_l
   properties {
    label Font Size=8
    label Text=All FK L Leg
    label Color {colorUseRule=UseColor_Specified, specifiedColor=<rgba 0 0 0 1>}
    normal Tint {colorUseRule=UseColor_Specified, specifiedColor=<rgba 0 0.4061 1 1>}
   }
   slot {
    class=CanvasPanelSlot
    properties {layoutData {alignment=<0 0>, anchors {maximum=<0.5 0.5>, minimum=<0.5 0.5>}, offsets=<ltrb 250 50 70 30>}}
   }
  }
  AllIKLeg_l {
   widgetClassPath=&EUW_Picker_Button
   bIsVariable=true
   parent=CanvasPanel_0
   widget=AllIKLeg_l
   properties {
    label Font Size=8
    label Text=All L IK L Leg
    label Color {colorUseRule=UseColor_Specified, specifiedColor=<rgba 0 0 0 1>}
    normal Tint {colorUseRule=UseColor_Specified, specifiedColor=<rgba 0.2195 0.5089 1 1>}
   }
   slot {
    class=CanvasPanelSlot
    properties {layoutData {alignment=<0 0>, anchors {maximum=<0.5 0.5>, minimum=<0.5 0.5>}, offsets=<ltrb 250 107 70 30>}}
   }
  }
  AllIKLegExtra_l {
   widgetClassPath=&EUW_Picker_Button
   bIsVariable=true
   parent=CanvasPanel_0
   widget=AllIKLegExtra_l
   properties {
    label Font Size=8
    label Text=All IK Feet Extras
    label Color {colorUseRule=UseColor_Specified, specifiedColor=<rgba 0 0 0 1>}
    normal Tint {colorUseRule=UseColor_Specified, specifiedColor=<rgba 0.2195 0.5089 1 1>}
   }
   slot {
    class=CanvasPanelSlot
    properties {layoutData {alignment=<0 0>, anchors {maximum=<0.5 0.5>, minimum=<0.5 0.5>}, offsets=<ltrb 250 145 70 30>}}
   }
  }
  ArmsFingers_l {
   widgetClassPath=&EUW_Picker_Button
   bIsVariable=true
   parent=CanvasPanel_0
   widget=ArmsFingers_l
   properties {
    label Font Size=8
    label Text=All L Arm + Fingers
    label Color {colorUseRule=UseColor_Specified, specifiedColor=<rgba 0 0 0 1>}
    normal Tint {colorUseRule=UseColor_Specified, specifiedColor=<rgba 0 0.4061 1 1>}
   }
   slot {
    class=CanvasPanelSlot
    properties {layoutData {alignment=<0 0>, anchors {maximum=<0.5 0.5>, minimum=<0.5 0.5>}, offsets=<ltrb 250 -220 70 30>}}
   }
  }
  AllArm_l {
   widgetClassPath=&EUW_Picker_Button
   bIsVariable=true
   parent=CanvasPanel_0
   widget=AllArm_l
   properties {
    label Font Size=8
    label Text=All L Arm
    label Color {colorUseRule=UseColor_Specified, specifiedColor=<rgba 0 0 0 1>}
    normal Tint {colorUseRule=UseColor_Specified, specifiedColor=<rgba 0 0.4061 1 1>}
   }
   slot {
    class=CanvasPanelSlot
    properties {layoutData {alignment=<0 0>, anchors {maximum=<0.5 0.5>, minimum=<0.5 0.5>}, offsets=<ltrb 280 -325 50 25>}}
   }
  }
  AllArm_r {
   widgetClassPath=&EUW_Picker_Button
   bIsVariable=true
   parent=CanvasPanel_0
   widget=AllArm_r
   properties {
    label Font Size=8
    label Text=All R Arm
    label Color {colorUseRule=UseColor_Specified, specifiedColor=<rgba 0 0 0 1>}
    normal Tint {colorUseRule=UseColor_Specified, specifiedColor=<rgba 1 0.08 0.0982 1>}
   }
   slot {
    class=CanvasPanelSlot
    properties {layoutData {alignment=<0 0>, anchors {maximum=<0.5 0.5>, minimum=<0.5 0.5>}, offsets=<ltrb -330 -325 50 25>}}
   }
  }
  ArmsFingers_r {
   widgetClassPath=&EUW_Picker_Button
   bIsVariable=true
   parent=CanvasPanel_0
   widget=ArmsFingers_r
   properties {
    label Font Size=8
    label Text=All R Arms + Fingers
    label Color {colorUseRule=UseColor_Specified, specifiedColor=<rgba 0 0 0 1>}
    normal Tint {colorUseRule=UseColor_Specified, specifiedColor=<rgba 1 0.08 0.0982 1>}
   }
   slot {
    class=CanvasPanelSlot
    properties {layoutData {alignment=<0 0>, anchors {maximum=<0.5 0.5>, minimum=<0.5 0.5>}, offsets=<ltrb -330 -216.5405 70 30>}}
   }
  }
  allFingers_r {
   widgetClassPath=&EUW_Picker_Button
   bIsVariable=true
   parent=CanvasPanel_0
   widget=allFingers_r
   properties {
    label Font Size=7
    label Text=All R Fingers
    label Color {colorUseRule=UseColor_Specified, specifiedColor=<rgba 0 0 0 1>}
    normal Tint {colorUseRule=UseColor_Specified, specifiedColor=<rgba 1 0.08 0.0982 1>}
   }
   slot {
    class=CanvasPanelSlot
    properties {layoutData {alignment=<0 0>, anchors {maximum=<0.5 0.5>, minimum=<0.5 0.5>}, offsets=<ltrb -355 -135 60 25>}}
   }
  }
  4Fingers_r {
   widgetClassPath=&EUW_Picker_Button
   bIsVariable=true
   parent=CanvasPanel_0
   widget=4Fingers_r
   properties {
    label Font Size=8
    label Text=4 R Fingers
    label Color {colorUseRule=UseColor_Specified, specifiedColor=<rgba 0 0 0 1>}
    normal Tint {colorUseRule=UseColor_Specified, specifiedColor=<rgba 1 0.33 0.5548 1>}
   }
   slot {
    class=CanvasPanelSlot
    properties {layoutData {alignment=<0 0>, anchors {maximum=<0.5 0.5>, minimum=<0.5 0.5>}, offsets=<ltrb -355 -105 60 25>}}
   }
  }
  AllFKLeg_r {
   widgetClassPath=&EUW_Picker_Button
   bIsVariable=true
   parent=CanvasPanel_0
   widget=AllFKLeg_r
   properties {
    label Font Size=8
    label Text=All FK R Leg
    label Color {colorUseRule=UseColor_Specified, specifiedColor=<rgba 0 0 0 1>}
    normal Tint {colorUseRule=UseColor_Specified, specifiedColor=<rgba 1 0.08 0.0982 1>}
   }
   slot {
    class=CanvasPanelSlot
    properties {layoutData {alignment=<0 0>, anchors {maximum=<0.5 0.5>, minimum=<0.5 0.5>}, offsets=<ltrb -330 50 70 30>}}
   }
  }
  AllIKLeg_r {
   widgetClassPath=&EUW_Picker_Button
   bIsVariable=true
   parent=CanvasPanel_0
   widget=AllIKLeg_r
   properties {
    label Font Size=8
    label Text=All IK R Leg
    label Color {colorUseRule=UseColor_Specified, specifiedColor=<rgba 0 0 0 1>}
    normal Tint {colorUseRule=UseColor_Specified, specifiedColor=<rgba 1 0.33 0.5548 1>}
   }
   slot {
    class=CanvasPanelSlot
    properties {layoutData {alignment=<0 0>, anchors {maximum=<0.5 0.5>, minimum=<0.5 0.5>}, offsets=<ltrb -330 107 70 30>}}
   }
  }
  AllIKLegExtras_r {
   widgetClassPath=&EUW_Picker_Button
   bIsVariable=true
   parent=CanvasPanel_0
   widget=AllIKLegExtras_r
   properties {
    label Font Size=8
    label Text=All IK Feet Extras
    label Color {colorUseRule=UseColor_Specified, specifiedColor=<rgba 0 0 0 1>}
    normal Tint {colorUseRule=UseColor_Specified, specifiedColor=<rgba 1 0.33 0.5548 1>}
   }
   slot {
    class=CanvasPanelSlot
    properties {layoutData {alignment=<0 0>, anchors {maximum=<0.5 0.5>, minimum=<0.5 0.5>}, offsets=<ltrb -330 145 70 30>}}
   }
  }
  4Fingers_l {
   widgetClassPath=&EUW_Picker_Button
   bIsVariable=true
   parent=CanvasPanel_0
   widget=4Fingers_l
   properties {
    label Font Size=8
    label Text=4 L Fingers
    label Color {colorUseRule=UseColor_Specified, specifiedColor=<rgba 0 0 0 1>}
    normal Tint {colorUseRule=UseColor_Specified, specifiedColor=<rgba 0.2195 0.5089 1 1>}
   }
   slot {
    class=CanvasPanelSlot
    properties {layoutData {alignment=<0 0>, anchors {maximum=<0.5 0.5>, minimum=<0.5 0.5>}, offsets=<ltrb 300 -105 60 25>}}
   }
  }
  AllArms {
   widgetClassPath=&EUW_Picker_Button
   bIsVariable=true
   parent=CanvasPanel_0
   widget=AllArms
   properties {
    label Text=All Arms
    label Color {colorUseRule=UseColor_Specified, specifiedColor=<rgba 0 0 0 1>}
    normal Tint {colorUseRule=UseColor_Specified, specifiedColor=<rgba 0.3618 0.03 1 1>}
   }
   slot {
    class=CanvasPanelSlot
    properties {layoutData {alignment=<0 0>, anchors {maximum=<0.5 0.5>, minimum=<0.5 0.5>}, offsets=<ltrb -459 -380 70 25>}}
   }
  }
  AllTwist {
   widgetClassPath=&EUW_Picker_Button
   bIsVariable=true
   parent=CanvasPanel_0
   widget=AllTwist
   properties {
    label Text=All Twists
    label Color {colorUseRule=UseColor_Specified, specifiedColor=<rgba 0 0 0 1>}
    normal Tint {colorUseRule=UseColor_Specified, specifiedColor=<rgba 0.3618 0.03 1 1>}
   }
   slot {
    class=CanvasPanelSlot
    properties {layoutData {alignment=<0 0>, anchors {maximum=<0.5 0.5>, minimum=<0.5 0.5>}, offsets=<ltrb 235 -440 70 25>}}
   }
  }
  AllClavicles {
   widgetClassPath=&EUW_Picker_Button
   bIsVariable=true
   parent=CanvasPanel_0
   widget=AllClavicles
   properties {
    label Font Size=9
    label Text=All Clavicles
    label Color {colorUseRule=UseColor_Specified, specifiedColor=<rgba 0 0 0 1>}
    normal Tint {colorUseRule=UseColor_Specified, specifiedColor=<rgba 0.3618 0.03 1 1>}
   }
   slot {
    class=CanvasPanelSlot
    properties {layoutData {alignment=<0 0>, anchors {maximum=<0.5 0.5>, minimum=<0.5 0.5>}, offsets=<ltrb -459 -350 70 25>}}
   }
  }
  AllFingers {
   widgetClassPath=&EUW_Picker_Button
   bIsVariable=true
   parent=CanvasPanel_0
   widget=AllFingers
   properties {
    label Text=All Fingers
    label Color {colorUseRule=UseColor_Specified, specifiedColor=<rgba 0 0 0 1>}
    normal Tint {colorUseRule=UseColor_Specified, specifiedColor=<rgba 0.3618 0.03 1 1>}
   }
   slot {
    class=CanvasPanelSlot
    properties {layoutData {alignment=<0 0>, anchors {maximum=<0.5 0.5>, minimum=<0.5 0.5>}, offsets=<ltrb -459 -320 70 25>}}
   }
  }
  IKMove {
   widgetClassPath=&EUW_Picker_Button
   bIsVariable=true
   parent=CanvasPanel_0
   widget=IKMove
   properties {
    label Text=IK Move
    label Color {colorUseRule=UseColor_Specified, specifiedColor=<rgba 0 0 0 1>}
    normal Tint {colorUseRule=UseColor_Specified, specifiedColor=<rgba 0.9911 0.4852 0.0452 1>}
   }
   slot {
    class=CanvasPanelSlot
    properties {layoutData {alignment=<0 0>, anchors {maximum=<0.5 0.5>, minimum=<0.5 0.5>}, offsets=<ltrb -459 -130 70 30>}}
   }
  }
  FKArms {
   widgetClassPath=&EUW_Picker_Button
   bIsVariable=true
   parent=CanvasPanel_0
   widget=FKArms
   properties {
    label Text=FK Arms
    label Color {colorUseRule=UseColor_Specified, specifiedColor=<rgba 0 0 0 1>}
    normal Tint {colorUseRule=UseColor_Specified, specifiedColor=<rgba 0.7126 0.16 1 1>}
   }
   slot {
    class=CanvasPanelSlot
    properties {layoutData {alignment=<0 0>, anchors {maximum=<0.5 0.5>, minimum=<0.5 0.5>}, offsets=<ltrb -497 -290 70 25>}}
   }
  }
  AllLegs {
   widgetClassPath=&EUW_Picker_Button
   bIsVariable=true
   parent=CanvasPanel_0
   widget=AllLegs
   properties {
    label Text=All Legs
    label Color {colorUseRule=UseColor_Specified, specifiedColor=<rgba 0 0 0 1>}
    normal Tint {colorUseRule=UseColor_Specified, specifiedColor=<rgba 0.3618 0.03 1 1>}
   }
   slot {
    class=CanvasPanelSlot
    properties {layoutData {alignment=<0 0>, anchors {maximum=<0.5 0.5>, minimum=<0.5 0.5>}, offsets=<ltrb -459 -60 70 25>}}
   }
  }
  AllFKLegs {
   widgetClassPath=&EUW_Picker_Button
   bIsVariable=true
   parent=CanvasPanel_0
   widget=AllFKLegs
   properties {
    label Text=All FK Legs
    label Color {colorUseRule=UseColor_Specified, specifiedColor=<rgba 0 0 0 1>}
    normal Tint {colorUseRule=UseColor_Specified, specifiedColor=<rgba 0.7126 0.16 1 1>}
   }
   slot {
    class=CanvasPanelSlot
    properties {layoutData {alignment=<0 0>, anchors {maximum=<0.5 0.5>, minimum=<0.5 0.5>}, offsets=<ltrb -497 -30 70 25>}}
   }
  }
  FKThighs {
   widgetClassPath=&EUW_Picker_Button
   bIsVariable=true
   parent=CanvasPanel_0
   widget=FKThighs
   properties {
    label Text=FK Thighs
    label Color {colorUseRule=UseColor_Specified, specifiedColor=<rgba 0 0 0 1>}
    normal Tint {colorUseRule=UseColor_Specified, specifiedColor=<rgba 0.8473 0.532 1 1>}
   }
   slot {
    class=CanvasPanelSlot
    properties {layoutData {alignment=<0 0>, anchors {maximum=<0.5 0.5>, minimum=<0.5 0.5>}, offsets=<ltrb -497 0 70 25>}}
   }
  }
  FKKnees {
   widgetClassPath=&EUW_Picker_Button
   bIsVariable=true
   parent=CanvasPanel_0
   widget=FKKnees
   properties {
    label Text=FK Knees
    label Color {colorUseRule=UseColor_Specified, specifiedColor=<rgba 0 0 0 1>}
    normal Tint {colorUseRule=UseColor_Specified, specifiedColor=<rgba 0.8473 0.532 1 1>}
   }
   slot {
    class=CanvasPanelSlot
    properties {layoutData {alignment=<0 0>, anchors {maximum=<0.5 0.5>, minimum=<0.5 0.5>}, offsets=<ltrb -497 30 70 25>}}
   }
  }
  FKFeets {
   widgetClassPath=&EUW_Picker_Button
   bIsVariable=true
   parent=CanvasPanel_0
   widget=FKFeets
   properties {
    label Text=FK Feets
    label Color {colorUseRule=UseColor_Specified, specifiedColor=<rgba 0 0 0 1>}
    normal Tint {colorUseRule=UseColor_Specified, specifiedColor=<rgba 0.8473 0.532 1 1>}
   }
   slot {
    class=CanvasPanelSlot
    properties {layoutData {alignment=<0 0>, anchors {maximum=<0.5 0.5>, minimum=<0.5 0.5>}, offsets=<ltrb -497 60 70 25>}}
   }
  }
  FKToes {
   widgetClassPath=&EUW_Picker_Button
   bIsVariable=true
   parent=CanvasPanel_0
   widget=FKToes
   properties {
    label Text=FK Toes
    label Color {colorUseRule=UseColor_Specified, specifiedColor=<rgba 0 0 0 1>}
    normal Tint {colorUseRule=UseColor_Specified, specifiedColor=<rgba 0.8473 0.532 1 1>}
   }
   slot {
    class=CanvasPanelSlot
    properties {layoutData {alignment=<0 0>, anchors {maximum=<0.5 0.5>, minimum=<0.5 0.5>}, offsets=<ltrb -497 90 70 25>}}
   }
  }
  IKToes {
   widgetClassPath=&EUW_Picker_Button
   bIsVariable=true
   parent=CanvasPanel_0
   widget=IKToes
   properties {
    label Text=IK Toes
    label Color {colorUseRule=UseColor_Specified, specifiedColor=<rgba 0 0 0 1>}
    normal Tint {colorUseRule=UseColor_Specified, specifiedColor=<rgba 1 0.37 0.9544 1>}
   }
   slot {
    class=CanvasPanelSlot
    properties {layoutData {alignment=<0 0>, anchors {maximum=<0.5 0.5>, minimum=<0.5 0.5>}, offsets=<ltrb -421 60 70 25>}}
   }
  }
  IKFeet {
   widgetClassPath=&EUW_Picker_Button
   bIsVariable=true
   parent=CanvasPanel_0
   widget=IKFeet
   properties {
    label Text=IK Feets
    label Color {colorUseRule=UseColor_Specified, specifiedColor=<rgba 0 0 0 1>}
    normal Tint {colorUseRule=UseColor_Specified, specifiedColor=<rgba 1 0.37 0.9544 1>}
   }
   slot {
    class=CanvasPanelSlot
    properties {layoutData {alignment=<0 0>, anchors {maximum=<0.5 0.5>, minimum=<0.5 0.5>}, offsets=<ltrb -421 30 70 25>}}
   }
  }
  IKKnees {
   widgetClassPath=&EUW_Picker_Button
   bIsVariable=true
   parent=CanvasPanel_0
   widget=IKKnees
   properties {
    label Text=IK Knees
    label Color {colorUseRule=UseColor_Specified, specifiedColor=<rgba 0 0 0 1>}
    normal Tint {colorUseRule=UseColor_Specified, specifiedColor=<rgba 1 0.37 0.9544 1>}
   }
   slot {
    class=CanvasPanelSlot
    properties {layoutData {alignment=<0 0>, anchors {maximum=<0.5 0.5>, minimum=<0.5 0.5>}, offsets=<ltrb -421 0 70 25>}}
   }
  }
  IKFeetExtras {
   widgetClassPath=&EUW_Picker_Button
   bIsVariable=true
   parent=CanvasPanel_0
   widget=IKFeetExtras
   properties {
    label Font Size=8
    label Text=IK Feet Extras
    label Color {colorUseRule=UseColor_Specified, specifiedColor=<rgba 0 0 0 1>}
    normal Tint {colorUseRule=UseColor_Specified, specifiedColor=<rgba 1 0.37 0.9544 1>}
   }
   slot {
    class=CanvasPanelSlot
    properties {layoutData {alignment=<0 0>, anchors {maximum=<0.5 0.5>, minimum=<0.5 0.5>}, offsets=<ltrb -421 90 70 25>}}
   }
  }
  AllIKLegs {
   widgetClassPath=&EUW_Picker_Button
   bIsVariable=true
   parent=CanvasPanel_0
   widget=AllIKLegs
   properties {
    label Text=All IK Legs
    label Color {colorUseRule=UseColor_Specified, specifiedColor=<rgba 0 0 0 1>}
    normal Tint {colorUseRule=UseColor_Specified, specifiedColor=<rgba 1 0.11 0.7072 1>}
   }
   slot {
    class=CanvasPanelSlot
    properties {layoutData {alignment=<0 0>, anchors {maximum=<0.5 0.5>, minimum=<0.5 0.5>}, offsets=<ltrb -421 -30 70 25>}}
   }
  }
  FKBiceps {
   widgetClassPath=&EUW_Picker_Button
   bIsVariable=true
   parent=CanvasPanel_0
   widget=FKBiceps
   properties {
    label Text=FK Biceps
    label Color {colorUseRule=UseColor_Specified, specifiedColor=<rgba 0 0 0 1>}
    normal Tint {colorUseRule=UseColor_Specified, specifiedColor=<rgba 0.8473 0.532 1 1>}
   }
   slot {
    class=CanvasPanelSlot
    properties {layoutData {alignment=<0 0>, anchors {maximum=<0.5 0.5>, minimum=<0.5 0.5>}, offsets=<ltrb -497 -260 70 25>}}
   }
  }
  FKElbows {
   widgetClassPath=&EUW_Picker_Button
   bIsVariable=true
   parent=CanvasPanel_0
   widget=FKElbows
   properties {
    label Text=FK Elbows
    label Color {colorUseRule=UseColor_Specified, specifiedColor=<rgba 0 0 0 1>}
    normal Tint {colorUseRule=UseColor_Specified, specifiedColor=<rgba 0.8473 0.532 1 1>}
   }
   slot {
    class=CanvasPanelSlot
    properties {layoutData {alignment=<0 0>, anchors {maximum=<0.5 0.5>, minimum=<0.5 0.5>}, offsets=<ltrb -497 -230 70 25>}}
   }
  }
  IKHands {
   widgetClassPath=&EUW_Picker_Button
   bIsVariable=true
   parent=CanvasPanel_0
   widget=IKHands
   properties {
    label Text=IK Hands
    label Color {colorUseRule=UseColor_Specified, specifiedColor=<rgba 0 0 0 1>}
    normal Tint {colorUseRule=UseColor_Specified, specifiedColor=<rgba 1 0.37 0.9544 1>}
   }
   slot {
    class=CanvasPanelSlot
    properties {layoutData {alignment=<0 0>, anchors {maximum=<0.5 0.5>, minimum=<0.5 0.5>}, offsets=<ltrb -421 -200 70 25>}}
   }
  }
  FKHands {
   widgetClassPath=&EUW_Picker_Button
   bIsVariable=true
   parent=CanvasPanel_0
   widget=FKHands
   properties {
    label Text=FK Hands
    label Color {colorUseRule=UseColor_Specified, specifiedColor=<rgba 0 0 0 1>}
    normal Tint {colorUseRule=UseColor_Specified, specifiedColor=<rgba 0.8473 0.532 1 1>}
   }
   slot {
    class=CanvasPanelSlot
    properties {layoutData {alignment=<0 0>, anchors {maximum=<0.5 0.5>, minimum=<0.5 0.5>}, offsets=<ltrb -497 -200 70 25>}}
   }
  }
  IKElbows {
   widgetClassPath=&EUW_Picker_Button
   bIsVariable=true
   parent=CanvasPanel_0
   widget=IKElbows
   properties {
    label Text=IK Elbows
    label Color {colorUseRule=UseColor_Specified, specifiedColor=<rgba 0 0 0 1>}
    normal Tint {colorUseRule=UseColor_Specified, specifiedColor=<rgba 1 0.37 0.9544 1>}
   }
   slot {
    class=CanvasPanelSlot
    properties {layoutData {alignment=<0 0>, anchors {maximum=<0.5 0.5>, minimum=<0.5 0.5>}, offsets=<ltrb -421 -230 70 25>}}
   }
  }
  IKArms {
   widgetClassPath=&EUW_Picker_Button
   bIsVariable=true
   parent=CanvasPanel_0
   widget=IKArms
   properties {
    label Text=IK Arms
    label Color {colorUseRule=UseColor_Specified, specifiedColor=<rgba 0 0 0 1>}
    normal Tint {colorUseRule=UseColor_Specified, specifiedColor=<rgba 1 0.11 0.7072 1>}
   }
   slot {
    class=CanvasPanelSlot
    properties {layoutData {alignment=<0 0>, anchors {maximum=<0.5 0.5>, minimum=<0.5 0.5>}, offsets=<ltrb -421 -290 70 25>}}
   }
  }
  ChangeBGbutton {
   widgetClassPath=&EUW_Picker_Button
   bIsVariable=true
   parent=CanvasPanel_0
   widget=ChangeBGbutton
   properties {label Text=Change BG, normal Tint {colorUseRule=UseColor_Specified, specifiedColor=<rgba 0.0012 0.2307 0.0137 1>}}
   slot {
    class=CanvasPanelSlot
    properties {layoutData {alignment=<0 0>, anchors {maximum=<0.5 0.5>, minimum=<0.5 0.5>}, offsets=<ltrb 270 -502 70.6373 25.986>}}
   }
  }
  BodyPickerButton {
   widgetClassPath=&EUW_Picker_Button
   bIsVariable=true
   parent=CanvasPanel_0
   widget=BodyPickerButton
   properties {label Font Size=25, label Text=Body Picker, normal Tint {colorUseRule=UseColor_Specified, specifiedColor=<rgba 0.0013 0.2292 0.0139 1>}}
   slot {
    class=CanvasPanelSlot
    properties {layoutData {alignment=<0 0>, anchors {maximum=<0.5 0.5>, minimum=<0.5 0.5>}, offsets=<ltrb 475.0391 -252.5405 200.7674 38.6653>}}
   }
  }
  BakeIKFKButton {
   widgetClassPath=&EUW_Picker_Button
   bIsVariable=true
   parent=CanvasPanel_0
   widget=BakeIKFKButton
   properties {label Font Size=15, label Text=Bake IKFK, normal Tint {colorUseRule=UseColor_Specified, specifiedColor=<rgba 0.15 0.0704 0.0497 1>}}
   slot {
    class=CanvasPanelSlot
    properties {layoutData {alignment=<0 0>, anchors {maximum=<0.5 0.5>, minimum=<0.5 0.5>}, offsets=<ltrb -268.9609 279.4595 117.7753 36.4813>}}
   }
  }
  DeleteTwistKeysButton {
   widgetClassPath=&EUW_Picker_Button
   bIsVariable=true
   parent=CanvasPanel_0
   widget=DeleteTwistKeysButton
   properties {label Font Size=15, label Text=Delete Twist Keys, normal Tint {colorUseRule=UseColor_Specified, specifiedColor=<rgba 0.3231 0.0343 0 1>}}
   slot {
    class=CanvasPanelSlot
    properties {layoutData {alignment=<0 0>, anchors {maximum=<0.5 0.5>, minimum=<0.5 0.5>}, offsets=<ltrb -144.9609 279.4595 173.5454 36.4813>}}
   }
  }
  3LFacePickerButton {
   widgetClassPath=&EUW_Picker_Button
   bIsVariable=true
   parent=CanvasPanel_0
   widget=3LFacePickerButton
   properties {label Font Size=25, label Text=3L Face Picker, normal Tint {colorUseRule=UseColor_Specified, specifiedColor=<rgba 0.0013 0.2292 0.0139 1>}}
   slot {
    class=CanvasPanelSlot
    properties {layoutData {alignment=<0 0>, anchors {maximum=<0.5 0.5>, minimum=<0.5 0.5>}, offsets=<ltrb 439.0391 -200.5405 261.8285 38.6653>}}
   }
  }
  3LOnFacePickerButton {
   widgetClassPath=&EUW_Picker_Button
   bIsVariable=true
   parent=CanvasPanel_0
   widget=3LOnFacePickerButton
   properties {label Font Size=25, label Text=3L OnFace Picker, normal Tint {colorUseRule=UseColor_Specified, specifiedColor=<rgba 0.0013 0.2292 0.0139 1>}}
   slot {
    class=CanvasPanelSlot
    properties {layoutData {alignment=<0 0>, anchors {maximum=<0.5 0.5>, minimum=<0.5 0.5>}, offsets=<ltrb 427.0391 -152.5405 283.8505 38.6653>}}
   }
  }
  LegacyFacePickerButton {
   widgetClassPath=&EUW_Picker_Button
   bIsVariable=true
   parent=CanvasPanel_0
   widget=LegacyFacePickerButton
   properties {
    label Font Size=25
    label Text=Legacy Face Picker
    normal Tint {colorUseRule=UseColor_Specified, specifiedColor=<rgba 0.0013 0.2292 0.0139 1>}
   }
   slot {
    class=CanvasPanelSlot
    properties {layoutData {alignment=<0 0>, anchors {maximum=<0.5 0.5>, minimum=<0.5 0.5>}, offsets=<ltrb 419.0391 -100.5405 301.8686 38.6653>}}
   }
  }
  FaceboardcameraButton {
   widgetClassPath=&EUW_Picker_Button
   bIsVariable=true
   parent=CanvasPanel_0
   widget=FaceboardcameraButton
   properties {label Font Size=25, label Text=Faceboard camera, normal Tint {colorUseRule=UseColor_Specified, specifiedColor=<rgba 0.0004 0.0729 0.0044 1>}}
   slot {
    class=CanvasPanelSlot
    properties {layoutData {alignment=<0 0>, anchors {maximum=<0.5 0.5>, minimum=<0.5 0.5>}, offsets=<ltrb 419.0391 -44.5405 301.8686 38.6653>}}
   }
  }
 ]
}
