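# Shared Control Rig for Fortnite biped player characters, providing animator controls for limbs, feet and fingers.
# Setup functions place FK, IK and pole vector controls from bone transforms, mirroring negative-side limbs and adding space parents.
# Limb solves switch between FK point-at and two-bone IK with stretch and softness, matching the inactive controls.
# Finger logic spreads proxy and interpolated control edits across fingers, requesting auto-key inside undo brackets.
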ControlRigBlueprint /Game/Characters/Player/Common/SharedRigs/Fortnite_Biped_CtrlRig
graphs [
 ContainedGraph {nodes [
   n1 = RigVMFunctionEntryNode()
   SequenceExecution_1 = &RigVMFunction_Sequence(ExecuteContext=@n1.ExecuteContext)
   SequenceExecution_1_1 = &RigVMFunction_Sequence(ExecuteContext=@SequenceExecution_1.B)
   SequenceExecution_1_2 = &RigVMFunction_Sequence(ExecuteContext=@SequenceExecution_1_1.B)
   RigVMFunctionReturnNode(A=@SequenceExecution_1.A, B=@SequenceExecution_1_1.A, C=@SequenceExecution_1_2.A, D=@SequenceExecution_1_2.B)
  ]}
 ContainedGraph {nodes [
   n1 = RigVMFunctionEntryNode()
   SequenceExecution_3 = &RigVMFunction_Sequence(ExecuteContext=@n1.ExecuteContext)
   SequenceExecution_3_1 = &RigVMFunction_Sequence(ExecuteContext=@SequenceExecution_3.B)
   SequenceExecution_3_2 = &RigVMFunction_Sequence(ExecuteContext=@SequenceExecution_3_1.B)
   SequenceExecution_3_3 = &RigVMFunction_Sequence(ExecuteContext=@SequenceExecution_3_2.B)
   RigVMFunctionReturnNode(A=@SequenceExecution_3.A, B=@SequenceExecution_3_1.A, C=@SequenceExecution_3_2.A, D=@SequenceExecution_3_3.A, E=@SequenceExecution_3_3.B)
  ]}
 ContainedGraph {nodes [
   n1 = RigVMFunctionEntryNode()
   SequenceExecution = &RigVMFunction_Sequence(ExecuteContext=@n1.ExecuteContext)
   SequenceExecution_1 = &RigVMFunction_Sequence(ExecuteContext=@SequenceExecution.B)
   SequenceExecution_2 = &RigVMFunction_Sequence(ExecuteContext=@SequenceExecution_1.B)
   SequenceExecution_3 = &RigVMFunction_Sequence(ExecuteContext=@SequenceExecution_2.B)
   RigVMFunctionReturnNode(A=@SequenceExecution.A, B=@SequenceExecution_1.A, C=@SequenceExecution_2.A, D=@SequenceExecution_3.A, E=@SequenceExecution_3.B)
  ]}
 ContainedGraph {nodes [
   n1 = RigVMFunctionEntryNode()
   n2 = /Script/RigVM.RigVMFunction_MathQuaternionToRotator(Value=@n1.Input.Rotation) {is=struct RigVMUnitNode, template="Cast::Execute(in Value,out Result)"}
   RerouteNode = RigVMDispatchNode(
    Elements=(Rotation=(X=0,Y=0,Z=0,W=1),Translation=(X=0,Y=0,Z=0),Scale3D=(X=-1,Y=1,Z=1))
    Elements.Scale3D.Y=@n1.Input.Scale3D.Y
    Elements.Scale3D.Z=@n1.Input.Scale3D.Z
    Elements.Translation=@n1.Input.Translation
    Elements.Scale3D.X=/Script/RigVM.RigVMFunction_MathFloatNegate(Value=@n1.Input.Scale3D.X) {is=struct RigVMUnitNode, template="Negate::Execute(in Value,out Result)"}.Result
    Elements.Rotation=/Script/RigVM.RigVMFunction_MathQuaternionFromRotator(
     Rotator=(Pitch=0,Yaw=0,Roll=0)
     Rotator.Roll=&RigVMFunction_MathFloatAdd(B=180, A=@n2.Result.Roll) {template="Add::Execute(in A,in B,out Result)"}.Result
     Rotator.Pitch=@n2.Result.Pitch
     Rotator.Yaw=@n2.Result.Yaw
    ) {is=struct RigVMUnitNode}.Result
   ) {template="DISPATCH_RigVMDispatch_MakeStruct(in Elements,out Struct)"}
   > RigVMFunctionReturnNode(Output=@RerouteNode.Struct)
  ]}
 ContainedGraph {nodes [
   RigVMFunctionEntryNode()
   > SequenceExecution = RigVMAggregateNode()
   n1 = RigVMVariableNode() {variable=Bones}
   RigVMDispatchNode(
    Num=RigVMDispatchNode(Array=@n1.Value) {template="DISPATCH_RigVMDispatch_ArrayGetNum(in Array,out Num)"}.Num
    ExecuteContext=@SequenceExecution.A
    Array=RigVMVariableNode() {variable=Transforms}.Value
   ) {template="DISPATCH_RigVMDispatch_ArraySetNum(io Array,in Num)"}
   > n2 = RigVMDispatchNode(Array=/Script/ControlRig.RigUnit_GetTransformItemArray(Space=GlobalSpace, Items=@n1.Value) {is=struct RigVMUnitNode}.Transforms) {template="DISPATCH_RigVMDispatch_ArrayIterator(in Array,out Element,out Index,out Count,out Ratio)"}
   Get Negative Transform = RigVMFunctionReferenceNode(Input=@n2.Element)
   n3 = RigVMDispatchNode(False=@n2.Element, True=@"Get Negative Transform".Output, Condition=RigVMVariableNode() {variable=Negative Side}.Value) {template="DISPATCH_RigVMDispatch_If(in Condition,in True,in False,out Result)"}
   MathIntEquals_1 = RigVMDispatchNode(
    A=@n2.Index
    B=/Script/RigVM.RigVMFunction_MathIntSub(B=1, A=@n2.Count) {is=struct RigVMUnitNode, template="Subtract::Execute(in A,in B,out Result)"}.Result
   ) {template="DISPATCH_RigVMDispatch_CoreEquals(in A,in B,out Result)"}
   n4 = &RigVMFunction_MathQuaternionToRotator(Value=@n3.Result.Rotation) {template="Cast::Execute(in Value,out Result)"}
   n5 = &RigVMFunction_MathFloatAdd(B=90, A=@n4.Result.Roll) {template="Add::Execute(in A,in B,out Result)"}
   > RigVMDispatchNode(
    Element=(Rotation=(X=0,Y=0,Z=0,W=1),Translation=(X=0,Y=0,Z=0),Scale3D=(X=1,Y=1,Z=1))
    Index=@n2.Index
    Element.Scale3D=@n3.Result.Scale3D
    Element.Translation=@n3.Result.Translation
    Element.Rotation=RigVMDispatchNode(
     False=@n3.Result.Rotation
     Condition=@MathIntEquals_1.Result
     True=/Script/RigVM.RigVMFunction_MathQuaternionFromRotatorV2(Value=(Pitch=0,Yaw=0,Roll=0), Value.Roll=@n5.Result, Value.Pitch=@n4.Result.Pitch, Value.Yaw=@n4.Result.Yaw) {is=struct RigVMUnitNode, template="Cast::Execute(in Value,out Result)"}.Result
    ) {template="DISPATCH_RigVMDispatch_If(in Condition,in True,in False,out Result)"}.Result
    Array=RigVMVariableNode() {variable=Transforms}.Value
   ) {template="DISPATCH_RigVMDispatch_ArraySetAtIndex(io Array,in Index,in Element)"}
   &RigUnit_SetTransform(
    Weight=1
    bInitial=true
    Value=RigVMDispatchNode(Index=0, Array=RigVMVariableNode() {variable=Transforms}.Value) {template="DISPATCH_RigVMDispatch_ArrayGetAtIndex(in Array,in Index,out Element)"}.Element
    ExecutePin=@SequenceExecution.C
    Item=RigVMVariableNode() {variable=FK Space}.Value
   )
   > /Script/ControlRig.RigUnit_SetTransformItemArray(
    Space=GlobalSpace
    Weight=1
    bInitial=true
    bPropagateToChildren=True
    Transforms=RigVMVariableNode() {variable=Transforms}.Value
    Items=RigVMVariableNode() {variable=FK Controls}.Value
   ) {is=struct RigVMUnitNode}
   n6 = &RigUnit_GetTransform(Item=RigVMDispatchNode(Index=2, Array=RigVMVariableNode() {variable=Bones}.Value) {template="DISPATCH_RigVMDispatch_ArrayGetAtIndex(in Array,in Index,out Element)"}.Element)
   Get Negative Transform_0 = RigVMFunctionReferenceNode(Input=@n6.Transform)
   > &RigUnit_SetTransform(
    Weight=1
    bInitial=true
    Value=RigVMDispatchNode(True=@"Get Negative Transform_0".Output, False=@n6.Transform, Condition=RigVMVariableNode() {variable=Negative Side}.Value) {template="DISPATCH_RigVMDispatch_If(in Condition,in True,in False,out Result)"}.Result
    Item=RigVMVariableNode() {variable=Fingers Space}.Value
   )
   n7 = RigVMVariableNode() {variable=Transforms}
   n8 = /Script/RigVM.RigVMFunction_MathFloatDiv(
    B=2
    A=&RigVMFunction_MathVectorDistance(
     A=RigVMDispatchNode(Index=0, Array=@n7.Value) {template="DISPATCH_RigVMDispatch_ArrayGetAtIndex(in Array,in Index,out Element)"}.Element.Translation
     B=RigVMDispatchNode(Index=2, Array=@n7.Value) {template="DISPATCH_RigVMDispatch_ArrayGetAtIndex(in Array,in Index,out Element)"}.Element.Translation
    ).Result
   ) {is=struct RigVMUnitNode, template="Divide::Execute(in A,in B,out Result)"}
   &RigUnit_SetTransform(
    Value=(Rotation=(X=0,Y=0,Z=0,W=1),Translation=(X=0,Y=0,Z=0),Scale3D=(X=1,Y=1,Z=1))
    Weight=1
    bInitial=true
    Value.Scale3D=RigVMDispatchNode(False=(X=1,Y=1,Z=1), True=(X=-1,Y=1,Z=1), Condition=RigVMVariableNode() {variable=Negative Side}.Value) {template="DISPATCH_RigVMDispatch_If(in Condition,in True,in False,out Result)"}.Result
    Value.Translation=&RigVMFunction_MathTransformMul(
     A=(Rotation=(X=0,Y=0,Z=0,W=1),Translation=(X=0,Y=0,Z=0),Scale3D=(X=1,Y=1,Z=1))
     B=RigVMDispatchNode(Index=1, Array=@n7.Value) {template="DISPATCH_RigVMDispatch_ArrayGetAtIndex(in Array,in Index,out Element)"}.Element
     A.Translation.Y=&RigVMFunction_MathFloatNegate(Value=@n8.Result) {template="Negate::Execute(in Value,out Result)"}.Result
    ) {template="Multiply::Execute(in A,in B,out Result)"}.Result.Translation
    ExecutePin=@SequenceExecution.D
    Item=RigVMVariableNode() {variable=Pole Vector Control}.Value
   )
   n9 = RigVMVariableNode() {variable=Transforms}
   n10 = RigVMDispatchNode(Array=@n9.Value) {template="DISPATCH_RigVMDispatch_ArrayGetNum(in Array,out Num)"}
   > &RigUnit_SetTransform(
    Weight=1
    bInitial=true
    Value=RigVMDispatchNode(Index=&RigVMFunction_MathIntSub(B=1, A=@n10.Num) {template="Subtract::Execute(in A,in B,out Result)"}.Result, Array=@n9.Value) {template="DISPATCH_RigVMDispatch_ArrayGetAtIndex(in Array,in Index,out Element)"}.Element
    Item=RigVMVariableNode() {variable=IK Control}.Value
   )
   n11 = RigVMDispatchNode(ExecuteContext=@SequenceExecution.E, Array=RigVMVariableNode() {variable=IK Parents}.Value) {template="DISPATCH_RigVMDispatch_ArrayIterator(in Array,out Element,out Index,out Count,out Ratio)"}
   > /Script/ControlRig.RigUnit_AddParent(DisplayLabel=None, Parent=@n11.Element, Child=RigVMVariableNode() {variable=IK Control}.Value) {is=struct RigVMUnitNode}
   > &RigUnit_AddParent(DisplayLabel=None, Parent=@n11.Element, Child=RigVMVariableNode() {variable=Pole Vector Control}.Value)
   At = RigVMDispatchNode(Index=0, Array=RigVMVariableNode() {variable=FK Controls}.Value) {template="DISPATCH_RigVMDispatch_ArrayGetAtIndex(in Array,in Index,out Element)"}
   Concat = /Script/RigVM.RigVMFunction_NameConcat(B=_body_offset_space, A=@At.Element.Name) {is=struct RigVMUnitNode, template="Concat::Execute(in A,in B,out Result)"}
   n12 = RigVMVariableNode() {variable=FK Space}
   n13 = &RigUnit_GetTransform(bInitial=true, Item=@n12.Value)
   RigUnit_GetTransform = &RigUnit_GetTransform(Item=(Type=Control,Name="body_offset_ctrl"), bInitial=true)
   n14 = &RigUnit_HierarchyAddNull(
    Space=GlobalSpace
    Transform=(Rotation=(X=0,Y=0,Z=0,W=1),Translation=(X=0,Y=0,Z=0),Scale3D=(X=1,Y=1,Z=1))
    ExecutePin=@n11.Completed
    Name=@Concat.Result
    Transform.Scale3D=@n13.Transform.Scale3D
    Transform.Translation=@n13.Transform.Translation
    Transform.Rotation=@RigUnit_GetTransform.Transform.Rotation
    Parent=@n12.Value
   )
   > &RigUnit_AddParent(DisplayLabel=None, Child=@At.Element, Parent=@n14.Item)
   > RigVMFunctionReturnNode()
  ]}
 ContainedGraph {nodes [
   n1 = RigVMFunctionEntryNode()
   SequenceExecution = &RigVMFunction_Sequence(ExecuteContext=@n1.ExecuteContext)
   SequenceExecution_1 = &RigVMFunction_Sequence(ExecuteContext=@SequenceExecution.B)
   SequenceExecution_2 = &RigVMFunction_Sequence(ExecuteContext=@SequenceExecution_1.B)
   SequenceExecution_3 = &RigVMFunction_Sequence(ExecuteContext=@SequenceExecution_2.B)
   RigVMFunctionReturnNode(A=@SequenceExecution.A, B=@SequenceExecution_1.A, C=@SequenceExecution_2.A, D=@SequenceExecution_3.A, E=@SequenceExecution_3.B)
  ]}
 ContainedGraph {nodes [
   n1 = RigVMFunctionEntryNode()
   n2 = &RigUnit_GetTransform(Item=@n1.Bone)
   Get Negative Transform_1 = RigVMFunctionReferenceNode(Input=@n2.Transform)
   > &RigUnit_SetTransform(
    Weight=1
    bInitial=true
    Item=@n1.Control
    Value=RigVMDispatchNode(True=@"Get Negative Transform_1".Output, False=@n2.Transform, Condition=RigVMVariableNode() {variable=Negative Side}.Value) {template="DISPATCH_RigVMDispatch_If(in Condition,in True,in False,out Result)"}.Result
   )
   > RigVMFunctionReturnNode()
  ]}
 ContainedGraph {nodes [
   n1 = RigVMFunctionEntryNode()
   n2 = RigVMDispatchNode(ExecuteContext=@n1.Execute, Array=RigVMVariableNode() {variable=Fingers Bones}.Value) {template="DISPATCH_RigVMDispatch_ArrayIterator(in Array,out Element,out Index,out Count,out Ratio)"}
   n3 = &RigUnit_GetTransform(Item=@n2.Element)
   n4 = RigVMVariableNode() {variable=Negative Side}
   n5 = &RigVMFunction_MathQuaternionToRotator(Value=@n3.Transform.Rotation) {template="Cast::Execute(in Value,out Result)"}
   n6 = &RigVMFunction_MathFloatAdd(B=180, A=@n5.Result.Roll) {template="Add::Execute(in A,in B,out Result)"}
   > &RigUnit_SetTransform(
    Item=(Type=Control)
    Value=(Rotation=(X=0,Y=0,Z=0,W=1),Translation=(X=0,Y=0,Z=0),Scale3D=(X=1,Y=1,Z=1))
    Weight=1
    bInitial=true
    Value.Translation=@n3.Transform.Translation
    Value.Scale3D=RigVMDispatchNode(False=(X=1,Y=1,Z=1), True=(X=-1,Y=1,Z=1), Condition=@n4.Value) {template="DISPATCH_RigVMDispatch_If(in Condition,in True,in False,out Result)"}.Result
    Value.Rotation=RigVMDispatchNode(
     False=@n3.Transform.Rotation
     True=&RigVMFunction_MathQuaternionFromRotatorV2(Value=(Pitch=0,Yaw=0,Roll=0), Value.Roll=@n6.Result, Value.Pitch=@n5.Result.Pitch, Value.Yaw=@n5.Result.Yaw) {template="Cast::Execute(in Value,out Result)"}.Result
     Condition=@n4.Value
    ) {template="DISPATCH_RigVMDispatch_If(in Condition,in True,in False,out Result)"}.Result
    Item.Name=&RigVMFunction_NameConcat(B=_ctrl, A=@n2.Element.Name) {template="Concat::Execute(in A,in B,out Result)"}.Result
   )
   RigVMFunctionReturnNode(Execute=@n2.Completed)
  ]}
 ContainedGraph {nodes [
   RigVMFunctionEntryNode()
   > SequenceExecution = RigVMAggregateNode()
   n1 = RigVMVariableNode() {variable=Bones}
   n2 = RigVMVariableNode() {variable=Transforms}
   RigVMDispatchNode(
    Num=RigVMDispatchNode(Array=@n1.Value) {template="DISPATCH_RigVMDispatch_ArrayGetNum(in Array,out Num)"}.Num
    ExecuteContext=@SequenceExecution.A
    Array=@n2.Value
   ) {template="DISPATCH_RigVMDispatch_ArraySetNum(io Array,in Num)"}
   > n3 = RigVMDispatchNode(Array=&RigUnit_GetTransformItemArray(Space=GlobalSpace, Items=@n1.Value).Transforms) {template="DISPATCH_RigVMDispatch_ArrayIterator(in Array,out Element,out Index,out Count,out Ratio)"}
   Get Negative Transform_1 = RigVMFunctionReferenceNode(Input=@n3.Element)
   > RigVMDispatchNode(
    Index=@n3.Index
    Element=RigVMDispatchNode(False=@n3.Element, True=@"Get Negative Transform_1".Output, Condition=RigVMVariableNode() {variable=Negative Side}.Value) {template="DISPATCH_RigVMDispatch_If(in Condition,in True,in False,out Result)"}.Result
    Array=RigVMVariableNode() {variable=Transforms}.Value
   ) {template="DISPATCH_RigVMDispatch_ArraySetAtIndex(io Array,in Index,in Element)"}
   /Script/ControlRig.RigUnit_DebugTransformArrayMutableItemSpace(
    Color=(R=1,G=1,B=1,A=1)
    DebugDrawSettings=(DepthPriority=SDPG_Foreground,Lifetime=-1)
    Mode=Axes
    ParentIndices=()
    Scale=10
    Space=(Type=Bone)
    Thickness=1
    WorldOffset=(Rotation=(X=0,Y=0,Z=0,W=1),Translation=(X=0,Y=0,Z=0),Scale3D=(X=1,Y=1,Z=1))
    bEnabled=True
    ExecutePin=@n3.Completed
    Transforms=@n2.Value
   ) {is=struct RigVMUnitNode}
   &RigUnit_SetTransform(
    Weight=1
    bInitial=true
    Value=RigVMDispatchNode(Index=0, Array=RigVMVariableNode() {variable=Transforms}.Value) {template="DISPATCH_RigVMDispatch_ArrayGetAtIndex(in Array,in Index,out Element)"}.Element
    ExecutePin=@SequenceExecution.B
    Item=RigVMVariableNode() {variable=FK Space}.Value
   )
   > &RigUnit_SetTransformItemArray(
    Space=GlobalSpace
    Weight=1
    bInitial=true
    bPropagateToChildren=True
    Transforms=RigVMVariableNode() {variable=Transforms}.Value
    Items=RigVMVariableNode() {variable=FK Controls}.Value
   )
   &RigUnit_SetTransform(
    Value=(Rotation=(X=0,Y=0,Z=0,W=1),Translation=(X=0,Y=0,Z=0),Scale3D=(X=1,Y=1,Z=1))
    Weight=1
    bInitial=true
    Value.Translation=RigVMDispatchNode(Index=-1, Array=RigVMVariableNode() {variable=Transforms}.Value) {template="DISPATCH_RigVMDispatch_ArrayGetAtIndex(in Array,in Index,out Element)"}.Element.Translation
    Value.Scale3D=RigVMDispatchNode(False=(X=1,Y=1,Z=1), True=(X=-1,Y=1,Z=1), Condition=RigVMVariableNode() {variable=Negative Side}.Value) {template="DISPATCH_RigVMDispatch_If(in Condition,in True,in False,out Result)"}.Result
    ExecutePin=@SequenceExecution.C
    Item=RigVMVariableNode() {variable=IK Control}.Value
   )
   > Compute Pole Vector Parent = RigVMFunctionReferenceNode(
    Initial=true
    Root Bone=RigVMDispatchNode(Index=0, Array=RigVMVariableNode() {variable=Bones}.Value) {template="DISPATCH_RigVMDispatch_ArrayGetAtIndex(in Array,in Index,out Element)"}.Element
    Foot Control=RigVMVariableNode() {variable=IK Control}.Value
    Pole Vector Null=RigVMVariableNode() {variable=Pole Vector Null}.Value
   )
   n4 = RigVMVariableNode() {variable=Transforms}
   n5 = &RigVMFunction_MathVectorDistance(
    A=RigVMDispatchNode(Index=0, Array=@n4.Value) {template="DISPATCH_RigVMDispatch_ArrayGetAtIndex(in Array,in Index,out Element)"}.Element.Translation
    B=RigVMDispatchNode(Index=2, Array=@n4.Value) {template="DISPATCH_RigVMDispatch_ArrayGetAtIndex(in Array,in Index,out Element)"}.Element.Translation
   )
   > &RigUnit_SetTransform(
    Value=(Rotation=(X=0,Y=0,Z=0,W=1),Translation=(X=0,Y=0,Z=0),Scale3D=(X=1,Y=1,Z=1))
    Weight=1
    bInitial=true
    Value.Scale3D=RigVMDispatchNode(False=(X=1,Y=1,Z=1), True=(X=-1,Y=1,Z=1), Condition=RigVMVariableNode() {variable=Negative Side}.Value) {template="DISPATCH_RigVMDispatch_If(in Condition,in True,in False,out Result)"}.Result
    Value.Translation=&RigVMFunction_MathTransformMul(
     A=(Rotation=(X=0,Y=0,Z=0,W=1),Translation=(X=0,Y=0,Z=0),Scale3D=(X=1,Y=1,Z=1))
     B=RigVMDispatchNode(Index=1, Array=@n4.Value) {template="DISPATCH_RigVMDispatch_ArrayGetAtIndex(in Array,in Index,out Element)"}.Element
     A.Translation.Y=&RigVMFunction_MathFloatDiv(B=2, A=@n5.Result) {template="Divide::Execute(in A,in B,out Result)"}.Result
    ) {template="Multiply::Execute(in A,in B,out Result)"}.Result.Translation
    Item=RigVMVariableNode() {variable=Pole Vector Control}.Value
   )
   > &RigUnit_AddParent(
    DisplayLabel=None
    Child=RigVMVariableNode() {variable=Pole Vector Control}.Value
    Parent=RigVMVariableNode() {variable=Pole Vector Null}.Value
   )
   n6 = RigVMDispatchNode(ExecuteContext=@SequenceExecution.E, Array=RigVMVariableNode() {variable=IK Parents}.Value) {template="DISPATCH_RigVMDispatch_ArrayIterator(in Array,out Element,out Index,out Count,out Ratio)"}
   > &RigUnit_AddParent(DisplayLabel=None, Parent=@n6.Element, Child=RigVMVariableNode() {variable=IK Control}.Value)
   > &RigUnit_AddParent(DisplayLabel=None, Parent=@n6.Element, Child=RigVMVariableNode() {variable=Pole Vector Control}.Value)
   RigVMFunctionReturnNode(ExecuteContext=@n6.Completed)
  ]}
 ContainedGraph {nodes [
   n1 = RigVMFunctionEntryNode()
   SequenceExecution = &RigVMFunction_Sequence(ExecuteContext=@n1.ExecuteContext)
   SequenceExecution_1 = &RigVMFunction_Sequence(ExecuteContext=@SequenceExecution.B)
   SequenceExecution_2 = &RigVMFunction_Sequence(ExecuteContext=@SequenceExecution_1.B)
   SequenceExecution_3 = &RigVMFunction_Sequence(ExecuteContext=@SequenceExecution_2.B)
   RigVMFunctionReturnNode(A=@SequenceExecution.A, B=@SequenceExecution_1.A, C=@SequenceExecution_2.A, D=@SequenceExecution_3.A, E=@SequenceExecution_3.B)
  ]}
 ContainedGraph {nodes [
   RigVMFunctionEntryNode()
   > SequenceExecution = RigVMAggregateNode()
   &RigUnit_SetTransform(
    Value=(Rotation=(X=0,Y=0,Z=0,W=1),Translation=(X=0,Y=0,Z=0),Scale3D=(X=1,Y=1,Z=1))
    Weight=1
    bInitial=true
    Value.Translation=&RigUnit_GetTransform(Item=RigVMVariableNode() {variable=Bone}.Value).Transform.Translation
    Value.Scale3D=RigVMDispatchNode(False=(X=1,Y=1,Z=1), True=(X=-1,Y=1,Z=1), Condition=RigVMVariableNode() {variable=Negative Side}.Value) {template="DISPATCH_RigVMDispatch_If(in Condition,in True,in False,out Result)"}.Result
    ExecutePin=@SequenceExecution.A
    Item=RigVMVariableNode() {variable=IK Control}.Value
   )
   n1 = &RigUnit_GetTransform(Item=RigVMVariableNode() {variable=Bone}.Value)
   n2 = &RigVMFunction_MathQuaternionToRotator(Value=@n1.Transform.Rotation) {template="Cast::Execute(in Value,out Result)"}
   n3 = &RigVMFunction_MathQuaternionFromRotatorV2(
    Value=(Pitch=0,Yaw=0,Roll=0)
    Value.Pitch=&RigVMFunction_MathFloatAdd(B=90, A=@n2.Result.Pitch) {template="Add::Execute(in A,in B,out Result)"}.Result
    Value.Roll=@n2.Result.Roll
    Value.Yaw=@n2.Result.Yaw
   ) {template="Cast::Execute(in Value,out Result)"}
   n4 = &RigVMFunction_MathQuaternionToEuler(RotationOrder=YXZ, Value=@n3.Result)
   n5 = &RigVMFunction_MathFloatAdd(B=180, A=@n4.Result.Y) {template="Add::Execute(in A,in B,out Result)"}
   n6 = RigVMVariableNode() {variable=Negative Side}
   &RigUnit_SetTransform(
    Value=(Rotation=(X=0,Y=0,Z=0,W=1),Translation=(X=0,Y=0,Z=0),Scale3D=(X=1,Y=1,Z=1))
    Weight=1
    bInitial=true
    Value.Translation=@n1.Transform.Translation
    Value.Rotation=RigVMDispatchNode(
     True=&RigVMFunction_MathQuaternionFromEuler(Euler=(X=0,Y=0,Z=0), RotationOrder=YXZ, Euler.Y=@n5.Result, Euler.X=@n4.Result.X, Euler.Z=@n4.Result.Z).Result
     False=@n3.Result
     Condition=@n6.Value
    ) {template="DISPATCH_RigVMDispatch_If(in Condition,in True,in False,out Result)"}.Result
    Value.Scale3D=RigVMDispatchNode(False=(X=1,Y=1,Z=1), True=(X=-1,Y=1,Z=1), Condition=@n6.Value) {template="DISPATCH_RigVMDispatch_If(in Condition,in True,in False,out Result)"}.Result
    ExecutePin=@SequenceExecution.B
    Item=RigVMVariableNode() {variable=FK Control}.Value
   )
   &RigUnit_SetTransform(
    Weight=1
    bInitial=true
    Value=&RigUnit_GetTransform(Item=RigVMVariableNode() {variable=Bone}.Value).Transform
    ExecutePin=@SequenceExecution.C
    Item=RigVMVariableNode() {variable=Toes Space}.Value
   )
   RigVMFunctionReturnNode(ExecuteContext=@SequenceExecution.D)
  ]}
 ContainedGraph {nodes [
   n1 = RigVMFunctionEntryNode()
   SequenceExecution = &RigVMFunction_Sequence(ExecuteContext=@n1.ExecuteContext)
   SequenceExecution_1 = &RigVMFunction_Sequence(ExecuteContext=@SequenceExecution.B)
   SequenceExecution_2 = &RigVMFunction_Sequence(ExecuteContext=@SequenceExecution_1.B)
   RigVMFunctionReturnNode(A=@SequenceExecution.A, B=@SequenceExecution_1.A, C=@SequenceExecution_2.A, D=@SequenceExecution_2.B)
  ]}
 ContainedGraph {nodes [
   n1 = RigVMFunctionEntryNode()
   n2 = RigVMVariableNode() {variable=IKFK Switch Control}
   RigVMVariableNode(
    ExecuteContext=@n1.Execute
    Value=/Script/RigVM.RigVMFunction_MathFloatGreater(B=0.5, A=&RigUnit_GetTransform(Space=LocalSpace, Item=@n2.Value).Transform.Translation.X) {is=struct RigVMUnitNode, template="Greater::Execute(in A,in B,out Result)"}.Result
   ) {variable=IK Mode}
   > SequenceExecution = RigVMAggregateNode()
   Set IKFK Visibility = RigVMFunctionReferenceNode(
    IK Controls=((),())
    Show Controls=true
    ExecuteContext=@SequenceExecution.A
    FK Controls=RigVMVariableNode() {variable=FK Controls}.Value
    IK Controls.0=RigVMVariableNode() {variable=IK Control}.Value
    IKFK Switch=RigVMVariableNode() {variable=IK Mode}.Value
    IK Controls.1=RigVMVariableNode() {variable=Pole Vector Control}.Value
   )
   > /Script/ControlRig.RigUnit_SetControlVisibility(
    bVisible=&RigVMFunction_MathBoolAnd(A=RigVMVariableNode() {variable=IK Mode}.Value, B=RigVMVariableNode() {variable=IK Orient}.Value).Result
    Item=RigVMVariableNode() {variable=IK Orient Control}.Value
   ) {is=struct RigVMUnitNode, template="SetControlVisibility::Execute(in Item,in Pattern,in bVisible)"}
   n3 = RigVMVariableNode() {variable=FK Orient Control}
   n4 = &RigUnit_GetTransform(Space=LocalSpace, Item=@n3.Value)
   n5 = RigVMVariableNode() {variable=FK Orient References}
   n6 = RigVMDispatchNode(Index=0, Array=@n5.Value) {template="DISPATCH_RigVMDispatch_ArrayGetAtIndex(in Array,in Index,out Element)"}
   n7 = RigVMVariableNode() {variable=Arm Space}
   n8 = /Script/ControlRig.RigUnit_ProjectTransformToNewParent(bChildInitial=True, bOldParentInitial=True, NewParent=@n6.Element, OldParent=@n6.Element, Child=@n7.Value) {is=struct RigVMUnitNode}
   n9 = RigVMDispatchNode(Index=1, Array=@n5.Value) {template="DISPATCH_RigVMDispatch_ArrayGetAtIndex(in Array,in Index,out Element)"}
   &RigUnit_SetTransform(
    Value=(Rotation=(X=0,Y=0,Z=0,W=1),Translation=(X=0,Y=0,Z=0),Scale3D=(X=1,Y=1,Z=1))
    Weight=1
    Value.Rotation=RigVMDispatchNode(
     Condition=&RigVMFunction_MathFloatGreater(A=@n4.Transform.Translation.X) {template="Greater::Execute(in A,in B,out Result)"}.Result
     False=@n8.Transform.Rotation
     True=&RigUnit_ProjectTransformToNewParent(bChildInitial=True, bOldParentInitial=True, NewParent=@n9.Element, OldParent=@n9.Element, Child=@n7.Value).Transform.Rotation
    ) {template="DISPATCH_RigVMDispatch_If(in Condition,in True,in False,out Result)"}.Result
    Value.Scale3D=@n8.Transform.Scale3D
    Value.Translation=@n8.Transform.Translation
    ExecutePin=@SequenceExecution.B
    Item=@n7.Value
   )
   At = RigVMDispatchNode(Index=0, Array=RigVMVariableNode() {variable=FK Controls}.Value) {template="DISPATCH_RigVMDispatch_ArrayGetAtIndex(in Array,in Index,out Element)"}
   Concat = &RigVMFunction_NameConcat(B=_body_offset_space, A=@At.Element.Name) {template="Concat::Execute(in A,in B,out Result)"}
   n10 = &RigUnit_GetTransform(bInitial=false, Item=RigVMVariableNode() {variable=Arm Space}.Value)
   RigUnit_GetTransform = &RigUnit_GetTransform(Item=(Type=Control,Name="body_offset_ctrl"), bInitial=false)
   > &RigUnit_SetTransform(
    Item=(Type=Null,Name="None")
    Value=(Rotation=(X=0,Y=0,Z=0,W=1),Translation=(X=0,Y=0,Z=0),Scale3D=(X=1,Y=1,Z=1))
    Weight=1
    bInitial=False
    bPropagateToChildren=true
    Item.Name=@Concat.Result
    Value.Scale3D=@n10.Transform.Scale3D
    Value.Translation=@n10.Transform.Translation
    Value.Rotation=@RigUnit_GetTransform.Transform.Rotation
   )
   n11 = &RigVMFunction_ControlFlowBranch(ExecuteContext=@SequenceExecution.C, Condition=RigVMVariableNode() {variable=IK Mode}.Value)
   FK Point At = RigVMFunctionReferenceNode(
    Initial=false
    Point At Axis=(X=1,Y=0,Z=0)
    Propagate To Children=true
    Up Vector Axis=(X=0,Y=0,Z=1)
    Up Vector Target=(X=0,Y=0,Z=1)
    ExecuteContext=@n11.False
    Negative Side=RigVMVariableNode() {variable=Negative Side}.Value
    Bones=RigVMVariableNode() {variable=Bones}.Value
    FK Controls=RigVMVariableNode() {variable=FK Controls}.Value
   )
   > Match IK onto FK = RigVMFunctionReferenceNode(
    Secondary Controls=()
    Bones=RigVMVariableNode() {variable=Bones}.Value
    Pole Vector Control=RigVMVariableNode() {variable=Pole Vector Control}.Value
    IK Control=RigVMVariableNode() {variable=IK Control}.Value
   )
   SequenceAggregate_1 = &RigVMFunction_Sequence(ExecuteContext=@n11.True)
   n12 = RigVMVariableNode() {variable=IK Orient}
   DISPATCH_RigDispatch_CoreNotEquals = RigVMDispatchNode(B=RigVMVariableNode() {variable=Previous Align}.Value, A=@n12.Value) {template="DISPATCH_RigVMDispatch_CoreNotEquals(in A,in B,out Result)"}
   n13 = &RigVMFunction_ControlFlowBranch(Condition=@DISPATCH_RigDispatch_CoreNotEquals.Result, ExecuteContext=@SequenceAggregate_1.A)
   n14 = &RigVMFunction_ControlFlowBranch(ExecuteContext=@n13.True, Condition=@n12.Value)
   n15 = RigVMVariableNode() {variable=IK Orient Control}
   n16 = RigVMVariableNode() {variable=IK Control}
   Set Transform_1 = &RigUnit_SetRotation(
    Space=GlobalSpace
    Weight=1
    bInitial=False
    ExecutePin=@n14.False
    Value=&RigUnit_GetTransform(Item=@n15.Value).Transform.Rotation
    Item=@n16.Value
   )
   > /Script/ControlRig.RigUnit_SendEvent(Event=RequestAutoKey, bEnable=True, bOnlyDuringInteraction=True, Item=@n16.Value) {is=struct RigVMUnitNode, template="SendEvent::Execute(in Event,in Item,in OffsetInSeconds,in bEnable,in bOnlyDuringInteraction)"}
   Set Transform_1_1 = &RigUnit_SetRotation(
    Space=GlobalSpace
    Weight=1
    bInitial=False
    ExecutePin=@n14.True
    Value=&RigUnit_GetTransform(Item=@n16.Value).Transform.Rotation
    Item=@n15.Value
   )
   > SendEvent = &RigUnit_SendEvent(Event=RequestAutoKey, bEnable=True, bOnlyDuringInteraction=True, Item=@n15.Value) {template="SendEvent::Execute(in Event,in Item,in OffsetInSeconds,in bEnable,in bOnlyDuringInteraction)"}
   IK 2 Bones = RigVMFunctionReferenceNode(
    Primary Axis=(X=1,Y=0,Z=0)
    Secondary Axis=(X=0,Y=-1,Z=0)
    Bones=RigVMVariableNode() {variable=Bones}.Value
    Enable Stretch=RigVMVariableNode() {variable=Enable Stretch}.Value
    Softness=RigVMVariableNode() {variable=Softness}.Value
    Pole Vector Control=RigVMVariableNode() {variable=Pole Vector Control}.Value
    Negative Side=RigVMVariableNode() {variable=Negative Side}.Value
    IK Control=RigVMVariableNode() {variable=IK Effector}.Value
    ExecuteContext=@SequenceAggregate_1.B
   )
   SequenceAggregate = &RigVMFunction_Sequence(ExecuteContext=@"IK 2 Bones".ExecuteContext)
   n17 = RigVMDispatchNode(Index=1, Array=RigVMVariableNode() {variable=Bones}.Value) {template="DISPATCH_RigVMDispatch_ArrayGetAtIndex(in Array,in Index,out Element)"}
   n18 = RigVMVariableNode() {variable=IK Orient Null}
   Set Transform = &RigUnit_SetRotation(
    Space=GlobalSpace
    Weight=1
    bInitial=False
    Value=&RigUnit_ProjectTransformToNewParent(bChildInitial=True, bOldParentInitial=True, NewParent=@n17.Element, OldParent=@n17.Element, Child=@n18.Value).Transform.Rotation
    ExecutePin=@SequenceAggregate.A
    Item=@n18.Value
   )
   n19 = &RigVMFunction_ControlFlowBranch(ExecuteContext=@"Set Transform".ExecutePin, Condition=RigVMVariableNode() {variable=IK Orient}.Value)
   n20 = RigVMDispatchNode(Index=-1, Array=RigVMVariableNode() {variable=Bones}.Value) {template="DISPATCH_RigVMDispatch_ArrayGetAtIndex(in Array,in Index,out Element)"}
   n21 = RigVMVariableNode() {variable=IK Orient Control}
   RigUnit_SetTransform_1 = &RigUnit_SetRotation(
    Space=GlobalSpace
    Weight=1
    bInitial=False
    bPropagateToChildren=true
    Item=@n20.Element
    ExecutePin=@n19.True
    Value=&RigUnit_ProjectTransformToNewParent(bChildInitial=True, bOldParentInitial=True, Child=@n20.Element, NewParent=@n21.Value, OldParent=@n21.Value).Transform.Rotation
   )
   n22 = RigVMVariableNode() {variable=FK Controls}
   Project to New Parents = RigVMFunctionReferenceNode(Propagate to Children=true, ExecuteContext=@SequenceAggregate.B, Parents=RigVMVariableNode() {variable=Bones}.Value, Children=@n22.Value)
   > For_Each = RigVMDispatchNode(Array=@n22.Value) {template="DISPATCH_RigVMDispatch_ArrayIterator(in Array,out Element,out Index,out Count,out Ratio)"}
   Get_Metadata = RigVMDispatchNode(Default=(Type=Bone,Name="None"), Name=Gimbal Control, NameSpace=Self, Item=@For_Each.Element) {template="DISPATCH_RigDispatch_GetMetadata(in Item,in Name,in NameSpace,in Default,out Value,out Found)"}
   RigVMFunction_ControlFlowBranch = &RigVMFunction_ControlFlowBranch(ExecuteContext=@For_Each.ExecuteContext, Condition=@Get_Metadata.Found)
   &RigUnit_SetTransform(
    Weight=1
    bInitial=False
    bPropagateToChildren=false
    Value=&RigUnit_GetTransform(bInitial=False, Item=@For_Each.Element).Transform
    Item=@Get_Metadata.Value
    ExecutePin=@RigVMFunction_ControlFlowBranch.True
   )
   SequenceExecution_1 = RigVMAggregateNode(ExecuteContext=@SequenceExecution.D)
   Set IKFK Keys = RigVMFunctionReferenceNode(
    "IK Controls "=((),())
    ExecuteContext=@SequenceExecution_1.A
    FK Controls=RigVMVariableNode() {variable=FK Controls}.Value
    IK Controls .0=RigVMVariableNode() {variable=IK Control}.Value
    IK Controls .1=RigVMVariableNode() {variable=Pole Vector Control}.Value
    Switch Control=RigVMVariableNode() {variable=IKFK Switch Control}.Value
    IK Mode=RigVMVariableNode() {variable=IK Mode}.Value
    Previous IK Mode=RigVMVariableNode() {variable=Previous IK Mode}.Value
   )
   n23 = &RigUnit_GetTransform(Item=RigVMDispatchNode(Index=2, Array=RigVMVariableNode() {variable=Bones}.Value) {template="DISPATCH_RigVMDispatch_ArrayGetAtIndex(in Array,in Index,out Element)"}.Element)
   Get Negative Transform = RigVMFunctionReferenceNode(Input=@n23.Transform)
   &RigUnit_SetTransform(
    Weight=1
    Value=RigVMDispatchNode(True=@"Get Negative Transform".Output, False=@n23.Transform, Condition=RigVMVariableNode() {variable=Negative Side}.Value) {template="DISPATCH_RigVMDispatch_If(in Condition,in True,in False,out Result)"}.Result
    ExecutePin=@SequenceExecution_1.C
    Item=RigVMVariableNode() {variable=Finger Space}.Value
   )
   /Script/ControlRig.RigUnit_GetControlBool(Control=@n3.Value.Name) {is=struct RigVMUnitNode, template="GetControlValue::Execute(in Control,out BoolValue)"}
   RigVMFunctionReturnNode(Execute=@SequenceExecution_1.D)
   RigVMVariableNode() {variable=IK Control}
  ]}
 ContainedGraph {nodes [
   n1 = RigVMFunctionEntryNode()
   SequenceExecution = &RigVMFunction_Sequence(ExecuteContext=@n1.ExecuteContext)
   SequenceExecution_1 = &RigVMFunction_Sequence(ExecuteContext=@SequenceExecution.B)
   SequenceExecution_2 = &RigVMFunction_Sequence(ExecuteContext=@SequenceExecution_1.B)
   RigVMFunctionReturnNode(A=@SequenceExecution.A, B=@SequenceExecution_1.A, C=@SequenceExecution_2.A, D=@SequenceExecution_2.B)
  ]}
 ContainedGraph {nodes [
   n1 = RigVMFunctionEntryNode()
   SequenceExecution_1 = &RigVMFunction_Sequence(ExecuteContext=@n1.ExecuteContext)
   SequenceExecution_1_1 = &RigVMFunction_Sequence(ExecuteContext=@SequenceExecution_1.B)
   SequenceExecution_1_2 = &RigVMFunction_Sequence(ExecuteContext=@SequenceExecution_1_1.B)
   RigVMFunctionReturnNode(A=@SequenceExecution_1.A, B=@SequenceExecution_1_1.A, C=@SequenceExecution_1_2.A, D=@SequenceExecution_1_2.B)
  ]}
 ContainedGraph {nodes [
   n1 = RigVMFunctionEntryNode()
   SendEvent = &RigUnit_SendEvent(Event=OpenUndoBracket, Item=(Type=None,Name="None"), bEnable=True, bOnlyDuringInteraction=True, ExecutePin=@n1.Execute) {template="SendEvent::Execute(in Event,in Item,in OffsetInSeconds,in bEnable,in bOnlyDuringInteraction)"}
   > SequenceAggregate = RigVMAggregateNode()
   Interpolate Interacting Controls = RigVMFunctionReferenceNode(
    ExecuteContext=@SequenceAggregate.A
    Controls=RigVMVariableNode() {variable=Metacarpals Controls}.Value
    Active=RigVMVariableNode() {variable=Auto Meta}.Value
   )
   > Interpolate Interacting Controls_1 = RigVMFunctionReferenceNode(Controls=RigVMVariableNode() {variable=Base Fingers Controls}.Value, Active=RigVMVariableNode() {variable=Auto Base}.Value)
   n2 = RigVMDispatchNode(ExecuteContext=@SequenceAggregate.B, Array=RigVMVariableNode() {variable=Proxy Controls}.Value) {template="DISPATCH_RigVMDispatch_ArrayIterator(in Array,out Element,out Index,out Count,out Ratio)"}
   n3 = &RigVMFunction_ControlFlowBranch(
    Condition=RigVMDispatchNode(Element=@n2.Element, Array=/Script/ControlRig.RigUnit_IsInteracting() {is=struct RigVMUnitNode}.Items) {template="DISPATCH_RigVMDispatch_ArrayFind(in Array,in Element,out Index,out Success)"}.Success
    ExecuteContext=@n2.ExecuteContext
   )
   Find Finger Controls = RigVMFunctionReferenceNode(Value=@n2.Element.Name, ExecuteContext=@n3.True)
   > n4 = RigVMDispatchNode(Array=@"Find Finger Controls".Items) {template="DISPATCH_RigVMDispatch_ArrayIterator(in Array,out Element,out Index,out Count,out Ratio)"}
   > &RigUnit_OffsetTransformForItem(
    Weight=1
    bPropagateToChildren=True
    Item=@n4.Element
    OffsetTransform=&RigUnit_GetTransform(Space=LocalSpace, Item=@n2.Element).Transform
   )
   > &RigUnit_SendEvent(Event=RequestAutoKey, bEnable=True, bOnlyDuringInteraction=True, Item=@n4.Element) {template="SendEvent::Execute(in Event,in Item,in OffsetInSeconds,in bEnable,in bOnlyDuringInteraction)"}
   &RigUnit_SetTransform(
    Space=LocalSpace
    Value=(Rotation=(X=0,Y=0,Z=0,W=1),Translation=(X=0,Y=0,Z=0),Scale3D=(X=1,Y=1,Z=1))
    Weight=1
    bInitial=False
    Item=@n2.Element
    ExecutePin=@n4.Completed
   )
   n5 = RigVMDispatchNode(ExecuteContext=@SequenceAggregate.C, Array=RigVMVariableNode() {variable=Fingers Bones}.Value) {template="DISPATCH_RigVMDispatch_ArrayIterator(in Array,out Element,out Index,out Count,out Ratio)"}
   n6 = &RigVMFunction_NameConcat(B=_ctrl, A=@n5.Element.Name) {template="Concat::Execute(in A,in B,out Result)"}
   > &RigUnit_SetTransform(
    Weight=1
    Item=@n5.Element
    Value=&RigUnit_ProjectTransformToNewParent(
     NewParent=(Type=Control)
     OldParent=(Type=Control)
     bChildInitial=True
     bOldParentInitial=True
     Child=@n5.Element
     NewParent.Name=@n6.Result
     OldParent.Name=@n6.Result
    ).Transform
   )
   n7 = &RigUnit_SendEvent(Event=CloseUndoBracket, Item=(Type=None,Name="None"), bEnable=True, bOnlyDuringInteraction=True, ExecutePin=@SequenceAggregate.D) {template="SendEvent::Execute(in Event,in Item,in OffsetInSeconds,in bEnable,in bOnlyDuringInteraction)"}
   RigVMFunctionReturnNode(Execute=@n7.ExecutePin)
  ]}
 ContainedGraph {nodes [
   n1 = RigVMFunctionEntryNode()
   SequenceAggregate = &RigVMFunction_Sequence(ExecuteContext=@n1.ExecuteContext)
   SequenceAggregate_1 = &RigVMFunction_Sequence(ExecuteContext=@SequenceAggregate.B)
   SequenceAggregate_2 = &RigVMFunction_Sequence(ExecuteContext=@SequenceAggregate_1.B)
   RigVMFunctionReturnNode(A=@SequenceAggregate.A, B=@SequenceAggregate_1.A, C=@SequenceAggregate_2.A, D=@SequenceAggregate_2.B)
  ]}
 ContainedGraph {nodes [
   n1 = RigVMFunctionEntryNode()
   > &RigUnit_SetTransform(
    Weight=1
    Item=@n1.Bone
    Value=&RigUnit_ProjectTransformToNewParent(bChildInitial=True, bOldParentInitial=True, Child=@n1.Bone, NewParent=@n1.Control, OldParent=@n1.Control).Transform
   )
   > RigVMFunctionReturnNode()
  ]}
 ContainedGraph {nodes [
   n1 = RigVMFunctionEntryNode()
   n2 = RigVMVariableNode() {variable=IKFK Switch Control}
   RigVMVariableNode(
    ExecuteContext=@n1.Execute
    Value=&RigVMFunction_MathFloatGreater(B=0.5, A=&RigUnit_GetTransform(Space=LocalSpace, Item=@n2.Value).Transform.Translation.X) {template="Greater::Execute(in A,in B,out Result)"}.Result
   ) {variable=IK Mode}
   > SequenceExecution = RigVMAggregateNode()
   Set IKFK Visibility = RigVMFunctionReferenceNode(
    IK Controls=((),())
    Show Controls=true
    ExecuteContext=@SequenceExecution.A
    IK Controls.0=RigVMVariableNode() {variable=IK Control}.Value
    FK Controls=RigVMVariableNode() {variable=FK Controls}.Value
    IK Controls.1=RigVMVariableNode() {variable=Pole Vector Control}.Value
    IKFK Switch=RigVMVariableNode() {variable=IK Mode}.Value
   )
   n3 = &RigVMFunction_ControlFlowBranch(ExecuteContext=@SequenceExecution.B, Condition=RigVMVariableNode() {variable=IK Mode}.Value)
   FK Point At = RigVMFunctionReferenceNode(
    Initial=false
    Point At Axis=(X=-1,Y=0,Z=0)
    Propagate To Children=true
    Up Vector Axis=(X=0,Y=0,Z=1)
    Up Vector Target=(X=0,Y=0,Z=1)
    ExecuteContext=@n3.False
    Negative Side=RigVMVariableNode() {variable=Negative Side}.Value
    FK Controls=RigVMVariableNode() {variable=FK Controls}.Value
    Bones=RigVMVariableNode() {variable=Bones}.Value
   )
   > Match IK onto FK = RigVMFunctionReferenceNode(
    Secondary Controls=RigVMVariableNode() {variable=Secondary IK Controls}.Value
    Bones=RigVMVariableNode() {variable=Bones}.Value
    Pole Vector Control=RigVMVariableNode() {variable=Pole Vector Control}.Value
    IK Control=RigVMVariableNode() {variable=IK Control}.Value
   )
   Compute Pole Vector Parent = RigVMFunctionReferenceNode(
    Initial=False
    Root Bone=RigVMDispatchNode(Index=0, Array=RigVMVariableNode() {variable=Bones}.Value) {template="DISPATCH_RigVMDispatch_ArrayGetAtIndex(in Array,in Index,out Element)"}.Element
    ExecuteContext=@n3.True
    Foot Control=RigVMVariableNode() {variable=IK Control}.Value
    Pole Vector Null=RigVMVariableNode() {variable=Pole Vector Null}.Value
   )
   > IK 2 Bones = RigVMFunctionReferenceNode(
    Primary Axis=(X=-1,Y=0,Z=0)
    Secondary Axis=(X=0,Y=1,Z=0)
    Bones=RigVMVariableNode() {variable=Bones}.Value
    Enable Stretch=RigVMVariableNode() {variable=Enable Stretch}.Value
    IK Control=RigVMVariableNode() {variable=IK Effector}.Value
    Softness=RigVMVariableNode() {variable=Softness}.Value
    Pole Vector Control=RigVMVariableNode() {variable=Pole Vector Control}.Value
    Negative Side=RigVMVariableNode() {variable=Negative Side}.Value
   )
   > Project to New Parents = RigVMFunctionReferenceNode(Propagate to Children=True, Parents=RigVMVariableNode() {variable=Bones}.Value, Children=RigVMVariableNode() {variable=FK Controls}.Value)
   Make_Array = RigVMDispatchNode(
    Values=((Type=None,Name="None"),((),Type=None,Name="None"))
    Values.0=RigVMVariableNode() {variable=IK Control}.Value
    Values.1=RigVMVariableNode() {variable=Pole Vector Control}.Value
   ) {template="DISPATCH_RigVMDispatch_ArrayMake(in Values,out Array)"}
   n4 = RigVMDispatchNode(Array=@Make_Array.Array, ExecuteContext=@SequenceExecution.C, Other=RigVMVariableNode() {variable=Secondary IK Controls}.Value) {template="DISPATCH_RigVMDispatch_ArrayUnion(io Array,in Other)"}
   > Set IKFK Keys = RigVMFunctionReferenceNode(
    "IK Controls "=@n4.Array
    FK Controls=RigVMVariableNode() {variable=FK Controls}.Value
    Switch Control=RigVMVariableNode() {variable=IKFK Switch Control}.Value
    Previous IK Mode=RigVMVariableNode() {variable=Previous IK Mode}.Value
    IK Mode=RigVMVariableNode() {variable=IK Mode}.Value
   )
   RigVMFunctionReturnNode(Execute=@SequenceExecution.D)
  ]}
 ContainedGraph {nodes [
   n1 = RigVMFunctionEntryNode()
   SequenceExecution = &RigVMFunction_Sequence(ExecuteContext=@n1.ExecuteContext)
   SequenceExecution_1 = &RigVMFunction_Sequence(ExecuteContext=@SequenceExecution.B)
   SequenceExecution_2 = &RigVMFunction_Sequence(ExecuteContext=@SequenceExecution_1.B)
   RigVMFunctionReturnNode(A=@SequenceExecution.A, B=@SequenceExecution_1.A, C=@SequenceExecution_2.A, D=@SequenceExecution_2.B)
  ]}
 ContainedGraph {nodes [
   n1 = &RigVMFunction_ControlFlowBranch(ExecuteContext=(), Condition=RigVMVariableNode() {variable=IK Mode}.Value)
   &RigUnit_SendEvent(Event=RequestAutoKey, bEnable=True, bOnlyDuringInteraction=True, ExecutePin=@n1.True, Item=RigVMVariableNode() {variable=IK Control}.Value) {template="SendEvent::Execute(in Event,in Item,in OffsetInSeconds,in bEnable,in bOnlyDuringInteraction)"}
   > &RigUnit_SendEvent(Event=RequestAutoKey, bEnable=True, bOnlyDuringInteraction=True, Item=RigVMVariableNode() {variable=FK Control}.Value) {template="SendEvent::Execute(in Event,in Item,in OffsetInSeconds,in bEnable,in bOnlyDuringInteraction)"}
   > RigVMVariableNode(Value=True) {variable=IK Mode}
   RigVMFunctionEntryNode()
   n2 = RigVMVariableNode() {variable=IKFK Switch Control}
   > RigVMVariableNode(Value=&RigVMFunction_MathFloatGreater(B=0.5, A=&RigUnit_GetTransform(Space=LocalSpace, Item=@n2.Value).Transform.Translation.X) {template="Greater::Execute(in A,in B,out Result)"}.Result) {variable=IK Mode}
   > SequenceExecution = RigVMAggregateNode()
   n3 = RigVMVariableNode() {variable=IK Mode}
   &RigUnit_SetControlVisibility(ExecutePin=@SequenceExecution.A, bVisible=@n3.Value, Item=RigVMVariableNode() {variable=IK Control}.Value) {template="SetControlVisibility::Execute(in Item,in Pattern,in bVisible)"}
   > &RigUnit_SetControlVisibility(bVisible=&RigVMFunction_MathBoolNot(Value=@n3.Value).Result, Item=RigVMVariableNode() {variable=FK Control}.Value) {template="SetControlVisibility::Execute(in Item,in Pattern,in bVisible)"}
   n4 = &RigVMFunction_ControlFlowBranch(ExecuteContext=@SequenceExecution.B, Condition=RigVMVariableNode() {variable=IK Mode}.Value)
   n5 = RigVMVariableNode() {variable=Bone}
   n6 = RigVMVariableNode() {variable=FK Control}
   &RigUnit_SetTransform(
    Weight=1
    ExecutePin=@n4.False
    Value=&RigUnit_ProjectTransformToNewParent(bChildInitial=True, bOldParentInitial=True, Child=@n5.Value, NewParent=@n6.Value, OldParent=@n6.Value).Transform
    Item=@n5.Value
   )
   n7 = RigVMVariableNode() {variable=IK Control}
   n8 = RigVMVariableNode() {variable=Bone}
   > &RigUnit_SetTransform(
    Weight=1
    Value=&RigUnit_ProjectTransformToNewParent(bChildInitial=True, bOldParentInitial=True, Child=@n7.Value, NewParent=@n8.Value, OldParent=@n8.Value).Transform
    Item=@n7.Value
   )
   n9 = RigVMVariableNode() {variable=Bone}
   n10 = RigVMVariableNode() {variable=IK Control}
   SetTransform_13 = &RigUnit_SetRotation(
    Space=GlobalSpace
    Weight=1
    bInitial=False
    ExecutePin=@n4.True
    Value=&RigUnit_ProjectTransformToNewParent(bChildInitial=True, bOldParentInitial=True, Child=@n9.Value, NewParent=@n10.Value, OldParent=@n10.Value).Transform.Rotation
    Item=@n9.Value
   )
   n11 = &RigUnit_GetTransform(Item=@n9.Value)
   n12 = &RigUnit_GetTransform(Item=@n10.Value)
   Add = /Script/RigVM.RigVMFunction_MathVectorAdd(
    B=&RigUnit_GetTransform(Space=LocalSpace, Item=@n10.Value).Transform.Translation
    A=&RigVMFunction_MathTransformMakeRelative(Global=@n11.Transform, Parent=@n12.Transform).Local.Translation
   ) {is=struct RigVMUnitNode, template="Add::Execute(in A,in B,out Result)"}
   > &RigUnit_OffsetTransformForItem(
    OffsetTransform=(Rotation=(X=0,Y=0,Z=0,W=1),Translation=(X=0,Y=0,Z=0),Scale3D=(X=1,Y=1,Z=1))
    Weight=1
    bPropagateToChildren=True
    OffsetTransform.Translation=&RigVMFunction_MathTransformMakeRelative(
     Parent=@n11.Transform
     Global=/Script/RigVM.RigVMFunction_MathTransformMakeAbsolute(Local=(Rotation=(X=0,Y=0,Z=0,W=1),Translation=(X=0,Y=0,Z=0),Scale3D=(X=1,Y=1,Z=1)), Local.Translation=@Add.Result, Parent=@n12.Transform) {is=struct RigVMUnitNode, template="Make Absolute::Execute(in Local,in Parent,out Global)"}.Global
    ).Local.Translation
    Item=@n9.Value
   )
   n13 = RigVMVariableNode() {variable=Bone}
   n14 = RigVMVariableNode() {variable=FK Control}
   > n15 = &RigUnit_SetTransform(
    Weight=1
    Value=&RigUnit_ProjectTransformToNewParent(bChildInitial=True, bOldParentInitial=True, NewParent=@n13.Value, OldParent=@n13.Value, Child=@n14.Value).Transform
    Item=@n14.Value
   )
   n16 = &RigVMFunction_ControlFlowBranch(ExecuteContext=@n15.ExecutePin, Condition=RigVMVariableNode() {variable=IK Mode}.Value)
   &RigUnit_SendEvent(Event=RequestAutoKey, bEnable=True, bOnlyDuringInteraction=True, ExecutePin=@n16.False, Item=RigVMVariableNode() {variable=IK Control}.Value) {template="SendEvent::Execute(in Event,in Item,in OffsetInSeconds,in bEnable,in bOnlyDuringInteraction)"}
   > &RigUnit_SendEvent(Event=RequestAutoKey, bEnable=True, bOnlyDuringInteraction=True, Item=RigVMVariableNode() {variable=FK Control}.Value) {template="SendEvent::Execute(in Event,in Item,in OffsetInSeconds,in bEnable,in bOnlyDuringInteraction)"}
   > RigVMVariableNode(Value=True) {variable=IK Mode}
   MathBoolNotEquals_2_1 = RigVMDispatchNode(B=RigVMVariableNode() {variable=IK Mode}.Value, A=RigVMVariableNode() {variable=Previous IK Mode}.Value) {template="DISPATCH_RigVMDispatch_CoreNotEquals(in A,in B,out Result)"}
   n17 = &RigVMFunction_ControlFlowBranch(Condition=@MathBoolNotEquals_2_1.Result, ExecuteContext=@SequenceExecution.C)
   &RigUnit_SendEvent(Event=RequestAutoKey, bEnable=True, bOnlyDuringInteraction=True, ExecutePin=@n17.True, Item=RigVMVariableNode() {variable=IK Control}.Value) {template="SendEvent::Execute(in Event,in Item,in OffsetInSeconds,in bEnable,in bOnlyDuringInteraction)"}
   > &RigUnit_SendEvent(Event=RequestAutoKey, bEnable=True, bOnlyDuringInteraction=True, Item=RigVMVariableNode() {variable=FK Control}.Value) {template="SendEvent::Execute(in Event,in Item,in OffsetInSeconds,in bEnable,in bOnlyDuringInteraction)"}
   n18 = RigVMVariableNode() {variable=Toes Space}
   n19 = RigVMVariableNode() {variable=Bone}
   &RigUnit_SetTransform(
    Weight=1
    Value=&RigUnit_ProjectTransformToNewParent(bChildInitial=True, bOldParentInitial=True, Child=@n18.Value, NewParent=@n19.Value, OldParent=@n19.Value).Transform
    ExecutePin=@SequenceExecution.D
    Item=@n18.Value
   )
   > RigVMFunctionReturnNode()
   &RigUnit_GetTransform(Item=(Type=Bone), bInitial=false)
  ]}
 ContainedGraph {nodes [
   n1 = RigVMFunctionEntryNode()
   SequenceExecution = &RigVMFunction_Sequence(ExecuteContext=@n1.ExecuteContext)
   SequenceExecution_1 = &RigVMFunction_Sequence(ExecuteContext=@SequenceExecution.B)
   SequenceExecution_2 = &RigVMFunction_Sequence(ExecuteContext=@SequenceExecution_1.B)
   RigVMFunctionReturnNode(A=@SequenceExecution.A, B=@SequenceExecution_1.A, C=@SequenceExecution_2.A, D=@SequenceExecution_2.B)
  ]}
 ContainedGraph {nodes [
   n1 = RigVMFunctionEntryNode()
   > SequenceExecution = RigVMAggregateNode()
   n2 = RigVMDispatchNode(ExecuteContext=@SequenceExecution.A, Array=RigVMVariableNode() {variable=FK Controls}.Value) {template="DISPATCH_RigVMDispatch_ArrayIterator(in Array,out Element,out Index,out Count,out Ratio)"}
   n3 = RigVMDispatchNode(False=False, Condition=@n1.Show Controls, True=&RigVMFunction_MathBoolNot(Value=@n1.IKFK Switch).Result) {template="DISPATCH_RigVMDispatch_If(in Condition,in True,in False,out Result)"}
   > n4 = &RigUnit_SetControlVisibility(Item=@n2.Element, bVisible=@n3.Result) {template="SetControlVisibility::Execute(in Item,in Pattern,in bVisible)"}
   Get_Metadata = RigVMDispatchNode(Default=(Type=Bone,Name="None"), Name=Gimbal Control, NameSpace=Self, Item=@n2.Element) {template="DISPATCH_RigDispatch_GetMetadata(in Item,in Name,in NameSpace,in Default,out Value,out Found)"}
   RigVMFunction_ControlFlowBranch = &RigVMFunction_ControlFlowBranch(Condition=@Get_Metadata.Found, ExecuteContext=@n4.ExecutePin)
   &RigUnit_SetControlVisibility(Item=@Get_Metadata.Value, bVisible=@n3.Result, ExecutePin=@RigVMFunction_ControlFlowBranch.True) {template="SetControlVisibility::Execute(in Item,in Pattern,in bVisible)"}
   n5 = RigVMDispatchNode(ExecuteContext=@SequenceExecution.B, Array=RigVMVariableNode() {variable=IK Controls}.Value) {template="DISPATCH_RigVMDispatch_ArrayIterator(in Array,out Element,out Index,out Count,out Ratio)"}
   > &RigUnit_SetControlVisibility(
    Item=@n5.Element
    bVisible=RigVMDispatchNode(False=False, True=@n1.IKFK Switch, Condition=@n1.Show Controls) {template="DISPATCH_RigVMDispatch_If(in Condition,in True,in False,out Result)"}.Result
   ) {template="SetControlVisibility::Execute(in Item,in Pattern,in bVisible)"}
   RigVMFunctionReturnNode(ExecuteContext=@SequenceExecution.D)
  ]}
 ContainedGraph {nodes [
   n1 = RigVMFunctionEntryNode()
   SequenceExecution = &RigVMFunction_Sequence(ExecuteContext=@n1.ExecuteContext)
   SequenceExecution_1 = &RigVMFunction_Sequence(ExecuteContext=@SequenceExecution.B)
   SequenceExecution_2 = &RigVMFunction_Sequence(ExecuteContext=@SequenceExecution_1.B)
   RigVMFunctionReturnNode(A=@SequenceExecution.A, B=@SequenceExecution_1.A, C=@SequenceExecution_2.A, D=@SequenceExecution_2.B)
  ]}
 ContainedGraph {nodes [
   RigVMFunctionEntryNode()
   > SequenceExecution = RigVMAggregateNode()
   Project to New Parent_1 = RigVMFunctionReferenceNode(
    Propagate to Children=True
    ExecuteContext=@SequenceExecution.A
    Children=RigVMVariableNode() {variable=FK Controls}.Value
    Parents=RigVMVariableNode() {variable=Bones}.Value
   )
   n1 = RigVMDispatchNode(Index=0, Array=RigVMVariableNode() {variable=FK Controls}.Value) {template="DISPATCH_RigVMDispatch_ArrayGetAtIndex(in Array,in Index,out Element)"}
   n2 = RigVMVariableNode() {variable=Pelvis Bone}
   RigUnit_SetTransform = &RigUnit_SetTransform(
    Weight=1
    Item=@n1.Element
    Value=&RigUnit_ProjectTransformToNewParent(bChildInitial=True, bOldParentInitial=True, Child=@n1.Element, NewParent=@n2.Value, OldParent=@n2.Value).Transform
    ExecutePin=@SequenceExecution.C
   )
   RigVMFunctionReturnNode(ExecuteContext=@SequenceExecution.D)
   Spine Match IK onto FK = RigVMFunctionReferenceNode(
    Execute=@SequenceExecution.B
    Offsets=RigVMVariableNode() {variable=Offsets}.Value
    Bones=RigVMVariableNode() {variable=Bones}.Value
    IK Controls=RigVMVariableNode() {variable=IK Controls}.Value
   )
  ]}
 ContainedGraph {nodes [
   n1 = RigVMFunctionEntryNode()
   SequenceExecution = &RigVMFunction_Sequence(ExecuteContext=@n1.ExecuteContext)
   SequenceExecution_1 = &RigVMFunction_Sequence(ExecuteContext=@SequenceExecution.B)
   SequenceExecution_2 = &RigVMFunction_Sequence(ExecuteContext=@SequenceExecution_1.B)
   RigVMFunctionReturnNode(A=@SequenceExecution.A, B=@SequenceExecution_1.A, C=@SequenceExecution_2.A, D=@SequenceExecution_2.B)
  ]}
 ContainedGraph {nodes [
   RigVMFunctionEntryNode()
   > SequenceExecution = RigVMAggregateNode()
   n1 = RigVMDispatchNode(Index=1, Array=RigVMVariableNode() {variable=Bones}.Value) {template="DISPATCH_RigVMDispatch_ArrayGetAtIndex(in Array,in Index,out Element)"}
   n2 = RigVMVariableNode() {variable=Head FK Space}
   n3 = RigVMVariableNode() {variable=Head Orient Reference}
   &RigUnit_SetTransform(
    Value=(Rotation=(X=0,Y=0,Z=0,W=1),Translation=(X=0,Y=0,Z=0),Scale3D=(X=1,Y=1,Z=1))
    Weight=1
    Value.Translation=&RigUnit_ProjectTransformToNewParent(bChildInitial=True, bOldParentInitial=True, NewParent=@n1.Element, OldParent=@n1.Element, Child=@n2.Value).Transform.Translation
    Value.Rotation=&RigUnit_ProjectTransformToNewParent(bChildInitial=True, bOldParentInitial=True, Child=@n2.Value, NewParent=@n3.Value, OldParent=@n3.Value).Transform.Rotation
    ExecutePin=@SequenceExecution.A
    Item=@n2.Value
   )
   n4 = RigVMVariableNode() {variable=Head IK Space}
   n5 = RigVMVariableNode() {variable=Head Orient Reference}
   n6 = /Script/ControlRig.RigUnit_HierarchyGetParent(bDefaultParent=True, Child=@n4.Value) {is=struct RigVMUnitNode}
   > &RigUnit_SetTransform(
    Value=(Rotation=(X=0,Y=0,Z=0,W=1),Translation=(X=0,Y=0,Z=0),Scale3D=(X=1,Y=1,Z=1))
    Weight=1
    Value.Rotation=&RigUnit_ProjectTransformToNewParent(bChildInitial=True, bOldParentInitial=True, Child=@n4.Value, NewParent=@n5.Value, OldParent=@n5.Value).Transform.Rotation
    Value.Translation=&RigUnit_ProjectTransformToNewParent(bChildInitial=True, bOldParentInitial=True, NewParent=@n6.Parent, OldParent=@n6.Parent, Child=@n4.Value).Transform.Translation
    Item=@n4.Value
   )
   Project to New Parents = RigVMFunctionReferenceNode(
    Propagate to Children=True
    ExecuteContext=@SequenceExecution.B
    Parents=RigVMVariableNode() {variable=Bones}.Value
    Children=RigVMVariableNode() {variable=FK Controls}.Value
   )
   n7 = RigVMVariableNode() {variable=Bones}
   n8 = RigVMDispatchNode(
    Index=&RigVMFunction_MathIntSub(B=1, A=RigVMDispatchNode(Array=@n7.Value) {template="DISPATCH_RigVMDispatch_ArrayGetNum(in Array,out Num)"}.Num) {template="Subtract::Execute(in A,in B,out Result)"}.Result
    Array=@n7.Value
   ) {template="DISPATCH_RigVMDispatch_ArrayGetAtIndex(in Array,in Index,out Element)"}
   n9 = RigVMVariableNode() {variable=IK Control}
   &RigUnit_SetTransform(
    Weight=1
    Value=&RigUnit_ProjectTransformToNewParent(bChildInitial=True, bOldParentInitial=True, NewParent=@n8.Element, OldParent=@n8.Element, Child=@n9.Value).Transform
    ExecutePin=@SequenceExecution.C
    Item=@n9.Value
   )
   RigVMFunctionReturnNode(ExecuteContext=@SequenceExecution.D)
  ]}
 ContainedGraph {nodes [
   n1 = RigVMFunctionEntryNode()
   SequenceExecution = &RigVMFunction_Sequence(ExecuteContext=@n1.ExecuteContext)
   SequenceExecution_1 = &RigVMFunction_Sequence(ExecuteContext=@SequenceExecution.B)
   SequenceExecution_2 = &RigVMFunction_Sequence(ExecuteContext=@SequenceExecution_1.B)
   RigVMFunctionReturnNode(A=@SequenceExecution.A, B=@SequenceExecution_1.A, C=@SequenceExecution_2.A, D=@SequenceExecution_2.B)
  ]}
 ContainedGraph {nodes [
   n1 = RigVMFunctionEntryNode()
   n2 = &RigUnit_SetTransform(
    Weight=1
    Item=@n1.Control
    ExecutePin=@n1.Execute
    Value=&RigUnit_ProjectTransformToNewParent(bChildInitial=True, bOldParentInitial=True, NewParent=@n1.Bone, OldParent=@n1.Bone, Child=@n1.Control).Transform
   )
   RigVMFunctionReturnNode(Execute=@n2.ExecutePin)
  ]}
 ContainedGraph {nodes [
   n1 = RigVMFunctionEntryNode()
   > SequenceExecution = RigVMAggregateNode()
   n2 = RigVMVariableNode() {variable=FK Orient References}
   n3 = RigVMDispatchNode(Index=1, Array=@n2.Value) {template="DISPATCH_RigVMDispatch_ArrayGetAtIndex(in Array,in Index,out Element)"}
   n4 = RigVMVariableNode() {variable=FK Arm Space}
   n5 = RigVMDispatchNode(Index=0, Array=@n2.Value) {template="DISPATCH_RigVMDispatch_ArrayGetAtIndex(in Array,in Index,out Element)"}
   n6 = &RigUnit_ProjectTransformToNewParent(bChildInitial=True, bOldParentInitial=True, NewParent=@n5.Element, OldParent=@n5.Element, Child=@n4.Value)
   &RigUnit_SetTransform(
    Value=(Rotation=(X=0,Y=0,Z=0,W=1),Translation=(X=0,Y=0,Z=0),Scale3D=(X=1,Y=1,Z=1))
    Weight=1
    Value.Rotation=&RigUnit_ProjectTransformToNewParent(bChildInitial=True, bOldParentInitial=True, NewParent=@n3.Element, OldParent=@n3.Element, Child=@n4.Value).Transform.Rotation
    Value.Scale3D=@n6.Transform.Scale3D
    Value.Translation=@n6.Transform.Translation
    ExecutePin=@SequenceExecution.A
    Item=@n4.Value
   )
   n7 = RigVMDispatchNode(Array=@n1.FK Controls, ExecuteContext=@SequenceExecution.B) {template="DISPATCH_RigVMDispatch_ArrayIterator(in Array,out Element,out Index,out Count,out Ratio)"}
   n8 = RigVMDispatchNode(Index=@n7.Index, Array=@n1.Bones) {template="DISPATCH_RigVMDispatch_ArrayGetAtIndex(in Array,in Index,out Element)"}
   > &RigUnit_SetTransform(
    Weight=1
    Item=@n7.Element
    Value=&RigUnit_ProjectTransformToNewParent(bChildInitial=True, bOldParentInitial=True, NewParent=@n8.Element, OldParent=@n8.Element, Child=@n7.Element).Transform
   )
   n9 = RigVMDispatchNode(Index=0, Array=RigVMVariableNode() {variable=Bones}.Value) {template="DISPATCH_RigVMDispatch_ArrayGetAtIndex(in Array,in Index,out Element)"}
   n10 = RigVMVariableNode() {variable=UpV Control}
   &RigUnit_SetTransform(
    Weight=1
    Value=&RigUnit_ProjectTransformToNewParent(bChildInitial=True, bOldParentInitial=True, NewParent=@n9.Element, OldParent=@n9.Element, Child=@n10.Value).Transform
    ExecutePin=@SequenceExecution.C
    Item=@n10.Value
   )
   n11 = RigVMDispatchNode(Index=2, Array=RigVMVariableNode() {variable=Bones}.Value) {template="DISPATCH_RigVMDispatch_ArrayGetAtIndex(in Array,in Index,out Element)"}
   n12 = RigVMVariableNode() {variable=Ik Control}
   > &RigUnit_SetTransform(
    Weight=1
    Value=&RigUnit_ProjectTransformToNewParent(bChildInitial=True, bOldParentInitial=True, NewParent=@n11.Element, OldParent=@n11.Element, Child=@n12.Value).Transform
    Item=@n12.Value
   )
   n13 = RigVMDispatchNode(Index=2, Array=RigVMVariableNode() {variable=Bones}.Value) {template="DISPATCH_RigVMDispatch_ArrayGetAtIndex(in Array,in Index,out Element)"}
   n14 = RigVMVariableNode() {variable=Fingers Space}
   &RigUnit_SetTransform(
    Weight=1
    Value=&RigUnit_ProjectTransformToNewParent(bChildInitial=True, bOldParentInitial=True, NewParent=@n13.Element, OldParent=@n13.Element, Child=@n14.Value).Transform
    ExecutePin=@SequenceExecution.D
    Item=@n14.Value
   )
   > RigVMFunctionReturnNode()
  ]}
 ContainedGraph {nodes [
   n1 = RigVMFunctionEntryNode()
   SequenceExecution = &RigVMFunction_Sequence(ExecuteContext=@n1.ExecuteContext)
   SequenceExecution_1 = &RigVMFunction_Sequence(ExecuteContext=@SequenceExecution.B)
   SequenceExecution_2 = &RigVMFunction_Sequence(ExecuteContext=@SequenceExecution_1.B)
   RigVMFunctionReturnNode(A=@SequenceExecution.A, B=@SequenceExecution_1.A, C=@SequenceExecution_2.A, D=@SequenceExecution_2.B)
  ]}
 ContainedGraph {nodes [
   n1 = RigVMFunctionEntryNode()
   n2 = RigVMDispatchNode(Array=@n1.Bones, ExecuteContext=@n1.Execute) {template="DISPATCH_RigVMDispatch_ArrayIterator(in Array,out Element,out Index,out Count,out Ratio)"}
   n3 = &RigVMFunction_NameConcat(B=_ctrl, A=@n2.Element.Name) {template="Concat::Execute(in A,in B,out Result)"}
   > &RigUnit_SetTransform(
    Item=(Type=Control)
    Weight=1
    bInitial=false
    Item.Name=@n3.Result
    Value=&RigUnit_ProjectTransformToNewParent(Child=(Type=Control), bChildInitial=True, bOldParentInitial=True, NewParent=@n2.Element, OldParent=@n2.Element, Child.Name=@n3.Result).Transform
   )
   RigVMFunctionReturnNode(Execute=@n2.Completed)
  ]}
 ContainedGraph {nodes [
   n1 = RigVMFunctionEntryNode()
   > SequenceExecution = RigVMAggregateNode()
   n2 = RigVMDispatchNode(Array=@n1.FK Controls, ExecuteContext=@SequenceExecution.A) {template="DISPATCH_RigVMDispatch_ArrayIterator(in Array,out Element,out Index,out Count,out Ratio)"}
   n3 = RigVMDispatchNode(Index=@n2.Index, Array=@n1.Bones) {template="DISPATCH_RigVMDispatch_ArrayGetAtIndex(in Array,in Index,out Element)"}
   > &RigUnit_SetTransform(
    Weight=1
    Item=@n2.Element
    Value=&RigUnit_ProjectTransformToNewParent(bChildInitial=True, bOldParentInitial=True, NewParent=@n3.Element, OldParent=@n3.Element, Child=@n2.Element).Transform
   )
   n4 = RigVMDispatchNode(Index=2, Array=RigVMVariableNode() {variable=Bones}.Value) {template="DISPATCH_RigVMDispatch_ArrayGetAtIndex(in Array,in Index,out Element)"}
   n5 = RigVMVariableNode() {variable=Ik Control}
   &RigUnit_SetTransform(
    Weight=1
    Value=&RigUnit_ProjectTransformToNewParent(bChildInitial=True, bOldParentInitial=True, NewParent=@n4.Element, OldParent=@n4.Element, Child=@n5.Value).Transform
    ExecutePin=@SequenceExecution.B
    Item=@n5.Value
   )
   n6 = RigVMDispatchNode(Index=0, Array=RigVMVariableNode() {variable=Bones}.Value) {template="DISPATCH_RigVMDispatch_ArrayGetAtIndex(in Array,in Index,out Element)"}
   Compute Pole Vector Parent_1 = RigVMFunctionReferenceNode(
    Initial=False
    Root Bone=@n6.Element
    ExecuteContext=@SequenceExecution.C
    Pole Vector Null=RigVMVariableNode() {variable=Pole Vector Null}.Value
    Foot Control=RigVMVariableNode() {variable=Ik Control}.Value
   )
   n7 = RigVMVariableNode() {variable=UpV Control}
   > &RigUnit_SetTransform(
    Weight=1
    Value=&RigUnit_ProjectTransformToNewParent(bChildInitial=True, bOldParentInitial=True, NewParent=@n6.Element, OldParent=@n6.Element, Child=@n7.Value).Transform
    Item=@n7.Value
   )
   RigVMFunctionReturnNode(ExecuteContext=@SequenceExecution.D)
  ]}
 ContainedGraph {nodes [
   n1 = RigVMFunctionEntryNode()
   SequenceExecution = &RigVMFunction_Sequence(ExecuteContext=@n1.ExecuteContext)
   SequenceExecution_1 = &RigVMFunction_Sequence(ExecuteContext=@SequenceExecution.B)
   SequenceExecution_2 = &RigVMFunction_Sequence(ExecuteContext=@SequenceExecution_1.B)
   RigVMFunctionReturnNode(A=@SequenceExecution.A, B=@SequenceExecution_1.A, C=@SequenceExecution_2.A, D=@SequenceExecution_2.B)
  ]}
 ContainedGraph {nodes [
   n1 = RigVMFunctionEntryNode()
   SequenceExecution = RigVMAggregateNode(ExecuteContext=@n1.Execute)
   n2 = RigVMVariableNode() {variable=Bone}
   n3 = RigVMVariableNode() {variable=FK Control}
   &RigUnit_SetTransform(
    Weight=1
    Value=&RigUnit_ProjectTransformToNewParent(bChildInitial=True, bOldParentInitial=True, NewParent=@n2.Value, OldParent=@n2.Value, Child=@n3.Value).Transform
    ExecutePin=@SequenceExecution.A
    Item=@n3.Value
   )
   n4 = RigVMVariableNode() {variable=Bone}
   n5 = RigVMVariableNode() {variable=IK Control}
   &RigUnit_SetTransform(
    Weight=1
    Value=&RigUnit_ProjectTransformToNewParent(bChildInitial=True, bOldParentInitial=True, NewParent=@n4.Value, OldParent=@n4.Value, Child=@n5.Value).Transform
    ExecutePin=@SequenceExecution.B
    Item=@n5.Value
   )
   &RigUnit_SetTransform(
    Weight=1
    bInitial=false
    Value=&RigUnit_GetTransform(Item=RigVMVariableNode() {variable=Bone}.Value).Transform
    ExecutePin=@SequenceExecution.C
    Item=RigVMVariableNode() {variable=Toes Space}.Value
   )
   RigVMFunctionReturnNode(Execute=@SequenceExecution.D)
  ]}
 ContainedGraph {nodes [
   n1 = RigVMFunctionEntryNode()
   SequenceExecution = &RigVMFunction_Sequence(ExecuteContext=@n1.ExecuteContext)
   SequenceExecution_1 = &RigVMFunction_Sequence(ExecuteContext=@SequenceExecution.B)
   SequenceExecution_2 = &RigVMFunction_Sequence(ExecuteContext=@SequenceExecution_1.B)
   RigVMFunctionReturnNode(A=@SequenceExecution.A, B=@SequenceExecution_1.A, C=@SequenceExecution_2.A, D=@SequenceExecution_2.B)
  ]}
 ContainedGraph {nodes [
   RigVMFunctionEntryNode()
   > SequenceExecution = RigVMAggregateNode()
   n1 = RigVMDispatchNode(ExecuteContext=@SequenceExecution.A, Array=RigVMVariableNode() {variable=Bones}.Value) {template="DISPATCH_RigVMDispatch_ArrayIterator(in Array,out Element,out Index,out Count,out Ratio)"}
   n2 = &RigVMFunction_ControlFlowBranch(
    ExecuteContext=@n1.ExecuteContext
    Condition=/Script/RigVM.RigVMFunction_MathIntGreaterEqual(A=@n1.Index, B=&RigVMFunction_MathIntSub(B=1, A=@n1.Count) {template="Subtract::Execute(in A,in B,out Result)"}.Result) {is=struct RigVMUnitNode, template="GreaterEqual::Execute(in A,in B,out Result)"}.Result
   )
   n3 = RigVMVariableNode() {variable=FK Controls}
   n4 = RigVMDispatchNode(Index=@n1.Index, Array=@n3.Value) {template="DISPATCH_RigVMDispatch_ArrayGetAtIndex(in Array,in Index,out Element)"}
   Get_Metadata = RigVMDispatchNode(Default=(Type=Bone,Name="None"), Name=Gimbal Control, NameSpace=Self, Item=@n4.Element) {template="DISPATCH_RigDispatch_GetMetadata(in Item,in Name,in NameSpace,in Default,out Value,out Found)"}
   n5 = /Script/ControlRig.RigUnit_AimBoneMath(
    DebugSettings=(Scale=10,WorldOffset=(Rotation=(X=0,Y=0,Z=0,W=1),Translation=(X=0,Y=0,Z=0),Scale3D=(X=1,Y=1,Z=1)))
    InputTransform=(Rotation=(X=0,Y=0,Z=0,W=1),Translation=(X=0,Y=0,Z=0),Scale3D=(X=1,Y=1,Z=1))
    Primary=(Weight=1,Axis=(X=1,Y=0,Z=0),Target=(X=0,Y=0,Z=0),Kind=Location,Space=(Type=Bone))
    Secondary=(Weight=1,Axis=(X=0,Y=0,Z=1),Target=(X=1,Y=0,Z=0),Kind=Direction,Space=(Type=Bone))
    Weight=1
    Primary.Space=RigVMDispatchNode(
     Index=/Script/RigVM.RigVMFunction_MathIntAdd(B=1, A=@n1.Index) {is=struct RigVMUnitNode, template="Add::Execute(in A,in B,out Result)"}.Result
     Array=@n3.Value
    ) {template="DISPATCH_RigVMDispatch_ArrayGetAtIndex(in Array,in Index,out Element)"}.Element
    InputTransform.Translation=RigVMDispatchNode(Index=@n1.Index, Array=&RigUnit_GetTransformItemArray(Space=GlobalSpace, Items=@n3.Value).Transforms) {template="DISPATCH_RigVMDispatch_ArrayGetAtIndex(in Array,in Index,out Element)"}.Element.Translation
    Secondary.Space=RigVMDispatchNode(False=@n4.Element, Condition=@Get_Metadata.Found, True=@Get_Metadata.Value) {template="DISPATCH_RigVMDispatch_If(in Condition,in True,in False,out Result)"}.Result
    Primary.Axis=RigVMVariableNode() {variable=Point At Axis}.Value
    Secondary.Axis=RigVMVariableNode() {variable=Up Vector Axis}.Value
    Secondary.Target=RigVMVariableNode() {variable=Up Vector Target}.Value
   ) {is=struct RigVMUnitNode}
   n6 = &RigVMFunction_MathQuaternionFromRotatorV2(Value=(Pitch=180,Yaw=0,Roll=0)) {template="Cast::Execute(in Value,out Result)"}
   &RigUnit_SetTransform(
    Weight=1
    Item=@n1.Element
    ExecutePin=@n2.False
    Value=RigVMDispatchNode(
     False=@n5.Result
     True=&RigVMFunction_MathTransformMul(A=(Rotation=(X=0,Y=0,Z=0,W=1),Translation=(X=0,Y=0,Z=0),Scale3D=(X=1,Y=1,Z=1)), B=@n5.Result, A.Rotation=@n6.Result) {template="Multiply::Execute(in A,in B,out Result)"}.Result
     Condition=RigVMVariableNode() {variable=Negative Side}.Value
    ) {template="DISPATCH_RigVMDispatch_If(in Condition,in True,in False,out Result)"}.Result
    bInitial=RigVMVariableNode() {variable=Initial}.Value
    bPropagateToChildren=RigVMVariableNode() {variable=Propagate To Children}.Value
   )
   n7 = RigVMVariableNode() {variable=Bones}
   n8 = &RigVMFunction_MathIntSub(B=1, A=RigVMDispatchNode(Array=@n7.Value) {template="DISPATCH_RigVMDispatch_ArrayGetNum(in Array,out Num)"}.Num) {template="Subtract::Execute(in A,in B,out Result)"}
   n9 = RigVMDispatchNode(Index=@n8.Result, Array=@n7.Value) {template="DISPATCH_RigVMDispatch_ArrayGetAtIndex(in Array,in Index,out Element)"}
   n10 = RigVMDispatchNode(Index=@n8.Result, Array=RigVMVariableNode() {variable=FK Controls}.Value) {template="DISPATCH_RigVMDispatch_ArrayGetAtIndex(in Array,in Index,out Element)"}
   &RigUnit_SetTransform(
    Weight=1
    bPropagateToChildren=true
    Item=@n9.Element
    Value=&RigUnit_ProjectTransformToNewParent(bChildInitial=true, bOldParentInitial=true, Child=@n9.Element, NewParent=@n10.Element, OldParent=@n10.Element).Transform
    ExecutePin=@SequenceExecution.B
   )
   RigVMFunctionReturnNode(ExecuteContext=@SequenceExecution.D)
  ]}
 ContainedGraph {nodes [
   n1 = RigVMFunctionEntryNode()
   SequenceExecution = &RigVMFunction_Sequence(ExecuteContext=@n1.ExecuteContext)
   SequenceExecution_1 = &RigVMFunction_Sequence(ExecuteContext=@SequenceExecution.B)
   SequenceExecution_2 = &RigVMFunction_Sequence(ExecuteContext=@SequenceExecution_1.B)
   RigVMFunctionReturnNode(A=@SequenceExecution.A, B=@SequenceExecution_1.A, C=@SequenceExecution_2.A, D=@SequenceExecution_2.B)
  ]}
 ContainedGraph {nodes [
   RigVMFunctionEntryNode()
   n1 = RigVMDispatchNode(Index=0, Array=RigVMVariableNode() {variable=Bones}.Value) {template="DISPATCH_RigVMDispatch_ArrayGetAtIndex(in Array,in Index,out Element)"}
   n2 = RigVMVariableNode() {variable=Pole Vector Control}
   > &RigUnit_SetTransform(
    Weight=1
    Value=&RigUnit_ProjectTransformToNewParent(bChildInitial=True, bOldParentInitial=True, NewParent=@n1.Element, OldParent=@n1.Element, Child=@n2.Value).Transform
    Item=@n2.Value
   )
   n3 = RigVMDispatchNode(Index=2, Array=RigVMVariableNode() {variable=Bones}.Value) {template="DISPATCH_RigVMDispatch_ArrayGetAtIndex(in Array,in Index,out Element)"}
   n4 = RigVMVariableNode() {variable=IK Control}
   > &RigUnit_SetTransform(
    Weight=1
    Value=&RigUnit_ProjectTransformToNewParent(bChildInitial=True, bOldParentInitial=True, NewParent=@n3.Element, OldParent=@n3.Element, Child=@n4.Value).Transform
    Item=@n4.Value
   )
   > For_Each = RigVMDispatchNode(Array=RigVMVariableNode() {variable=Secondary Controls}.Value) {template="DISPATCH_RigVMDispatch_ArrayIterator(in Array,out Element,out Index,out Count,out Ratio)"}
   > &RigUnit_SetTransform(
    Space=LocalSpace
    Value=(Rotation=(X=0,Y=0,Z=0,W=1),Translation=(X=0,Y=0,Z=0),Scale3D=(X=1,Y=1,Z=1))
    Weight=1
    bInitial=False
    bPropagateToChildren=true
    Item=@For_Each.Element
   )
   RigVMFunctionReturnNode(ExecuteContext=@For_Each.Completed)
  ]}
 ContainedGraph {nodes [
   n1 = &RigUnit_IsInteracting()
   n2 = RigVMVariableNode() {variable=IK Control}
   MathBoolAnd = RigVMAggregateNode(
    B=RigVMDispatchNode(Array=@n1.Items, Element=@n2.Value) {template="DISPATCH_RigVMDispatch_ArrayFind(in Array,in Element,out Index,out Success)"}.Success
    C=@n1.bIsTranslating
    A=/Script/ControlRig.RigUnit_GetBoolAnimationChannel(Channel=Keep Aligned, bInitial=False, Control=@n2.Value.Name) {is=struct RigVMUnitNode, template="GetAnimationChannel::Execute(out Value,in Control,in Channel,in bInitial)"}.Value
   )
   n3 = &RigVMFunction_ControlFlowBranch(ExecuteContext=(), Condition=@MathBoolAnd.Result)
   n4 = RigVMVariableNode() {variable=IK Control}
   RigUnit_GetTransform = &RigUnit_GetTransform(Item=RigVMDispatchNode(Index=1, Array=RigVMVariableNode() {variable=Bones}.Value) {template="DISPATCH_RigVMDispatch_ArrayGetAtIndex(in Array,in Index,out Element)"}.Element)
   DISPATCH_RigDispatch_SetMetadata = RigVMDispatchNode(
    Name=Alignment
    NameSpace=Self
    ExecuteContext=@n3.False
    Value=&RigVMFunction_MathTransformMakeRelative(Global=&RigUnit_GetTransform(Item=@n4.Value).Transform, Parent=@RigUnit_GetTransform.Transform).Local
    Item=@n4.Value
   ) {template="DISPATCH_RigDispatch_SetMetadata(in Item,in Name,in NameSpace,in Value,out Success)"}
   DISPATCH_RigDispatch_GetMetadata = RigVMDispatchNode(Default=(Rotation=(X=0,Y=0,Z=0,W=1),Translation=(X=0,Y=0,Z=0),Scale3D=(X=1,Y=1,Z=1)), Name=Alignment, NameSpace=Self, Item=@n2.Value) {template="DISPATCH_RigDispatch_GetMetadata(in Item,in Name,in NameSpace,in Default,out Value,out Found)"}
   n5 = RigVMDispatchNode(Index=1, Array=RigVMVariableNode() {variable=Bones}.Value) {template="DISPATCH_RigVMDispatch_ArrayGetAtIndex(in Array,in Index,out Element)"}
   Set Transform = &RigUnit_SetRotation(
    Space=GlobalSpace
    Weight=1
    bInitial=False
    ExecutePin=@n3.True
    Value=&RigVMFunction_MathTransformMakeAbsolute(Local=@DISPATCH_RigDispatch_GetMetadata.Value, Parent=&RigUnit_GetTransform(Item=@n5.Element).Transform) {template="Make Absolute::Execute(in Local,in Parent,out Global)"}.Global.Rotation
    Item=@n2.Value
   )
   RigVMFunctionEntryNode()
   > SequenceExecution = RigVMAggregateNode()
   n6 = RigVMVariableNode() {variable=Bones}
   n7 = RigVMVariableNode() {variable=Primary Axis}
   n8 = RigVMVariableNode() {variable=Negative Side}
   n9 = RigVMVariableNode() {variable=Secondary Axis}
   n10 = RigVMVariableNode() {variable=Bones}
   n11 = RigVMVariableNode() {variable=IK Control}
   Softness = RigVMCollapseNode(
    Bones=RigVMVariableNode() {variable=Bones}.Value
    Softness=RigVMVariableNode() {variable=Softness}.Value
    Effector=RigVMVariableNode() {variable=IK Control}.Value
   )
   /Script/ControlRig.RigUnit_TwoBoneIKSimplePerItem(
    DebugSettings=(Scale=10,WorldOffset=(Rotation=(X=0,Y=0,Z=0,W=1),Translation=(X=0,Y=0,Z=0),Scale3D=(X=1,Y=1,Z=1)))
    PoleVector=(X=0,Y=0,Z=0)
    PoleVectorKind=Location
    SecondaryAxisWeight=1
    StretchMaximumRatio=2
    StretchStartRatio=1
    Weight=1
    bPropagateToChildren=true
    ItemB=RigVMDispatchNode(Index=1, Array=@n6.Value) {template="DISPATCH_RigVMDispatch_ArrayGetAtIndex(in Array,in Index,out Element)"}.Element
    EffectorItem=RigVMDispatchNode(Index=2, Array=@n6.Value) {template="DISPATCH_RigVMDispatch_ArrayGetAtIndex(in Array,in Index,out Element)"}.Element
    ItemA=RigVMDispatchNode(Index=0, Array=@n6.Value) {template="DISPATCH_RigVMDispatch_ArrayGetAtIndex(in Array,in Index,out Element)"}.Element
    PrimaryAxis=RigVMDispatchNode(
     True=/Script/RigVM.RigVMFunction_MathVectorNegate(Value=@n7.Value) {is=struct RigVMUnitNode, template="Negate::Execute(in Value,out Result)"}.Result
     Condition=@n8.Value
     False=@n7.Value
    ) {template="DISPATCH_RigVMDispatch_If(in Condition,in True,in False,out Result)"}.Result
    SecondaryAxis=RigVMDispatchNode(
     True=&RigVMFunction_MathVectorNegate(Value=@n9.Value) {template="Negate::Execute(in Value,out Result)"}.Result
     Condition=@n8.Value
     False=@n9.Value
    ) {template="DISPATCH_RigVMDispatch_If(in Condition,in True,in False,out Result)"}.Result
    Effector=&RigUnit_ProjectTransformToNewParent(
     bChildInitial=True
     bOldParentInitial=True
     Child=RigVMDispatchNode(Index=2, Array=@n10.Value) {template="DISPATCH_RigVMDispatch_ArrayGetAtIndex(in Array,in Index,out Element)"}.Element
     NewParent=@n11.Value
     OldParent=@n11.Value
    ).Transform
    ExecutePin=@SequenceExecution.A
    ItemALength=@Softness.Length A
    ItemBLength=@Softness.Length B
    PoleVectorSpace=RigVMVariableNode() {variable=Pole Vector Control}.Value
    bEnableStretch=RigVMVariableNode() {variable=Enable Stretch}.Value
   ) {is=struct RigVMUnitNode}
   n12 = RigVMVariableNode() {variable=Bones}
   /Script/ControlRig.RigUnit_DebugLineItemSpace(
    Color=(R=0.2,G=0.2,B=0.2,A=1)
    DebugDrawSettings=(DepthPriority=SDPG_Foreground,Lifetime=-1)
    Space=(Type=Bone)
    Thickness=0.5
    WorldOffset=(Rotation=(X=0,Y=0,Z=0,W=1),Translation=(X=0,Y=0,Z=0),Scale3D=(X=1,Y=1,Z=1))
    bEnabled=True
    B=&RigUnit_GetTransform(Item=RigVMVariableNode() {variable=Pole Vector Control}.Value).Transform.Translation
    A=&RigUnit_GetTransform(Item=RigVMDispatchNode(Index=1, Array=@n12.Value) {template="DISPATCH_RigVMDispatch_ArrayGetAtIndex(in Array,in Index,out Element)"}.Element).Transform.Translation
    ExecutePin=@SequenceExecution.B
   ) {is=struct RigVMUnitNode}
   RigVMFunctionReturnNode(ExecuteContext=@SequenceExecution.D)
  ]}
 ContainedGraph {nodes [
   n1 = RigVMVariableNode() {variable=Softness}
   n2 = RigVMFunctionEntryNode()
   n3 = /Script/RigVM.RigVMFunction_MathFloatAbs(Value=&RigUnit_GetTransform(
     Space=LocalSpace
     bInitial=true
     Item=RigVMDispatchNode(Index=1, Array=@n2.Bones) {template="DISPATCH_RigVMDispatch_ArrayGetAtIndex(in Array,in Index,out Element)"}.Element
    ).Transform.Translation.X) {is=struct RigVMUnitNode, template="Absolute::Execute(in Value,out Result)"}
   n4 = &RigVMFunction_MathFloatAbs(Value=&RigUnit_GetTransform(
     Space=LocalSpace
     bInitial=true
     Item=RigVMDispatchNode(Index=2, Array=@n2.Bones) {template="DISPATCH_RigVMDispatch_ArrayGetAtIndex(in Array,in Index,out Element)"}.Element
    ).Transform.Translation.X) {template="Absolute::Execute(in Value,out Result)"}
   n5 = &RigVMFunction_MathFloatAdd(A=@n3.Result, B=@n4.Result) {template="Add::Execute(in A,in B,out Result)"}
   n6 = &RigVMFunction_MathFloatMul(
    B=0.02
    A=&RigVMFunction_MathFloatMul(A=@n5.Result, B=RigVMVariableNode() {variable=Softness}.Value) {template="Multiply::Execute(in A,in B,out Result)"}.Result
   ) {template="Multiply::Execute(in A,in B,out Result)"}
   n7 = /Script/RigVM.RigVMFunction_MathFloatSub(A=@n5.Result, B=@n6.Result) {is=struct RigVMUnitNode, template="Subtract::Execute(in A,in B,out Result)"}
   n8 = &RigVMFunction_MathVectorDistance(
    B=&RigUnit_GetTransform(Item=@n2.Effector).Transform.Translation
    A=&RigUnit_GetTransform(Item=RigVMDispatchNode(Index=0, Array=@n2.Bones) {template="DISPATCH_RigVMDispatch_ArrayGetAtIndex(in Array,in Index,out Element)"}.Element).Transform.Translation
   )
   Exponential Softness = RigVMCollapseNode(Softness=@n6.Result, DA=@n7.Result, Distance=@n8.Result, Stretch Distance=@n8.Result)
   n9 = RigVMDispatchNode(
    False=1
    Condition=&RigVMFunction_MathBoolAnd(
     A=&RigVMFunction_MathFloatGreater(B=0, A=@n1.Value) {template="Greater::Execute(in A,in B,out Result)"}.Result
     B=&RigVMFunction_MathFloatGreater(B=@n7.Result, A=@n8.Result) {template="Greater::Execute(in A,in B,out Result)"}.Result
    ).Result
    True=/Script/RigVM.RigVMFunction_MathFloatMax(B=1, A=@"Exponential Softness".Result) {is=struct RigVMUnitNode, template="Maximum::Execute(in A,in B,out Result)"}.Result
   ) {template="DISPATCH_RigVMDispatch_If(in Condition,in True,in False,out Result)"}
   RigVMFunctionReturnNode(
    Length A=&RigVMFunction_MathFloatMul(B=@n9.Result, A=@n3.Result) {template="Multiply::Execute(in A,in B,out Result)"}.Result
    Length B=&RigVMFunction_MathFloatMul(B=@n9.Result, A=@n4.Result) {template="Multiply::Execute(in A,in B,out Result)"}.Result
   )
  ]}
 ContainedGraph {nodes [
   n1 = RigVMFunctionEntryNode()
   n2 = &RigVMFunction_MathFloatNegate(Value=&RigVMFunction_MathFloatDiv(
     B=@n1.Softness
     A=&RigVMFunction_MathFloatSub(B=@n1.DA, A=@n1.Distance) {template="Subtract::Execute(in A,in B,out Result)"}.Result
    ) {template="Divide::Execute(in A,in B,out Result)"}.Result) {template="Negate::Execute(in Value,out Result)"}
   n3 = &RigVMFunction_MathFloatMul(
    B=@n1.Softness
    A=&RigVMFunction_MathFloatSub(
     A=1
     B=/Script/RigVM.RigVMFunction_MathFloatExponential(Value=@n2.Result) {is=struct RigVMUnitNode, template="Exponential::Execute(in Value,out Result)"}.Result
    ) {template="Subtract::Execute(in A,in B,out Result)"}.Result
   ) {template="Multiply::Execute(in A,in B,out Result)"}
   RigVMFunctionReturnNode(Result=&RigVMFunction_MathFloatDiv(
     A=@n1.Stretch Distance
     B=&RigVMFunction_MathFloatAdd(B=@n1.DA, A=@n3.Result) {template="Add::Execute(in A,in B,out Result)"}.Result
    ) {template="Divide::Execute(in A,in B,out Result)"}.Result)
  ]}
 ContainedGraph {nodes [
   n1 = RigVMFunctionEntryNode()
   SequenceExecution = &RigVMFunction_Sequence(ExecuteContext=@n1.ExecuteContext)
   SequenceExecution_1 = &RigVMFunction_Sequence(ExecuteContext=@SequenceExecution.B)
   SequenceExecution_2 = &RigVMFunction_Sequence(ExecuteContext=@SequenceExecution_1.B)
   RigVMFunctionReturnNode(A=@SequenceExecution.A, B=@SequenceExecution_1.A, C=@SequenceExecution_2.A, D=@SequenceExecution_2.B)
  ]}
 ContainedGraph {nodes [
   n1 = RigVMFunctionEntryNode()
   RigVMFunctionReturnNode(Result=&RigVMFunction_MathBoolAnd(B=@n1.C, A=&RigVMFunction_MathBoolAnd(A=@n1.A, B=@n1.B).Result).Result)
  ]}
 ContainedGraph {nodes [
   n1 = RigVMFunctionEntryNode()
   > n2 = RigVMDispatchNode(Array=@n1.Children) {template="DISPATCH_RigVMDispatch_ArrayIterator(in Array,out Element,out Index,out Count,out Ratio)"}
   n3 = RigVMDispatchNode(Index=@n2.Index, Array=@n1.Parents) {template="DISPATCH_RigVMDispatch_ArrayGetAtIndex(in Array,in Index,out Element)"}
   > &RigUnit_SetTransform(
    Weight=1
    Item=@n2.Element
    bPropagateToChildren=@n1.Propagate to Children
    Value=&RigUnit_ProjectTransformToNewParent(bChildInitial=True, bOldParentInitial=True, NewParent=@n3.Element, OldParent=@n3.Element, Child=@n2.Element).Transform
   )
   RigVMFunctionReturnNode(ExecuteContext=@n2.Completed)
  ]}
 ContainedGraph {nodes [
   n1 = RigVMFunctionEntryNode()
   > SequenceExecution_1 = RigVMAggregateNode()
   MathBoolNotEquals = RigVMDispatchNode(A=@n1.IK Mode, B=@n1.Previous IK Mode) {template="DISPATCH_RigVMDispatch_CoreNotEquals(in A,in B,out Result)"}
   n2 = &RigVMFunction_ControlFlowBranch(Condition=@MathBoolNotEquals.Result, ExecuteContext=@SequenceExecution_1.A)
   SequenceExecution = RigVMAggregateNode(ExecuteContext=@n2.True)
   n3 = RigVMDispatchNode(ExecuteContext=@SequenceExecution.A, Array=RigVMVariableNode() {variable=FK Controls}.Value) {template="DISPATCH_RigVMDispatch_ArrayIterator(in Array,out Element,out Index,out Count,out Ratio)"}
   > n4 = &RigUnit_SendEvent(Event=RequestAutoKey, bEnable=True, bOnlyDuringInteraction=True, Item=@n3.Element) {template="SendEvent::Execute(in Event,in Item,in OffsetInSeconds,in bEnable,in bOnlyDuringInteraction)"}
   Get_Metadata = RigVMDispatchNode(Default=(Type=Bone,Name="None"), Name=Gimbal Control, NameSpace=Self, Item=@n3.Element) {template="DISPATCH_RigDispatch_GetMetadata(in Item,in Name,in NameSpace,in Default,out Value,out Found)"}
   RigVMFunction_ControlFlowBranch = &RigVMFunction_ControlFlowBranch(Condition=@Get_Metadata.Found, ExecuteContext=@n4.ExecutePin)
   &RigUnit_SendEvent(Event=RequestAutoKey, bEnable=True, bOnlyDuringInteraction=True, Item=@Get_Metadata.Value, ExecutePin=@RigVMFunction_ControlFlowBranch.True) {template="SendEvent::Execute(in Event,in Item,in OffsetInSeconds,in bEnable,in bOnlyDuringInteraction)"}
   n5 = RigVMDispatchNode(ExecuteContext=@SequenceExecution.B, Array=RigVMVariableNode() {variable="IK Controls "}.Value) {template="DISPATCH_RigVMDispatch_ArrayIterator(in Array,out Element,out Index,out Count,out Ratio)"}
   > &RigUnit_SendEvent(Event=RequestAutoKey, bEnable=True, bOnlyDuringInteraction=True, Item=@n5.Element) {template="SendEvent::Execute(in Event,in Item,in OffsetInSeconds,in bEnable,in bOnlyDuringInteraction)"}
   &RigUnit_SendEvent(
    Event=RequestAutoKey
    bEnable=True
    bOnlyDuringInteraction=True
    ExecutePin=@SequenceExecution.C
    Item=RigVMVariableNode() {variable=Switch Control}.Value
   ) {template="SendEvent::Execute(in Event,in Item,in OffsetInSeconds,in bEnable,in bOnlyDuringInteraction)"}
   RigVMFunctionReturnNode(ExecuteContext=@SequenceExecution_1.D)
  ]}
 ContainedGraph {nodes [
   n1 = RigVMFunctionEntryNode()
   SequenceExecution_1 = &RigVMFunction_Sequence(ExecuteContext=@n1.ExecuteContext)
   SequenceExecution_1_1 = &RigVMFunction_Sequence(ExecuteContext=@SequenceExecution_1.B)
   SequenceExecution_1_2 = &RigVMFunction_Sequence(ExecuteContext=@SequenceExecution_1_1.B)
   RigVMFunctionReturnNode(A=@SequenceExecution_1.A, B=@SequenceExecution_1_1.A, C=@SequenceExecution_1_2.A, D=@SequenceExecution_1_2.B)
  ]}
 ContainedGraph {nodes [
   n1 = RigVMFunctionEntryNode()
   SequenceExecution = &RigVMFunction_Sequence(ExecuteContext=@n1.ExecuteContext)
   SequenceExecution_1 = &RigVMFunction_Sequence(ExecuteContext=@SequenceExecution.B)
   SequenceExecution_2 = &RigVMFunction_Sequence(ExecuteContext=@SequenceExecution_1.B)
   RigVMFunctionReturnNode(A=@SequenceExecution.A, B=@SequenceExecution_1.A, C=@SequenceExecution_2.A, D=@SequenceExecution_2.B)
  ]}
 ContainedGraph {nodes [
   RigVMFunctionEntryNode()
   > SequenceExecution = RigVMAggregateNode()
   n1 = RigVMDispatchNode(ExecuteContext=@SequenceExecution.A, Array=RigVMVariableNode() {variable=Bones}.Value) {template="DISPATCH_RigVMDispatch_ArrayIterator(in Array,out Element,out Index,out Count,out Ratio)"}
   n2 = &RigUnit_GetTransform(Item=RigVMDispatchNode(Index=0, Array=RigVMVariableNode() {variable=Bones}.Value) {template="DISPATCH_RigVMDispatch_ArrayGetAtIndex(in Array,in Index,out Element)"}.Element)
   n3 = &RigUnit_GetTransform(Item=@n1.Element)
   Get Negative Transform_1 = RigVMFunctionReferenceNode(
    Input=(Rotation=(X=0,Y=0,Z=0,W=1),Translation=(X=0,Y=0,Z=0),Scale3D=(X=0,Y=0,Z=0))
    Input.Scale3D=@n3.Transform.Scale3D
    Input.Translation=@n3.Transform.Translation
   )
   n4 = RigVMDispatchNode(
    False=(Rotation=(X=0,Y=0,Z=0,W=1),Translation=(X=0,Y=0,Z=0),Scale3D=(X=0,Y=0,Z=0))
    True=@"Get Negative Transform_1".Output
    False.Scale3D=@n3.Transform.Scale3D
    False.Translation=@n3.Transform.Translation
    False.Rotation=@n2.Transform.Rotation
    Condition=RigVMVariableNode() {variable=Negative Side}.Value
   ) {template="DISPATCH_RigVMDispatch_If(in Condition,in True,in False,out Result)"}
   > &RigUnit_SetTransform(
    Value=(Rotation=(X=0,Y=0,Z=0,W=1),Translation=(X=0,Y=0,Z=0),Scale3D=(X=1,Y=1,Z=1))
    Weight=1
    bInitial=true
    Item=RigVMDispatchNode(Index=@n1.Index, Array=RigVMVariableNode() {variable=BK Buffers}.Value) {template="DISPATCH_RigVMDispatch_ArrayGetAtIndex(in Array,in Index,out Element)"}.Element
    Value.Rotation=@n2.Transform.Rotation
    Value.Scale3D=@n4.Result.Scale3D
    Value.Translation=@n4.Result.Translation
   )
   > &RigUnit_SetTransform(
    Value=(Rotation=(X=0,Y=0,Z=0,W=1),Translation=(X=0,Y=0,Z=0),Scale3D=(X=1,Y=1,Z=1))
    Weight=1
    bInitial=true
    Item=RigVMDispatchNode(Index=@n1.Index, Array=RigVMVariableNode() {variable=BK Controls}.Value) {template="DISPATCH_RigVMDispatch_ArrayGetAtIndex(in Array,in Index,out Element)"}.Element
    Value.Rotation=@n2.Transform.Rotation
    Value.Scale3D=@n4.Result.Scale3D
    Value.Translation=@n4.Result.Translation
   )
   n5 = RigVMVariableNode() {variable=Bones}
   n6 = RigVMDispatchNode(Array=@n5.Value) {template="DISPATCH_RigVMDispatch_ArrayGetNum(in Array,out Num)"}
   n7 = &RigUnit_GetTransform(Item=RigVMDispatchNode(Index=&RigVMFunction_MathIntSub(B=1, A=@n6.Num) {template="Subtract::Execute(in A,in B,out Result)"}.Result, Array=@n5.Value) {template="DISPATCH_RigVMDispatch_ArrayGetAtIndex(in Array,in Index,out Element)"}.Element)
   n8 = RigVMDispatchNode(False=(X=1,Y=1,Z=1), True=(X=-1,Y=1,Z=1), Condition=RigVMVariableNode() {variable=Negative Side}.Value) {template="DISPATCH_RigVMDispatch_If(in Condition,in True,in False,out Result)"}
   &RigUnit_SetTransform(
    Value=(Rotation=(X=0,Y=0,Z=0,W=1),Translation=(X=0,Y=0,Z=0),Scale3D=(X=1,Y=1,Z=1))
    Weight=1
    bInitial=true
    Value.Translation=@n7.Transform.Translation
    Value.Rotation=@n2.Transform.Rotation
    Value.Scale3D=@n8.Result
    ExecutePin=@SequenceExecution.B
    Item=RigVMVariableNode() {variable=Roll Buffer}.Value
   )
   > &RigUnit_SetTransform(
    Value=(Rotation=(X=0,Y=0,Z=0,W=1),Translation=(X=0,Y=0,Z=0),Scale3D=(X=1,Y=1,Z=1))
    Weight=1
    bInitial=true
    Value.Translation=@n7.Transform.Translation
    Value.Rotation=@n2.Transform.Rotation
    Value.Scale3D=@n8.Result
    Item=RigVMVariableNode() {variable=Roll Control}.Value
   )
   RigVMFunctionReturnNode(ExecuteContext=@SequenceExecution.D)
  ]}
 ContainedGraph {nodes [
   n1 = RigVMFunctionEntryNode()
   SequenceExecution = &RigVMFunction_Sequence(ExecuteContext=@n1.ExecuteContext)
   SequenceExecution_1 = &RigVMFunction_Sequence(ExecuteContext=@SequenceExecution.B)
   SequenceExecution_2 = &RigVMFunction_Sequence(ExecuteContext=@SequenceExecution_1.B)
   RigVMFunctionReturnNode(A=@SequenceExecution.A, B=@SequenceExecution_1.A, C=@SequenceExecution_2.A, D=@SequenceExecution_2.B)
  ]}
 ContainedGraph {nodes [
   RigVMFunctionEntryNode()
   n1 = RigVMVariableNode() {variable=IKFK Switch Control}
   > RigVMVariableNode(Value=&RigVMFunction_MathFloatGreater(B=0.5, A=&RigUnit_GetTransform(Space=LocalSpace, Item=@n1.Value).Transform.Translation.X) {template="Greater::Execute(in A,in B,out Result)"}.Result) {variable=IK Mode}
   > SequenceExecution = RigVMAggregateNode()
   n2 = RigVMDispatchNode(ExecuteContext=@SequenceExecution.A, Array=RigVMVariableNode() {variable=BK Controls}.Value) {template="DISPATCH_RigVMDispatch_ArrayIterator(in Array,out Element,out Index,out Count,out Ratio)"}
   n3 = RigVMVariableNode() {variable=IK Mode}
   > SetControlVisibility = &RigUnit_SetControlVisibility(Item=@n2.Element, bVisible=@n3.Value) {template="SetControlVisibility::Execute(in Item,in Pattern,in bVisible)"}
   &RigUnit_SetControlVisibility(ExecutePin=@n2.Completed, bVisible=@n3.Value, Item=RigVMVariableNode() {variable=Roll Control}.Value) {template="SetControlVisibility::Execute(in Item,in Pattern,in bVisible)"}
   RigUnit_GetTransform = &RigUnit_GetTransform(Space=LocalSpace, Item=RigVMVariableNode() {variable=Roll Control}.Value)
   n4 = /Script/RigVM.RigVMFunction_MathFloatClamp(
    Maximum=180
    Minimum=0
    Value=&RigVMFunction_MathQuaternionToRotator(Value=@RigUnit_GetTransform.Transform.Rotation) {template="Cast::Execute(in Value,out Result)"}.Result.Yaw
   ) {is=struct RigVMUnitNode, template="Clamp::Execute(in Value,in Minimum,in Maximum,out Result)"}
   n5 = &RigUnit_GetTransform(
    Space=LocalSpace
    bInitial=false
    Item=RigVMDispatchNode(Index=0, Array=RigVMVariableNode() {variable=BK Controls}.Value) {template="DISPATCH_RigVMDispatch_ArrayGetAtIndex(in Array,in Index,out Element)"}.Element
   )
   n6 = &RigVMFunction_MathTransformMakeAbsolute(
    Local=&RigVMFunction_MathTransformMul(
     B=(Rotation=(X=0,Y=0,Z=0,W=1),Translation=(X=0,Y=0,Z=0),Scale3D=(X=1,Y=1,Z=1))
     A=&RigVMFunction_MathTransformMakeRelative(Global=(Rotation=(X=0,Y=0,Z=0,W=1),Translation=(X=0,Y=0,Z=0),Scale3D=(X=1,Y=1,Z=1)), Parent=@n5.Transform).Local
     B.Rotation=@n5.Transform.Rotation
    ) {template="Multiply::Execute(in A,in B,out Result)"}.Result
    Parent=@n5.Transform
   ) {template="Make Absolute::Execute(in Local,in Parent,out Global)"}
   n7 = &RigUnit_GetTransform(
    Space=LocalSpace
    bInitial=true
    Item=RigVMDispatchNode(Index=0, Array=RigVMVariableNode() {variable=BK Buffers}.Value) {template="DISPATCH_RigVMDispatch_ArrayGetAtIndex(in Array,in Index,out Element)"}.Element
   )
   RigUnit_SetTransform = &RigUnit_SetTransform(
    Space=LocalSpace
    Weight=1
    Item=RigVMDispatchNode(Index=0, Array=RigVMVariableNode() {variable=BK Buffers}.Value) {template="DISPATCH_RigVMDispatch_ArrayGetAtIndex(in Array,in Index,out Element)"}.Element
    Value=&RigVMFunction_MathTransformMul(
     A=(Rotation=(X=0,Y=0,Z=0,W=1),Translation=(X=0,Y=0,Z=0),Scale3D=(X=1,Y=1,Z=1))
     A.Rotation=&RigVMFunction_MathQuaternionFromRotatorV2(Value=(Pitch=0,Yaw=0,Roll=0), Value.Yaw=@n4.Result) {template="Cast::Execute(in Value,out Result)"}.Result
     B=&RigVMFunction_MathTransformMul(A=@n6.Global, B=@n7.Transform) {template="Multiply::Execute(in A,in B,out Result)"}.Result
    ) {template="Multiply::Execute(in A,in B,out Result)"}.Result
    ExecutePin=@SequenceExecution.B
   )
   n8 = RigVMDispatchNode(
    Index=0
    Array=RigVMDispatchNode(Array=RigVMVariableNode() {variable=BK Buffers}.Value) {template="DISPATCH_RigVMDispatch_ArrayClone(in Array,out Clone)"}.Clone
    ExecuteContext=@SequenceExecution.C
   ) {template="DISPATCH_RigVMDispatch_ArrayRemove(io Array,in Index)"}
   > n9 = RigVMDispatchNode(Array=@n8.Array) {template="DISPATCH_RigVMDispatch_ArrayReverse(io Array)"}
   > n10 = RigVMDispatchNode(Array=@n9.Array) {template="DISPATCH_RigVMDispatch_ArrayIterator(in Array,out Element,out Index,out Count,out Ratio)"}
   n11 = &RigVMFunction_MathFloatMul(A=RigVMVariableNode() {variable=Roll Clamp}.Value, B=RigVMVariableNode() {variable=Roll Blend}.Value) {template="Multiply::Execute(in A,in B,out Result)"}
   MathIntEquals = RigVMDispatchNode(B=&RigVMFunction_MathIntSub(B=1, A=@n10.Count) {template="Subtract::Execute(in A,in B,out Result)"}.Result, A=@n10.Index) {template="DISPATCH_RigVMDispatch_CoreEquals(in A,in B,out Result)"}
   n12 = &RigVMFunction_MathFloatMul(
    B=@n11.Result
    A=/Script/RigVM.RigVMFunction_MathIntToFloat(Value=@n10.Index) {is=struct RigVMUnitNode, template="Cast::Execute(in Value,out Result)"}.Result
   ) {template="Multiply::Execute(in A,in B,out Result)"}
   n13 = RigVMVariableNode() {variable=Roll Control}
   n14 = &RigVMFunction_MathFloatNegate(Value=&RigVMFunction_MathQuaternionToRotator(Value=&RigUnit_GetTransform(Space=LocalSpace, Item=@n13.Value).Transform.Rotation) {template="Cast::Execute(in Value,out Result)"}.Result.Yaw) {template="Negate::Execute(in Value,out Result)"}
   n15 = &RigVMFunction_MathFloatNegate(Value=&RigVMFunction_MathFloatClamp(
     Minimum=0
     Maximum=RigVMDispatchNode(True=180, False=@n11.Result, Condition=@MathIntEquals.Result) {template="DISPATCH_RigVMDispatch_If(in Condition,in True,in False,out Result)"}.Result
     Value=&RigVMFunction_MathFloatSub(B=@n12.Result, A=@n14.Result) {template="Subtract::Execute(in A,in B,out Result)"}.Result
    ) {template="Clamp::Execute(in Value,in Minimum,in Maximum,out Result)"}.Result) {template="Negate::Execute(in Value,out Result)"}
   > &RigUnit_SetRotation(
    Weight=1
    Item=@n10.Element
    Value=&RigVMFunction_MathTransformMul(
     B=(Rotation=(X=0,Y=0,Z=0,W=1),Translation=(X=0,Y=0,Z=0),Scale3D=(X=1,Y=1,Z=1))
     B.Rotation=&RigVMFunction_MathQuaternionFromRotatorV2(Value=(Pitch=0,Yaw=0,Roll=0), Value.Yaw=@n15.Result) {template="Cast::Execute(in Value,out Result)"}.Result
     A=&RigUnit_GetTransform(Space=LocalSpace, bInitial=true, Item=@n10.Element).Transform
    ) {template="Multiply::Execute(in A,in B,out Result)"}.Result.Rotation
   )
   SequenceExecution_1 = RigVMAggregateNode(ExecuteContext=@SequenceExecution.D)
   n16 = RigVMDispatchNode(Index=1, Array=RigVMVariableNode() {variable=BK Buffers}.Value) {template="DISPATCH_RigVMDispatch_ArrayGetAtIndex(in Array,in Index,out Element)"}
   n17 = &RigUnit_GetTransform(
    Space=LocalSpace
    bInitial=false
    Item=RigVMDispatchNode(Index=1, Array=RigVMVariableNode() {variable=BK Controls}.Value) {template="DISPATCH_RigVMDispatch_ArrayGetAtIndex(in Array,in Index,out Element)"}.Element
   )
   n18 = &RigVMFunction_MathTransformMakeAbsolute(
    Local=&RigVMFunction_MathTransformMul(
     B=(Rotation=(X=0,Y=0,Z=0,W=1),Translation=(X=0,Y=0,Z=0),Scale3D=(X=1,Y=1,Z=1))
     A=&RigVMFunction_MathTransformMakeRelative(Global=(Rotation=(X=0,Y=0,Z=0,W=1),Translation=(X=0,Y=0,Z=0),Scale3D=(X=1,Y=1,Z=1)), Parent=@n17.Transform).Local
     B.Rotation=@n17.Transform.Rotation
    ) {template="Multiply::Execute(in A,in B,out Result)"}.Result
    Parent=@n17.Transform
   ) {template="Make Absolute::Execute(in Local,in Parent,out Global)"}
   n19 = &RigUnit_GetTransform(Space=LocalSpace, bInitial=true, Item=@n16.Element)
   &RigUnit_SetTransform(
    Space=LocalSpace
    Weight=1
    Item=@n16.Element
    Value=&RigVMFunction_MathTransformMul(
     A=(Rotation=(X=0,Y=0,Z=0,W=1),Translation=(X=0,Y=0,Z=0),Scale3D=(X=1,Y=1,Z=1))
     A.Rotation=&RigUnit_GetTransform(Space=LocalSpace, Item=@n16.Element).Transform.Rotation
     B=&RigVMFunction_MathTransformMul(A=@n18.Global, B=@n19.Transform) {template="Multiply::Execute(in A,in B,out Result)"}.Result
    ) {template="Multiply::Execute(in A,in B,out Result)"}.Result
    ExecutePin=@SequenceExecution_1.A
   )
   n20 = &RigUnit_HierarchyGetParent(
    bDefaultParent=True
    Child=RigVMDispatchNode(Index=0, Array=RigVMVariableNode() {variable=BK Buffers}.Value) {template="DISPATCH_RigVMDispatch_ArrayGetAtIndex(in Array,in Index,out Element)"}.Element
   )
   n21 = RigVMVariableNode() {variable=Roll Buffer}
   RigUnit_SetRotation = &RigUnit_SetRotation(
    Space=GlobalSpace
    Weight=1
    Value=&RigUnit_ProjectTransformToNewParent(bChildInitial=True, bOldParentInitial=True, NewParent=@n20.Parent, OldParent=@n20.Parent, Child=@n21.Value).Transform.Rotation
    ExecutePin=@SequenceExecution_1.B
    Item=@n21.Value
   )
   RigVMFunctionReturnNode(ExecuteContext=@SequenceExecution_1.D)
  ]}
 ContainedGraph {nodes [
   n1 = RigVMFunctionEntryNode()
   SequenceExecution = &RigVMFunction_Sequence(ExecuteContext=@n1.ExecuteContext)
   SequenceExecution_1 = &RigVMFunction_Sequence(ExecuteContext=@SequenceExecution.B)
   SequenceExecution_2 = &RigVMFunction_Sequence(ExecuteContext=@SequenceExecution_1.B)
   RigVMFunctionReturnNode(A=@SequenceExecution.A, B=@SequenceExecution_1.A, C=@SequenceExecution_2.A, D=@SequenceExecution_2.B)
  ]}
 ContainedGraph {nodes [
   n1 = RigVMFunctionEntryNode()
   SequenceExecution_1 = &RigVMFunction_Sequence(ExecuteContext=@n1.ExecuteContext)
   SequenceExecution_1_1 = &RigVMFunction_Sequence(ExecuteContext=@SequenceExecution_1.B)
   SequenceExecution_1_2 = &RigVMFunction_Sequence(ExecuteContext=@SequenceExecution_1_1.B)
   RigVMFunctionReturnNode(A=@SequenceExecution_1.A, B=@SequenceExecution_1_1.A, C=@SequenceExecution_1_2.A, D=@SequenceExecution_1_2.B)
  ]}
 ContainedGraph {nodes [
   n1 = RigVMFunctionEntryNode()
   n2 = RigVMDispatchNode(Num=4, Array=RigVMVariableNode() {variable=Outputs}.Value, ExecuteContext=@n1.Execute) {template="DISPATCH_RigVMDispatch_ArraySetNum(io Array,in Num)"}
   n3 = RigVMVariableNode() {variable=Point 0}
   > n4 = RigVMDispatchNode(Index=0, Array=@n2.Array, Element=@n3.Value) {template="DISPATCH_RigVMDispatch_ArraySetAtIndex(io Array,in Index,in Element)"}
   n5 = RigVMVariableNode() {variable=Point 3}
   n6 = &RigVMFunction_MathVectorDistance(B=@n1.Point 0, A=@n1.Point 1)
   n7 = &RigVMFunction_MathFloatAdd(A=@n6.Result, B=&RigVMFunction_MathVectorDistance(A=@n1.Point 1, B=@n1.Point 2).Result) {template="Add::Execute(in A,in B,out Result)"}
   n8 = &RigVMFunction_MathFloatAdd(A=@n7.Result, B=&RigVMFunction_MathVectorDistance(A=@n1.Point 2, B=@n1.Point 3).Result) {template="Add::Execute(in A,in B,out Result)"}
   n9 = RigVMVariableNode() {variable=Auto Weights}
   n10 = RigVMDispatchNode(
    True=&RigVMFunction_MathFloatDiv(B=@n8.Result, A=@n6.Result) {template="Divide::Execute(in A,in B,out Result)"}.Result
    Condition=@n9.Value
    False=RigVMVariableNode() {variable=Weight 1}.Value
   ) {template="DISPATCH_RigVMDispatch_If(in Condition,in True,in False,out Result)"}
   b3t1 = RigVMCollapseNode(t=@n10.Result)
   n11 = RigVMVariableNode() {variable=Point 0}
   b0t1 = RigVMCollapseNode(t=@n10.Result)
   n12 = /Script/RigVM.RigVMFunction_MathVectorScale(Value=@n11.Value, Factor=@b0t1.Result) {is=struct RigVMUnitNode, template="Scale::Execute(in Value,in Factor,out Result)"}
   n13 = RigVMVariableNode() {variable=Point 1}
   n14 = RigVMDispatchNode(
    True=&RigVMFunction_MathFloatDiv(A=@n7.Result, B=@n8.Result) {template="Divide::Execute(in A,in B,out Result)"}.Result
    Condition=@n9.Value
    False=RigVMVariableNode() {variable=Weight 2}.Value
   ) {template="DISPATCH_RigVMDispatch_If(in Condition,in True,in False,out Result)"}
   b3t2 = RigVMCollapseNode(t=@n14.Result)
   b0t2 = RigVMCollapseNode(t=@n14.Result)
   n15 = &RigVMFunction_MathVectorScale(Value=@n11.Value, Factor=@b0t2.Result) {template="Scale::Execute(in Value,in Factor,out Result)"}
   n16 = RigVMVariableNode() {variable=Point 2}
   b1t1 = RigVMCollapseNode(t=@n10.Result)
   b1t2 = RigVMCollapseNode(t=@n14.Result)
   b2t1 = RigVMCollapseNode(t=@n10.Result)
   b2t2 = RigVMCollapseNode(t=@n14.Result)
   Solve = RigVMCollapseNode(
    c=/Script/RigVM.RigVMFunction_MathVectorSub(
     B=&RigVMFunction_MathVectorScale(Value=@n5.Value, Factor=@b3t1.Result) {template="Scale::Execute(in Value,in Factor,out Result)"}.Result
     A=&RigVMFunction_MathVectorSub(B=@n12.Result, A=@n13.Value) {template="Subtract::Execute(in A,in B,out Result)"}.Result
    ) {is=struct RigVMUnitNode, template="Subtract::Execute(in A,in B,out Result)"}.Result
    f=&RigVMFunction_MathVectorSub(
     B=&RigVMFunction_MathVectorScale(Value=@n5.Value, Factor=@b3t2.Result) {template="Scale::Execute(in Value,in Factor,out Result)"}.Result
     A=&RigVMFunction_MathVectorSub(B=@n15.Result, A=@n16.Value) {template="Subtract::Execute(in A,in B,out Result)"}.Result
    ) {template="Subtract::Execute(in A,in B,out Result)"}.Result
    a=@b1t1.Result
    d=@b1t2.Result
    b=@b2t1.Result
    e=@b2t2.Result
   )
   > n17 = RigVMDispatchNode(Index=1, Array=@n4.Array, Element=@Solve.i) {template="DISPATCH_RigVMDispatch_ArraySetAtIndex(io Array,in Index,in Element)"}
   > n18 = RigVMDispatchNode(Index=2, Array=@n17.Array, Element=@Solve.j) {template="DISPATCH_RigVMDispatch_ArraySetAtIndex(io Array,in Index,in Element)"}
   n19 = RigVMVariableNode() {variable=Point 3}
   > n20 = RigVMDispatchNode(Index=3, Array=@n18.Array, Element=@n19.Value) {template="DISPATCH_RigVMDispatch_ArraySetAtIndex(io Array,in Index,in Element)"}
   RigVMFunctionReturnNode(
    Control Points=@n20.Array
    Execute=@n20.ExecuteContext
    Spline=/Script/ControlRigSpline.RigUnit_ControlRigSplineFromPoints(
     Points=((X=0,Y=0,Z=0),(X=0,Y=0,Z=0),(X=0,Y=0,Z=0),(X=0,Y=0,Z=0))
     SamplesPerSegment=16
     SplineMode=BSpline
     bClosed=False
     Points.1=@Solve.i
     Points.2=@Solve.j
     Points.0=@n3.Value
     Points.3=@n19.Value
    ) {is=struct RigVMUnitNode}.Spline
   )
  ]}
 ContainedGraph {nodes [
   RigVMFunctionEntryNode()
   n1 = &RigVMFunction_MathFloatDiv(
    A=&RigVMFunction_MathFloatMul(A=RigVMVariableNode() {variable=a}.Value, B=RigVMVariableNode() {variable=e}.Value) {template="Multiply::Execute(in A,in B,out Result)"}.Result
    B=RigVMVariableNode() {variable=d}.Value
   ) {template="Divide::Execute(in A,in B,out Result)"}
   n2 = RigVMVariableNode() {variable=b}
   n3 = &RigVMFunction_MathFloatDiv(A=RigVMVariableNode() {variable=a}.Value, B=RigVMVariableNode() {variable=d}.Value) {template="Divide::Execute(in A,in B,out Result)"}
   n4 = RigVMVariableNode() {variable=f}
   n5 = &RigVMFunction_MathVectorScale(
    Factor=&RigVMFunction_MathFloatDiv(A=1, B=&RigVMFunction_MathFloatSub(B=@n1.Result, A=@n2.Value) {template="Subtract::Execute(in A,in B,out Result)"}.Result) {template="Divide::Execute(in A,in B,out Result)"}.Result
    Value=&RigVMFunction_MathVectorSub(
     B=&RigVMFunction_MathVectorScale(Factor=@n3.Result, Value=@n4.Value) {template="Scale::Execute(in Value,in Factor,out Result)"}.Result
     A=RigVMVariableNode() {variable=c}.Value
    ) {template="Subtract::Execute(in A,in B,out Result)"}.Result
   ) {template="Scale::Execute(in Value,in Factor,out Result)"}
   n6 = RigVMVariableNode() {variable=a}
   n7 = &RigVMFunction_MathVectorScale(Value=@n5.Result, Factor=RigVMVariableNode() {variable=b}.Value) {template="Scale::Execute(in Value,in Factor,out Result)"}
   n8 = RigVMVariableNode() {variable=c}
   RigVMFunctionReturnNode(
    j=@n5.Result
    i=&RigVMFunction_MathVectorScale(
     Factor=&RigVMFunction_MathFloatDiv(A=1, B=@n6.Value) {template="Divide::Execute(in A,in B,out Result)"}.Result
     Value=&RigVMFunction_MathVectorSub(B=@n7.Result, A=@n8.Value) {template="Subtract::Execute(in A,in B,out Result)"}.Result
    ) {template="Scale::Execute(in Value,in Factor,out Result)"}.Result
   )
  ]}
 ContainedGraph {nodes [
   n1 = RigVMFunctionEntryNode()
   n2 = &RigVMFunction_MathFloatSub(A=1, B=@n1.t) {template="Subtract::Execute(in A,in B,out Result)"}
   n3 = &RigVMFunction_MathFloatMul(A=@n2.Result, B=@n2.Result) {template="Multiply::Execute(in A,in B,out Result)"}
   RigVMFunctionReturnNode(Result=&RigVMFunction_MathFloatMul(B=3, A=&RigVMFunction_MathFloatMul(A=@n1.t, B=@n3.Result) {template="Multiply::Execute(in A,in B,out Result)"}.Result) {template="Multiply::Execute(in A,in B,out Result)"}.Result)
  ]}
 ContainedGraph {nodes [
   n1 = RigVMFunctionEntryNode()
   n2 = &RigVMFunction_MathFloatMul(A=@n1.t, B=@n1.t) {template="Multiply::Execute(in A,in B,out Result)"}
   n3 = &RigVMFunction_MathFloatSub(A=1, B=@n1.t) {template="Subtract::Execute(in A,in B,out Result)"}
   RigVMFunctionReturnNode(Result=&RigVMFunction_MathFloatMul(B=3, A=&RigVMFunction_MathFloatMul(B=@n2.Result, A=@n3.Result) {template="Multiply::Execute(in A,in B,out Result)"}.Result) {template="Multiply::Execute(in A,in B,out Result)"}.Result)
  ]}
 ContainedGraph {nodes [
   n1 = RigVMFunctionEntryNode()
   n2 = &RigVMFunction_MathFloatSub(A=1, B=@n1.t) {template="Subtract::Execute(in A,in B,out Result)"}
   n3 = &RigVMFunction_MathFloatMul(A=@n2.Result, B=@n2.Result) {template="Multiply::Execute(in A,in B,out Result)"}
   RigVMFunctionReturnNode(Result=&RigVMFunction_MathFloatMul(B=3, A=&RigVMFunction_MathFloatMul(A=@n1.t, B=@n3.Result) {template="Multiply::Execute(in A,in B,out Result)"}.Result) {template="Multiply::Execute(in A,in B,out Result)"}.Result)
  ]}
 ContainedGraph {nodes [
   n1 = RigVMFunctionEntryNode()
   n2 = &RigVMFunction_MathFloatMul(A=@n1.t, B=@n1.t) {template="Multiply::Execute(in A,in B,out Result)"}
   n3 = &RigVMFunction_MathFloatSub(A=1, B=@n1.t) {template="Subtract::Execute(in A,in B,out Result)"}
   RigVMFunctionReturnNode(Result=&RigVMFunction_MathFloatMul(B=3, A=&RigVMFunction_MathFloatMul(B=@n2.Result, A=@n3.Result) {template="Multiply::Execute(in A,in B,out Result)"}.Result) {template="Multiply::Execute(in A,in B,out Result)"}.Result)
  ]}
 ContainedGraph {nodes [
   n1 = RigVMFunctionEntryNode()
   RigVMFunctionReturnNode(Result=/Script/RigVM.RigVMFunction_MathFloatPow(B=3, A=&RigVMFunction_MathFloatSub(A=1, B=@n1.t) {template="Subtract::Execute(in A,in B,out Result)"}.Result) {is=struct RigVMUnitNode, template="Power::Execute(in A,in B,out Result)"}.Result)
  ]}
 ContainedGraph {nodes [RigVMFunctionReturnNode(Result=&RigVMFunction_MathFloatPow(B=3, A=RigVMFunctionEntryNode().t) {template="Power::Execute(in A,in B,out Result)"}.Result)]}
 ContainedGraph {nodes [
   n1 = RigVMFunctionEntryNode()
   RigVMFunctionReturnNode(Result=&RigVMFunction_MathFloatPow(B=3, A=&RigVMFunction_MathFloatSub(A=1, B=@n1.t) {template="Subtract::Execute(in A,in B,out Result)"}.Result) {template="Power::Execute(in A,in B,out Result)"}.Result)
  ]}
 ContainedGraph {nodes [RigVMFunctionReturnNode(Result=&RigVMFunction_MathFloatPow(B=3, A=RigVMFunctionEntryNode().t) {template="Power::Execute(in A,in B,out Result)"}.Result)]}
 ContainedGraph {nodes [
   n1 = &RigUnit_GetTransformItemArray(Space=GlobalSpace, Items=RigVMVariableNode() {variable=Bones}.Value)
   SequenceExecution = RigVMAggregateNode(ExecuteContext=@n3.ExecuteContext)
   BSpline Passing thru 4 Points_1 = RigVMFunctionReferenceNode(
    Auto Weights=False
    Weight 1=0.333
    Weight 2=0.666
    Point 0=RigVMDispatchNode(Index=0, Array=@n1.Transforms) {template="DISPATCH_RigVMDispatch_ArrayGetAtIndex(in Array,in Index,out Element)"}.Element.Translation
    Point 1=RigVMDispatchNode(Index=1, Array=@n1.Transforms) {template="DISPATCH_RigVMDispatch_ArrayGetAtIndex(in Array,in Index,out Element)"}.Element.Translation
    Point 2=RigVMDispatchNode(Index=2, Array=@n1.Transforms) {template="DISPATCH_RigVMDispatch_ArrayGetAtIndex(in Array,in Index,out Element)"}.Element.Translation
    Point 3=RigVMDispatchNode(Index=3, Array=@n1.Transforms) {template="DISPATCH_RigVMDispatch_ArrayGetAtIndex(in Array,in Index,out Element)"}.Element.Translation
    Execute=@SequenceExecution.A
   )
   n2 = RigVMDispatchNode(ExecuteContext=@"BSpline Passing thru 4 Points_1".Execute, Array=RigVMVariableNode() {variable=Bones}.Value) {template="DISPATCH_RigVMDispatch_ArrayIterator(in Array,out Element,out Index,out Count,out Ratio)"}
   > &RigUnit_SetTransform(
    Value=(Rotation=(X=0,Y=0,Z=0,W=1),Translation=(X=0,Y=0,Z=0),Scale3D=(X=1,Y=1,Z=1))
    Weight=1
    bInitial=true
    bPropagateToChildren=false
    Item=RigVMDispatchNode(Index=@n2.Index, Array=RigVMVariableNode() {variable=FK Controls}.Value) {template="DISPATCH_RigVMDispatch_ArrayGetAtIndex(in Array,in Index,out Element)"}.Element
    Value.Translation=&RigUnit_GetTransform(Item=@n2.Element).Transform.Translation
   )
   > &RigUnit_SetTransform(
    Value=(Rotation=(X=0,Y=0,Z=0,W=1),Translation=(X=0,Y=0,Z=0),Scale3D=(X=1,Y=1,Z=1))
    Weight=1
    bInitial=true
    bPropagateToChildren=false
    Value.Translation=RigVMDispatchNode(Index=@n2.Index, Array=@"BSpline Passing thru 4 Points_1".Control Points) {template="DISPATCH_RigVMDispatch_ArrayGetAtIndex(in Array,in Index,out Element)"}.Element
    Item=RigVMDispatchNode(Index=@n2.Index, Array=RigVMVariableNode() {variable=IK Controls}.Value) {template="DISPATCH_RigVMDispatch_ArrayGetAtIndex(in Array,in Index,out Element)"}.Element
   )
   n3 = RigVMFunctionEntryNode()
   n4 = RigVMDispatchNode(ExecuteContext=@n7.ExecuteContext, Array=RigVMVariableNode() {variable=Bones}.Value) {template="DISPATCH_RigVMDispatch_ArrayIterator(in Array,out Element,out Index,out Count,out Ratio)"}
   Spline From Element Setup = RigVMCollapseNode(Items=RigVMVariableNode() {variable=IK Controls}.Value)
   > RigVMDispatchNode(
    Element=&RigVMFunction_MathTransformMakeRelative(
     Global=&RigUnit_GetTransform(Item=@n4.Element).Transform
     Parent=/Script/ControlRigSpline.RigUnit_TransformFromControlRigSpline(UpVector=(X=0,Y=1,Z=0), U=@n4.Ratio, Spline=@"Spline From Element Setup".Spline) {is=struct RigVMUnitNode}.Transform
    ).Local
    Array=RigVMVariableNode() {variable=IK Offsets}.Value
   ) {template="DISPATCH_RigVMDispatch_ArrayAdd(io Array,in Element,out Index)"}
   RigVMFunctionReturnNode(
    Length=/Script/ControlRigSpline.RigUnit_GetLengthControlRigSpline(Spline=@"BSpline Passing thru 4 Points_1".Spline) {is=struct RigVMUnitNode}.Length
    ExecuteContext=@SequenceExecution.D
    Spline Offsets=RigVMVariableNode() {variable=IK Offsets}.Value
   )
   n5 = RigVMVariableNode() {variable=Bones}
   n6 = &RigVMFunction_MathIntSub(B=1, A=RigVMDispatchNode(Array=@n5.Value) {template="DISPATCH_RigVMDispatch_ArrayGetNum(in Array,out Num)"}.Num) {template="Subtract::Execute(in A,in B,out Result)"}
   &RigUnit_SetTransform(
    Weight=1
    bInitial=true
    bPropagateToChildren=false
    Value=&RigUnit_GetTransform(Item=RigVMDispatchNode(Index=@n6.Result, Array=@n5.Value) {template="DISPATCH_RigVMDispatch_ArrayGetAtIndex(in Array,in Index,out Element)"}.Element).Transform
    ExecutePin=@SequenceExecution.C
    Item=RigVMVariableNode() {variable=Chest Space}.Value
   )
   n7 = RigVMDispatchNode(ExecuteContext=@SequenceExecution.B, Array=RigVMVariableNode() {variable=IK Offsets}.Value) {template="DISPATCH_RigVMDispatch_ArrayReset(io Array)"}
  ]}
 ContainedGraph {nodes [
   n1 = &RigUnit_GetTransformItemArray(Space=GlobalSpace, Items=RigVMFunctionEntryNode().Items)
   RigVMFunctionReturnNode(Spline=&RigUnit_ControlRigSplineFromPoints(
     Points=((X=0,Y=0,Z=0),(X=0,Y=0,Z=0),(X=0,Y=0,Z=0),(X=0,Y=0,Z=0))
     SamplesPerSegment=16
     SplineMode=BSpline
     bClosed=False
     Points.0=RigVMDispatchNode(Index=0, Array=@n1.Transforms) {template="DISPATCH_RigVMDispatch_ArrayGetAtIndex(in Array,in Index,out Element)"}.Element.Translation
     Points.1=RigVMDispatchNode(Index=1, Array=@n1.Transforms) {template="DISPATCH_RigVMDispatch_ArrayGetAtIndex(in Array,in Index,out Element)"}.Element.Translation
     Points.2=RigVMDispatchNode(Index=2, Array=@n1.Transforms) {template="DISPATCH_RigVMDispatch_ArrayGetAtIndex(in Array,in Index,out Element)"}.Element.Translation
     Points.3=RigVMDispatchNode(Index=3, Array=@n1.Transforms) {template="DISPATCH_RigVMDispatch_ArrayGetAtIndex(in Array,in Index,out Element)"}.Element.Translation
    ).Spline)
  ]}
 ContainedGraph {nodes [
   n1 = RigVMFunctionEntryNode()
   SequenceExecution = &RigVMFunction_Sequence(ExecuteContext=@n1.ExecuteContext)
   SequenceExecution_1 = &RigVMFunction_Sequence(ExecuteContext=@SequenceExecution.B)
   SequenceExecution_2 = &RigVMFunction_Sequence(ExecuteContext=@SequenceExecution_1.B)
   RigVMFunctionReturnNode(A=@SequenceExecution.A, B=@SequenceExecution_1.A, C=@SequenceExecution_2.A, D=@SequenceExecution_2.B)
  ]}
 ContainedGraph {nodes [
   RigVMFunctionEntryNode()
   n1 = RigVMVariableNode() {variable=IKFK Switch Control}
   > RigVMVariableNode(Value=&RigVMFunction_MathFloatGreater(B=0.5, A=&RigUnit_GetTransform(Space=LocalSpace, Item=@n1.Value).Transform.Translation.X) {template="Greater::Execute(in A,in B,out Result)"}.Result) {variable=IK Mode}
   > SequenceExecution = RigVMAggregateNode()
   n2 = RigVMDispatchNode(
    Index=0
    Array=RigVMDispatchNode(Array=RigVMVariableNode() {variable=FK Controls}.Value) {template="DISPATCH_RigVMDispatch_ArrayClone(in Array,out Clone)"}.Clone
    ExecuteContext=@SequenceExecution.A
   ) {template="DISPATCH_RigVMDispatch_ArrayRemove(io Array,in Index)"}
   > n3 = RigVMDispatchNode(
    Index=0
    Array=RigVMDispatchNode(Array=RigVMVariableNode() {variable=IK Controls}.Value) {template="DISPATCH_RigVMDispatch_ArrayClone(in Array,out Clone)"}.Clone
   ) {template="DISPATCH_RigVMDispatch_ArrayRemove(io Array,in Index)"}
   > Set IKFK Visibility = RigVMFunctionReferenceNode(Show Controls=true, FK Controls=@n2.Array, IK Controls=@n3.Array, IKFK Switch=RigVMVariableNode() {variable=IK Mode}.Value)
   n4 = RigVMDispatchNode(Index=0, Array=RigVMVariableNode() {variable=FK Controls}.Value) {template="DISPATCH_RigVMDispatch_ArrayGetAtIndex(in Array,in Index,out Element)"}
   n5 = RigVMVariableNode() {variable=Pelvis Bone}
   &RigUnit_SetTransform(
    Weight=1
    Value=&RigUnit_ProjectTransformToNewParent(bChildInitial=True, bOldParentInitial=True, NewParent=@n4.Element, OldParent=@n4.Element, Child=@n5.Value).Transform
    ExecutePin=@SequenceExecution.B
    Item=@n5.Value
   )
   n6 = &RigVMFunction_ControlFlowBranch(ExecuteContext=@SequenceExecution.C, Condition=RigVMVariableNode() {variable=IK Mode}.Value)
   FK Point At_1 = RigVMFunctionReferenceNode(
    Initial=false
    Point At Axis=(X=1,Y=0,Z=0)
    Propagate To Children=true
    Up Vector Axis=(X=0,Y=0,Z=-1)
    Up Vector Target=(X=1,Y=0,Z=0)
    ExecuteContext=@n6.False
    Bones=RigVMVariableNode() {variable=Bones}.Value
    FK Controls=RigVMVariableNode() {variable=FK Controls}.Value
   )
   Spine Match IK onto FK = RigVMFunctionReferenceNode(
    Execute=@"FK Point At_1".ExecuteContext
    Bones=RigVMVariableNode() {variable=Bones}.Value
    Offsets=RigVMVariableNode() {variable=Spine Offsets}.Value
    IK Controls=RigVMVariableNode() {variable=IK Controls}.Value
   )
   SequenceExecution_1 = RigVMAggregateNode(ExecuteContext=@SequenceExecution.D)
   n7 = RigVMVariableNode() {variable=Bones}
   n8 = RigVMDispatchNode(
    Index=&RigVMFunction_MathIntSub(B=1, A=RigVMDispatchNode(Array=@n7.Value) {template="DISPATCH_RigVMDispatch_ArrayGetNum(in Array,out Num)"}.Num) {template="Subtract::Execute(in A,in B,out Result)"}.Result
    Array=@n7.Value
   ) {template="DISPATCH_RigVMDispatch_ArrayGetAtIndex(in Array,in Index,out Element)"}
   n9 = RigVMVariableNode() {variable=Chest Space}
   &RigUnit_SetTransform(
    Weight=1
    Value=&RigUnit_ProjectTransformToNewParent(bChildInitial=True, bOldParentInitial=True, NewParent=@n8.Element, OldParent=@n8.Element, Child=@n9.Value).Transform
    ExecutePin=@SequenceExecution_1.A
    Item=@n9.Value
   )
   Set IKFK Keys_1 = RigVMFunctionReferenceNode(
    ExecuteContext=@SequenceExecution_1.B
    Previous IK Mode=RigVMVariableNode() {variable=Previous IK Mode}.Value
    FK Controls=RigVMVariableNode() {variable=FK Controls}.Value
    "IK Controls "=RigVMVariableNode() {variable=IK Controls}.Value
    Switch Control=RigVMVariableNode() {variable=IKFK Switch Control}.Value
    IK Mode=RigVMVariableNode() {variable=IK Mode}.Value
   )
   RigVMFunctionReturnNode(ExecuteContext=@SequenceExecution_1.D)
   Spine IK Solve = RigVMCollapseNode(
    Completed=@n6.True
    Default Length=RigVMVariableNode() {variable=Default Length}.Value
    Enable Stretch=RigVMVariableNode() {variable="Enable Stretch "}.Value
    Bones=RigVMVariableNode() {variable=Bones}.Value
    IK Controls=RigVMVariableNode() {variable=IK Controls}.Value
    Spine Offsets=RigVMVariableNode() {variable=Spine Offsets}.Value
   )
   n10 = RigVMDispatchNode(
    Index=0
    Array=RigVMDispatchNode(Array=RigVMVariableNode() {variable=FK Controls}.Value) {template="DISPATCH_RigVMDispatch_ArrayClone(in Array,out Clone)"}.Clone
    ExecuteContext=@"Spine IK Solve".Completed
   ) {template="DISPATCH_RigVMDispatch_ArrayRemove(io Array,in Index)"}
   > n11 = RigVMDispatchNode(
    Index=0
    Array=RigVMDispatchNode(Array=RigVMVariableNode() {variable=Bones}.Value) {template="DISPATCH_RigVMDispatch_ArrayClone(in Array,out Clone)"}.Clone
   ) {template="DISPATCH_RigVMDispatch_ArrayRemove(io Array,in Index)"}
   > Project to New Parent = RigVMFunctionReferenceNode(Propagate to Children=True, Children=@n10.Array, Parents=@n11.Array)
  ]}
 ContainedGraph {nodes [
   n1 = RigVMDispatchNode(ExecuteContext=(), Array=RigVMVariableNode() {variable=Bones}.Value) {template="DISPATCH_RigVMDispatch_ArrayIterator(in Array,out Element,out Index,out Count,out Ratio)"}
   n2 = &RigVMFunction_ControlFlowBranch(
    ExecuteContext=@n1.ExecuteContext
    Condition=/Script/RigVM.RigVMFunction_MathIntEquals(A=@n1.Index, B=&RigVMFunction_MathIntSub(B=1, A=@n1.Count) {template="Subtract::Execute(in A,in B,out Result)"}.Result) {is=struct RigVMUnitNode}.Result
   )
   n3 = &RigUnit_AimBoneMath(
    DebugSettings=(Scale=10,WorldOffset=(Rotation=(X=0,Y=0,Z=0,W=1),Translation=(X=0,Y=0,Z=0),Scale3D=(X=1,Y=1,Z=1)))
    Primary=(Weight=1,Axis=(X=1,Y=0,Z=0),Target=(X=0,Y=0,Z=0),Kind=Location,Space=(Type=Bone))
    Secondary=(Weight=1,Axis=(X=0,Y=0,Z=1),Target=(X=0,Y=0,Z=1),Kind=Direction,Space=(Type=Bone))
    Weight=1
    Primary.Space=RigVMDispatchNode(
     Index=&RigVMFunction_MathIntAdd(B=1, A=@n1.Index) {template="Add::Execute(in A,in B,out Result)"}.Result
     Array=RigVMVariableNode() {variable=Bones}.Value
    ) {template="DISPATCH_RigVMDispatch_ArrayGetAtIndex(in Array,in Index,out Element)"}.Element
    Secondary.Space=@n1.Element
    InputTransform=&RigUnit_GetTransform(Item=@n1.Element).Transform
   )
   &RigUnit_SetTransform(Weight=1, bPropagateToChildren=false, Value=@n3.Result, Item=@n1.Element, ExecutePin=@n2.False)
   n4 = RigVMFunctionEntryNode()
   n5 = /Script/ControlRig.RigUnit_SequenceExecution(ExecuteContext=@n4.Completed) {is=struct RigVMUnitNode}
   n6 = RigVMDispatchNode(ExecuteContext=@n5.A, Array=RigVMVariableNode() {variable=Bones}.Value) {template="DISPATCH_RigVMDispatch_ArrayIterator(in Array,out Element,out Index,out Count,out Ratio)"}
   Is First or Last = RigVMCollapseNode(Count=@n6.Count, Index=@n6.Index)
   n7 = RigVMVariableNode() {variable=IK Controls}
   Spline From Element = RigVMCollapseNode(Items=@n7.Value)
   n8 = &RigUnit_GetLengthControlRigSpline(Spline=@"Spline From Element".Spline)
   n9 = RigVMVariableNode() {variable=Default Length}
   n10 = RigVMDispatchNode(
    True=1
    False=/Script/RigVM.RigVMFunction_MathFloatMin(B=1, A=&RigVMFunction_MathFloatDiv(B=@n8.Length, A=@n9.Value) {template="Divide::Execute(in A,in B,out Result)"}.Result) {is=struct RigVMUnitNode, template="Minimum::Execute(in A,in B,out Result)"}.Result
    Condition=RigVMVariableNode() {variable=Enable Stretch}.Value
   ) {template="DISPATCH_RigVMDispatch_If(in Condition,in True,in False,out Result)"}
   UpVector = RigVMCollapseNode(
    Start=RigVMDispatchNode(Index=0, Array=@n7.Value) {template="DISPATCH_RigVMDispatch_ArrayGetAtIndex(in Array,in Index,out Element)"}.Element
    End=RigVMDispatchNode(Index=3, Array=@n7.Value) {template="DISPATCH_RigVMDispatch_ArrayGetAtIndex(in Array,in Index,out Element)"}.Element
    T=@n6.Ratio
   )
   n11 = &RigVMFunction_MathTransformMakeAbsolute(
    Local=RigVMDispatchNode(Index=@n6.Index, Array=RigVMVariableNode() {variable=Spine Offsets}.Value) {template="DISPATCH_RigVMDispatch_ArrayGetAtIndex(in Array,in Index,out Element)"}.Element
    Parent=&RigUnit_TransformFromControlRigSpline(
     Roll=0
     U=&RigVMFunction_MathFloatMul(B=@n6.Ratio, A=@n10.Result) {template="Multiply::Execute(in A,in B,out Result)"}.Result
     Spline=@"Spline From Element".Spline
     UpVector=@UpVector.Up Vector
    ).Transform
   ) {template="Make Absolute::Execute(in Local,in Parent,out Global)"}
   n12 = RigVMDispatchNode(Index=@n6.Index, Array=RigVMVariableNode() {variable=IK Controls}.Value) {template="DISPATCH_RigVMDispatch_ArrayGetAtIndex(in Array,in Index,out Element)"}
   > &RigUnit_SetTransform(
    Weight=1
    Item=@n6.Element
    Value=RigVMDispatchNode(
     True=(Rotation=(X=0,Y=0,Z=0,W=1),Translation=(X=0,Y=0,Z=0),Scale3D=(X=0,Y=0,Z=0))
     Condition=@"Is First or Last".Result
     False=@n11.Global
     True.Scale3D=@n11.Global.Scale3D
     True.Translation=@n11.Global.Translation
     True.Rotation=&RigUnit_ProjectTransformToNewParent(bChildInitial=True, bOldParentInitial=True, NewParent=@n12.Element, OldParent=@n12.Element, Child=@n6.Element).Transform.Rotation
    ) {template="DISPATCH_RigVMDispatch_If(in Condition,in True,in False,out Result)"}.Result
   )
   RigVMFunctionReturnNode(Completed=@n5.D)
  ]}
 ContainedGraph {nodes [
   n1 = &RigUnit_GetTransformItemArray(Space=GlobalSpace, Items=RigVMFunctionEntryNode().Items)
   RigVMFunctionReturnNode(Spline=&RigUnit_ControlRigSplineFromPoints(
     Points=((X=0,Y=0,Z=0),(X=0,Y=0,Z=0),(X=0,Y=0,Z=0),(X=0,Y=0,Z=0))
     SamplesPerSegment=16
     SplineMode=BSpline
     bClosed=False
     Points.0=RigVMDispatchNode(Index=0, Array=@n1.Transforms) {template="DISPATCH_RigVMDispatch_ArrayGetAtIndex(in Array,in Index,out Element)"}.Element.Translation
     Points.1=RigVMDispatchNode(Index=1, Array=@n1.Transforms) {template="DISPATCH_RigVMDispatch_ArrayGetAtIndex(in Array,in Index,out Element)"}.Element.Translation
     Points.2=RigVMDispatchNode(Index=2, Array=@n1.Transforms) {template="DISPATCH_RigVMDispatch_ArrayGetAtIndex(in Array,in Index,out Element)"}.Element.Translation
     Points.3=RigVMDispatchNode(Index=3, Array=@n1.Transforms) {template="DISPATCH_RigVMDispatch_ArrayGetAtIndex(in Array,in Index,out Element)"}.Element.Translation
    ).Spline)
  ]}
 ContainedGraph {nodes [
   n1 = RigVMFunctionEntryNode()
   n2 = &RigUnit_GetTransform(Item=@n1.Start)
   n3 = &RigUnit_GetTransform(Item=@n1.End)
   RigVMFunctionReturnNode(Up Vector=/Script/RigVM.RigVMFunction_MathQuaternionRotateVector(
     Vector=(X=0,Y=1,Z=0)
     Transform=/Script/RigVM.RigVMFunction_MathQuaternionSlerp(T=@n1.T, A=@n2.Transform.Rotation, B=@n3.Transform.Rotation) {is=struct RigVMUnitNode, template="Interpolate::Execute(in A,in B,in T,out Result)"}.Result
    ) {is=struct RigVMUnitNode, template="Rotate Vector::Execute(in Transform,in Vector,out Result)"}.Result)
  ]}
 ContainedGraph {nodes [
   n1 = RigVMFunctionEntryNode()
   n2 = &RigVMFunction_MathIntSub(B=1, A=@n1.Count) {template="Subtract::Execute(in A,in B,out Result)"}
   RigVMFunctionReturnNode(Result=&RigVMFunction_MathBoolOr(
     A=&RigVMFunction_MathIntEquals(A=@n1.Index).Result
     B=&RigVMFunction_MathIntEquals(A=@n1.Index, B=@n2.Result).Result
    ).Result)
  ]}
 ContainedGraph {nodes [
   n1 = RigVMFunctionEntryNode()
   SequenceExecution = &RigVMFunction_Sequence(ExecuteContext=@n1.ExecuteContext)
   SequenceExecution_1 = &RigVMFunction_Sequence(ExecuteContext=@SequenceExecution.B)
   SequenceExecution_2 = &RigVMFunction_Sequence(ExecuteContext=@SequenceExecution_1.B)
   RigVMFunctionReturnNode(A=@SequenceExecution.A, B=@SequenceExecution_1.A, C=@SequenceExecution_2.A, D=@SequenceExecution_2.B)
  ]}
 ContainedGraph {nodes [
   n1 = RigVMFunctionEntryNode()
   SequenceExecution_1 = &RigVMFunction_Sequence(ExecuteContext=@n1.ExecuteContext)
   SequenceExecution_1_1 = &RigVMFunction_Sequence(ExecuteContext=@SequenceExecution_1.B)
   SequenceExecution_1_2 = &RigVMFunction_Sequence(ExecuteContext=@SequenceExecution_1_1.B)
   RigVMFunctionReturnNode(A=@SequenceExecution_1.A, B=@SequenceExecution_1_1.A, C=@SequenceExecution_1_2.A, D=@SequenceExecution_1_2.B)
  ]}
 ContainedGraph {nodes [
   RigVMFunctionEntryNode()
   > SequenceExecution = RigVMAggregateNode()
   n1 = RigVMVariableNode() {variable=Bones}
   n2 = RigVMDispatchNode(Array=@n1.Value) {template="DISPATCH_RigVMDispatch_ArrayGetNum(in Array,out Num)"}
   n3 = &RigUnit_GetTransform(Item=RigVMDispatchNode(Index=&RigVMFunction_MathIntSub(B=1, A=@n2.Num) {template="Subtract::Execute(in A,in B,out Result)"}.Result, Array=@n1.Value) {template="DISPATCH_RigVMDispatch_ArrayGetAtIndex(in Array,in Index,out Element)"}.Element)
   &RigUnit_SetTransform(
    Value=(Rotation=(X=0,Y=0,Z=0,W=1),Translation=(X=0,Y=0,Z=0),Scale3D=(X=1,Y=1,Z=1))
    Weight=1
    bInitial=true
    Value.Translation=@n3.Transform.Translation
    ExecutePin=@SequenceExecution.A
    Item=RigVMVariableNode() {variable=Head FK Space}.Value
   )
   > &RigUnit_SetTransform(
    Value=(Rotation=(X=0,Y=0,Z=0,W=1),Translation=(X=0,Y=0,Z=0),Scale3D=(X=1,Y=1,Z=1))
    Weight=1
    bInitial=true
    Value.Translation=@n3.Transform.Translation
    Item=RigVMVariableNode() {variable=Head IK Space}.Value
   )
   n4 = RigVMDispatchNode(ExecuteContext=@SequenceExecution.B, Array=RigVMVariableNode() {variable=FK Controls}.Value) {template="DISPATCH_RigVMDispatch_ArrayIterator(in Array,out Element,out Index,out Count,out Ratio)"}
   n5 = RigVMVariableNode() {variable=Bones}
   > &RigUnit_SetTransform(
    Value=(Rotation=(X=0,Y=0,Z=0,W=1),Translation=(X=0,Y=0,Z=0),Scale3D=(X=1,Y=1,Z=1))
    Weight=1
    bInitial=true
    Item=@n4.Element
    Value.Translation=&RigUnit_GetTransform(Item=RigVMDispatchNode(Index=@n4.Index, Array=@n5.Value) {template="DISPATCH_RigVMDispatch_ArrayGetAtIndex(in Array,in Index,out Element)"}.Element).Transform.Translation
   )
   n6 = RigVMVariableNode() {variable=Bones}
   n7 = &RigVMFunction_MathIntSub(B=1, A=RigVMDispatchNode(Array=@n6.Value) {template="DISPATCH_RigVMDispatch_ArrayGetNum(in Array,out Num)"}.Num) {template="Subtract::Execute(in A,in B,out Result)"}
   &RigUnit_SetTransform(
    Value=(Rotation=(X=0,Y=0,Z=0,W=1),Translation=(X=0,Y=0,Z=0),Scale3D=(X=1,Y=1,Z=1))
    Weight=1
    bInitial=true
    Value.Translation=&RigUnit_GetTransform(Item=RigVMDispatchNode(Index=@n7.Result, Array=@n6.Value) {template="DISPATCH_RigVMDispatch_ArrayGetAtIndex(in Array,in Index,out Element)"}.Element).Transform.Translation
    ExecutePin=@SequenceExecution.C
    Item=RigVMVariableNode() {variable=IK Control}.Value
   )
   n8 = &RigUnit_GetTransformItemArray(Space=GlobalSpace, Items=RigVMVariableNode() {variable=Bones}.Value)
   n9 = RigVMDispatchNode(Index=0, Array=@n8.Transforms) {template="DISPATCH_RigVMDispatch_ArrayGetAtIndex(in Array,in Index,out Element)"}
   n10 = RigVMDispatchNode(Index=1, Array=@n8.Transforms) {template="DISPATCH_RigVMDispatch_ArrayGetAtIndex(in Array,in Index,out Element)"}
   n11 = RigVMDispatchNode(Index=2, Array=@n8.Transforms) {template="DISPATCH_RigVMDispatch_ArrayGetAtIndex(in Array,in Index,out Element)"}
   RigVMFunctionReturnNode(
    Length=&RigUnit_GetLengthControlRigSpline(Spline=&RigUnit_ControlRigSplineFromPoints(
      Points=((X=0,Y=0,Z=0),(X=0,Y=0,Z=0),(X=0,Y=0,Z=0),(X=0,Y=0,Z=0))
      SamplesPerSegment=16
      SplineMode=BSpline
      bClosed=False
      Points.0=@n9.Element.Translation
      Points.1=@n9.Element.Translation
      Points.2=@n10.Element.Translation
      Points.3=@n11.Element.Translation
     ).Spline).Length
    ExecuteContext=@SequenceExecution.D
   )
  ]}
 ContainedGraph {nodes [
   n1 = RigVMFunctionEntryNode()
   SequenceExecution = &RigVMFunction_Sequence(ExecuteContext=@n1.ExecuteContext)
   SequenceExecution_1 = &RigVMFunction_Sequence(ExecuteContext=@SequenceExecution.B)
   SequenceExecution_2 = &RigVMFunction_Sequence(ExecuteContext=@SequenceExecution_1.B)
   RigVMFunctionReturnNode(A=@SequenceExecution.A, B=@SequenceExecution_1.A, C=@SequenceExecution_2.A, D=@SequenceExecution_2.B)
  ]}
 ContainedGraph {nodes [
   RigVMFunctionEntryNode()
   n1 = RigVMVariableNode() {variable=IKFK Switch Control}
   > RigVMVariableNode(Value=&RigVMFunction_MathFloatGreater(B=0.5, A=&RigUnit_GetTransform(Space=LocalSpace, Item=@n1.Value).Transform.Translation.X) {template="Greater::Execute(in A,in B,out Result)"}.Result) {variable=IK Mode}
   n2 = RigVMVariableNode() {variable=Head Orient Control}
   > RigVMVariableNode(Value=&RigVMFunction_MathFloatGreater(B=0.5, A=&RigUnit_GetTransform(Space=LocalSpace, Item=@n2.Value).Transform.Translation.X) {template="Greater::Execute(in A,in B,out Result)"}.Result) {variable=Local Space}
   > SequenceExecution = RigVMAggregateNode()
   Set IKFK Visibility = RigVMFunctionReferenceNode(
    IK Controls=((Type=Control,Name="head_ik_ctrl"))
    Show Controls=true
    ExecuteContext=@SequenceExecution.A
    IKFK Switch=RigVMVariableNode() {variable=IK Mode}.Value
    FK Controls=RigVMVariableNode() {variable=FK Controls}.Value
    IK Controls.0=RigVMVariableNode() {variable=IK Control}.Value
   )
   n3 = RigVMDispatchNode(Index=1, Array=RigVMVariableNode() {variable=FK Controls}.Value) {template="DISPATCH_RigVMDispatch_ArrayGetAtIndex(in Array,in Index,out Element)"}
   n4 = RigVMVariableNode() {variable=Head FK Space}
   n5 = &RigUnit_ProjectTransformToNewParent(bChildInitial=True, bOldParentInitial=True, NewParent=@n3.Element, OldParent=@n3.Element, Child=@n4.Value)
   n6 = RigVMVariableNode() {variable=Head Orient Reference}
   &RigUnit_SetTransform(
    Value=(Rotation=(X=0,Y=0,Z=0,W=1),Translation=(X=0,Y=0,Z=0),Scale3D=(X=1,Y=1,Z=1))
    Weight=1
    Value.Rotation=RigVMDispatchNode(
     False=@n5.Transform.Rotation
     True=&RigUnit_ProjectTransformToNewParent(bChildInitial=True, bOldParentInitial=True, Child=@n4.Value, NewParent=@n6.Value, OldParent=@n6.Value).Transform.Rotation
     Condition=RigVMVariableNode() {variable=Local Space}.Value
    ) {template="DISPATCH_RigVMDispatch_If(in Condition,in True,in False,out Result)"}.Result
    Value.Translation=@n5.Transform.Translation
    ExecutePin=@SequenceExecution.B
    Item=@n4.Value
   )
   n7 = RigVMVariableNode() {variable=Head IK Space}
   n8 = RigVMVariableNode() {variable=Head Orient Reference}
   n9 = &RigUnit_HierarchyGetParent(bDefaultParent=True, Child=@n7.Value)
   n10 = &RigUnit_ProjectTransformToNewParent(bChildInitial=True, bOldParentInitial=True, NewParent=@n9.Parent, OldParent=@n9.Parent, Child=@n7.Value)
   > &RigUnit_SetTransform(
    Value=(Rotation=(X=0,Y=0,Z=0,W=1),Translation=(X=0,Y=0,Z=0),Scale3D=(X=1,Y=1,Z=1))
    Weight=1
    Value.Rotation=RigVMDispatchNode(
     True=&RigUnit_ProjectTransformToNewParent(bChildInitial=True, bOldParentInitial=True, Child=@n7.Value, NewParent=@n8.Value, OldParent=@n8.Value).Transform.Rotation
     False=@n10.Transform.Rotation
     Condition=RigVMVariableNode() {variable=Local Space}.Value
    ) {template="DISPATCH_RigVMDispatch_If(in Condition,in True,in False,out Result)"}.Result
    Value.Translation=@n10.Transform.Translation
    Item=@n7.Value
   )
   n11 = &RigVMFunction_ControlFlowBranch(ExecuteContext=@SequenceExecution.C, Condition=RigVMVariableNode() {variable=IK Mode}.Value)
   Project to New Parents_2 = RigVMFunctionReferenceNode(
    Propagate to Children=True
    ExecuteContext=@n11.False
    Children=RigVMVariableNode() {variable=Bones}.Value
    Parents=RigVMVariableNode() {variable=FK Controls}.Value
   )
   n12 = RigVMVariableNode() {variable=Bones}
   n13 = RigVMDispatchNode(
    Index=&RigVMFunction_MathIntSub(B=1, A=RigVMDispatchNode(Array=@n12.Value) {template="DISPATCH_RigVMDispatch_ArrayGetNum(in Array,out Num)"}.Num) {template="Subtract::Execute(in A,in B,out Result)"}.Result
    Array=@n12.Value
   ) {template="DISPATCH_RigVMDispatch_ArrayGetAtIndex(in Array,in Index,out Element)"}
   n14 = RigVMVariableNode() {variable=IK Control}
   > &RigUnit_SetTransform(
    Weight=1
    Value=&RigUnit_ProjectTransformToNewParent(bChildInitial=True, bOldParentInitial=True, NewParent=@n13.Element, OldParent=@n13.Element, Child=@n14.Value).Transform
    Item=@n14.Value
   )
   Neck IK Neck = RigVMCollapseNode(
    ExecuteContext=@n11.True
    Length=RigVMVariableNode() {variable=Length}.Value
    Enable Stretch=RigVMVariableNode() {variable=Enable Stretch}.Value
    IK Control=RigVMVariableNode() {variable=IK Control}.Value
    Mid Null=RigVMVariableNode() {variable=Mid IK Null}.Value
    Bones=RigVMVariableNode() {variable=Bones}.Value
    Positions=RigVMVariableNode() {variable=Positions}.Value
   )
   > Project to New Parents = RigVMFunctionReferenceNode(Propagate to Children=True, Parents=RigVMVariableNode() {variable=Bones}.Value, Children=RigVMVariableNode() {variable=FK Controls}.Value)
   Set IKFK Keys = RigVMFunctionReferenceNode(
    "IK Controls "=((Type=Control,Name="head_ik_ctrl"))
    ExecuteContext=@SequenceExecution.D
    IK Mode=RigVMVariableNode() {variable=IK Mode}.Value
    Previous IK Mode=RigVMVariableNode() {variable=Previous IK Mode}.Value
    FK Controls=RigVMVariableNode() {variable=FK Controls}.Value
    Switch Control=RigVMVariableNode() {variable=IKFK Switch Control}.Value
    IK Controls .0=RigVMVariableNode() {variable=IK Control}.Value
   )
   > RigVMFunctionReturnNode()
  ]}
 ContainedGraph {nodes [
   n1 = RigVMFunctionEntryNode()
   > Get Points on Spline = RigVMCollapseNode(
    Bones=@n1.Bones
    Enable Stretch=@n1.Enable Stretch
    IK Control=@n1.IK Control
    Length=@n1.Length
    Mid Null=@n1.Mid Null
    Array=RigVMVariableNode() {variable=Positions}.Value
   )
   n2 = &RigUnit_SequenceExecution(ExecuteContext=@"Get Points on Spline".ExecuteContext)
   n3 = RigVMDispatchNode(ExecuteContext=@n2.A, Array=RigVMVariableNode() {variable=Bones}.Value) {template="DISPATCH_RigVMDispatch_ArrayIterator(in Array,out Element,out Index,out Count,out Ratio)"}
   At and Next = RigVMCollapseNode(Index=@n3.Index, Array=RigVMVariableNode() {variable=Positions}.Value)
   First Last Rotation = RigVMCollapseNode(
    First Bone=RigVMDispatchNode(Index=0, Array=RigVMVariableNode() {variable=Bones}.Value) {template="DISPATCH_RigVMDispatch_ArrayGetAtIndex(in Array,in Index,out Element)"}.Element
    T=@n3.Ratio
    IK Control=RigVMVariableNode() {variable=IK Control}.Value
   )
   n4 = &RigUnit_AimBoneMath(
    DebugSettings=(Scale=10,WorldOffset=(Rotation=(X=0,Y=0,Z=0,W=1),Translation=(X=0,Y=0,Z=0),Scale3D=(X=1,Y=1,Z=1)))
    InputTransform=(Rotation=(X=0,Y=0,Z=0,W=1),Translation=(X=0,Y=0,Z=0),Scale3D=(X=1,Y=1,Z=1))
    Primary=(Weight=1,Axis=(X=1,Y=0,Z=0),Target=(X=1,Y=0,Z=0),Kind=Location,Space=(Type=Bone))
    Secondary=(Weight=1,Axis=(X=0,Y=0,Z=1),Target=(X=1,Y=0,Z=0),Kind=Direction,Space=(Type=Bone))
    Weight=1
    InputTransform.Translation=@"At and Next".At
    Primary.Target=@"At and Next".Next
    Secondary.Target=&RigVMFunction_MathQuaternionRotateVector(Vector=(X=0,Y=0,Z=1), Transform=@"First Last Rotation".Interpolate) {template="Rotate Vector::Execute(in Transform,in Vector,out Result)"}.Result
   )
   Is Last = RigVMCollapseNode(Count=@n3.Count, Index=@n3.Index)
   > &RigUnit_SetTransform(
    Value=(Rotation=(X=0,Y=0,Z=0,W=1),Translation=(X=0,Y=0,Z=0),Scale3D=(X=1,Y=1,Z=1))
    Weight=1
    Item=@n3.Element
    Value.Translation=@"At and Next".At
    Value.Rotation=RigVMDispatchNode(False=@n4.Result.Rotation, True=@"First Last Rotation".Last, Condition=@"Is Last".Result) {template="DISPATCH_RigVMDispatch_If(in Condition,in True,in False,out Result)"}.Result
   )
   RigVMFunctionReturnNode(ExecuteContext=@n2.D)
  ]}
 ContainedGraph {nodes [
   n1 = RigVMFunctionEntryNode()
   n2 = &RigUnit_SequenceExecution(ExecuteContext=@n1.ExecuteContext)
   n3 = RigVMDispatchNode(ExecuteContext=@n2.A, Array=RigVMVariableNode() {variable=Array}.Value) {template="DISPATCH_RigVMDispatch_ArrayReset(io Array)"}
   > n4 = RigVMDispatchNode(Array=RigVMVariableNode() {variable=Bones}.Value) {template="DISPATCH_RigVMDispatch_ArrayIterator(in Array,out Element,out Index,out Count,out Ratio)"}
   n5 = &RigUnit_GetTransform(Item=RigVMDispatchNode(Index=0, Array=RigVMVariableNode() {variable=Bones}.Value) {template="DISPATCH_RigVMDispatch_ArrayGetAtIndex(in Array,in Index,out Element)"}.Element)
   RigUnit_GetTransform = &RigUnit_GetTransform(Item=RigVMVariableNode() {variable=Mid Null}.Value)
   n6 = &RigUnit_ControlRigSplineFromPoints(
    Points=((X=0,Y=0,Z=0),(X=0,Y=0,Z=0),(X=0,Y=0,Z=0),(X=0,Y=0,Z=0))
    SamplesPerSegment=16
    SplineMode=BSpline
    bClosed=False
    Points.0=@n5.Transform.Translation
    Points.1=@n5.Transform.Translation
    Points.3=&RigUnit_GetTransform(Item=RigVMVariableNode() {variable=IK Control}.Value).Transform.Translation
    Points.2=@RigUnit_GetTransform.Transform.Translation
   )
   n7 = &RigVMFunction_MathFloatDiv(B=&RigUnit_GetLengthControlRigSpline(Spline=@n6.Spline).Length, A=RigVMVariableNode() {variable=Length}.Value) {template="Divide::Execute(in A,in B,out Result)"}
   n8 = &RigVMFunction_MathFloatMul(
    A=@n4.Ratio
    B=RigVMDispatchNode(
     True=1
     False=&RigVMFunction_MathFloatMin(B=1, A=@n7.Result) {template="Minimum::Execute(in A,in B,out Result)"}.Result
     Condition=RigVMVariableNode() {variable=Enable Stretch}.Value
    ) {template="DISPATCH_RigVMDispatch_If(in Condition,in True,in False,out Result)"}.Result
   ) {template="Multiply::Execute(in A,in B,out Result)"}
   > RigVMDispatchNode(
    Array=@n3.Array
    Element=/Script/ControlRigSpline.RigUnit_PositionFromControlRigSpline(
     Spline=@n6.Spline
     U=/Script/ControlRigSpline.RigUnit_ParameterAtPercentage(Spline=@n6.Spline, Percentage=@n8.Result) {is=struct RigVMUnitNode}.U
    ) {is=struct RigVMUnitNode}.Position
   ) {template="DISPATCH_RigVMDispatch_ArrayAdd(io Array,in Element,out Index)"}
   RigVMFunctionReturnNode(ExecuteContext=@n2.D, Positions=RigVMVariableNode() {variable=Array}.Value)
  ]}
 ContainedGraph {nodes [
   n1 = RigVMFunctionEntryNode()
   RigVMFunctionReturnNode(Result=&RigVMFunction_MathIntEquals(A=@n1.Index, B=&RigVMFunction_MathIntSub(B=1, A=@n1.Count) {template="Subtract::Execute(in A,in B,out Result)"}.Result).Result)
  ]}
 ContainedGraph {nodes [
   n1 = RigVMFunctionEntryNode()
   RigVMFunctionReturnNode(
    At=RigVMDispatchNode(Array=@n1.Array, Index=@n1.Index) {template="DISPATCH_RigVMDispatch_ArrayGetAtIndex(in Array,in Index,out Element)"}.Element
    Next=RigVMDispatchNode(Array=@n1.Array, Index=&RigVMFunction_MathIntAdd(B=1, A=@n1.Index) {template="Add::Execute(in A,in B,out Result)"}.Result) {template="DISPATCH_RigVMDispatch_ArrayGetAtIndex(in Array,in Index,out Element)"}.Element
   )
  ]}
 ContainedGraph {nodes [
   n1 = RigVMFunctionEntryNode()
   n2 = /Script/RigVM.RigVMFunction_MathQuaternionMul(
    A=&RigUnit_GetTransform(Item=@n1.IK Control).Transform.Rotation
    B=&RigVMFunction_MathQuaternionFromRotator(Rotator=(Pitch=90,Yaw=0,Roll=0)).Result
   ) {is=struct RigVMUnitNode, template="Multiply::Execute(in A,in B,out Result)"}
   n3 = &RigUnit_HierarchyGetParent(bDefaultParent=True, Child=@n1.First Bone)
   n4 = &RigUnit_ProjectTransformToNewParent(bChildInitial=True, bOldParentInitial=True, Child=@n1.First Bone, NewParent=@n3.Parent, OldParent=@n3.Parent)
   RigVMFunctionReturnNode(
    Last=@n2.Result
    Interpolate=&RigVMFunction_MathQuaternionSlerp(T=@n1.T, B=@n2.Result, A=@n4.Transform.Rotation) {template="Interpolate::Execute(in A,in B,in T,out Result)"}.Result
    First=@n4.Transform.Rotation
   )
  ]}
 ContainedGraph {nodes [
   n1 = RigVMFunctionEntryNode()
   SequenceExecution = &RigVMFunction_Sequence(ExecuteContext=@n1.ExecuteContext)
   SequenceExecution_1 = &RigVMFunction_Sequence(ExecuteContext=@SequenceExecution.B)
   SequenceExecution_2 = &RigVMFunction_Sequence(ExecuteContext=@SequenceExecution_1.B)
   RigVMFunctionReturnNode(A=@SequenceExecution.A, B=@SequenceExecution_1.A, C=@SequenceExecution_2.A, D=@SequenceExecution_2.B)
  ]}
 ContainedGraph {nodes [
   n1 = &RigUnit_GetTransformItemArray(Space=GlobalSpace, Items=RigVMVariableNode() {variable=Bones}.Value)
   SequenceExecution = RigVMAggregateNode(ExecuteContext=@n7.Execute)
   BSpline Passing thru 4 Points_1 = RigVMFunctionReferenceNode(
    Auto Weights=false
    Weight 1=0.33
    Weight 2=0.66
    Point 0=RigVMDispatchNode(Index=0, Array=@n1.Transforms) {template="DISPATCH_RigVMDispatch_ArrayGetAtIndex(in Array,in Index,out Element)"}.Element.Translation
    Point 1=RigVMDispatchNode(Index=1, Array=@n1.Transforms) {template="DISPATCH_RigVMDispatch_ArrayGetAtIndex(in Array,in Index,out Element)"}.Element.Translation
    Point 2=RigVMDispatchNode(Index=2, Array=@n1.Transforms) {template="DISPATCH_RigVMDispatch_ArrayGetAtIndex(in Array,in Index,out Element)"}.Element.Translation
    Point 3=RigVMDispatchNode(Index=3, Array=@n1.Transforms) {template="DISPATCH_RigVMDispatch_ArrayGetAtIndex(in Array,in Index,out Element)"}.Element.Translation
    Execute=@SequenceExecution.A
   )
   n2 = RigVMDispatchNode(ExecuteContext=@"BSpline Passing thru 4 Points_1".Execute, Array=RigVMVariableNode() {variable=IK Controls}.Value) {template="DISPATCH_RigVMDispatch_ArrayIterator(in Array,out Element,out Index,out Count,out Ratio)"}
   MathIntEquals_1 = RigVMDispatchNode(B=0, A=@n2.Index) {template="DISPATCH_RigVMDispatch_CoreEquals(in A,in B,out Result)"}
   n3 = &RigVMFunction_ControlFlowBranch(ExecuteContext=@n2.ExecuteContext, Condition=@MathIntEquals_1.Result)
   n4 = RigVMDispatchNode(Index=@n2.Index, Array=RigVMVariableNode() {variable=Bones}.Value) {template="DISPATCH_RigVMDispatch_ArrayGetAtIndex(in Array,in Index,out Element)"}
   n5 = &RigUnit_ProjectTransformToNewParent(bChildInitial=True, bOldParentInitial=True, NewParent=@n4.Element, OldParent=@n4.Element, Child=@n2.Element)
   n6 = &RigVMFunction_MathTransformMakeAbsolute(
    Local=(Rotation=(X=0,Y=0,Z=0,W=1),Translation=(X=0,Y=0,Z=0),Scale3D=(X=1,Y=1,Z=1))
    Parent=(Rotation=(X=0,Y=0,Z=0,W=1),Translation=(X=0,Y=0,Z=0),Scale3D=(X=1,Y=1,Z=1))
    Parent.Translation=RigVMDispatchNode(Index=@n2.Index, Array=@"BSpline Passing thru 4 Points_1".Control Points) {template="DISPATCH_RigVMDispatch_ArrayGetAtIndex(in Array,in Index,out Element)"}.Element
    Parent.Rotation=@n5.Transform.Rotation
    Parent.Scale3D=@n5.Transform.Scale3D
   ) {template="Make Absolute::Execute(in Local,in Parent,out Global)"}
   &RigUnit_SetTransform(Weight=1, bPropagateToChildren=false, Item=@n2.Element, ExecutePin=@n3.False, Value=@n6.Global)
   n7 = RigVMFunctionEntryNode()
   RigVMDispatchNode(Index=@n2.Index, Array=RigVMVariableNode() {variable=Offsets}.Value) {template="DISPATCH_RigVMDispatch_ArrayGetAtIndex(in Array,in Index,out Element)"}
   RigVMFunctionReturnNode(Execute=@SequenceExecution.D)
  ]}
 ContainedGraph {nodes [
   n1 = RigVMFunctionEntryNode()
   SequenceExecution = &RigVMFunction_Sequence(ExecuteContext=@n1.ExecuteContext)
   SequenceExecution_1 = &RigVMFunction_Sequence(ExecuteContext=@SequenceExecution.B)
   SequenceExecution_2 = &RigVMFunction_Sequence(ExecuteContext=@SequenceExecution_1.B)
   RigVMFunctionReturnNode(A=@SequenceExecution.A, B=@SequenceExecution_1.A, C=@SequenceExecution_2.A, D=@SequenceExecution_2.B)
  ]}
 ContainedGraph {nodes [
   n1 = RigVMFunctionEntryNode()
   > SequenceAggregate = RigVMAggregateNode()
   n2 = &RigVMFunction_ControlFlowBranch(Condition=@n1.Active, ExecuteContext=@SequenceAggregate.A)
   Greater = /Script/RigVM.RigVMFunction_MathIntGreater(
    B=2
    A=RigVMDispatchNode(Array=RigVMVariableNode() {variable=Controls}.Value) {template="DISPATCH_RigVMDispatch_ArrayGetNum(in Array,out Num)"}.Num
   ) {is=struct RigVMUnitNode, template="Greater::Execute(in A,in B,out Result)"}
   n3 = &RigVMFunction_ControlFlowBranch(
    ExecuteContext=@n2.True
    Condition=&RigVMFunction_MathBoolAnd(B=@Greater.Result, A=&RigUnit_IsInteracting().bIsInteracting).Result
   )
   n4 = RigVMVariableNode(
    Value=RigVMDispatchNode(Index=0, Array=&RigUnit_IsInteracting().Items) {template="DISPATCH_RigVMDispatch_ArrayGetAtIndex(in Array,in Index,out Element)"}.Element
    ExecuteContext=@n3.True
   ) {variable=Interacting Item}
   n5 = RigVMDispatchNode(Element=RigVMVariableNode() {variable=Interacting Item}.Value, Array=RigVMVariableNode() {variable=Controls}.Value) {template="DISPATCH_RigVMDispatch_ArrayFind(in Array,in Element,out Index,out Success)"}
   DISPATCH_RigDispatch_CoreEquals_1 = RigVMDispatchNode(B=0, A=@n5.Index) {template="DISPATCH_RigVMDispatch_CoreEquals(in A,in B,out Result)"}
   n6 = RigVMVariableNode() {variable=Controls}
   DISPATCH_RigDispatch_CoreEquals_2 = RigVMDispatchNode(
    A=@n5.Index
    B=&RigVMFunction_MathIntSub(B=1, A=RigVMDispatchNode(Array=@n6.Value) {template="DISPATCH_RigVMDispatch_ArrayGetNum(in Array,out Num)"}.Num) {template="Subtract::Execute(in A,in B,out Result)"}.Result
   ) {template="DISPATCH_RigVMDispatch_CoreEquals(in A,in B,out Result)"}
   n7 = &RigVMFunction_ControlFlowBranch(
    Condition=&RigVMFunction_MathBoolOr(B=@DISPATCH_RigDispatch_CoreEquals_1.Result, A=@DISPATCH_RigDispatch_CoreEquals_2.Result).Result
    ExecuteContext=@n4.ExecuteContext
   )
   SequenceAggregate_3 = RigVMAggregateNode(ExecuteContext=@n7.True)
   n8 = RigVMVariableNode(ExecuteContext=@SequenceAggregate_3.A, Value=RigVMVariableNode() {variable=Controls}.Value) {variable=Loop Controls}
   n9 = &RigVMFunction_ControlFlowBranch(Condition=@DISPATCH_RigDispatch_CoreEquals_2.Result, ExecuteContext=@n8.ExecuteContext)
   RigVMDispatchNode(ExecuteContext=@n9.True, Array=RigVMVariableNode() {variable=Loop Controls}.Value) {template="DISPATCH_RigVMDispatch_ArrayReverse(io Array)"}
   n10 = RigVMDispatchNode(ExecuteContext=@SequenceAggregate_3.B, Array=RigVMVariableNode() {variable=Loop Controls}.Value) {template="DISPATCH_RigVMDispatch_ArrayIterator(in Array,out Element,out Index,out Count,out Ratio)"}
   DISPATCH_RigDispatch_CoreEquals_3 = RigVMDispatchNode(B=0, A=@n10.Index) {template="DISPATCH_RigVMDispatch_CoreEquals(in A,in B,out Result)"}
   n11 = &RigVMFunction_ControlFlowBranch(ExecuteContext=@n10.ExecuteContext, Condition=@DISPATCH_RigDispatch_CoreEquals_3.Result)
   n12 = RigVMVariableNode() {variable=Interacting Item}
   n13 = RigVMDispatchNode(
    Default=(Rotation=(X=0,Y=0,Z=0,W=1),Translation=(X=0,Y=0,Z=0),Scale3D=(X=1,Y=1,Z=1))
    Name=Previous Transform
    NameSpace=Self
    Item=@n12.Value
   ) {template="DISPATCH_RigDispatch_GetMetadata(in Item,in Name,in NameSpace,in Default,out Value,out Found)"}
   n14 = &RigUnit_GetTransform(Space=LocalSpace, Item=@n12.Value)
   &RigUnit_OffsetTransformForItem(
    Weight=1
    bPropagateToChildren=True
    Item=@n10.Element
    ExecutePin=@n11.False
    OffsetTransform=/Script/RigVM.RigVMFunction_MathTransformLerp(
     A=(Rotation=(X=0,Y=0,Z=0,W=1),Translation=(X=0,Y=0,Z=0),Scale3D=(X=1,Y=1,Z=1))
     B=&RigVMFunction_MathTransformMakeRelative(Parent=@n13.Value, Global=@n14.Transform).Local
     T=&RigVMFunction_MathDoubleSub(A=1, B=@n10.Ratio).Result
    ) {is=struct RigVMUnitNode, template="Interpolate::Execute(in A,in B,in T,out Result)"}.Result
   )
   n15 = RigVMDispatchNode(ExecuteContext=@SequenceAggregate_3.C, Array=RigVMVariableNode() {variable=Controls}.Value) {template="DISPATCH_RigVMDispatch_ArrayIterator(in Array,out Element,out Index,out Count,out Ratio)"}
   > SendEvent = &RigUnit_SendEvent(Event=RequestAutoKey, bEnable=True, bOnlyDuringInteraction=True, Item=@n15.Element) {template="SendEvent::Execute(in Event,in Item,in OffsetInSeconds,in bEnable,in bOnlyDuringInteraction)"}
   n16 = RigVMDispatchNode(ExecuteContext=@SequenceAggregate.B, Array=RigVMVariableNode() {variable=Controls}.Value) {template="DISPATCH_RigVMDispatch_ArrayIterator(in Array,out Element,out Index,out Count,out Ratio)"}
   > RigVMDispatchNode(
    Name=Previous Transform
    NameSpace=Self
    Item=@n16.Element
    Value=&RigUnit_GetTransform(Space=LocalSpace, Item=@n16.Element).Transform
   ) {template="DISPATCH_RigDispatch_SetMetadata(in Item,in Name,in NameSpace,in Value,out Success)"}
   RigVMFunctionReturnNode(ExecuteContext=@SequenceAggregate.C)
  ]}
 ContainedGraph {nodes [
   n1 = RigVMFunctionEntryNode()
   SequenceAggregate_3 = &RigVMFunction_Sequence(ExecuteContext=@n1.ExecuteContext)
   SequenceAggregate_3_1 = &RigVMFunction_Sequence(ExecuteContext=@SequenceAggregate_3.B)
   RigVMFunctionReturnNode(A=@SequenceAggregate_3.A, B=@SequenceAggregate_3_1.A, C=@SequenceAggregate_3_1.B)
  ]}
 ContainedGraph {nodes [
   n1 = RigVMFunctionEntryNode()
   SequenceAggregate = &RigVMFunction_Sequence(ExecuteContext=@n1.ExecuteContext)
   SequenceAggregate_1 = &RigVMFunction_Sequence(ExecuteContext=@SequenceAggregate.B)
   RigVMFunctionReturnNode(A=@SequenceAggregate.A, B=@SequenceAggregate_1.A, C=@SequenceAggregate_1.B)
  ]}
 ContainedGraph {nodes [
   n1 = RigVMFunctionEntryNode()
   n2 = &RigUnit_AimBoneMath(
    DebugSettings=(Scale=10,WorldOffset=(Rotation=(X=0,Y=0,Z=0,W=1),Translation=(X=0,Y=0,Z=0),Scale3D=(X=1,Y=1,Z=1)))
    Primary=(Weight=1,Axis=(X=1,Y=0,Z=0),Target=(X=0,Y=0,Z=0),Kind=Location,Space=(Type=Bone))
    Secondary=(Weight=1,Axis=(X=0,Y=0,Z=1),Target=(X=1,Y=0,Z=0),Kind=Direction,Space=(Type=Bone))
    Weight=1
    Primary.Space=@n1.Foot Control
    Secondary.Space=@n1.Foot Control
    InputTransform=&RigUnit_GetTransform(bInitial=@n1.Initial, Item=@n1.Root Bone).Transform
   )
   > /Script/ControlRig.RigUnit_DebugTransformMutableItemSpace(
    Color=(R=1,G=0,B=0,A=1)
    DebugDrawSettings=(DepthPriority=SDPG_Foreground,Lifetime=-1)
    Mode=Axes
    Scale=10
    Space=(Type=Bone)
    Thickness=1
    WorldOffset=(Rotation=(X=0,Y=0,Z=0,W=1),Translation=(X=0,Y=0,Z=0),Scale3D=(X=1,Y=1,Z=1))
    bEnabled=false
    Transform=@n2.Result
   ) {is=struct RigVMUnitNode}
   > &RigUnit_SetTransform(Weight=1, Value=@n2.Result, bInitial=@n1.Initial, Item=@n1.Pole Vector Null)
   > RigVMFunctionReturnNode()
  ]}
 ContainedGraph {nodes [
   n1 = RigVMFunctionEntryNode()
   SequenceAggregate = &RigVMFunction_Sequence(ExecuteContext=@n1.ExecuteContext)
   RigVMDispatchNode(ExecuteContext=@SequenceAggregate.A, Array=RigVMVariableNode() {variable=Matching Items}.Value) {template="DISPATCH_RigVMDispatch_ArrayReset(io Array)"}
   > n2 = RigVMDispatchNode(Array=/Script/ControlRig.RigUnit_CollectionNameSearchArray(
     TypeToSearch=Control
     PartialName=/Script/RigVM.RigVMFunction_NameTruncate(Count=5, FromEnd=false, Name=@n1.Value) {is=struct RigVMUnitNode, template="Chop::Execute(in Name,in Count,in FromEnd,out Remainder,out Chopped)"}.Chopped
    ) {is=struct RigVMUnitNode}.Items) {template="DISPATCH_RigVMDispatch_ArrayIterator(in Array,out Element,out Index,out Count,out Ratio)"}
   Contains = /Script/RigVM.RigVMFunction_Contains(Search=meta, Name=@n2.Element.Name) {is=struct RigVMUnitNode, template="Contains::Execute(in Name,in Search,out Result)"}
   MathBoolAnd = RigVMAggregateNode(
    C=/Script/RigVM.RigVMFunction_EndsWith(
     Name=@n2.Element.Name
     Ending=&RigVMFunction_NameTruncate(Count=6, FromEnd=True, Name=@n1.Value) {template="Chop::Execute(in Name,in Count,in FromEnd,out Remainder,out Chopped)"}.Chopped
    ) {is=struct RigVMUnitNode, template="EndsWith::Execute(in Name,in Ending,out Result)"}.Result
    A=&RigVMFunction_MathBoolNot(Value=@Contains.Result).Result
    B=&RigVMFunction_MathBoolNot(Value=&RigVMFunction_Contains(Search=toe, Name=@n2.Element.Name) {template="Contains::Execute(in Name,in Search,out Result)"}.Result).Result
   )
   n3 = &RigVMFunction_ControlFlowBranch(ExecuteContext=@n2.ExecuteContext, Condition=@MathBoolAnd.Result)
   RigVMDispatchNode(Element=@n2.Element, ExecuteContext=@n3.True, Array=RigVMVariableNode() {variable=Matching Items}.Value) {template="DISPATCH_RigVMDispatch_ArrayAdd(io Array,in Element,out Index)"}
   RigVMFunctionReturnNode(ExecuteContext=@SequenceAggregate.B, Items=RigVMVariableNode() {variable=Matching Items}.Value)
  ]}
 ContainedGraph {nodes [
   n1 = RigVMFunctionEntryNode()
   RigVMFunctionReturnNode(Result=&RigVMFunction_MathBoolAnd(B=@n1.C, A=&RigVMFunction_MathBoolAnd(A=@n1.A, B=@n1.B).Result).Result)
  ]}
 ContainedGraph {nodes [
   n1 = RigVMFunctionEntryNode()
   n2 = RigVMVariableNode(ExecuteContext=@n1.Execute Context) {variable=Local Segment Param}
   n3 = &RigUnit_GetTransform(Item=(Type=Bone), bInitial=true, Item.Name=@n1.Point A)
   Subtract = &RigVMFunction_MathVectorSub(
    B=@n3.Transform.Translation
    A=&RigUnit_GetTransform(Item=(Type=Bone), bInitial=true, Item.Name=@n1.Point B).Transform.Translation
   ) {template="Subtract::Execute(in A,in B,out Result)"}
   n4 = /Script/RigVM.RigVMFunction_MathVectorLength(Value=@Subtract.Result) {is=struct RigVMUnitNode, template="Length::Execute(in Value,out Result)"}
   IsNearlyZero = /Script/RigVM.RigVMFunction_MathDoubleIsNearlyZero(Tolerance=0, Value=@n4.Result) {is=struct RigVMUnitNode, template="IsNearlyZero::Execute(in Value,in Tolerance,out Result)"}
   n5 = &RigVMFunction_ControlFlowBranch(Condition=@IsNearlyZero.Result, ExecuteContext=@n2.ExecuteContext)
   n6 = &RigUnit_GetTransform(Item=(Type=Bone), bInitial=true, Item.Name=@n1.Segment Location)
   Unit = /Script/RigVM.RigVMFunction_MathVectorUnit(Value=@Subtract.Result) {is=struct RigVMUnitNode, template="Unit::Execute(in Value,out Result)"}
   Dot__ = /Script/RigVM.RigVMFunction_MathVectorDot(
    A=&RigVMFunction_MathVectorSub(B=@n3.Transform.Translation, A=@n6.Transform.Translation) {template="Subtract::Execute(in A,in B,out Result)"}.Result
    B=@Unit.Result
   ) {is=struct RigVMUnitNode, template="Dot,|::Execute(in A,in B,out Result)"}
   Divide = &RigVMFunction_MathDoubleDiv(A=@Dot__.Result, B=@n4.Result)
   RigVMVariableNode(ExecuteContext=@n5.False, Value=@Divide.Result) {variable=Local Segment Param}
   Multiply = /Script/RigVM.RigVMFunction_MathVectorMul(B=(X=1,Y=1,Z=1), B.X=@Divide.Result, B.Y=@Divide.Result, B.Z=@Divide.Result, A=@Subtract.Result) {is=struct RigVMUnitNode, template="Multiply::Execute(in A,in B,out Result)"}
   Add = &RigVMFunction_MathVectorAdd(A=@n3.Transform.Translation, B=@Multiply.Result) {template="Add::Execute(in A,in B,out Result)"}
   n7 = RigVMVariableNode() {variable=Local Segment Param}
   Interpolate = &RigVMFunction_MathQuaternionSlerp(A=@n3.Transform.Rotation, B=@n6.Transform.Rotation, T=@n7.Value) {template="Interpolate::Execute(in A,in B,in T,out Result)"}
   RigVMFunctionReturnNode(
    Param Transform=(Rotation=(X=0,Y=0,Z=0,W=1),Translation=(X=0,Y=0,Z=0),Scale3D=(X=1,Y=1,Z=1))
    Param Transform.Translation=@Add.Result
    Execute Context=@n5.Completed
    Param Transform.Scale3D=@n3.Transform.Scale3D
    Param Transform.Rotation=@Interpolate.Result
    Segment Param=@n7.Value
   )
  ]}
 ContainedGraph {nodes [
   n1 = RigVMFunctionEntryNode()
   n2 = &RigUnit_SequenceExecution(ExecuteContext=@n1.Execute)
   n3 = RigVMVariableNode() {variable=Start Transform}
   Get_Metadata = RigVMDispatchNode(Default=(Type=Bone,Name="None"), Name=Gimbal Control, NameSpace=Self, Item=@n3.Value) {template="DISPATCH_RigDispatch_GetMetadata(in Item,in Name,in NameSpace,in Default,out Value,out Found)"}
   n4 = RigVMDispatchNode(Condition=@Get_Metadata.Found, True=@Get_Metadata.Value, False=@n3.Value) {template="DISPATCH_RigVMDispatch_If(in Condition,in True,in False,out Result)"}
   n5 = &RigUnit_GetTransform(bInitial=false, Item=@n4.Result)
   RigUnit_GetTransform = &RigUnit_GetTransform(bInitial=false, Item=&RigUnit_HierarchyGetParent(bDefaultParent=True, Child=@n3.Value).Parent)
   Blend Swing Inverse Twist = RigVMFunctionReferenceNode(
    Tracked Rotation=&RigVMFunction_MathTransformMakeRelative(
     Parent=&RigUnit_GetTransform(Space=LocalSpace, bInitial=true, Item=@n4.Result).Transform
     Global=&RigVMFunction_MathTransformMakeRelative(Global=@n5.Transform, Parent=@RigUnit_GetTransform.Transform).Local
    ).Local.Rotation
    Swing Blend=RigVMVariableNode() {variable=Swing Blend}.Value
   )
   &RigUnit_SetRotation(
    Weight=1
    bInitial=False
    bPropagateToChildren=true
    Value=@"Blend Swing Inverse Twist".Result
    ExecutePin=@n2.A
    Item=RigVMVariableNode() {variable="Root Corrective "}.Value
   )
   n6 = RigVMDispatchNode(ExecuteContext=@n2.B, Array=RigVMVariableNode() {variable=Twist Bones}.Value) {template="DISPATCH_RigVMDispatch_ArrayIterator(in Array,out Element,out Index,out Count,out Ratio)"}
   Concat = &RigVMFunction_NameConcat(B=_space, A=@n6.Element.Name) {template="Concat::Execute(in A,in B,out Result)"}
   n7 = &RigUnit_GetTransform(Item=@n4.Result)
   n8 = RigVMVariableNode() {variable=End Transform}
   Interpolate = &RigVMFunction_MathTransformLerp(
    T=RigVMDispatchNode(Index=@n6.Index, Array=RigVMVariableNode() {variable=Position Blend}.Value) {template="DISPATCH_RigVMDispatch_ArrayGetAtIndex(in Array,in Index,out Element)"}.Element
    A=@n7.Transform
    B=&RigUnit_GetTransform(Item=@n8.Value).Transform
   ) {template="Interpolate::Execute(in A,in B,in T,out Result)"}
   > &RigUnit_SetTransform(
    Item=(Type=Null,Name="None")
    Value=(Rotation=(X=0,Y=0,Z=0,W=1),Translation=(X=0,Y=0,Z=0),Scale3D=(X=1,Y=1,Z=1))
    Weight=1
    bInitial=false
    Item.Name=@Concat.Result
    Value.Rotation=@n7.Transform.Rotation
    Value.Scale3D=@n7.Transform.Scale3D
    Value.Translation=@Interpolate.Result.Translation
   )
   n9 = RigVMVariableNode() {variable=Twist Blend}
   n10 = &RigVMFunction_MathQuaternionSwingTwist(
    TwistAxis=(X=1,Y=0,Z=0)
    Input=&RigVMFunction_MathTransformMakeRelative(
     Parent=&RigUnit_GetTransform(Space=LocalSpace, bInitial=true, Item=@n8.Value).Transform
     Global=&RigUnit_GetTransform(Space=LocalSpace, bInitial=false, Item=@n8.Value).Transform
    ).Local.Rotation
   )
   n11 = RigVMVariableNode() {variable=Twist From End}
   > &RigUnit_OffsetTransformForItem(
    Item=(Type=Null)
    OffsetTransform=(Rotation=(X=0,Y=0,Z=0,W=1),Translation=(X=0,Y=0,Z=0),Scale3D=(X=1,Y=1,Z=1))
    Weight=1
    bPropagateToChildren=true
    Item.Name=@Concat.Result
    OffsetTransform.Rotation=&RigVMFunction_MathQuaternionSlerp(
     B=(X=0,Y=0,Z=0,W=1)
     T=RigVMDispatchNode(Index=@n6.Index, Array=@n9.Value) {template="DISPATCH_RigVMDispatch_ArrayGetAtIndex(in Array,in Index,out Element)"}.Element
     A=RigVMDispatchNode(False=@"Blend Swing Inverse Twist".Inversed Twist, True=@n10.Twist, Condition=@n11.Value) {template="DISPATCH_RigVMDispatch_If(in Condition,in True,in False,out Result)"}.Result
    ) {template="Interpolate::Execute(in A,in B,in T,out Result)"}.Result
   )
   RigVMFunctionReturnNode(Execute=@n2.D)
  ]}
 ContainedGraph {nodes [
   n1 = RigVMFunctionEntryNode()
   > SequenceAggregate = RigVMAggregateNode()
   RigVMDispatchNode(ExecuteContext=@SequenceAggregate.A, Array=RigVMVariableNode() {variable=Twist Transforms}.Value) {template="DISPATCH_RigVMDispatch_ArrayReset(io Array)"}
   > n2 = RigVMDispatchNode(Array=@n1.twist joints) {template="DISPATCH_RigVMDispatch_ArrayIterator(in Array,out Element,out Index,out Count,out Ratio)"}
   Calculate Ref Segment Param_8_1 = RigVMFunctionReferenceNode(Segment Location=@n2.Element.Name, Execute Context=@n2.ExecuteContext, Point B=@n1.segment end joint, Point A=@n1.segment start joint)
   RigVMDispatchNode(
    ExecuteContext=@"Calculate Ref Segment Param_8_1".Execute Context
    Element=@"Calculate Ref Segment Param_8_1".Param Transform
    Array=RigVMVariableNode() {variable=Twist Transforms}.Value
   ) {template="DISPATCH_RigVMDispatch_ArrayAdd(io Array,in Element,out Index)"}
   > RigVMDispatchNode(Name=Location Param, NameSpace=Self, Item=@n2.Element, Value=@"Calculate Ref Segment Param_8_1".Segment Param) {template="DISPATCH_RigDispatch_SetMetadata(in Item,in Name,in NameSpace,in Value,out Success)"}
   > RigVMDispatchNode(
    Name=Twist Param
    NameSpace=Self
    Value=RigVMDispatchNode(Index=@n2.Index, Array=@n1.twist amounts) {template="DISPATCH_RigVMDispatch_ArrayGetAtIndex(in Array,in Index,out Element)"}.Element
    Item=@n2.Element
   ) {template="DISPATCH_RigDispatch_SetMetadata(in Item,in Name,in NameSpace,in Value,out Success)"}
   n3 = RigVMVariableNode() {variable=Twist Transforms}
   DrawTransforms = RigVMFunctionReferenceNode(Color=(R=1,G=0,B=0,A=1), Mode=Axes, Scale=10, Active=@n1.Debug, ExecuteContext=@SequenceAggregate.B, Transforms=@n3.Value)
   RigVMFunctionReturnNode(ExecuteContext=@SequenceAggregate.D, Twist Param Transforms=@n3.Value)
  ]}
 ContainedGraph {nodes [
   n1 = RigVMFunctionEntryNode()
   SequenceAggregate = &RigVMFunction_Sequence(ExecuteContext=@n1.ExecuteContext)
   SequenceAggregate_1 = &RigVMFunction_Sequence(ExecuteContext=@SequenceAggregate.B)
   SequenceAggregate_2 = &RigVMFunction_Sequence(ExecuteContext=@SequenceAggregate_1.B)
   RigVMFunctionReturnNode(A=@SequenceAggregate.A, B=@SequenceAggregate_1.A, C=@SequenceAggregate_2.A, D=@SequenceAggregate_2.B)
  ]}
 ContainedGraph {nodes [
   n1 = RigVMFunctionEntryNode()
   n2 = &RigVMFunction_MathQuaternionSwingTwist(TwistAxis=(X=1,Y=0,Z=0), Input=@n1.Tracked Rotation)
   n3 = /Script/RigVM.RigVMFunction_MathQuaternionInverse(Value=@n2.Twist) {is=struct RigVMUnitNode, template="Inverse::Execute(in Value,out Result)"}
   n4 = &RigVMFunction_MathQuaternionInverse(Value=@n2.Swing) {template="Inverse::Execute(in Value,out Result)"}
   n5 = RigVMVariableNode() {variable=Swing Blend}
   > RigVMFunctionReturnNode(
    Inversed Twist=@n3.Result
    Result=&RigVMFunction_MathQuaternionMul(
     A=@n3.Result
     B=&RigVMFunction_MathQuaternionSlerp(B=(X=0,Y=0,Z=0,W=1), A=@n4.Result, T=@n5.Value) {template="Interpolate::Execute(in A,in B,in T,out Result)"}.Result
    ) {template="Multiply::Execute(in A,in B,out Result)"}.Result
   )
  ]}
 ContainedGraph {nodes [
   n1 = RigVMFunctionEntryNode()
   > n2 = RigVMDispatchNode(Array=@n1.Twist Joints) {template="DISPATCH_RigVMDispatch_ArrayIterator(in Array,out Element,out Index,out Count,out Ratio)"}
   Get_Metadata = RigVMDispatchNode(Default=(Type=Bone,Name="None"), Name=Control, NameSpace=Self, Item=@n2.Element) {template="DISPATCH_RigDispatch_GetMetadata(in Item,in Name,in NameSpace,in Default,out Value,out Found)"}
   GetAnimationChannel = /Script/ControlRig.RigUnit_GetFloatAnimationChannel(Channel=Twist Factor, bInitial=False, Control=@Get_Metadata.Value.Name) {is=struct RigVMUnitNode, template="GetAnimationChannel::Execute(out Value,in Control,in Channel,in bInitial)"}
   > RigVMDispatchNode(Element=@GetAnimationChannel.Value, Array=RigVMVariableNode() {variable=Twist Params Local}.Value) {template="DISPATCH_RigVMDispatch_ArrayAdd(io Array,in Element,out Index)"}
   > RigVMDispatchNode(
    Element=RigVMDispatchNode(Default=0, Name=Location Param, NameSpace=Self, Item=@n2.Element) {template="DISPATCH_RigDispatch_GetMetadata(in Item,in Name,in NameSpace,in Default,out Value,out Found)"}.Value
    Array=RigVMVariableNode() {variable=Location Params Local}.Value
   ) {template="DISPATCH_RigVMDispatch_ArrayAdd(io Array,in Element,out Index)"}
   RigVMFunctionReturnNode(
    ExecuteContext=@n2.Completed
    Twist Params=RigVMVariableNode() {variable=Twist Params Local}.Value
    Location Params=RigVMVariableNode() {variable=Location Params Local}.Value
   )
  ]}
 ContainedGraph {nodes [
   n1 = RigVMFunctionEntryNode()
   > n2 = RigVMDispatchNode(Array=@n1.Twist Joints) {template="DISPATCH_RigVMDispatch_ArrayIterator(in Array,out Element,out Index,out Count,out Ratio)"}
   DISPATCH_RigDispatch_GetMetadata = RigVMDispatchNode(Default=(Type=Bone,Name="None"), Name=Control, NameSpace=Self, Item=@n2.Element) {template="DISPATCH_RigDispatch_GetMetadata(in Item,in Name,in NameSpace,in Default,out Value,out Found)"}
   n3 = &RigVMFunction_ControlFlowBranch(ExecuteContext=@n2.ExecuteContext, Condition=@DISPATCH_RigDispatch_GetMetadata.Found)
   &RigUnit_SetTransform(
    Weight=1
    bInitial=false
    bPropagateToChildren=true
    Item=@n2.Element
    ExecutePin=@n3.True
    Value=&RigUnit_ProjectTransformToNewParent(
     bChildInitial=True
     bOldParentInitial=True
     Child=@n2.Element
     NewParent=@DISPATCH_RigDispatch_GetMetadata.Value
     OldParent=@DISPATCH_RigDispatch_GetMetadata.Value
    ).Transform
   )
   RigVMFunctionReturnNode(ExecuteContext=@n2.Completed)
   &RigUnit_GetTransform(Item=@DISPATCH_RigDispatch_GetMetadata.Value)
  ]}
 ContainedGraph {nodes [
   n1 = RigVMFunctionEntryNode()
   > n2 = RigVMDispatchNode(Array=@n1.Twist Joints) {template="DISPATCH_RigVMDispatch_ArrayIterator(in Array,out Element,out Index,out Count,out Ratio)"}
   n3 = &RigUnit_GetTransform(bInitial=true, Item=@n1.Parent)
   > n4 = &RigUnit_HierarchyAddNull(
    Space=GlobalSpace
    Transform=(Rotation=(X=0,Y=0,Z=0,W=1),Translation=(X=0,Y=0,Z=0),Scale3D=(X=1,Y=1,Z=1))
    Transform.Translation=RigVMDispatchNode(Index=@n2.Index, Array=@n1.Twist Param Transforms) {template="DISPATCH_RigVMDispatch_ArrayGetAtIndex(in Array,in Index,out Element)"}.Element.Translation
    Name=&RigVMFunction_NameConcat(B=_space, A=@n2.Element.Name) {template="Concat::Execute(in A,in B,out Result)"}.Result
    Parent=@n1.Parent
    Transform.Rotation=@n3.Transform.Rotation
    Transform.Scale3D=@n3.Transform.Scale3D
   )
   Concat = &RigVMFunction_NameConcat(B=_ctrl, A=@n2.Element.Name) {template="Concat::Execute(in A,in B,out Result)"}
   > SpawnControl = /Script/ControlRig.RigUnit_HierarchyAddControlTransform(
    InitialValue=(Rotation=(X=0,Y=0,Z=0,W=1),Translation=(X=0,Y=0,Z=0),Scale3D=(X=1,Y=1,Z=1))
    OffsetSpace=LocalSpace
    OffsetTransform=(Rotation=(X=0,Y=0,Z=0,W=1),Translation=(X=0,Y=0,Z=0),Scale3D=(X=1,Y=1,Z=1))
    Settings="(Shape=(bVisible=True,Name=\"Sphere_Thick\",Color=(R=1,G=0,B=0,A=1),Transform=(Rotation=(X=0,Y=0,Z=0,W=1),Translation=(X=0,Y=0,Z=0),Scale3D=(X=0.5,Y=0.5,Z=0.5))),Proxy=(bIsProxy=False,DrivenControls=((Type=None,Name=\"None\")),ShapeVisibility=UserDefined),Disp ...(truncated)"
    Name=@Concat.Result
    Settings.Shape=@n1.Shape Settings
    Parent=@n4.Item
   ) {
    is=struct RigVMUnitNode
    template="SpawnControl::Execute(in InitialValue,in Settings,in OffsetTransform,in OffsetSpace,in Parent,in Name,out Item)"
   }
   > &RigUnit_SetTransform(
    Value=(Rotation=(X=0,Y=0,Z=0,W=1),Translation=(X=0,Y=0,Z=0),Scale3D=(X=1,Y=1,Z=1))
    Weight=1
    bInitial=true
    bPropagateToChildren=true
    Value.Translation=&RigUnit_GetTransform(bInitial=true, Item=@n2.Element).Transform.Translation
    Value.Rotation=@n3.Transform.Rotation
    Value.Scale3D=@n3.Transform.Scale3D
    Item=@SpawnControl.Item
   )
   > DISPATCH_RigDispatch_SetMetadata = RigVMDispatchNode(Name=Control, NameSpace=Self, Item=@n2.Element, Value=@SpawnControl.Item) {template="DISPATCH_RigDispatch_SetMetadata(in Item,in Name,in NameSpace,in Value,out Success)"}
   > RigVMDispatchNode(Name=Snap, NameSpace=Self, Item=@n2.Element, Value=@n4.Item) {template="DISPATCH_RigDispatch_SetMetadata(in Item,in Name,in NameSpace,in Value,out Success)"}
   At = RigVMDispatchNode(Index=@n2.Index, Array=@n1.Twist Factors) {template="DISPATCH_RigVMDispatch_ArrayGetAtIndex(in Array,in Index,out Element)"}
   > SpawnAnimationChannel = /Script/ControlRig.RigUnit_HierarchyAddAnimationChannelFloat(
    LimitsEnabled=(Enabled=(bMinimum=True,bMaximum=True))
    MaximumValue=1
    MinimumValue=0
    Name=Twist Factor
    InitialValue=@At.Element
    Parent=@SpawnControl.Item
   ) {
    is=struct RigVMUnitNode
    template="SpawnAnimationChannel::Execute(in InitialValue,in MinimumValue,in MaximumValue,in LimitsEnabled,in Parent,in Name,out Item)"
   }
   RigVMFunctionReturnNode(ExecuteContext=@n2.Completed)
   &RigUnit_HierarchyGetParent(bDefaultParent=True, Child=@n2.Element)
  ]}
 ContainedGraph {nodes [
   n1 = RigVMFunctionEntryNode()
   > SequenceAggregate = RigVMAggregateNode()
   n2 = RigVMVariableNode() {variable=Twist Joints}
   Build Segment Twist Param Arrays = RigVMFunctionReferenceNode(ExecuteContext=@SequenceAggregate.A, Twist Joints=@n2.Value)
   Compute Twist = RigVMFunctionReferenceNode(
    Swing Blend=0.5
    Execute=@"Build Segment Twist Param Arrays".ExecuteContext
    Position Blend=@"Build Segment Twist Param Arrays".Location Params
    Twist Blend=@"Build Segment Twist Param Arrays".Twist Params
    End Transform=@n1.End Bone
    "Root Corrective "=@n1.Root Corrective
    Start Transform=@n1.Start Bone
    Twist From End=@n1.Twist From End
    Twist Bones=@n2.Value
   )
   n3 = RigVMDispatchNode(ExecuteContext=@"Compute Twist".Execute, Array=@n2.Value) {template="DISPATCH_RigVMDispatch_ArrayIterator(in Array,out Element,out Index,out Count,out Ratio)"}
   DISPATCH_RigDispatch_GetMetadata = RigVMDispatchNode(Default=(Type=Bone,Name="None"), Name=Control, NameSpace=Self, Item=@n3.Element) {template="DISPATCH_RigDispatch_GetMetadata(in Item,in Name,in NameSpace,in Default,out Value,out Found)"}
   > &RigUnit_SetTransform(
    Weight=1
    bInitial=False
    Item=@DISPATCH_RigDispatch_GetMetadata.Value
    Value=&RigUnit_ProjectTransformToNewParent(bChildInitial=True, bOldParentInitial=True, NewParent=@n3.Element, OldParent=@n3.Element, Child=@DISPATCH_RigDispatch_GetMetadata.Value).Transform
   )
   RigVMFunctionReturnNode(ExecuteContext=@SequenceAggregate.C)
  ]}
 ContainedGraph {nodes [
   n1 = RigVMFunctionEntryNode()
   SequenceAggregate = &RigVMFunction_Sequence(ExecuteContext=@n1.ExecuteContext)
   SequenceAggregate_1 = &RigVMFunction_Sequence(ExecuteContext=@SequenceAggregate.B)
   RigVMFunctionReturnNode(A=@SequenceAggregate.A, B=@SequenceAggregate_1.A, C=@SequenceAggregate_1.B)
  ]}
 ContainedGraph {nodes [
   n1 = RigVMFunctionEntryNode()
   > SequenceAggregate = RigVMAggregateNode()
   n2 = RigVMDispatchNode(ExecuteContext=@SequenceAggregate.A, Array=RigVMVariableNode() {variable=aimed_joints}.Value) {template="DISPATCH_RigVMDispatch_ArrayReset(io Array)"}
   > RigVMVariableNode(Value=@n1.Parent) {variable=Previous Aimed Joint}
   > n3 = RigVMDispatchNode(Array=@n1.Bones) {template="DISPATCH_RigVMDispatch_ArrayIterator(in Array,out Element,out Index,out Count,out Ratio)"}
   Concat = &RigVMFunction_NameConcat(B=_aimed, A=@n3.Element.Name) {template="Concat::Execute(in A,in B,out Result)"}
   > n4 = /Script/ControlRig.RigUnit_HierarchyAddBone(
    Space=GlobalSpace
    Name=@Concat.Result
    Transform=&RigUnit_GetTransform(bInitial=true, Item=@n3.Element).Transform
    Parent=RigVMVariableNode() {variable=Previous Aimed Joint}.Value
   ) {is=struct RigVMUnitNode}
   > RigVMVariableNode(Value=@n4.Item) {variable=Previous Aimed Joint}
   > RigVMDispatchNode(Array=@n2.Array, Element=@n4.Item) {template="DISPATCH_RigVMDispatch_ArrayAdd(io Array,in Element,out Index)"}
   n5 = RigVMVariableNode() {variable=aimed_joints}
   AlignThreeJointChainsToPlane = RigVMFunctionReferenceNode(
    Debug=@n1.Debug
    IK Plane Twist Compensation=@n1.IK Plane Twist Compensation
    Negated Axis=@n1.Negative Side
    ExecuteContext=@SequenceAggregate.B
    Bones=@n5.Value
   )
   RigVMFunctionReturnNode(ExecuteContext=@SequenceAggregate.E, Aimed Bones=@n5.Value)
  ]}
 ContainedGraph {nodes [
   n1 = RigVMFunctionEntryNode()
   SequenceAggregate = &RigVMFunction_Sequence(ExecuteContext=@n1.ExecuteContext)
   SequenceAggregate_1 = &RigVMFunction_Sequence(ExecuteContext=@SequenceAggregate.B)
   SequenceAggregate_2 = &RigVMFunction_Sequence(ExecuteContext=@SequenceAggregate_1.B)
   SequenceAggregate_3 = &RigVMFunction_Sequence(ExecuteContext=@SequenceAggregate_2.B)
   RigVMFunctionReturnNode(A=@SequenceAggregate.A, B=@SequenceAggregate_1.A, C=@SequenceAggregate_2.A, D=@SequenceAggregate_3.A, E=@SequenceAggregate_3.B)
  ]}
 New Function_ContainedGraph {nodes [
   n1 = RigVMFunctionEntryNode()
   n2 = RigVMDispatchNode(Index=0, Array=@n1.Bones) {template="DISPATCH_RigVMDispatch_ArrayGetAtIndex(in Array,in Index,out Element)"}
   Concat = &RigVMFunction_NameConcat(B=_ik_plane, A=@n2.Element.Name) {template="Concat::Execute(in A,in B,out Result)"}
   > SpawnControl = &RigUnit_HierarchyAddControlTransform(
    InitialValue=(Rotation=(X=0,Y=0,Z=0,W=1),Translation=(X=0,Y=0,Z=0),Scale3D=(X=1,Y=1,Z=1))
    OffsetSpace=LocalSpace
    OffsetTransform=(Rotation=(X=0,Y=0,Z=0,W=1),Translation=(X=0,Y=0,Z=0),Scale3D=(X=1,Y=1,Z=1))
    Settings="(Shape=(bVisible=True,Name=\"Box_Solid\",Color=(R=1,G=0,B=0,A=1),Transform=(Rotation=(X=0,Y=0,Z=0,W=1),Translation=(X=0,Y=0,Z=0),Scale3D=(X=15,Y=0,Z=5))),Proxy=(bIsProxy=False,DrivenControls=((Type=None,Name=\"None\")),ShapeVisibility=BasedOnSelection),Display ...(truncated)"
    Parent=@n2.Element
    Name=@Concat.Result
    Settings.Shape.bVisible=@n1.Debug
    Settings.Shape.Transform.Rotation=&RigVMFunction_MathQuaternionFromEuler(Euler=(X=90,Y=0,Z=0)).Result
   ) {template="SpawnControl::Execute(in InitialValue,in Settings,in OffsetTransform,in OffsetSpace,in Parent,in Name,out Item)"}
   n3 = &RigUnit_GetTransform(bInitial=true, Item=@n2.Element)
   n4 = RigVMDispatchNode(Index=1, Array=@n1.Bones) {template="DISPATCH_RigVMDispatch_ArrayGetAtIndex(in Array,in Index,out Element)"}
   n5 = &RigUnit_GetTransform(bInitial=true, Item=@n4.Element)
   Subtract = &RigVMFunction_MathVectorSub(B=@n3.Transform.Translation, A=@n5.Transform.Translation) {template="Subtract::Execute(in A,in B,out Result)"}
   n6 = RigVMDispatchNode(
    Condition=@n1.Negated Axis
    True=&RigVMFunction_MathVectorMul(B=(X=-1,Y=-1,Z=-1), A=@Subtract.Result) {template="Multiply::Execute(in A,in B,out Result)"}.Result
    False=@Subtract.Result
   ) {template="DISPATCH_RigVMDispatch_If(in Condition,in True,in False,out Result)"}
   n7 = RigVMDispatchNode(Index=2, Array=@n1.Bones) {template="DISPATCH_RigVMDispatch_ArrayGetAtIndex(in Array,in Index,out Element)"}
   n8 = &RigUnit_GetTransform(bInitial=true, Item=@n7.Element)
   n9 = /Script/RigVM.RigVMFunction_MathVectorCross(
    A=@n6.Result
    B=&RigVMFunction_MathVectorSub(B=@n3.Transform.Translation, A=@n8.Transform.Translation) {template="Subtract::Execute(in A,in B,out Result)"}.Result
   ) {is=struct RigVMUnitNode, template="Cross::Execute(in A,in B,out Result)"}
   n10 = &RigVMFunction_MathVectorCross(B=@n6.Result, A=@n9.Result) {template="Cross::Execute(in A,in B,out Result)"}
   Unit = &RigVMFunction_MathVectorUnit(Value=RigVMDispatchNode(
     Condition=@n1.Negated Axis
     False=@n10.Result
     True=&RigVMFunction_MathVectorMul(B=(X=-1,Y=-1,Z=-1), A=@n10.Result) {template="Multiply::Execute(in A,in B,out Result)"}.Result
    ) {template="DISPATCH_RigVMDispatch_If(in Condition,in True,in False,out Result)"}.Result) {template="Unit::Execute(in Value,out Result)"}
   n11 = RigVMDispatchNode(
    Condition=@n1.Negated Axis
    False=@n9.Result
    True=&RigVMFunction_MathVectorMul(B=(X=-1,Y=-1,Z=-1), A=@n9.Result) {template="Multiply::Execute(in A,in B,out Result)"}.Result
   ) {template="DISPATCH_RigVMDispatch_If(in Condition,in True,in False,out Result)"}
   n12 = /Script/RigVM.RigVMFunction_MathMatrixToTransform(Value=/Script/RigVM.RigVMFunction_MathMatrixFromVectors(
     Origin=@n3.Transform.Translation
     Y=@Unit.Result
     Z=&RigVMFunction_MathVectorUnit(Value=@n11.Result) {template="Unit::Execute(in Value,out Result)"}.Result
     X=&RigVMFunction_MathVectorUnit(Value=@n6.Result) {template="Unit::Execute(in Value,out Result)"}.Result
    ) {is=struct RigVMUnitNode}.Result) {is=struct RigVMUnitNode, template="Cast::Execute(in Value,out Result)"}
   > &RigUnit_SetTransform(Weight=1, bInitial=true, bPropagateToChildren=false, Value=@n12.Result, Item=@SpawnControl.Item)
   n13 = /Script/RigVM.RigVMFunction_MathMatrixFromTransformV2(Value=@n3.Transform) {is=struct RigVMUnitNode, template="Cast::Execute(in Value,out Result)"}
   n14 = /Script/RigVM.RigVMFunction_MathQuaternionFromAxisAndAngle(
    Axis=(X=1,Y=0,Z=0)
    Angle=/Script/RigVM.RigVMFunction_MathVectorAngle(A=/Script/RigVM.RigVMFunction_MathMatrixToVectors(Value=@n13.Result) {is=struct RigVMUnitNode}.Y, B=@Unit.Result) {is=struct RigVMUnitNode, template="AngleBetween::Execute(in A,in B,out Result)"}.Result
   ) {is=struct RigVMUnitNode, template="FromAxisAndAngle::Execute(in Axis,in Angle,out Result)"}
   > &RigUnit_OffsetTransformForItem(
    OffsetTransform=(Rotation=(X=0,Y=0,Z=0,W=1),Translation=(X=0,Y=0,Z=0),Scale3D=(X=1,Y=1,Z=1))
    Weight=1
    bPropagateToChildren=false
    OffsetTransform.Rotation=RigVMDispatchNode(
     Condition=@n1.IK Plane Twist Compensation
     True=@n14.Result
     False=&RigVMFunction_MathQuaternionFromEuler(Euler=(X=0,Y=0,Z=0)).Result
    ) {template="DISPATCH_RigVMDispatch_If(in Condition,in True,in False,out Result)"}.Result
    Item=@SpawnControl.Item
   )
   Multiply = &RigVMFunction_MathTransformMul(A=(Rotation=(X=0,Y=0,Z=0,W=1),Translation=(X=0,Y=0,Z=0),Scale3D=(X=1,Y=1,Z=1)), B=@n12.Result, A.Rotation=@n14.Result) {template="Multiply::Execute(in A,in B,out Result)"}
   n15 = RigVMDispatchNode(False=@n12.Result, Condition=@n1.IK Plane Twist Compensation, True=@Multiply.Result) {template="DISPATCH_RigVMDispatch_If(in Condition,in True,in False,out Result)"}
   > &RigUnit_SetTransform(Weight=1, bInitial=true, bPropagateToChildren=false, Item=@n2.Element, Value=@n15.Result)
   n16 = &RigVMFunction_MathMatrixToVectors(Value=&RigVMFunction_MathMatrixFromTransformV2(Value=@n15.Result) {template="Cast::Execute(in Value,out Result)"}.Result)
   n17 = &RigVMFunction_MathVectorSub(B=@n5.Transform.Translation, A=@n8.Transform.Translation) {template="Subtract::Execute(in A,in B,out Result)"}
   n18 = RigVMDispatchNode(
    True=&RigVMFunction_MathVectorMul(B=(X=-1,Y=-1,Z=-1), A=@n17.Result) {template="Multiply::Execute(in A,in B,out Result)"}.Result
    False=@n17.Result
    Condition=@n1.Negated Axis
   ) {template="DISPATCH_RigVMDispatch_If(in Condition,in True,in False,out Result)"}
   n19 = &RigVMFunction_MathMatrixToTransform(Value=&RigVMFunction_MathMatrixFromVectors(
     Origin=@n5.Transform.Translation
     Z=@n16.Z
     Y=&RigVMFunction_MathVectorCross(B=@n18.Result, A=@n16.Z) {template="Cross::Execute(in A,in B,out Result)"}.Result
     X=&RigVMFunction_MathVectorUnit(Value=@n18.Result) {template="Unit::Execute(in Value,out Result)"}.Result
    ).Result) {template="Cast::Execute(in Value,out Result)"}
   > &RigUnit_SetTransform(
    Value=(Rotation=(X=0,Y=0,Z=0,W=1),Translation=(X=0,Y=0,Z=0),Scale3D=(X=1,Y=1,Z=1))
    Weight=1
    bInitial=true
    bPropagateToChildren=false
    Item=@n4.Element
    Value.Rotation=@n19.Result.Rotation
    Value.Translation=@n19.Result.Translation
    Value.Scale3D=@n5.Transform.Scale3D
   )
   > &RigUnit_SetTransform(
    Value=(Rotation=(X=0,Y=0,Z=0,W=1),Translation=(X=0,Y=0,Z=0),Scale3D=(X=1,Y=1,Z=1))
    Weight=1
    bInitial=true
    bPropagateToChildren=false
    Item=@n7.Element
    Value.Rotation=@n19.Result.Rotation
    Value.Scale3D=@n8.Transform.Scale3D
    Value.Translation=@n8.Transform.Translation
   )
   > RigVMFunctionReturnNode()
  ]}
 New Function_ContainedGraph {nodes [
   n1 = RigVMFunctionEntryNode()
   > SequenceAggregate = RigVMAggregateNode()
   Concat = &RigVMFunction_NameConcat(B=_ctrl, A=@n1.Bone.Name) {template="Concat::Execute(in A,in B,out Result)"}
   DISPATCH_RigVMDispatch_If = RigVMDispatchNode(
    True=@n1.Parent
    False=&RigUnit_HierarchyGetParent(bDefaultParent=True, Child=@n1.Bone).Parent
    Condition=/Script/ControlRig.RigUnit_ItemExists(Item=@n1.Parent) {is=struct RigVMUnitNode}.Exists
   ) {template="DISPATCH_RigVMDispatch_If(in Condition,in True,in False,out Result)"}
   SpawnControl = &RigUnit_HierarchyAddControlTransform(
    InitialValue=(Rotation=(X=0,Y=0,Z=0,W=1),Translation=(X=0,Y=0,Z=0),Scale3D=(X=1,Y=1,Z=1))
    OffsetSpace=LocalSpace
    OffsetTransform=(Rotation=(X=0,Y=0,Z=0,W=1),Translation=(X=0,Y=0,Z=0),Scale3D=(X=1,Y=1,Z=1))
    Name=@Concat.Result
    Parent=@DISPATCH_RigVMDispatch_If.Result
    Settings=@n1.Control Shape Settings
    ExecutePin=@SequenceAggregate.A
   ) {template="SpawnControl::Execute(in InitialValue,in Settings,in OffsetTransform,in OffsetSpace,in Parent,in Name,out Item)"}
   n2 = &RigUnit_GetTransform(bInitial=true, Item=@n1.Bone)
   > &RigUnit_SetTransform(Weight=1, bInitial=true, bPropagateToChildren=true, Value=@n2.Transform, Item=@SpawnControl.Item)
   > DISPATCH_RigDispatch_SetMetadata = RigVMDispatchNode(Name=Control, NameSpace=Self, Item=@n1.Bone, Value=@SpawnControl.Item) {template="DISPATCH_RigDispatch_SetMetadata(in Item,in Name,in NameSpace,in Value,out Success)"}
   RigVMFunction_ControlFlowBranch = &RigVMFunction_ControlFlowBranch(Condition=@n1.Build Buffer Null, ExecuteContext=@SequenceAggregate.B)
   n3 = &RigUnit_HierarchyAddNull(
    Space=GlobalSpace
    Name=&RigVMFunction_NameConcat(B=_space, A=@n1.Bone.Name) {template="Concat::Execute(in A,in B,out Result)"}.Result
    Parent=@DISPATCH_RigVMDispatch_If.Result
    Transform=@n2.Transform
    ExecutePin=@RigVMFunction_ControlFlowBranch.True
   )
   > /Script/ControlRig.RigUnit_SwitchParent(Mode=ParentItem, bMaintainGlobal=True, Parent=@n3.Item, Child=@SpawnControl.Item) {is=struct RigVMUnitNode}
   RigVMFunctionReturnNode(ExecuteContext=@SequenceAggregate.C, Argument=@SpawnControl.Item)
  ]}
 New Function_ContainedGraph {nodes [
   n1 = RigVMFunctionEntryNode()
   DISPATCH_RigDispatch_GetMetadata = RigVMDispatchNode(Default=(Type=Bone,Name="None"), Name=Control, NameSpace=Self, Item=@n1.Bone) {template="DISPATCH_RigDispatch_GetMetadata(in Item,in Name,in NameSpace,in Default,out Value,out Found)"}
   DISPATCH_RigVMDispatch_If = RigVMDispatchNode(True=@DISPATCH_RigDispatch_GetMetadata.Value, False=@n1.Bone, Condition=@n1.Forward Solve) {template="DISPATCH_RigVMDispatch_If(in Condition,in True,in False,out Result)"}
   > &RigUnit_SetTransform(
    Weight=1
    bInitial=False
    bPropagateToChildren=true
    Item=RigVMDispatchNode(False=@DISPATCH_RigDispatch_GetMetadata.Value, True=@n1.Bone, Condition=@n1.Forward Solve) {template="DISPATCH_RigVMDispatch_If(in Condition,in True,in False,out Result)"}.Result
    Value=&RigUnit_GetTransform(Item=@DISPATCH_RigVMDispatch_If.Result).Transform
   )
   > RigVMFunctionReturnNode()
  ]}
 New Function_ContainedGraph {nodes [
   n1 = RigVMFunctionEntryNode()
   SequenceAggregate = &RigVMFunction_Sequence(ExecuteContext=@n1.ExecuteContext)
   n2 = RigVMDispatchNode(Array=@n1.Items, ExecuteContext=@SequenceAggregate.A) {template="DISPATCH_RigVMDispatch_ArrayIterator(in Array,out Element,out Index,out Count,out Ratio)"}
   > &RigUnit_SwitchParent(Mode=ParentItem, bMaintainGlobal=true, Child=@n2.Element, Parent=@n1.Parent)
   RigVMFunctionReturnNode(ExecuteContext=@SequenceAggregate.B)
   &RigUnit_SetTransform(
    ExecutePin=()
    Space=LocalSpace
    Weight=1
    bInitial=true
    bPropagateToChildren=true
    Item=@n2.Element
    Value=&RigUnit_GetRelativeTransformForItem(bChildInitial=true, bParentInitial=true, Child=@n2.Element, Parent=@n1.Parent).RelativeTransform
   )
  ]}
 New Function_ContainedGraph {nodes [
   n1 = RigVMFunctionEntryNode()
   > DISPATCH_RigVMDispatch_ArrayIterator = RigVMDispatchNode(Array=@n1.Control List) {template="DISPATCH_RigVMDispatch_ArrayIterator(in Array,out Element,out Index,out Count,out Ratio)"}
   > RigVMDispatchNode(Name=Mirror Axis, NameSpace=Self, Item=@DISPATCH_RigVMDispatch_ArrayIterator.Element, Value=@n1.Mirror Axis) {template="DISPATCH_RigDispatch_SetMetadata(in Item,in Name,in NameSpace,in Value,out Success)"}
   > DISPATCH_RigDispatch_SetMetadata = RigVMDispatchNode(Name=Mirror behavioral, NameSpace=Self, Item=@DISPATCH_RigVMDispatch_ArrayIterator.Element, Value=@n1.Mirror Behavioral) {template="DISPATCH_RigDispatch_SetMetadata(in Item,in Name,in NameSpace,in Value,out Success)"}
   > RigVMDispatchNode(Name=Mirror No Flip, NameSpace=Self, Item=@DISPATCH_RigVMDispatch_ArrayIterator.Element, Value=@n1.Mirror No Flip) {template="DISPATCH_RigDispatch_SetMetadata(in Item,in Name,in NameSpace,in Value,out Success)"}
   RigVMFunctionReturnNode(ExecuteContext=@DISPATCH_RigVMDispatch_ArrayIterator.Completed)
  ]}
 RigVMFunctionLibrary {nodes [
   Add Control Metadata = RigVMCollapseNode(Control List=(), ExecuteContext=(), Mirror Axis=(X=0,Y=0,Z=0), Mirror Behavioral=False, Mirror No Flip=False)
   AlignThreeJointChainsToPlane = RigVMCollapseNode(Bones=(), Debug=false, ExecuteContext=(), IK Plane Twist Compensation=false, Negated Axis=false)
   BSpline Passing thru 4 Points = RigVMCollapseNode(
    Auto Weights=true
    Execute=()
    Point 0=(X=0,Y=0,Z=0)
    Point 1=(X=0,Y=0,Z=0)
    Point 2=(X=0,Y=0,Z=0)
    Point 3=(X=0,Y=0,Z=0)
    Weight 1=0.33
    Weight 2=0.66
   )
   Backward Arm = RigVMCollapseNode(Bones=(), ExecuteContext=(), FK Arm Space=(), FK Controls=(), FK Orient References=(), Fingers Space=(), Ik Control=(), UpV Control=())
   Backward Clavicle = RigVMCollapseNode(Bone=(), Control=(), Execute=())
   Backward Fingers = RigVMCollapseNode(Bones=(), Execute=())
   Backward Foot = RigVMCollapseNode(Bone=(), Execute=(), FK Control=(), IK Control=(), Toes Space=())
   Backward Leg = RigVMCollapseNode(Bones=(), ExecuteContext=(), FK Controls=(), Ik Control=(), Pole Vector Null=(Type=Control,Name="FootParent"), UpV Control=())
   Backward Neck = RigVMCollapseNode(Bones=(), ExecuteContext=(), FK Controls=(), Head FK Space=(), Head IK Space=(), Head Orient Reference=(), IK Control=())
   Backward Spine = RigVMCollapseNode(Bones=(), ExecuteContext=(), FK Controls=(), IK Controls=(), Offsets=(), Pelvis Bone=(Type=Bone,Name="pelvis"))
   Backwards Solve Twist Controls = RigVMCollapseNode(
    End Bone=(Type=None,Name="None")
    ExecuteContext=()
    Root Corrective=(Type=None,Name="None")
    Start Bone=(Type=None,Name="None")
    Twist From End=False
    Twist Joints=()
   )
   Blend Swing Inverse Twist = RigVMCollapseNode(Swing Blend=0.5, Tracked Rotation=(X=0,Y=0,Z=0,W=1))
   Build Aimed Chains = RigVMCollapseNode(
    Bones=()
    Debug=False
    Do Aim=False
    ExecuteContext=()
    FK Controls=()
    IK Plane Twist Compensation=False
    Negative Side=False
    Parent=(Type=None,Name="None")
    Point At Axis=(X=1,Y=0,Z=0)
    Up Vector Axis=(X=0,Y=0,Z=1)
    Up Vector Target=(X=0,Y=0,Z=1)
   )
   Build Control From Joint = RigVMCollapseNode(
    Bone=(Type=None,Name="None")
    Build Buffer Null=False
    Control Shape Settings="(Shape=(bVisible=True,Name=\"Default\",Color=(R=1,G=0,B=0,A=1),Transform=(Rotation=(X=0,Y=0,Z=0,W=1),Translation=(X=0,Y=0,Z=0),Scale3D=(X=1,Y=1,Z=1))),Proxy=(bIsProxy=False,ShapeVisibility=BasedOnSelection),DisplayName=\"None\",bSelectable=True,InitialSpace=Lo ...(truncated)"
    ExecuteContext=()
    Parent=(Type=None,Name="None")
   )
   Build Segment Twist Param Arrays = RigVMCollapseNode(ExecuteContext=(), Twist Joints=())
   Calculate Ref Segment Param = RigVMCollapseNode(Execute Context=())
   Calculate Twist Joint Metadata = RigVMCollapseNode(Debug=false, ExecuteContext=(), twist amounts=(), twist joints=())
   Compute Pole Vector Parent = RigVMCollapseNode(
    ExecuteContext=()
    Foot Control=(Type=Control,Name="foot_l_ik_ctrl")
    Initial=False
    Pole Vector Null=(Type=Control,Name="FootParent")
    Root Bone=(Type=Bone,Name="thigh_l")
   )
   Compute Twist = RigVMCollapseNode(End Transform=(), Execute=(), Position Blend=(), "Root Corrective "=(), Start Transform=(), Swing Blend=0.5, Twist Blend=(), Twist Bones=())
   Create Twist Controls = RigVMCollapseNode(
    ExecuteContext=()
    Parent=(Type=None,Name="None")
    Shape Settings=(bVisible=True,Name="Sphere_Thick",Color=(R=1,G=1,B=0,A=1),Transform=(Rotation=(X=0,Y=0,Z=0,W=1),Translation=(X=0,Y=0,Z=0),Scale3D=(X=0.5,Y=0.5,Z=0.5)))
    Twist Factors=()
    Twist Joints=()
    Twist Param Transforms=()
   )
   Direct Control Solve = RigVMCollapseNode(Bone=(Type=None,Name="None"), ExecuteContext=(), Forward Solve=False)
   FK Point At = RigVMCollapseNode(
    Bones=()
    ExecuteContext=()
    FK Controls=()
    Initial=false
    Negative Side=False
    Point At Axis=(X=1,Y=0,Z=0)
    Propagate To Children=true
    Up Vector Axis=(X=0,Y=0,Z=1)
    Up Vector Target=(X=0,Y=0,Z=1)
   )
   Find Finger Controls = RigVMCollapseNode(ExecuteContext=(), Value=None)
   Forward Arm = RigVMCollapseNode(
    Arm Space=(Name="None")
    Bones=()
    Enable Stretch=False
    Execute=()
    FK Controls=()
    FK Orient Control=(Name="None")
    FK Orient References=()
    Finger Space=(Name="None")
    IK Control=(Name="None")
    IK Effector=(Name="None")
    IK Orient=False
    IK Orient Control=(Name="None")
    IK Orient Null=(Name="None")
    IKFK Switch Control=(Name="None")
    Negative Side=False
    Pole Vector Control=(Name="None")
    Previous Align=False
    Previous IK Mode=False
    Softness=1
   )
   Forward Clavicle = RigVMCollapseNode(Bone=(), Control=(), ExecuteContext=())
   Forward Fingers = RigVMCollapseNode(Auto Base=False, Auto Meta=False, Base Fingers Controls=(), Execute=(), Fingers Bones=(), Metacarpals Controls=(), Proxy Controls=())
   Forward Foot = RigVMCollapseNode(Bone=(), ExecuteContext=(), FK Control=(), IK Control=(), IKFK Switch Control=(), Toes Space=())
   Forward Leg = RigVMCollapseNode(
    Bones=()
    Enable Stretch=False
    Execute=()
    FK Controls=()
    IK Control=(Name="None")
    IK Effector=(Name="None")
    IKFK Switch Control=(Name="None")
    Negative Side=False
    Pole Vector Control=(Name="None")
    Pole Vector Null=(Name="None")
    Previous IK Mode=False
    Secondary IK Controls=()
    Softness=1
   )
   Forward Neck = RigVMCollapseNode(
    Bones=()
    ExecuteContext=()
    FK Controls=()
    Head FK Space=()
    Head IK Space=()
    Head Orient Control=()
    Head Orient Reference=()
    IK Control=()
    IKFK Switch Control=()
    Mid IK Null=()
   )
   Forward Reverse Foot = RigVMCollapseNode(BK Buffers=(), BK Controls=(), ExecuteContext=(), IKFK Switch Control=(), Roll Buffer=(), Roll Clamp=30, Roll Control=())
   Forward Spine = RigVMCollapseNode(Bones=(), Chest Space=(), ExecuteContext=(), FK Controls=(), IK Controls=(), IKFK Switch Control=(), Pelvis Bone=(), Spline Offsets=())
   Get Negative Transform = RigVMCollapseNode(Input=(Rotation=(X=0,Y=0,Z=0,W=1),Translation=(X=0,Y=0,Z=0),Scale3D=(X=0,Y=0,Z=0)))
   IK 2 Bones = RigVMCollapseNode(
    Bones=()
    Enable Stretch=true
    ExecuteContext=()
    IK Control=(Type=Control,Name="hand_l_ik_ctrl")
    Pole Vector Control=(Type=Control,Name="arm_l_pv_ik_ctrl")
    Primary Axis=(X=1,Y=0,Z=0)
    Secondary Axis=(X=0,Y=-1,Z=0)
    Softness=1
   )
   Interpolate Interacting Controls = RigVMCollapseNode(Active=False, Controls=(), ExecuteContext=())
   Match 2 Bones IK onto FK = RigVMCollapseNode(
    Bones=()
    ExecuteContext=()
    IK Control=(Type=Control,Name="hand_l_ik_ctrl")
    Pole Vector Control=(Type=Control,Name="arm_l_pv_ik_ctrl")
    Secondary Controls=()
   )
   Project to New Parents = RigVMCollapseNode(Children=(), ExecuteContext=(), Parents=(), Propagate to Children=True)
   Reparent Items = RigVMCollapseNode(ExecuteContext=(), Items=(), Parent=(Type=None,Name="None"))
   Set IKFK Keys = RigVMCollapseNode(ExecuteContext=(), FK Controls=(), "IK Controls "=(), Switch Control=())
   Set IKFK Visibility = RigVMCollapseNode(ExecuteContext=(), FK Controls=(), IK Controls=(), IKFK Switch=True)
   Setup Arm = RigVMCollapseNode(Bones=(), ExecuteContext=(), FK Controls=(), FK Space=(), Fingers Space=(), IK Control=(), IK Parents=(), Pole Vector Control=())
   Setup Clavicle = RigVMCollapseNode(Bone=(), Control=(), ExecuteContext=())
   Setup Fingers = RigVMCollapseNode(Execute=(), Fingers Bones=(), Negative Side=False)
   Setup Foot = RigVMCollapseNode(Bone=(), ExecuteContext=(), FK Control=(), IK Control=(), Toes Space=())
   Setup Leg = RigVMCollapseNode(
    Bones=()
    ExecuteContext=()
    FK Controls=()
    FK Space=(Name="None")
    IK Control=(Name="None")
    IK Parents=()
    Negative Side=False
    Pole Vector Control=(Name="None")
    Pole Vector Null=(Name="None")
   )
   Setup Neck = RigVMCollapseNode(Bones=(), ExecuteContext=(), FK Controls=(), Head FK Space=(), Head IK Space=(), IK Control=())
   Setup Reverse Foot = RigVMCollapseNode(BK Buffers=(), BK Controls=(), Bones=(), ExecuteContext=(), Roll Buffer=(), Roll Control=())
   Setup Spine = RigVMCollapseNode(Bones=(), Chest Space=(), ExecuteContext=(), FK Controls=(), IK Controls=())
   Snap Twist Joints To Twist Controls = RigVMCollapseNode(ExecuteContext=(), Twist Joints=())
   Spine Match IK onto FK = RigVMCollapseNode(Bones=(), Execute=(), IK Controls=(), Offsets=())
  ]}
 RigVMModel {nodes [
   SequenceExecution_1 = RigVMAggregateNode(ExecuteContext=@n2.ExecutePin)
   Backward Clavicle_1 = RigVMFunctionReferenceNode(Bone=(Type=Bone,Name="clavicle_l"), Control=(Type=Control,Name="clavicle_l_ctrl"), Execute=@SequenceExecution_1.B)
   Backward Clavicle = RigVMFunctionReferenceNode(Bone=(Type=Bone,Name="clavicle_r"), Control=(Type=Control,Name="clavicle_r_ctrl"), Execute=@"Backward Clavicle_1".Execute)
   RerouteNode_12_1 = RigVMDispatchNode(Values=((Type=Control,Name="clavicle_l_ctrl"),((),Type=Control,Name="body_ctrl"))) {template="DISPATCH_RigVMDispatch_ArrayMake(in Values,out Array)"}
   RerouteNode_55_1 = RigVMDispatchNode(Values=((Type=Bone,Name="upperarm_l"),((),Type=Bone,Name="lowerarm_l"),((),Type=Bone,Name="hand_l"))) {template="DISPATCH_RigVMDispatch_ArrayMake(in Values,out Array)"}
   RerouteNode_56_1 = RigVMDispatchNode(Values=((Type=Control,Name="upperarm_l_fk_ctrl"),((),Type=Control,Name="lowerarm_l_fk_ctrl"),((),Type=Control,Name="hand_l_fk_ctrl"))) {template="DISPATCH_RigVMDispatch_ArrayMake(in Values,out Array)"}
   Backward Arm = RigVMFunctionReferenceNode(
    FK Arm Space=(Type=Null,Name="upperarm_l_fk_ctrl_space")
    Fingers Space=(Type=Null,Name="fingers_l_space")
    Ik Control=(Type=Control,Name="hand_l_ik_ctrl")
    UpV Control=(Type=Control,Name="arm_l_pv_ik_ctrl")
    ExecuteContext=@"Backward Clavicle".Execute
    FK Orient References=@RerouteNode_12_1.Array
    Bones=@RerouteNode_55_1.Array
    FK Controls=@RerouteNode_56_1.Array
   )
   RerouteNode_13_1 = RigVMDispatchNode(Values=((Type=Control,Name="clavicle_r_ctrl"),((),Type=Control,Name="body_ctrl"))) {template="DISPATCH_RigVMDispatch_ArrayMake(in Values,out Array)"}
   RerouteNode_57_1 = RigVMDispatchNode(Values=((Type=Control,Name="upperarm_r_fk_ctrl"),((),Type=Control,Name="lowerarm_r_fk_ctrl"),((),Type=Control,Name="hand_r_fk_ctrl"))) {template="DISPATCH_RigVMDispatch_ArrayMake(in Values,out Array)"}
   RerouteNode_58_1 = RigVMDispatchNode(Values=((Type=Bone,Name="upperarm_r"),((),Type=Bone,Name="lowerarm_r"),((),Type=Bone,Name="hand_r"))) {template="DISPATCH_RigVMDispatch_ArrayMake(in Values,out Array)"}
   > Backward Arm_1 = RigVMFunctionReferenceNode(
    FK Arm Space=(Type=Null,Name="upperarm_r_fk_ctrl_space")
    Fingers Space=(Type=Null,Name="fingers_r_space")
    Ik Control=(Type=Control,Name="hand_r_ik_ctrl")
    UpV Control=(Type=Control,Name="arm_r_pv_ik_ctrl")
    FK Orient References=@RerouteNode_13_1.Array
    FK Controls=@RerouteNode_57_1.Array
    Bones=@RerouteNode_58_1.Array
   )
   RerouteNode_59_2 = RigVMDispatchNode(Values="((Type=Bone,Name=\"thumb_01_l\"),((),Type=Bone,Name=\"thumb_02_l\"),((),Type=Bone,Name=\"thumb_03_l\"),((),Type=Bone,Name=\"index_metacarpal_l\"),((),Type=Bone,Name=\"index_01_l\"),((),Type=Bone,Name=\"index_02_l\"),((),Type=Bone,Name=\"index_03_l\"),((),Type=Bone,Name= ...(truncated)") {template="DISPATCH_RigVMDispatch_ArrayMake(in Values,out Array)"}
   Backward Fingers = RigVMFunctionReferenceNode(Execute=@"Backward Arm_1".ExecuteContext, Bones=@RerouteNode_59_2.Array)
   RerouteNode_15 = RigVMDispatchNode(Values=((Type=Control,Name="thigh_l_fk_ctrl"),((),Type=Control,Name="calf_l_fk_ctrl"),((),Type=Control,Name="foot_l_fk_ctrl"))) {template="DISPATCH_RigVMDispatch_ArrayMake(in Values,out Array)"}
   Make_Array = RigVMDispatchNode(Values=((Type=Control,Name="foot_roll_l_ctrl"),(Type=Control,Name="foot_bk1_l_ctrl"),((),Type=Control,Name="heel_l_ctrl"),((),Type=Control,Name="tip_l_ctrl"))) {template="DISPATCH_RigVMDispatch_ArrayMake(in Values,out Array)"}
   Forward Leg_1 = RigVMFunctionReferenceNode(
    IK Control=(Type=Control,Name="foot_l_ik_ctrl")
    IK Effector=(Type=Null,Name="foot_ik_l")
    IKFK Switch Control=(Type=Control,Name="leg_l_fk_ik_switch")
    Negative Side=False
    Pole Vector Control=(Type=Control,Name="leg_l_pv_ik_ctrl")
    Pole Vector Null=(Type=Null,Name="leg_l_pv_footFollow")
    Execute=@"Forward Reverse Foot_1".ExecuteContext
    FK Controls=@RerouteNode_15.Array
    Enable Stretch=&RigUnit_GetBoolAnimationChannel(Channel=Stretch, Control=foot_l_ik_ctrl, bInitial=False) {template="GetAnimationChannel::Execute(out Value,in Control,in Channel,in bInitial)"}.Value
    Softness=&RigUnit_GetFloatAnimationChannel(Channel=Softness, Control=foot_l_ik_ctrl, bInitial=False) {template="GetAnimationChannel::Execute(out Value,in Control,in Channel,in bInitial)"}.Value
    Previous IK Mode=RigVMVariableNode() {variable="L Leg IK Mode "}.Value
    Bones=RigVMVariableNode() {variable=leg_l_aimed}.Value
    Secondary IK Controls=@Make_Array.Array
   )
   Make_Array_1 = RigVMDispatchNode(Values=((Type=Control,Name="foot_roll_r_ctrl"),(Type=Control,Name="foot_bk1_r_ctrl"),((),Type=Control,Name="heel_r_ctrl"),((),Type=Control,Name="tip_r_ctrl"))) {template="DISPATCH_RigVMDispatch_ArrayMake(in Values,out Array)"}
   RerouteNode_14 = RigVMDispatchNode(Values=((Type=Control,Name="thigh_r_fk_ctrl"),((),Type=Control,Name="calf_r_fk_ctrl"),((),Type=Control,Name="foot_r_fk_ctrl"))) {template="DISPATCH_RigVMDispatch_ArrayMake(in Values,out Array)"}
   Forward Leg = RigVMFunctionReferenceNode(
    IK Control=(Type=Control,Name="foot_r_ik_ctrl")
    IK Effector=(Type=Null,Name="foot_ik_r")
    IKFK Switch Control=(Type=Control,Name="leg_r_fk_ik_switch")
    Negative Side=true
    Pole Vector Control=(Type=Control,Name="leg_r_pv_ik_ctrl")
    Pole Vector Null=(Type=Null,Name="leg_r_pv_footFollow")
    Execute=@"Forward Leg_1".Execute
    Secondary IK Controls=@Make_Array_1.Array
    FK Controls=@RerouteNode_14.Array
    Enable Stretch=&RigUnit_GetBoolAnimationChannel(Channel=Stretch, Control=foot_r_ik_ctrl, bInitial=False) {template="GetAnimationChannel::Execute(out Value,in Control,in Channel,in bInitial)"}.Value
    Softness=&RigUnit_GetFloatAnimationChannel(Channel=Softness, Control=foot_r_ik_ctrl, bInitial=False) {template="GetAnimationChannel::Execute(out Value,in Control,in Channel,in bInitial)"}.Value
    Previous IK Mode=RigVMVariableNode() {variable=R Leg IK Mode}.Value
    Bones=RigVMVariableNode() {variable=leg_r_aimed}.Value
   )
   RerouteNode_17 = RigVMDispatchNode(Values=((Type=Bone,Name="thigh_l"),((),Type=Bone,Name="calf_l"),((),Type=Bone,Name="foot_l"))) {template="DISPATCH_RigVMDispatch_ArrayMake(in Values,out Array)"}
   Project to New Parents_0_2 = RigVMFunctionReferenceNode(
    Propagate to Children=true
    ExecuteContext=@"Forward Leg".Execute
    Children=@RerouteNode_17.Array
    Parents=RigVMVariableNode() {variable=leg_l_aimed}.Value
   )
   RerouteNode_16 = RigVMDispatchNode(Values=((Type=Bone,Name="thigh_r"),((),Type=Bone,Name="calf_r"),((),Type=Bone,Name="foot_r"))) {template="DISPATCH_RigVMDispatch_ArrayMake(in Values,out Array)"}
   > Project to New Parents_0_1_1 = RigVMFunctionReferenceNode(Propagate to Children=true, Children=@RerouteNode_16.Array, Parents=RigVMVariableNode() {variable=leg_r_aimed}.Value)
   > Forward Foot = RigVMFunctionReferenceNode(
    Bone=(Type=Bone,Name="ball_l")
    FK Control=(Type=Control,Name="ball_l_fk_ctrl")
    IK Control=(Type=Control,Name="ball_l_ik_ctrl")
    IKFK Switch Control=(Type=Control,Name="leg_l_fk_ik_switch")
    Toes Space=(Type=Null,Name="toes_l_space")
    Previous IK Mode=RigVMVariableNode() {variable="L Leg IK Mode "}.Value
   )
   > Forward Foot_1 = RigVMFunctionReferenceNode(
    Bone=(Type=Bone,Name="ball_r")
    FK Control=(Type=Control,Name="ball_r_fk_ctrl")
    IK Control=(Type=Control,Name="ball_r_ik_ctrl")
    IKFK Switch Control=(Type=Control,Name="leg_r_fk_ik_switch")
    Toes Space=(Type=Null,Name="toes_r_space")
    Previous IK Mode=RigVMVariableNode() {variable=R Leg IK Mode}.Value
   )
   > RigVMVariableNode(Value=&RigUnit_GetBoolAnimationChannel(Channel=leg_l_fk_ik_switch, Control=global_ctrl, bInitial=False) {template="GetAnimationChannel::Execute(out Value,in Control,in Channel,in bInitial)"}.Value) {variable="L Leg IK Mode "}
   > n1 = RigVMVariableNode(Value=&RigUnit_GetBoolAnimationChannel(Channel=leg_r_fk_ik_switch, Control=global_ctrl, bInitial=False) {template="GetAnimationChannel::Execute(out Value,in Control,in Channel,in bInitial)"}.Value) {variable=R Leg IK Mode}
   RerouteNode_2_1 = RigVMDispatchNode(Values="((Type=Bone,Name=\"indextoe_01_l\"),((),Type=Bone,Name=\"indextoe_02_l\"),((),Type=Bone,Name=\"bigtoe_01_l\"),((),Type=Bone,Name=\"bigtoe_02_l\"),((),Type=Bone,Name=\"littletoe_01_l\"),((),Type=Bone,Name=\"littletoe_02_l\"),((),Type=Bone,Name=\"middletoe_01_l\"),((),Typ ...(truncated)") {template="DISPATCH_RigVMDispatch_ArrayMake(in Values,out Array)"}
   Forward Fingers_1 = RigVMFunctionReferenceNode(
    Auto Base=False
    Auto Meta=False
    Base Fingers Controls=()
    Metacarpals Controls=()
    Proxy Controls=()
    Fingers Bones=@RerouteNode_2_1.Array
    Execute=@n1.ExecuteContext
   )
   n2 = /Script/ControlRig.RigUnit_InverseExecution() {is=struct RigVMUnitNode}
   n3 = &RigUnit_PrepareForExecution()
   SequenceAggregate = RigVMAggregateNode(ExecuteContext=@n3.ExecutePin)
   RerouteNode_71 = RigVMDispatchNode(Values=((Type=Bone,Name="upperarm_l"),((),Type=Bone,Name="lowerarm_l"),((),Type=Bone,Name="hand_l"))) {template="DISPATCH_RigVMDispatch_ArrayMake(in Values,out Array)"}
   RerouteNode_72 = RigVMDispatchNode(Values=((Type=Control,Name="upperarm_l_fk_ctrl"),((),Type=Control,Name="lowerarm_l_fk_ctrl"),((),Type=Control,Name="hand_l_fk_ctrl"))) {template="DISPATCH_RigVMDispatch_ArrayMake(in Values,out Array)"}
   Build Aimed Chains = RigVMFunctionReferenceNode(
    Debug=false
    Do Aim=false
    IK Plane Twist Compensation=true
    Negative Side=false
    Parent=(Type=Bone,Name="clavicle_l")
    Point At Axis=(X=1,Y=0,Z=0)
    Up Vector Axis=(X=0,Y=0,Z=1)
    Up Vector Target=(X=0,Y=0,Z=1)
    Bones=@RerouteNode_71.Array
    FK Controls=@RerouteNode_72.Array
    ExecuteContext=@SequenceAggregate.A
   )
   > RigVMVariableNode(Value=@"Build Aimed Chains".Aimed Bones) {variable=arm_l_aimed}
   RerouteNode_71_1 = RigVMDispatchNode(Values=((Type=Bone,Name="upperarm_r"),((),Type=Bone,Name="lowerarm_r"),((),Type=Bone,Name="hand_r"))) {template="DISPATCH_RigVMDispatch_ArrayMake(in Values,out Array)"}
   RerouteNode_72_1 = RigVMDispatchNode(Values=((Type=Control,Name="upperarm_r_fk_ctrl"),((),Type=Control,Name="lowerarm_r_fk_ctrl"),((),Type=Control,Name="hand_r_fk_ctrl"))) {template="DISPATCH_RigVMDispatch_ArrayMake(in Values,out Array)"}
   > Build Aimed Chains_1 = RigVMFunctionReferenceNode(
    Debug=false
    Do Aim=false
    IK Plane Twist Compensation=true
    Negative Side=true
    Parent=(Type=Bone,Name="clavicle_r")
    Point At Axis=(X=1,Y=0,Z=0)
    Up Vector Axis=(X=0,Y=0,Z=1)
    Up Vector Target=(X=0,Y=0,Z=1)
    Bones=@RerouteNode_71_1.Array
    FK Controls=@RerouteNode_72_1.Array
   )
   > RigVMVariableNode(Value=@"Build Aimed Chains_1".Aimed Bones) {variable=arm_r_aimed}
   RerouteNode_15_1 = RigVMDispatchNode(Values=((Type=Control,Name="thigh_l_fk_ctrl"),((),Type=Control,Name="calf_l_fk_ctrl"),((),Type=Control,Name="foot_l_fk_ctrl"))) {template="DISPATCH_RigVMDispatch_ArrayMake(in Values,out Array)"}
   RerouteNode_17_1 = RigVMDispatchNode(Values=((Type=Bone,Name="thigh_l"),((),Type=Bone,Name="calf_l"),((),Type=Bone,Name="foot_l"))) {template="DISPATCH_RigVMDispatch_ArrayMake(in Values,out Array)"}
   > Build Aimed Chains_2 = RigVMFunctionReferenceNode(
    Debug=false
    Do Aim=false
    IK Plane Twist Compensation=true
    Negative Side=true
    Parent=(Type=Bone,Name="pelvis")
    Point At Axis=(X=-1,Y=0,Z=0)
    Up Vector Axis=(X=0,Y=0,Z=1)
    Up Vector Target=(X=0,Y=0,Z=1)
    FK Controls=@RerouteNode_15_1.Array
    Bones=@RerouteNode_17_1.Array
   )
   > RigVMVariableNode(Value=@"Build Aimed Chains_2".Aimed Bones) {variable=leg_l_aimed}
   RerouteNode_14_1 = RigVMDispatchNode(Values=((Type=Control,Name="thigh_r_fk_ctrl"),((),Type=Control,Name="calf_r_fk_ctrl"),((),Type=Control,Name="foot_r_fk_ctrl"))) {template="DISPATCH_RigVMDispatch_ArrayMake(in Values,out Array)"}
   RerouteNode_16_1 = RigVMDispatchNode(Values=((Type=Bone,Name="thigh_r"),((),Type=Bone,Name="calf_r"),((),Type=Bone,Name="foot_r"))) {template="DISPATCH_RigVMDispatch_ArrayMake(in Values,out Array)"}
   > Build Aimed Chains_2_1 = RigVMFunctionReferenceNode(
    Debug=false
    Do Aim=false
    IK Plane Twist Compensation=true
    Negative Side=false
    Parent=(Type=Bone,Name="pelvis")
    Point At Axis=(X=-1,Y=0,Z=0)
    Up Vector Axis=(X=0,Y=0,Z=1)
    Up Vector Target=(X=0,Y=0,Z=1)
    FK Controls=@RerouteNode_14_1.Array
    Bones=@RerouteNode_16_1.Array
   )
   > RigVMVariableNode(Value=@"Build Aimed Chains_2_1".Aimed Bones) {variable=leg_r_aimed}
   Build Control From Joint_0 = RigVMFunctionReferenceNode(
    Bone=(Type=Bone,Name="attach")
    Build Buffer Null=false
    Control Shape Settings="(DisplayName=\"None\",Shape=(bVisible=True,Name=\"Box_Thick\",Color=(R=1,G=1,B=0,A=1),Transform=(Rotation=(X=0,Y=0,Z=0,W=1),Translation=(X=0,Y=0,Z=0),Scale3D=(X=1,Y=1,Z=1))),Proxy=(bIsProxy=False,ShapeVisibility=BasedOnSelection),bSelectable=True,InitialSpace= ...(truncated)"
    Parent=(Type=Control,Name="body_offset_ctrl")
    ExecuteContext=@SequenceAggregate.B
   )
   > Build Control From Joint_2 = RigVMFunctionReferenceNode(
    Bone=(Type=Bone,Name="attach_emote_01")
    Build Buffer Null=false
    Control Shape Settings="(DisplayName=\"None\",Shape=(bVisible=True,Name=\"Box_Thick\",Color=(R=1,G=0.381,B=0.099,A=1),Transform=(Rotation=(X=0,Y=0,Z=0,W=1),Translation=(X=0,Y=0,Z=0),Scale3D=(X=0.75,Y=0.75,Z=0.75))),Proxy=(bIsProxy=False,ShapeVisibility=BasedOnSelection),bSelectable=T ...(truncated)"
    Parent=(Type=Control,Name="body_offset_ctrl")
   )
   > Build Control From Joint_3 = RigVMFunctionReferenceNode(
    Bone=(Type=Bone,Name="attach_emote_02")
    Build Buffer Null=false
    Control Shape Settings="(DisplayName=\"None\",Shape=(bVisible=True,Name=\"Box_Thick\",Color=(R=1,G=0.381,B=0.099,A=1),Transform=(Rotation=(X=0,Y=0,Z=0,W=1),Translation=(X=0,Y=0,Z=0),Scale3D=(X=0.65,Y=0.65,Z=0.65))),Proxy=(bIsProxy=False,ShapeVisibility=BasedOnSelection),bSelectable=T ...(truncated)"
    Parent=(Type=Control,Name="body_offset_ctrl")
   )
   > Build Control From Joint_4 = RigVMFunctionReferenceNode(
    Bone=(Type=Bone,Name="attach_emote_03")
    Build Buffer Null=false
    Control Shape Settings="(DisplayName=\"None\",Shape=(bVisible=True,Name=\"Box_Thick\",Color=(R=1,G=0.381,B=0.099,A=1),Transform=(Rotation=(X=0,Y=0,Z=0,W=1),Translation=(X=0,Y=0,Z=0),Scale3D=(X=0.55,Y=0.55,Z=0.55))),Proxy=(bIsProxy=False,ShapeVisibility=BasedOnSelection),bSelectable=T ...(truncated)"
    Parent=(Type=Control,Name="body_offset_ctrl")
   )
   SequenceExecution_3 = RigVMAggregateNode(ExecuteContext=@SequenceAggregate.C)
   RigUnit_SetTransform = &RigUnit_SetTransform(
    Item=(Type=Control,Name="body_ctrl")
    Value=(Rotation=(X=0,Y=0,Z=0,W=1),Translation=(X=0,Y=2.5434,Z=91.1834),Scale3D=(X=1,Y=1,Z=1))
    Weight=1
    bInitial=true
    Value.Translation=&RigVMFunction_MathTransformLerp(
     T=0.5
     A=&RigUnit_GetTransform(Item=(Type=Bone,Name="spine_01"), bInitial=true).Transform
     B=&RigUnit_GetTransform(Item=(Type=Bone,Name="spine_02"), bInitial=true).Transform
    ) {template="Interpolate::Execute(in A,in B,in T,out Result)"}.Result.Translation
    ExecutePin=@SequenceExecution_3.A
   )
   RerouteNode_11_1 = RigVMDispatchNode(Values=((Type=Control,Name="hips_ctrl"),((),Type=Control,Name="hips_tan_ctrl"),((),Type=Control,Name="chest_tan_ctrl"),((),Type=Control,Name="chest_ctrl"))) {template="DISPATCH_RigVMDispatch_ArrayMake(in Values,out Array)"}
   RerouteNode_52 = RigVMDispatchNode(Values=((Type=Bone,Name="spine_01"),((),Type=Bone,Name="spine_02"),((),Type=Bone,Name="spine_03"),((),Type=Bone,Name="spine_04"))) {template="DISPATCH_RigVMDispatch_ArrayMake(in Values,out Array)"}
   RerouteNode_53 = RigVMDispatchNode(Values=((Type=Control,Name="hips_ctrl"),((),Type=Control,Name="spine_01_ctrl"),((),Type=Control,Name="spine_02_ctrl"),((),Type=Control,Name="spine_03_ctrl"))) {template="DISPATCH_RigVMDispatch_ArrayMake(in Values,out Array)"}
   > Setup Spine = RigVMFunctionReferenceNode(
    Chest Space=(Type=Null,Name="chest_space")
    IK Controls=@RerouteNode_11_1.Array
    Bones=@RerouteNode_52.Array
    FK Controls=@RerouteNode_53.Array
   )
   > RigVMVariableNode(Value=@"Setup Spine".Spline Offsets) {variable=Spine Offsets}
   > RigVMVariableNode(Value=@"Setup Spine".Length) {variable=Spine Length}
   RerouteNode_51 = RigVMDispatchNode(Values=((Type=Bone,Name="neck_01"),((),Type=Bone,Name="neck_02"),((),Type=Bone,Name="head"))) {template="DISPATCH_RigVMDispatch_ArrayMake(in Values,out Array)"}
   RerouteNode_54 = RigVMDispatchNode(Values=((Type=Control,Name="neck_01_ctrl"),((),Type=Control,Name="neck_02_ctrl"),((),Type=Control,Name="head_ctrl"))) {template="DISPATCH_RigVMDispatch_ArrayMake(in Values,out Array)"}
   > Setup Neck_1 = RigVMFunctionReferenceNode(
    Head FK Space=(Type=Null,Name="head_fk_space")
    Head IK Space=(Type=Null,Name="head_ik_space")
    IK Control=(Type=Control,Name="head_ik_ctrl")
    Bones=@RerouteNode_51.Array
    FK Controls=@RerouteNode_54.Array
   )
   > RigVMVariableNode(Value=@"Setup Neck_1".Length) {variable=Neck Length}
   Setup Clavicle = RigVMFunctionReferenceNode(Bone=(Type=Bone,Name="clavicle_l"), Control=(Type=Control,Name="clavicle_l_ctrl"), ExecuteContext=@SequenceExecution_3.B)
   > Setup Clavicle_1 = RigVMFunctionReferenceNode(Bone=(Type=Bone,Name="clavicle_r"), Control=(Type=Control,Name="clavicle_r_ctrl"), Negative Side=true)
   RerouteNode = RigVMDispatchNode(Values=((Type=Control,Name="global_ctrl"),((),Type=Control,Name="attach_ctrl"),((),Type=Control,Name="body_ctrl"),((),Type=Control,Name="hips_ctrl"),((),Type=Null,Name="chest_space"),((),Type=Control,Name="clavicle_l_ctrl"))) {template="DISPATCH_RigVMDispatch_ArrayMake(in Values,out Array)"}
   RerouteNode_56 = RigVMDispatchNode(Values=((Type=Control,Name="upperarm_l_fk_ctrl"),((),Type=Control,Name="lowerarm_l_fk_ctrl"),((),Type=Control,Name="hand_l_fk_ctrl"))) {template="DISPATCH_RigVMDispatch_ArrayMake(in Values,out Array)"}
   > Setup Arm = RigVMFunctionReferenceNode(
    FK Space=(Type=Null,Name="upperarm_l_fk_ctrl_space")
    Fingers Space=(Type=Null,Name="fingers_l_space")
    IK Control=(Type=Control,Name="hand_l_ik_ctrl")
    Pole Vector Control=(Type=Control,Name="arm_l_pv_ik_ctrl")
    IK Parents=@RerouteNode.Array
    FK Controls=@RerouteNode_56.Array
    Bones=RigVMVariableNode() {variable=arm_l_aimed}.Value
   )
   RerouteNode_34 = RigVMDispatchNode(Values=((Type=Control,Name="global_ctrl"),((),Type=Control,Name="attach_ctrl"),((),Type=Control,Name="body_ctrl"),((),Type=Control,Name="hips_ctrl"),((),Type=Null,Name="chest_space"),((),Type=Control,Name="clavicle_r_ctrl"))) {template="DISPATCH_RigVMDispatch_ArrayMake(in Values,out Array)"}
   RerouteNode_57 = RigVMDispatchNode(Values=((Type=Control,Name="upperarm_r_fk_ctrl"),((),Type=Control,Name="lowerarm_r_fk_ctrl"),((),Type=Control,Name="hand_r_fk_ctrl"))) {template="DISPATCH_RigVMDispatch_ArrayMake(in Values,out Array)"}
   > Setup Arm_1 = RigVMFunctionReferenceNode(
    FK Space=(Type=Null,Name="upperarm_r_fk_ctrl_space")
    Fingers Space=(Type=Null,Name="fingers_r_space")
    IK Control=(Type=Control,Name="hand_r_ik_ctrl")
    Negative Side=true
    Pole Vector Control=(Type=Control,Name="arm_r_pv_ik_ctrl")
    IK Parents=@RerouteNode_34.Array
    FK Controls=@RerouteNode_57.Array
    Bones=RigVMVariableNode() {variable=arm_r_aimed}.Value
   )
   RerouteNode_45_1 = RigVMDispatchNode(Values="((Type=Bone,Name=\"thumb_01_l\"),((),Type=Bone,Name=\"thumb_02_l\"),((),Type=Bone,Name=\"thumb_03_l\"),((),Type=Bone,Name=\"index_metacarpal_l\"),((),Type=Bone,Name=\"index_01_l\"),((),Type=Bone,Name=\"index_02_l\"),((),Type=Bone,Name=\"index_03_l\"),((),Type=Bone,Name= ...(truncated)") {template="DISPATCH_RigVMDispatch_ArrayMake(in Values,out Array)"}
   Setup Fingers = RigVMFunctionReferenceNode(Negative Side=False, Fingers Bones=@RerouteNode_45_1.Array, Execute=@"Setup Arm_1".ExecuteContext)
   RerouteNode_35 = RigVMDispatchNode(Values=((Type=Control,Name="global_ctrl"),((),Type=Control,Name="attach_ctrl"),((),Type=Control,Name="body_ctrl"),((),Type=Control,Name="hips_ctrl"))) {template="DISPATCH_RigVMDispatch_ArrayMake(in Values,out Array)"}
   RerouteNode_61 = RigVMDispatchNode(Values=((Type=Control,Name="thigh_l_fk_ctrl"),((),Type=Control,Name="calf_l_fk_ctrl"),((),Type=Control,Name="foot_l_fk_ctrl"))) {template="DISPATCH_RigVMDispatch_ArrayMake(in Values,out Array)"}
   Setup Leg = RigVMFunctionReferenceNode(
    FK Space=(Type=Null,Name="thigh_l_fk_ctrl_space")
    IK Control=(Type=Control,Name="foot_l_ik_ctrl")
    Negative Side=false
    Pole Vector Control=(Type=Control,Name="leg_l_pv_ik_ctrl")
    Pole Vector Null=(Type=Null,Name="leg_l_pv_footFollow")
    IK Parents=@RerouteNode_35.Array
    FK Controls=@RerouteNode_61.Array
    ExecuteContext=@SequenceExecution_3.D
    Bones=RigVMVariableNode() {variable=leg_l_aimed}.Value
   )
   RerouteNode_35_1 = RigVMDispatchNode(Values=((Type=Control,Name="global_ctrl"),((),Type=Control,Name="attach_ctrl"),((),Type=Control,Name="body_ctrl"),((),Type=Control,Name="hips_ctrl"))) {template="DISPATCH_RigVMDispatch_ArrayMake(in Values,out Array)"}
   RerouteNode_62 = RigVMDispatchNode(Values=((Type=Control,Name="thigh_r_fk_ctrl"),((),Type=Control,Name="calf_r_fk_ctrl"),((),Type=Control,Name="foot_r_fk_ctrl"))) {template="DISPATCH_RigVMDispatch_ArrayMake(in Values,out Array)"}
   > Setup Leg_1 = RigVMFunctionReferenceNode(
    FK Space=(Type=Null,Name="thigh_r_fk_ctrl_space")
    IK Control=(Type=Control,Name="foot_r_ik_ctrl")
    Negative Side=true
    Pole Vector Control=(Type=Control,Name="leg_r_pv_ik_ctrl")
    Pole Vector Null=(Type=Null,Name="leg_r_pv_footFollow")
    IK Parents=@RerouteNode_35_1.Array
    FK Controls=@RerouteNode_62.Array
    Bones=RigVMVariableNode() {variable=leg_r_aimed}.Value
   )
   > Setup Foot = RigVMFunctionReferenceNode(
    Bone=(Type=Bone,Name="ball_l")
    FK Control=(Type=Control,Name="ball_l_fk_ctrl")
    IK Control=(Type=Control,Name="ball_l_ik_ctrl")
    Toes Space=(Type=Null,Name="toes_l_space")
   )
   > Setup Foot_1 = RigVMFunctionReferenceNode(
    Bone=(Type=Bone,Name="ball_r")
    FK Control=(Type=Control,Name="ball_r_fk_ctrl")
    IK Control=(Type=Control,Name="ball_r_ik_ctrl")
    Negative Side=true
    Toes Space=(Type=Null,Name="toes_r_space")
   )
   RerouteNode_2 = RigVMDispatchNode(Values="((Type=Bone,Name=\"indextoe_01_l\"),((),Type=Bone,Name=\"indextoe_02_l\"),((),Type=Bone,Name=\"bigtoe_01_l\"),((),Type=Bone,Name=\"bigtoe_02_l\"),((),Type=Bone,Name=\"littletoe_01_l\"),((),Type=Bone,Name=\"littletoe_02_l\"),((),Type=Bone,Name=\"middletoe_01_l\"),((),Typ ...(truncated)") {template="DISPATCH_RigVMDispatch_ArrayMake(in Values,out Array)"}
   Setup Fingers_2 = RigVMFunctionReferenceNode(Negative Side=False, Fingers Bones=@RerouteNode_2.Array, Execute=@"Setup Foot_1".ExecuteContext)
   &RigUnit_AddParent(Child=(Type=Control,Name="body_ctrl"), DisplayLabel=None, Parent=(Type=Control,Name="global_ctrl"), ExecutePin=@SequenceExecution_3.C)
   SequenceAggregate_1 = RigVMAggregateNode(ExecuteContext=@SequenceExecution_3.E)
   n4 = &RigVMFunction_ControlFlowBranch(ExecuteContext=@SequenceAggregate_1.A, Condition=RigVMVariableNode() {variable=Do Twist Controls}.Value)
   n5 = RigVMVariableNode() {variable=l_upperarm_twists}
   n6 = RigVMVariableNode() {variable=upperarm_twist_params}
   Calculate Twist Joint Metadata = RigVMFunctionReferenceNode(
    Debug=false
    segment end joint=lowerarm_l
    segment start joint=upperarm_l
    ExecuteContext=@n4.True
    twist joints=@n5.Value
    twist amounts=@n6.Value
   )
   > Create Twist Controls = RigVMFunctionReferenceNode(
    Parent=(Type=Control,Name="upperarm_l_fk_ctrl")
    Shape Settings=(bVisible=True,Name="Sphere_Thick",Color=(R=0,G=0,B=1,A=1),Transform=(Rotation=(X=0,Y=0,Z=0,W=1),Translation=(X=0,Y=0,Z=0),Scale3D=(X=0.5,Y=0.5,Z=0.5)))
    Twist Param Transforms=@"Calculate Twist Joint Metadata".Twist Param Transforms
    Twist Joints=@n5.Value
    Twist Factors=@n6.Value
   )
   n7 = RigVMVariableNode() {variable=l_lowerarm_twists}
   n8 = RigVMVariableNode() {variable=lowerarm_twist_params}
   > Calculate Twist Joint Metadata_1 = RigVMFunctionReferenceNode(Debug=false, segment end joint=hand_l_aimed, segment start joint=lowerarm_l, twist joints=@n7.Value, twist amounts=@n8.Value)
   > Create Twist Controls_1 = RigVMFunctionReferenceNode(
    Parent=(Type=Control,Name="lowerarm_l_fk_ctrl")
    Shape Settings=(bVisible=True,Name="Sphere_Thick",Color=(R=0,G=0,B=1,A=1),Transform=(Rotation=(X=0,Y=0,Z=0,W=1),Translation=(X=0,Y=0,Z=0),Scale3D=(X=0.5,Y=0.5,Z=0.5)))
    Twist Param Transforms=@"Calculate Twist Joint Metadata_1".Twist Param Transforms
    Twist Joints=@n7.Value
    Twist Factors=@n8.Value
   )
   n9 = RigVMVariableNode() {variable=l_thigh_twists}
   n10 = RigVMVariableNode() {variable=thigh_twist_params}
   > Calculate Twist Joint Metadata_1_1 = RigVMFunctionReferenceNode(Debug=false, segment end joint=calf_l, segment start joint=thigh_l, twist joints=@n9.Value, twist amounts=@n10.Value)
   > Create Twist Controls_1_1 = RigVMFunctionReferenceNode(
    Parent=(Type=Control,Name="thigh_l_fk_ctrl")
    Shape Settings=(bVisible=True,Name="Sphere_Thick",Color=(R=0,G=0,B=1,A=1),Transform=(Rotation=(X=0,Y=0,Z=0,W=1),Translation=(X=0,Y=0,Z=0),Scale3D=(X=0.5,Y=0.5,Z=0.5)))
    Twist Param Transforms=@"Calculate Twist Joint Metadata_1_1".Twist Param Transforms
    Twist Joints=@n9.Value
    Twist Factors=@n10.Value
   )
   n11 = RigVMVariableNode() {variable=l_calf_twists}
   n12 = RigVMVariableNode() {variable=calf_twist_params}
   > Calculate Twist Joint Metadata_1_1_1 = RigVMFunctionReferenceNode(Debug=false, segment end joint=foot_l, segment start joint=calf_l, twist joints=@n11.Value, twist amounts=@n12.Value)
   > Create Twist Controls_1_1_1 = RigVMFunctionReferenceNode(
    Parent=(Type=Control,Name="calf_l_fk_ctrl")
    Shape Settings=(bVisible=True,Name="Sphere_Thick",Color=(R=0,G=0,B=1,A=1),Transform=(Rotation=(X=0,Y=0,Z=0,W=1),Translation=(X=0,Y=0,Z=0),Scale3D=(X=0.5,Y=0.5,Z=0.5)))
    Twist Param Transforms=@"Calculate Twist Joint Metadata_1_1_1".Twist Param Transforms
    Twist Joints=@n11.Value
    Twist Factors=@n12.Value
   )
   n13 = RigVMVariableNode() {variable=r_upperarm_twists}
   > Calculate Twist Joint Metadata_2 = RigVMFunctionReferenceNode(Debug=false, segment end joint=lowerarm_r, segment start joint=upperarm_r, twist joints=@n13.Value, twist amounts=@n6.Value)
   > Create Twist Controls_2 = RigVMFunctionReferenceNode(
    Parent=(Type=Control,Name="upperarm_r_fk_ctrl")
    Shape Settings=(bVisible=True,Name="Sphere_Thick",Color=(R=1,G=0,B=0,A=1),Transform=(Rotation=(X=0,Y=0,Z=0,W=1),Translation=(X=0,Y=0,Z=0),Scale3D=(X=0.5,Y=0.5,Z=0.5)))
    Twist Param Transforms=@"Calculate Twist Joint Metadata_2".Twist Param Transforms
    Twist Joints=@n13.Value
    Twist Factors=@n6.Value
   )
   n14 = RigVMVariableNode() {variable=r_lowerarm_twists}
   > Calculate Twist Joint Metadata_1_2 = RigVMFunctionReferenceNode(Debug=false, segment end joint=hand_r, segment start joint=lowerarm_r, twist amounts=@n8.Value, twist joints=@n14.Value)
   > Create Twist Controls_1_2 = RigVMFunctionReferenceNode(
    Parent=(Type=Control,Name="lowerarm_r_fk_ctrl")
    Shape Settings=(bVisible=True,Name="Sphere_Thick",Color=(R=1,G=0,B=0,A=1),Transform=(Rotation=(X=0,Y=0,Z=0,W=1),Translation=(X=0,Y=0,Z=0),Scale3D=(X=0.5,Y=0.5,Z=0.5)))
    Twist Param Transforms=@"Calculate Twist Joint Metadata_1_2".Twist Param Transforms
    Twist Factors=@n8.Value
    Twist Joints=@n14.Value
   )
   n15 = RigVMVariableNode() {variable=r_thigh_twists}
   > Calculate Twist Joint Metadata_1_1_2 = RigVMFunctionReferenceNode(Debug=false, segment end joint=calf_r, segment start joint=thigh_r, twist amounts=@n10.Value, twist joints=@n15.Value)
   > Create Twist Controls_1_1_2 = RigVMFunctionReferenceNode(
    Parent=(Type=Control,Name="thigh_r_fk_ctrl")
    Shape Settings=(bVisible=True,Name="Sphere_Thick",Color=(R=1,G=0,B=0,A=1),Transform=(Rotation=(X=0,Y=0,Z=0,W=1),Translation=(X=0,Y=0,Z=0),Scale3D=(X=0.5,Y=0.5,Z=0.5)))
    Twist Param Transforms=@"Calculate Twist Joint Metadata_1_1_2".Twist Param Transforms
    Twist Factors=@n10.Value
    Twist Joints=@n15.Value
   )
   n16 = RigVMVariableNode() {variable=r_calf_twists}
   > Calculate Twist Joint Metadata_1_1_1_1 = RigVMFunctionReferenceNode(Debug=false, segment end joint=foot_r, segment start joint=calf_r, twist amounts=@n12.Value, twist joints=@n16.Value)
   > Create Twist Controls_1_1_1_1 = RigVMFunctionReferenceNode(
    Parent=(Type=Control,Name="calf_r_fk_ctrl")
    Shape Settings=(bVisible=True,Name="Sphere_Thick",Color=(R=1,G=0,B=0,A=1),Transform=(Rotation=(X=0,Y=0,Z=0,W=1),Translation=(X=0,Y=0,Z=0),Scale3D=(X=0.5,Y=0.5,Z=0.5)))
    Twist Param Transforms=@"Calculate Twist Joint Metadata_1_1_1_1".Twist Param Transforms
    Twist Factors=@n12.Value
    Twist Joints=@n16.Value
   )
   DISPATCH_RigVMDispatch_ArrayIterator = RigVMDispatchNode(ExecuteContext=@SequenceAggregate_1.B, Array=RigVMVariableNode() {variable=hand_l_attach_control_joints}.Value) {template="DISPATCH_RigVMDispatch_ArrayIterator(in Array,out Element,out Index,out Count,out Ratio)"}
   > Build Control From Joint = RigVMFunctionReferenceNode(
    Build Buffer Null=false
    Control Shape Settings="(bSelectable=True,InitialSpace=LocalSpace,bUsePreferredRotationOrder=False,PreferredRotationOrder=YZX,Limits=(LimitTranslationX=(bMinimum=False,bMaximum=False),LimitTranslationY=(bMinimum=False,bMaximum=False),LimitTranslationZ=(bMinimum=False,bMaximum=Fal ...(truncated)"
    Parent=(Type=Control,Name="hand_l_fk_ctrl")
    Bone=@DISPATCH_RigVMDispatch_ArrayIterator.Element
   )
   n17 = RigVMDispatchNode(ExecuteContext=@DISPATCH_RigVMDispatch_ArrayIterator.Completed, Array=RigVMVariableNode() {variable=hand_r_attach_control_joints}.Value) {template="DISPATCH_RigVMDispatch_ArrayIterator(in Array,out Element,out Index,out Count,out Ratio)"}
   > Build Control From Joint_1 = RigVMFunctionReferenceNode(
    Control Shape Settings="(bSelectable=True,InitialSpace=LocalSpace,bUsePreferredRotationOrder=False,PreferredRotationOrder=YZX,Limits=(LimitTranslationX=(bMinimum=False,bMaximum=False),LimitTranslationY=(bMinimum=False,bMaximum=False),LimitTranslationZ=(bMinimum=False,bMaximum=Fal ...(truncated)"
    Parent=(Type=Control,Name="hand_r_fk_ctrl")
    Bone=@n17.Element
   )
   Build Control From Joint_0_1 = RigVMFunctionReferenceNode(
    Bone=(Type=Bone,Name="attach_backpack")
    Control Shape Settings="(DisplayName=\"None\",Shape=(bVisible=True,Name=\"Box_Thick\",Color=(R=1,G=1,B=0,A=1),Transform=(Rotation=(X=0,Y=0,Z=0,W=1),Translation=(X=0,Y=0,Z=0),Scale3D=(X=0.75,Y=0.75,Z=0.75))),Proxy=(bIsProxy=False,ShapeVisibility=BasedOnSelection),bSelectable=True,Init ...(truncated)"
    Parent=(Type=Control,Name="spine_04_ctrl")
    ExecuteContext=@n17.Completed
   )
   > Build Control From Joint_0_1_1 = RigVMFunctionReferenceNode(
    Bone=(Type=Bone,Name="attach_cape")
    Control Shape Settings="(DisplayName=\"None\",Shape=(bVisible=True,Name=\"Box_Thick\",Color=(R=1,G=1,B=0,A=1),Transform=(Rotation=(X=0,Y=0,Z=0,W=1),Translation=(X=0,Y=0,Z=0),Scale3D=(X=0.75,Y=0.75,Z=0.75))),Proxy=(bIsProxy=False,ShapeVisibility=BasedOnSelection),bSelectable=True,Init ...(truncated)"
    Parent=(Type=Control,Name="spine_04_ctrl")
   )
   > Build Control From Joint_0_2 = RigVMFunctionReferenceNode(
    Bone=(Type=Bone,Name="root_fp")
    Build Buffer Null=false
    Control Shape Settings="(DisplayName=\"None\",Shape=(bVisible=True,Name=\"Square_Thin\",Color=(R=1,G=1,B=0,A=1),Transform=(Rotation=(X=0,Y=0,Z=0,W=1),Translation=(X=0,Y=0,Z=0),Scale3D=(X=4,Y=6.5,Z=1))),Proxy=(bIsProxy=False,ShapeVisibility=BasedOnSelection),bSelectable=True,InitialSp ...(truncated)"
    Parent=(Type=Control,Name="body_offset_ctrl")
   )
   > Build Control From Joint_0_2_1_1 = RigVMFunctionReferenceNode(
    Bone=(Type=Bone,Name="attach_fp")
    Build Buffer Null=false
    Control Shape Settings="(DisplayName=\"None\",Shape=(bVisible=True,Name=\"Arrow_Thin\",Color=(R=1,G=1,B=0,A=1),Transform=(Rotation=(X=-0.7071,Y=0,Z=-0.7071,W=0),Translation=(X=0,Y=0,Z=0),Scale3D=(X=2,Y=2,Z=1))),Proxy=(bIsProxy=False,ShapeVisibility=BasedOnSelection),bSelectable=True, ...(truncated)"
    Parent=@"Build Control From Joint_0_2".Argument
   )
   > Build Control From Joint_0_2_1 = RigVMFunctionReferenceNode(
    Bone=(Type=Bone,Name="camera_root")
    Build Buffer Null=false
    Control Shape Settings="(DisplayName=\"None\",Shape=(bVisible=True,Name=\"Square_Thin\",Color=(R=1,G=1,B=0,A=1),Transform=(Rotation=(X=0,Y=0,Z=0,W=1),Translation=(X=0,Y=0,Z=0),Scale3D=(X=3.5,Y=6,Z=1))),Proxy=(bIsProxy=False,ShapeVisibility=BasedOnSelection),bSelectable=True,InitialSp ...(truncated)"
    Parent=@"Build Control From Joint_0_2".Argument
   )
   > Build Control From Joint_0_2_1_2 = RigVMFunctionReferenceNode(
    Bone=(Type=Bone,Name="camera")
    Build Buffer Null=false
    Control Shape Settings="(DisplayName=\"None\",Shape=(bVisible=True,Name=\"Circle_Thin\",Color=(R=1,G=1,B=0,A=1),Transform=(Rotation=(X=0,Y=0,Z=0,W=1),Translation=(X=0,Y=0,Z=0),Scale3D=(X=2.5,Y=2.5,Z=2))),Proxy=(bIsProxy=False,ShapeVisibility=BasedOnSelection),bSelectable=True,Initial ...(truncated)"
    Parent=@"Build Control From Joint_0_2_1".Argument
   )
   Get Controls_1 = RigVMFunctionReferenceNode()
   RigVMVariableNode(Value=@"Get Controls_1".Items, ExecuteContext=@SequenceAggregate_1.D) {variable=ControlList}
   > SequenceAggregate_4 = RigVMAggregateNode()
   n18 = RigVMDispatchNode(Array=@"Get Controls_1".Items, ExecuteContext=@SequenceAggregate_4.A) {template="DISPATCH_RigVMDispatch_ArrayIterator(in Array,out Element,out Index,out Count,out Ratio)"}
   Contains = &RigVMFunction_Contains(Search=_r_, Name=@n18.Element.Name) {template="Contains::Execute(in Name,in Search,out Result)"}
   RigVMFunction_MathBoolOr = &RigVMFunction_MathBoolOr(
    A=@Contains.Result
    B=&RigVMFunction_Contains(Search=_l_, Name=@n18.Element.Name) {template="Contains::Execute(in Name,in Search,out Result)"}.Result
   )
   n19 = &RigVMFunction_ControlFlowBranch(ExecuteContext=@n18.ExecuteContext, Condition=@RigVMFunction_MathBoolOr.Result)
   n20 = RigVMDispatchNode(Array=(), ExecuteContext=@n19.False) {template="DISPATCH_RigVMDispatch_ArrayReset(io Array)"}
   > n21 = RigVMDispatchNode(Element=@n18.Element, Array=@n20.Array) {template="DISPATCH_RigVMDispatch_ArrayAdd(io Array,in Element,out Index)"}
   > Add Pose Mirroring Metadata_1 = RigVMFunctionReferenceNode(Mirror Axis=(X=0,Y=1,Z=1), Mirror Behavioral=false, Mirror No Flip=false, Control List=@n21.Array)
   DISPATCH_RigVMDispatch_ArrayReset = RigVMDispatchNode(Array=(), ExecuteContext=@n19.True) {template="DISPATCH_RigVMDispatch_ArrayReset(io Array)"}
   > DISPATCH_RigVMDispatch_ArrayAdd = RigVMDispatchNode(Element=@n18.Element, Array=@DISPATCH_RigVMDispatch_ArrayReset.Array) {template="DISPATCH_RigVMDispatch_ArrayAdd(io Array,in Element,out Index)"}
   > Add Pose Mirroring Metadata = RigVMFunctionReferenceNode(Mirror Axis=(X=0,Y=1,Z=1), Mirror Behavioral=true, Mirror No Flip=true, Control List=@DISPATCH_RigVMDispatch_ArrayAdd.Array)
   SpawnAnimationChannel = /Script/ControlRig.RigUnit_HierarchyAddAnimationChannelBool(
    InitialValue=False
    LimitsEnabled=()
    MaximumValue=True
    MinimumValue=False
    Name=spine_fk_auto_aim
    Parent=(Type=Control,Name="global_ctrl")
    ExecutePin=@SequenceAggregate_4.B
   ) {
    is=struct RigVMUnitNode
    template="SpawnAnimationChannel::Execute(in InitialValue,in MinimumValue,in MaximumValue,in LimitsEnabled,in Parent,in Name,out Item)"
   }
   n22 = RigVMDispatchNode(Array=@"Get Controls_1".Items, ExecuteContext=@SequenceAggregate_4.C) {template="DISPATCH_RigVMDispatch_ArrayIterator(in Array,out Element,out Index,out Count,out Ratio)"}
   > RigVMDispatchNode(
    Name=Control Display Scale Init
    NameSpace=Self
    Item=@n22.Element
    Value=/Script/ControlRig.RigUnit_HierarchyGetShapeSettings(Item=@n22.Element) {is=struct RigVMUnitNode}.Settings.Transform.Scale3D
   ) {template="DISPATCH_RigDispatch_SetMetadata(in Item,in Name,in NameSpace,in Value,out Success)"}
   &RigUnit_HierarchyAddAnimationChannelFloat(
    InitialValue=1
    LimitsEnabled=(Enabled=(bMinimum=True,bMaximum=True))
    MaximumValue=30
    MinimumValue=0
    Name=Control Display Scale
    Parent=(Type=Control,Name="global_ctrl")
    ExecutePin=@n22.Completed
   ) {template="SpawnAnimationChannel::Execute(in InitialValue,in MinimumValue,in MaximumValue,in LimitsEnabled,in Parent,in Name,out Item)"}
   For_Each = RigVMDispatchNode(Array=RigVMVariableNode() {variable=Gimbaled Controls}.Value, ExecuteContext=@SequenceAggregate_1.E) {template="DISPATCH_RigVMDispatch_ArrayIterator(in Array,out Element,out Index,out Count,out Ratio)"}
   n23 = &RigUnit_HierarchyGetShapeSettings(Item=@For_Each.Element)
   Replace = /Script/RigVM.RigVMFunction_NameReplace(New=_gimbal_ctrl, Old=_ctrl, Name=@For_Each.Element.Name) {is=struct RigVMUnitNode, template="Replace::Execute(in Name,in Old,in New,out Result)"}
   > n24 = &RigUnit_HierarchyAddControlTransform(
    InitialValue=(Rotation=(X=0,Y=0,Z=0,W=1),Translation=(X=0,Y=0,Z=0),Scale3D=(X=1,Y=1,Z=1))
    OffsetSpace=LocalSpace
    OffsetTransform=(Rotation=(X=0,Y=0,Z=0,W=1),Translation=(X=0,Y=0,Z=0),Scale3D=(X=1,Y=1,Z=1))
    Settings="(InitialSpace=LocalSpace,bUsePreferredRotationOrder=False,PreferredRotationOrder=YZX,Limits=(LimitTranslationX=(bMinimum=False,bMaximum=False),LimitTranslationY=(bMinimum=False,bMaximum=False),LimitTranslationZ=(bMinimum=False,bMaximum=False),LimitPitch=(b ...(truncated)"
    Parent=@For_Each.Element
    Settings.Shape.Color=@n23.Settings.Color
    Settings.Shape.Name=@n23.Settings.Name
    Settings.Shape.Transform.Rotation=@n23.Settings.Transform.Rotation
    Settings.Shape.Transform.Translation=@n23.Settings.Transform.Translation
    Settings.Shape.bVisible=@n23.Settings.bVisible
    Settings.Shape.Transform.Scale3D=&RigVMFunction_MathVectorMul(B=(X=0.75,Y=0.75,Z=0.75), A=@n23.Settings.Transform.Scale3D) {template="Multiply::Execute(in A,in B,out Result)"}.Result
    Name=@Replace.Result
   ) {template="SpawnControl::Execute(in InitialValue,in Settings,in OffsetTransform,in OffsetSpace,in Parent,in Name,out Item)"}
   > Set_Metadata = RigVMDispatchNode(Name=Gimbal Control, NameSpace=Self, Item=@For_Each.Element, Value=@n24.Item) {template="DISPATCH_RigDispatch_SetMetadata(in Item,in Name,in NameSpace,in Value,out Success)"}
   > For_Each_1 = RigVMDispatchNode(Array=/Script/ControlRig.RigUnit_CollectionChildrenArray(TypeToSearch=All, bDefaultChildren=True, bIncludeParent=False, bRecursive=False, Parent=@For_Each.Element) {is=struct RigVMUnitNode}.Items) {template="DISPATCH_RigVMDispatch_ArrayIterator(in Array,out Element,out Index,out Count,out Ratio)"}
   Not_Equals = RigVMDispatchNode(A=@For_Each_1.Element, B=@n24.Item) {template="DISPATCH_RigVMDispatch_CoreNotEquals(in A,in B,out Result)"}
   EndsWith = &RigVMFunction_EndsWith(Ending=_fk_local, Name=@For_Each_1.Element.Name) {template="EndsWith::Execute(in Name,in Ending,out Result)"}
   RigVMFunction_MathBoolNot = &RigVMFunction_MathBoolNot(Value=@EndsWith.Result)
   RigVMFunction_MathBoolAnd = &RigVMFunction_MathBoolAnd(A=@Not_Equals.Result, B=@RigVMFunction_MathBoolNot.Result)
   n25 = &RigVMFunction_ControlFlowBranch(ExecuteContext=@For_Each_1.ExecuteContext, Condition=@RigVMFunction_MathBoolAnd.Result)
   /Script/ControlRig.RigUnit_SetDefaultParent(Child=@For_Each_1.Element, ExecutePin=@n25.True, Parent=@n24.Item) {is=struct RigVMUnitNode}
   n26 = &RigUnit_BeginExecution()
   SequenceExecution = RigVMAggregateNode(ExecuteContext=@n26.ExecutePin)
   &RigUnit_SetTransform(
    Item=(Type=Bone,Name="root")
    Weight=1
    Value=&RigUnit_GetTransform(Item=(Type=Control,Name="root_ctrl")).Transform
    ExecutePin=@SequenceExecution.A
   )
   RerouteNode_11_1_1 = RigVMDispatchNode(Values=((Type=Control,Name="hips_ctrl"),((),Type=Control,Name="hips_tan_ctrl"),((),Type=Control,Name="chest_tan_ctrl"),((),Type=Control,Name="chest_ctrl"))) {template="DISPATCH_RigVMDispatch_ArrayMake(in Values,out Array)"}
   RerouteNode_21 = RigVMDispatchNode(Values=((Type=Control,Name="hips_ctrl"),((),Type=Control,Name="spine_01_ctrl"),((),Type=Control,Name="spine_02_ctrl"),((),Type=Control,Name="spine_03_ctrl"))) {template="DISPATCH_RigVMDispatch_ArrayMake(in Values,out Array)"}
   RerouteNode_4 = RigVMDispatchNode(Values=((Type=Bone,Name="spine_01"),((),Type=Bone,Name="spine_02"),((),Type=Bone,Name="spine_03"),((),Type=Bone,Name="spine_04"))) {template="DISPATCH_RigVMDispatch_ArrayMake(in Values,out Array)"}
   n27 = RigVMVariableNode() {variable=Spine IK Mode}
   > Forward Spine = RigVMFunctionReferenceNode(
    Chest Space=(Type=Null,Name="chest_space")
    IKFK Switch Control=(Type=Control,Name="spine_fk_ik_switch")
    Pelvis Bone=(Type=Bone,Name="pelvis")
    IK Controls=@RerouteNode_11_1_1.Array
    FK Controls=@RerouteNode_21.Array
    Bones=@RerouteNode_4.Array
    "Enable Stretch "=&RigUnit_GetBoolAnimationChannel(Channel=spine_stretch_switch, Control=chest_ctrl, bInitial=False) {template="GetAnimationChannel::Execute(out Value,in Control,in Channel,in bInitial)"}.Value
    Previous IK Mode=@n27.Value
    Default Length=RigVMVariableNode() {variable=Spine Length}.Value
    Spline Offsets=RigVMVariableNode() {variable=Spine Offsets}.Value
   )
   > n28 = RigVMVariableNode(Value=&RigUnit_GetBoolAnimationChannel(Channel=spine_fk_ik_switch, Control=global_ctrl, bInitial=False) {template="GetAnimationChannel::Execute(out Value,in Control,in Channel,in bInitial)"}.Value) {variable=Spine IK Mode}
   RigVMFunction_ControlFlowBranch = &RigVMFunction_ControlFlowBranch(ExecuteContext=@n28.ExecuteContext, Condition=@n27.Value)
   n29 = &RigUnit_ProjectTransformToNewParent(
    Child=(Type=Bone,Name="spine_05")
    NewParent=(Type=Control,Name="spine_04_ctrl")
    OldParent=(Type=Control,Name="spine_04_ctrl")
    bChildInitial=True
    bOldParentInitial=True
   )
   &RigUnit_SetTransform(
    Item=(Type=Bone,Name="spine_05")
    Weight=1
    bInitial=False
    bPropagateToChildren=true
    Value=@n29.Transform
    ExecutePin=@RigVMFunction_ControlFlowBranch.Completed
   )
   RigUnit_GetIntAnimationChannel = /Script/ControlRig.RigUnit_GetIntAnimationChannel(Channel=space, Control=head_ctrl, bInitial=False) {is=struct RigVMUnitNode, template="GetAnimationChannel::Execute(out Value,in Control,in Channel,in bInitial)"}
   DISPATCH_RigVMDispatch_ArrayGetAtIndex = RigVMDispatchNode(
    Index=@RigUnit_GetIntAnimationChannel.Value
    Array=&RigUnit_ItemArray(Items=((Type=Control,Name="neck_02_ctrl"),(Type=Control,Name="body_ctrl"),(Type=Control,Name="global_ctrl"))).Items
   ) {template="DISPATCH_RigVMDispatch_ArrayGetAtIndex(in Array,in Index,out Element)"}
   RerouteNode_5_1 = RigVMDispatchNode(Values=((Type=Control,Name="neck_01_ctrl"),((),Type=Control,Name="neck_02_ctrl"),((),Type=Control,Name="head_ctrl"))) {template="DISPATCH_RigVMDispatch_ArrayMake(in Values,out Array)"}
   RerouteNode_9 = RigVMDispatchNode(Values=((Type=Bone,Name="neck_01"),((),Type=Bone,Name="neck_02"),((),Type=Bone,Name="head"))) {template="DISPATCH_RigVMDispatch_ArrayMake(in Values,out Array)"}
   > Forward Neck_1 = RigVMFunctionReferenceNode(
    Head FK Space=(Type=Null,Name="head_fk_space")
    Head IK Space=(Type=Null,Name="head_ik_space")
    Head Orient Control=(Type=Control,Name="head_local")
    IK Control=(Type=Control,Name="head_ik_ctrl")
    IKFK Switch Control=(Type=Control,Name="neck_fk_ik_switch")
    Mid IK Null=(Type=Null,Name="head_mid")
    Head Orient Reference=@DISPATCH_RigVMDispatch_ArrayGetAtIndex.Element
    FK Controls=@RerouteNode_5_1.Array
    Bones=@RerouteNode_9.Array
    Enable Stretch=&RigUnit_GetBoolAnimationChannel(Channel=Stretch, Control=head_ctrl, bInitial=False) {template="GetAnimationChannel::Execute(out Value,in Control,in Channel,in bInitial)"}.Value
    Length=RigVMVariableNode() {variable=Neck Length}.Value
    Previous IK Mode=RigVMVariableNode() {variable=Neck IK Mode}.Value
   )
   > RigVMVariableNode(Value=&RigUnit_GetBoolAnimationChannel(Channel=neck_fk_ik_switch, Control=global_ctrl, bInitial=False) {template="GetAnimationChannel::Execute(out Value,in Control,in Channel,in bInitial)"}.Value) {variable=Neck IK Mode}
   GetAnimationChannel = &RigUnit_GetBoolAnimationChannel(Channel=spine_fk_auto_aim, Control=global_ctrl, bInitial=False) {template="GetAnimationChannel::Execute(out Value,in Control,in Channel,in bInitial)"}
   n30 = &RigVMFunction_ControlFlowBranch(Condition=@GetAnimationChannel.Value, ExecuteContext=@RigVMFunction_ControlFlowBranch.False)
   Project to New Parents_0_4 = RigVMFunctionReferenceNode(Propagate to Children=false, Parents=@RerouteNode_21.Array, Children=@RerouteNode_4.Array, ExecuteContext=@n30.False)
   Forward Clavicle_1 = RigVMFunctionReferenceNode(Bone=(Type=Bone,Name="clavicle_l"), Control=(Type=Control,Name="clavicle_l_ctrl"), ExecuteContext=@SequenceExecution.B)
   > Forward Clavicle = RigVMFunctionReferenceNode(Bone=(Type=Bone,Name="clavicle_r"), Control=(Type=Control,Name="clavicle_r_ctrl"))
   RerouteNode_12 = RigVMDispatchNode(Values=((Type=Control,Name="clavicle_l_ctrl"),((),Type=Control,Name="body_ctrl"))) {template="DISPATCH_RigVMDispatch_ArrayMake(in Values,out Array)"}
   RerouteNode_65 = RigVMDispatchNode(Values=((Type=Control,Name="upperarm_l_fk_ctrl"),((),Type=Control,Name="lowerarm_l_fk_ctrl"),((),Type=Control,Name="hand_l_fk_ctrl"))) {template="DISPATCH_RigVMDispatch_ArrayMake(in Values,out Array)"}
   RigUnit_GetBoolAnimationChannel = &RigUnit_GetBoolAnimationChannel(Channel=Stretch, Control=hand_l_ik_ctrl, bInitial=False) {template="GetAnimationChannel::Execute(out Value,in Control,in Channel,in bInitial)"}
   Forward Arm = RigVMFunctionReferenceNode(
    Arm Space=(Type=Null,Name="upperarm_l_fk_ctrl_space")
    FK Orient Control=(Type=Control,Name="arm_l_fk_local")
    Finger Space=(Type=Null,Name="fingers_l_space")
    IK Control=(Type=Control,Name="hand_l_ik_ctrl")
    IK Effector=(Type=Control,Name="hand_l_ik_ctrl")
    IK Orient Control=(Type=Control,Name="hand_l_ikorient_ctrl")
    IK Orient Null=(Type=Null,Name="hand_l_ikorient")
    IKFK Switch Control=(Type=Control,Name="arm_l_fk_ik_switch")
    Negative Side=false
    Pole Vector Control=(Type=Control,Name="arm_l_pv_ik_ctrl")
    Execute=@"Forward Clavicle".ExecuteContext
    FK Orient References=@RerouteNode_12.Array
    FK Controls=@RerouteNode_65.Array
    Enable Stretch=@RigUnit_GetBoolAnimationChannel.Value
    IK Orient=&RigUnit_GetBoolAnimationChannel(Channel=Keep Aligned, Control=hand_l_ik_ctrl, bInitial=False) {template="GetAnimationChannel::Execute(out Value,in Control,in Channel,in bInitial)"}.Value
    Softness=&RigUnit_GetFloatAnimationChannel(Channel=Softness, Control=hand_l_ik_ctrl, bInitial=False) {template="GetAnimationChannel::Execute(out Value,in Control,in Channel,in bInitial)"}.Value
    Previous IK Mode=RigVMVariableNode() {variable=L Arm IK Mode}.Value
    Previous Align=RigVMVariableNode() {variable=L Arm IK Align}.Value
    Bones=RigVMVariableNode() {variable=arm_l_aimed}.Value
   )
   RigVMVariableNode(
    ExecuteContext=@"Forward Arm".Execute
    Value=&RigUnit_GetBoolAnimationChannel(Channel=arm_l_fk_ik_switch, Control=global_ctrl, bInitial=False) {template="GetAnimationChannel::Execute(out Value,in Control,in Channel,in bInitial)"}.Value
   ) {variable=L Arm IK Mode}
   > n31 = RigVMVariableNode(Value=&RigUnit_GetBoolAnimationChannel(Channel=Keep Aligned, Control=hand_l_ik_ctrl, bInitial=False) {template="GetAnimationChannel::Execute(out Value,in Control,in Channel,in bInitial)"}.Value) {variable=L Arm IK Align}
   RerouteNode_13 = RigVMDispatchNode(Values=((Type=Control,Name="clavicle_r_ctrl"),((),Type=Control,Name="body_ctrl"))) {template="DISPATCH_RigVMDispatch_ArrayMake(in Values,out Array)"}
   RerouteNode_67 = RigVMDispatchNode(Values=((Type=Control,Name="upperarm_r_fk_ctrl"),((),Type=Control,Name="lowerarm_r_fk_ctrl"),((),Type=Control,Name="hand_r_fk_ctrl"))) {template="DISPATCH_RigVMDispatch_ArrayMake(in Values,out Array)"}
   RigUnit_GetFloatAnimationChannel = &RigUnit_GetFloatAnimationChannel(Channel=Softness, Control=hand_r_ik_ctrl, bInitial=False) {template="GetAnimationChannel::Execute(out Value,in Control,in Channel,in bInitial)"}
   Forward Arm_1 = RigVMFunctionReferenceNode(
    Arm Space=(Type=Null,Name="upperarm_r_fk_ctrl_space")
    FK Orient Control=(Type=Control,Name="arm_r_fk_local")
    Finger Space=(Type=Null,Name="fingers_r_space")
    IK Control=(Type=Control,Name="hand_r_ik_ctrl")
    IK Effector=(Type=Control,Name="hand_r_ik_ctrl")
    IK Orient Control=(Type=Control,Name="hand_r_ikorient_ctrl")
    IK Orient Null=(Type=Null,Name="hand_r_ikorient")
    IKFK Switch Control=(Type=Control,Name="arm_r_fk_ik_switch")
    Negative Side=true
    Pole Vector Control=(Type=Control,Name="arm_r_pv_ik_ctrl")
    FK Orient References=@RerouteNode_13.Array
    FK Controls=@RerouteNode_67.Array
    Enable Stretch=&RigUnit_GetBoolAnimationChannel(Channel=Stretch, Control=hand_r_ik_ctrl, bInitial=False) {template="GetAnimationChannel::Execute(out Value,in Control,in Channel,in bInitial)"}.Value
    IK Orient=&RigUnit_GetBoolAnimationChannel(Channel=Keep Aligned, Control=hand_r_ik_ctrl, bInitial=False) {template="GetAnimationChannel::Execute(out Value,in Control,in Channel,in bInitial)"}.Value
    Softness=@RigUnit_GetFloatAnimationChannel.Value
    Previous IK Mode=RigVMVariableNode() {variable=R Arm IK Mode}.Value
    Previous Align=RigVMVariableNode() {variable=R Arm IK Align}.Value
    Execute=@n31.ExecuteContext
    Bones=RigVMVariableNode() {variable=arm_r_aimed}.Value
   )
   RigVMVariableNode(
    ExecuteContext=@"Forward Arm_1".Execute
    Value=&RigUnit_GetBoolAnimationChannel(Channel=arm_r_fk_ik_switch, Control=global_ctrl, bInitial=False) {template="GetAnimationChannel::Execute(out Value,in Control,in Channel,in bInitial)"}.Value
   ) {variable=R Arm IK Mode}
   > RigVMVariableNode(Value=&RigUnit_GetBoolAnimationChannel(Channel=Keep Aligned, Control=hand_r_ik_ctrl, bInitial=False) {template="GetAnimationChannel::Execute(out Value,in Control,in Channel,in bInitial)"}.Value) {variable=R Arm IK Align}
   RerouteNode_66 = RigVMDispatchNode(Values=((Type=Bone,Name="upperarm_l"),((),Type=Bone,Name="lowerarm_l"),((),Type=Bone,Name="hand_l"))) {template="DISPATCH_RigVMDispatch_ArrayMake(in Values,out Array)"}
   > Project to New Parents_0 = RigVMFunctionReferenceNode(Propagate to Children=true, Children=@RerouteNode_66.Array, Parents=RigVMVariableNode() {variable=arm_l_aimed}.Value)
   RerouteNode_68 = RigVMDispatchNode(Values=((Type=Bone,Name="upperarm_r"),((),Type=Bone,Name="lowerarm_r"),((),Type=Bone,Name="hand_r"))) {template="DISPATCH_RigVMDispatch_ArrayMake(in Values,out Array)"}
   > Project to New Parents_0_1 = RigVMFunctionReferenceNode(Propagate to Children=true, Children=@RerouteNode_68.Array, Parents=RigVMVariableNode() {variable=arm_r_aimed}.Value)
   RerouteNode_44 = RigVMDispatchNode(Values=((Type=Control,Name="index_metacarpal_l_ctrl"),((),Type=Control,Name="middle_metacarpal_l_ctrl"),((),Type=Control,Name="ring_metacarpal_l_ctrl"),((),Type=Control,Name="pinky_metacarpal_l_ctrl"))) {template="DISPATCH_RigVMDispatch_ArrayMake(in Values,out Array)"}
   RerouteNode_45 = RigVMDispatchNode(Values="((Type=Bone,Name=\"thumb_01_l\"),((),Type=Bone,Name=\"thumb_02_l\"),((),Type=Bone,Name=\"thumb_03_l\"),((),Type=Bone,Name=\"index_metacarpal_l\"),((),Type=Bone,Name=\"index_01_l\"),((),Type=Bone,Name=\"index_02_l\"),((),Type=Bone,Name=\"index_03_l\"),((),Type=Bone,Name= ...(truncated)") {template="DISPATCH_RigVMDispatch_ArrayMake(in Values,out Array)"}
   RerouteNode_46 = RigVMDispatchNode(Values=((Type=Control,Name="index_01_l_ctrl"),((),Type=Control,Name="middle_01_l_ctrl"),((),Type=Control,Name="ring_01_l_ctrl"),((),Type=Control,Name="pinky_01_l_ctrl"))) {template="DISPATCH_RigVMDispatch_ArrayMake(in Values,out Array)"}
   RerouteNode_47 = RigVMDispatchNode(Values=((Type=Control,Name="thumb_curl_l_ctrl"),((),Type=Control,Name="index_curl_l_ctrl"),((),Type=Control,Name="middle_curl_l_ctrl"),((),Type=Control,Name="ring_curl_l_ctrl"),((),Type=Control,Name="pinky_curl_l_ctrl"))) {template="DISPATCH_RigVMDispatch_ArrayMake(in Values,out Array)"}
   Forward Fingers_4 = RigVMFunctionReferenceNode(
    Execute=@"Project to New Parents_0_1".ExecuteContext
    Metacarpals Controls=@RerouteNode_44.Array
    Fingers Bones=@RerouteNode_45.Array
    Base Fingers Controls=@RerouteNode_46.Array
    Proxy Controls=@RerouteNode_47.Array
    Auto Meta=&RigUnit_GetBoolAnimationChannel(Channel=Auto Interpolate, Control=pinky_metacarpal_l_ctrl, bInitial=False) {template="GetAnimationChannel::Execute(out Value,in Control,in Channel,in bInitial)"}.Value
    Auto Base=&RigUnit_GetBoolAnimationChannel(Channel=Auto Interpolate, Control=pinky_01_l_ctrl, bInitial=False) {template="GetAnimationChannel::Execute(out Value,in Control,in Channel,in bInitial)"}.Value
   )
   RerouteNode_7 = RigVMDispatchNode(Values=((Type=Null,Name="thigh_l_fk_ctrl_space"),((),Type=Null,Name="thigh_r_fk_ctrl_space"))) {template="DISPATCH_RigVMDispatch_ArrayMake(in Values,out Array)"}
   n32 = RigVMDispatchNode(Array=@RerouteNode_7.Array, ExecuteContext=@SequenceExecution.C) {template="DISPATCH_RigVMDispatch_ArrayIterator(in Array,out Element,out Index,out Count,out Ratio)"}
   n33 = &RigUnit_ProjectTransformToNewParent(
    NewParent=(Type=Control,Name="hips_ctrl")
    OldParent=(Type=Control,Name="hips_ctrl")
    bChildInitial=True
    bOldParentInitial=True
    Child=@n32.Element
   )
   > &RigUnit_SetTransform(Weight=1, Item=@n32.Element, Value=@n33.Transform)
   RerouteNode_25 = RigVMDispatchNode(Values=((Type=Null,Name="heel_l"),((),Type=Null,Name="tip_l"),((),Type=Null,Name="foot_bk1_l"))) {template="DISPATCH_RigVMDispatch_ArrayMake(in Values,out Array)"}
   RerouteNode_28 = RigVMDispatchNode(Values=((Type=Control,Name="heel_l_ctrl"),((),Type=Control,Name="tip_l_ctrl"),((),Type=Control,Name="foot_bk1_l_ctrl"))) {template="DISPATCH_RigVMDispatch_ArrayMake(in Values,out Array)"}
   RigUnit_GetControlFloat = /Script/ControlRig.RigUnit_GetControlFloat(Control=roll_blend_l_ctrl) {is=struct RigVMUnitNode, template="GetControlValue::Execute(in Control,out FloatValue,out Minimum,out Maximum)"}
   Forward Reverse Foot = RigVMFunctionReferenceNode(
    IKFK Switch Control=(Type=Control,Name="leg_l_fk_ik_switch")
    Roll Buffer=(Type=Null,Name="foot_roll_l_bfr")
    Roll Clamp=40
    Roll Control=(Type=Control,Name="foot_roll_l_ctrl")
    ExecuteContext=@n32.Completed
    BK Buffers=@RerouteNode_25.Array
    BK Controls=@RerouteNode_28.Array
    Roll Blend=@RigUnit_GetControlFloat.FloatValue
   )
   RerouteNode_32 = RigVMDispatchNode(Values=((Type=Null,Name="heel_r"),((),Type=Null,Name="tip_r"),((),Type=Null,Name="foot_bk1_r"))) {template="DISPATCH_RigVMDispatch_ArrayMake(in Values,out Array)"}
   RerouteNode_33 = RigVMDispatchNode(Values=((Type=Control,Name="heel_r_ctrl"),((),Type=Control,Name="tip_r_ctrl"),((),Type=Control,Name="foot_bk1_r_ctrl"))) {template="DISPATCH_RigVMDispatch_ArrayMake(in Values,out Array)"}
   > Forward Reverse Foot_1 = RigVMFunctionReferenceNode(
    IKFK Switch Control=(Type=Control,Name="leg_r_fk_ik_switch")
    Roll Buffer=(Type=Null,Name="foot_roll_r_bfr")
    Roll Clamp=40
    Roll Control=(Type=Control,Name="foot_roll_r_ctrl")
    BK Buffers=@RerouteNode_32.Array
    BK Controls=@RerouteNode_33.Array
    Roll Blend=&RigUnit_GetControlFloat(Control=roll_blend_r_ctrl) {template="GetControlValue::Execute(in Control,out FloatValue,out Minimum,out Maximum)"}.FloatValue
   )
   SequenceAggregate_2 = RigVMAggregateNode(ExecuteContext=@SequenceExecution.D)
   n34 = &RigVMFunction_ControlFlowBranch(ExecuteContext=@SequenceAggregate_2.A, Condition=RigVMVariableNode() {variable=Do Twist Controls}.Value)
   n35 = RigVMVariableNode() {variable=l_upperarm_twists}
   Build Segment Twist Param Arrays = RigVMFunctionReferenceNode(ExecuteContext=@n34.True, Twist Joints=@n35.Value)
   Compute Twist = RigVMFunctionReferenceNode(
    End Transform=(Type=Control,Name="lowerarm_l_fk_ctrl")
    "Root Corrective "=(Type=Bone,Name="upperarm_correctiveRoot_l")
    Start Transform=(Type=Control,Name="upperarm_l_fk_ctrl")
    Swing Blend=0.5
    Twist From End=false
    Execute=@"Build Segment Twist Param Arrays".ExecuteContext
    Position Blend=@"Build Segment Twist Param Arrays".Location Params
    Twist Blend=@"Build Segment Twist Param Arrays".Twist Params
    Twist Bones=@n35.Value
   )
   Snap Twist Joints To Twist Controls_8 = RigVMFunctionReferenceNode(ExecuteContext=@"Compute Twist".Execute, Twist Joints=@n35.Value)
   n36 = RigVMVariableNode() {variable=l_lowerarm_twists}
   > Build Segment Twist Param Arrays_1 = RigVMFunctionReferenceNode(Twist Joints=@n36.Value)
   Compute Twist_1 = RigVMFunctionReferenceNode(
    End Transform=(Type=Control,Name="hand_l_fk_ctrl")
    "Root Corrective "=(Type=Bone,Name="lowerarm_correctiveRoot_l")
    Start Transform=(Type=Control,Name="lowerarm_l_fk_ctrl")
    Swing Blend=0.5
    Twist From End=true
    Execute=@"Build Segment Twist Param Arrays_1".ExecuteContext
    Position Blend=@"Build Segment Twist Param Arrays_1".Location Params
    Twist Blend=@"Build Segment Twist Param Arrays_1".Twist Params
    Twist Bones=@n36.Value
   )
   Snap Twist Joints To Twist Controls_8_1 = RigVMFunctionReferenceNode(ExecuteContext=@"Compute Twist_1".Execute, Twist Joints=@n36.Value)
   n37 = RigVMVariableNode() {variable=r_upperarm_twists}
   > Build Segment Twist Param Arrays_1_1 = RigVMFunctionReferenceNode(Twist Joints=@n37.Value)
   Compute Twist_1_1 = RigVMFunctionReferenceNode(
    End Transform=(Type=Control,Name="lowerarm_r_fk_ctrl")
    "Root Corrective "=(Type=Bone,Name="upperarm_correctiveRoot_r")
    Start Transform=(Type=Control,Name="upperarm_r_fk_ctrl")
    Swing Blend=0.5
    Twist From End=false
    Execute=@"Build Segment Twist Param Arrays_1_1".ExecuteContext
    Position Blend=@"Build Segment Twist Param Arrays_1_1".Location Params
    Twist Blend=@"Build Segment Twist Param Arrays_1_1".Twist Params
    Twist Bones=@n37.Value
   )
   Snap Twist Joints To Twist Controls_8_1_1 = RigVMFunctionReferenceNode(ExecuteContext=@"Compute Twist_1_1".Execute, Twist Joints=@n37.Value)
   n38 = RigVMVariableNode() {variable=r_lowerarm_twists}
   > Build Segment Twist Param Arrays_1_1_1 = RigVMFunctionReferenceNode(Twist Joints=@n38.Value)
   Compute Twist_1_1_1 = RigVMFunctionReferenceNode(
    End Transform=(Type=Control,Name="hand_r_fk_ctrl")
    "Root Corrective "=(Type=Bone,Name="lowerarm_correctiveRoot_r")
    Start Transform=(Type=Control,Name="lowerarm_r_fk_ctrl")
    Swing Blend=0.5
    Twist From End=true
    Execute=@"Build Segment Twist Param Arrays_1_1_1".ExecuteContext
    Position Blend=@"Build Segment Twist Param Arrays_1_1_1".Location Params
    Twist Blend=@"Build Segment Twist Param Arrays_1_1_1".Twist Params
    Twist Bones=@n38.Value
   )
   Snap Twist Joints To Twist Controls_8_1_1_1 = RigVMFunctionReferenceNode(ExecuteContext=@"Compute Twist_1_1_1".Execute, Twist Joints=@n38.Value)
   n39 = RigVMVariableNode() {variable=l_thigh_twists}
   > Build Segment Twist Param Arrays_1_1_1_1 = RigVMFunctionReferenceNode(Twist Joints=@n39.Value)
   Compute Twist_1_1_1_1 = RigVMFunctionReferenceNode(
    End Transform=(Type=Control,Name="calf_l_fk_ctrl")
    "Root Corrective "=(Type=Bone,Name="thigh_correctiveRoot_l")
    Start Transform=(Type=Control,Name="thigh_l_fk_ctrl")
    Swing Blend=0.5
    Execute=@"Build Segment Twist Param Arrays_1_1_1_1".ExecuteContext
    Position Blend=@"Build Segment Twist Param Arrays_1_1_1_1".Location Params
    Twist Blend=@"Build Segment Twist Param Arrays_1_1_1_1".Twist Params
    Twist Bones=@n39.Value
   )
   Snap Twist Joints To Twist Controls_8_1_1_1_1 = RigVMFunctionReferenceNode(ExecuteContext=@"Compute Twist_1_1_1_1".Execute, Twist Joints=@n39.Value)
   n40 = RigVMVariableNode() {variable=l_calf_twists}
   > Build Segment Twist Param Arrays_1_1_1_1_1 = RigVMFunctionReferenceNode(Twist Joints=@n40.Value)
   Compute Twist_1_1_1_1_1 = RigVMFunctionReferenceNode(
    End Transform=(Type=Control,Name="foot_l_fk_ctrl")
    "Root Corrective "=(Type=Bone,Name="calf_correctiveRoot_l")
    Start Transform=(Type=Control,Name="calf_l_fk_ctrl")
    Swing Blend=0.5
    Twist From End=true
    Execute=@"Build Segment Twist Param Arrays_1_1_1_1_1".ExecuteContext
    Position Blend=@"Build Segment Twist Param Arrays_1_1_1_1_1".Location Params
    Twist Blend=@"Build Segment Twist Param Arrays_1_1_1_1_1".Twist Params
    Twist Bones=@n40.Value
   )
   Snap Twist Joints To Twist Controls_8_1_1_1_1_1 = RigVMFunctionReferenceNode(ExecuteContext=@"Compute Twist_1_1_1_1_1".Execute, Twist Joints=@n40.Value)
   n41 = RigVMVariableNode() {variable=r_thigh_twists}
   > Build Segment Twist Param Arrays_1_1_1_1_1_1_1 = RigVMFunctionReferenceNode(Twist Joints=@n41.Value)
   Compute Twist_1_1_1_1_1_1 = RigVMFunctionReferenceNode(
    End Transform=(Type=Control,Name="calf_r_fk_ctrl")
    "Root Corrective "=(Type=Bone,Name="thigh_correctiveRoot_r")
    Start Transform=(Type=Control,Name="thigh_r_fk_ctrl")
    Swing Blend=0.5
    Execute=@"Build Segment Twist Param Arrays_1_1_1_1_1_1_1".ExecuteContext
    Position Blend=@"Build Segment Twist Param Arrays_1_1_1_1_1_1_1".Location Params
    Twist Blend=@"Build Segment Twist Param Arrays_1_1_1_1_1_1_1".Twist Params
    Twist Bones=@n41.Value
   )
   Snap Twist Joints To Twist Controls_8_1_1_1_1_1_1 = RigVMFunctionReferenceNode(ExecuteContext=@"Compute Twist_1_1_1_1_1_1".Execute, Twist Joints=@n41.Value)
   n42 = RigVMVariableNode() {variable=r_calf_twists}
   > Build Segment Twist Param Arrays_1_1_1_1_1_1 = RigVMFunctionReferenceNode(Twist Joints=@n42.Value)
   Compute Twist_1_1_1_1_1_1_1 = RigVMFunctionReferenceNode(
    End Transform=(Type=Control,Name="foot_r_fk_ctrl")
    "Root Corrective "=(Type=Bone,Name="calf_correctiveRoot_r")
    Start Transform=(Type=Control,Name="calf_r_fk_ctrl")
    Swing Blend=0.5
    Twist From End=true
    Execute=@"Build Segment Twist Param Arrays_1_1_1_1_1_1".ExecuteContext
    Position Blend=@"Build Segment Twist Param Arrays_1_1_1_1_1_1".Location Params
    Twist Blend=@"Build Segment Twist Param Arrays_1_1_1_1_1_1".Twist Params
    Twist Bones=@n42.Value
   )
   Snap Twist Joints To Twist Controls_8_1_1_1_1_1_1_1 = RigVMFunctionReferenceNode(ExecuteContext=@"Compute Twist_1_1_1_1_1_1_1".Execute, Twist Joints=@n42.Value)
   n43 = RigVMDispatchNode(ExecuteContext=@SequenceAggregate_2.B, Array=RigVMVariableNode() {variable=hand_l_attach_control_joints}.Value) {template="DISPATCH_RigVMDispatch_ArrayIterator(in Array,out Element,out Index,out Count,out Ratio)"}
   > Direct Control Solve = RigVMFunctionReferenceNode(Forward Solve=true, Bone=@n43.Element)
   n44 = RigVMDispatchNode(ExecuteContext=@n43.Completed, Array=RigVMVariableNode() {variable=hand_r_attach_control_joints}.Value) {template="DISPATCH_RigVMDispatch_ArrayIterator(in Array,out Element,out Index,out Count,out Ratio)"}
   > Direct Control Solve_1 = RigVMFunctionReferenceNode(Forward Solve=true, Bone=@n44.Element)
   > Direct Control Solve_0_2 = RigVMFunctionReferenceNode(Bone=(Type=Bone,Name="attach"), Forward Solve=true)
   > Direct Control Solve_0_1_2 = RigVMFunctionReferenceNode(Bone=(Type=Bone,Name="attach_backpack"), Forward Solve=true)
   > Direct Control Solve_0_1_1_1 = RigVMFunctionReferenceNode(Bone=(Type=Bone,Name="attach_cape"), Forward Solve=true)
   > Direct Control Solve_6 = RigVMFunctionReferenceNode(Bone=(Type=Bone,Name="attach_emote_01"), Forward Solve=true)
   > Direct Control Solve_7 = RigVMFunctionReferenceNode(Bone=(Type=Bone,Name="attach_emote_02"), Forward Solve=true)
   > Direct Control Solve_8 = RigVMFunctionReferenceNode(Bone=(Type=Bone,Name="attach_emote_03"), Forward Solve=true)
   > Direct Control Solve_0_1_1_2_2 = RigVMFunctionReferenceNode(Bone=(Type=Bone,Name="root_fp"), Forward Solve=true)
   > Direct Control Solve_0_1_1_2_1_2 = RigVMFunctionReferenceNode(Bone=(Type=Bone,Name="attach_fp"), Forward Solve=true)
   > Direct Control Solve_0_1_1_2_1_1_2 = RigVMFunctionReferenceNode(Bone=(Type=Bone,Name="camera_root"), Forward Solve=true)
   > Direct Control Solve_0_1_1_2_1_1_1_1 = RigVMFunctionReferenceNode(Bone=(Type=Bone,Name="camera"), Forward Solve=true)
   n45 = RigVMDispatchNode(ExecuteContext=@SequenceAggregate_2.D, Array=RigVMVariableNode() {variable=ControlList}.Value) {template="DISPATCH_RigVMDispatch_ArrayIterator(in Array,out Element,out Index,out Count,out Ratio)"}
   n46 = /Script/ControlRig.RigUnit_GetShapeTransform(Control=@n45.Element.Name) {is=struct RigVMUnitNode}
   DISPATCH_RigDispatch_GetMetadata = RigVMDispatchNode(Default=(X=1,Y=1,Z=1), Name=Control Display Scale Init, NameSpace=Self, Item=@n45.Element) {template="DISPATCH_RigDispatch_GetMetadata(in Item,in Name,in NameSpace,in Default,out Value,out Found)"}
   n47 = &RigUnit_GetFloatAnimationChannel(Channel=Control Display Scale, Control=global_Ctrl, bInitial=False) {template="GetAnimationChannel::Execute(out Value,in Control,in Channel,in bInitial)"}
   Multiply = &RigVMFunction_MathVectorMul(B=(X=1,Y=1,Z=1), A=@DISPATCH_RigDispatch_GetMetadata.Value, B.X=@n47.Value, B.Y=@n47.Value, B.Z=@n47.Value) {template="Multiply::Execute(in A,in B,out Result)"}
   > /Script/ControlRig.RigUnit_SetShapeTransform(
    Transform=(Rotation=(X=0,Y=0,Z=0,W=1),Translation=(X=0,Y=0,Z=0),Scale3D=(X=1,Y=1,Z=1))
    Control=@n45.Element.Name
    Transform.Rotation=@n46.Transform.Rotation
    Transform.Translation=@n46.Transform.Translation
    Transform.Scale3D=@Multiply.Result
   ) {is=struct RigVMUnitNode}
   RigUnit_ItemArray = &RigUnit_ItemArray(Items=((Type=Bone,Name="ik_foot_root"),(Type=Bone,Name="ik_hand_root"),(Type=Bone,Name="ik_foot_l"),(Type=Bone,Name="ik_foot_r"),(Type=Bone,Name="ik_hand_gun"),(Type=Bone,Name="ik_hand_l"),(Type=Bone,Name="ik_hand_r")))
   n48 = &RigUnit_ItemArray(Items=((Type=Bone,Name="root"),(Type=Bone,Name="root"),(Type=Bone,Name="foot_l"),(Type=Bone,Name="foot_r"),(Type=Bone,Name="hand_r"),(Type=Bone,Name="hand_l"),(Type=Bone,Name="hand_r")))
   Project to New Parents_0_3 = RigVMFunctionReferenceNode(Propagate to Children=true, Children=@RigUnit_ItemArray.Items, Parents=@n48.Items, ExecuteContext=@SequenceAggregate_2.C)
   RerouteNode_63 = RigVMDispatchNode(Values="((Type=Bone,Name=\"littletoe_01_r\"),((),Type=Bone,Name=\"littletoe_02_r\"),((),Type=Bone,Name=\"middletoe_01_r\"),((),Type=Bone,Name=\"middletoe_02_r\"),((),Type=Bone,Name=\"bigtoe_01_r\"),((),Type=Bone,Name=\"bigtoe_02_r\"),((),Type=Bone,Name=\"indextoe_01_r\"),((),Ty ...(truncated)") {template="DISPATCH_RigVMDispatch_ArrayMake(in Values,out Array)"}
   Setup Fingers_2_1 = RigVMFunctionReferenceNode(Negative Side=true, Fingers Bones=@RerouteNode_63.Array, Execute=@"Setup Fingers_2".Execute)
   RerouteNode_11 = RigVMDispatchNode(Values=((Type=Bone,Name="heel_l"),((),Type=Bone,Name="tip_l"),((),Type=Bone,Name="ball_l"),((),Type=Bone,Name="foot_l"))) {template="DISPATCH_RigVMDispatch_ArrayMake(in Values,out Array)"}
   RerouteNode_18 = RigVMDispatchNode(Values=((Type=Control,Name="heel_l_ctrl"),((),Type=Control,Name="tip_l_ctrl"),((),Type=Control,Name="foot_bk1_l_ctrl"),((),Type=Null,Name="foot_ik_l"))) {template="DISPATCH_RigVMDispatch_ArrayMake(in Values,out Array)"}
   RerouteNode_23 = RigVMDispatchNode(Values=((Type=Null,Name="heel_l"),((),Type=Null,Name="tip_l"),((),Type=Null,Name="foot_bk1_l"),((),Type=Null,Name="foot_ik_l"))) {template="DISPATCH_RigVMDispatch_ArrayMake(in Values,out Array)"}
   Setup Reverse Foot = RigVMFunctionReferenceNode(
    Roll Buffer=(Type=Null,Name="foot_roll_l_bfr")
    Roll Control=(Type=Control,Name="foot_roll_l_ctrl")
    Bones=@RerouteNode_11.Array
    BK Controls=@RerouteNode_18.Array
    BK Buffers=@RerouteNode_23.Array
    ExecuteContext=@"Setup Fingers_2_1".Execute
   )
   RerouteNode_29 = RigVMDispatchNode(Values=((Type=Null,Name="heel_r"),((),Type=Null,Name="tip_r"),((),Type=Null,Name="foot_bk1_r"),((),Type=Null,Name="foot_ik_r"))) {template="DISPATCH_RigVMDispatch_ArrayMake(in Values,out Array)"}
   RerouteNode_30 = RigVMDispatchNode(Values=((Type=Control,Name="heel_r_ctrl"),((),Type=Control,Name="tip_r_ctrl"),((),Type=Control,Name="foot_bk1_r_ctrl"),((),Type=Null,Name="foot_ik_r"))) {template="DISPATCH_RigVMDispatch_ArrayMake(in Values,out Array)"}
   RerouteNode_31 = RigVMDispatchNode(Values=((Type=Bone,Name="heel_r"),((),Type=Bone,Name="tip_r"),((),Type=Bone,Name="ball_r"),((),Type=Bone,Name="foot_r"))) {template="DISPATCH_RigVMDispatch_ArrayMake(in Values,out Array)"}
   > Setup Reverse Foot_1 = RigVMFunctionReferenceNode(
    Negative Side=true
    Roll Buffer=(Type=Null,Name="foot_roll_r_bfr")
    Roll Control=(Type=Control,Name="foot_roll_r_ctrl")
    BK Buffers=@RerouteNode_29.Array
    BK Controls=@RerouteNode_30.Array
    Bones=@RerouteNode_31.Array
   )
   SpawnControl = &RigUnit_HierarchyAddControlTransform(
    ExecutePin=()
    InitialValue=(Rotation=(X=0,Y=0,Z=0,W=1),Translation=(X=0,Y=0,Z=0),Scale3D=(X=1,Y=1,Z=1))
    Name=spine_04_ctrl
    OffsetSpace=LocalSpace
    OffsetTransform=(Rotation=(X=0,Y=0,Z=0,W=1),Translation=(X=0,Y=0,Z=0),Scale3D=(X=1,Y=1,Z=1))
    Parent=(Type=Null,Name="chest_space")
    Settings="(Shape=(bVisible=true,Name=\"RoundedSquare_Thick\",Color=(R=1,G=0.7765,B=0,A=1),Transform=(Rotation=(X=0,Y=0,Z=0,W=1),Translation=(X=0,Y=0,Z=0),Scale3D=(X=3,Y=3,Z=10))),Proxy=(bIsProxy=False,ShapeVisibility=BasedOnSelection),FilteredChannels=(),DisplayName=\" ...(truncated)"
   ) {template="SpawnControl::Execute(in InitialValue,in Settings,in OffsetTransform,in OffsetSpace,in Parent,in Name,out Item)"}
   n49 = &RigUnit_GetTransform(Item=(Type=Bone,Name="spine_05"), bInitial=true)
   > &RigUnit_SetTransform(
    Value=(Rotation=(X=0,Y=0,Z=0,W=1),Translation=(X=0,Y=0,Z=0),Scale3D=(X=1,Y=1,Z=1))
    Weight=1
    bInitial=true
    bPropagateToChildren=false
    Value.Rotation=&RigUnit_GetTransform(Item=(Type=Control,Name="spine_03_ctrl"), bInitial=true).Transform.Rotation
    Value.Scale3D=@n49.Transform.Scale3D
    Value.Translation=@n49.Transform.Translation
    Item=@SpawnControl.Item
   )
   n50 = &RigUnit_ItemArray(Items=((Type=Control,Name="neck_01_ctrl"),(Type=Control,Name="clavicle_l_ctrl"),(Type=Control,Name="clavicle_r_ctrl"),(Type=Null,Name="head_fk_space"),(Type=Null,Name="head_ik_space")))
   > Reparent Items = RigVMFunctionReferenceNode(Parent=(Type=Control,Name="spine_04_ctrl"), Items=@n50.Items)
   RerouteNode_60_2 = RigVMDispatchNode(Values="((Type=Bone,Name=\"thumb_01_r\"),((),Type=Bone,Name=\"thumb_02_r\"),((),Type=Bone,Name=\"thumb_03_r\"),((),Type=Bone,Name=\"index_metacarpal_r\"),((),Type=Bone,Name=\"index_01_r\"),((),Type=Bone,Name=\"index_02_r\"),((),Type=Bone,Name=\"index_03_r\"),((),Type=Bone,Name= ...(truncated)") {template="DISPATCH_RigVMDispatch_ArrayMake(in Values,out Array)"}
   Backward Fingers_1 = RigVMFunctionReferenceNode(Execute=@"Backward Fingers".Execute, Bones=@RerouteNode_60_2.Array)
   Backward Foot_1 = RigVMFunctionReferenceNode(
    Bone=(Type=Bone,Name="ball_l")
    FK Control=(Type=Control,Name="ball_l_fk_ctrl")
    IK Control=(Type=Control,Name="ball_l_ik_ctrl")
    Toes Space=(Type=Null,Name="toes_l_space")
    Execute=@"Backward Leg_1".ExecuteContext
   )
   Backward Foot = RigVMFunctionReferenceNode(
    Bone=(Type=Bone,Name="ball_r")
    FK Control=(Type=Control,Name="ball_r_fk_ctrl")
    IK Control=(Type=Control,Name="ball_r_ik_ctrl")
    Toes Space=(Type=Null,Name="toes_r_space")
    Execute=@"Backward Foot_1".Execute
   )
   RerouteNode_2_2 = RigVMDispatchNode(Values="((Type=Bone,Name=\"indextoe_01_l\"),((),Type=Bone,Name=\"indextoe_02_l\"),((),Type=Bone,Name=\"bigtoe_01_l\"),((),Type=Bone,Name=\"bigtoe_02_l\"),((),Type=Bone,Name=\"littletoe_01_l\"),((),Type=Bone,Name=\"littletoe_02_l\"),((),Type=Bone,Name=\"middletoe_01_l\"),((),Typ ...(truncated)") {template="DISPATCH_RigVMDispatch_ArrayMake(in Values,out Array)"}
   Backward Fingers_2 = RigVMFunctionReferenceNode(Execute=@"Backward Foot".Execute, Bones=@RerouteNode_2_2.Array)
   RerouteNode_63_2 = RigVMDispatchNode(Values="((Type=Bone,Name=\"littletoe_01_r\"),((),Type=Bone,Name=\"littletoe_02_r\"),((),Type=Bone,Name=\"middletoe_01_r\"),((),Type=Bone,Name=\"middletoe_02_r\"),((),Type=Bone,Name=\"bigtoe_01_r\"),((),Type=Bone,Name=\"bigtoe_02_r\"),((),Type=Bone,Name=\"indextoe_01_r\"),((),Ty ...(truncated)") {template="DISPATCH_RigVMDispatch_ArrayMake(in Values,out Array)"}
   Backward Fingers_1_1 = RigVMFunctionReferenceNode(Execute=@"Backward Fingers_2".Execute, Bones=@RerouteNode_63_2.Array)
   RerouteNode_20 = RigVMDispatchNode(Values=((Type=Control,Name="neck_01_ctrl"),((),Type=Control,Name="neck_02_ctrl"),((),Type=Control,Name="head_ctrl"))) {template="DISPATCH_RigVMDispatch_ArrayMake(in Values,out Array)"}
   RerouteNode_24 = RigVMDispatchNode(Values=((Type=Bone,Name="neck_01"),((),Type=Bone,Name="neck_02"),((),Type=Bone,Name="head"))) {template="DISPATCH_RigVMDispatch_ArrayMake(in Values,out Array)"}
   Backward Neck_1 = RigVMFunctionReferenceNode(
    Head FK Space=(Type=Null,Name="head_fk_space")
    Head IK Space=(Type=Null,Name="head_ik_space")
    Head Orient Reference=(Type=Control,Name="body_ctrl")
    IK Control=(Type=Control,Name="head_ik_ctrl")
    FK Controls=@RerouteNode_20.Array
    Bones=@RerouteNode_24.Array
    ExecuteContext=@n56.ExecutePin
   )
   Backwards Solve Twist Controls_3_1 = RigVMFunctionReferenceNode(
    End Bone=(Type=Control,Name="foot_r_fk_ctrl")
    Root Corrective=(Type=Bone,Name="calf_correctiveRoot_r")
    Start Bone=(Type=Control,Name="calf_r_fk_ctrl")
    Twist From End=true
    ExecuteContext=@"Backwards Solve Twist Controls_2_1".ExecuteContext
    Twist Joints=RigVMVariableNode() {variable=r_calf_twists}.Value
   )
   Direct Control Solve_0_1_1_2_1_1_1 = RigVMFunctionReferenceNode(Bone=(Type=Bone,Name="camera"), Forward Solve=False, ExecuteContext=@"Direct Control Solve_0_1_1_2_1_1".ExecuteContext)
   n51 = RigVMDispatchNode(ExecuteContext=@n52.Completed, Array=RigVMVariableNode() {variable=hand_r_attach_control_joints}.Value) {template="DISPATCH_RigVMDispatch_ArrayIterator(in Array,out Element,out Index,out Count,out Ratio)"}
   > Direct Control Solve_1_1 = RigVMFunctionReferenceNode(Forward Solve=false, Bone=@n51.Element)
   n52 = RigVMDispatchNode(ExecuteContext=@SequenceAggregate_3.B, Array=RigVMVariableNode() {variable=hand_l_attach_control_joints}.Value) {template="DISPATCH_RigVMDispatch_ArrayIterator(in Array,out Element,out Index,out Count,out Ratio)"}
   > Direct Control Solve_2 = RigVMFunctionReferenceNode(Forward Solve=false, Bone=@n52.Element)
   RerouteNode_36_1 = RigVMDispatchNode(Values=((Type=Control,Name="index_metacarpal_r_ctrl"),((),Type=Control,Name="middle_metacarpal_r_ctrl"),((),Type=Control,Name="ring_metacarpal_r_ctrl"),((),Type=Control,Name="pinky_metacarpal_r_ctrl"))) {template="DISPATCH_RigVMDispatch_ArrayMake(in Values,out Array)"}
   RerouteNode_37 = RigVMDispatchNode(Values=((Type=Control,Name="thumb_curl_r_ctrl"),((),Type=Control,Name="index_curl_r_ctrl"),((),Type=Control,Name="middle_curl_r_ctrl"),((),Type=Control,Name="ring_curl_r_ctrl"),((),Type=Control,Name="pinky_curl_r_ctrl"))) {template="DISPATCH_RigVMDispatch_ArrayMake(in Values,out Array)"}
   RerouteNode_48 = RigVMDispatchNode(Values=((Type=Control,Name="index_01_r_ctrl"),((),Type=Control,Name="middle_01_r_ctrl"),((),Type=Control,Name="ring_01_r_ctrl"),((),Type=Control,Name="pinky_01_r_ctrl"))) {template="DISPATCH_RigVMDispatch_ArrayMake(in Values,out Array)"}
   RerouteNode_60_1 = RigVMDispatchNode(Values="((Type=Bone,Name=\"thumb_01_r\"),((),Type=Bone,Name=\"thumb_02_r\"),((),Type=Bone,Name=\"thumb_03_r\"),((),Type=Bone,Name=\"index_metacarpal_r\"),((),Type=Bone,Name=\"index_01_r\"),((),Type=Bone,Name=\"index_02_r\"),((),Type=Bone,Name=\"index_03_r\"),((),Type=Bone,Name= ...(truncated)") {template="DISPATCH_RigVMDispatch_ArrayMake(in Values,out Array)"}
   Forward Fingers = RigVMFunctionReferenceNode(
    Execute=@"Forward Fingers_4".Execute
    Metacarpals Controls=@RerouteNode_36_1.Array
    Proxy Controls=@RerouteNode_37.Array
    Base Fingers Controls=@RerouteNode_48.Array
    Fingers Bones=@RerouteNode_60_1.Array
    Auto Meta=&RigUnit_GetBoolAnimationChannel(Channel=Auto Interpolate, Control=pinky_metacarpal_r_ctrl, bInitial=False) {template="GetAnimationChannel::Execute(out Value,in Control,in Channel,in bInitial)"}.Value
    Auto Base=&RigUnit_GetBoolAnimationChannel(Channel=Auto Interpolate, Control=pinky_01_r_ctrl, bInitial=False) {template="GetAnimationChannel::Execute(out Value,in Control,in Channel,in bInitial)"}.Value
   )
   RerouteNode_63_1 = RigVMDispatchNode(Values="((Type=Bone,Name=\"littletoe_01_r\"),((),Type=Bone,Name=\"littletoe_02_r\"),((),Type=Bone,Name=\"middletoe_01_r\"),((),Type=Bone,Name=\"middletoe_02_r\"),((),Type=Bone,Name=\"bigtoe_01_r\"),((),Type=Bone,Name=\"bigtoe_02_r\"),((),Type=Bone,Name=\"indextoe_01_r\"),((),Ty ...(truncated)") {template="DISPATCH_RigVMDispatch_ArrayMake(in Values,out Array)"}
   Forward Fingers_5 = RigVMFunctionReferenceNode(
    Auto Base=False
    Auto Meta=False
    Base Fingers Controls=()
    Metacarpals Controls=()
    Proxy Controls=()
    Execute=@"Forward Fingers_1".Execute
    Fingers Bones=@RerouteNode_63_1.Array
   )
   RerouteNode_60 = RigVMDispatchNode(Values="((Type=Bone,Name=\"thumb_01_r\"),((),Type=Bone,Name=\"thumb_02_r\"),((),Type=Bone,Name=\"thumb_03_r\"),((),Type=Bone,Name=\"index_metacarpal_r\"),((),Type=Bone,Name=\"index_01_r\"),((),Type=Bone,Name=\"index_02_r\"),((),Type=Bone,Name=\"index_03_r\"),((),Type=Bone,Name= ...(truncated)") {template="DISPATCH_RigVMDispatch_ArrayMake(in Values,out Array)"}
   Setup Fingers_1 = RigVMFunctionReferenceNode(Negative Side=true, Fingers Bones=@RerouteNode_60.Array, Execute=@"Setup Fingers".Execute)
   RerouteNode_3 = RigVMDispatchNode(Values=((Type=Null,Name="thigh_l_fk_ctrl_space"),((),Type=Null,Name="thigh_r_fk_ctrl_space"))) {template="DISPATCH_RigVMDispatch_ArrayMake(in Values,out Array)"}
   n53 = RigVMDispatchNode(Array=@RerouteNode_3.Array, ExecuteContext=@SequenceExecution_1.C) {template="DISPATCH_RigVMDispatch_ArrayIterator(in Array,out Element,out Index,out Count,out Ratio)"}
   n54 = &RigUnit_ProjectTransformToNewParent(
    NewParent=(Type=Control,Name="hips_ctrl")
    OldParent=(Type=Control,Name="hips_ctrl")
    bChildInitial=True
    bOldParentInitial=True
    Child=@n53.Element
   )
   > &RigUnit_SetTransform(Weight=1, Item=@n53.Element, Value=@n54.Transform)
   RerouteNode_1 = RigVMDispatchNode(Values=((Type=Control,Name="thigh_l_fk_ctrl"),((),Type=Control,Name="calf_l_fk_ctrl"),((),Type=Control,Name="foot_l_fk_ctrl"))) {template="DISPATCH_RigVMDispatch_ArrayMake(in Values,out Array)"}
   RerouteNode_10 = RigVMDispatchNode(Values=((Type=Bone,Name="thigh_l"),((),Type=Bone,Name="calf_l"),((),Type=Bone,Name="foot_l"))) {template="DISPATCH_RigVMDispatch_ArrayMake(in Values,out Array)"}
   > Backward Leg = RigVMFunctionReferenceNode(
    Ik Control=(Type=Control,Name="foot_l_ik_ctrl")
    Pole Vector Null=(Type=Null,Name="leg_l_pv_footFollow")
    UpV Control=(Type=Control,Name="leg_l_pv_ik_ctrl")
    FK Controls=@RerouteNode_1.Array
    Bones=@RerouteNode_10.Array
   )
   RerouteNode_6 = RigVMDispatchNode(Values=((Type=Control,Name="thigh_r_fk_ctrl"),((),Type=Control,Name="calf_r_fk_ctrl"),((),Type=Control,Name="foot_r_fk_ctrl"))) {template="DISPATCH_RigVMDispatch_ArrayMake(in Values,out Array)"}
   RerouteNode_8 = RigVMDispatchNode(Values=((Type=Bone,Name="thigh_r"),((),Type=Bone,Name="calf_r"),((),Type=Bone,Name="foot_r"))) {template="DISPATCH_RigVMDispatch_ArrayMake(in Values,out Array)"}
   > Backward Leg_1 = RigVMFunctionReferenceNode(
    Ik Control=(Type=Control,Name="foot_r_ik_ctrl")
    Pole Vector Null=(Type=Null,Name="leg_r_pv_footFollow")
    UpV Control=(Type=Control,Name="leg_r_pv_ik_ctrl")
    FK Controls=@RerouteNode_6.Array
    Bones=@RerouteNode_8.Array
   )
   RerouteNode_11_1_2 = RigVMDispatchNode(Values=((Type=Control,Name="hips_ctrl"),((),Type=Control,Name="hips_tan_ctrl"),((),Type=Control,Name="chest_tan_ctrl"),((),Type=Control,Name="chest_ctrl"))) {template="DISPATCH_RigVMDispatch_ArrayMake(in Values,out Array)"}
   RerouteNode_52_1 = RigVMDispatchNode(Values=((Type=Bone,Name="spine_01"),((),Type=Bone,Name="spine_02"),((),Type=Bone,Name="spine_03"),((),Type=Bone,Name="spine_04"))) {template="DISPATCH_RigVMDispatch_ArrayMake(in Values,out Array)"}
   RerouteNode_53_1 = RigVMDispatchNode(Values=((Type=Control,Name="hips_ctrl"),((),Type=Control,Name="spine_01_ctrl"),((),Type=Control,Name="spine_02_ctrl"),((),Type=Control,Name="spine_03_ctrl"))) {template="DISPATCH_RigVMDispatch_ArrayMake(in Values,out Array)"}
   Backward Spine = RigVMFunctionReferenceNode(
    Pelvis Bone=(Type=Bone,Name="pelvis")
    IK Controls=@RerouteNode_11_1_2.Array
    Bones=@RerouteNode_52_1.Array
    FK Controls=@RerouteNode_53_1.Array
    ExecuteContext=@n61.ExecutePin
    Offsets=RigVMVariableNode() {variable=Spine Offsets}.Value
   )
   n55 = &RigUnit_ProjectTransformToNewParent(
    Child=(Type=Control,Name="spine_04_ctrl")
    NewParent=(Type=Bone,Name="spine_05")
    OldParent=(Type=Bone,Name="spine_05")
    bChildInitial=True
    bOldParentInitial=True
   )
   > n56 = &RigUnit_SetTransform(Item=(Type=Control,Name="spine_04_ctrl"), Weight=1, bInitial=False, bPropagateToChildren=true, Value=@n55.Transform)
   Backwards Solve Twist Controls = RigVMFunctionReferenceNode(
    End Bone=(Type=Control,Name="lowerarm_l_fk_ctrl")
    Root Corrective=(Type=Bone,Name="upperarm_correctiveRoot_l")
    Start Bone=(Type=Control,Name="upperarm_l_fk_ctrl")
    Twist From End=False
    ExecuteContext=@n57.True
    Twist Joints=RigVMVariableNode() {variable=l_upperarm_twists}.Value
   )
   > Backwards Solve Twist Controls_1 = RigVMFunctionReferenceNode(
    End Bone=(Type=Control,Name="hand_l_fk_ctrl")
    Root Corrective=(Type=Bone,Name="lowerarm_correctiveRoot_l")
    Start Bone=(Type=Control,Name="lowerarm_l_fk_ctrl")
    Twist From End=true
    Twist Joints=RigVMVariableNode() {variable=l_lowerarm_twists}.Value
   )
   > Backwards Solve Twist Controls_2 = RigVMFunctionReferenceNode(
    End Bone=(Type=Control,Name="calf_l_fk_ctrl")
    Root Corrective=(Type=Bone,Name="thigh_correctiveRoot_l")
    Start Bone=(Type=Control,Name="thigh_l_fk_ctrl")
    Twist From End=False
    Twist Joints=RigVMVariableNode() {variable=l_thigh_twists}.Value
   )
   > Backwards Solve Twist Controls_3 = RigVMFunctionReferenceNode(
    End Bone=(Type=Control,Name="foot_l_fk_ctrl")
    Root Corrective=(Type=Bone,Name="calf_correctiveRoot_l")
    Start Bone=(Type=Control,Name="calf_l_fk_ctrl")
    Twist From End=true
    Twist Joints=RigVMVariableNode() {variable=l_calf_twists}.Value
   )
   > Backwards Solve Twist Controls_4 = RigVMFunctionReferenceNode(
    End Bone=(Type=Control,Name="lowerarm_r_fk_ctrl")
    Root Corrective=(Type=Bone,Name="upperarm_correctiveRoot_r")
    Start Bone=(Type=Control,Name="upperarm_r_fk_ctrl")
    Twist From End=False
    Twist Joints=RigVMVariableNode() {variable=r_upperarm_twists}.Value
   )
   > Backwards Solve Twist Controls_1_1 = RigVMFunctionReferenceNode(
    End Bone=(Type=Control,Name="hand_r_fk_ctrl")
    Root Corrective=(Type=Bone,Name="lowerarm_correctiveRoot_r")
    Start Bone=(Type=Control,Name="lowerarm_r_fk_ctrl")
    Twist From End=true
    Twist Joints=RigVMVariableNode() {variable=r_lowerarm_twists}.Value
   )
   > Backwards Solve Twist Controls_2_1 = RigVMFunctionReferenceNode(
    End Bone=(Type=Control,Name="calf_r_fk_ctrl")
    Root Corrective=(Type=Bone,Name="thigh_correctiveRoot_r")
    Start Bone=(Type=Control,Name="thigh_r_fk_ctrl")
    Twist From End=False
    Twist Joints=RigVMVariableNode() {variable=r_thigh_twists}.Value
   )
   n57 = &RigVMFunction_ControlFlowBranch(ExecuteContext=@SequenceAggregate_3.A, Condition=RigVMVariableNode() {variable=Do Twist Controls}.Value)
   Direct Control Solve_0 = RigVMFunctionReferenceNode(Bone=(Type=Bone,Name="attach"), Forward Solve=False, ExecuteContext=@n51.Completed)
   > Direct Control Solve_0_1 = RigVMFunctionReferenceNode(Bone=(Type=Bone,Name="attach_backpack"), Forward Solve=False)
   > Direct Control Solve_0_1_1 = RigVMFunctionReferenceNode(Bone=(Type=Bone,Name="attach_cape"), Forward Solve=False)
   > Direct Control Solve_3 = RigVMFunctionReferenceNode(Bone=(Type=Bone,Name="attach_emote_01"), Forward Solve=False)
   > Direct Control Solve_4 = RigVMFunctionReferenceNode(Bone=(Type=Bone,Name="attach_emote_02"), Forward Solve=False)
   > Direct Control Solve_5 = RigVMFunctionReferenceNode(Bone=(Type=Bone,Name="attach_emote_03"), Forward Solve=False)
   > Direct Control Solve_0_1_1_2 = RigVMFunctionReferenceNode(Bone=(Type=Bone,Name="root_fp"), Forward Solve=False)
   > Direct Control Solve_0_1_1_2_1 = RigVMFunctionReferenceNode(Bone=(Type=Bone,Name="attach_fp"), Forward Solve=False)
   > Direct Control Solve_0_1_1_2_1_1 = RigVMFunctionReferenceNode(Bone=(Type=Bone,Name="camera_root"), Forward Solve=False)
   n58 = &RigUnit_GetTransform(Item=(Type=Bone,Name="spine_05"))
   n59 = &RigUnit_GetTransform(Item=(Type=Bone,Name="root"), bInitial=false)
   n60 = &RigUnit_ProjectTransformToNewParent(
    Child=(Type=Control,Name="body_ctrl")
    NewParent=(Type=Bone,Name="spine_01")
    OldParent=(Type=Bone,Name="spine_01")
    bChildInitial=True
    bOldParentInitial=True
   )
   SequenceAggregate_3 = &RigVMFunction_Sequence(ExecuteContext=@SequenceExecution_1.D)
   &RigUnit_SetTransform(Item=(Type=Control,Name="root_ctrl"), Weight=1, Value=@n59.Transform, ExecutePin=@n62.ExecuteContext)
   > &RigUnit_SetTransform(Item=(Type=Control,Name="body_offset_ctrl"), Weight=1, Value=@n59.Transform)
   > n61 = &RigUnit_SetTransform(Item=(Type=Control,Name="body_ctrl"), Weight=1, Value=@n60.Transform)
   n62 = RigVMVariableNode(Value=@n58.Transform, ExecuteContext=@SequenceExecution_1.A) {variable=spine_05_Transform}
  ]}
 RigVMModel Match Event Graph {nodes [
   RigVMFunction_UserDefinedEvent = /Script/RigVM.RigVMFunction_UserDefinedEvent(EventName=Match Right Arm IK) {is=struct RigVMUnitNode}
   Match IK onto FK_1 = RigVMFunctionReferenceNode(
    IK Control=(Type=Control,Name="hand_r_ik_ctrl")
    Pole Vector Control=(Type=Control,Name="arm_r_pv_ik_ctrl")
    Secondary Controls=()
    ExecuteContext=@RigVMFunction_UserDefinedEvent.ExecutePin
    Bones=RigVMVariableNode() {variable=arm_r_aimed}.Value
   )
   n1 = &RigVMFunction_UserDefinedEvent(EventName=Match Right Arm FK)
   RerouteNode_57 = RigVMDispatchNode(Values=((Type=Control,Name="upperarm_r_fk_ctrl"),((),Type=Control,Name="lowerarm_r_fk_ctrl"),((),Type=Control,Name="hand_r_fk_ctrl"))) {template="DISPATCH_RigVMDispatch_ArrayMake(in Values,out Array)"}
   Project to New Parents_1 = RigVMFunctionReferenceNode(
    Propagate to Children=True
    Parents=@RerouteNode_57.Array
    ExecuteContext=@n1.ExecutePin
    Children=RigVMVariableNode() {variable=arm_r_aimed}.Value
   )
   n2 = &RigVMFunction_UserDefinedEvent(EventName=Match Left Arm IK)
   Match IK onto FK = RigVMFunctionReferenceNode(
    IK Control=(Type=Control,Name="hand_l_ik_ctrl")
    Pole Vector Control=(Type=Control,Name="arm_l_pv_ik_ctrl")
    Secondary Controls=()
    ExecuteContext=@n2.ExecutePin
    Bones=RigVMVariableNode() {variable=arm_l_aimed}.Value
   )
   n3 = &RigVMFunction_UserDefinedEvent(EventName=Match Left Arm FK)
   RerouteNode_56 = RigVMDispatchNode(Values=((Type=Control,Name="upperarm_l_fk_ctrl"),((),Type=Control,Name="lowerarm_l_fk_ctrl"),((),Type=Control,Name="hand_l_fk_ctrl"))) {template="DISPATCH_RigVMDispatch_ArrayMake(in Values,out Array)"}
   Project to New Parents = RigVMFunctionReferenceNode(
    Propagate to Children=True
    Parents=@RerouteNode_56.Array
    ExecuteContext=@n3.ExecutePin
    Children=RigVMVariableNode() {variable=arm_l_aimed}.Value
   )
   n4 = &RigVMFunction_UserDefinedEvent(EventName=Match Left Leg IK)
   Make_Array = RigVMDispatchNode(Values=((Type=Control,Name="foot_roll_l_ctrl"),(Type=Control,Name="foot_bk1_l_ctrl"),((),Type=Control,Name="heel_l_ctrl"),((),Type=Control,Name="tip_l_ctrl"))) {template="DISPATCH_RigVMDispatch_ArrayMake(in Values,out Array)"}
   Match IK onto FK_2 = RigVMFunctionReferenceNode(
    IK Control=(Type=Control,Name="foot_l_ik_ctrl")
    Pole Vector Control=(Type=Control,Name="leg_l_pv_ik_ctrl")
    Secondary Controls=@Make_Array.Array
    ExecuteContext=@n4.ExecutePin
    Bones=RigVMVariableNode() {variable=leg_l_aimed}.Value
   )
   n5 = &RigVMFunction_UserDefinedEvent(EventName=Match Left Leg FK)
   RerouteNode_61 = RigVMDispatchNode(Values=((Type=Control,Name="thigh_l_fk_ctrl"),((),Type=Control,Name="calf_l_fk_ctrl"),((),Type=Control,Name="foot_l_fk_ctrl"))) {template="DISPATCH_RigVMDispatch_ArrayMake(in Values,out Array)"}
   Project to New Parents_2 = RigVMFunctionReferenceNode(
    Propagate to Children=True
    Parents=@RerouteNode_61.Array
    ExecuteContext=@n5.ExecutePin
    Children=RigVMVariableNode() {variable=leg_l_aimed}.Value
   )
   n6 = &RigVMFunction_UserDefinedEvent(EventName=Match Right Leg IK)
   Make_Array_1 = RigVMDispatchNode(Values=((Type=Control,Name="foot_roll_r_ctrl"),(Type=Control,Name="foot_bk1_r_ctrl"),((),Type=Control,Name="heel_r_ctrl"),((),Type=Control,Name="tip_r_ctrl"))) {template="DISPATCH_RigVMDispatch_ArrayMake(in Values,out Array)"}
   Match IK onto FK_3 = RigVMFunctionReferenceNode(
    IK Control=(Type=Control,Name="foot_r_ik_ctrl")
    Pole Vector Control=(Type=Control,Name="leg_r_pv_ik_ctrl")
    Secondary Controls=@Make_Array_1.Array
    ExecuteContext=@n6.ExecutePin
    Bones=RigVMVariableNode() {variable=leg_r_aimed}.Value
   )
   n7 = &RigVMFunction_UserDefinedEvent(EventName=Match Right Leg FK)
   RerouteNode_62 = RigVMDispatchNode(Values=((Type=Control,Name="thigh_r_fk_ctrl"),((),Type=Control,Name="calf_r_fk_ctrl"),((),Type=Control,Name="foot_r_fk_ctrl"))) {template="DISPATCH_RigVMDispatch_ArrayMake(in Values,out Array)"}
   Project to New Parents_3 = RigVMFunctionReferenceNode(
    Propagate to Children=True
    Parents=@RerouteNode_62.Array
    ExecuteContext=@n7.ExecutePin
    Children=RigVMVariableNode() {variable=leg_r_aimed}.Value
   )
  ]}
 SequenceAggregate_1_1_ContainedGraph {nodes [
   n1 = RigVMFunctionEntryNode()
   SequenceAggregate_1 = &RigVMFunction_Sequence(ExecuteContext=@n1.ExecuteContext)
   SequenceAggregate_1_1 = &RigVMFunction_Sequence(ExecuteContext=@SequenceAggregate_1.B)
   SequenceAggregate_1_2 = &RigVMFunction_Sequence(ExecuteContext=@SequenceAggregate_1_1.B)
   SequenceAggregate_2 = &RigVMFunction_Sequence(ExecuteContext=@SequenceAggregate_1_2.B)
   RigVMFunctionReturnNode(A=@SequenceAggregate_1.A, B=@SequenceAggregate_1_1.A, C=@SequenceAggregate_1_2.A, D=@SequenceAggregate_2.A, E=@SequenceAggregate_2.B)
  ]}
 SequenceAggregate_1_ContainedGraph {nodes [
   n1 = RigVMFunctionEntryNode()
   SequenceAggregate = &RigVMFunction_Sequence(ExecuteContext=@n1.ExecuteContext)
   SequenceAggregate_1 = &RigVMFunction_Sequence(ExecuteContext=@SequenceAggregate.B)
   RigVMFunctionReturnNode(A=@SequenceAggregate.A, B=@SequenceAggregate_1.A, C=@SequenceAggregate_1.B)
  ]}
 SequenceAggregate_2_1_ContainedGraph {nodes [
   n1 = RigVMFunctionEntryNode()
   SequenceAggregate_2 = &RigVMFunction_Sequence(ExecuteContext=@n1.ExecuteContext)
   SequenceAggregate_2_1 = &RigVMFunction_Sequence(ExecuteContext=@SequenceAggregate_2.B)
   SequenceAggregate_2_2 = &RigVMFunction_Sequence(ExecuteContext=@SequenceAggregate_2_1.B)
   RigVMFunctionReturnNode(A=@SequenceAggregate_2.A, B=@SequenceAggregate_2_1.A, C=@SequenceAggregate_2_2.A, D=@SequenceAggregate_2_2.B)
  ]}
 SequenceAggregate_4_1_ContainedGraph {nodes [
   n1 = RigVMFunctionEntryNode()
   SequenceAggregate_4 = &RigVMFunction_Sequence(ExecuteContext=@n1.ExecuteContext)
   SequenceAggregate_4_1 = &RigVMFunction_Sequence(ExecuteContext=@SequenceAggregate_4.B)
   RigVMFunctionReturnNode(A=@SequenceAggregate_4.A, B=@SequenceAggregate_4_1.A, C=@SequenceAggregate_4_1.B)
  ]}
 SequenceAggregate_5_ContainedGraph {nodes [
   n1 = RigVMFunctionEntryNode()
   SequenceAggregate = &RigVMFunction_Sequence(ExecuteContext=@n1.ExecuteContext)
   SequenceAggregate_1 = &RigVMFunction_Sequence(ExecuteContext=@SequenceAggregate.B)
   RigVMFunctionReturnNode(A=@SequenceAggregate.A, B=@SequenceAggregate_1.A, C=@SequenceAggregate_1.B)
  ]}
]
hierarchy [
 root {loc=<0 0 0>, rot=<0 0 0 1>}
 pelvis {parent=root, loc=<0 -0.7071 93.1191>, rot=<0.0127 0.707 -0.0127 -0.707>}
 spine_01 {parent=pelvis, loc=<6.7494 -0.6388 0>, rot=<0 0 0.034 -0.9994>}
 spine_02 {parent=spine_01, loc=<7.2518 -0.035 0>, rot=<0 0 0.0238 -0.9997>}
 spine_03 {parent=spine_02, loc=<6.0104 0 0>, rot=<0 0 0.0149 -0.9999>}
 spine_04 {parent=spine_03, loc=<8.1024 0 0>, rot=<0 0 -0.0132 -0.9999>}
 spine_05 {parent=spine_04, loc=<10.0146 0 0>, rot=<0 0 -0.0015 -1>}
 neck_01 {parent=spine_05, loc=<11.3138 0.1529 0>, rot=<0 0 -0.2003 -0.9797>}
 neck_02 {parent=neck_01, loc=<4.5627 0 0>, rot=<0 0 0.0504 -0.9987>}
 head {parent=neck_02, loc=<5.3914 0 0>, rot=<0 0 0.115 -0.9934>}
 clavicle_l {parent=spine_05, loc=<4.3575 -0.1466 -2.3223>, rot=<0.0465 -0.7289 -0.009 -0.6829>}
 upperarm_l {parent=clavicle_l, loc=<15.5111 0 0>, rot=<0.0226 -0.3234 -0.0088 -0.9459>}
 lowerarm_l {parent=upperarm_l, loc=<26.6514 0 0>, rot=<0.0391 -0.0099 -0.2453 -0.9686>}
 hand_l {parent=lowerarm_l, loc=<22.3491 0 0>, rot=<-0.7081 -0.0507 0.0975 -0.6975>}
 index_metacarpal_l {parent=hand_l, loc=<1.6535 0.0183 -2.96>, rot=<0.0158 -0.0124 -0.0173 -0.9996>}
 index_01_l {parent=index_metacarpal_l, loc=<8.3008 0.0357 -0.0524>, rot=<-0.0044 0.0485 0.0924 -0.9945>}
 index_02_l {parent=index_01_l, loc=<3.3587 -0.3004 0.0452>, rot=<-0.002 0.0027 0.3225 -0.9466>}
 index_03_l {parent=index_02_l, loc=<2.5578 -0.123 -0.0252>, rot=<-0.002 0.0067 0.2558 -0.9667>}
 index_03_bulge_l {parent=index_03_l, loc=<0.8328 -0.6522 -0.0609>, rot=<-0.7754 -0.0132 -0.0293 -0.6306>}
 index_03_half_l {parent=index_03_l, loc=<0 0 0>, rot=<0 0 0 1>}
 index_02_dip_l {parent=index_03_half_l, loc=<-0.0483 0.5099 0.0634>, rot=<0.6284 -0.03 0.0569 -0.7753>}
 index_03_in_l {parent=index_03_half_l, loc=<0.0068 -0.6099 -0.0993>, rot=<-0.7585 -0.0043 -0.0144 -0.6515>}
 index_02_bulge_l {parent=index_02_l, loc=<1.2679 -0.7245 -0.1244>, rot=<-0.7248 -0.0082 -0.0356 -0.688>}
 index_02_side_out_l {parent=index_02_l, loc=<1.2194 -0.0243 0.622>, rot=<-0.0119 -0.0165 -0.7152 -0.6986>}
 index_02_side_inn_l {parent=index_02_l, loc=<1.2392 -0.0465 -0.7297>, rot=<0.7005 0.6947 0.1384 -0.0865>}
 index_02_half_l {parent=index_02_l, loc=<0 0 0>, rot=<0 0 0 1>}
 index_02_in_l {parent=index_02_half_l, loc=<0.202 -0.7363 -0.168>, rot=<-0.7214 0.0528 0.0174 -0.6903>}
 index_02_pip_l {parent=index_02_half_l, loc=<0.0518 0.6386 0.0257>, rot=<0.6993 0.0168 -0.0017 -0.7147>}
 index_01_palmMid_l {parent=index_01_l, loc=<1.4779 -0.9649 -0.1824>, rot=<-0.7072 -0.1278 -0.0351 -0.6945>}
 index_01_bulge_l {parent=index_01_l, loc=<2.7004 -0.8645 -0.2221>, rot=<-0.7809 -0.0063 0.0096 -0.6246>}
 index_01_side_out_l {parent=index_01_l, loc=<2.7526 0.1164 0.7381>, rot=<0.0069 -0.0662 -0.7063 -0.7047>}
 index_01_side_inn_l {parent=index_01_l, loc=<2.6653 -0.0812 -0.8637>, rot=<-0.7045 -0.7074 0.0543 -0.0187>}
 index_01_half_l {parent=index_01_l, loc=<0 0 0>, rot=<0 0 0 1>}
 index_01_mcp_l {parent=index_01_half_l, loc=<-0.2824 0.8445 0.1505>, rot=<0.7167 -0.0113 -0.0196 -0.697>}
 index_01_palm_l {parent=index_01_half_l, loc=<0.486 -1.2095 -0.4188>, rot=<-0.7242 -0.1505 -0.102 -0.6652>}
 index_metacarpal_slide_l {parent=index_metacarpal_l, loc=<2.0671 0.3152 0.2008>, rot=<0.6834 -0.0204 0.0521 -0.7279>}
 middle_metacarpal_l {parent=hand_l, loc=<2.1856 -0.1725 -0.5984>, rot=<0.006 0.0168 -0.0329 -0.9993>}
 middle_01_l {parent=middle_metacarpal_l, loc=<7.3275 -0.2447 0.0129>, rot=<-0.0208 0.0042 0.0858 -0.9961>}
 middle_02_l {parent=middle_01_l, loc=<4.0123 0.1401 0.1175>, rot=<-0.0218 0.0258 0.3512 -0.9357>}
 middle_03_l {parent=middle_02_l, loc=<2.6077 0.0534 0.1373>, rot=<0.0082 0.0188 0.1708 -0.9851>}
 middle_03_bulge_l {parent=middle_03_l, loc=<0.6039 -0.5838 -0.0914>, rot=<-0.7696 -0.007 -0.0255 -0.6379>}
 middle_03_half_l {parent=middle_03_l, loc=<0 0 0>, rot=<0 0 0 1>}
 middle_03_in_l {parent=middle_03_half_l, loc=<-0.0216 -0.5365 -0.087>, rot=<-0.7646 -0.0054 0.0016 -0.6445>}
 middle_02_dip_l {parent=middle_03_half_l, loc=<0.0188 0.58 0.096>, rot=<0.6666 -0.0259 0.015 -0.7448>}
 middle_02_bulge_l {parent=middle_02_l, loc=<1.3356 -0.6387 -0.1187>, rot=<-0.7529 0.018 0.034 -0.657>}
 middle_02_side_out_l {parent=middle_02_l, loc=<1.2894 0.0587 0.6112>, rot=<-0.0099 0.0174 -0.7084 -0.7055>}
 middle_02_side_inn_l {parent=middle_02_l, loc=<1.2774 0.0693 -0.7116>, rot=<-0.6836 -0.7292 -0.0303 0.0062>}
 middle_02_half_l {parent=middle_02_l, loc=<0 0 0>, rot=<0 0 0 1>}
 middle_02_pip_l {parent=middle_02_half_l, loc=<-0.1178 0.7082 0.0526>, rot=<0.6625 0.0441 -0.0455 -0.7464>}
 middle_02_in_l {parent=middle_02_half_l, loc=<0.2359 -0.6452 -0.1065>, rot=<-0.7503 0.0044 0.0137 -0.661>}
 middle_01_palmMid_l {parent=middle_01_l, loc=<1.6802 -1.1053 -0.1343>, rot=<-0.7489 -0.0824 -0.1177 -0.6469>}
 middle_01_bulge_l {parent=middle_01_l, loc=<3.1937 -0.8245 -0.0515>, rot=<-0.7031 -0.0127 -0.0163 -0.7108>}
 middle_01_side_out_l {parent=middle_01_l, loc=<3.2592 0.0643 0.7134>, rot=<-0.0169 -0.0342 -0.707 -0.7062>}
 middle_01_side_inn_l {parent=middle_01_l, loc=<3.1057 -0.2057 -0.7603>, rot=<-0.6831 -0.7191 0.1237 -0.0303>}
 middle_01_half_l {parent=middle_01_l, loc=<0 0 0>, rot=<0 0 0 1>}
 middle_01_mcp_l {parent=middle_01_half_l, loc=<0.0119 0.7212 0.1249>, rot=<0.7046 -0.0133 -0.0348 -0.7086>}
 middle_01_palm_l {parent=middle_01_half_l, loc=<0.6493 -1.5 -0.2651>, rot=<-0.7196 -0.1134 -0.1741 -0.6626>}
 middle_metacarpal_slide_l {parent=middle_metacarpal_l, loc=<2.4672 0.2465 -0.0038>, rot=<0.7195 -0.0053 0.0079 -0.6944>}
 pinky_metacarpal_l {parent=hand_l, loc=<1.7616 -0.9753 3.3814>, rot=<-0.0397 0.0814 -0.069 -0.9935>}
 pinky_01_l {parent=pinky_metacarpal_l, loc=<7.1709 -0.4831 -0.1094>, rot=<-0.0654 -0.0035 0.2242 -0.9723>}
 pinky_02_l {parent=pinky_01_l, loc=<2.8274 0.0893 -0.1043>, rot=<0.0037 0.0011 0.2854 -0.9584>}
 pinky_03_l {parent=pinky_02_l, loc=<2.2046 0.0307 -0.0078>, rot=<-0.0002 0.0098 0.1445 -0.9895>}
 pinky_03_bulge_l {parent=pinky_03_l, loc=<0.5993 -0.5216 -0.0399>, rot=<-0.697 0.0164 0.0076 -0.7168>}
 pinky_03_half_l {parent=pinky_03_l, loc=<0 0 0>, rot=<0 0 0 1>}
 pinky_03_in_l {parent=pinky_03_half_l, loc=<-0.0453 -0.469 -0.0096>, rot=<-0.7033 0.0203 -0.0094 -0.7105>}
 pinky_02_dip_l {parent=pinky_03_half_l, loc=<0.0005 0.5697 0.0186>, rot=<0.7032 -0.0026 -0.0216 -0.7107>}
 pinky_02_bulge_l {parent=pinky_02_l, loc=<1.0154 -0.5119 0.1478>, rot=<-0.6081 -0.0182 -0.0274 -0.7932>}
 pinky_02_side_out_l {parent=pinky_02_l, loc=<1.0387 -0.0455 0.5578>, rot=<0.0814 -0.0847 -0.713 -0.6912>}
 pinky_02_side_inn_l {parent=pinky_02_l, loc=<0.9943 0.0126 -0.5825>, rot=<0.7148 0.6988 0.0254 0.0038>}
 pinky_02_half_l {parent=pinky_02_l, loc=<0 0 0>, rot=<0 0 0 1>}
 pinky_02_in_l {parent=pinky_02_half_l, loc=<0.1673 -0.5556 0.081>, rot=<-0.6579 0.0017 -0.0123 -0.753>}
 pinky_02_pip_l {parent=pinky_02_half_l, loc=<-0.1412 0.5745 0.0362>, rot=<0.6974 0.0445 -0.0668 -0.7121>}
 pinky_01_palmMid_l {parent=pinky_01_l, loc=<0.9102 -0.6387 0.004>, rot=<-0.6398 -0.0317 -0.0411 -0.7668>}
 pinky_01_bulge_l {parent=pinky_01_l, loc=<1.8657 -0.6076 0.0246>, rot=<-0.6596 -0.0031 -0.0055 -0.7516>}
 pinky_01_side_out_l {parent=pinky_01_l, loc=<1.9698 0.1582 0.5742>, rot=<0.0472 -0.0628 -0.6895 -0.72>}
 pinky_01_side_inn_l {parent=pinky_01_l, loc=<1.9589 -0.1083 -0.6995>, rot=<0.7305 0.6691 -0.0969 0.0966>}
 pinky_01_half_l {parent=pinky_01_l, loc=<0 0 0>, rot=<0 0 0 1>}
 pinky_01_mcp_l {parent=pinky_01_half_l, loc=<-0.2097 0.7978 -0.0339>, rot=<0.7244 -0.0981 -0.0292 -0.6817>}
 pinky_01_palm_l {parent=pinky_01_half_l, loc=<0.4415 -0.7347 0.3849>, rot=<-0.5643 -0.106 -0.0209 -0.8184>}
 pinky_metacarpal_slide_l {parent=pinky_metacarpal_l, loc=<1.6675 0.3116 -0.1582>, rot=<0.743 -0.0428 -0.0046 -0.6679>}
 ring_metacarpal_l {parent=hand_l, loc=<2.2074 -0.3625 1.3638>, rot=<-0.0284 0.0722 -0.0458 -0.9959>}
 ring_01_l {parent=ring_metacarpal_l, loc=<7.0295 -0.6197 -0.1446>, rot=<-0.0489 -0.0252 0.1892 -0.9804>}
 ring_02_l {parent=ring_01_l, loc=<3.6174 0.4647 -0.1027>, rot=<0 0 0.2806 -0.9598>}
 ring_03_l {parent=ring_02_l, loc=<2.4152 0.114 0.0199>, rot=<0 0 0.1189 -0.9929>}
 ring_03_bulge_l {parent=ring_03_l, loc=<0.3996 -0.5585 0.0882>, rot=<-0.7026 0.0054 -0.0216 -0.7112>}
 ring_03_half_l {parent=ring_03_l, loc=<0 0 0>, rot=<0 0 0 1>}
 ring_02_dip_l {parent=ring_03_half_l, loc=<-0.2342 0.5823 0.075>, rot=<0.7224 0.0478 0 -0.6898>}
 ring_03_in_l {parent=ring_03_half_l, loc=<-0.2296 -0.5924 0.0786>, rot=<-0.6897 -0.0254 -0.061 -0.7211>}
 ring_02_bulge_l {parent=ring_02_l, loc=<1.5943 -0.6099 0.0165>, rot=<-0.6818 0.0039 -0.0223 -0.7312>}
 ring_02_side_out_l {parent=ring_02_l, loc=<1.5325 0.0786 0.6239>, rot=<0.0026 0.0443 -0.7048 -0.708>}
 ring_02_side_inn_l {parent=ring_02_l, loc=<1.6242 0.2523 -0.5907>, rot=<0.713 0.6998 0.0413 -0.0176>}
 ring_02_half_l {parent=ring_02_l, loc=<0 0 0>, rot=<0 0 0 1>}
 ring_02_in_l {parent=ring_02_half_l, loc=<0.4183 -0.6712 -0.0629>, rot=<-0.762 -0.012 -0.0782 -0.6427>}
 ring_02_pip_l {parent=ring_02_half_l, loc=<0.0241 0.6734 -0.0393>, rot=<0.7047 0.0634 -0.0533 -0.7046>}
 ring_01_palmMid_l {parent=ring_01_l, loc=<1.6474 -0.92 -0.1888>, rot=<-0.7049 0.0039 -0.0272 -0.7088>}
 ring_01_bulge_l {parent=ring_01_l, loc=<2.8565 -0.8656 -0.0855>, rot=<-0.7262 0.0479 0.0121 -0.6858>}
 ring_01_side_out_l {parent=ring_01_l, loc=<2.814 0.1223 0.6404>, rot=<0.0011 -0.0029 -0.6693 -0.7429>}
 ring_01_side_inn_l {parent=ring_01_l, loc=<2.9145 -0.1311 -0.7792>, rot=<-0.7034 -0.7071 0.053 -0.0496>}
 ring_01_half_l {parent=ring_01_l, loc=<0 0 0>, rot=<0 0 0 1>}
 ring_01_mcp_l {parent=ring_01_half_l, loc=<0.1592 0.7105 0.1499>, rot=<0.6797 -0.0132 0.0333 -0.7326>}
 ring_01_palm_l {parent=ring_01_half_l, loc=<0.6198 -1.3254 -0.2137>, rot=<-0.6411 -0.0679 -0.1097 -0.7565>}
 ring_metacarpal_slide_l {parent=ring_metacarpal_l, loc=<2.2473 0.3637 -0.0304>, rot=<0.7223 0.0127 -0.0266 -0.6909>}
 thumb_01_l {parent=hand_l, loc=<1.3118 -0.3191 -2.459>, rot=<0.6546 -0.4042 0.0664 -0.6354>}
 thumb_02_l {parent=thumb_01_l, loc=<5.9336 -0.0448 0.0401>, rot=<-0.0037 0.0213 0.298 -0.9543>}
 thumb_03_l {parent=thumb_02_l, loc=<2.2223 0.3682 -0.0907>, rot=<0.0118 -0.013 0.0066 -0.9998>}
 thumb_03_bulge_l {parent=thumb_03_l, loc=<0.854 -0.9279 -0.1364>, rot=<-0.7582 -0.0381 -0.0057 -0.6509>}
 thumb_03_side_out_l {parent=thumb_03_l, loc=<0.3379 -0.2074 0.7481>, rot=<-0.009 -0.018 0.9998 -0.0013>}
 thumb_03_side_inn_l {parent=thumb_03_l, loc=<0.2417 -0.0743 -0.8274>, rot=<0.9952 -0.0755 -0.0151 -0.0604>}
 thumb_03_half_l {parent=thumb_03_l, loc=<0 0 0>, rot=<0 0 0 1>}
 thumb_03_pip_l {parent=thumb_03_half_l, loc=<0.0282 0.7126 0.1327>, rot=<0.6567 -0.0165 -0.0216 -0.7536>}
 thumb_03_in_l {parent=thumb_03_half_l, loc=<0.0161 -0.788 -0.1277>, rot=<-0.7292 -0.0572 -0.0191 -0.6816>}
 thumb_02_bulge_l {parent=thumb_02_l, loc=<1.6746 -0.8593 -0.1133>, rot=<-0.7435 -0.0285 0.0014 -0.6682>}
 thumb_02_side_out_l {parent=thumb_02_l, loc=<1.0629 -0.2661 0.813>, rot=<0.0312 -0.1082 0.9902 -0.0823>}
 thumb_02_side_inn_l {parent=thumb_02_l, loc=<0.9576 0.1105 -0.7045>, rot=<0.9902 0.0823 -0.0312 -0.1082>}
 thumb_02_half_l {parent=thumb_02_l, loc=<0 0 0>, rot=<0 0 0 1>}
 thumb_02_in_l {parent=thumb_02_half_l, loc=<0.7327 -1.0138 -0.096>, rot=<-0.8303 -0.0684 -0.0426 -0.5515>}
 thumb_02_mcp_l {parent=thumb_02_half_l, loc=<-0.5189 0.6084 -0.0585>, rot=<0.7115 0.0887 0.0008 -0.6971>}
 thumb_01_side_out_l {parent=thumb_01_l, loc=<0.5727 0.672 -1.1092>, rot=<0.8946 0.0392 0.0667 -0.4402>}
 thumb_01_side_inn_l {parent=thumb_01_l, loc=<4.168 -1.1419 1.296>, rot=<-0.2021 -0.6228 -0.064 -0.7531>}
 wrist_inner_l {parent=hand_l, loc=<-0.7036 -1.6588 0.038>, rot=<-0.7163 -0.0808 0.1161 -0.6833>}
 wrist_outer_l {parent=hand_l, loc=<0.0197 2.2385 -0.1241>, rot=<0.6833 -0.1161 -0.0808 -0.7163>}
 weapon_l {parent=hand_l, loc=<-0.1974 -3.4111 0.3811>, rot=<0 0 0 1>}
 hand_attach_l {parent=hand_l, loc=<8.5 -3 0>, rot=<0.7071 0.7071 0 0>}
 lowerarm_twist_02_l {parent=lowerarm_l, loc=<9.6199 0 0>, rot=<0 0 0 1>}
 lowerarm_twist_01_l {parent=lowerarm_l, loc=<17.8549 0 0>, rot=<0 0 0 1>}
 lowerarm_correctiveRoot_l {parent=lowerarm_l, loc=<0 0 0>, rot=<0 0 0 1>}
 lowerarm_out_l {parent=lowerarm_correctiveRoot_l, loc=<0.584 -0.5752 2.7212>, rot=<0 0 0 1>}
 lowerarm_in_l {parent=lowerarm_correctiveRoot_l, loc=<0.7172 -0.7188 -3.9889>, rot=<1 0 0 0>}
 lowerarm_fwd_l {parent=lowerarm_correctiveRoot_l, loc=<1.6935 3.6591 -0.4389>, rot=<0.7071 0 0 -0.7071>}
 lowerarm_bck_l {parent=lowerarm_correctiveRoot_l, loc=<0.6623 -4.9183 -1.1643>, rot=<-0.7071 0 0 -0.7071>}
 upperarm_twist_01_l {parent=upperarm_l, loc=<0.0041 0 0>, rot=<0 0 0 1>}
 upperarm_twistCor_01_l {parent=upperarm_twist_01_l, loc=<0 0 0>, rot=<-0.0001 0.0188 -0.0028 -0.9998>}
 upperarm_twist_02_l {parent=upperarm_l, loc=<13.7941 0 0>, rot=<0 0 0 1>}
 upperarm_tricep_l {parent=upperarm_twist_02_l, loc=<0.6398 -4.8806 -0.2017>, rot=<-0.737 0.0081 0.0024 -0.6759>}
 upperarm_bicep_l {parent=upperarm_twist_02_l, loc=<0.6797 4.4394 -1.7918>, rot=<0.6759 -0.0024 0.0081 -0.737>}
 upperarm_twistCor_02_l {parent=upperarm_twist_02_l, loc=<0 0 0>, rot=<-0.0002 0.0563 -0.0085 -0.9984>}
 upperarm_correctiveRoot_l {parent=upperarm_l, loc=<0 0 0>, rot=<0 0 0 1>}
 upperarm_bck_l {parent=upperarm_correctiveRoot_l, loc=<1.4677 -7.6444 0.3859>, rot=<-0.5831 0 0 -0.8124>}
 upperarm_fwd_l {parent=upperarm_correctiveRoot_l, loc=<3.255 8.5092 -1.063>, rot=<0.7571 -0.0676 0.0801 -0.6449>}
 upperarm_in_l {parent=upperarm_correctiveRoot_l, loc=<6.4101 0.5444 -4.7618>, rot=<0.9129 0 0.4083 0>}
 upperarm_out_l {parent=upperarm_correctiveRoot_l, loc=<0.2203 -0.0517 5.8041>, rot=<0 0 0 1>}
 clavicle_out_l {parent=clavicle_l, loc=<11.8134 0.4439 6.2818>, rot=<-0.0026 0.0271 -0.0375 -0.9989>}
 clavicle_scap_l {parent=clavicle_l, loc=<9.9858 -7.5782 -1.9931>, rot=<-0.0271 -0.0026 0.9989 -0.0375>}
 clavicle_r {parent=spine_05, loc=<4.3575 -0.1466 2.3223>, rot=<0.7289 0.0465 -0.6829 0.009>}
 upperarm_r {parent=clavicle_r, loc=<-15.5111 0 0>, rot=<0.0226 -0.3234 -0.0088 -0.9459>}
 lowerarm_r {parent=upperarm_r, loc=<-26.6515 0 0>, rot=<0.0391 -0.0099 -0.2453 -0.9686>}
 hand_r {parent=lowerarm_r, loc=<-22.3492 0 0>, rot=<-0.7081 -0.0507 0.0975 -0.6975>}
 index_metacarpal_r {parent=hand_r, loc=<-1.6535 -0.0183 2.96>, rot=<0.0158 -0.0124 -0.0173 -0.9996>}
 index_01_r {parent=index_metacarpal_r, loc=<-8.3008 -0.0357 0.0524>, rot=<-0.0044 0.0485 0.0924 -0.9945>}
 index_02_r {parent=index_01_r, loc=<-3.3587 0.3004 -0.0452>, rot=<-0.002 0.0027 0.3225 -0.9466>}
 index_03_r {parent=index_02_r, loc=<-2.5578 0.1231 0.0251>, rot=<-0.002 0.0067 0.2558 -0.9667>}
 index_03_bulge_r {parent=index_03_r, loc=<-0.8528 0.6522 0.0609>, rot=<-0.7754 -0.0132 -0.0293 -0.6306>}
 index_03_half_r {parent=index_03_r, loc=<-0.02 0 0>, rot=<0 0 0 1>}
 index_02_dip_r {parent=index_03_half_r, loc=<0.0484 -0.5099 -0.0634>, rot=<0.6284 -0.03 0.0569 -0.7753>}
 index_03_in_r {parent=index_03_half_r, loc=<-0.0068 0.6099 0.0993>, rot=<-0.7585 -0.0043 -0.0144 -0.6515>}
 index_02_bulge_r {parent=index_02_r, loc=<-1.268 0.7244 0.1244>, rot=<-0.7248 -0.0082 -0.0356 -0.688>}
 index_02_side_out_r {parent=index_02_r, loc=<-1.2194 0.0242 -0.622>, rot=<-0.0119 -0.0165 -0.7152 -0.6986>}
 index_02_side_inn_r {parent=index_02_r, loc=<-1.2392 0.0466 0.7297>, rot=<0.7005 0.6947 0.1384 -0.0865>}
 index_02_half_r {parent=index_02_r, loc=<0 0 0>, rot=<0 0 0 1>}
 index_02_in_r {parent=index_02_half_r, loc=<-0.202 0.7363 0.168>, rot=<-0.7214 0.0528 0.0174 -0.6903>}
 index_02_pip_r {parent=index_02_half_r, loc=<-0.0518 -0.6387 -0.0257>, rot=<0.6993 0.0168 -0.0017 -0.7147>}
 index_01_palmMid_r {parent=index_01_r, loc=<-1.4778 0.9649 0.1824>, rot=<-0.7072 -0.1278 -0.0351 -0.6945>}
 index_01_bulge_r {parent=index_01_r, loc=<-2.7003 0.8646 0.2222>, rot=<-0.7809 -0.0063 0.0096 -0.6246>}
 index_01_side_out_r {parent=index_01_r, loc=<-2.7525 -0.1164 -0.7382>, rot=<0.0069 -0.0662 -0.7063 -0.7047>}
 index_01_side_inn_r {parent=index_01_r, loc=<-2.6652 0.0813 0.8637>, rot=<-0.7045 -0.7074 0.0543 -0.0187>}
 index_01_half_r {parent=index_01_r, loc=<0 0 0>, rot=<0 0 0 1>}
 index_01_mcp_r {parent=index_01_half_r, loc=<0.2823 -0.8444 -0.1505>, rot=<0.7167 -0.0113 -0.0196 -0.697>}
 index_01_palm_r {parent=index_01_half_r, loc=<-0.486 1.2096 0.4188>, rot=<-0.7242 -0.1505 -0.102 -0.6652>}
 index_metacarpal_slide_r {parent=index_metacarpal_r, loc=<-2.0671 -0.3153 -0.2008>, rot=<0.6834 -0.0204 0.0521 -0.7279>}
 middle_metacarpal_r {parent=hand_r, loc=<-2.1856 0.1725 0.5984>, rot=<0.006 0.0168 -0.0329 -0.9993>}
 middle_01_r {parent=middle_metacarpal_r, loc=<-7.3275 0.2447 -0.0129>, rot=<-0.0208 0.0042 0.0858 -0.9961>}
 middle_02_r {parent=middle_01_r, loc=<-4.0123 -0.1401 -0.1175>, rot=<-0.0218 0.0258 0.3512 -0.9357>}
 middle_03_r {parent=middle_02_r, loc=<-2.6077 -0.0534 -0.1373>, rot=<0.0082 0.0188 0.1708 -0.9851>}
 middle_03_bulge_r {parent=middle_03_r, loc=<-0.6039 0.5838 0.0914>, rot=<-0.7696 -0.007 -0.0255 -0.6379>}
 middle_03_half_r {parent=middle_03_r, loc=<0 0 0>, rot=<0 0 0 -1>}
 middle_03_in_r {parent=middle_03_half_r, loc=<0.0216 0.5364 0.087>, rot=<-0.7646 -0.0054 0.0016 -0.6445>}
 middle_02_dip_r {parent=middle_03_half_r, loc=<-0.0187 -0.5799 -0.0961>, rot=<0.6666 -0.0259 0.015 -0.7448>}
 middle_02_bulge_r {parent=middle_02_r, loc=<-1.3356 0.6387 0.1187>, rot=<-0.7529 0.018 0.034 -0.657>}
 middle_02_side_out_r {parent=middle_02_r, loc=<-1.2894 -0.0587 -0.6112>, rot=<-0.0099 0.0174 -0.7084 -0.7055>}
 middle_02_side_inn_r {parent=middle_02_r, loc=<-1.2773 -0.0693 0.7117>, rot=<-0.6836 -0.7292 -0.0303 0.0062>}
 middle_02_half_r {parent=middle_02_r, loc=<0 0 0>, rot=<0 0 0 1>}
 middle_02_pip_r {parent=middle_02_half_r, loc=<0.1178 -0.7082 -0.0526>, rot=<0.6625 0.0441 -0.0455 -0.7464>}
 middle_02_in_r {parent=middle_02_half_r, loc=<-0.2358 0.6451 0.1065>, rot=<-0.7503 0.0044 0.0137 -0.661>}
 middle_01_palmMid_r {parent=middle_01_r, loc=<-1.6801 1.1052 0.1343>, rot=<-0.7489 -0.0824 -0.1177 -0.6469>}
 middle_01_bulge_r {parent=middle_01_r, loc=<-3.1936 0.8245 0.0514>, rot=<-0.7031 -0.0127 -0.0163 -0.7108>}
 middle_01_side_out_r {parent=middle_01_r, loc=<-3.2591 -0.0643 -0.7134>, rot=<-0.0169 -0.0342 -0.707 -0.7062>}
 middle_01_side_inn_r {parent=middle_01_r, loc=<-3.1056 0.2058 0.7603>, rot=<-0.6831 -0.7191 0.1237 -0.0303>}
 middle_01_half_r {parent=middle_01_r, loc=<0 0 0>, rot=<0 0 0 1>}
 middle_01_mcp_r {parent=middle_01_half_r, loc=<-0.0118 -0.7211 -0.125>, rot=<0.7046 -0.0133 -0.0348 -0.7086>}
 middle_01_palm_r {parent=middle_01_half_r, loc=<-0.6492 1.5 0.265>, rot=<-0.7196 -0.1134 -0.1741 -0.6626>}
 middle_metacarpal_slide_r {parent=middle_metacarpal_r, loc=<-2.4672 -0.2464 0.0038>, rot=<0.7195 -0.0053 0.0079 -0.6944>}
 pinky_metacarpal_r {parent=hand_r, loc=<-1.7616 0.9753 -3.3814>, rot=<-0.0397 0.0814 -0.069 -0.9935>}
 pinky_01_r {parent=pinky_metacarpal_r, loc=<-7.1709 0.4831 0.1094>, rot=<-0.0654 -0.0035 0.2242 -0.9723>}
 pinky_02_r {parent=pinky_01_r, loc=<-2.8274 -0.0892 0.1043>, rot=<0.0037 0.0011 0.2854 -0.9584>}
 pinky_03_r {parent=pinky_02_r, loc=<-2.2046 -0.0307 0.0078>, rot=<-0.0002 0.0098 0.1445 -0.9895>}
 pinky_03_bulge_r {parent=pinky_03_r, loc=<-0.6193 0.5204 0.0399>, rot=<-0.697 0.0164 0.0076 -0.7168>}
 pinky_03_half_r {parent=pinky_03_r, loc=<-0.02 -0.0013 0>, rot=<0 0 0 1>}
 pinky_03_in_r {parent=pinky_03_half_r, loc=<0.0453 0.469 0.0097>, rot=<-0.7033 0.0203 -0.0094 -0.7105>}
 pinky_02_dip_r {parent=pinky_03_half_r, loc=<-0.0004 -0.5697 -0.0185>, rot=<0.7032 -0.0026 -0.0216 -0.7107>}
 pinky_02_bulge_r {parent=pinky_02_r, loc=<-1.0155 0.5119 -0.1478>, rot=<-0.6081 -0.0182 -0.0274 -0.7932>}
 pinky_02_side_out_r {parent=pinky_02_r, loc=<-1.0387 0.0455 -0.5577>, rot=<0.0814 -0.0847 -0.713 -0.6912>}
 pinky_02_side_inn_r {parent=pinky_02_r, loc=<-0.9943 -0.0125 0.5825>, rot=<0.7148 0.6988 0.0254 0.0038>}
 pinky_02_half_r {parent=pinky_02_r, loc=<0 0 0>, rot=<0 0 0 1>}
 pinky_02_in_r {parent=pinky_02_half_r, loc=<-0.1673 0.5556 -0.0809>, rot=<-0.6579 0.0017 -0.0123 -0.753>}
 pinky_02_pip_r {parent=pinky_02_half_r, loc=<0.1413 -0.5745 -0.0362>, rot=<0.6974 0.0445 -0.0668 -0.7121>}
 pinky_01_palmMid_r {parent=pinky_01_r, loc=<-0.9103 0.6387 -0.0039>, rot=<-0.6398 -0.0317 -0.0411 -0.7668>}
 pinky_01_bulge_r {parent=pinky_01_r, loc=<-1.8657 0.6076 -0.0246>, rot=<-0.6596 -0.0031 -0.0055 -0.7516>}
 pinky_01_side_out_r {parent=pinky_01_r, loc=<-1.9698 -0.1582 -0.5742>, rot=<0.0472 -0.0628 -0.6895 -0.72>}
 pinky_01_side_inn_r {parent=pinky_01_r, loc=<-1.9589 0.1083 0.6996>, rot=<0.7305 0.6691 -0.0969 0.0966>}
 pinky_01_half_r {parent=pinky_01_r, loc=<0 0 0>, rot=<0 0 0 1>}
 pinky_01_mcp_r {parent=pinky_01_half_r, loc=<0.2097 -0.7978 0.0339>, rot=<0.7244 -0.0981 -0.0292 -0.6817>}
 pinky_01_palm_r {parent=pinky_01_half_r, loc=<-0.4415 0.7346 -0.3849>, rot=<-0.5643 -0.106 -0.0209 -0.8184>}
 pinky_metacarpal_slide_r {parent=pinky_metacarpal_r, loc=<-1.6675 -0.3117 0.1581>, rot=<0.743 -0.0428 -0.0046 -0.6679>}
 ring_metacarpal_r {parent=hand_r, loc=<-2.2074 0.3625 -1.3638>, rot=<-0.0284 0.0722 -0.0458 -0.9959>}
 ring_01_r {parent=ring_metacarpal_r, loc=<-7.0295 0.6197 0.1446>, rot=<-0.0489 -0.0252 0.1892 -0.9804>}
 ring_02_r {parent=ring_01_r, loc=<-3.6174 -0.4647 0.1027>, rot=<0 0 0.2806 -0.9598>}
 ring_03_r {parent=ring_02_r, loc=<-2.4153 -0.1141 -0.0199>, rot=<0 0 0.1189 -0.9929>}
 ring_03_bulge_r {parent=ring_03_r, loc=<-0.3997 0.5585 -0.0882>, rot=<-0.7026 0.0054 -0.0216 -0.7112>}
 ring_03_half_r {parent=ring_03_r, loc=<0 0 0>, rot=<0 0 0 1>}
 ring_02_dip_r {parent=ring_03_half_r, loc=<0.2342 -0.5824 -0.0749>, rot=<0.7224 0.0478 0 -0.6898>}
 ring_03_in_r {parent=ring_03_half_r, loc=<0.2295 0.5923 -0.0785>, rot=<-0.6897 -0.0254 -0.061 -0.7211>}
 ring_02_bulge_r {parent=ring_02_r, loc=<-1.5943 0.6099 -0.0166>, rot=<-0.6818 0.0039 -0.0223 -0.7312>}
 ring_02_side_out_r {parent=ring_02_r, loc=<-1.5325 -0.0787 -0.6239>, rot=<0.0026 0.0443 -0.7048 -0.708>}
 ring_02_side_inn_r {parent=ring_02_r, loc=<-1.6242 -0.2523 0.5907>, rot=<0.713 0.6998 0.0413 -0.0176>}
 ring_02_half_r {parent=ring_02_r, loc=<0 0 0>, rot=<0 0 0 1>}
 ring_02_in_r {parent=ring_02_half_r, loc=<-0.4183 0.6712 0.0629>, rot=<-0.762 -0.012 -0.0782 -0.6427>}
 ring_02_pip_r {parent=ring_02_half_r, loc=<-0.0242 -0.6734 0.0393>, rot=<0.7047 0.0634 -0.0533 -0.7046>}
 ring_01_palmMid_r {parent=ring_01_r, loc=<-1.6473 0.9199 0.1887>, rot=<-0.7049 0.0039 -0.0272 -0.7088>}
 ring_01_bulge_r {parent=ring_01_r, loc=<-2.8564 0.8655 0.0855>, rot=<-0.7262 0.0479 0.0121 -0.6858>}
 ring_01_side_out_r {parent=ring_01_r, loc=<-2.8139 -0.1224 -0.6404>, rot=<0.0011 -0.0029 -0.6693 -0.7429>}
 ring_01_side_inn_r {parent=ring_01_r, loc=<-2.9145 0.1311 0.7792>, rot=<-0.7034 -0.7071 0.053 -0.0496>}
 ring_01_half_r {parent=ring_01_r, loc=<0 0 0>, rot=<0 0 0 1>}
 ring_01_mcp_r {parent=ring_01_half_r, loc=<-0.1592 -0.7106 -0.1499>, rot=<0.6797 -0.0132 0.0333 -0.7326>}
 ring_01_palm_r {parent=ring_01_half_r, loc=<-0.6198 1.3254 0.2137>, rot=<-0.6411 -0.0679 -0.1097 -0.7565>}
 ring_metacarpal_slide_r {parent=ring_metacarpal_r, loc=<-2.2474 -0.3637 0.0305>, rot=<0.7223 0.0127 -0.0266 -0.6909>}
 thumb_01_r {parent=hand_r, loc=<-1.3118 0.3191 2.459>, rot=<0.6546 -0.4042 0.0664 -0.6354>}
 thumb_02_r {parent=thumb_01_r, loc=<-5.9335 0.0447 -0.0401>, rot=<-0.0037 0.0213 0.298 -0.9543>}
 thumb_03_r {parent=thumb_02_r, loc=<-2.2223 -0.3682 0.0907>, rot=<0.0118 -0.013 0.0066 -0.9998>}
 thumb_03_bulge_r {parent=thumb_03_r, loc=<-0.8539 0.9278 0.1363>, rot=<-0.7582 -0.0381 -0.0057 -0.6509>}
 thumb_03_side_out_r {parent=thumb_03_r, loc=<-0.3379 0.2074 -0.7482>, rot=<-0.009 -0.018 0.9998 -0.0013>}
 thumb_03_side_inn_r {parent=thumb_03_r, loc=<-0.2417 0.0743 0.8274>, rot=<0.9952 -0.0755 -0.0151 -0.0604>}
 thumb_03_half_r {parent=thumb_03_r, loc=<0 0 0>, rot=<0 0 0 1>}
 thumb_03_pip_r {parent=thumb_03_half_r, loc=<-0.0282 -0.7125 -0.1327>, rot=<0.6567 -0.0165 -0.0216 -0.7536>}
 thumb_03_in_r {parent=thumb_03_half_r, loc=<-0.0161 0.7879 0.1277>, rot=<-0.7292 -0.0572 -0.0191 -0.6816>}
 thumb_02_bulge_r {parent=thumb_02_r, loc=<-1.6745 0.8593 0.1133>, rot=<-0.7435 -0.0285 0.0014 -0.6682>}
 thumb_02_side_out_r {parent=thumb_02_r, loc=<-1.0629 0.2662 -0.813>, rot=<0.0312 -0.1082 0.9902 -0.0823>}
 thumb_02_side_inn_r {parent=thumb_02_r, loc=<-0.9576 -0.1106 0.7046>, rot=<0.9902 0.0823 -0.0312 -0.1082>}
 thumb_02_half_r {parent=thumb_02_r, loc=<0 0 0>, rot=<0 0 0 1>}
 thumb_02_in_r {parent=thumb_02_half_r, loc=<-0.7327 1.0138 0.096>, rot=<-0.8303 -0.0684 -0.0426 -0.5515>}
 thumb_02_mcp_r {parent=thumb_02_half_r, loc=<0.5189 -0.6085 0.0585>, rot=<0.7115 0.0887 0.0008 -0.6971>}
 thumb_01_side_out_r {parent=thumb_01_r, loc=<-0.5727 -0.672 1.1092>, rot=<0.8946 0.0392 0.0667 -0.4402>}
 thumb_01_side_inn_r {parent=thumb_01_r, loc=<-4.1679 1.1419 -1.296>, rot=<-0.2021 -0.6228 -0.064 -0.7531>}
 wrist_inner_r {parent=hand_r, loc=<0.7033 1.6587 -0.0382>, rot=<-0.7163 -0.0808 0.1161 -0.6833>}
 wrist_outer_r {parent=hand_r, loc=<-0.0199 -2.2387 0.124>, rot=<0.6833 -0.1161 -0.0808 -0.7163>}
 weapon_r {parent=hand_r, loc=<0.1975 3.4112 -0.3811>, rot=<0 0 0 1>}
 hand_attach_r {parent=hand_r, loc=<-8.5 3 0>, rot=<0 0 -0.7071 -0.7071>}
 lowerarm_twist_02_r {parent=lowerarm_r, loc=<-9.6199 0 0>, rot=<0 0 0 -1>}
 lowerarm_twist_01_r {parent=lowerarm_r, loc=<-17.8549 0 0>, rot=<0 0 0 -1>}
 lowerarm_correctiveRoot_r {parent=lowerarm_r, loc=<0 0 -0.0003>, rot=<-0.0026 -0.0088 -0.0044 0.9999>}
 lowerarm_out_r {parent=lowerarm_correctiveRoot_r, loc=<-0.7071 0.6178 -2.9245>, rot=<0 0 0 -1>}
 lowerarm_fwd_r {parent=lowerarm_correctiveRoot_r, loc=<-1.8127 -3.5877 0.226>, rot=<0.7071 0 0 -0.7071>}
 lowerarm_in_r {parent=lowerarm_correctiveRoot_r, loc=<-0.8578 0.8169 3.8255>, rot=<1 0 0 0>}
 lowerarm_bck_r {parent=lowerarm_correctiveRoot_r, loc=<-0.7723 5.0016 0.9858>, rot=<-0.7071 0 0 -0.7071>}
 upperarm_twist_01_r {parent=upperarm_r, loc=<-0.0041 0 0>, rot=<0 0 0 -1>}
 upperarm_twistCor_01_r {parent=upperarm_twist_01_r, loc=<0 0 0>, rot=<-0.0001 0.0188 -0.0028 -0.9998>}
 upperarm_twist_02_r {parent=upperarm_r, loc=<-13.7941 0 0>, rot=<0 0 0 -1>}
 upperarm_tricep_r {parent=upperarm_twist_02_r, loc=<-0.7575 4.8717 0.0035>, rot=<-0.737 0.0081 0.0024 -0.6759>}
 upperarm_bicep_r {parent=upperarm_twist_02_r, loc=<-0.7985 -4.4483 1.5933>, rot=<0.6759 -0.0024 0.0081 -0.737>}
 upperarm_twistCor_02_r {parent=upperarm_twist_02_r, loc=<0 0 0>, rot=<-0.0002 0.0563 -0.0085 -0.9984>}
 upperarm_correctiveRoot_r {parent=upperarm_r, loc=<0 0 0>, rot=<0 0 0 1>}
 upperarm_fwd_r {parent=upperarm_correctiveRoot_r, loc=<-3.388 -8.5167 0.8737>, rot=<0.7571 -0.0676 0.0801 -0.6449>}
 upperarm_bck_r {parent=upperarm_correctiveRoot_r, loc=<-1.5996 7.6377 -0.5764>, rot=<-0.5831 0 0 -0.8124>}
 upperarm_in_r {parent=upperarm_correctiveRoot_r, loc=<-6.5511 -0.5461 4.557>, rot=<0.9129 0 0.4083 0>}
 upperarm_out_r {parent=upperarm_correctiveRoot_r, loc=<-0.3533 0.0442 -5.9935>, rot=<0 0 0 1>}
 clavicle_out_r {parent=clavicle_r, loc=<-12.0417 -0.4423 -6.3166>, rot=<-0.0026 0.0271 -0.0375 -0.9989>}
 clavicle_scap_r {parent=clavicle_r, loc=<-10.2145 7.5677 1.9571>, rot=<-0.0271 -0.0026 0.9989 -0.0375>}
 clavicle_pec_r {parent=spine_05, loc=<-9.2378 7.9155 9.6355>, rot=<0.7255 0.1085 -0.6793 0.0214>}
 spine_04_latissimus_l {parent=spine_05, loc=<-9.9618 -4.6263 -15.4435>, rot=<0.1673 -0.7577 0.0193 -0.6305>}
 spine_04_latissimus_r {parent=spine_05, loc=<-9.9671 -4.627 15.4547>, rot=<0.732 0.047 -0.6751 0.0785>}
 clavicle_pec_l {parent=spine_05, loc=<-9.2429 7.6848 -9.3102>, rot=<0.1085 -0.7255 -0.0214 -0.6793>}
 attach_backpack {parent=spine_05, loc=<-0.123 -13.473 0>, rot=<-0.0093 -0.707 0.0093 -0.707>}
 attach_cape {parent=spine_05, loc=<9.017 -11.714 0>, rot=<0.0616 -0.7044 -0.0616 -0.7044>}
 thigh_r {parent=pelvis, loc=<-2.5693 -0.0109 7.7501>, rot=<0.0572 -0.0012 0.9982 0.0155>}
 calf_r {parent=thigh_r, loc=<39.0484 0 0>, rot=<0.0006 -0.0161 -0.0391 -0.9991>}
 foot_r {parent=calf_r, loc=<42.561 -0.1077 -0.4292>, rot=<0.0418 -0.0455 0.0343 -0.9975>}
 ball_r {parent=foot_r, loc=<6.1598 -13.8264 -0.6806>, rot=<0.0001 -0.0001 -0.7063 -0.7079>}
 littletoe_01_r {parent=ball_r, loc=<1.1336 0.8229 4.2796>, rot=<-0.0392 -0.0068 -0.1132 -0.9928>}
 littletoe_02_r {parent=littletoe_01_r, loc=<-1.29 0 0>, rot=<0 0 -0.0872 -0.9962>}
 middletoe_01_r {parent=ball_r, loc=<-0.4086 1.5427 1.852>, rot=<-0.0392 -0.0068 -0.1132 -0.9928>}
 middletoe_02_r {parent=middletoe_01_r, loc=<-2.1519 0 0>, rot=<0 0 -0.0872 -0.9962>}
 bigtoe_01_r {parent=ball_r, loc=<-0.6422 1.0469 -2.1821>, rot=<-0.0022 -0.0373 -0.0284 -0.9989>}
 bigtoe_02_r {parent=bigtoe_01_r, loc=<-2.6424 0 0>, rot=<0 0 0 1>}
 indextoe_01_r {parent=ball_r, loc=<-0.6412 1.603 0.16>, rot=<-0.0043 -0.0319 -0.0741 -0.9967>}
 indextoe_02_r {parent=indextoe_01_r, loc=<-2.6423 0 0>, rot=<0 0 -0.1045 -0.9945>}
 ringtoe_01_r {parent=ball_r, loc=<0.2691 1.3415 3.1953>, rot=<-0.0392 -0.0068 -0.1132 -0.9928>}
 ringtoe_02_r {parent=ringtoe_01_r, loc=<-1.872 0 0>, rot=<0 0 -0.0872 -0.9962>}
 ankle_bck_r {parent=foot_r, loc=<-0.5266 4.0418 0.9291>, rot=<-0.7021 -0.0055 -0.05 -0.7103>}
 ankle_fwd_r {parent=foot_r, loc=<-2.1681 -5.0974 -0.3948>, rot=<0.7103 0.05 -0.0055 -0.7021>}
 heel_r {parent=foot_r, loc=<9.591 4.984 0.359>, rot=<0 0 -0.7071 0.7071>}
 tip_r {parent=heel_r, loc=<25.968 -0.136 -0.509>, rot=<0 0 0 1>}
 calf_twist_02_r {parent=calf_r, loc=<18.8064 0.0967 -0.0397>, rot=<0.0237 0.0095 -0.0035 -0.9997>}
 calf_twistCor_02_r {parent=calf_twist_02_r, loc=<0 0 0>, rot=<0 0 0 1>}
 calf_twist_01_r {parent=calf_r, loc=<35.813 0.0967 -0.0397>, rot=<0.0237 0.0095 -0.0035 -0.9997>}
 calf_correctiveRoot_r {parent=calf_r, loc=<0 0 0>, rot=<-0.0002 -0.008 0.0076 -0.9999>}
 calf_kneeBack_r {parent=calf_correctiveRoot_r, loc=<0.1554 5.7634 -0.4276>, rot=<-0.6933 -0.013 -0.0215 -0.7202>}
 calf_knee_r {parent=calf_correctiveRoot_r, loc=<0.19 -5.2458 -0.0016>, rot=<0.7202 0.0215 -0.013 -0.6933>}
 thigh_twist_01_r {parent=thigh_r, loc=<21.6303 0 0>, rot=<0.0013 0 0 -1>}
 thigh_twistCor_01_r {parent=thigh_twist_01_r, loc=<0 0 0>, rot=<0 0 0 1>}
 thigh_twist_02_r {parent=thigh_r, loc=<28.3644 0.3467 -0.0299>, rot=<0.0043 0 0 1>}
 thigh_twistCor_02_r {parent=thigh_twist_02_r, loc=<0 0 0>, rot=<0 0 0 1>}
 thigh_correctiveRoot_r {parent=thigh_r, loc=<0 0 0>, rot=<0 0 0 1>}
 thigh_fwd_r {parent=thigh_correctiveRoot_r, loc=<-4.2409 -6.7747 -0.7551>, rot=<0.7071 0 0 -0.7071>}
 thigh_bck_r {parent=thigh_correctiveRoot_r, loc=<-5.0375 10.0623 -1.5192>, rot=<-0.7071 0 0 -0.7071>}
 thigh_out_r {parent=thigh_correctiveRoot_r, loc=<-5.8166 0.6006 5.1102>, rot=<1 0 0 0>}
 thigh_in_r {parent=thigh_correctiveRoot_r, loc=<10.9283 -0.7659 -8.5779>, rot=<0 0 0 -1>}
 thigh_bck_lwr_r {parent=thigh_correctiveRoot_r, loc=<5.2762 10.6485 -1.4992>, rot=<-0.7071 0 0 -0.7071>}
 thigh_fwd_lwr_r {parent=thigh_correctiveRoot_r, loc=<1.028 -6.8997 -0.7243>, rot=<0.7071 0 0 -0.7071>}
 thigh_l {parent=pelvis, loc=<-2.5693 -0.0109 -7.7501>, rot=<0.0012 0.0572 0.0155 -0.9982>}
 calf_l {parent=thigh_l, loc=<-39.0484 0 0>, rot=<0.0006 -0.0161 -0.0391 -0.9991>}
 foot_l {parent=calf_l, loc=<-42.561 0.1077 0.4292>, rot=<0.0418 -0.0455 0.0343 -0.9975>}
 ball_l {parent=foot_l, loc=<-6.1599 13.8264 0.6806>, rot=<0.0001 -0.0001 -0.7063 -0.7079>}
 indextoe_01_l {parent=ball_l, loc=<0.6412 -1.603 -0.16>, rot=<-0.0043 -0.0319 -0.0741 -0.9967>}
 indextoe_02_l {parent=indextoe_01_l, loc=<2.6424 0 0>, rot=<0 0 -0.1045 -0.9945>}
 bigtoe_01_l {parent=ball_l, loc=<0.6422 -1.0469 2.1821>, rot=<-0.0022 -0.0373 -0.0284 -0.9989>}
 bigtoe_02_l {parent=bigtoe_01_l, loc=<2.6424 0 0>, rot=<0 0 0 1>}
 littletoe_01_l {parent=ball_l, loc=<-1.1336 -0.8229 -4.2796>, rot=<-0.0392 -0.0068 -0.1132 -0.9928>}
 littletoe_02_l {parent=littletoe_01_l, loc=<1.29 0 0>, rot=<0 0 -0.0872 -0.9962>}
 middletoe_01_l {parent=ball_l, loc=<0.4086 -1.5427 -1.852>, rot=<-0.0392 -0.0068 -0.1132 -0.9928>}
 middletoe_02_l {parent=middletoe_01_l, loc=<2.152 0 0>, rot=<0 0 -0.0872 -0.9962>}
 ringtoe_01_l {parent=ball_l, loc=<-0.2692 -1.3415 -3.1952>, rot=<-0.0392 -0.0068 -0.1132 -0.9928>}
 ringtoe_02_l {parent=ringtoe_01_l, loc=<1.872 0 0>, rot=<0 0 -0.0872 -0.9962>}
 ankle_bck_l {parent=foot_l, loc=<0.6127 -3.4437 -0.4925>, rot=<-0.7502 -0.002 -0.0503 -0.6593>}
 ankle_fwd_l {parent=foot_l, loc=<1.7798 4.5948 -0.2646>, rot=<0.6593 0.0503 -0.002 -0.7502>}
 heel_l {parent=foot_l, loc=<-9.591 -4.984 -0.359>, rot=<0.7071 -0.7071 0 0>}
 tip_l {parent=heel_l, loc=<-25.968 0.136 -0.509>, rot=<0 0 0 1>}
 calf_twist_02_l {parent=calf_l, loc=<-18.8064 -0.0967 0.0397>, rot=<0.0237 0.0095 -0.0035 -0.9997>}
 calf_twistCor_02_l {parent=calf_twist_02_l, loc=<0 0 0>, rot=<0 0 0 1>}
 calf_twist_01_l {parent=calf_l, loc=<-35.813 -0.0967 0.0397>, rot=<0.0237 0.0095 -0.0035 -0.9997>}
 calf_correctiveRoot_l {parent=calf_l, loc=<0 0 0>, rot=<-0.0002 -0.008 0.0076 -0.9999>}
 calf_kneeBack_l {parent=calf_correctiveRoot_l, loc=<-0.1586 -5.7585 0.404>, rot=<-0.6933 -0.013 -0.0215 -0.7202>}
 calf_knee_l {parent=calf_correctiveRoot_l, loc=<-0.1928 5.2492 -0.0011>, rot=<0.7202 0.0215 -0.013 -0.6933>}
 thigh_twist_01_l {parent=thigh_l, loc=<-21.6303 0 0>, rot=<0.0013 0 0 -1>}
 thigh_twistCor_01_l {parent=thigh_twist_01_l, loc=<0 0 0>, rot=<0 0 0 1>}
 thigh_twist_02_l {parent=thigh_l, loc=<-28.3643 -0.3467 0.0299>, rot=<0.0043 0 0 1>}
 thigh_twistCor_02_l {parent=thigh_twist_02_l, loc=<0 0 0>, rot=<0 0 0 1>}
 thigh_correctiveRoot_l {parent=thigh_l, loc=<0 0 0>, rot=<0 0 0 1>}
 thigh_bck_l {parent=thigh_correctiveRoot_l, loc=<5.0355 -10.0558 1.493>, rot=<-0.7071 0 0 -0.7071>}
 thigh_fwd_l {parent=thigh_correctiveRoot_l, loc=<4.2296 6.7731 0.7474>, rot=<0.7071 0 0 -0.7071>}
 thigh_out_l {parent=thigh_correctiveRoot_l, loc=<5.8141 -0.5864 -5.1227>, rot=<1 0 0 0>}
 thigh_in_l {parent=thigh_correctiveRoot_l, loc=<-10.8688 0.8267 8.5552>, rot=<0 0 0 1>}
 thigh_bck_lwr_l {parent=thigh_correctiveRoot_l, loc=<-5.2985 -10.4915 1.2302>, rot=<-0.7071 0 0 -0.7071>}
 thigh_fwd_lwr_l {parent=thigh_correctiveRoot_l, loc=<-1.1209 7.3878 0.6484>, rot=<0.7071 0 0 -0.7071>}
 attach {parent=root, loc=<0 0 0>, rot=<0 0 0 1>}
 VB attach_ik_hand_gun {parent=attach, loc=<-51.9729 4.0883 102.6935>, rot=<0.6084 0.1793 0.4613 -0.6205>}
 ik_foot_root {parent=root, loc=<0 0 0>, rot=<0 0 0 1>}
 ik_foot_l {parent=ik_foot_root, loc=<15.2608 -4.4176 9.4433>, rot=<0.0357 0.7045 0.0284 -0.7082>}
 ik_foot_r {parent=ik_foot_root, loc=<-15.2608 -4.4176 9.4433>, rot=<-0.7082 -0.0284 0.7045 -0.0357>}
 ik_hand_root {parent=root, loc=<0 0 0>, rot=<0 0 0 1>}
 ik_hand_gun {parent=ik_hand_root, loc=<-51.9729 4.0883 102.6935>, rot=<0.6084 0.1793 0.4613 -0.6205>}
 ik_hand_l {parent=ik_hand_gun, loc=<53.0286 82.1782 35.205>, rot=<0.5102 0.7906 0.3387 0>}
 ik_hand_r {parent=ik_hand_gun, loc=<0 0 0>, rot=<0 0 0 1>}
 root_fp {parent=root, loc=<0 0 0>, rot=<0 0 0 1>}
 attach_fp {parent=root_fp, loc=<0 0 0>, rot=<0 0 0 1>}
 camera_root {parent=root_fp, loc=<0 0 0>, rot=<0 0 0 1>}
 camera {parent=camera_root, loc=<0 0 0>, rot=<0 0 0 1>}
 attach_emote_01 {parent=root, loc=<0 0 0>, rot=<0 0 0 1>}
 attach_emote_02 {parent=root, loc=<0 0 0>, rot=<0 0 0 1>}
 attach_emote_03 {parent=root, loc=<0 0 0>, rot=<0 0 0 1>}
 global_ctrl {controlType=EULER_TRANSFORM, type=Control, loc=<0 0 0>, rot=<0 0 0 1>}
 root_ctrl {controlType=EULER_TRANSFORM, parent=global_ctrl, type=Control, loc=<0 0 0>, rot=<0 0 0 1>}
 body_offset_ctrl {controlType=EULER_TRANSFORM, parent=global_ctrl, type=Control, loc=<0 0 0>, rot=<0 0 0 1>}
 leg_l_controls_space {parent=body_offset_ctrl, type=Null, loc=<0 0 0>, rot=<0 0 0 1>}
 thigh_l_fk_ctrl_space {parent=leg_l_controls_space, type=Null, loc=<7.7501 -0.81 90.5518>, rot=<-0.0019 -0.7464 0.0016 0.6655>}
 thigh_l_fk_ctrl {controlType=EULER_TRANSFORM, parent=thigh_l_fk_ctrl_space, type=Control, loc=<0 0 0>, rot=<0 0 0 1>}
 calf_l_fk_ctrl {controlType=EULER_TRANSFORM, parent=thigh_l_fk_ctrl, type=Control, loc=<0 0 0>, rot=<0 0 0 1>}
 foot_l_fk_ctrl {controlType=EULER_TRANSFORM, parent=calf_l_fk_ctrl, type=Control, loc=<0 0 0>, rot=<0 0 0 1>}
 ball_l_fk_ctrl {controlType=EULER_TRANSFORM, parent=foot_l_fk_ctrl, type=Control, loc=<0 0 0>, rot=<0 0 0 1>}
 toes_l_space {parent=leg_l_controls_space, type=Null, loc=<15.7991 9.3506 3.1406>, rot=<0.5229 0.4735 -0.4801 -0.5214>}
 foot_l_ik_ctrl {controlType=EULER_TRANSFORM, parent=leg_l_controls_space, type=Control, loc=<0 0 0>, rot=<0 0 0 1>}
 leg_l_stretch_switch {controlType=BOOL, parent=foot_l_ik_ctrl, type=Control, loc=<0 0 0>, rot=<0 0 0 1>}
 heel_l {parent=foot_l_ik_ctrl, type=Null, loc=<-0.1481 -5.0928 -9.5393>, rot=<0.4807 -0.5208 0.5234 -0.4729>}
 tip_l {parent=heel_l, type=Null, loc=<-25.968 0.136 -0.509>, rot=<0 0 0 1>}
 ball_l_ik_ctrl {controlType=EULER_TRANSFORM, parent=tip_l, type=Control, loc=<0 0 0>, rot=<0 0 0 1>}
 foot_bk1_l {parent=tip_l, type=Null, loc=<7.1576 -3.5671 -0.5306>, rot=<0 0 0 1>}
 foot_bk1_l_ctrl {controlType=EULER_TRANSFORM, parent=foot_bk1_l, type=Control, loc=<0 0 0>, rot=<0 0 0 1>}
 foot_ik_l {parent=foot_bk1_l_ctrl, type=Null, loc=<13.8264 -6.1599 0.6806>, rot=<0 0 0 1>}
 foot_roll_l_bfr {parent=foot_bk1_l_ctrl, type=Null, loc=<13.8264 -6.1599 0.6806>, rot=<0 0 0 1>}
 foot_roll_l_ctrl {controlType=EULER_TRANSFORM, parent=foot_roll_l_bfr, type=Control, loc=<0 0 0>, rot=<0 0 0 1>}
 roll_blend_l_ctrl {controlType=FLOAT, parent=foot_roll_l_ctrl, type=Control, loc=<1 0 0>, rot=<0 0 0 1>}
 tip_l_ctrl {controlType=EULER_TRANSFORM, parent=heel_l, type=Control, loc=<0 0 0>, rot=<0 0 0 1>}
 heel_l_ctrl {controlType=EULER_TRANSFORM, parent=foot_l_ik_ctrl, type=Control, loc=<0 0 0>, rot=<0 0 0 1>}
 leg_l_softness {controlType=FLOAT, parent=foot_l_ik_ctrl, type=Control, loc=<0 0 0>, rot=<0 0 0 1>}
 leg_l_pv_footFollow {parent=leg_l_controls_space, type=Null, loc=<7.7501 -0.81 90.5518>, rot=<0.0164 -0.6736 0.015 -0.7388>}
 leg_l_pv_ik_ctrl {controlType=EULER_TRANSFORM, parent=leg_l_pv_footFollow, type=Control, loc=<0 0 0>, rot=<0 0 0 1>}
 leg_r_controls_space {parent=body_offset_ctrl, type=Null, loc=<0 0 0>, rot=<0 0 0 1>, scale=<-1 1 1>}
 thigh_r_fk_ctrl_space {parent=leg_r_controls_space, type=Null, loc=<7.7501 -0.81 90.5518>, rot=<-0.0019 -0.7464 0.0016 0.6655>}
 thigh_r_fk_ctrl {controlType=EULER_TRANSFORM, parent=thigh_r_fk_ctrl_space, type=Control, loc=<0 0 0>, rot=<0 0 0 1>}
 calf_r_fk_ctrl {controlType=EULER_TRANSFORM, parent=thigh_r_fk_ctrl, type=Control, loc=<0 0 0>, rot=<0 0 0 1>}
 foot_r_fk_ctrl {controlType=EULER_TRANSFORM, parent=calf_r_fk_ctrl, type=Control, loc=<0 0 0>, rot=<0 0 0 1>}
 ball_r_fk_ctrl {controlType=EULER_TRANSFORM, parent=foot_r_fk_ctrl, type=Control, loc=<0 0 0>, rot=<0 0 0 1>}
 toes_r_space {parent=leg_r_controls_space, type=Null, loc=<15.7991 9.3506 3.1406>, rot=<0.5214 0.4801 0.4735 0.5229>, scale=<-1 1 1>}
 foot_r_ik_ctrl {controlType=EULER_TRANSFORM, parent=leg_r_controls_space, type=Control, loc=<0 0 0>, rot=<0 0 0 1>}
 leg_r_stretch_switch {controlType=BOOL, parent=foot_r_ik_ctrl, type=Control, loc=<0 0 0>, rot=<0 0 0 1>}
 heel_r {parent=foot_r_ik_ctrl, type=Null, loc=<-0.1481 -5.0928 -9.5393>, rot=<0.4807 -0.5208 0.5234 -0.4729>}
 tip_r {parent=heel_r, type=Null, loc=<-25.968 -0.136 -0.509>, rot=<0 0 0 1>}
 ball_r_ik_ctrl {controlType=EULER_TRANSFORM, parent=tip_r, type=Control, loc=<0 0 0>, rot=<0 0 0 1>}
 foot_bk1_r {parent=tip_r, type=Null, loc=<7.1576 -3.2952 -0.5306>, rot=<0 0 0 1>}
 foot_bk1_r_ctrl {controlType=EULER_TRANSFORM, parent=foot_bk1_r, type=Control, loc=<0 0 0>, rot=<0 0 0 1>}
 foot_ik_r {parent=foot_bk1_r_ctrl, type=Null, loc=<13.8264 -6.1598 0.6806>, rot=<0 0 0 1>}
 foot_roll_r_bfr {parent=foot_bk1_r_ctrl, type=Null, loc=<13.8264 -6.1598 0.6806>, rot=<0 0 0 1>}
 foot_roll_r_ctrl {controlType=EULER_TRANSFORM, parent=foot_roll_r_bfr, type=Control, loc=<0 0 0>, rot=<0 0 0 1>}
 roll_blend_r_ctrl {controlType=FLOAT, parent=foot_roll_r_ctrl, type=Control, loc=<1 0 0>, rot=<0 0 0 1>}
 tip_r_ctrl {controlType=EULER_TRANSFORM, parent=heel_r, type=Control, loc=<0 0 0>, rot=<0 0 0 1>}
 heel_r_ctrl {controlType=EULER_TRANSFORM, parent=foot_r_ik_ctrl, type=Control, loc=<0 0 0>, rot=<0 0 0 1>}
 leg_r_softness {controlType=FLOAT, parent=foot_r_ik_ctrl, type=Control, loc=<0 0 0>, rot=<0 0 0 1>}
 leg_r_pv_footFollow {parent=leg_r_controls_space, type=Null, loc=<7.7501 -0.81 90.5518>, rot=<-0.015 -0.7388 0.0164 0.6736>, scale=<-1 1 1>}
 leg_r_pv_ik_ctrl {controlType=EULER_TRANSFORM, parent=leg_r_pv_footFollow, type=Control, loc=<0 0 0>, rot=<0 0 0 1>}
 arm_l_controls_space {parent=body_offset_ctrl, type=Null, loc=<0 0 0>, rot=<0 0 0 1>}
 upperarm_l_fk_ctrl_space {parent=arm_l_controls_space, type=Null, loc=<17.7788 -4.8444 134.4501>, rot=<-0.0134 0.3555 -0.014 0.9345>}
 upperarm_l_fk_ctrl {controlType=EULER_TRANSFORM, parent=upperarm_l_fk_ctrl_space, type=Control, loc=<0 0 0>, rot=<0 0 0 1>}
 lowerarm_l_fk_ctrl {controlType=EULER_TRANSFORM, parent=upperarm_l_fk_ctrl, type=Control, loc=<0 0 0>, rot=<0 0 0 1>}
 hand_l_fk_ctrl {controlType=EULER_TRANSFORM, parent=lowerarm_l_fk_ctrl, type=Control, loc=<0 0 0>, rot=<0 0 0 1>}
 arm_l_fk_local {controlType=BOOL, parent=upperarm_l_fk_ctrl, type=Control, loc=<1 0 0>, rot=<0 0 0 1>}
 fingers_l_space {parent=arm_l_controls_space, type=Null, loc=<51.9727 4.0883 102.6935>, rot=<0.0744 0.357 0.2158 0.9058>}
 thumb_01_l_ctrl {controlType=EULER_TRANSFORM, parent=fingers_l_space, type=Control, loc=<0 0 0>, rot=<0 0 0 1>}
 thumb_02_l_ctrl {controlType=EULER_TRANSFORM, parent=thumb_01_l_ctrl, type=Control, loc=<0 0 0>, rot=<0 0 0 1>}
 thumb_03_l_ctrl {controlType=EULER_TRANSFORM, parent=thumb_02_l_ctrl, type=Control, loc=<0 0 0>, rot=<0 0 0 1>}
 thumb_curl_l_ctrl {controlType=EULER_TRANSFORM, parent=thumb_01_l_ctrl, type=Control, loc=<0 0 0>, rot=<0 0 0 1>}
 index_metacarpal_l_ctrl {controlType=EULER_TRANSFORM, parent=fingers_l_space, type=Control, loc=<0 0 0>, rot=<0 0 0 1>}
 index_01_l_ctrl {controlType=EULER_TRANSFORM, parent=index_metacarpal_l_ctrl, type=Control, loc=<0 0 0>, rot=<0 0 0 1>}
 index_02_l_ctrl {controlType=EULER_TRANSFORM, parent=index_01_l_ctrl, type=Control, loc=<0 0 0>, rot=<0 0 0 1>}
 index_03_l_ctrl {controlType=EULER_TRANSFORM, parent=index_02_l_ctrl, type=Control, loc=<0 0 0>, rot=<0 0 0 1>}
 index_curl_l_ctrl {controlType=EULER_TRANSFORM, parent=index_01_l_ctrl, type=Control, loc=<0 0 0>, rot=<0 0 0 1>}
 middle_metacarpal_l_ctrl {controlType=EULER_TRANSFORM, parent=fingers_l_space, type=Control, loc=<0 0 0>, rot=<0 0 0 1>}
 middle_01_l_ctrl {controlType=EULER_TRANSFORM, parent=middle_metacarpal_l_ctrl, type=Control, loc=<0 0 0>, rot=<0 0 0 1>}
 middle_02_l_ctrl {controlType=EULER_TRANSFORM, parent=middle_01_l_ctrl, type=Control, loc=<0 0 0>, rot=<0 0 0 1>}
 middle_03_l_ctrl {controlType=EULER_TRANSFORM, parent=middle_02_l_ctrl, type=Control, loc=<0 0 0>, rot=<0 0 0 1>}
 middle_curl_l_ctrl {controlType=EULER_TRANSFORM, parent=middle_01_l_ctrl, type=Control, loc=<0 0 0>, rot=<0 0 0 1>}
 ring_metacarpal_l_ctrl {controlType=EULER_TRANSFORM, parent=fingers_l_space, type=Control, loc=<0 0 0>, rot=<0 0 0 1>}
 ring_01_l_ctrl {controlType=EULER_TRANSFORM, parent=ring_metacarpal_l_ctrl, type=Control, loc=<0 0 0>, rot=<0 0 0 1>}
 ring_02_l_ctrl {controlType=EULER_TRANSFORM, parent=ring_01_l_ctrl, type=Control, loc=<0 0 0>, rot=<0 0 0 1>}
 ring_03_l_ctrl {controlType=EULER_TRANSFORM, parent=ring_02_l_ctrl, type=Control, loc=<0 0 0>, rot=<0 0 0 1>}
 ring_curl_l_ctrl {controlType=EULER_TRANSFORM, parent=ring_01_l_ctrl, type=Control, loc=<0 0 0>, rot=<0 0 0 1>}
 pinky_metacarpal_l_ctrl {controlType=EULER_TRANSFORM, parent=fingers_l_space, type=Control, loc=<0 0 0>, rot=<0 0 0 1>}
 pinky_01_l_ctrl {controlType=EULER_TRANSFORM, parent=pinky_metacarpal_l_ctrl, type=Control, loc=<0 0 0>, rot=<0 0 0 1>}
 pinky_02_l_ctrl {controlType=EULER_TRANSFORM, parent=pinky_01_l_ctrl, type=Control, loc=<0 0 0>, rot=<0 0 0 1>}
 pinky_03_l_ctrl {controlType=EULER_TRANSFORM, parent=pinky_02_l_ctrl, type=Control, loc=<0 0 0>, rot=<0 0 0 1>}
 pinky_curl_l_ctrl {controlType=EULER_TRANSFORM, parent=pinky_01_l_ctrl, type=Control, loc=<0 0 0>, rot=<0 0 0 1>}
 Auto Interpolate_3 {controlType=BOOL, parent=pinky_01_l_ctrl, type=Control, loc=<0 0 0>, rot=<0 0 0 1>}
 Auto Interpolate {controlType=BOOL, parent=pinky_metacarpal_l_ctrl, type=Control, loc=<0 0 0>, rot=<0 0 0 1>}
 hand_l_ik_ctrl {controlType=EULER_TRANSFORM, parent=arm_l_controls_space, type=Control, loc=<0 0 0>, rot=<0 0 0 1>}
 arm_l_stretch_switch {controlType=BOOL, parent=hand_l_ik_ctrl, type=Control, loc=<0 0 0>, rot=<0 0 0 1>}
 Softness {controlType=FLOAT, parent=hand_l_ik_ctrl, type=Control, loc=<0 0 0>, rot=<0 0 0 1>}
 Keep Aligned {controlType=BOOL, parent=hand_l_ik_ctrl, type=Control, loc=<0 0 0>, rot=<0 0 0 1>}
 hand_l_ikorient {parent=hand_l_ik_ctrl, type=Null, loc=<0 0 0>, rot=<0 0 0 1>}
 hand_l_ikorient_ctrl {controlType=ROTATOR, parent=hand_l_ikorient, type=Control, loc=<0 0 0>, rot=<0 0 0 1>}
 arm_l_pv_ik_ctrl {controlType=EULER_TRANSFORM, parent=arm_l_controls_space, type=Control, loc=<0 0 0>, rot=<0 0 0 1>}
 arm_r_controls_space {parent=body_offset_ctrl, type=Null, loc=<0 0 0>, rot=<0 0 0 1>, scale=<-1 1 1>}
 upperarm_r_fk_ctrl_space {parent=arm_r_controls_space, type=Null, loc=<17.7788 -4.8444 134.4502>, rot=<-0.0134 0.3555 -0.014 0.9345>}
 upperarm_r_fk_ctrl {controlType=EULER_TRANSFORM, parent=upperarm_r_fk_ctrl_space, type=Control, loc=<0 0 0>, rot=<0 0 0 1>}
 lowerarm_r_fk_ctrl {controlType=EULER_TRANSFORM, parent=upperarm_r_fk_ctrl, type=Control, loc=<0 0 0>, rot=<0 0 0 1>}
 hand_r_fk_ctrl {controlType=EULER_TRANSFORM, parent=lowerarm_r_fk_ctrl, type=Control, loc=<0 0 0>, rot=<0 0 0 1>}
 arm_r_fk_local {controlType=BOOL, parent=upperarm_r_fk_ctrl, type=Control, loc=<1 0 0>, rot=<0 0 0 1>}
 fingers_r_space {parent=arm_r_controls_space, type=Null, loc=<51.9729 4.0883 102.6935>, rot=<0.0744 0.357 0.2158 0.9058>}
 thumb_01_r_ctrl {controlType=EULER_TRANSFORM, parent=fingers_r_space, type=Control, loc=<0 0 0>, rot=<0 0 0 1>}
 thumb_02_r_ctrl {controlType=EULER_TRANSFORM, parent=thumb_01_r_ctrl, type=Control, loc=<0 0 0>, rot=<0 0 0 1>}
 thumb_03_r_ctrl {controlType=EULER_TRANSFORM, parent=thumb_02_r_ctrl, type=Control, loc=<0 0 0>, rot=<0 0 0 1>}
 thumb_curl_r_ctrl {controlType=EULER_TRANSFORM, parent=thumb_01_r_ctrl, type=Control, loc=<0 0 0>, rot=<0 0 0 1>}
 index_metacarpal_r_ctrl {controlType=EULER_TRANSFORM, parent=fingers_r_space, type=Control, loc=<0 0 0>, rot=<0 0 0 1>}
 index_01_r_ctrl {controlType=EULER_TRANSFORM, parent=index_metacarpal_r_ctrl, type=Control, loc=<0 0 0>, rot=<0 0 0 1>}
 index_02_r_ctrl {controlType=EULER_TRANSFORM, parent=index_01_r_ctrl, type=Control, loc=<0 0 0>, rot=<0 0 0 1>}
 index_03_r_ctrl {controlType=EULER_TRANSFORM, parent=index_02_r_ctrl, type=Control, loc=<0 0 0>, rot=<0 0 0 1>}
 index_curl_r_ctrl {controlType=EULER_TRANSFORM, parent=index_01_r_ctrl, type=Control, loc=<0 0 0>, rot=<0 0 0 1>}
 middle_metacarpal_r_ctrl {controlType=EULER_TRANSFORM, parent=fingers_r_space, type=Control, loc=<0 0 0>, rot=<0 0 0 1>}
 middle_01_r_ctrl {controlType=EULER_TRANSFORM, parent=middle_metacarpal_r_ctrl, type=Control, loc=<0 0 0>, rot=<0 0 0 1>}
 middle_02_r_ctrl {controlType=EULER_TRANSFORM, parent=middle_01_r_ctrl, type=Control, loc=<0 0 0>, rot=<0 0 0 1>}
 middle_03_r_ctrl {controlType=EULER_TRANSFORM, parent=middle_02_r_ctrl, type=Control, loc=<0 0 0>, rot=<0 0 0 1>}
 middle_curl_r_ctrl {controlType=EULER_TRANSFORM, parent=middle_01_r_ctrl, type=Control, loc=<0 0 0>, rot=<0 0 0 1>}
 ring_metacarpal_r_ctrl {controlType=EULER_TRANSFORM, parent=fingers_r_space, type=Control, loc=<0 0 0>, rot=<0 0 0 1>}
 ring_01_r_ctrl {controlType=EULER_TRANSFORM, parent=ring_metacarpal_r_ctrl, type=Control, loc=<0 0 0>, rot=<0 0 0 1>}
 ring_02_r_ctrl {controlType=EULER_TRANSFORM, parent=ring_01_r_ctrl, type=Control, loc=<0 0 0>, rot=<0 0 0 1>}
 ring_03_r_ctrl {controlType=EULER_TRANSFORM, parent=ring_02_r_ctrl, type=Control, loc=<0 0 0>, rot=<0 0 0 1>}
 ring_curl_r_ctrl {controlType=EULER_TRANSFORM, parent=ring_01_r_ctrl, type=Control, loc=<0 0 0>, rot=<0 0 0 1>}
 pinky_metacarpal_r_ctrl {controlType=EULER_TRANSFORM, parent=fingers_r_space, type=Control, loc=<0 0 0>, rot=<0 0 0 1>}
 pinky_01_r_ctrl {controlType=EULER_TRANSFORM, parent=pinky_metacarpal_r_ctrl, type=Control, loc=<0 0 0>, rot=<0 0 0 1>}
 pinky_02_r_ctrl {controlType=EULER_TRANSFORM, parent=pinky_01_r_ctrl, type=Control, loc=<0 0 0>, rot=<0 0 0 1>}
 pinky_03_r_ctrl {controlType=EULER_TRANSFORM, parent=pinky_02_r_ctrl, type=Control, loc=<0 0 0>, rot=<0 0 0 1>}
 pinky_curl_r_ctrl {controlType=EULER_TRANSFORM, parent=pinky_01_r_ctrl, type=Control, loc=<0 0 0>, rot=<0 0 0 1>}
 Auto Interpolate_4 {controlType=BOOL, parent=pinky_01_r_ctrl, type=Control, loc=<0 0 0>, rot=<0 0 0 1>}
 Auto Interpolate_2 {controlType=BOOL, parent=pinky_metacarpal_r_ctrl, type=Control, loc=<0 0 0>, rot=<0 0 0 1>}
 hand_r_ik_ctrl {controlType=EULER_TRANSFORM, parent=arm_r_controls_space, type=Control, loc=<0 0 0>, rot=<0 0 0 1>}
 arm_r_stretch_switch {controlType=BOOL, parent=hand_r_ik_ctrl, type=Control, loc=<0 0 0>, rot=<0 0 0 1>}
 Softness_2 {controlType=FLOAT, parent=hand_r_ik_ctrl, type=Control, loc=<0 0 0>, rot=<0 0 0 1>}
 Keep Aligned_2 {controlType=BOOL, parent=hand_r_ik_ctrl, type=Control, loc=<0 0 0>, rot=<0 0 0 1>}
 hand_r_ikorient {parent=hand_r_ik_ctrl, type=Null, loc=<0 0 0>, rot=<0 0 0 1>}
 hand_r_ikorient_ctrl {controlType=ROTATOR, parent=hand_r_ikorient, type=Control, loc=<0 0 0>, rot=<0 0 0 1>}
 arm_r_pv_ik_ctrl {controlType=EULER_TRANSFORM, parent=arm_r_controls_space, type=Control, loc=<0 0 0>, rot=<0 0 0 1>}
 body_ctrl {controlType=EULER_TRANSFORM, parent=body_offset_ctrl, type=Control, loc=<0 0 0>, rot=<0 0 0 1>}
 hips_ctrl {controlType=EULER_TRANSFORM, parent=body_ctrl, type=Control, loc=<0 0 0>, rot=<0 0 0 1>}
 hips_tan_ctrl {controlType=EULER_TRANSFORM, parent=hips_ctrl, type=Control, loc=<0 0 0>, rot=<0 0 0 1>}
 spine_01_ctrl {controlType=EULER_TRANSFORM, parent=body_ctrl, type=Control, loc=<0 0 0>, rot=<0 0 0 1>}
 spine_02_ctrl {controlType=EULER_TRANSFORM, parent=spine_01_ctrl, type=Control, loc=<0 0 0>, rot=<0 0 0 1>}
 spine_03_ctrl {controlType=EULER_TRANSFORM, parent=spine_02_ctrl, type=Control, loc=<0 0 0>, rot=<0 0 0 1>}
 chest_space {parent=spine_03_ctrl, type=Null, loc=<0 0 0>, rot=<-0.0294 0.7065 0.0294 -0.7065>}
 spine_04_ctrl {controlType=EULER_TRANSFORM, parent=chest_space, type=Control, loc=<0 0 0>, rot=<0 0 0 1>}
 head_fk_space {parent=spine_04_ctrl, type=Null, loc=<0 1.8821 20.8752>, rot=<0 0 0 1>}
 head_ctrl {controlType=EULER_TRANSFORM, parent=head_fk_space, type=Control, loc=<0 0 0>, rot=<0 0 0 1>}
 head_local {controlType=BOOL, parent=head_ctrl, type=Control, loc=<1 0 0>, rot=<0 0 0 1>}
 Stretch {controlType=BOOL, parent=head_ctrl, type=Control, loc=<0 0 0>, rot=<0 0 0 1>}
 space {controlType=INTEGER, parent=head_ctrl, type=Control, loc=<1 0 0>, rot=<0 0 0 1>}
 clavicle_l_ctrl {controlType=EULER_TRANSFORM, parent=spine_04_ctrl, type=Control, loc=<0 0 0>, rot=<0 0 0 1>}
 clavicle_r_ctrl {controlType=EULER_TRANSFORM, parent=spine_04_ctrl, type=Control, loc=<0 0 0>, rot=<0 0 0 1>}
 neck_01_ctrl {controlType=EULER_TRANSFORM, parent=spine_04_ctrl, type=Control, loc=<0 0 0>, rot=<0 0 0 1>}
 neck_02_ctrl {controlType=EULER_TRANSFORM, parent=neck_01_ctrl, type=Control, loc=<0 0 0>, rot=<0 0 0 1>}
 head_ik_space {parent=spine_04_ctrl, type=Null, loc=<0 1.8821 20.8752>, rot=<0 0 0 1>}
 head_ik_ctrl {controlType=EULER_TRANSFORM, parent=head_ik_space, type=Control, loc=<0 0 0>, rot=<0 0 0 1>}
 head_mid {parent=head_ik_ctrl, type=Null, loc=<-0.0001 -1.0459 -5.1749>, rot=<0 0 0 1>}
 head_stretch {controlType=BOOL, parent=head_ik_ctrl, type=Control, loc=<0 0 0>, rot=<0 0 0 1>}
 chest_ctrl {controlType=EULER_TRANSFORM, parent=body_ctrl, type=Control, loc=<0 0 0>, rot=<0 0 0 1>}
 chest_tan_ctrl {controlType=EULER_TRANSFORM, parent=chest_ctrl, type=Control, loc=<0 0 0>, rot=<0 0 0 1>}
 spine_stretch_switch {controlType=BOOL, parent=chest_ctrl, type=Control, loc=<0 0 0>, rot=<0 0 0 1>}
 arm_l_fk_ik_switch {controlType=BOOL, parent=global_ctrl, type=Control, loc=<0 0 0>, rot=<0 0 0 1>}
 leg_l_fk_ik_switch {controlType=BOOL, parent=global_ctrl, type=Control, loc=<1 0 0>, rot=<0 0 0 1>}
 arm_r_fk_ik_switch {controlType=BOOL, parent=global_ctrl, type=Control, loc=<0 0 0>, rot=<0 0 0 1>}
 leg_r_fk_ik_switch {controlType=BOOL, parent=global_ctrl, type=Control, loc=<1 0 0>, rot=<0 0 0 1>}
 ShowBodyControls {controlType=BOOL, parent=global_ctrl, type=Control, loc=<1 0 0>, rot=<0 0 0 1>}
 spine_fk_ik_switch {controlType=BOOL, parent=global_ctrl, type=Control, loc=<0 0 0>, rot=<0 0 0 1>}
 neck_fk_ik_switch {controlType=BOOL, parent=global_ctrl, type=Control, loc=<0 0 0>, rot=<0 0 0 1>}
 r_knee_corrective {type=Curve, loc=<0 0 0>, rot=<0 0 0 1>}
 l_knee_corrective {type=Curve, loc=<0 0 0>, rot=<0 0 0 1>}
 Male_Med_Head_01_Pout {type=Curve, loc=<0 0 0>, rot=<0 0 0 1>}
 Male_Med_Head_02_Pout {type=Curve, loc=<0 0 0>, rot=<0 0 0 1>}
 Male_Med_Head_03_Pout {type=Curve, loc=<0 0 0>, rot=<0 0 0 1>}
 Male_Med_Head_05_Effort {type=Curve, loc=<0 0 0>, rot=<0 0 0 1>}
 Stretch {type=Curve, loc=<0 0 0>, rot=<0 0 0 1>}
 Squash {type=Curve, loc=<0 0 0>, rot=<0 0 0 1>}
 polySurface151 {type=Curve, loc=<0 0 0>, rot=<0 0 0 1>}
 l_cheek_puff {type=Curve, loc=<0 0 0>, rot=<0 0 0 1>}
 r_cheek_puff {type=Curve, loc=<0 0 0>, rot=<0 0 0 1>}
 l_smile_fix {type=Curve, loc=<0 0 0>, rot=<0 0 0 1>}
 r_smile_fix {type=Curve, loc=<0 0 0>, rot=<0 0 0 1>}
 jaw_open_fix {type=Curve, loc=<0 0 0>, rot=<0 0 0 1>}
 r_lip_sneer_fix {type=Curve, loc=<0 0 0>, rot=<0 0 0 1>}
 l_lip_sneer_fix {type=Curve, loc=<0 0 0>, rot=<0 0 0 1>}
 L_Blink {type=Curve, loc=<0 0 0>, rot=<0 0 0 1>}
 R_Blink {type=Curve, loc=<0 0 0>, rot=<0 0 0 1>}
 L_Arm_Stretch {type=Curve, loc=<0 0 0>, rot=<0 0 0 1>}
 Hair_Hat {type=Curve, loc=<0 0 0>, rot=<0 0 0 1>}
 Hair_Helmet {type=Curve, loc=<0 0 0>, rot=<0 0 0 1>}
 Male_Med_Head_02_What {type=Curve, loc=<0 0 0>, rot=<0 0 0 1>}
 Male_Med_Head_02_Smile {type=Curve, loc=<0 0 0>, rot=<0 0 0 1>}
 Male_Med_Head_02_Scowl {type=Curve, loc=<0 0 0>, rot=<0 0 0 1>}
 Male_Med_Head_02_Hmmph {type=Curve, loc=<0 0 0>, rot=<0 0 0 1>}
 Male_Med_Head_02_Grrr {type=Curve, loc=<0 0 0>, rot=<0 0 0 1>}
 Male_Med_Head_02_Effort {type=Curve, loc=<0 0 0>, rot=<0 0 0 1>}
 Male_Med_Head_03_What {type=Curve, loc=<0 0 0>, rot=<0 0 0 1>}
 Male_Med_Head_03_Grrr {type=Curve, loc=<0 0 0>, rot=<0 0 0 1>}
 Male_Med_Head_03_Effort {type=Curve, loc=<0 0 0>, rot=<0 0 0 1>}
 Male_Med_Head_03_Hmmph {type=Curve, loc=<0 0 0>, rot=<0 0 0 1>}
 Male_Med_Head_03_Scowl {type=Curve, loc=<0 0 0>, rot=<0 0 0 1>}
 Male_Med_Head_03_Smile {type=Curve, loc=<0 0 0>, rot=<0 0 0 1>}
 Male_Med_Head_05_Teeth_Low_What {type=Curve, loc=<0 0 0>, rot=<0 0 0 1>}
 Male_Med_Head_05_Teeth_Low_Pout {type=Curve, loc=<0 0 0>, rot=<0 0 0 1>}
 Male_Med_Head_05_Teeth_Low_Hmmph {type=Curve, loc=<0 0 0>, rot=<0 0 0 1>}
 Male_Med_Head_05_Teeth_Low_Grrr {type=Curve, loc=<0 0 0>, rot=<0 0 0 1>}
 Male_Med_Head_05_Teeth_Low_Effort {type=Curve, loc=<0 0 0>, rot=<0 0 0 1>}
 Male_Med_Head_05_What {type=Curve, loc=<0 0 0>, rot=<0 0 0 1>}
 Male_Med_Head_05_Grrr {type=Curve, loc=<0 0 0>, rot=<0 0 0 1>}
 Male_Med_Head_05_Hmmph {type=Curve, loc=<0 0 0>, rot=<0 0 0 1>}
 Male_Med_Head_05_Pout {type=Curve, loc=<0 0 0>, rot=<0 0 0 1>}
 Male_Med_Head_05_Tongue_What {type=Curve, loc=<0 0 0>, rot=<0 0 0 1>}
 Male_Med_Head_05_Tongue_Pout {type=Curve, loc=<0 0 0>, rot=<0 0 0 1>}
 Male_Med_Head_05_Tongue_Hmmph {type=Curve, loc=<0 0 0>, rot=<0 0 0 1>}
 Male_Med_Head_05_Tongue_Grrr {type=Curve, loc=<0 0 0>, rot=<0 0 0 1>}
 Male_Med_Head_05_Tongue_Effort {type=Curve, loc=<0 0 0>, rot=<0 0 0 1>}
 AdditiveLayer {type=Curve, loc=<0 0 0>, rot=<0 0 0 1>}
 RightHandIK {type=Curve, loc=<0 0 0>, rot=<0 0 0 1>}
 LeftHandIK {type=Curve, loc=<0 0 0>, rot=<0 0 0 1>}
 LeftHandFK {type=Curve, loc=<0 0 0>, rot=<0 0 0 1>}
 Male_Med_Head_01_Grrr {type=Curve, loc=<0 0 0>, rot=<0 0 0 1>}
 WeaponAiming {type=Curve, loc=<0 0 0>, rot=<0 0 0 1>}
 Male_Med_Head_01_Hmmph {type=Curve, loc=<0 0 0>, rot=<0 0 0 1>}
 DragonSlash {type=Curve, loc=<0 0 0>, rot=<0 0 0 1>}
 Male_Med_Head_01_Effort {type=Curve, loc=<0 0 0>, rot=<0 0 0 1>}
 LeftHankFK {type=Curve, loc=<0 0 0>, rot=<0 0 0 1>}
 blendOrient1 {type=Curve, loc=<0 0 0>, rot=<0 0 0 1>}
 blendParent1 {type=Curve, loc=<0 0 0>, rot=<0 0 0 1>}
 bottom_R_Corner_Down {type=Curve, loc=<0 0 0>, rot=<0 0 0 1>}
 bottom_R_Corner_Up {type=Curve, loc=<0 0 0>, rot=<0 0 0 1>}
 top_L_Corner_Up {type=Curve, loc=<0 0 0>, rot=<0 0 0 1>}
 top_L_Corner_Down {type=Curve, loc=<0 0 0>, rot=<0 0 0 1>}
 top_R_Corner_Down {type=Curve, loc=<0 0 0>, rot=<0 0 0 1>}
 top_R_Corner_Up {type=Curve, loc=<0 0 0>, rot=<0 0 0 1>}
 bottom_L_Corner_Down {type=Curve, loc=<0 0 0>, rot=<0 0 0 1>}
 bottomt_L_Corner_Up {type=Curve, loc=<0 0 0>, rot=<0 0 0 1>}
 TrailCurve1 {type=Curve, loc=<0 0 0>, rot=<0 0 0 1>}
 BothHandFK {type=Curve, loc=<0 0 0>, rot=<0 0 0 1>}
 bottom_R_Corner_Down1 {type=Curve, loc=<0 0 0>, rot=<0 0 0 1>}
 blendPoint1 {type=Curve, loc=<0 0 0>, rot=<0 0 0 1>}
 WeaponAimingFreeze {type=Curve, loc=<0 0 0>, rot=<0 0 0 1>}
 l_calf_top_taper {type=Curve, loc=<0 0 0>, rot=<0 0 0 1>}
 JogBlendSpaceRotation {type=Curve, loc=<0 0 0>, rot=<0 0 0 1>}
 ZeroOutPitchWeight {type=Curve, loc=<0 0 0>, rot=<0 0 0 1>}
 MeleeTwist {type=Curve, loc=<0 0 0>, rot=<0 0 0 1>}
 MeleeAnim {type=Curve, loc=<0 0 0>, rot=<0 0 0 1>}
 l_sneer_fix {type=Curve, loc=<0 0 0>, rot=<0 0 0 1>}
 jaw_fix {type=Curve, loc=<0 0 0>, rot=<0 0 0 1>}
 r_sneer_fix {type=Curve, loc=<0 0 0>, rot=<0 0 0 1>}
 TurnRotationAmount {type=Curve, loc=<0 0 0>, rot=<0 0 0 1>}
 TurnEnd {type=Curve, loc=<0 0 0>, rot=<0 0 0 1>}
 BlendIn {type=Curve, loc=<0 0 0>, rot=<0 0 0 1>}
 BlendOut {type=Curve, loc=<0 0 0>, rot=<0 0 0 1>}
 BodyRotateX {type=Curve, loc=<0 0 0>, rot=<0 0 0 1>}
 BodyTranslateX {type=Curve, loc=<0 0 0>, rot=<0 0 0 1>}
 CrescentKickScale {type=Curve, loc=<0 0 0>, rot=<0 0 0 1>}
 mask_geo {type=Curve, loc=<0 0 0>, rot=<0 0 0 1>}
 RemoveJumpAdditive {type=Curve, loc=<0 0 0>, rot=<0 0 0 1>}
 RightLeg_IK {type=Curve, loc=<0 0 0>, rot=<0 0 0 1>}
 SnapWeapon_LHand {type=Curve, loc=<0 0 0>, rot=<0 0 0 1>}
 PivotAnim {type=Curve, loc=<0 0 0>, rot=<0 0 0 1>}
 DefaultAdditiveStateWeight {type=Curve, loc=<0 0 0>, rot=<0 0 0 1>}
 StartAnim {type=Curve, loc=<0 0 0>, rot=<0 0 0 1>}
 StopAnim {type=Curve, loc=<0 0 0>, rot=<0 0 0 1>}
 DisableSpeedWarping {type=Curve, loc=<0 0 0>, rot=<0 0 0 1>}
 DisableDBNOSlopeWarp {type=Curve, loc=<0 0 0>, rot=<0 0 0 1>}
 MaskRightArm {type=Curve, loc=<0 0 0>, rot=<0 0 0 1>}
 TurnInPlaceAnim {type=Curve, loc=<0 0 0>, rot=<0 0 0 1>}
 Head_MESH {type=Curve, loc=<0 0 0>, rot=<0 0 0 1>}
 Hair_MESH {type=Curve, loc=<0 0 0>, rot=<0 0 0 1>}
 LaunchPadTransitionOff {type=Curve, loc=<0 0 0>, rot=<0 0 0 1>}
 C_glabella_up_pose {type=Curve, loc=<0 0 0>, rot=<0 0 0 1>}
 jaw_open_pose {type=Curve, loc=<0 0 0>, rot=<0 0 0 1>}
 L_blink_pose {type=Curve, loc=<0 0 0>, rot=<0 0 0 1>}
 L_brow_down_pose {type=Curve, loc=<0 0 0>, rot=<0 0 0 1>}
 L_brow_squeeze_in_pose {type=Curve, loc=<0 0 0>, rot=<0 0 0 1>}
 L_brow_up_pose {type=Curve, loc=<0 0 0>, rot=<0 0 0 1>}
 L_frown_pose {type=Curve, loc=<0 0 0>, rot=<0 0 0 1>}
 L_lip_corner_narrow_pose {type=Curve, loc=<0 0 0>, rot=<0 0 0 1>}
 L_squint_inner_pose {type=Curve, loc=<0 0 0>, rot=<0 0 0 1>}
 L_wide_pose {type=Curve, loc=<0 0 0>, rot=<0 0 0 1>}
 look_down_pose {type=Curve, loc=<0 0 0>, rot=<0 0 0 1>}
 look_left_pose {type=Curve, loc=<0 0 0>, rot=<0 0 0 1>}
 look_right_pose {type=Curve, loc=<0 0 0>, rot=<0 0 0 1>}
 look_up_pose {type=Curve, loc=<0 0 0>, rot=<0 0 0 1>}
 phoneme_fv_pose {type=Curve, loc=<0 0 0>, rot=<0 0 0 1>}
 phoneme_oo_pose {type=Curve, loc=<0 0 0>, rot=<0 0 0 1>}
 R_blink_pose {type=Curve, loc=<0 0 0>, rot=<0 0 0 1>}
 R_brow_down_pose {type=Curve, loc=<0 0 0>, rot=<0 0 0 1>}
 R_brow_squeeze_in_pose {type=Curve, loc=<0 0 0>, rot=<0 0 0 1>}
 R_brow_up_pose {type=Curve, loc=<0 0 0>, rot=<0 0 0 1>}
 R_frown_pose {type=Curve, loc=<0 0 0>, rot=<0 0 0 1>}
 R_lip_corner_narrow_pose {type=Curve, loc=<0 0 0>, rot=<0 0 0 1>}
 R_squint_inner_pose {type=Curve, loc=<0 0 0>, rot=<0 0 0 1>}
 R_wide_pose {type=Curve, loc=<0 0 0>, rot=<0 0 0 1>}
 DisableArmsHeadAdditive {type=Curve, loc=<0 0 0>, rot=<0 0 0 1>}
 FrontendRotation {type=Curve, loc=<0 0 0>, rot=<0 0 0 1>}
 BothLegFK {type=Curve, loc=<0 0 0>, rot=<0 0 0 1>}
 AllowTurnInPlace {type=Curve, loc=<0 0 0>, rot=<0 0 0 1>}
 MontageTimeStretchCurve {type=Curve, loc=<0 0 0>, rot=<0 0 0 1>}
 ParaGlideLeanAlpha {type=Curve, loc=<0 0 0>, rot=<0 0 0 1>}
 WeaponRInChestSpace {type=Curve, loc=<0 0 0>, rot=<0 0 0 1>}
 JimBeam_AO {type=Curve, loc=<0 0 0>, rot=<0 0 0 1>}
 R_squeeze_pose {type=Curve, loc=<0 0 0>, rot=<0 0 0 1>}
 L_squeeze_pose {type=Curve, loc=<0 0 0>, rot=<0 0 0 1>}
 R_upper_lip_raiser_pose {type=Curve, loc=<0 0 0>, rot=<0 0 0 1>}
 L_cheek_up_pose {type=Curve, loc=<0 0 0>, rot=<0 0 0 1>}
 R_cheek_up_pose {type=Curve, loc=<0 0 0>, rot=<0 0 0 1>}
 DisableFaceOverride {type=Curve, loc=<0 0 0>, rot=<0 0 0 1>}
 EnableFullBodyOverride {type=Curve, loc=<0 0 0>, rot=<0 0 0 1>}
 C_glabella_down_pose {type=Curve, loc=<0 0 0>, rot=<0 0 0 1>}
 C_nose_wrinkler_pose {type=Curve, loc=<0 0 0>, rot=<0 0 0 1>}
 eye_convergence_pose {type=Curve, loc=<0 0 0>, rot=<0 0 0 1>}
 jaw_back_pose {type=Curve, loc=<0 0 0>, rot=<0 0 0 1>}
 jaw_forward_pose {type=Curve, loc=<0 0 0>, rot=<0 0 0 1>}
 jaw_left_pose {type=Curve, loc=<0 0 0>, rot=<0 0 0 1>}
 jaw_right_pose {type=Curve, loc=<0 0 0>, rot=<0 0 0 1>}
 jaw_up_pose {type=Curve, loc=<0 0 0>, rot=<0 0 0 1>}
 L_lip_corner_wide_pose {type=Curve, loc=<0 0 0>, rot=<0 0 0 1>}
 L_lower_lip_down_pose {type=Curve, loc=<0 0 0>, rot=<0 0 0 1>}
 L_lower_lip_up_pose {type=Curve, loc=<0 0 0>, rot=<0 0 0 1>}
 L_smile_pose {type=Curve, loc=<0 0 0>, rot=<0 0 0 1>}
 L_upper_lip_lower_pose {type=Curve, loc=<0 0 0>, rot=<0 0 0 1>}
 L_upper_lip_raiser_pose {type=Curve, loc=<0 0 0>, rot=<0 0 0 1>}
 L_upperLidDown_pose {type=Curve, loc=<0 0 0>, rot=<0 0 0 1>}
 mouth_down_pose {type=Curve, loc=<0 0 0>, rot=<0 0 0 1>}
 mouth_up_pose {type=Curve, loc=<0 0 0>, rot=<0 0 0 1>}
 phoneme_ch_pose {type=Curve, loc=<0 0 0>, rot=<0 0 0 1>}
 phoneme_mbp_pose {type=Curve, loc=<0 0 0>, rot=<0 0 0 1>}
 R_lip_corner_wide_pose {type=Curve, loc=<0 0 0>, rot=<0 0 0 1>}
 R_lower_lip_down_pose {type=Curve, loc=<0 0 0>, rot=<0 0 0 1>}
 R_lower_lip_up_pose {type=Curve, loc=<0 0 0>, rot=<0 0 0 1>}
 R_smile_pose {type=Curve, loc=<0 0 0>, rot=<0 0 0 1>}
 R_upper_lip_lower_pose {type=Curve, loc=<0 0 0>, rot=<0 0 0 1>}
 R_upperLidDown_pose {type=Curve, loc=<0 0 0>, rot=<0 0 0 1>}
 tongue_back_pose {type=Curve, loc=<0 0 0>, rot=<0 0 0 1>}
 tongue_down_pose {type=Curve, loc=<0 0 0>, rot=<0 0 0 1>}
 tongue_forward_pose {type=Curve, loc=<0 0 0>, rot=<0 0 0 1>}
 tongue_up_pose {type=Curve, loc=<0 0 0>, rot=<0 0 0 1>}
 LiveLink {type=Curve, loc=<0 0 0>, rot=<0 0 0 1>}
 ParaGlideDeploy {type=Curve, loc=<0 0 0>, rot=<0 0 0 1>}
 HideLeftWeapon {type=Curve, loc=<0 0 0>, rot=<0 0 0 1>}
 DisableLeftHandIKGolfCart {type=Curve, loc=<0 0 0>, rot=<0 0 0 1>}
 DisableRightHandIKGolfCart {type=Curve, loc=<0 0 0>, rot=<0 0 0 1>}
 DisableHandIKRetargeting {type=Curve, loc=<0 0 0>, rot=<0 0 0 1>}
 DisableConsumableOffsetPose {type=Curve, loc=<0 0 0>, rot=<0 0 0 1>}
 UpperBodyNoAdditive {type=Curve, loc=<0 0 0>, rot=<0 0 0 1>}
 MaskLeftArm {type=Curve, loc=<0 0 0>, rot=<0 0 0 1>}
 ForceUpperBody {type=Curve, loc=<0 0 0>, rot=<0 0 0 1>}
 RightHandFK {type=Curve, loc=<0 0 0>, rot=<0 0 0 1>}
 ParaGlideLeanTransition {type=Curve, loc=<0 0 0>, rot=<0 0 0 1>}
 InMotionEmote {type=Curve, loc=<0 0 0>, rot=<0 0 0 1>}
 ForceFullBodyInFrontEnd {type=Curve, loc=<0 0 0>, rot=<0 0 0 1>}
 WeaponLinFloorSpace {type=Curve, loc=<0 0 0>, rot=<0 0 0 1>}
 FrontEnd_Idle_Bypass {type=Curve, loc=<0 0 0>, rot=<0 0 0 1>}
 DisableSlopeWarping {type=Curve, loc=<0 0 0>, rot=<0 0 0 1>}
 Disable_Custom {type=Curve, loc=<0 0 0>, rot=<0 0 0 1>}
 EmoteRequiresWeapon {type=Curve, loc=<0 0 0>, rot=<0 0 0 1>}
 DisableIKRootAdditive {type=Curve, loc=<0 0 0>, rot=<0 0 0 1>}
 DisableFullBodySprint {type=Curve, loc=<0 0 0>, rot=<0 0 0 1>}
 DetachFeet_Vehicle {type=Curve, loc=<0 0 0>, rot=<0 0 0 1>}
 IndigoBoost {type=Curve, loc=<0 0 0>, rot=<0 0 0 1>}
 DisableUpperBodyAdditiveMeshSpace {type=Curve, loc=<0 0 0>, rot=<0 0 0 1>}
 MeleeAOWeight {type=Curve, loc=<0 0 0>, rot=<0 0 0 1>}
 HandsInRootSpace {type=Curve, loc=<0 0 0>, rot=<0 0 0 1>}
 CrouchPelvisAdjustment {type=Curve, loc=<0 0 0>, rot=<0 0 0 1>}
 EnableAccelYawCorrection {type=Curve, loc=<0 0 0>, rot=<0 0 0 1>}
 ParaGlideModifyRoot {type=Curve, loc=<0 0 0>, rot=<0 0 0 1>}
 UseScalarParameterForBlink {type=Curve, loc=<0 0 0>, rot=<0 0 0 1>}
 BlinkControl {type=Curve, loc=<0 0 0>, rot=<0 0 0 1>}
 ForceTargeting {type=Curve, loc=<0 0 0>, rot=<0 0 0 1>}
 BlinkControlProp {type=Curve, loc=<0 0 0>, rot=<0 0 0 1>}
 UseScalarParameterForBlinkProp {type=Curve, loc=<0 0 0>, rot=<0 0 0 1>}
 CrackshotCozyHeadCorrection {type=Curve, loc=<0 0 0>, rot=<0 0 0 1>}
 ForwardMotion {type=Curve, loc=<0 0 0>, rot=<0 0 0 1>}
 EmoteForcedRootYaw {type=Curve, loc=<0 0 0>, rot=<0 0 0 1>}
 MeleeTrailWidth {type=Curve, loc=<0 0 0>, rot=<0 0 0 1>}
 space_world {type=Curve, loc=<0 0 0>, rot=<0 0 0 1>}
 PeelyBlenderFace {type=Curve, loc=<0 0 0>, rot=<0 0 0 1>}
 Width1 {type=Curve, loc=<0 0 0>, rot=<0 0 0 1>}
 OverrideIKHands {type=Curve, loc=<0 0 0>, rot=<0 0 0 1>}
 ParaGlideAddAlpha {type=Curve, loc=<0 0 0>, rot=<0 0 0 1>}
 ParaGlideTurn {type=Curve, loc=<0 0 0>, rot=<0 0 0 1>}
 DisablePelvisAdjustment {type=Curve, loc=<0 0 0>, rot=<0 0 0 1>}
 UncleBrollyDeploy {type=Curve, loc=<0 0 0>, rot=<0 0 0 1>}
 ForceRootSlopeOrientation {type=Curve, loc=<0 0 0>, rot=<0 0 0 1>}
 DisableAdditiveOffsetPose {type=Curve, loc=<0 0 0>, rot=<0 0 0 1>}
 DisableIdleNoiseAdditive {type=Curve, loc=<0 0 0>, rot=<0 0 0 1>}
 UsingConsumable {type=Curve, loc=<0 0 0>, rot=<0 0 0 1>}
 DisableRightArmAdditive {type=Curve, loc=<0 0 0>, rot=<0 0 0 1>}
 NoGlider {type=Curve, loc=<0 0 0>, rot=<0 0 0 1>}
 ParaGlideModifyPelvis {type=Curve, loc=<0 0 0>, rot=<0 0 0 1>}
 ParaGlideModifyLegs {type=Curve, loc=<0 0 0>, rot=<0 0 0 1>}
 HT_L1_TransitionCurve {type=Curve, loc=<0 0 0>, rot=<0 0 0 1>}
 KeepPelvisAdjustment {type=Curve, loc=<0 0 0>, rot=<0 0 0 1>}
 PatrolDisableIdleDuringWalk {type=Curve, loc=<0 0 0>, rot=<0 0 0 1>}
 PatrolWalkUseFullbodyBlendspace {type=Curve, loc=<0 0 0>, rot=<0 0 0 1>}
 RemoveJumpUpAdditive {type=Curve, loc=<0 0 0>, rot=<0 0 0 1>}
 RemoveJumpLoopAdditive {type=Curve, loc=<0 0 0>, rot=<0 0 0 1>}
 RemoveJumpFallAdditive {type=Curve, loc=<0 0 0>, rot=<0 0 0 1>}
 JH_TransformCurve {type=Curve, loc=<0 0 0>, rot=<0 0 0 1>}
 PhysicsOff {type=Curve, loc=<0 0 0>, rot=<0 0 0 1>}
 DisableWeaponAdditivePose {type=Curve, loc=<0 0 0>, rot=<0 0 0 1>}
 Lexa_TransformationCurve {type=Curve, loc=<0 0 0>, rot=<0 0 0 1>}
 HideProp {type=Curve, loc=<0 0 0>, rot=<0 0 0 1>}
 CopyPose {type=Curve, loc=<0 0 0>, rot=<0 0 0 1>}
 MaterialCurveA {type=Curve, loc=<0 0 0>, rot=<0 0 0 1>}
 MaterialCurveB {type=Curve, loc=<0 0 0>, rot=<0 0 0 1>}
 MaterialCurveC {type=Curve, loc=<0 0 0>, rot=<0 0 0 1>}
 MaterialCurveD {type=Curve, loc=<0 0 0>, rot=<0 0 0 1>}
 MaterialCurveE {type=Curve, loc=<0 0 0>, rot=<0 0 0 1>}
 UseEmoteAnimOffset {type=Curve, loc=<0 0 0>, rot=<0 0 0 1>}
 emissive mult 2 {type=Curve, loc=<0 0 0>, rot=<0 0 0 1>}
 MaxHandle {type=Curve, loc=<0 0 0>, rot=<0 0 0 1>}
 FootPhase {type=Curve, loc=<0 0 0>, rot=<0 0 0 1>}
 TacticalDoorBash {type=Curve, loc=<0 0 0>, rot=<0 0 0 1>}
 Custom_Frontend_Face_Idle {type=Curve, loc=<0 0 0>, rot=<0 0 0 1>}
 HandIKInAttachBoneSpace {type=Curve, loc=<0 0 0>, rot=<0 0 0 1>}
 SwingPhase {type=Curve, loc=<0 0 0>, rot=<0 0 0 1>}
 Moving {type=Curve, loc=<0 0 0>, rot=<0 0 0 1>}
 HandAttachRInChestSpace {type=Curve, loc=<0 0 0>, rot=<0 0 0 1>}
 emote_BuiltIn {type=Curve, loc=<0 0 0>, rot=<0 0 0 1>}
 OhanaPackAdj {type=Curve, loc=<0 0 0>, rot=<0 0 0 1>}
 UseRootSlopeOrientation {type=Curve, loc=<0 0 0>, rot=<0 0 0 1>}
 Backpack_Anim_Override {type=Curve, loc=<0 0 0>, rot=<0 0 0 1>}
 PhysicsGroundOffset {type=Curve, loc=<0 0 0>, rot=<0 0 0 1>}
 AllowRidingNoiseAdditive {type=Curve, loc=<0 0 0>, rot=<0 0 0 1>}
 LeftLegIKOnGroundVehicle {type=Curve, loc=<0 0 0>, rot=<0 0 0 1>}
 RightLegIKOnGroundVehicle {type=Curve, loc=<0 0 0>, rot=<0 0 0 1>}
 BaseHeadDefaultFaceIdle {type=Curve, loc=<0 0 0>, rot=<0 0 0 1>}
 Hurdle_100fwd {type=Curve, loc=<0 0 0>, rot=<0 0 0 1>}
 Hurdle_250fwd {type=Curve, loc=<0 0 0>, rot=<0 0 0 1>}
 Hurdle_100up {type=Curve, loc=<0 0 0>, rot=<0 0 0 1>}
 Hurdle_200up {type=Curve, loc=<0 0 0>, rot=<0 0 0 1>}
 Hurdle_50up {type=Curve, loc=<0 0 0>, rot=<0 0 0 1>}
 Hurdle_Mantle {type=Curve, loc=<0 0 0>, rot=<0 0 0 1>}
 Idle {type=Curve, loc=<0 0 0>, rot=<0 0 0 1>}
 Interruptible {type=Curve, loc=<0 0 0>, rot=<0 0 0 1>}
 CustomFaceIdle {type=Curve, loc=<0 0 0>, rot=<0 0 0 1>}
 Hurdle_Land {type=Curve, loc=<0 0 0>, rot=<0 0 0 1>}
 Hurdle_Fall {type=Curve, loc=<0 0 0>, rot=<0 0 0 1>}
 TacticalJump {type=Curve, loc=<0 0 0>, rot=<0 0 0 1>}
 copyWeaponRtoAttachL {type=Curve, loc=<0 0 0>, rot=<0 0 0 1>}
 CTRL_expressions_browDownL {type=Curve, loc=<0 0 0>, rot=<0 0 0 1>}
 CTRL_expressions_browDownR {type=Curve, loc=<0 0 0>, rot=<0 0 0 1>}
 CTRL_expressions_browLateralL {type=Curve, loc=<0 0 0>, rot=<0 0 0 1>}
 CTRL_expressions_browLateralR {type=Curve, loc=<0 0 0>, rot=<0 0 0 1>}
 CTRL_expressions_browRaiseInL {type=Curve, loc=<0 0 0>, rot=<0 0 0 1>}
 CTRL_expressions_browRaiseInR {type=Curve, loc=<0 0 0>, rot=<0 0 0 1>}
 CTRL_expressions_browRaiseOuterL {type=Curve, loc=<0 0 0>, rot=<0 0 0 1>}
 CTRL_expressions_browRaiseOuterR {type=Curve, loc=<0 0 0>, rot=<0 0 0 1>}
 CTRL_expressions_browUINL {type=Curve, loc=<0 0 0>, rot=<0 0 0 1>}
 CTRL_expressions_browUINR {type=Curve, loc=<0 0 0>, rot=<0 0 0 1>}
 CTRL_expressions_browUOUTL {type=Curve, loc=<0 0 0>, rot=<0 0 0 1>}
 CTRL_expressions_browUOUTR {type=Curve, loc=<0 0 0>, rot=<0 0 0 1>}
 CTRL_expressions_browDINL {type=Curve, loc=<0 0 0>, rot=<0 0 0 1>}
 CTRL_expressions_browDINR {type=Curve, loc=<0 0 0>, rot=<0 0 0 1>}
 CTRL_expressions_browDOUTL {type=Curve, loc=<0 0 0>, rot=<0 0 0 1>}
 CTRL_expressions_browDOUTR {type=Curve, loc=<0 0 0>, rot=<0 0 0 1>}
 CTRL_expressions_eyeBlinkL {type=Curve, loc=<0 0 0>, rot=<0 0 0 1>}
 CTRL_expressions_eyeBlinkR {type=Curve, loc=<0 0 0>, rot=<0 0 0 1>}
 CTRL_expressions_eyeWidenL {type=Curve, loc=<0 0 0>, rot=<0 0 0 1>}
 CTRL_expressions_eyeWidenR {type=Curve, loc=<0 0 0>, rot=<0 0 0 1>}
 CTRL_expressions_eyeSquintInnerL {type=Curve, loc=<0 0 0>, rot=<0 0 0 1>}
 CTRL_expressions_eyeSquintInnerR {type=Curve, loc=<0 0 0>, rot=<0 0 0 1>}
 CTRL_expressions_eyeCheekRaiseL {type=Curve, loc=<0 0 0>, rot=<0 0 0 1>}
 CTRL_expressions_eyeCheekRaiseR {type=Curve, loc=<0 0 0>, rot=<0 0 0 1>}
 CTRL_expressions_eyeUpperLidUpL {type=Curve, loc=<0 0 0>, rot=<0 0 0 1>}
 CTRL_expressions_eyeUpperLidUpR {type=Curve, loc=<0 0 0>, rot=<0 0 0 1>}
 CTRL_expressions_eyeRelaxL {type=Curve, loc=<0 0 0>, rot=<0 0 0 1>}
 CTRL_expressions_eyeRelaxR {type=Curve, loc=<0 0 0>, rot=<0 0 0 1>}
 CTRL_expressions_eyeLowerLidUpL {type=Curve, loc=<0 0 0>, rot=<0 0 0 1>}
 CTRL_expressions_eyeLowerLidUpR {type=Curve, loc=<0 0 0>, rot=<0 0 0 1>}
 CTRL_expressions_eyeLowerLidDownL {type=Curve, loc=<0 0 0>, rot=<0 0 0 1>}
 CTRL_expressions_eyeLowerLidDownR {type=Curve, loc=<0 0 0>, rot=<0 0 0 1>}
 CTRL_expressions_eyelidRollPosUL {type=Curve, loc=<0 0 0>, rot=<0 0 0 1>}
 CTRL_expressions_eyelidRollPosUR {type=Curve, loc=<0 0 0>, rot=<0 0 0 1>}
 CTRL_expressions_eyelidRollPosDL {type=Curve, loc=<0 0 0>, rot=<0 0 0 1>}
 CTRL_expressions_eyelidRollPosDR {type=Curve, loc=<0 0 0>, rot=<0 0 0 1>}
 CTRL_expressions_eyelidRollNegUL {type=Curve, loc=<0 0 0>, rot=<0 0 0 1>}
 CTRL_expressions_eyelidRollNegUR {type=Curve, loc=<0 0 0>, rot=<0 0 0 1>}
 CTRL_expressions_eyelidRollNegDL {type=Curve, loc=<0 0 0>, rot=<0 0 0 1>}
 CTRL_expressions_eyelidRollNegDR {type=Curve, loc=<0 0 0>, rot=<0 0 0 1>}
 CTRL_expressions_eyeLookUpL {type=Curve, loc=<0 0 0>, rot=<0 0 0 1>}
 CTRL_expressions_eyeLookUpR {type=Curve, loc=<0 0 0>, rot=<0 0 0 1>}
 CTRL_expressions_eyeLookDownL {type=Curve, loc=<0 0 0>, rot=<0 0 0 1>}
 CTRL_expressions_eyeLookDownR {type=Curve, loc=<0 0 0>, rot=<0 0 0 1>}
 CTRL_expressions_eyeLookLeftL {type=Curve, loc=<0 0 0>, rot=<0 0 0 1>}
 CTRL_expressions_eyeLookLeftR {type=Curve, loc=<0 0 0>, rot=<0 0 0 1>}
 CTRL_expressions_eyeLookRightL {type=Curve, loc=<0 0 0>, rot=<0 0 0 1>}
 CTRL_expressions_eyeLookRightR {type=Curve, loc=<0 0 0>, rot=<0 0 0 1>}
 CTRL_expressions_eyeEyelidsReduce {type=Curve, loc=<0 0 0>, rot=<0 0 0 1>}
 CTRL_expressions_noseWrinkleL {type=Curve, loc=<0 0 0>, rot=<0 0 0 1>}
 CTRL_expressions_noseWrinkleR {type=Curve, loc=<0 0 0>, rot=<0 0 0 1>}
 CTRL_expressions_noseWrinkleUpperL {type=Curve, loc=<0 0 0>, rot=<0 0 0 1>}
 CTRL_expressions_noseWrinkleUpperR {type=Curve, loc=<0 0 0>, rot=<0 0 0 1>}
 CTRL_expressions_noseNostrilDepressL {type=Curve, loc=<0 0 0>, rot=<0 0 0 1>}
 CTRL_expressions_noseNostrilDepressR {type=Curve, loc=<0 0 0>, rot=<0 0 0 1>}
 CTRL_expressions_noseNostrilDilateL {type=Curve, loc=<0 0 0>, rot=<0 0 0 1>}
 CTRL_expressions_noseNostrilDilateR {type=Curve, loc=<0 0 0>, rot=<0 0 0 1>}
 CTRL_expressions_noseNostrilCompressL {type=Curve, loc=<0 0 0>, rot=<0 0 0 1>}
 CTRL_expressions_noseNostrilCompressR {type=Curve, loc=<0 0 0>, rot=<0 0 0 1>}
 CTRL_expressions_noseRotateL {type=Curve, loc=<0 0 0>, rot=<0 0 0 1>}
 CTRL_expressions_noseRotateR {type=Curve, loc=<0 0 0>, rot=<0 0 0 1>}
 CTRL_expressions_noseSquash {type=Curve, loc=<0 0 0>, rot=<0 0 0 1>}
 CTRL_expressions_noseStretch {type=Curve, loc=<0 0 0>, rot=<0 0 0 1>}
 CTRL_expressions_mouthCheekSuckL {type=Curve, loc=<0 0 0>, rot=<0 0 0 1>}
 CTRL_expressions_mouthCheekSuckR {type=Curve, loc=<0 0 0>, rot=<0 0 0 1>}
 CTRL_expressions_mouthCheekBlowL {type=Curve, loc=<0 0 0>, rot=<0 0 0 1>}
 CTRL_expressions_mouthCheekBlowR {type=Curve, loc=<0 0 0>, rot=<0 0 0 1>}
 CTRL_expressions_mouthLipsBlowL {type=Curve, loc=<0 0 0>, rot=<0 0 0 1>}
 CTRL_expressions_mouthLipsBlowR {type=Curve, loc=<0 0 0>, rot=<0 0 0 1>}
 CTRL_expressions_mouthLeft {type=Curve, loc=<0 0 0>, rot=<0 0 0 1>}
 CTRL_expressions_mouthRight {type=Curve, loc=<0 0 0>, rot=<0 0 0 1>}
 CTRL_expressions_mouthUp {type=Curve, loc=<0 0 0>, rot=<0 0 0 1>}
 CTRL_expressions_mouthDown {type=Curve, loc=<0 0 0>, rot=<0 0 0 1>}
 CTRL_expressions_mouthUpperLipRaiseL {type=Curve, loc=<0 0 0>, rot=<0 0 0 1>}
 CTRL_expressions_mouthUpperLipRaiseR {type=Curve, loc=<0 0 0>, rot=<0 0 0 1>}
 CTRL_expressions_mouthLowerLipDepressL {type=Curve, loc=<0 0 0>, rot=<0 0 0 1>}
 CTRL_expressions_mouthLowerLipDepressR {type=Curve, loc=<0 0 0>, rot=<0 0 0 1>}
 CTRL_expressions_mouthCornerPullL {type=Curve, loc=<0 0 0>, rot=<0 0 0 1>}
 CTRL_expressions_mouthCornerPullR {type=Curve, loc=<0 0 0>, rot=<0 0 0 1>}
 CTRL_expressions_mouthStretchL {type=Curve, loc=<0 0 0>, rot=<0 0 0 1>}
 CTRL_expressions_mouthStretchR {type=Curve, loc=<0 0 0>, rot=<0 0 0 1>}
 CTRL_expressions_mouthDimpleL {type=Curve, loc=<0 0 0>, rot=<0 0 0 1>}
 CTRL_expressions_mouthDimpleR {type=Curve, loc=<0 0 0>, rot=<0 0 0 1>}
 CTRL_expressions_mouthCornerDepressL {type=Curve, loc=<0 0 0>, rot=<0 0 0 1>}
 CTRL_expressions_mouthCornerDepressR {type=Curve, loc=<0 0 0>, rot=<0 0 0 1>}
 CTRL_expressions_mouthPressUL {type=Curve, loc=<0 0 0>, rot=<0 0 0 1>}
 CTRL_expressions_mouthPressUR {type=Curve, loc=<0 0 0>, rot=<0 0 0 1>}
 CTRL_expressions_mouthPressDL {type=Curve, loc=<0 0 0>, rot=<0 0 0 1>}
 CTRL_expressions_mouthPressDR {type=Curve, loc=<0 0 0>, rot=<0 0 0 1>}
 CTRL_expressions_mouthLipsPurseUL {type=Curve, loc=<0 0 0>, rot=<0 0 0 1>}
 CTRL_expressions_mouthLipsPurseUR {type=Curve, loc=<0 0 0>, rot=<0 0 0 1>}
 CTRL_expressions_mouthLipsPurseDL {type=Curve, loc=<0 0 0>, rot=<0 0 0 1>}
 CTRL_expressions_mouthLipsPurseDR {type=Curve, loc=<0 0 0>, rot=<0 0 0 1>}
 CTRL_expressions_mouthLipsTowardsUL {type=Curve, loc=<0 0 0>, rot=<0 0 0 1>}
 CTRL_expressions_mouthLipsTowardsUR {type=Curve, loc=<0 0 0>, rot=<0 0 0 1>}
 CTRL_expressions_mouthLipsTowardsDL {type=Curve, loc=<0 0 0>, rot=<0 0 0 1>}
 CTRL_expressions_mouthLipsTowardsDR {type=Curve, loc=<0 0 0>, rot=<0 0 0 1>}
 CTRL_expressions_mouthFunnelUL {type=Curve, loc=<0 0 0>, rot=<0 0 0 1>}
 CTRL_expressions_mouthFunnelUR {type=Curve, loc=<0 0 0>, rot=<0 0 0 1>}
 CTRL_expressions_mouthFunnelDL {type=Curve, loc=<0 0 0>, rot=<0 0 0 1>}
 CTRL_expressions_mouthFunnelDR {type=Curve, loc=<0 0 0>, rot=<0 0 0 1>}
 CTRL_expressions_mouthLipsTogetherUL {type=Curve, loc=<0 0 0>, rot=<0 0 0 1>}
 CTRL_expressions_mouthLipsTogetherUR {type=Curve, loc=<0 0 0>, rot=<0 0 0 1>}
 CTRL_expressions_mouthLipsTogetherDL {type=Curve, loc=<0 0 0>, rot=<0 0 0 1>}
 CTRL_expressions_mouthLipsTogetherDR {type=Curve, loc=<0 0 0>, rot=<0 0 0 1>}
 CTRL_expressions_mouthUpperLipBiteL {type=Curve, loc=<0 0 0>, rot=<0 0 0 1>}
 CTRL_expressions_mouthUpperLipBiteR {type=Curve, loc=<0 0 0>, rot=<0 0 0 1>}
 CTRL_expressions_mouthLowerLipBiteL {type=Curve, loc=<0 0 0>, rot=<0 0 0 1>}
 CTRL_expressions_mouthLowerLipBiteR {type=Curve, loc=<0 0 0>, rot=<0 0 0 1>}
 CTRL_expressions_mouthCornerSharpenUL {type=Curve, loc=<0 0 0>, rot=<0 0 0 1>}
 CTRL_expressions_mouthCornerSharpenUR {type=Curve, loc=<0 0 0>, rot=<0 0 0 1>}
 CTRL_expressions_mouthCornerSharpenDL {type=Curve, loc=<0 0 0>, rot=<0 0 0 1>}
 CTRL_expressions_mouthCornerSharpenDR {type=Curve, loc=<0 0 0>, rot=<0 0 0 1>}
 CTRL_expressions_mouthCornerRounderUL {type=Curve, loc=<0 0 0>, rot=<0 0 0 1>}
 CTRL_expressions_mouthCornerRounderUR {type=Curve, loc=<0 0 0>, rot=<0 0 0 1>}
 CTRL_expressions_mouthCornerRounderDL {type=Curve, loc=<0 0 0>, rot=<0 0 0 1>}
 CTRL_expressions_mouthCornerRounderDR {type=Curve, loc=<0 0 0>, rot=<0 0 0 1>}
 CTRL_expressions_mouthUpperLipTowardsTeethL {type=Curve, loc=<0 0 0>, rot=<0 0 0 1>}
 CTRL_expressions_mouthUpperLipTowardsTeethR {type=Curve, loc=<0 0 0>, rot=<0 0 0 1>}
 CTRL_expressions_mouthLowerLipTowardsTeethL {type=Curve, loc=<0 0 0>, rot=<0 0 0 1>}
 CTRL_expressions_mouthLowerLipTowardsTeethR {type=Curve, loc=<0 0 0>, rot=<0 0 0 1>}
 CTRL_expressions_mouthCornerUpL {type=Curve, loc=<0 0 0>, rot=<0 0 0 1>}
 CTRL_expressions_mouthCornerUpR {type=Curve, loc=<0 0 0>, rot=<0 0 0 1>}
 CTRL_expressions_mouthCornerDownL {type=Curve, loc=<0 0 0>, rot=<0 0 0 1>}
 CTRL_expressions_mouthCornerDownR {type=Curve, loc=<0 0 0>, rot=<0 0 0 1>}
 CTRL_expressions_mouthCornerWideL {type=Curve, loc=<0 0 0>, rot=<0 0 0 1>}
 CTRL_expressions_mouthCornerWideR {type=Curve, loc=<0 0 0>, rot=<0 0 0 1>}
 CTRL_expressions_mouthCornerNarrowL {type=Curve, loc=<0 0 0>, rot=<0 0 0 1>}
 CTRL_expressions_mouthCornerNarrowR {type=Curve, loc=<0 0 0>, rot=<0 0 0 1>}
 CTRL_expressions_tongueUp {type=Curve, loc=<0 0 0>, rot=<0 0 0 1>}
 CTRL_expressions_tongueDown {type=Curve, loc=<0 0 0>, rot=<0 0 0 1>}
 CTRL_expressions_tongueLeft {type=Curve, loc=<0 0 0>, rot=<0 0 0 1>}
 CTRL_expressions_tongueRight {type=Curve, loc=<0 0 0>, rot=<0 0 0 1>}
 CTRL_expressions_tongueOut {type=Curve, loc=<0 0 0>, rot=<0 0 0 1>}
 CTRL_expressions_tongueIn {type=Curve, loc=<0 0 0>, rot=<0 0 0 1>}
 CTRL_expressions_tongueRollUp {type=Curve, loc=<0 0 0>, rot=<0 0 0 1>}
 CTRL_expressions_tongueRollDown {type=Curve, loc=<0 0 0>, rot=<0 0 0 1>}
 CTRL_expressions_tongueRollLeft {type=Curve, loc=<0 0 0>, rot=<0 0 0 1>}
 CTRL_expressions_tongueRollRight {type=Curve, loc=<0 0 0>, rot=<0 0 0 1>}
 CTRL_expressions_tongueWide {type=Curve, loc=<0 0 0>, rot=<0 0 0 1>}
 CTRL_expressions_tongueNarrow {type=Curve, loc=<0 0 0>, rot=<0 0 0 1>}
 CTRL_expressions_jawOpen {type=Curve, loc=<0 0 0>, rot=<0 0 0 1>}
 CTRL_expressions_jawLeft {type=Curve, loc=<0 0 0>, rot=<0 0 0 1>}
 CTRL_expressions_jawRight {type=Curve, loc=<0 0 0>, rot=<0 0 0 1>}
 CTRL_expressions_jawFwd {type=Curve, loc=<0 0 0>, rot=<0 0 0 1>}
 CTRL_expressions_jawBack {type=Curve, loc=<0 0 0>, rot=<0 0 0 1>}
 CTRL_expressions_jawChinRaiseDL {type=Curve, loc=<0 0 0>, rot=<0 0 0 1>}
 CTRL_expressions_jawChinRaiseDR {type=Curve, loc=<0 0 0>, rot=<0 0 0 1>}
 CTRL_expressions_jawChinRaiseUL {type=Curve, loc=<0 0 0>, rot=<0 0 0 1>}
 CTRL_expressions_jawChinRaiseUR {type=Curve, loc=<0 0 0>, rot=<0 0 0 1>}
 CTRL_expressions_jawOpenExtreme {type=Curve, loc=<0 0 0>, rot=<0 0 0 1>}
 CTRL_expressions_teethUpU {type=Curve, loc=<0 0 0>, rot=<0 0 0 1>}
 CTRL_expressions_teethUpD {type=Curve, loc=<0 0 0>, rot=<0 0 0 1>}
 CTRL_expressions_teethDownU {type=Curve, loc=<0 0 0>, rot=<0 0 0 1>}
 CTRL_expressions_teethDownD {type=Curve, loc=<0 0 0>, rot=<0 0 0 1>}
 CTRL_expressions_teethLeftU {type=Curve, loc=<0 0 0>, rot=<0 0 0 1>}
 CTRL_expressions_teethLeftD {type=Curve, loc=<0 0 0>, rot=<0 0 0 1>}
 CTRL_expressions_teethRightU {type=Curve, loc=<0 0 0>, rot=<0 0 0 1>}
 CTRL_expressions_teethRightD {type=Curve, loc=<0 0 0>, rot=<0 0 0 1>}
 CTRL_expressions_teethFwdU {type=Curve, loc=<0 0 0>, rot=<0 0 0 1>}
 CTRL_expressions_teethFwdD {type=Curve, loc=<0 0 0>, rot=<0 0 0 1>}
 CTRL_expressions_teethBackU {type=Curve, loc=<0 0 0>, rot=<0 0 0 1>}
 CTRL_expressions_teethBackD {type=Curve, loc=<0 0 0>, rot=<0 0 0 1>}
 CTRL_expressions_teethRotateFwdU {type=Curve, loc=<0 0 0>, rot=<0 0 0 1>}
 CTRL_expressions_teethRotateFwdD {type=Curve, loc=<0 0 0>, rot=<0 0 0 1>}
 CTRL_expressions_teethRotateBackU {type=Curve, loc=<0 0 0>, rot=<0 0 0 1>}
 CTRL_expressions_teethRotateBackD {type=Curve, loc=<0 0 0>, rot=<0 0 0 1>}
 is_3L {type=Curve, loc=<0 0 0>, rot=<0 0 0 1>}
 Mirrored {type=Curve, loc=<0 0 0>, rot=<0 0 0 1>}
 LockMovement {type=Curve, loc=<0 0 0>, rot=<0 0 0 1>}
 Walk {type=Curve, loc=<0 0 0>, rot=<0 0 0 1>}
 Sprint {type=Curve, loc=<0 0 0>, rot=<0 0 0 1>}
 Jog {type=Curve, loc=<0 0 0>, rot=<0 0 0 1>}
 Run {type=Curve, loc=<0 0 0>, rot=<0 0 0 1>}
 LeftFootOnGrindRail {type=Curve, loc=<0 0 0>, rot=<0 0 0 1>}
 RightFootOnGrindRail {type=Curve, loc=<0 0 0>, rot=<0 0 0 1>}
 DisableLeftHandIKHipAttach {type=Curve, loc=<0 0 0>, rot=<0 0 0 1>}
 YawCorrectionDirectionSouth {type=Curve, loc=<0 0 0>, rot=<0 0 0 1>}
 YawCorrectionDirectionEast {type=Curve, loc=<0 0 0>, rot=<0 0 0 1>}
 YawCorrectionDirectionWest {type=Curve, loc=<0 0 0>, rot=<0 0 0 1>}
 YawCorrectionDirectionNorth {type=Curve, loc=<0 0 0>, rot=<0 0 0 1>}
 is_face_bone_anim {type=Curve, loc=<0 0 0>, rot=<0 0 0 1>}
 Hover {type=Curve, loc=<0 0 0>, rot=<0 0 0 1>}
 WeaponRinFloorSpace {type=Curve, loc=<0 0 0>, rot=<0 0 0 1>}
 AlphaBlur_R {type=Curve, loc=<0 0 0>, rot=<0 0 0 1>}
 AlphaBlur_L {type=Curve, loc=<0 0 0>, rot=<0 0 0 1>}
 RotationDirection_R {type=Curve, loc=<0 0 0>, rot=<0 0 0 1>}
 RotationDirection_L {type=Curve, loc=<0 0 0>, rot=<0 0 0 1>}
 PlayLowerBodyLocomotionOverFullBodySlot {type=Curve, loc=<0 0 0>, rot=<0 0 0 1>}
]
variables [
 ControlList {type="TArray<FRigElementKey>"}
 Do Twist Controls {type=bool, default=True}
 Gimbaled Controls {type="TArray<FRigElementKey>", default=((Type=Control,Name="upperarm_r_fk_ctrl"),(Type=Control,Name="upperarm_l_fk_ctrl"))}
 L Arm IK Align {type=bool, default=False}
 L Arm IK Mode {type=bool, default=False}
 "L Leg IK Mode " {type=bool, default=True}
 Neck IK Mode {type=bool, default=False}
 Neck Length {type=double, default=0}
 Pole Vector Distance Mult {type=double, default=0}
 R Arm IK Align {type=bool, default=False}
 R Arm IK Mode {type=bool, default=False}
 R Leg IK Mode {type=bool, default=True}
 Spine IK Mode {type=bool, default=False}
 Spine Length {type=double, default=0}
 Spine Offsets {type="TArray<FTransform>"}
 Use Pole Vector Distance {type=bool, default=False}
 arm_l_aimed {type="TArray<FRigElementKey>"}
 arm_r_aimed {type="TArray<FRigElementKey>"}
 calf_twist_params {type="TArray<double>", default=(0.25,0.5)}
 hand_l_attach_control_joints {type="TArray<FRigElementKey>", default=((Type=Bone,Name="hand_attach_l"),(Type=Bone,Name="weapon_l"))}
 hand_r_attach_control_joints {type="TArray<FRigElementKey>", default=((Type=Bone,Name="hand_attach_r"),(Type=Bone,Name="weapon_r"))}
 l_calf_twists {type="TArray<FRigElementKey>", default=((Type=Bone,Name="calf_twist_01_l"),(Type=Bone,Name="calf_twist_02_l"))}
 l_lowerarm_twists {type="TArray<FRigElementKey>", default=((Type=Bone,Name="lowerarm_twist_01_l"),(Type=Bone,Name="lowerarm_twist_02_l"))}
 l_thigh_twists {type="TArray<FRigElementKey>", default=((Type=Bone,Name="thigh_twist_01_l"))}
 l_upperarm_twists {type="TArray<FRigElementKey>", default=((Type=Bone,Name="upperarm_twist_01_l"),(Type=Bone,Name="upperarm_twist_02_l"))}
 leg_l_aimed {type="TArray<FRigElementKey>"}
 leg_r_aimed {type="TArray<FRigElementKey>"}
 lowerarm_twist_params {type="TArray<double>", default=(0.25,0.5)}
 r_calf_twists {type="TArray<FRigElementKey>", default=((Type=Bone,Name="calf_twist_01_r"),(Type=Bone,Name="calf_twist_02_r"))}
 r_lowerarm_twists {type="TArray<FRigElementKey>", default=((Type=Bone,Name="lowerarm_twist_01_r"),(Type=Bone,Name="lowerarm_twist_02_r"))}
 r_thigh_twists {type="TArray<FRigElementKey>", default=((Type=Bone,Name="thigh_twist_01_r"))}
 r_upperarm_twists {type="TArray<FRigElementKey>", default=((Type=Bone,Name="upperarm_twist_01_r"),(Type=Bone,Name="upperarm_twist_02_r"))}
 sides {type="TArray<FName>", default=("l","r")}
 spine_05_Transform {type=FTransform, default=(Rotation=(X=0,Y=0,Z=0,W=1),Translation=(X=0,Y=0,Z=0),Scale3D=(X=1,Y=1,Z=1))}
 thigh_twist_params {type="TArray<double>", default=(0)}
 upperarm_twist_params {type="TArray<double>", default=(0.1,0.4)}
]
