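# Control Rig for the player mannequin body, with FK and IK controls for limbs plus spine, neck, head and finger controls.
# Setup places controls from initial bone transforms, mirroring left-side limb placement to the right.
# Forward solve drives bones from controls, with rotation-space toggles for head and FK arms and an FK/IK switch that swaps visible arm controls.
# Backward solve snaps controls to bones using initial-offset projection.
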ControlRigBlueprint /Game/Characters/Player/Common/SharedRigs/Mannequin_Body_CtrlRig
graphs [
 RigVMFunctionLibrary {links [], nodes []}
 RigVMModel {nodes [
   n1 = /Script/ControlRig.RigUnit_InverseExecution() {is=struct RigVMUnitNode}
   n2 = &RigUnit_GetTransform(Item=(Type=Bone,Name="Root"), bInitial=false)
   &RigUnit_SetTransform(Item=(Type=Control,Name="root_ctrl"), Weight=1, Value=@n2.Transform, ExecutePin=@n1.ExecutePin)
   > n3 = &RigUnit_SetTransform(Item=(Type=Control,Name="body_offset_ctrl"), Weight=1, Value=@n2.Transform)
   n4 = /Script/ControlRig.RigUnit_SequenceExecution(ExecuteContext=@n3.ExecutePin) {is=struct RigVMUnitNode}
   n5 = /Script/ControlRig.RigUnit_ProjectTransformToNewParent(
    Child=(Type=Control,Name="body_ctrl")
    NewParent=(Type=Bone,Name="spine_01")
    OldParent=(Type=Bone,Name="spine_01")
    bChildInitial=True
    bOldParentInitial=True
   ) {is=struct RigVMUnitNode}
   &RigUnit_SetTransform(Item=(Type=Control,Name="body_ctrl"), Weight=1, Value=@n5.Transform, ExecutePin=@n4.A)
   n6 = &RigUnit_ProjectTransformToNewParent(
    Child=(Type=Control,Name="hips_ctrl")
    NewParent=(Type=Bone,Name="Spine_01")
    OldParent=(Type=Bone,Name="spine_01")
    bChildInitial=True
    bOldParentInitial=True
   )
   > &RigUnit_SetTransform(Item=(Type=Control,Name="hips_ctrl"), Weight=1, Value=@n6.Transform)
   n7 = &RigUnit_ProjectTransformToNewParent(
    Child=(Type=Control,Name="spine_01_ctrl")
    NewParent=(Type=Bone,Name="Spine_02")
    OldParent=(Type=Bone,Name="Spine_02")
    bChildInitial=True
    bOldParentInitial=True
   )
   > &RigUnit_SetTransform(Item=(Type=Control,Name="spine_01_ctrl"), Weight=1, Value=@n7.Transform)
   n8 = &RigUnit_ProjectTransformToNewParent(
    Child=(Type=Control,Name="spine_02_ctrl")
    NewParent=(Type=Bone,Name="Spine_03")
    OldParent=(Type=Bone,Name="Spine_03")
    bChildInitial=True
    bOldParentInitial=True
   )
   > &RigUnit_SetTransform(Item=(Type=Control,Name="spine_02_ctrl"), Weight=1, Value=@n8.Transform)
   n9 = &RigUnit_ProjectTransformToNewParent(
    Child=(Type=Control,Name="spine_03_ctrl")
    NewParent=(Type=Bone,Name="spine_04")
    OldParent=(Type=Bone,Name="spine_04")
    bChildInitial=True
    bOldParentInitial=True
   )
   > &RigUnit_SetTransform(Item=(Type=Control,Name="spine_03_ctrl"), Weight=1, Value=@n9.Transform)
   n10 = &RigUnit_ProjectTransformToNewParent(
    Child=(Type=Control,Name="neck_01_ctrl")
    NewParent=(Type=Bone,Name="neck_01")
    OldParent=(Type=Bone,Name="neck_01")
    bChildInitial=True
    bOldParentInitial=True
   )
   > &RigUnit_SetTransform(Item=(Type=Control,Name="neck_01_ctrl"), Weight=1, Value=@n10.Transform)
   n11 = &RigUnit_ProjectTransformToNewParent(
    Child=(Type=Control,Name="neck_02_ctrl")
    NewParent=(Type=Bone,Name="neck_02")
    OldParent=(Type=Bone,Name="neck_02")
    bChildInitial=True
    bOldParentInitial=True
   )
   > &RigUnit_SetTransform(Item=(Type=Control,Name="neck_02_ctrl"), Weight=1, Value=@n11.Transform)
   n12 = &RigUnit_ProjectTransformToNewParent(
    Child=(Type=Null,Name="head_ctrl_space")
    NewParent=(Type=Control,Name="body_ctrl")
    OldParent=(Type=Control,Name="body_ctrl")
    bChildInitial=True
    bOldParentInitial=True
   )
   n13 = &RigUnit_ProjectTransformToNewParent(
    Child=(Type=Null,Name="head_ctrl_space")
    NewParent=(Type=Control,Name="neck_02_ctrl")
    OldParent=(Type=Control,Name="neck_02_ctrl")
    bChildInitial=True
    bOldParentInitial=True
   )
   > &RigUnit_SetTransform(
    Item=(Type=Null,Name="head_ctrl_space")
    Value=(Rotation=(X=0,Y=0,Z=0,W=1),Translation=(X=0,Y=0,Z=0),Scale3D=(X=1,Y=1,Z=1))
    Weight=1
    Value.Rotation=@n12.Transform.Rotation
    Value.Translation=@n13.Transform.Translation
   )
   n14 = &RigUnit_ProjectTransformToNewParent(
    Child=(Type=Control,Name="head_ctrl")
    NewParent=(Type=Bone,Name="head")
    OldParent=(Type=Bone,Name="head")
    bChildInitial=True
    bOldParentInitial=True
   )
   > &RigUnit_SetTransform(Item=(Type=Control,Name="head_ctrl"), Weight=1, Value=@n14.Transform)
   > n15 = /Script/ControlRig.RigUnit_CollectionLoop(Collection=/Script/ControlRig.RigUnit_CollectionItems(Items=((Type=Null,Name="thigh_l_fk_ctrl_space"),(Type=Null,Name="thigh_r_fk_ctrl_space"))) {is=struct RigVMUnitNode}.Collection) {is=struct RigVMUnitNode}
   n16 = &RigUnit_ProjectTransformToNewParent(
    NewParent=(Type=Control,Name="hips_ctrl")
    OldParent=(Type=Control,Name="hips_ctrl")
    bChildInitial=True
    bOldParentInitial=True
    Child=@n15.Item
   )
   > &RigUnit_SetTransform(Weight=1, Item=@n15.Item, Value=@n16.Transform)
   n17 = &RigUnit_SequenceExecution(ExecuteContext=@n4.C)
   n18 = &RigUnit_CollectionLoop(
    Collection=&RigUnit_CollectionItems(Items=((Type=Bone,Name="thigh_r"),(Type=Bone,Name="calf_r"),(Type=Bone,Name="foot_r"),(Type=Bone,Name="ball_r"))).Collection
    ExecutePin=@n17.D
   )
   n19 = /Script/RigVM.RigVMFunction_NameConcat(B=_fk_ctrl, A=@n18.Item.Name) {is=struct RigVMUnitNode, template="Concat::Execute(in A,in B,out Result)"}
   > &RigUnit_SetTransform(
    Item=(Type=Control)
    Weight=1
    Item.Name=@n19.Result
    Value=&RigUnit_ProjectTransformToNewParent(Child=(Type=Control), bChildInitial=True, bOldParentInitial=True, NewParent=@n18.Item, OldParent=@n18.Item, Child.Name=@n19.Result).Transform
   )
   n20 = &RigUnit_ProjectTransformToNewParent(
    Child=(Type=Control,Name="leg_r_pv_ik_ctrl")
    NewParent=(Type=Bone,Name="thigh_r")
    OldParent=(Type=Bone,Name="thigh_r")
    bChildInitial=True
    bOldParentInitial=True
   )
   &RigUnit_SetTransform(Item=(Type=Control,Name="leg_r_pv_ik_ctrl"), Weight=1, Value=@n20.Transform, ExecutePin=@n17.C)
   n21 = &RigUnit_ProjectTransformToNewParent(
    Child=(Type=Control,Name="foot_r_ik_ctrl")
    NewParent=(Type=Bone,Name="foot_r")
    OldParent=(Type=Bone,Name="foot_r")
    bChildInitial=True
    bOldParentInitial=True
   )
   > &RigUnit_SetTransform(Item=(Type=Control,Name="foot_r_ik_ctrl"), Weight=1, Value=@n21.Transform)
   n22 = &RigUnit_ProjectTransformToNewParent(
    Child=(Type=Control,Name="ball_r_ik_ctrl")
    NewParent=(Type=Bone,Name="ball_r")
    OldParent=(Type=Bone,Name="ball_r")
    bChildInitial=True
    bOldParentInitial=True
   )
   > &RigUnit_SetTransform(Item=(Type=Control,Name="ball_r_ik_ctrl"), Weight=1, Value=@n22.Transform)
   n23 = &RigUnit_CollectionLoop(
    Collection=&RigUnit_CollectionItems(Items=((Type=Bone,Name="thigh_l"),(Type=Bone,Name="calf_l"),(Type=Bone,Name="foot_l"),(Type=Bone,Name="ball_l"))).Collection
    ExecutePin=@n17.B
   )
   n24 = &RigVMFunction_NameConcat(B=_fk_ctrl, A=@n23.Item.Name) {template="Concat::Execute(in A,in B,out Result)"}
   > &RigUnit_SetTransform(
    Item=(Type=Control)
    Weight=1
    Item.Name=@n24.Result
    Value=&RigUnit_ProjectTransformToNewParent(Child=(Type=Control), bChildInitial=True, bOldParentInitial=True, NewParent=@n23.Item, OldParent=@n23.Item, Child.Name=@n24.Result).Transform
   )
   n25 = &RigUnit_ProjectTransformToNewParent(
    Child=(Type=Control,Name="leg_l_pv_ik_ctrl")
    NewParent=(Type=Bone,Name="thigh_l")
    OldParent=(Type=Bone,Name="thigh_l")
    bChildInitial=True
    bOldParentInitial=True
   )
   &RigUnit_SetTransform(Item=(Type=Control,Name="leg_l_pv_ik_ctrl"), Weight=1, Value=@n25.Transform, ExecutePin=@n17.A)
   n26 = &RigUnit_ProjectTransformToNewParent(
    Child=(Type=Control,Name="foot_l_ik_ctrl")
    NewParent=(Type=Bone,Name="foot_l")
    OldParent=(Type=Bone,Name="foot_l")
    bChildInitial=True
    bOldParentInitial=True
   )
   > &RigUnit_SetTransform(Item=(Type=Control,Name="foot_l_ik_ctrl"), Weight=1, Value=@n26.Transform)
   n27 = &RigUnit_ProjectTransformToNewParent(
    Child=(Type=Control,Name="ball_l_ik_ctrl")
    NewParent=(Type=Bone,Name="ball_l")
    OldParent=(Type=Bone,Name="ball_l")
    bChildInitial=True
    bOldParentInitial=True
   )
   > &RigUnit_SetTransform(Item=(Type=Control,Name="ball_l_ik_ctrl"), Weight=1, Value=@n27.Transform)
   n28 = &RigUnit_SequenceExecution(ExecuteContext=@n4.B)
   &RigUnit_SetTransform(
    Item=(Type=Control,Name="clavicle_l_ctrl")
    Weight=1
    Value=&RigUnit_GetTransform(Item=(Type=Bone,Name="clavicle_l")).Transform
    ExecutePin=@n28.A
   )
   n29 = &RigUnit_ProjectTransformToNewParent(
    Child=(Type=Null,Name="upperarm_l_fk_ctrl_space")
    NewParent=(Type=Control,Name="body_ctrl")
    OldParent=(Type=Control,Name="body_ctrl")
    bChildInitial=True
    bOldParentInitial=True
   )
   n30 = &RigUnit_ProjectTransformToNewParent(
    Child=(Type=Null,Name="upperarm_l_fk_ctrl_space")
    NewParent=(Type=Control,Name="clavicle_l_ctrl")
    OldParent=(Type=Control,Name="clavicle_l_ctrl")
    bChildInitial=True
    bOldParentInitial=True
   )
   > &RigUnit_SetTransform(
    Item=(Type=Null,Name="upperarm_l_fk_ctrl_space")
    Value=(Rotation=(X=0,Y=0,Z=0,W=1),Translation=(X=0,Y=0,Z=0),Scale3D=(X=1,Y=1,Z=1))
    Weight=1
    Value.Rotation=@n29.Transform.Rotation
    Value.Translation=@n30.Transform.Translation
   )
   > n31 = &RigUnit_CollectionLoop(Collection=&RigUnit_CollectionItems(Items=((Type=Bone,Name="upperarm_l"),(Type=Bone,Name="lowerarm_l"),(Type=Bone,Name="hand_l"))).Collection)
   n32 = &RigVMFunction_NameConcat(B=_fk_ctrl, A=@n31.Item.Name) {template="Concat::Execute(in A,in B,out Result)"}
   > &RigUnit_SetTransform(
    Item=(Type=Control)
    Weight=1
    Item.Name=@n32.Result
    Value=&RigUnit_ProjectTransformToNewParent(Child=(Type=Control), bChildInitial=True, bOldParentInitial=True, NewParent=@n31.Item, OldParent=@n31.Item, Child.Name=@n32.Result).Transform
   )
   n33 = &RigUnit_ProjectTransformToNewParent(
    Child=(Type=Null,Name="fingers_l_space")
    NewParent=(Type=Bone,Name="hand_l")
    OldParent=(Type=Bone,Name="hand_l")
    bChildInitial=True
    bOldParentInitial=True
   )
   &RigUnit_SetTransform(Item=(Type=Null,Name="fingers_l_space"), Weight=1, ExecutePin=@n31.Completed, Value=@n33.Transform)
   n34 = &RigUnit_CollectionItems(Items="((Type=Bone,Name=\"thumb_01_l\"),(Type=Bone,Name=\"thumb_02_l\"),(Type=Bone,Name=\"thumb_03_l\"),(Type=Bone,Name=\"index_metacarpal_l\"),(Type=Bone,Name=\"index_01_l\"),(Type=Bone,Name=\"index_02_l\"),(Type=Bone,Name=\"index_03_l\"),(Type=Bone,Name=\"middle_metacarpal_l\" ...(truncated)")
   > n35 = &RigUnit_CollectionLoop(Collection=@n34.Collection)
   > &RigUnit_SetTransform(
    Item=(Type=Control)
    Weight=1
    Value=&RigUnit_GetTransform(Item=@n35.Item).Transform
    Item.Name=&RigVMFunction_NameConcat(B=_ctrl, A=@n35.Item.Name) {template="Concat::Execute(in A,in B,out Result)"}.Result
   )
   n36 = &RigUnit_ProjectTransformToNewParent(
    Child=(Type=Control,Name="arm_l_pv_ik_ctrl")
    NewParent=(Type=Bone,Name="upperarm_l")
    OldParent=(Type=Bone,Name="upperarm_l")
    bChildInitial=True
    bOldParentInitial=True
   )
   &RigUnit_SetTransform(Item=(Type=Control,Name="arm_l_pv_ik_ctrl"), Weight=1, Value=@n36.Transform, ExecutePin=@n28.B)
   n37 = &RigUnit_ProjectTransformToNewParent(
    Child=(Type=Control,Name="hand_l_ik_ctrl")
    NewParent=(Type=Bone,Name="hand_l")
    OldParent=(Type=Bone,Name="hand_l")
    bChildInitial=True
    bOldParentInitial=True
   )
   > &RigUnit_SetTransform(Item=(Type=Control,Name="hand_l_ik_ctrl"), Weight=1, Value=@n37.Transform)
   n38 = &RigUnit_ProjectTransformToNewParent(
    Child=(Type=Control,Name="clavicle_r_ctrl")
    NewParent=(Type=Bone,Name="clavicle_r")
    OldParent=(Type=Bone,Name="clavicle_r")
    bChildInitial=True
    bOldParentInitial=True
   )
   &RigUnit_SetTransform(Item=(Type=Control,Name="clavicle_r_ctrl"), Weight=1, Value=@n38.Transform, ExecutePin=@n28.C)
   n39 = &RigUnit_ProjectTransformToNewParent(
    Child=(Type=Null,Name="upperarm_r_fk_ctrl_space")
    NewParent=(Type=Control,Name="body_ctrl")
    OldParent=(Type=Control,Name="body_ctrl")
    bChildInitial=True
    bOldParentInitial=True
   )
   n40 = &RigUnit_ProjectTransformToNewParent(
    Child=(Type=Null,Name="upperarm_r_fk_ctrl_space")
    NewParent=(Type=Bone,Name="clavicle_r")
    OldParent=(Type=Bone,Name="clavicle_r")
    bChildInitial=True
    bOldParentInitial=True
   )
   > &RigUnit_SetTransform(
    Item=(Type=Null,Name="upperarm_r_fk_ctrl_space")
    Value=(Rotation=(X=0,Y=0,Z=0,W=1),Translation=(X=0,Y=0,Z=0),Scale3D=(X=1,Y=1,Z=1))
    Weight=1
    Value.Rotation=@n39.Transform.Rotation
    Value.Scale3D=@n40.Transform.Scale3D
    Value.Translation=@n40.Transform.Translation
   )
   > n41 = &RigUnit_CollectionLoop(Collection=&RigUnit_CollectionItems(Items=((Type=Bone,Name="upperarm_r"),(Type=Bone,Name="lowerarm_r"),(Type=Bone,Name="hand_r"))).Collection)
   n42 = &RigVMFunction_NameConcat(B=_fk_ctrl, A=@n41.Item.Name) {template="Concat::Execute(in A,in B,out Result)"}
   > &RigUnit_SetTransform(
    Item=(Type=Control)
    Weight=1
    Item.Name=@n42.Result
    Value=&RigUnit_ProjectTransformToNewParent(Child=(Type=Control), bChildInitial=True, bOldParentInitial=True, NewParent=@n41.Item, OldParent=@n41.Item, Child.Name=@n42.Result).Transform
   )
   n43 = &RigUnit_ProjectTransformToNewParent(
    Child=(Type=Null,Name="fingers_r_space")
    NewParent=(Type=Bone,Name="hand_r")
    OldParent=(Type=Bone,Name="hand_r")
    bChildInitial=True
    bOldParentInitial=True
   )
   &RigUnit_SetTransform(Item=(Type=Null,Name="fingers_r_space"), Weight=1, ExecutePin=@n41.Completed, Value=@n43.Transform)
   n44 = &RigUnit_CollectionItems(Items="((Type=Bone,Name=\"thumb_01_r\"),(Type=Bone,Name=\"thumb_02_r\"),(Type=Bone,Name=\"thumb_03_r\"),(Type=Bone,Name=\"index_metacarpal_r\"),(Type=Bone,Name=\"index_01_r\"),(Type=Bone,Name=\"index_02_r\"),(Type=Bone,Name=\"index_03_r\"),(Type=Bone,Name=\"middle_metacarpal_r\" ...(truncated)")
   > n45 = &RigUnit_CollectionLoop(Collection=@n44.Collection)
   n46 = &RigVMFunction_NameConcat(B=_ctrl, A=@n45.Item.Name) {template="Concat::Execute(in A,in B,out Result)"}
   > &RigUnit_SetTransform(
    Item=(Type=Control,Name="_ctrl")
    Weight=1
    Item.Name=@n46.Result
    Value=&RigUnit_ProjectTransformToNewParent(
     Child=(Type=Control,Name="_ctrl")
     bChildInitial=True
     bOldParentInitial=True
     NewParent=@n45.Item
     OldParent=@n45.Item
     Child.Name=@n46.Result
    ).Transform
   )
   n47 = &RigUnit_ProjectTransformToNewParent(
    Child=(Type=Control,Name="arm_r_pv_ik_ctrl")
    NewParent=(Type=Bone,Name="upperarm_r")
    OldParent=(Type=Bone,Name="upperarm_r")
    bChildInitial=True
    bOldParentInitial=True
   )
   &RigUnit_SetTransform(Item=(Type=Control,Name="arm_r_pv_ik_ctrl"), Weight=1, Value=@n47.Transform, ExecutePin=@n28.D)
   n48 = &RigUnit_ProjectTransformToNewParent(
    Child=(Type=Control,Name="hand_r_ik_ctrl")
    NewParent=(Type=Bone,Name="hand_r")
    OldParent=(Type=Bone,Name="hand_r")
    bChildInitial=True
    bOldParentInitial=True
   )
   > &RigUnit_SetTransform(Item=(Type=Control,Name="hand_r_ik_ctrl"), Weight=1, Value=@n48.Transform)
   n49 = &RigUnit_PrepareForExecution()
   n50 = &RigUnit_SequenceExecution(ExecuteContext=@n49.ExecutePin)
   n51 = &RigUnit_GetTransform(Item=(Type=Bone,Name="spine_01"), bInitial=true)
   n52 = &RigUnit_GetTransform(Item=(Type=Bone,Name="Spine_02"), bInitial=true)
   SetControlOffset = /Script/ControlRig.RigUnit_SetControlOffset(
    Control=body_ctrl
    Offset=(Rotation=(X=0,Y=0,Z=0,W=1),Translation=(X=0,Y=0,Z=0),Scale3D=(X=1,Y=1,Z=1))
    Space=GlobalSpace
    Offset.Translation=/Script/RigVM.RigVMFunction_MathTransformLerp(T=0.5, A=@n51.Transform, B=@n52.Transform) {is=struct RigVMUnitNode, template="Interpolate::Execute(in A,in B,in T,out Result)"}.Result.Translation
    ExecutePin=@n50.A
   ) {is=struct RigVMUnitNode, template="SetControlOffset::Execute(in Control,io Offset,in Space)"}
   > &RigUnit_SetControlOffset(
    Control=hips_ctrl
    Offset=(Rotation=(X=0,Y=0,Z=0,W=1),Translation=(X=0,Y=0,Z=0),Scale3D=(X=1,Y=1,Z=1))
    Space=GlobalSpace
    Offset.Translation=@n51.Transform.Translation
   ) {template="SetControlOffset::Execute(in Control,io Offset,in Space)"}
   > &RigUnit_SetControlOffset(
    Control=spine_01_ctrl
    Offset=(Rotation=(X=0,Y=0,Z=0,W=1),Translation=(X=0,Y=0,Z=0),Scale3D=(X=1,Y=1,Z=1))
    Space=GlobalSpace
    Offset.Translation=@n52.Transform.Translation
   ) {template="SetControlOffset::Execute(in Control,io Offset,in Space)"}
   > &RigUnit_SetControlOffset(
    Control=spine_02_ctrl
    Offset=(Rotation=(X=0,Y=0,Z=0,W=1),Translation=(X=0,Y=0,Z=0),Scale3D=(X=1,Y=1,Z=1))
    Space=GlobalSpace
    Offset.Translation=&RigUnit_GetTransform(Item=(Type=Bone,Name="Spine_03"), bInitial=true).Transform.Translation
   ) {template="SetControlOffset::Execute(in Control,io Offset,in Space)"}
   > &RigUnit_SetControlOffset(
    Control=spine_03_ctrl
    Offset=(Rotation=(X=0,Y=0,Z=0,W=1),Translation=(X=0,Y=0,Z=0),Scale3D=(X=1,Y=1,Z=1))
    Space=GlobalSpace
    Offset.Translation=&RigUnit_GetTransform(Item=(Type=Bone,Name="spine_04"), bInitial=true).Transform.Translation
   ) {template="SetControlOffset::Execute(in Control,io Offset,in Space)"}
   > &RigUnit_SetControlOffset(
    Control=neck_01_ctrl
    Offset=(Rotation=(X=0,Y=0,Z=0,W=1),Translation=(X=0,Y=0,Z=0),Scale3D=(X=1,Y=1,Z=1))
    Space=GlobalSpace
    Offset.Translation=&RigUnit_GetTransform(Item=(Type=Bone,Name="neck_01"), bInitial=true).Transform.Translation
   ) {template="SetControlOffset::Execute(in Control,io Offset,in Space)"}
   > &RigUnit_SetControlOffset(
    Control=neck_02_ctrl
    Offset=(Rotation=(X=0,Y=0,Z=0,W=1),Translation=(X=0,Y=0,Z=0),Scale3D=(X=1,Y=1,Z=1))
    Space=GlobalSpace
    Offset.Translation=&RigUnit_GetTransform(Item=(Type=Bone,Name="neck_02"), bInitial=true).Transform.Translation
   ) {template="SetControlOffset::Execute(in Control,io Offset,in Space)"}
   n53 = &RigUnit_GetTransform(Item=(Type=Bone,Name="head"), bInitial=true)
   > &RigUnit_SetTransform(
    Item=(Type=Null,Name="head_ctrl_space")
    Value=(Rotation=(X=0,Y=0,Z=0,W=1),Translation=(X=0,Y=0,Z=0),Scale3D=(X=1,Y=1,Z=1))
    Weight=1
    bInitial=true
    Value.Translation=@n53.Transform.Translation
   )
   > &RigUnit_SetTransform(
    Item=(Type=Control,Name="head_ctrl")
    Value=(Rotation=(X=0,Y=0,Z=0,W=1),Translation=(X=0,Y=0,Z=0),Scale3D=(X=1,Y=1,Z=1))
    Weight=1
    bInitial=true
    Value.Translation=@n53.Transform.Translation
   )
   &RigUnit_SetTransform(
    Item=(Type=Null,Name="thigh_l_fk_ctrl_space")
    Weight=1
    bInitial=true
    Value=&RigUnit_GetTransform(Item=(Type=Bone,Name="thigh_l"), bInitial=true).Transform
    ExecutePin=@n50.C
   )
   > &RigUnit_SetControlOffset(
    Control=calf_l_fk_ctrl
    Space=GlobalSpace
    Offset=&RigUnit_GetTransform(Item=(Type=Bone,Name="calf_l"), bInitial=true).Transform
   ) {template="SetControlOffset::Execute(in Control,io Offset,in Space)"}
   n54 = &RigUnit_GetTransform(Item=(Type=Bone,Name="foot_l"), bInitial=true)
   > &RigUnit_SetControlOffset(Control=foot_l_fk_ctrl, Space=GlobalSpace, Offset=@n54.Transform) {template="SetControlOffset::Execute(in Control,io Offset,in Space)"}
   > &RigUnit_SetTransform(
    Item=(Type=Control,Name="foot_l_ik_ctrl")
    Value=(Rotation=(X=0,Y=0,Z=0,W=1),Translation=(X=0,Y=0,Z=0),Scale3D=(X=1,Y=1,Z=1))
    Weight=1
    bInitial=true
    bPropagateToChildren=false
    Value.Translation=@n54.Transform.Translation
   )
   n55 = &RigUnit_GetTransform(Item=(Type=Bone,Name="ball_l"), bInitial=true)
   n56 = /Script/RigVM.RigVMFunction_MathQuaternionToRotator(Value=@n55.Transform.Rotation) {is=struct RigVMUnitNode, template="Cast::Execute(in Value,out Result)"}
   > n57 = &RigUnit_SetControlOffset(
    Control=ball_l_fk_ctrl
    Offset=(Rotation=(X=0,Y=0,Z=0,W=1),Translation=(X=0,Y=0,Z=0),Scale3D=(X=1,Y=1,Z=1))
    Space=GlobalSpace
    Offset.Translation=@n55.Transform.Translation
    Offset.Rotation=/Script/RigVM.RigVMFunction_MathQuaternionFromRotator(
     Rotator=(Pitch=0,Yaw=0,Roll=0)
     Rotator.Pitch=&RigVMFunction_MathFloatAdd(B=90, A=@n56.Result.Pitch) {template="Add::Execute(in A,in B,out Result)"}.Result
     Rotator.Roll=@n56.Result.Roll
     Rotator.Yaw=@n56.Result.Yaw
    ) {is=struct RigVMUnitNode}.Result
   ) {template="SetControlOffset::Execute(in Control,io Offset,in Space)"}
   > &RigUnit_SetControlOffset(
    Control=ball_l_ik_ctrl
    Offset=(Rotation=(X=0,Y=0,Z=0,W=1),Translation=(X=0,Y=0,Z=0),Scale3D=(X=1,Y=1,Z=1))
    Space=GlobalSpace
    Offset.Translation=@n57.Offset.Translation
   ) {template="SetControlOffset::Execute(in Control,io Offset,in Space)"}
   n58 = &RigUnit_GetTransform(Item=(Type=Bone,Name="calf_l"), bInitial=true)
   n59 = /Script/RigVM.RigVMFunction_MathFloatDiv(
    B=2
    A=&RigVMFunction_MathVectorDistance(
     B=&RigUnit_GetTransform(Item=(Type=Bone,Name="thigh_l"), bInitial=true).Transform.Translation
     A=@n54.Transform.Translation
    ).Result
   ) {is=struct RigVMUnitNode, template="Divide::Execute(in A,in B,out Result)"}
   n60 = &RigVMFunction_MathTransformMakeRelative(Parent=@n58.Transform, Global=&RigUnit_GetTransform(Item=(Type=Bone,Name="calf_l"), bInitial=true).Transform)
   > &RigUnit_SetTransform(
    Item=(Type=Control,Name="leg_l_pv_ik_ctrl")
    Value=(Rotation=(X=0,Y=0,Z=0,W=1),Translation=(X=0,Y=0,Z=0),Scale3D=(X=1,Y=1,Z=1))
    Weight=1
    bInitial=true
    Value.Translation=/Script/RigVM.RigVMFunction_MathTransformMakeAbsolute(
     Parent=@n58.Transform
     Local=&RigVMFunction_MathTransformMul(B=(Rotation=(X=0,Y=0,Z=0,W=1),Translation=(X=0,Y=0,Z=0),Scale3D=(X=1,Y=1,Z=1)), B.Translation.Y=@n59.Result, A=@n60.Local) {template="Multiply::Execute(in A,in B,out Result)"}.Result
    ) {is=struct RigVMUnitNode, template="Make Absolute::Execute(in Local,in Parent,out Global)"}.Global.Translation
   )
   n61 = &RigUnit_CollectionItems(Items="((Type=Null,Name=\"thigh_l_fk_ctrl_space\"),(Type=Control,Name=\"calf_l_fk_ctrl\"),(Type=Control,Name=\"foot_l_fk_ctrl\"),(Type=Control,Name=\"ball_l_fk_ctrl\"),(Type=Control,Name=\"foot_l_ik_ctrl\"),(Type=Control,Name=\"ball_l_ik_ctrl\"),(Type=Control,Name=\"leg_l_pv_ ...(truncated)")
   > n62 = &RigUnit_CollectionLoop(Collection=@n61.Collection)
   > &RigUnit_SetTransform(
    Item=(Type=Bone)
    Weight=1
    bInitial=true
    Item.Type=@n62.Item.Type
    Value=&RigVMFunction_MathTransformMul(
     B=(Rotation=(X=0,Y=0,Z=0,W=1),Translation=(X=0,Y=0,Z=0),Scale3D=(X=-1,Y=1,Z=1))
     A=&RigUnit_GetTransform(bInitial=true, Item=@n62.Item).Transform
    ) {template="Multiply::Execute(in A,in B,out Result)"}.Result
    Item.Name=/Script/RigVM.RigVMFunction_NameReplace(New=_r_, Old=_l_, Name=@n62.Item.Name) {is=struct RigVMUnitNode, template="Replace::Execute(in Name,in Old,in New,out Result)"}.Result
   )
   &RigUnit_SetControlOffset(
    Control=clavicle_l_ctrl
    Space=GlobalSpace
    Offset=&RigUnit_GetTransform(Item=(Type=Bone,Name="clavicle_l"), bInitial=true).Transform
    ExecutePin=@n50.B
   ) {template="SetControlOffset::Execute(in Control,io Offset,in Space)"}
   n63 = &RigUnit_GetTransform(Item=(Type=Bone,Name="upperarm_l"), bInitial=true)
   > &RigUnit_SetTransform(Item=(Type=Null,Name="upperarm_l_fk_ctrl_space"), Weight=1, bInitial=true, Value=@n63.Transform)
   > &RigUnit_SetControlOffset(
    Control=lowerarm_l_fk_ctrl
    Space=GlobalSpace
    Offset=&RigUnit_GetTransform(Item=(Type=Bone,Name="lowerarm_l"), bInitial=true).Transform
   ) {template="SetControlOffset::Execute(in Control,io Offset,in Space)"}
   n64 = &RigUnit_GetTransform(Item=(Type=Bone,Name="lowerarm_l"), bInitial=true)
   n65 = &RigUnit_GetTransform(Item=(Type=Bone,Name="hand_l"), bInitial=true)
   n66 = /Script/RigVM.RigVMFunction_MathFloatNegate(Value=&RigVMFunction_MathFloatDiv(B=2, A=&RigVMFunction_MathVectorDistance(A=@n65.Transform.Translation, B=@n63.Transform.Translation).Result) {template="Divide::Execute(in A,in B,out Result)"}.Result) {is=struct RigVMUnitNode, template="Negate::Execute(in Value,out Result)"}
   n67 = &RigVMFunction_MathTransformMakeRelative(Global=@n64.Transform, Parent=@n64.Transform)
   > &RigUnit_SetTransform(
    Item=(Type=Control,Name="arm_l_pv_ik_ctrl")
    Value=(Rotation=(X=0,Y=0,Z=0,W=1),Translation=(X=0,Y=0,Z=0),Scale3D=(X=1,Y=1,Z=1))
    Weight=1
    bInitial=true
    Value.Translation=&RigVMFunction_MathTransformMakeAbsolute(
     Parent=@n64.Transform
     Local=&RigVMFunction_MathTransformMul(B=(Rotation=(X=0,Y=0,Z=0,W=1),Translation=(X=0,Y=0,Z=0),Scale3D=(X=1,Y=1,Z=1)), B.Translation.Y=@n66.Result, A=@n67.Local) {template="Multiply::Execute(in A,in B,out Result)"}.Result
    ) {template="Make Absolute::Execute(in Local,in Parent,out Global)"}.Global.Translation
   )
   n68 = &RigUnit_GetTransform(Item=(Type=Bone,Name="hand_l"), bInitial=true)
   n69 = &RigVMFunction_MathQuaternionToRotator(Value=@n68.Transform.Rotation) {template="Cast::Execute(in Value,out Result)"}
   n70 = &RigVMFunction_MathQuaternionFromRotator(
    Rotator=(Pitch=0,Yaw=0,Roll=0)
    Rotator.Roll=&RigVMFunction_MathFloatAdd(B=90, A=@n69.Result.Roll) {template="Add::Execute(in A,in B,out Result)"}.Result
    Rotator.Pitch=@n69.Result.Pitch
    Rotator.Yaw=@n69.Result.Yaw
   )
   > &RigUnit_SetTransform(
    Item=(Type=Control,Name="hand_l_ik_ctrl")
    Value=(Rotation=(X=0,Y=0,Z=0,W=1),Translation=(X=0,Y=0,Z=0),Scale3D=(X=1,Y=1,Z=1))
    Weight=1
    bInitial=true
    Value.Translation=@n68.Transform.Translation
    Value.Rotation=@n70.Result
   )
   n71 = &RigUnit_GetTransform(Item=(Type=Bone,Name="hand_l"), bInitial=true)
   > &RigUnit_SetTransform(
    Item=(Type=Control,Name="hand_l_fk_ctrl")
    Value=(Rotation=(X=0,Y=0,Z=0,W=1),Translation=(X=0,Y=0,Z=0),Scale3D=(X=1,Y=1,Z=1))
    Weight=1
    bInitial=true
    Value.Translation=@n71.Transform.Translation
    Value.Rotation=@n70.Result
   )
   > &RigUnit_SetTransform(Item=(Type=Null,Name="fingers_l_space"), Weight=1, bInitial=true, Value=@n71.Transform)
   n72 = &RigUnit_CollectionItems(Items="((Type=Bone,Name=\"thumb_01_l\"),(Type=Bone,Name=\"thumb_02_l\"),(Type=Bone,Name=\"thumb_03_l\"),(Type=Bone,Name=\"index_metacarpal_l\"),(Type=Bone,Name=\"index_01_l\"),(Type=Bone,Name=\"index_02_l\"),(Type=Bone,Name=\"index_03_l\"),(Type=Bone,Name=\"middle_metacarpal_l\" ...(truncated)")
   > n73 = &RigUnit_CollectionLoop(Collection=@n72.Collection)
   > &RigUnit_SetControlOffset(
    Space=GlobalSpace
    Offset=&RigUnit_GetTransform(bInitial=true, Item=@n73.Item).Transform
    Control=&RigVMFunction_NameConcat(B=_ctrl, A=@n73.Item.Name) {template="Concat::Execute(in A,in B,out Result)"}.Result
   ) {template="SetControlOffset::Execute(in Control,io Offset,in Space)"}
   n74 = &RigUnit_CollectionItems(Items="((Type=Control,Name=\"clavicle_r_ctrl\"),(Type=Null,Name=\"upperarm_r_fk_ctrl_space\"),(Type=Control,Name=\"lowerarm_r_fk_ctrl\"),(Type=Control,Name=\"hand_r_fk_ctrl\"),(Type=Control,Name=\"hand_r_ik_ctrl\"),(Type=Control,Name=\"arm_r_pv_ik_ctrl\"),(Type=Null,Name=\"fi ...(truncated)")
   n75 = &RigUnit_CollectionLoop(Collection=@n74.Collection, ExecutePin=@n73.Completed)
   n76 = &RigVMFunction_NameReplace(New=_l_, Old=_r_, Name=@n75.Item.Name) {template="Replace::Execute(in Name,in Old,in New,out Result)"}
   > &RigUnit_SetTransform(
    Item=(Type=Bone)
    Weight=1
    bInitial=true
    Item.Name=@n75.Item.Name
    Value=&RigVMFunction_MathTransformMul(
     B=(Rotation=(X=0,Y=0,Z=0,W=1),Translation=(X=0,Y=0,Z=0),Scale3D=(X=-1,Y=1,Z=1))
     A=&RigUnit_GetTransform(Item=(Type=All), bInitial=true, Item.Name=@n76.Result, Item.Type=@n75.Item.Type).Transform
    ) {template="Multiply::Execute(in A,in B,out Result)"}.Result
    Item.Type=@n75.Item.Type
   )
   n77 = &RigUnit_CollectionItems(Items="((Type=Bone,Name=\"thumb_01_r\"),(Type=Bone,Name=\"thumb_02_r\"),(Type=Bone,Name=\"thumb_03_r\"),(Type=Bone,Name=\"index_metacarpal_r\"),(Type=Bone,Name=\"index_01_r\"),(Type=Bone,Name=\"index_02_r\"),(Type=Bone,Name=\"index_03_r\"),(Type=Bone,Name=\"middle_metacarpal_r\" ...(truncated)")
   n78 = &RigUnit_CollectionLoop(Collection=@n77.Collection, ExecutePin=@n75.Completed)
   n79 = &RigVMFunction_NameReplace(New=_l, Old=_r, Name=@n78.Item.Name) {template="Replace::Execute(in Name,in Old,in New,out Result)"}
   > &RigUnit_SetControlOffset(
    Space=GlobalSpace
    Offset=&RigVMFunction_MathTransformMul(
     B=(Rotation=(X=0,Y=0,Z=0,W=1),Translation=(X=0,Y=0,Z=0),Scale3D=(X=-1,Y=1,Z=1))
     A=&RigUnit_GetTransform(Item=(Type=Bone,Name="hand_l"), bInitial=true, Item.Name=@n79.Result).Transform
    ) {template="Multiply::Execute(in A,in B,out Result)"}.Result
    Control=&RigVMFunction_NameConcat(B=_ctrl, A=@n78.Item.Name) {template="Concat::Execute(in A,in B,out Result)"}.Result
   ) {template="SetControlOffset::Execute(in Control,io Offset,in Space)"}
   RigUnit_BeginExecution = &RigUnit_BeginExecution()
   n80 = &RigUnit_SequenceExecution(ExecuteContext=@RigUnit_BeginExecution.ExecutePin)
   n81 = &RigUnit_SequenceExecution(ExecuteContext=@n80.A)
   &RigUnit_SetTransform(
    Item=(Type=Bone,Name="Root")
    Space=LocalSpace
    Weight=1
    Value=&RigUnit_GetTransform(Item=(Type=Control,Name="root_ctrl")).Transform
    ExecutePin=@n81.A
   )
   n82 = &RigUnit_ProjectTransformToNewParent(
    Child=(Type=Bone,Name="pelvis")
    NewParent=(Type=Control,Name="hips_ctrl")
    OldParent=(Type=Control,Name="hips_ctrl")
    bChildInitial=True
    bOldParentInitial=True
   )
   > &RigUnit_SetTransform(Item=(Type=Bone,Name="pelvis"), Weight=1, Value=@n82.Transform)
   n83 = &RigUnit_ProjectTransformToNewParent(
    Child=(Type=Bone,Name="Spine_01")
    NewParent=(Type=Control,Name="hips_ctrl")
    OldParent=(Type=Control,Name="hips_ctrl")
    bChildInitial=True
    bOldParentInitial=True
   )
   > &RigUnit_SetTransform(Item=(Type=Bone,Name="Spine_01"), Weight=1, Value=@n83.Transform)
   n84 = &RigUnit_ProjectTransformToNewParent(
    Child=(Type=Bone,Name="Spine_02")
    NewParent=(Type=Control,Name="spine_01_ctrl")
    OldParent=(Type=Control,Name="spine_01_ctrl")
    bChildInitial=True
    bOldParentInitial=True
   )
   > &RigUnit_SetTransform(Item=(Type=Bone,Name="Spine_02"), Weight=1, Value=@n84.Transform)
   n85 = &RigUnit_ProjectTransformToNewParent(
    Child=(Type=Bone,Name="Spine_03")
    NewParent=(Type=Control,Name="spine_02_ctrl")
    OldParent=(Type=Control,Name="spine_02_ctrl")
    bChildInitial=True
    bOldParentInitial=True
   )
   > &RigUnit_SetTransform(Item=(Type=Bone,Name="Spine_03"), Weight=1, Value=@n85.Transform)
   n86 = &RigUnit_ProjectTransformToNewParent(
    Child=(Type=Bone,Name="spine_04")
    NewParent=(Type=Control,Name="spine_03_ctrl")
    OldParent=(Type=Control,Name="spine_03_ctrl")
    bChildInitial=True
    bOldParentInitial=True
   )
   > &RigUnit_SetTransform(Item=(Type=Bone,Name="spine_04"), Weight=1, Value=@n86.Transform)
   > n87 = &RigUnit_CollectionLoop(Collection=&RigUnit_CollectionItems(Items=((Type=Null,Name="thigh_l_fk_ctrl_space"),(Type=Null,Name="thigh_r_fk_ctrl_space"))).Collection)
   n88 = &RigUnit_ProjectTransformToNewParent(
    NewParent=(Type=Control,Name="hips_ctrl")
    OldParent=(Type=Control,Name="hips_ctrl")
    bChildInitial=True
    bOldParentInitial=True
    Child=@n87.Item
   )
   > &RigUnit_SetTransform(Weight=1, Item=@n87.Item, Value=@n88.Transform)
   n89 = &RigUnit_ProjectTransformToNewParent(
    Child=(Type=Bone,Name="neck_01")
    NewParent=(Type=Control,Name="neck_01_ctrl")
    OldParent=(Type=Control,Name="neck_01_ctrl")
    bChildInitial=True
    bOldParentInitial=True
   )
   &RigUnit_SetTransform(Item=(Type=Bone,Name="neck_01"), Weight=1, ExecutePin=@n87.Completed, Value=@n89.Transform)
   n90 = &RigUnit_ProjectTransformToNewParent(
    Child=(Type=Bone,Name="neck_02")
    NewParent=(Type=Control,Name="neck_02_ctrl")
    OldParent=(Type=Control,Name="neck_02_ctrl")
    bChildInitial=True
    bOldParentInitial=True
   )
   > &RigUnit_SetTransform(Item=(Type=Bone,Name="neck_02"), Weight=1, Value=@n90.Transform)
   n91 = &RigUnit_ProjectTransformToNewParent(
    Child=(Type=Null,Name="head_ctrl_space")
    NewParent=(Type=Control,Name="neck_02_ctrl")
    OldParent=(Type=Control,Name="neck_02_ctrl")
    bChildInitial=True
    bOldParentInitial=True
   )
   > n92 = /Script/ControlRig.RigUnit_SetTranslation(Item=(Type=Null,Name="head_ctrl_space"), Space=GlobalSpace, Weight=1, bPropagateToChildren=True, Value=@n91.Transform.Translation) {is=struct RigVMUnitNode, template="Set Transform::Execute(in Item,in Space,in bInitial,in Value,in Weight,in bPropagateToChildren)"}
   n93 = &RigUnit_SequenceExecution(ExecuteContext=@n92.ExecutePin)
   Branch_7_RigVMUnitNode = &RigVMFunction_ControlFlowBranch(
    Condition=/Script/ControlRig.RigUnit_GetControlBool(Control=head_local) {is=struct RigVMUnitNode, template="GetControlValue::Execute(in Control,out BoolValue)"}.BoolValue
    ExecuteContext=@n93.A
   )
   &RigUnit_SetRotation(
    Item=(Type=Null,Name="head_ctrl_space")
    Space=GlobalSpace
    Weight=1
    ExecutePin=@Branch_7_RigVMUnitNode.False
    Value=@n91.Transform.Rotation
   )
   n94 = &RigUnit_ProjectTransformToNewParent(
    Child=(Type=Null,Name="head_ctrl_space")
    NewParent=(Type=Control,Name="body_ctrl")
    OldParent=(Type=Control,Name="body_ctrl")
    bChildInitial=True
    bOldParentInitial=True
   )
   &RigUnit_SetRotation(Item=(Type=Null,Name="head_ctrl_space"), Space=GlobalSpace, Weight=1, ExecutePin=@Branch_7_RigVMUnitNode.True, Value=@n94.Transform.Rotation)
   n95 = &RigUnit_ProjectTransformToNewParent(
    Child=(Type=Bone,Name="head")
    NewParent=(Type=Control,Name="head_ctrl")
    OldParent=(Type=Control,Name="head_ctrl")
    bChildInitial=True
    bOldParentInitial=True
   )
   n96 = &RigUnit_SetTransform(Item=(Type=Bone,Name="head"), Weight=1, Value=@n95.Transform, ExecutePin=@n93.B)
   n97 = &RigUnit_SequenceExecution(ExecuteContext=@n96.ExecutePin)
   &RigUnit_SetTransform(
    Item=(Type=Bone,Name="clavicle_l")
    Weight=1
    Value=&RigUnit_GetTransform(Item=(Type=Control,Name="clavicle_l_ctrl")).Transform
    ExecutePin=@n97.A
   )
   n98 = &RigUnit_ProjectTransformToNewParent(
    Child=(Type=Null,Name="upperarm_l_fk_ctrl_space")
    NewParent=(Type=Control,Name="clavicle_l_ctrl")
    OldParent=(Type=Control,Name="clavicle_l_ctrl")
    bChildInitial=True
    bOldParentInitial=True
   )
   > n99 = &RigUnit_SetTranslation(Item=(Type=Null,Name="upperarm_l_fk_ctrl_space"), Space=GlobalSpace, Weight=1, bPropagateToChildren=True, Value=@n98.Transform.Translation) {template="Set Transform::Execute(in Item,in Space,in bInitial,in Value,in Weight,in bPropagateToChildren)"}
   Branch_6_RigVMUnitNode = &RigVMFunction_ControlFlowBranch(
    Condition=&RigUnit_GetControlBool(Control=arm_l_fk_local) {template="GetControlValue::Execute(in Control,out BoolValue)"}.BoolValue
    ExecuteContext=@n99.ExecutePin
   )
   &RigUnit_SetRotation(
    Item=(Type=Null,Name="upperarm_l_fk_ctrl_space")
    Space=GlobalSpace
    Weight=1
    ExecutePin=@Branch_6_RigVMUnitNode.False
    Value=@n98.Transform.Rotation
   )
   n100 = &RigUnit_ProjectTransformToNewParent(
    Child=(Type=Null,Name="upperarm_l_fk_ctrl_space")
    NewParent=(Type=Control,Name="body_ctrl")
    OldParent=(Type=Control,Name="body_ctrl")
    bChildInitial=True
    bOldParentInitial=True
   )
   &RigUnit_SetRotation(
    Item=(Type=Null,Name="upperarm_l_fk_ctrl_space")
    Space=GlobalSpace
    Weight=1
    ExecutePin=@Branch_6_RigVMUnitNode.True
    Value=@n100.Transform.Rotation
   )
   n101 = &RigUnit_ProjectTransformToNewParent(
    Child=(Type=Bone,Name="clavicle_r")
    NewParent=(Type=Control,Name="clavicle_r_ctrl")
    OldParent=(Type=Control,Name="clavicle_r_ctrl")
    bChildInitial=True
    bOldParentInitial=True
   )
   &RigUnit_SetTransform(Item=(Type=Bone,Name="clavicle_r"), Weight=1, Value=@n101.Transform, ExecutePin=@n97.B)
   n102 = &RigUnit_ProjectTransformToNewParent(
    Child=(Type=Null,Name="upperarm_r_fk_ctrl_space")
    NewParent=(Type=Control,Name="clavicle_r_ctrl")
    OldParent=(Type=Control,Name="clavicle_r_ctrl")
    bChildInitial=True
    bOldParentInitial=True
   )
   > n103 = &RigUnit_SetTranslation(Item=(Type=Null,Name="upperarm_r_fk_ctrl_space"), Space=GlobalSpace, Weight=1, bPropagateToChildren=True, Value=@n102.Transform.Translation) {template="Set Transform::Execute(in Item,in Space,in bInitial,in Value,in Weight,in bPropagateToChildren)"}
   Branch_6_1_RigVMUnitNode = &RigVMFunction_ControlFlowBranch(
    Condition=&RigUnit_GetControlBool(Control=arm_r_fk_local) {template="GetControlValue::Execute(in Control,out BoolValue)"}.BoolValue
    ExecuteContext=@n103.ExecutePin
   )
   &RigUnit_SetRotation(
    Item=(Type=Null,Name="upperarm_r_fk_ctrl_space")
    Space=GlobalSpace
    Weight=1
    ExecutePin=@Branch_6_1_RigVMUnitNode.False
    Value=@n102.Transform.Rotation
   )
   n104 = &RigUnit_ProjectTransformToNewParent(
    Child=(Type=Null,Name="upperarm_r_fk_ctrl_space")
    NewParent=(Type=Control,Name="body_ctrl")
    OldParent=(Type=Control,Name="body_ctrl")
    bChildInitial=True
    bOldParentInitial=True
   )
   &RigUnit_SetRotation(
    Item=(Type=Null,Name="upperarm_r_fk_ctrl_space")
    Space=GlobalSpace
    Weight=1
    ExecutePin=@Branch_6_1_RigVMUnitNode.True
    Value=@n104.Transform.Rotation
   )
   n105 = &RigUnit_SequenceExecution(ExecuteContext=@n81.B)
   n106 = &RigUnit_GetControlBool(Control=arm_l_fk_ik_switch) {template="GetControlValue::Execute(in Control,out BoolValue)"}
   n107 = &RigVMFunction_MathBoolNot(Value=@n106.BoolValue)
   SetControlVisibility = /Script/ControlRig.RigUnit_SetControlVisibility(Item=(Type=Control,Name="upperarm_l_fk_ctrl"), bVisible=@n107.Result, ExecutePin=@n105.A) {is=struct RigVMUnitNode, template="SetControlVisibility::Execute(in Item,in Pattern,in bVisible)"}
   > &RigUnit_SetControlVisibility(Item=(Type=Control,Name="lowerarm_l_fk_ctrl"), bVisible=@n107.Result) {template="SetControlVisibility::Execute(in Item,in Pattern,in bVisible)"}
   > &RigUnit_SetControlVisibility(Item=(Type=Control,Name="hand_l_fk_ctrl"), bVisible=@n107.Result) {template="SetControlVisibility::Execute(in Item,in Pattern,in bVisible)"}
   > &RigUnit_SetControlVisibility(Item=(Type=Control,Name="hand_l_ik_ctrl"), bVisible=@n106.BoolValue) {template="SetControlVisibility::Execute(in Item,in Pattern,in bVisible)"}
   > n108 = &RigUnit_SetControlVisibility(Item=(Type=Control,Name="arm_l_pv_ik_ctrl"), bVisible=@n106.BoolValue) {template="SetControlVisibility::Execute(in Item,in Pattern,in bVisible)"}
   Branch_RigVMUnitNode = &RigVMFunction_ControlFlowBranch(ExecuteContext=@n108.ExecutePin, Condition=@n106.BoolValue)
   n109 = &RigUnit_CollectionItems(Items=((Type=Bone,Name="upperarm_l"),(Type=Bone,Name="lowerarm_l"),(Type=Bone,Name="hand_l")))
   n110 = &RigUnit_CollectionLoop(ExecutePin=@Branch_RigVMUnitNode.False, Collection=@n109.Collection)
   n111 = &RigVMFunction_NameConcat(B=_fk_ctrl, A=@n110.Item.Name) {template="Concat::Execute(in A,in B,out Result)"}
   > &RigUnit_SetTransform(
    Item=(Type=Bone,Name="arm_r_settings_ctrl")
    Weight=1
    Item.Name=@n110.Item.Name
    Value=&RigUnit_ProjectTransformToNewParent(
     Child=(Type=Bone)
     NewParent=(Type=Control)
     OldParent=(Type=Control)
     bChildInitial=True
     bOldParentInitial=True
     Child.Name=@n110.Item.Name
     NewParent.Name=@n111.Result
     OldParent.Name=@n111.Result
    ).Transform
   )
   n112 = &RigUnit_ProjectTransformToNewParent(
    Child=(Type=Control,Name="arm_l_pv_ik_ctrl")
    NewParent=(Type=Bone,Name="upperarm_l")
    OldParent=(Type=Bone,Name="upperarm_l")
    bChildInitial=True
    bOldParentInitial=True
   )
   &RigUnit_SetTransform(Item=(Type=Control,Name="arm_l_pv_ik_ctrl"), Weight=1, ExecutePin=@n110.Completed, Value=@n112.Transform)
   n113 = &RigUnit_ProjectTransformToNewParent(
    Child=(Type=Control,Name="hand_l_ik_ctrl")
    NewParent=(Type=Bone,Name="hand_l")
    OldParent=(Type=Bone,Name="hand_l")
    bChildInitial=True
    bOldParentInitial=True
   )
   > n114 = &RigUnit_SetTransform(Item=(Type=Control,Name="hand_l_ik_ctrl"), Weight=1, Value=@n113.Transform)
   n115 = RigVMVariableNode() {variable=arm_l_ik_mode}
   Branch_4_RigVMUnitNode = &RigVMFunction_ControlFlowBranch(ExecuteContext=@n114.ExecutePin, Condition=@n115.Value)
   n116 = &RigUnit_CollectionItems(Items="((Type=Control,Name=\"upperarm_l_fk_ctrl\"),(Type=Control,Name=\"lowerarm_l_fk_ctrl\"),(Type=Control,Name=\"hand_l_fk_ctrl\"),(Type=Control,Name=\"arm_l_pv_ik_ctrl\"),(Type=Control,Name=\"hand_l_ik_ctrl\"),(Type=Control,Name=\"arm_l_fk_ik_switch\"),(Type=Control,Name= ...(truncated)")
   n117 = &RigUnit_CollectionLoop(ExecutePin=@Branch_4_RigVMUnitNode.True, Collection=@n116.Collection)
   > /Script/ControlRig.RigUnit_SendEvent(Event=RequestAutoKey, bOnlyDuringInteraction=True, Item=@n117.Item, bEnable=@n115.Value) {is=struct RigVMUnitNode, template="SendEvent::Execute(in Event,in Item,in OffsetInSeconds,in bEnable,in bOnlyDuringInteraction)"}
   RigVMVariableNode(Value=false, ExecuteContext=@n117.Completed) {variable=arm_l_ik_mode}
   n118 = &RigUnit_ProjectTransformToNewParent(
    Child=(Type=Bone,Name="hand_l")
    NewParent=(Type=Control,Name="hand_l_ik_ctrl")
    OldParent=(Type=Control,Name="hand_l_ik_ctrl")
    bChildInitial=True
    bOldParentInitial=True
   )
   /Script/ControlRig.RigUnit_TwoBoneIKSimplePerItem(
    DebugSettings=(Scale=10,WorldOffset=(Rotation=(X=0,Y=0,Z=0,W=1),Translation=(X=0,Y=0,Z=0),Scale3D=(X=1,Y=1,Z=1)))
    EffectorItem=(Type=Bone,Name="hand_l")
    ItemA=(Type=Bone,Name="upperarm_l")
    ItemB=(Type=Bone,Name="lowerarm_l")
    PoleVectorKind=Location
    PoleVectorSpace=(Type=Bone)
    PrimaryAxis=(X=1,Y=0,Z=0)
    SecondaryAxis=(X=0,Y=-1,Z=0)
    SecondaryAxisWeight=1
    StretchMaximumRatio=1.1
    StretchStartRatio=1
    Weight=1
    bEnableStretch=true
    bPropagateToChildren=True
    ExecutePin=@Branch_RigVMUnitNode.True
    PoleVector=&RigUnit_GetTransform(Item=(Type=Control,Name="arm_l_pv_ik_ctrl")).Transform.Translation
    Effector=@n118.Transform
   ) {is=struct RigVMUnitNode}
   > /Script/ControlRig.RigUnit_DebugLineItemSpace(
    Color=(R=0,G=0,B=1,A=1)
    DebugDrawSettings=(DepthPriority=SDPG_Foreground,Lifetime=-1)
    Space=(Type=Bone)
    Thickness=0.5
    WorldOffset=(Rotation=(X=0,Y=0,Z=0,W=1),Translation=(X=0,Y=0,Z=0),Scale3D=(X=1,Y=1,Z=1))
    bEnabled=True
    B=&RigUnit_GetTransform(Item=(Type=Control,Name="arm_l_pv_ik_ctrl")).Transform.Translation
    A=&RigUnit_GetTransform(Item=(Type=Bone,Name="lowerarm_l")).Transform.Translation
   ) {is=struct RigVMUnitNode}
   > n119 = &RigUnit_CollectionLoop(Collection=@n109.Collection)
   n120 = &RigVMFunction_NameConcat(B=_fk_ctrl, A=@n119.Item.Name) {template="Concat::Execute(in A,in B,out Result)"}
   > &RigUnit_SetTransform(
    Item=(Type=Control,Name="None")
    Weight=1
    Item.Name=@n120.Result
    Value=&RigUnit_ProjectTransformToNewParent(
     Child=(Type=Control)
     NewParent=(Type=Bone)
     OldParent=(Type=Bone)
     bChildInitial=True
     bOldParentInitial=True
     NewParent.Name=@n119.Item.Name
     OldParent.Name=@n119.Item.Name
     Child.Name=@n120.Result
    ).Transform
   )
   Branch_3_RigVMUnitNode = &RigVMFunction_ControlFlowBranch(ExecuteContext=@n119.Completed, Condition=RigVMVariableNode() {variable=arm_l_ik_mode}.Value)
   n121 = &RigUnit_CollectionLoop(ExecutePin=@Branch_3_RigVMUnitNode.False, Collection=@n116.Collection)
   > SendEvent = &RigUnit_SendEvent(Event=RequestAutoKey, bEnable=True, bOnlyDuringInteraction=True, Item=@n121.Item) {template="SendEvent::Execute(in Event,in Item,in OffsetInSeconds,in bEnable,in bOnlyDuringInteraction)"}
   RigVMVariableNode(Value=true, ExecuteContext=@n121.Completed) {variable=arm_l_ik_mode}
   n122 = &RigUnit_ProjectTransformToNewParent(
    Child=(Type=Null,Name="fingers_l_space")
    NewParent=(Type=Bone,Name="hand_l")
    OldParent=(Type=Bone,Name="hand_l")
    bChildInitial=True
    bOldParentInitial=True
   )
   &RigUnit_SetTransform(Item=(Type=Null,Name="fingers_l_space"), Weight=1, Value=@n122.Transform, ExecutePin=@n105.C)
   n123 = &RigUnit_CollectionItems(Items="((Type=Bone,Name=\"thumb_01_l\"),(Type=Bone,Name=\"thumb_02_l\"),(Type=Bone,Name=\"thumb_03_l\"),(Type=Bone,Name=\"index_metacarpal_l\"),(Type=Bone,Name=\"index_01_l\"),(Type=Bone,Name=\"index_02_l\"),(Type=Bone,Name=\"index_03_l\"),(Type=Bone,Name=\"middle_metacarpal_l\" ...(truncated)")
   > n124 = &RigUnit_CollectionLoop(Collection=@n123.Collection)
   n125 = &RigVMFunction_NameConcat(B=_ctrl, A=@n124.Item.Name) {template="Concat::Execute(in A,in B,out Result)"}
   > &RigUnit_SetTransform(
    Weight=1
    Item=@n124.Item
    Value=&RigUnit_ProjectTransformToNewParent(
     NewParent=(Type=Control)
     OldParent=(Type=Control)
     bChildInitial=True
     bOldParentInitial=True
     Child=@n124.Item
     NewParent.Name=@n125.Result
     OldParent.Name=@n125.Result
    ).Transform
   )
   n126 = &RigUnit_ProjectTransformToNewParent(
    Child=(Type=Null,Name="fingers_r_space")
    NewParent=(Type=Bone,Name="hand_r")
    OldParent=(Type=Bone,Name="hand_r")
    bChildInitial=True
    bOldParentInitial=True
   )
   &RigUnit_SetTransform(Item=(Type=Null,Name="fingers_r_space"), Weight=1, ExecutePin=@n124.Completed, Value=@n126.Transform)
   n127 = &RigUnit_CollectionItems(Items="((Type=Bone,Name=\"thumb_01_r\"),(Type=Bone,Name=\"thumb_02_r\"),(Type=Bone,Name=\"thumb_03_r\"),(Type=Bone,Name=\"index_metacarpal_r\"),(Type=Bone,Name=\"index_01_r\"),(Type=Bone,Name=\"index_02_r\"),(Type=Bone,Name=\"index_03_r\"),(Type=Bone,Name=\"middle_metacarpal_r\" ...(truncated)")
   > n128 = &RigUnit_CollectionLoop(Collection=@n127.Collection)
   n129 = &RigVMFunction_NameConcat(B=_ctrl, A=@n128.Item.Name) {template="Concat::Execute(in A,in B,out Result)"}
   n130 = &RigUnit_ProjectTransformToNewParent(
    Child=(Type=Bone)
    NewParent=(Type=Control,Name="_ctrl")
    OldParent=(Type=Control,Name="_ctrl")
    bChildInitial=True
    bOldParentInitial=True
    Child.Name=@n128.Item.Name
    NewParent.Name=@n129.Result
    OldParent.Name=@n129.Result
   )
   > &RigUnit_SetTransform(Item=(Type=Bone), Weight=1, Item.Name=@n128.Item.Name, Value=@n130.Transform)
   n131 = &RigUnit_GetControlBool(Control=arm_r_fk_ik_switch) {template="GetControlValue::Execute(in Control,out BoolValue)"}
   n132 = &RigVMFunction_MathBoolNot(Value=@n131.BoolValue)
   &RigUnit_SetControlVisibility(Item=(Type=Control,Name="upperarm_r_fk_ctrl"), bVisible=@n132.Result, ExecutePin=@n105.B) {template="SetControlVisibility::Execute(in Item,in Pattern,in bVisible)"}
   > &RigUnit_SetControlVisibility(Item=(Type=Control,Name="lowerarm_r_fk_ctrl"), bVisible=@n132.Result) {template="SetControlVisibility::Execute(in Item,in Pattern,in bVisible)"}
   > &RigUnit_SetControlVisibility(Item=(Type=Control,Name="hand_r_fk_ctrl"), bVisible=@n132.Result) {template="SetControlVisibility::Execute(in Item,in Pattern,in bVisible)"}
   > &RigUnit_SetControlVisibility(Item=(Type=Control,Name="hand_r_ik_ctrl"), bVisible=@n131.BoolValue) {template="SetControlVisibility::Execute(in Item,in Pattern,in bVisible)"}
   > n133 = &RigUnit_SetControlVisibility(Item=(Type=Control,Name="arm_r_pv_ik_ctrl"), bVisible=@n131.BoolValue) {template="SetControlVisibility::Execute(in Item,in Pattern,in bVisible)"}
   Branch_2_RigVMUnitNode = &RigVMFunction_ControlFlowBranch(ExecuteContext=@n133.ExecutePin, Condition=@n131.BoolValue)
   n134 = &RigUnit_CollectionItems(Items=((Type=Bone,Name="upperarm_r"),(Type=Bone,Name="lowerarm_r"),(Type=Bone,Name="hand_r")))
   n135 = &RigUnit_CollectionLoop(ExecutePin=@Branch_2_RigVMUnitNode.False, Collection=@n134.Collection)
   n136 = &RigVMFunction_NameConcat(B=_fk_ctrl, A=@n135.Item.Name) {template="Concat::Execute(in A,in B,out Result)"}
   > &RigUnit_SetTransform(
    Item=(Type=Bone,Name="arm_l_settings_ctrl")
    Weight=1
    Item.Name=@n135.Item.Name
    Value=&RigUnit_ProjectTransformToNewParent(
     Child=(Type=Bone)
     NewParent=(Type=Control)
     OldParent=(Type=Control)
     bChildInitial=True
     bOldParentInitial=True
     Child.Name=@n135.Item.Name
     NewParent.Name=@n136.Result
     OldParent.Name=@n136.Result
    ).Transform
   )
   n137 = &RigUnit_ProjectTransformToNewParent(
    Child=(Type=Control,Name="arm_r_pv_ik_ctrl")
    NewParent=(Type=Bone,Name="upperarm_r")
    OldParent=(Type=Bone,Name="upperarm_r")
    bChildInitial=True
    bOldParentInitial=True
   )
   &RigUnit_SetTransform(Item=(Type=Control,Name="arm_r_pv_ik_ctrl"), Weight=1, ExecutePin=@n135.Completed, Value=@n137.Transform)
   n138 = &RigUnit_ProjectTransformToNewParent(
    Child=(Type=Control,Name="hand_r_ik_ctrl")
    NewParent=(Type=Bone,Name="hand_r")
    OldParent=(Type=Bone,Name="hand_r")
    bChildInitial=True
    bOldParentInitial=True
   )
   > n139 = &RigUnit_SetTransform(Item=(Type=Control,Name="hand_r_ik_ctrl"), Weight=1, Value=@n138.Transform)
   n140 = RigVMVariableNode() {variable=arm_r_ik_mode}
   Branch_4_1_RigVMUnitNode = &RigVMFunction_ControlFlowBranch(ExecuteContext=@n139.ExecutePin, Condition=@n140.Value)
   n141 = &RigUnit_CollectionItems(Items="((Type=Control,Name=\"upperarm_r_fk_ctrl\"),(Type=Control,Name=\"lowerarm_r_fk_ctrl\"),(Type=Control,Name=\"hand_r_fk_ctrl\"),(Type=Control,Name=\"arm_r_pv_ik_ctrl\"),(Type=Control,Name=\"hand_r_ik_ctrl\"),(Type=Control,Name=\"arm_r_fk_ik_switch\"),(Type=Control,Name= ...(truncated)")
   n142 = &RigUnit_CollectionLoop(ExecutePin=@Branch_4_1_RigVMUnitNode.True, Collection=@n141.Collection)
   > &RigUnit_SendEvent(Event=RequestAutoKey, Item=(Type=Control), bOnlyDuringInteraction=True, Item.Name=@n142.Item.Name, bEnable=@n140.Value) {template="SendEvent::Execute(in Event,in Item,in OffsetInSeconds,in bEnable,in bOnlyDuringInteraction)"}
   RigVMVariableNode(Value=false, ExecuteContext=@n142.Completed) {variable=arm_r_ik_mode}
   n143 = &RigUnit_ProjectTransformToNewParent(
    Child=(Type=Bone,Name="hand_r")
    NewParent=(Type=Control,Name="hand_r_ik_ctrl")
    OldParent=(Type=Control,Name="hand_r_ik_ctrl")
    bChildInitial=True
    bOldParentInitial=True
   )
   &RigUnit_TwoBoneIKSimplePerItem(
    DebugSettings=(Scale=10,WorldOffset=(Rotation=(X=0,Y=0,Z=0,W=1),Translation=(X=0,Y=0,Z=0),Scale3D=(X=1,Y=1,Z=1)))
    EffectorItem=(Type=Bone,Name="hand_r")
    ItemA=(Type=Bone,Name="upperarm_r")
    ItemB=(Type=Bone,Name="lowerarm_r")
    PoleVectorKind=Location
    PoleVectorSpace=(Type=Bone)
    PrimaryAxis=(X=-1,Y=0,Z=0)
    SecondaryAxis=(X=0,Y=1,Z=0)
    SecondaryAxisWeight=1
    StretchMaximumRatio=1.1
    StretchStartRatio=1
    Weight=1
    bEnableStretch=true
    bPropagateToChildren=True
    ExecutePin=@Branch_2_RigVMUnitNode.True
    PoleVector=&RigUnit_GetTransform(Item=(Type=Control,Name="arm_r_pv_ik_ctrl")).Transform.Translation
    Effector=@n143.Transform
   )
   > &RigUnit_DebugLineItemSpace(
    Color=(R=1,G=0,B=0,A=1)
    DebugDrawSettings=(DepthPriority=SDPG_Foreground,Lifetime=-1)
    Space=(Type=Bone)
    Thickness=0.5
    WorldOffset=(Rotation=(X=0,Y=0,Z=0,W=1),Translation=(X=0,Y=0,Z=0),Scale3D=(X=1,Y=1,Z=1))
    bEnabled=True
    B=&RigUnit_GetTransform(Item=(Type=Control,Name="arm_r_pv_ik_ctrl")).Transform.Translation
    A=&RigUnit_GetTransform(Item=(Type=Bone,Name="lowerarm_r")).Transform.Translation
   )
   > n144 = &RigUnit_CollectionLoop(Collection=@n134.Collection)
   n145 = &RigVMFunction_NameConcat(B=_fk_ctrl, A=@n144.Item.Name) {template="Concat::Execute(in A,in B,out Result)"}
   > &RigUnit_SetTransform(
    Item=(Type=Control,Name="None")
    Weight=1
    Item.Name=@n145.Result
    Value=&RigUnit_ProjectTransformToNewParent(
     Child=(Type=Control)
     NewParent=(Type=Bone)
     OldParent=(Type=Bone)
     bChildInitial=True
     bOldParentInitial=True
     NewParent.Name=@n144.Item.Name
     OldParent.Name=@n144.Item.Name
     Child.Name=@n145.Result
    ).Transform
   )
   Branch_3_1_RigVMUnitNode = &RigVMFunction_ControlFlowBranch(ExecuteContext=@n144.Completed, Condition=RigVMVariableNode() {variable=arm_r_ik_mode}.Value)
   n146 = &RigUnit_CollectionLoop(ExecutePin=@Branch_3_1_RigVMUnitNode.False, Collection=@n141.Collection)
   > &RigUnit_SendEvent(Event=RequestAutoKey, bEnable=True, bOnlyDuringInteraction=True, Item=@n146.Item) {template="SendEvent::Execute(in Event,in Item,in OffsetInSeconds,in bEnable,in bOnlyDuringInteraction)"}
   RigVMVariableNode(Value=true, ExecuteContext=@n146.Completed) {variable=arm_r_ik_mode}
   n147 = &RigUnit_SequenceExecution(ExecuteContext=@n81.C)
   n148 = &RigUnit_GetControlBool(Control=leg_l_fk_ik_switch) {template="GetControlValue::Execute(in Control,out BoolValue)"}
   &RigUnit_SetControlVisibility(Item=(Type=Control,Name="foot_l_ik_ctrl"), bVisible=@n148.BoolValue, ExecutePin=@n147.A) {template="SetControlVisibility::Execute(in Item,in Pattern,in bVisible)"}
   > &RigUnit_SetControlVisibility(Item=(Type=Control,Name="leg_l_pv_ik_ctrl"), bVisible=@n148.BoolValue) {template="SetControlVisibility::Execute(in Item,in Pattern,in bVisible)"}
   > &RigUnit_SetControlVisibility(Item=(Type=Control,Name="ball_l_ik_ctrl"), bVisible=@n148.BoolValue) {template="SetControlVisibility::Execute(in Item,in Pattern,in bVisible)"}
   n149 = &RigVMFunction_MathBoolNot(Value=@n148.BoolValue)
   > &RigUnit_SetControlVisibility(Item=(Type=Control,Name="thigh_l_fk_ctrl"), bVisible=@n149.Result) {template="SetControlVisibility::Execute(in Item,in Pattern,in bVisible)"}
   > &RigUnit_SetControlVisibility(Item=(Type=Control,Name="calf_l_fk_ctrl"), bVisible=@n149.Result) {template="SetControlVisibility::Execute(in Item,in Pattern,in bVisible)"}
   > &RigUnit_SetControlVisibility(Item=(Type=Control,Name="foot_l_fk_ctrl"), bVisible=@n149.Result) {template="SetControlVisibility::Execute(in Item,in Pattern,in bVisible)"}
   > n150 = &RigUnit_SetControlVisibility(Item=(Type=Control,Name="ball_l_fk_ctrl"), bVisible=@n149.Result) {template="SetControlVisibility::Execute(in Item,in Pattern,in bVisible)"}
   Branch_1_RigVMUnitNode = &RigVMFunction_ControlFlowBranch(Condition=@n148.BoolValue, ExecuteContext=@n150.ExecutePin)
   n151 = &RigUnit_CollectionItems(Items=((Type=Bone,Name="thigh_l"),(Type=Bone,Name="calf_l"),(Type=Bone,Name="foot_l"),(Type=Bone,Name="ball_l")))
   n152 = &RigUnit_CollectionLoop(ExecutePin=@Branch_1_RigVMUnitNode.False, Collection=@n151.Collection)
   n153 = &RigVMFunction_NameConcat(B=_fk_ctrl, A=@n152.Item.Name) {template="Concat::Execute(in A,in B,out Result)"}
   n154 = &RigUnit_ProjectTransformToNewParent(
    Child=(Type=Bone,Name="leg_l_pv_ik_ctrl")
    NewParent=(Type=Control,Name="thigh_l")
    OldParent=(Type=Control,Name="thigh_l")
    bChildInitial=True
    bOldParentInitial=True
    Child.Name=@n152.Item.Name
    NewParent.Name=@n153.Result
    OldParent.Name=@n153.Result
   )
   > &RigUnit_SetTransform(Item=(Type=Bone), Weight=1, Item.Name=@n152.Item.Name, Value=@n154.Transform)
   n155 = &RigUnit_ProjectTransformToNewParent(
    Child=(Type=Control,Name="leg_l_pv_ik_ctrl")
    NewParent=(Type=Bone,Name="thigh_l")
    OldParent=(Type=Bone,Name="thigh_l")
    bChildInitial=True
    bOldParentInitial=True
   )
   &RigUnit_SetTransform(Item=(Type=Control,Name="leg_l_pv_ik_ctrl"), Weight=1, ExecutePin=@n152.Completed, Value=@n155.Transform)
   n156 = &RigUnit_ProjectTransformToNewParent(
    Child=(Type=Control,Name="foot_l_ik_ctrl")
    NewParent=(Type=Bone,Name="foot_l")
    OldParent=(Type=Bone,Name="foot_l")
    bChildInitial=True
    bOldParentInitial=True
   )
   > &RigUnit_SetTransform(Item=(Type=Control,Name="foot_l_ik_ctrl"), Weight=1, Value=@n156.Transform)
   n157 = &RigUnit_ProjectTransformToNewParent(
    Child=(Type=Control,Name="ball_l_ik_ctrl")
    NewParent=(Type=Bone,Name="ball_l")
    OldParent=(Type=Bone,Name="ball_l")
    bChildInitial=True
    bOldParentInitial=True
   )
   > n158 = &RigUnit_SetTransform(Item=(Type=Control,Name="ball_l_ik_ctrl"), Weight=1, Value=@n157.Transform)
   n159 = RigVMVariableNode() {variable=leg_l_ik_mode}
   Branch_4_2_RigVMUnitNode = &RigVMFunction_ControlFlowBranch(ExecuteContext=@n158.ExecutePin, Condition=@n159.Value)
   n160 = &RigUnit_CollectionItems(Items="((Type=Control,Name=\"thigh_l_fk_ctrl\"),(Type=Control,Name=\"calf_l_fk_ctrl\"),(Type=Control,Name=\"foot_l_fk_ctrl\"),(Type=Control,Name=\"ball_l_fk_ctrl\"),(Type=Control,Name=\"leg_l_pv_ik_ctrl\"),(Type=Control,Name=\"foot_l_ik_ctrl\"),(Type=Control,Name=\"ball_l_ik_ ...(truncated)")
   n161 = &RigUnit_CollectionLoop(ExecutePin=@Branch_4_2_RigVMUnitNode.True, Collection=@n160.Collection)
   > &RigUnit_SendEvent(Event=RequestAutoKey, bOnlyDuringInteraction=True, Item=@n161.Item, bEnable=@n159.Value) {template="SendEvent::Execute(in Event,in Item,in OffsetInSeconds,in bEnable,in bOnlyDuringInteraction)"}
   RigVMVariableNode(Value=false, ExecuteContext=@n161.Completed) {variable=leg_l_ik_mode}
   n162 = &RigUnit_ProjectTransformToNewParent(
    Child=(Type=Bone,Name="foot_l")
    NewParent=(Type=Control,Name="foot_l_ik_ctrl")
    OldParent=(Type=Control,Name="foot_l_ik_ctrl")
    bChildInitial=True
    bOldParentInitial=True
   )
   &RigUnit_TwoBoneIKSimplePerItem(
    DebugSettings=(Scale=10,WorldOffset=(Rotation=(X=0,Y=0,Z=0,W=1),Translation=(X=0,Y=0,Z=0),Scale3D=(X=1,Y=1,Z=1)))
    EffectorItem=(Type=Bone,Name="foot_l")
    ItemA=(Type=Bone,Name="thigh_l")
    ItemALength=0
    ItemB=(Type=Bone,Name="calf_l")
    ItemBLength=0
    PoleVectorKind=Location
    PoleVectorSpace=(Type=Bone)
    PrimaryAxis=(X=-1,Y=0,Z=0)
    SecondaryAxis=(X=0,Y=1,Z=0)
    SecondaryAxisWeight=1
    StretchMaximumRatio=1.1
    StretchStartRatio=1
    Weight=1
    bEnableStretch=true
    bPropagateToChildren=True
    ExecutePin=@Branch_1_RigVMUnitNode.True
    PoleVector=&RigUnit_GetTransform(Item=(Type=Control,Name="leg_l_pv_ik_ctrl")).Transform.Translation
    Effector=@n162.Transform
   )
   > &RigUnit_DebugLineItemSpace(
    Color=(R=0,G=0,B=0.5,A=1)
    DebugDrawSettings=(DepthPriority=SDPG_Foreground,Lifetime=-1)
    Space=(Type=Bone)
    Thickness=0.5
    WorldOffset=(Rotation=(X=0,Y=0,Z=0,W=1),Translation=(X=0,Y=0,Z=0),Scale3D=(X=1,Y=1,Z=1))
    bEnabled=True
    A=&RigUnit_GetTransform(Item=(Type=Bone,Name="calf_l")).Transform.Translation
    B=&RigUnit_GetTransform(Item=(Type=Control,Name="leg_l_pv_ik_ctrl")).Transform.Translation
   )
   n163 = &RigUnit_ProjectTransformToNewParent(
    Child=(Type=Bone,Name="ball_l")
    NewParent=(Type=Control,Name="ball_l_ik_ctrl")
    OldParent=(Type=Control,Name="ball_l_ik_ctrl")
    bChildInitial=True
    bOldParentInitial=True
   )
   > &RigUnit_SetTransform(Item=(Type=Bone,Name="ball_l"), Weight=1, Value=@n163.Transform)
   > n164 = &RigUnit_CollectionLoop(Collection=@n151.Collection)
   n165 = &RigVMFunction_NameConcat(B=_fk_ctrl, A=@n164.Item.Name) {template="Concat::Execute(in A,in B,out Result)"}
   > &RigUnit_SetTransform(
    Item=(Type=Control,Name="ball_l")
    Weight=1
    Item.Name=@n165.Result
    Value=&RigUnit_ProjectTransformToNewParent(
     Child=(Type=Control,Name="foot_l")
     bChildInitial=True
     bOldParentInitial=True
     NewParent=@n164.Item
     OldParent=@n164.Item
     Child.Name=@n165.Result
    ).Transform
   )
   n166 = RigVMVariableNode() {variable=leg_l_ik_mode}
   Branch_3_2_RigVMUnitNode = &RigVMFunction_ControlFlowBranch(ExecuteContext=@n164.Completed, Condition=@n166.Value)
   n167 = &RigUnit_CollectionLoop(ExecutePin=@Branch_3_2_RigVMUnitNode.False, Collection=@n160.Collection)
   > &RigUnit_SendEvent(Event=RequestAutoKey, bOnlyDuringInteraction=True, Item=@n167.Item, bEnable=@n166.Value) {template="SendEvent::Execute(in Event,in Item,in OffsetInSeconds,in bEnable,in bOnlyDuringInteraction)"}
   RigVMVariableNode(Value=true, ExecuteContext=@n167.Completed) {variable=leg_l_ik_mode}
   n168 = &RigUnit_GetControlBool(Control=leg_r_fk_ik_switch) {template="GetControlValue::Execute(in Control,out BoolValue)"}
   &RigUnit_SetControlVisibility(Item=(Type=Control,Name="foot_r_ik_ctrl"), bVisible=@n168.BoolValue, ExecutePin=@n147.B) {template="SetControlVisibility::Execute(in Item,in Pattern,in bVisible)"}
   > &RigUnit_SetControlVisibility(Item=(Type=Control,Name="leg_r_pv_ik_ctrl"), bVisible=@n168.BoolValue) {template="SetControlVisibility::Execute(in Item,in Pattern,in bVisible)"}
   > &RigUnit_SetControlVisibility(Item=(Type=Control,Name="ball_r_ik_ctrl"), bVisible=@n168.BoolValue) {template="SetControlVisibility::Execute(in Item,in Pattern,in bVisible)"}
   n169 = &RigVMFunction_MathBoolNot(Value=@n168.BoolValue)
   > &RigUnit_SetControlVisibility(Item=(Type=Control,Name="thigh_r_fk_ctrl"), bVisible=@n169.Result) {template="SetControlVisibility::Execute(in Item,in Pattern,in bVisible)"}
   > &RigUnit_SetControlVisibility(Item=(Type=Control,Name="calf_r_fk_ctrl"), bVisible=@n169.Result) {template="SetControlVisibility::Execute(in Item,in Pattern,in bVisible)"}
   > &RigUnit_SetControlVisibility(Item=(Type=Control,Name="foot_r_fk_ctrl"), bVisible=@n169.Result) {template="SetControlVisibility::Execute(in Item,in Pattern,in bVisible)"}
   > n170 = &RigUnit_SetControlVisibility(Item=(Type=Control,Name="ball_r_fk_ctrl"), bVisible=@n169.Result) {template="SetControlVisibility::Execute(in Item,in Pattern,in bVisible)"}
   Branch_1_1_RigVMUnitNode = &RigVMFunction_ControlFlowBranch(Condition=@n168.BoolValue, ExecuteContext=@n170.ExecutePin)
   n171 = &RigUnit_CollectionItems(Items=((Type=Bone,Name="thigh_r"),(Type=Bone,Name="calf_r"),(Type=Bone,Name="foot_r"),(Type=Bone,Name="ball_r")))
   n172 = &RigUnit_CollectionLoop(ExecutePin=@Branch_1_1_RigVMUnitNode.False, Collection=@n171.Collection)
   n173 = &RigVMFunction_NameConcat(B=_fk_ctrl, A=@n172.Item.Name) {template="Concat::Execute(in A,in B,out Result)"}
   > &RigUnit_SetTransform(
    Item=(Type=Bone,Name="leg_l_settings_ctrl")
    Weight=1
    Item.Name=@n172.Item.Name
    Value=&RigUnit_ProjectTransformToNewParent(
     Child=(Type=Bone)
     NewParent=(Type=Control)
     OldParent=(Type=Control)
     bChildInitial=True
     bOldParentInitial=True
     Child.Name=@n172.Item.Name
     NewParent.Name=@n173.Result
     OldParent.Name=@n173.Result
    ).Transform
   )
   n174 = &RigUnit_ProjectTransformToNewParent(
    Child=(Type=Control,Name="leg_r_pv_ik_ctrl")
    NewParent=(Type=Bone,Name="thigh_r")
    OldParent=(Type=Bone,Name="thigh_r")
    bChildInitial=True
    bOldParentInitial=True
   )
   &RigUnit_SetTransform(Item=(Type=Control,Name="leg_r_pv_ik_ctrl"), Weight=1, ExecutePin=@n172.Completed, Value=@n174.Transform)
   n175 = &RigUnit_ProjectTransformToNewParent(
    Child=(Type=Control,Name="foot_r_ik_ctrl")
    NewParent=(Type=Bone,Name="foot_r")
    OldParent=(Type=Bone,Name="foot_r")
    bChildInitial=True
    bOldParentInitial=True
   )
   > &RigUnit_SetTransform(Item=(Type=Control,Name="foot_r_ik_ctrl"), Weight=1, Value=@n175.Transform)
   n176 = &RigUnit_ProjectTransformToNewParent(
    Child=(Type=Control,Name="ball_r_ik_ctrl")
    NewParent=(Type=Bone,Name="ball_r")
    OldParent=(Type=Bone,Name="ball_r")
    bChildInitial=True
    bOldParentInitial=True
   )
   > n177 = &RigUnit_SetTransform(Item=(Type=Control,Name="ball_r_ik_ctrl"), Weight=1, Value=@n176.Transform)
   n178 = RigVMVariableNode() {variable=leg_r_ik_mode}
   Branch_4_2_1_RigVMUnitNode = &RigVMFunction_ControlFlowBranch(ExecuteContext=@n177.ExecutePin, Condition=@n178.Value)
   n179 = &RigUnit_CollectionItems(Items="((Type=Control,Name=\"thigh_r_fk_ctrl\"),(Type=Control,Name=\"calf_r_fk_ctrl\"),(Type=Control,Name=\"foot_r_fk_ctrl\"),(Type=Control,Name=\"ball_r_fk_ctrl\"),(Type=Control,Name=\"leg_r_pv_ik_ctrl\"),(Type=Control,Name=\"foot_r_ik_ctrl\"),(Type=Control,Name=\"ball_r_ik_ ...(truncated)")
   n180 = &RigUnit_CollectionLoop(ExecutePin=@Branch_4_2_1_RigVMUnitNode.True, Collection=@n179.Collection)
   > &RigUnit_SendEvent(Event=RequestAutoKey, bOnlyDuringInteraction=True, Item=@n180.Item, bEnable=@n178.Value) {template="SendEvent::Execute(in Event,in Item,in OffsetInSeconds,in bEnable,in bOnlyDuringInteraction)"}
   RigVMVariableNode(Value=false, ExecuteContext=@n180.Completed) {variable=leg_r_ik_mode}
   n181 = &RigUnit_ProjectTransformToNewParent(
    Child=(Type=Bone,Name="foot_r")
    NewParent=(Type=Control,Name="foot_r_ik_ctrl")
    OldParent=(Type=Control,Name="foot_r_ik_ctrl")
    bChildInitial=True
    bOldParentInitial=True
   )
   &RigUnit_TwoBoneIKSimplePerItem(
    DebugSettings=(Scale=10,WorldOffset=(Rotation=(X=0,Y=0,Z=0,W=1),Translation=(X=0,Y=0,Z=0),Scale3D=(X=1,Y=1,Z=1)))
    EffectorItem=(Type=Bone,Name="foot_r")
    ItemA=(Type=Bone,Name="thigh_r")
    ItemB=(Type=Bone,Name="calf_r")
    PoleVectorKind=Location
    PoleVectorSpace=(Type=Bone)
    PrimaryAxis=(X=1,Y=0,Z=0)
    SecondaryAxis=(X=0,Y=-1,Z=0)
    SecondaryAxisWeight=1
    StretchMaximumRatio=1.1
    StretchStartRatio=1
    Weight=1
    bEnableStretch=true
    bPropagateToChildren=True
    ExecutePin=@Branch_1_1_RigVMUnitNode.True
    PoleVector=&RigUnit_GetTransform(Item=(Type=Control,Name="leg_r_pv_ik_ctrl")).Transform.Translation
    Effector=@n181.Transform
   )
   > &RigUnit_DebugLineItemSpace(
    Color=(R=1,G=0,B=0,A=1)
    DebugDrawSettings=(DepthPriority=SDPG_Foreground,Lifetime=-1)
    Space=(Type=Bone)
    Thickness=0.5
    WorldOffset=(Rotation=(X=0,Y=0,Z=0,W=1),Translation=(X=0,Y=0,Z=0),Scale3D=(X=1,Y=1,Z=1))
    bEnabled=True
    B=&RigUnit_GetTransform(Item=(Type=Control,Name="leg_r_pv_ik_ctrl")).Transform.Translation
    A=&RigUnit_GetTransform(Item=(Type=Bone,Name="calf_r")).Transform.Translation
   )
   n182 = &RigUnit_ProjectTransformToNewParent(
    Child=(Type=Bone,Name="ball_r")
    NewParent=(Type=Control,Name="ball_r_ik_ctrl")
    OldParent=(Type=Control,Name="ball_r_ik_ctrl")
    bChildInitial=True
    bOldParentInitial=True
   )
   > &RigUnit_SetTransform(Item=(Type=Bone,Name="ball_r"), Weight=1, Value=@n182.Transform)
   > n183 = &RigUnit_CollectionLoop(Collection=@n171.Collection)
   n184 = &RigVMFunction_NameConcat(B=_fk_ctrl, A=@n183.Item.Name) {template="Concat::Execute(in A,in B,out Result)"}
   > &RigUnit_SetTransform(
    Item=(Type=Control,Name="None")
    Weight=1
    Item.Name=@n184.Result
    Value=&RigUnit_ProjectTransformToNewParent(
     Child=(Type=Control)
     NewParent=(Type=Bone)
     OldParent=(Type=Bone)
     bChildInitial=True
     bOldParentInitial=True
     NewParent.Name=@n183.Item.Name
     OldParent.Name=@n183.Item.Name
     Child.Name=@n184.Result
    ).Transform
   )
   n185 = RigVMVariableNode() {variable=leg_r_ik_mode}
   Branch_3_2_1_RigVMUnitNode = &RigVMFunction_ControlFlowBranch(ExecuteContext=@n183.Completed, Condition=@n185.Value)
   n186 = &RigUnit_CollectionLoop(ExecutePin=@Branch_3_2_1_RigVMUnitNode.False, Collection=@n179.Collection)
   > &RigUnit_SendEvent(Event=RequestAutoKey, bOnlyDuringInteraction=True, Item=@n186.Item, bEnable=@n185.Value) {template="SendEvent::Execute(in Event,in Item,in OffsetInSeconds,in bEnable,in bOnlyDuringInteraction)"}
   RigVMVariableNode(Value=true, ExecuteContext=@n186.Completed) {variable=leg_r_ik_mode}
   n187 = &RigUnit_CollectionItems(Items="((Type=Control,Name=\"thumb_01_l_ctrl\"),(Type=Control,Name=\"thumb_02_l_ctrl\"),(Type=Control,Name=\"thumb_03_l_ctrl\"),(Type=Control,Name=\"index_metacarpal_l_ctrl\"),(Type=Control,Name=\"index_01_l_ctrl\"),(Type=Control,Name=\"index_02_l_ctrl\"),(Type=Control,Name= ...(truncated)")
   n188 = &RigUnit_CollectionItems(Items="((Type=Control,Name=\"thumb_01_r_ctrl\"),(Type=Control,Name=\"thumb_02_r_ctrl\"),(Type=Control,Name=\"thumb_03_r_ctrl\"),(Type=Control,Name=\"index_metacarpal_r_ctrl\"),(Type=Control,Name=\"index_01_r_ctrl\"),(Type=Control,Name=\"index_02_r_ctrl\"),(Type=Control,Name= ...(truncated)")
   n189 = &RigUnit_CollectionItems(Items="((Type=Control,Name=\"neck_01_ctrl\"),(Type=Control,Name=\"neck_02_ctrl\"),(Type=Control,Name=\"head_ctrl\"),(Type=Control,Name=\"clavicle_l_ctrl\"),(Type=Control,Name=\"clavicle_r_ctrl\"),(Type=Control,Name=\"body_offset_ctrl\"),(Type=Control,Name=\"hips_ctrl\"),(Type= ...(truncated)")
   n190 = &RigUnit_CollectionLoop(
    Collection=/Script/ControlRig.RigUnit_CollectionUnion(A=&RigUnit_CollectionUnion(A=@n187.Collection, B=@n188.Collection).Collection, B=@n189.Collection) {is=struct RigVMUnitNode}.Collection
    ExecutePin=@n81.D
   )
   n191 = &RigUnit_GetControlBool(Control=ShowBodyControls) {template="GetControlValue::Execute(in Control,out BoolValue)"}
   > &RigUnit_SetControlVisibility(Item=(Type=Control), Item.Name=@n190.Item.Name, bVisible=@n191.BoolValue) {template="SetControlVisibility::Execute(in Item,in Pattern,in bVisible)"}
   n192 = &RigUnit_SequenceExecution(ExecuteContext=@n190.Completed)
   Branch_5_RigVMUnitNode = &RigVMFunction_ControlFlowBranch(
    Condition=/Script/RigVM.RigVMFunction_MathBoolEquals(
     B=false
     A=&RigUnit_GetControlBool(Control=arm_l_fk_ik_switch) {template="GetControlValue::Execute(in Control,out BoolValue)"}.BoolValue
    ) {is=struct RigVMUnitNode}.Result
    ExecuteContext=@n192.A
   )
   n193 = &RigUnit_CollectionLoop(
    ExecutePin=@Branch_5_RigVMUnitNode.False
    Collection=&RigUnit_CollectionItems(Items=((Type=Control,Name="hand_l_ik_ctrl"),(Type=Control,Name="arm_l_pv_ik_ctrl"))).Collection
   )
   > &RigUnit_SetControlVisibility(Item=(Type=Control), Item.Name=@n193.Item.Name, bVisible=@n191.BoolValue) {template="SetControlVisibility::Execute(in Item,in Pattern,in bVisible)"}
   n194 = &RigUnit_CollectionLoop(
    ExecutePin=@Branch_5_RigVMUnitNode.True
    Collection=&RigUnit_CollectionItems(Items=((Type=Control,Name="upperarm_l_fk_ctrl"),(Type=Control,Name="lowerarm_l_fk_ctrl"),(Type=Control,Name="hand_l_fk_ctrl"))).Collection
   )
   > &RigUnit_SetControlVisibility(Item=(Type=Control), Item.Name=@n194.Item.Name, bVisible=@n191.BoolValue) {template="SetControlVisibility::Execute(in Item,in Pattern,in bVisible)"}
   Branch_5_1_RigVMUnitNode = &RigVMFunction_ControlFlowBranch(
    Condition=&RigVMFunction_MathBoolEquals(
     B=false
     A=&RigUnit_GetControlBool(Control=arm_r_fk_ik_switch) {template="GetControlValue::Execute(in Control,out BoolValue)"}.BoolValue
    ).Result
    ExecuteContext=@n192.B
   )
   n195 = &RigUnit_CollectionLoop(
    ExecutePin=@Branch_5_1_RigVMUnitNode.True
    Collection=&RigUnit_CollectionItems(Items=((Type=Control,Name="upperarm_r_fk_ctrl"),(Type=Control,Name="lowerarm_r_fk_ctrl"),(Type=Control,Name="hand_r_fk_ctrl"))).Collection
   )
   > &RigUnit_SetControlVisibility(Item=(Type=Control), Item.Name=@n195.Item.Name, bVisible=@n191.BoolValue) {template="SetControlVisibility::Execute(in Item,in Pattern,in bVisible)"}
   n196 = &RigUnit_CollectionLoop(
    Collection=&RigUnit_CollectionItems(Items=((Type=Control,Name="hand_r_ik_ctrl"),(Type=Control,Name="arm_r_pv_ik_ctrl"))).Collection
    ExecutePin=@Branch_5_1_RigVMUnitNode.False
   )
   > &RigUnit_SetControlVisibility(Item=(Type=Control), Item.Name=@n196.Item.Name, bVisible=@n191.BoolValue) {template="SetControlVisibility::Execute(in Item,in Pattern,in bVisible)"}
   Branch_5_1_1_RigVMUnitNode = &RigVMFunction_ControlFlowBranch(
    Condition=&RigVMFunction_MathBoolEquals(
     B=false
     A=&RigUnit_GetControlBool(Control=leg_l_fk_ik_switch) {template="GetControlValue::Execute(in Control,out BoolValue)"}.BoolValue
    ).Result
    ExecuteContext=@n192.C
   )
   n197 = &RigUnit_CollectionItems(Items=((Type=Control,Name="thigh_l_fk_ctrl"),(Type=Control,Name="calf_l_fk_ctrl"),(Type=Control,Name="foot_l_fk_ctrl"),(Type=Control,Name="ball_l_fk_ctrl")))
   n198 = &RigUnit_CollectionLoop(ExecutePin=@Branch_5_1_1_RigVMUnitNode.True, Collection=@n197.Collection)
   > &RigUnit_SetControlVisibility(Item=(Type=Control), Item.Name=@n198.Item.Name, bVisible=@n191.BoolValue) {template="SetControlVisibility::Execute(in Item,in Pattern,in bVisible)"}
   n199 = &RigUnit_CollectionLoop(
    Collection=&RigUnit_CollectionItems(Items=((Type=Control,Name="foot_l_ik_ctrl"),(Type=Control,Name="ball_l_ik_ctrl"),(Type=Control,Name="leg_l_pv_ik_ctrl"))).Collection
    ExecutePin=@Branch_5_1_1_RigVMUnitNode.False
   )
   > &RigUnit_SetControlVisibility(Item=(Type=Control), Item.Name=@n199.Item.Name, bVisible=@n191.BoolValue) {template="SetControlVisibility::Execute(in Item,in Pattern,in bVisible)"}
   Branch_5_1_2_RigVMUnitNode = &RigVMFunction_ControlFlowBranch(
    Condition=&RigVMFunction_MathBoolEquals(
     B=false
     A=&RigUnit_GetControlBool(Control=leg_r_fk_ik_switch) {template="GetControlValue::Execute(in Control,out BoolValue)"}.BoolValue
    ).Result
    ExecuteContext=@n192.D
   )
   n200 = &RigUnit_CollectionItems(Items=((Type=Control,Name="thigh_r_fk_ctrl"),(Type=Control,Name="calf_r_fk_ctrl"),(Type=Control,Name="foot_r_fk_ctrl"),(Type=Control,Name="ball_r_fk_ctrl")))
   n201 = &RigUnit_CollectionLoop(ExecutePin=@Branch_5_1_2_RigVMUnitNode.True, Collection=@n200.Collection)
   > &RigUnit_SetControlVisibility(Item=(Type=Control), Item.Name=@n201.Item.Name, bVisible=@n191.BoolValue) {template="SetControlVisibility::Execute(in Item,in Pattern,in bVisible)"}
   n202 = &RigUnit_CollectionLoop(
    Collection=&RigUnit_CollectionItems(Items=((Type=Control,Name="foot_r_ik_ctrl"),(Type=Control,Name="ball_r_ik_ctrl"),(Type=Control,Name="leg_r_pv_ik_ctrl"))).Collection
    ExecutePin=@Branch_5_1_2_RigVMUnitNode.False
   )
   > &RigUnit_SetControlVisibility(Item=(Type=Control), Item.Name=@n202.Item.Name, bVisible=@n191.BoolValue) {template="SetControlVisibility::Execute(in Item,in Pattern,in bVisible)"}
  ]}
]
hierarchy [
 global_ctrl {controlType=EULER_TRANSFORM, type=Control, loc=<0 0 0>, rot=<0 0 0 1>}
 root_ctrl {controlType=EULER_TRANSFORM, parent=global_ctrl, type=Control, loc=<0 0 0>, rot=<0 0 0 1>}
 body_offset_ctrl {controlType=EULER_TRANSFORM, parent=global_ctrl, type=Control, loc=<0 0 0>, rot=<0 0 0 1>}
 leg_l_controls_space {parent=body_offset_ctrl, type=Null, loc=<0 0 0>, rot=<0 0 0 1>}
 thigh_l_fk_ctrl_space {parent=leg_l_controls_space, type=Null, loc=<8.7855 -1.2823 84.6044>, rot=<-0.0087 -0.7071 -0.028 0.7065>}
 thigh_l_fk_ctrl {controlType=EULER_TRANSFORM, parent=thigh_l_fk_ctrl_space, type=Control, loc=<0 0 0>, rot=<0 0 0 1>}
 calf_l_fk_ctrl {controlType=EULER_TRANSFORM, parent=thigh_l_fk_ctrl, type=Control, loc=<0 0 0>, rot=<0 0 0 1>}
 foot_l_fk_ctrl {controlType=EULER_TRANSFORM, parent=calf_l_fk_ctrl, type=Control, loc=<0 0 0>, rot=<0 0 0 1>}
 ball_l_fk_ctrl {controlType=EULER_TRANSFORM, parent=foot_l_fk_ctrl, type=Control, loc=<0 0 0>, rot=<0 0 0 1>}
 foot_l_ik_ctrl {controlType=EULER_TRANSFORM, parent=leg_l_controls_space, type=Control, loc=<0 0 0>, rot=<0 0 0 1>}
 ball_l_ik_ctrl {controlType=EULER_TRANSFORM, parent=foot_l_ik_ctrl, type=Control, loc=<0 0 0>, rot=<0 0 0 1>}
 leg_l_pv_ik_ctrl {controlType=EULER_TRANSFORM, parent=leg_l_controls_space, type=Control, loc=<0 0 0>, rot=<0 0 0 1>}
 leg_r_controls_space {parent=body_offset_ctrl, type=Null, loc=<0 0 0>, rot=<0 0 0 1>, scale=<-1 1 1>}
 thigh_r_fk_ctrl_space {parent=leg_r_controls_space, type=Null, loc=<8.7855 -1.2823 84.6044>, rot=<-0.0087 -0.7071 -0.028 0.7065>}
 thigh_r_fk_ctrl {controlType=EULER_TRANSFORM, parent=thigh_r_fk_ctrl_space, type=Control, loc=<0 0 0>, rot=<0 0 0 1>}
 calf_r_fk_ctrl {controlType=EULER_TRANSFORM, parent=thigh_r_fk_ctrl, type=Control, loc=<0 0 0>, rot=<0 0 0 1>}
 foot_r_fk_ctrl {controlType=EULER_TRANSFORM, parent=calf_r_fk_ctrl, type=Control, loc=<0 0 0>, rot=<0 0 0 1>}
 ball_r_fk_ctrl {controlType=EULER_TRANSFORM, parent=foot_r_fk_ctrl, type=Control, loc=<0 0 0>, rot=<0 0 0 1>}
 foot_r_ik_ctrl {controlType=EULER_TRANSFORM, parent=leg_r_controls_space, type=Control, loc=<0 0 0>, rot=<0 0 0 1>}
 ball_r_ik_ctrl {controlType=EULER_TRANSFORM, parent=foot_r_ik_ctrl, type=Control, loc=<0 0 0>, rot=<0 0 0 1>}
 leg_r_pv_ik_ctrl {controlType=EULER_TRANSFORM, parent=leg_r_controls_space, type=Control, loc=<0 0 0>, rot=<0 0 0 1>}
 arm_l_controls_space {parent=body_offset_ctrl, type=Null, loc=<0 0 0>, rot=<0 0 0 1>}
 upperarm_l_fk_ctrl_space {parent=arm_l_controls_space, type=Null, loc=<13.966 -4.7181 128.0171>, rot=<-0.0069 0.5199 -0.0278 0.8537>}
 upperarm_l_fk_ctrl {controlType=EULER_TRANSFORM, parent=upperarm_l_fk_ctrl_space, type=Control, loc=<0 0 0>, rot=<0 0 0 1>}
 lowerarm_l_fk_ctrl {controlType=EULER_TRANSFORM, parent=upperarm_l_fk_ctrl, type=Control, loc=<0 0 0>, rot=<0 0 0 1>}
 hand_l_fk_ctrl {controlType=EULER_TRANSFORM, parent=lowerarm_l_fk_ctrl, type=Control, loc=<0 0 0>, rot=<0 0 0 1>}
 arm_l_fk_local {controlType=BOOL, parent=upperarm_l_fk_ctrl, type=Control, loc=<1 0 0>, rot=<0 0 0 1>}
 fingers_l_space {parent=arm_l_controls_space, type=Null, loc=<39.5619 6.118 82.7825>, rot=<0.4804 0.5538 -0.1448 0.6645>}
 thumb_01_l_ctrl {controlType=EULER_TRANSFORM, parent=fingers_l_space, type=Control, loc=<0 0 0>, rot=<0 0 0 1>}
 thumb_02_l_ctrl {controlType=EULER_TRANSFORM, parent=thumb_01_l_ctrl, type=Control, loc=<0 0 0>, rot=<0 0 0 1>}
 thumb_03_l_ctrl {controlType=EULER_TRANSFORM, parent=thumb_02_l_ctrl, type=Control, loc=<0 0 0>, rot=<0 0 0 1>}
 index_metacarpal_l_ctrl {controlType=EULER_TRANSFORM, parent=fingers_l_space, type=Control, loc=<0 0 0>, rot=<0 0 0 1>}
 index_01_l_ctrl {controlType=EULER_TRANSFORM, parent=index_metacarpal_l_ctrl, type=Control, loc=<0 0 0>, rot=<0 0 0 1>}
 index_02_l_ctrl {controlType=EULER_TRANSFORM, parent=index_01_l_ctrl, type=Control, loc=<0 0 0>, rot=<0 0 0 1>}
 index_03_l_ctrl {controlType=EULER_TRANSFORM, parent=index_02_l_ctrl, type=Control, loc=<0 0 0>, rot=<0 0 0 1>}
 middle_metacarpal_l_ctrl {controlType=EULER_TRANSFORM, parent=fingers_l_space, type=Control, loc=<0 0 0>, rot=<0 0 0 1>}
 middle_01_l_ctrl {controlType=EULER_TRANSFORM, parent=middle_metacarpal_l_ctrl, type=Control, loc=<0 0 0>, rot=<0 0 0 1>}
 middle_02_l_ctrl {controlType=EULER_TRANSFORM, parent=middle_01_l_ctrl, type=Control, loc=<0 0 0>, rot=<0 0 0 1>}
 middle_03_l_ctrl {controlType=EULER_TRANSFORM, parent=middle_02_l_ctrl, type=Control, loc=<0 0 0>, rot=<0 0 0 1>}
 ring_metacarpal_l_ctrl {controlType=EULER_TRANSFORM, parent=fingers_l_space, type=Control, loc=<0 0 0>, rot=<0 0 0 1>}
 ring_01_l_ctrl {controlType=EULER_TRANSFORM, parent=ring_metacarpal_l_ctrl, type=Control, loc=<0 0 0>, rot=<0 0 0 1>}
 ring_02_l_ctrl {controlType=EULER_TRANSFORM, parent=ring_01_l_ctrl, type=Control, loc=<0 0 0>, rot=<0 0 0 1>}
 ring_03_l_ctrl {controlType=EULER_TRANSFORM, parent=ring_02_l_ctrl, type=Control, loc=<0 0 0>, rot=<0 0 0 1>}
 pinky_metacarpal_l_ctrl {controlType=EULER_TRANSFORM, parent=fingers_l_space, type=Control, loc=<0 0 0>, rot=<0 0 0 1>}
 pinky_01_l_ctrl {controlType=EULER_TRANSFORM, parent=pinky_metacarpal_l_ctrl, type=Control, loc=<0 0 0>, rot=<0 0 0 1>}
 pinky_02_l_ctrl {controlType=EULER_TRANSFORM, parent=pinky_01_l_ctrl, type=Control, loc=<0 0 0>, rot=<0 0 0 1>}
 pinky_03_l_ctrl {controlType=EULER_TRANSFORM, parent=pinky_02_l_ctrl, type=Control, loc=<0 0 0>, rot=<0 0 0 1>}
 hand_l_ik_ctrl {controlType=EULER_TRANSFORM, parent=arm_l_controls_space, type=Control, loc=<0 0 0>, rot=<0 0 0 1>}
 arm_l_pv_ik_ctrl {controlType=EULER_TRANSFORM, parent=arm_l_controls_space, type=Control, loc=<0 0 0>, rot=<0 0 0 1>}
 arm_r_controls_space {parent=body_offset_ctrl, type=Null, loc=<0 0 0>, rot=<0 0 0 1>, scale=<-1 1 1>}
 upperarm_r_fk_ctrl_space {parent=arm_r_controls_space, type=Null, loc=<13.966 -4.7181 128.0171>, rot=<-0.0069 0.5199 -0.0278 0.8537>}
 upperarm_r_fk_ctrl {controlType=EULER_TRANSFORM, parent=upperarm_r_fk_ctrl_space, type=Control, loc=<0 0 0>, rot=<0 0 0 1>}
 lowerarm_r_fk_ctrl {controlType=EULER_TRANSFORM, parent=upperarm_r_fk_ctrl, type=Control, loc=<0 0 0>, rot=<0 0 0 1>}
 hand_r_fk_ctrl {controlType=EULER_TRANSFORM, parent=lowerarm_r_fk_ctrl, type=Control, loc=<0 0 0>, rot=<0 0 0 1>}
 arm_r_fk_local {controlType=BOOL, parent=upperarm_r_fk_ctrl, type=Control, loc=<1 0 0>, rot=<0 0 0 1>}
 fingers_r_space {parent=arm_r_controls_space, type=Null, loc=<39.5619 6.118 82.7825>, rot=<0.4804 0.5538 -0.1448 0.6645>}
 thumb_01_r_ctrl {controlType=EULER_TRANSFORM, parent=fingers_r_space, type=Control, loc=<0 0 0>, rot=<0 0 0 1>}
 thumb_02_r_ctrl {controlType=EULER_TRANSFORM, parent=thumb_01_r_ctrl, type=Control, loc=<0 0 0>, rot=<0 0 0 1>}
 thumb_03_r_ctrl {controlType=EULER_TRANSFORM, parent=thumb_02_r_ctrl, type=Control, loc=<0 0 0>, rot=<0 0 0 1>}
 index_metacarpal_r_ctrl {controlType=EULER_TRANSFORM, parent=fingers_r_space, type=Control, loc=<0 0 0>, rot=<0 0 0 1>}
 index_01_r_ctrl {controlType=EULER_TRANSFORM, parent=index_metacarpal_r_ctrl, type=Control, loc=<0 0 0>, rot=<0 0 0 1>}
 index_02_r_ctrl {controlType=EULER_TRANSFORM, parent=index_01_r_ctrl, type=Control, loc=<0 0 0>, rot=<0 0 0 1>}
 index_03_r_ctrl {controlType=EULER_TRANSFORM, parent=index_02_r_ctrl, type=Control, loc=<0 0 0>, rot=<0 0 0 1>}
 middle_metacarpal_r_ctrl {controlType=EULER_TRANSFORM, parent=fingers_r_space, type=Control, loc=<0 0 0>, rot=<0 0 0 1>}
 middle_01_r_ctrl {controlType=EULER_TRANSFORM, parent=middle_metacarpal_r_ctrl, type=Control, loc=<0 0 0>, rot=<0 0 0 1>}
 middle_02_r_ctrl {controlType=EULER_TRANSFORM, parent=middle_01_r_ctrl, type=Control, loc=<0 0 0>, rot=<0 0 0 1>}
 middle_03_r_ctrl {controlType=EULER_TRANSFORM, parent=middle_02_r_ctrl, type=Control, loc=<0 0 0>, rot=<0 0 0 1>}
 ring_metacarpal_r_ctrl {controlType=EULER_TRANSFORM, parent=fingers_r_space, type=Control, loc=<0 0 0>, rot=<0 0 0 1>}
 ring_01_r_ctrl {controlType=EULER_TRANSFORM, parent=ring_metacarpal_r_ctrl, type=Control, loc=<0 0 0>, rot=<0 0 0 1>}
 ring_02_r_ctrl {controlType=EULER_TRANSFORM, parent=ring_01_r_ctrl, type=Control, loc=<0 0 0>, rot=<0 0 0 1>}
 ring_03_r_ctrl {controlType=EULER_TRANSFORM, parent=ring_02_r_ctrl, type=Control, loc=<0 0 0>, rot=<0 0 0 1>}
 pinky_metacarpal_r_ctrl {controlType=EULER_TRANSFORM, parent=fingers_r_space, type=Control, loc=<0 0 0>, rot=<0 0 0 1>}
 pinky_01_r_ctrl {controlType=EULER_TRANSFORM, parent=pinky_metacarpal_r_ctrl, type=Control, loc=<0 0 0>, rot=<0 0 0 1>}
 pinky_02_r_ctrl {controlType=EULER_TRANSFORM, parent=pinky_01_r_ctrl, type=Control, loc=<0 0 0>, rot=<0 0 0 1>}
 pinky_03_r_ctrl {controlType=EULER_TRANSFORM, parent=pinky_02_r_ctrl, type=Control, loc=<0 0 0>, rot=<0 0 0 1>}
 hand_r_ik_ctrl {controlType=EULER_TRANSFORM, parent=arm_r_controls_space, type=Control, loc=<0 0 0>, rot=<0 0 0 1>}
 arm_r_pv_ik_ctrl {controlType=EULER_TRANSFORM, parent=arm_r_controls_space, type=Control, loc=<0 0 0>, rot=<0 0 0 1>}
 body_ctrl {controlType=EULER_TRANSFORM, parent=body_offset_ctrl, type=Control, loc=<0 0 0>, rot=<0 0 0 1>}
 hips_ctrl {controlType=EULER_TRANSFORM, parent=body_ctrl, type=Control, loc=<0 0 0>, rot=<0 0 0 1>}
 spine_01_ctrl {controlType=EULER_TRANSFORM, parent=body_ctrl, type=Control, loc=<0 0 0>, rot=<0 0 0 1>}
 spine_02_ctrl {controlType=EULER_TRANSFORM, parent=spine_01_ctrl, type=Control, loc=<0 0 0>, rot=<0 0 0 1>}
 spine_03_ctrl {controlType=EULER_TRANSFORM, parent=spine_02_ctrl, type=Control, loc=<0 0 0>, rot=<0 0 0 1>}
 head_ctrl_space {parent=spine_03_ctrl, type=Null, loc=<0 1.0117 34.8021>, rot=<0 0 0 1>}
 head_ctrl {controlType=EULER_TRANSFORM, parent=head_ctrl_space, type=Control, loc=<0 0 0>, rot=<0 0 0 1>}
 head_local {controlType=BOOL, parent=head_ctrl, type=Control, loc=<1 0 0>, rot=<0 0 0 1>}
 clavicle_l_ctrl {controlType=EULER_TRANSFORM, parent=spine_03_ctrl, type=Control, loc=<0 0 0>, rot=<0 0 0 1>}
 clavicle_r_ctrl {controlType=EULER_TRANSFORM, parent=spine_03_ctrl, type=Control, loc=<0 0 0>, rot=<0 0 0 1>}
 neck_01_ctrl {controlType=EULER_TRANSFORM, parent=spine_03_ctrl, type=Control, loc=<0 0 0>, rot=<0 0 0 1>}
 neck_02_ctrl {controlType=EULER_TRANSFORM, parent=neck_01_ctrl, type=Control, loc=<0 0 0>, rot=<0 0 0 1>}
 arm_l_fk_ik_switch {controlType=BOOL, parent=global_ctrl, type=Control, loc=<0 0 0>, rot=<0 0 0 1>}
 leg_l_fk_ik_switch {controlType=BOOL, parent=global_ctrl, type=Control, loc=<1 0 0>, rot=<0 0 0 1>}
 arm_r_fk_ik_switch {controlType=BOOL, parent=global_ctrl, type=Control, loc=<0 0 0>, rot=<0 0 0 1>}
 leg_r_fk_ik_switch {controlType=BOOL, parent=global_ctrl, type=Control, loc=<1 0 0>, rot=<0 0 0 1>}
 ShowBodyControls {controlType=BOOL, parent=global_ctrl, type=Control, loc=<1 0 0>, rot=<0 0 0 1>}
 eyeBlinkLeft {type=Curve, loc=<0 0 0>, rot=<0 0 0 1>}
 eyeBlinkRight {type=Curve, loc=<0 0 0>, rot=<0 0 0 1>}
 eyeSquintLeft {type=Curve, loc=<0 0 0>, rot=<0 0 0 1>}
 eyeSquintRight {type=Curve, loc=<0 0 0>, rot=<0 0 0 1>}
 cheekSquintLeft {type=Curve, loc=<0 0 0>, rot=<0 0 0 1>}
 cheekSquintRight {type=Curve, loc=<0 0 0>, rot=<0 0 0 1>}
 eyeWideLeft {type=Curve, loc=<0 0 0>, rot=<0 0 0 1>}
 eyeWideRight {type=Curve, loc=<0 0 0>, rot=<0 0 0 1>}
 eyeLookUpLeft {type=Curve, loc=<0 0 0>, rot=<0 0 0 1>}
 eyeLookUpRight {type=Curve, loc=<0 0 0>, rot=<0 0 0 1>}
 eyeLookDownLeft {type=Curve, loc=<0 0 0>, rot=<0 0 0 1>}
 eyeLookDownRight {type=Curve, loc=<0 0 0>, rot=<0 0 0 1>}
 eyeLookOutLeft {type=Curve, loc=<0 0 0>, rot=<0 0 0 1>}
 eyeLookInRight {type=Curve, loc=<0 0 0>, rot=<0 0 0 1>}
 eyeLookInLeft {type=Curve, loc=<0 0 0>, rot=<0 0 0 1>}
 eyeLookOutRight {type=Curve, loc=<0 0 0>, rot=<0 0 0 1>}
 eye_blinkElookUp_L {type=Curve, loc=<0 0 0>, rot=<0 0 0 1>}
 eye_blinkElookUp_R {type=Curve, loc=<0 0 0>, rot=<0 0 0 1>}
 eye_blinkElookDown_L {type=Curve, loc=<0 0 0>, rot=<0 0 0 1>}
 eye_blinkElookDown_R {type=Curve, loc=<0 0 0>, rot=<0 0 0 1>}
 eye_blinkElookLeft_L {type=Curve, loc=<0 0 0>, rot=<0 0 0 1>}
 eye_blinkElookLeft_R {type=Curve, loc=<0 0 0>, rot=<0 0 0 1>}
 eye_blinkElookRight_L {type=Curve, loc=<0 0 0>, rot=<0 0 0 1>}
 eye_blinkElookRight_R {type=Curve, loc=<0 0 0>, rot=<0 0 0 1>}
 eye_lookUpLeft_L {type=Curve, loc=<0 0 0>, rot=<0 0 0 1>}
 eye_lookUpLeft_R {type=Curve, loc=<0 0 0>, rot=<0 0 0 1>}
 eye_lookUpRight_L {type=Curve, loc=<0 0 0>, rot=<0 0 0 1>}
 eye_lookUpRight_R {type=Curve, loc=<0 0 0>, rot=<0 0 0 1>}
 eye_lookDownLeft_L {type=Curve, loc=<0 0 0>, rot=<0 0 0 1>}
 eye_lookDownLeft_R {type=Curve, loc=<0 0 0>, rot=<0 0 0 1>}
 eye_lookDownRight_L {type=Curve, loc=<0 0 0>, rot=<0 0 0 1>}
 eye_lookDownRight_R {type=Curve, loc=<0 0 0>, rot=<0 0 0 1>}
 eye_blinkElookUpLeft_L {type=Curve, loc=<0 0 0>, rot=<0 0 0 1>}
 eye_blinkElookUpLeft_R {type=Curve, loc=<0 0 0>, rot=<0 0 0 1>}
 eye_blinkElookUpRight_L {type=Curve, loc=<0 0 0>, rot=<0 0 0 1>}
 eye_blinkElookUpRight_R {type=Curve, loc=<0 0 0>, rot=<0 0 0 1>}
 eye_blinkElookDownLeft_L {type=Curve, loc=<0 0 0>, rot=<0 0 0 1>}
 eye_blinkElookDownLeft_R {type=Curve, loc=<0 0 0>, rot=<0 0 0 1>}
 eye_blinkElookDownRight_L {type=Curve, loc=<0 0 0>, rot=<0 0 0 1>}
 eye_blinkElookDownRight_R {type=Curve, loc=<0 0 0>, rot=<0 0 0 1>}
 eye_widenElookDown_L {type=Curve, loc=<0 0 0>, rot=<0 0 0 1>}
 eye_widenElookDown_R {type=Curve, loc=<0 0 0>, rot=<0 0 0 1>}
 eye_squintInnerEblink_L {type=Curve, loc=<0 0 0>, rot=<0 0 0 1>}
 eye_squintInnerEblink_R {type=Curve, loc=<0 0 0>, rot=<0 0 0 1>}
 eye_cheekRaiseEsquintInner_L {type=Curve, loc=<0 0 0>, rot=<0 0 0 1>}
 eye_cheekRaiseEsquintInner_R {type=Curve, loc=<0 0 0>, rot=<0 0 0 1>}
 eye_cheekRaiseEblink_L {type=Curve, loc=<0 0 0>, rot=<0 0 0 1>}
 eye_cheekRaiseEblink_R {type=Curve, loc=<0 0 0>, rot=<0 0 0 1>}
 eye_cheekRaiseEsquintInnerEblink_L {type=Curve, loc=<0 0 0>, rot=<0 0 0 1>}
 eye_cheekRaiseEsquintInnerEblink_R {type=Curve, loc=<0 0 0>, rot=<0 0 0 1>}
 eye_cheekRaiseNwrinkle_L {type=Curve, loc=<0 0 0>, rot=<0 0 0 1>}
 eye_cheekRaiseNwrinkle_R {type=Curve, loc=<0 0 0>, rot=<0 0 0 1>}
 mouth_cheekRaiseCornerPull_L {type=Curve, loc=<0 0 0>, rot=<0 0 0 1>}
 mouth_cheekRaiseCornerPull_R {type=Curve, loc=<0 0 0>, rot=<0 0 0 1>}
 L_BipedArm_A_Elbow_Jnt_0_0_0 {type=Curve, loc=<0 0 0>, rot=<0 0 0 1>}
 L_BipedArm_A_Elbow_Jnt_0_0_n60 {type=Curve, loc=<0 0 0>, rot=<0 0 0 1>}
 L_BipedArm_A_Elbow_Jnt_0_0_n90 {type=Curve, loc=<0 0 0>, rot=<0 0 0 1>}
 L_BipedArm_A_Elbow_Jnt_0_0_30 {type=Curve, loc=<0 0 0>, rot=<0 0 0 1>}
 R_BipedArm_A_Elbow_Jnt_0_0_0 {type=Curve, loc=<0 0 0>, rot=<0 0 0 1>}
 R_BipedArm_A_Elbow_Jnt_0_0_n60 {type=Curve, loc=<0 0 0>, rot=<0 0 0 1>}
 R_BipedArm_A_Elbow_Jnt_0_0_n90 {type=Curve, loc=<0 0 0>, rot=<0 0 0 1>}
 R_BipedArm_A_Elbow_Jnt_0_0_30 {type=Curve, loc=<0 0 0>, rot=<0 0 0 1>}
 L_BipedArm_A_Wrist_Jnt_0_0_0 {type=Curve, loc=<0 0 0>, rot=<0 0 0 1>}
 L_BipedArm_A_Wrist_Jnt_0_n60_0 {type=Curve, loc=<0 0 0>, rot=<0 0 0 1>}
 L_BipedArm_A_Wrist_Jnt_0_n120_0 {type=Curve, loc=<0 0 0>, rot=<0 0 0 1>}
 L_BipedArm_A_Wrist_Jnt_0_0_n60 {type=Curve, loc=<0 0 0>, rot=<0 0 0 1>}
 L_BipedArm_A_Wrist_Jnt_0_0_60 {type=Curve, loc=<0 0 0>, rot=<0 0 0 1>}
 L_BipedArm_A_Elbow_Jnt_0_0_n120 {type=Curve, loc=<0 0 0>, rot=<0 0 0 1>}
 R_BipedArm_A_Elbow_Jnt_0_0_n120 {type=Curve, loc=<0 0 0>, rot=<0 0 0 1>}
 R_BipedArm_A_Wrist_Jnt_0_0_0 {type=Curve, loc=<0 0 0>, rot=<0 0 0 1>}
 L_BipedArm_A_Wrist_Jnt_0_60_0 {type=Curve, loc=<0 0 0>, rot=<0 0 0 1>}
 R_BipedArm_A_Wrist_Jnt_0_n60_0 {type=Curve, loc=<0 0 0>, rot=<0 0 0 1>}
 L_BipedArm_A_Shoulder_Jnt_0_35_0 {type=Curve, loc=<0 0 0>, rot=<0 0 0 1>}
 L_BipedArm_A_Shoulder_Jnt_0_n55_0 {type=Curve, loc=<0 0 0>, rot=<0 0 0 1>}
 L_BipedArm_A_Shoulder_Jnt_0_0_n60 {type=Curve, loc=<0 0 0>, rot=<0 0 0 1>}
 L_BipedArm_A_Shoulder_Jnt_0_0_n120 {type=Curve, loc=<0 0 0>, rot=<0 0 0 1>}
 L_BipedArm_A_Shoulder_Jnt_0_0_45 {type=Curve, loc=<0 0 0>, rot=<0 0 0 1>}
 L_BipedArm_A_Shoulder_Jnt_0_0_90 {type=Curve, loc=<0 0 0>, rot=<0 0 0 1>}
 L_BipedArm_A_Shoulder_Jnt_0_n45_n90 {type=Curve, loc=<0 0 0>, rot=<0 0 0 1>}
 L_BipedArm_A_Shoulder_Jnt_0_n45_n150 {type=Curve, loc=<0 0 0>, rot=<0 0 0 1>}
 R_BipedArm_A_Shoulder_Jnt_0_35_0 {type=Curve, loc=<0 0 0>, rot=<0 0 0 1>}
 R_BipedArm_A_Shoulder_Jnt_0_n55_0 {type=Curve, loc=<0 0 0>, rot=<0 0 0 1>}
 R_BipedArm_A_Shoulder_Jnt_0_0_n60 {type=Curve, loc=<0 0 0>, rot=<0 0 0 1>}
 R_BipedArm_A_Shoulder_Jnt_0_0_n120 {type=Curve, loc=<0 0 0>, rot=<0 0 0 1>}
 R_BipedArm_A_Shoulder_Jnt_0_0_45 {type=Curve, loc=<0 0 0>, rot=<0 0 0 1>}
 R_BipedArm_A_Shoulder_Jnt_0_0_90 {type=Curve, loc=<0 0 0>, rot=<0 0 0 1>}
 R_BipedArm_A_Shoulder_Jnt_0_n45_n90 {type=Curve, loc=<0 0 0>, rot=<0 0 0 1>}
 R_BipedArm_A_Shoulder_Jnt_0_n45_n150 {type=Curve, loc=<0 0 0>, rot=<0 0 0 1>}
 L_BipedArm_A_Shoulder_Jnt_0_n120_0 {type=Curve, loc=<0 0 0>, rot=<0 0 0 1>}
 R_BipedArm_A_Shoulder_Jnt_0_n120_0 {type=Curve, loc=<0 0 0>, rot=<0 0 0 1>}
 L_BipedArm_A_Shoulder_Jnt_0_30_n90 {type=Curve, loc=<0 0 0>, rot=<0 0 0 1>}
 L_BipedArm_A_Shoulder_Jnt_0_n15_n90 {type=Curve, loc=<0 0 0>, rot=<0 0 0 1>}
 browDownLeft {type=Curve, loc=<0 0 0>, rot=<0 0 0 1>}
 browDownRight {type=Curve, loc=<0 0 0>, rot=<0 0 0 1>}
 browInnerUp {type=Curve, loc=<0 0 0>, rot=<0 0 0 1>}
 browOuterUpLeft {type=Curve, loc=<0 0 0>, rot=<0 0 0 1>}
 browOuterUpRight {type=Curve, loc=<0 0 0>, rot=<0 0 0 1>}
 noseSneerLeft {type=Curve, loc=<0 0 0>, rot=<0 0 0 1>}
 noseSneerRight {type=Curve, loc=<0 0 0>, rot=<0 0 0 1>}
 mouthLeft {type=Curve, loc=<0 0 0>, rot=<0 0 0 1>}
 mouthRight {type=Curve, loc=<0 0 0>, rot=<0 0 0 1>}
 mouthUpperUpLeft {type=Curve, loc=<0 0 0>, rot=<0 0 0 1>}
 mouthUpperUpRight {type=Curve, loc=<0 0 0>, rot=<0 0 0 1>}
 mouthSmileLeft {type=Curve, loc=<0 0 0>, rot=<0 0 0 1>}
 mouthSmileRight {type=Curve, loc=<0 0 0>, rot=<0 0 0 1>}
 mouthDimpleLeft {type=Curve, loc=<0 0 0>, rot=<0 0 0 1>}
 mouthDimpleRight {type=Curve, loc=<0 0 0>, rot=<0 0 0 1>}
 mouthFrownLeft {type=Curve, loc=<0 0 0>, rot=<0 0 0 1>}
 mouthFrownRight {type=Curve, loc=<0 0 0>, rot=<0 0 0 1>}
 mouthStretchLeft {type=Curve, loc=<0 0 0>, rot=<0 0 0 1>}
 mouthStretchRight {type=Curve, loc=<0 0 0>, rot=<0 0 0 1>}
 mouthLowerDownLeft {type=Curve, loc=<0 0 0>, rot=<0 0 0 1>}
 mouthLowerDownRight {type=Curve, loc=<0 0 0>, rot=<0 0 0 1>}
 mouthFunnel {type=Curve, loc=<0 0 0>, rot=<0 0 0 1>}
 mouthClose {type=Curve, loc=<0 0 0>, rot=<0 0 0 1>}
 mouthPucker {type=Curve, loc=<0 0 0>, rot=<0 0 0 1>}
 jawOpen {type=Curve, loc=<0 0 0>, rot=<0 0 0 1>}
 jawLeft {type=Curve, loc=<0 0 0>, rot=<0 0 0 1>}
 jawRight {type=Curve, loc=<0 0 0>, rot=<0 0 0 1>}
 jawForward {type=Curve, loc=<0 0 0>, rot=<0 0 0 1>}
 mouthShrugLower {type=Curve, loc=<0 0 0>, rot=<0 0 0 1>}
 mouthShrugUpper {type=Curve, loc=<0 0 0>, rot=<0 0 0 1>}
 mouthRollLower {type=Curve, loc=<0 0 0>, rot=<0 0 0 1>}
 mouthRollUpper {type=Curve, loc=<0 0 0>, rot=<0 0 0 1>}
 mouthPressLeft {type=Curve, loc=<0 0 0>, rot=<0 0 0 1>}
 mouthPressRight {type=Curve, loc=<0 0 0>, rot=<0 0 0 1>}
 cheekPuff {type=Curve, loc=<0 0 0>, rot=<0 0 0 1>}
 browRaise_L {type=Curve, loc=<0 0 0>, rot=<0 0 0 1>}
 browRaise_R {type=Curve, loc=<0 0 0>, rot=<0 0 0 1>}
 browDown_innerRaise_L {type=Curve, loc=<0 0 0>, rot=<0 0 0 1>}
 browDown_innerRaise_R {type=Curve, loc=<0 0 0>, rot=<0 0 0 1>}
 mouth_upperLipRaiseNwrinkle_L {type=Curve, loc=<0 0 0>, rot=<0 0 0 1>}
 mouth_upperLipRaiseNwrinkle_R {type=Curve, loc=<0 0 0>, rot=<0 0 0 1>}
 mouth_upperLipRaiseJopen_L {type=Curve, loc=<0 0 0>, rot=<0 0 0 1>}
 mouth_upperLipRaiseJopen_R {type=Curve, loc=<0 0 0>, rot=<0 0 0 1>}
 mouth_upperLipRaiseDimple_L {type=Curve, loc=<0 0 0>, rot=<0 0 0 1>}
 mouth_upperLipRaiseDimple_R {type=Curve, loc=<0 0 0>, rot=<0 0 0 1>}
 mouth_cornerPullLowerLipDepress_L {type=Curve, loc=<0 0 0>, rot=<0 0 0 1>}
 mouth_cornerPullLowerLipDepress_R {type=Curve, loc=<0 0 0>, rot=<0 0 0 1>}
 mouth_upperLipRaiseCornerPull_L {type=Curve, loc=<0 0 0>, rot=<0 0 0 1>}
 mouth_upperLipRaiseCornerPull_R {type=Curve, loc=<0 0 0>, rot=<0 0 0 1>}
 mouth_cornerPullJopen_L {type=Curve, loc=<0 0 0>, rot=<0 0 0 1>}
 mouth_cornerPullJopen_R {type=Curve, loc=<0 0 0>, rot=<0 0 0 1>}
 mouth_dimpleJopen_L {type=Curve, loc=<0 0 0>, rot=<0 0 0 1>}
 mouth_dimpleJopen_R {type=Curve, loc=<0 0 0>, rot=<0 0 0 1>}
 mouth_stretchJopen_L {type=Curve, loc=<0 0 0>, rot=<0 0 0 1>}
 mouth_stretchJopen_R {type=Curve, loc=<0 0 0>, rot=<0 0 0 1>}
 mouth_cornerPullMstretch_L {type=Curve, loc=<0 0 0>, rot=<0 0 0 1>}
 mouth_cornerPullMstretch_R {type=Curve, loc=<0 0 0>, rot=<0 0 0 1>}
 mouth_cornerPullMstretchJopen_L {type=Curve, loc=<0 0 0>, rot=<0 0 0 1>}
 mouth_cornerPullMstretchJopen_R {type=Curve, loc=<0 0 0>, rot=<0 0 0 1>}
 mouth_dimpleLowerLipDepress_L {type=Curve, loc=<0 0 0>, rot=<0 0 0 1>}
 mouth_dimpleLowerLipDepress_R {type=Curve, loc=<0 0 0>, rot=<0 0 0 1>}
 mouth_lipsTogetherJopen_cor {type=Curve, loc=<0 0 0>, rot=<0 0 0 1>}
 mouth_funnelJopen_cor {type=Curve, loc=<0 0 0>, rot=<0 0 0 1>}
 mouth_puckerJopen_cor {type=Curve, loc=<0 0 0>, rot=<0 0 0 1>}
 mouth_oh_cor {type=Curve, loc=<0 0 0>, rot=<0 0 0 1>}
 mouth_ohJopen_cor {type=Curve, loc=<0 0 0>, rot=<0 0 0 1>}
 jaw_leftJopen_cor {type=Curve, loc=<0 0 0>, rot=<0 0 0 1>}
 jaw_rightJopen_cor {type=Curve, loc=<0 0 0>, rot=<0 0 0 1>}
 mouth_lipsTogetherPucker_cor {type=Curve, loc=<0 0 0>, rot=<0 0 0 1>}
 mouth_lipsTogetherPuckerJopen_cor {type=Curve, loc=<0 0 0>, rot=<0 0 0 1>}
 L_BipedLeg_A_Knee_Jnt_0_0_0 {type=Curve, loc=<0 0 0>, rot=<0 0 0 1>}
 L_BipedLeg_A_Knee_Jnt_0_0_45 {type=Curve, loc=<0 0 0>, rot=<0 0 0 1>}
 L_BipedLeg_A_Knee_Jnt_0_0_90 {type=Curve, loc=<0 0 0>, rot=<0 0 0 1>}
 L_BipedLeg_A_Knee_Jnt_0_0_120 {type=Curve, loc=<0 0 0>, rot=<0 0 0 1>}
 L_BipedLeg_A_Knee_Jnt_0_0_n30 {type=Curve, loc=<0 0 0>, rot=<0 0 0 1>}
 R_BipedLeg_A_Knee_Jnt_0_0_0 {type=Curve, loc=<0 0 0>, rot=<0 0 0 1>}
 R_BipedLeg_A_Knee_Jnt_0_0_45 {type=Curve, loc=<0 0 0>, rot=<0 0 0 1>}
 R_BipedLeg_A_Knee_Jnt_0_0_90 {type=Curve, loc=<0 0 0>, rot=<0 0 0 1>}
 R_BipedLeg_A_Knee_Jnt_0_0_120 {type=Curve, loc=<0 0 0>, rot=<0 0 0 1>}
 R_BipedLeg_A_Knee_Jnt_0_0_n30 {type=Curve, loc=<0 0 0>, rot=<0 0 0 1>}
 C_BipedHeadNeck_A_NeckMid_Jnt_0_0_4 {type=Curve, loc=<0 0 0>, rot=<0 0 0 1>}
 C_BipedHeadNeck_A_NeckMid_Jnt_n11_44_n1 {type=Curve, loc=<0 0 0>, rot=<0 0 0 1>}
 C_BipedHeadNeck_A_NeckMid_Jnt_11_n44_n1 {type=Curve, loc=<0 0 0>, rot=<0 0 0 1>}
 C_BipedHeadNeck_A_NeckMid_Jnt_0_0_49 {type=Curve, loc=<0 0 0>, rot=<0 0 0 1>}
 C_BipedHeadNeck_A_NeckMid_Jnt_0_0_n41 {type=Curve, loc=<0 0 0>, rot=<0 0 0 1>}
 R_BipedArm_A_Shoulder_Jnt_0_30_n90 {type=Curve, loc=<0 0 0>, rot=<0 0 0 1>}
 R_BipedArm_A_Shoulder_Jnt_0_n15_n90 {type=Curve, loc=<0 0 0>, rot=<0 0 0 1>}
 L_BipedArm_A_Shoulder_Jnt_0_n90_0 {type=Curve, loc=<0 0 0>, rot=<0 0 0 1>}
 R_BipedArm_A_Shoulder_Jnt_0_n90_0 {type=Curve, loc=<0 0 0>, rot=<0 0 0 1>}
 R_BipedArm_A_Wrist_Jnt_0_n120_0 {type=Curve, loc=<0 0 0>, rot=<0 0 0 1>}
 R_BipedArm_A_Wrist_Jnt_0_0_n60 {type=Curve, loc=<0 0 0>, rot=<0 0 0 1>}
 R_BipedArm_A_Wrist_Jnt_0_0_60 {type=Curve, loc=<0 0 0>, rot=<0 0 0 1>}
 R_BipedArm_A_Wrist_Jnt_0_60_0 {type=Curve, loc=<0 0 0>, rot=<0 0 0 1>}
 L_BipedLeg_A_Knee_Jnt_0_0_150 {type=Curve, loc=<0 0 0>, rot=<0 0 0 1>}
 R_BipedLeg_A_Knee_Jnt_0_0_150 {type=Curve, loc=<0 0 0>, rot=<0 0 0 1>}
 MoveData_Speed {type=Curve, loc=<0 0 0>, rot=<0 0 0 1>}
 base_pose {type=Curve, loc=<0 0 0>, rot=<0 0 0 1>}
 calf_l_back_50 {type=Curve, loc=<0 0 0>, rot=<0 0 0 1>}
 calf_l_back_120 {type=Curve, loc=<0 0 0>, rot=<0 0 0 1>}
 calf_l_back_90 {type=Curve, loc=<0 0 0>, rot=<0 0 0 1>}
 calf_l_back_150 {type=Curve, loc=<0 0 0>, rot=<0 0 0 1>}
 calf_r_back_50 {type=Curve, loc=<0 0 0>, rot=<0 0 0 1>}
 calf_r_back_120 {type=Curve, loc=<0 0 0>, rot=<0 0 0 1>}
 calf_r_back_90 {type=Curve, loc=<0 0 0>, rot=<0 0 0 1>}
 calf_r_back_150 {type=Curve, loc=<0 0 0>, rot=<0 0 0 1>}
 clavicle_l_fwd_30 {type=Curve, loc=<0 0 0>, rot=<0 0 0 1>}
 clavicle_l_down_20 {type=Curve, loc=<0 0 0>, rot=<0 0 0 1>}
 clavicle_l_back_30 {type=Curve, loc=<0 0 0>, rot=<0 0 0 1>}
 clavicle_l_up_40 {type=Curve, loc=<0 0 0>, rot=<0 0 0 1>}
 clavicle_r_fwd_30 {type=Curve, loc=<0 0 0>, rot=<0 0 0 1>}
 clavicle_r_down_20 {type=Curve, loc=<0 0 0>, rot=<0 0 0 1>}
 clavicle_r_back_30 {type=Curve, loc=<0 0 0>, rot=<0 0 0 1>}
 clavicle_r_up_40 {type=Curve, loc=<0 0 0>, rot=<0 0 0 1>}
 clavicle_r_back_310 {type=Curve, loc=<0 0 0>, rot=<0 0 0 1>}
 clavicle_r_up_401 {type=Curve, loc=<0 0 0>, rot=<0 0 0 1>}
 foot_l_down_60 {type=Curve, loc=<0 0 0>, rot=<0 0 0 1>}
 foot_l_up_35 {type=Curve, loc=<0 0 0>, rot=<0 0 0 1>}
 foot_r_up_35 {type=Curve, loc=<0 0 0>, rot=<0 0 0 1>}
 foot_r_down_60 {type=Curve, loc=<0 0 0>, rot=<0 0 0 1>}
 hand_l_down_90 {type=Curve, loc=<0 0 0>, rot=<0 0 0 1>}
 hand_l_up_20 {type=Curve, loc=<0 0 0>, rot=<0 0 0 1>}
 hand_l_up_90 {type=Curve, loc=<0 0 0>, rot=<0 0 0 1>}
 hand_r_up_90 {type=Curve, loc=<0 0 0>, rot=<0 0 0 1>}
 hand_r_down_90 {type=Curve, loc=<0 0 0>, rot=<0 0 0 1>}
 hand_r_up_20 {type=Curve, loc=<0 0 0>, rot=<0 0 0 1>}
 Pose_0 {type=Curve, loc=<0 0 0>, rot=<0 0 0 1>}
 Pose_1 {type=Curve, loc=<0 0 0>, rot=<0 0 0 1>}
 Pose_2 {type=Curve, loc=<0 0 0>, rot=<0 0 0 1>}
 Pose_3 {type=Curve, loc=<0 0 0>, rot=<0 0 0 1>}
 Pose_4 {type=Curve, loc=<0 0 0>, rot=<0 0 0 1>}
 Pose_5 {type=Curve, loc=<0 0 0>, rot=<0 0 0 1>}
 Pose_6 {type=Curve, loc=<0 0 0>, rot=<0 0 0 1>}
 Pose_7 {type=Curve, loc=<0 0 0>, rot=<0 0 0 1>}
 Pose_8 {type=Curve, loc=<0 0 0>, rot=<0 0 0 1>}
 lowerarm_l_in_110 {type=Curve, loc=<0 0 0>, rot=<0 0 0 1>}
 lowerarm_l_in_35 {type=Curve, loc=<0 0 0>, rot=<0 0 0 1>}
 lowerarm_l_out_35 {type=Curve, loc=<0 0 0>, rot=<0 0 0 1>}
 lowerarm_l_in_50 {type=Curve, loc=<0 0 0>, rot=<0 0 0 1>}
 lowerarm_l_in_75 {type=Curve, loc=<0 0 0>, rot=<0 0 0 1>}
 lowerarm_l_in_10 {type=Curve, loc=<0 0 0>, rot=<0 0 0 1>}
 lowerarm_l_out_10 {type=Curve, loc=<0 0 0>, rot=<0 0 0 1>}
 lowerarm_l_in_90 {type=Curve, loc=<0 0 0>, rot=<0 0 0 1>}
 lowerarm_r_in_10 {type=Curve, loc=<0 0 0>, rot=<0 0 0 1>}
 lowerarm_r_out_10 {type=Curve, loc=<0 0 0>, rot=<0 0 0 1>}
 lowerarm_r_out_35 {type=Curve, loc=<0 0 0>, rot=<0 0 0 1>}
 lowerarm_r_in_75 {type=Curve, loc=<0 0 0>, rot=<0 0 0 1>}
 lowerarm_r_in_50 {type=Curve, loc=<0 0 0>, rot=<0 0 0 1>}
 lowerarm_r_in_110 {type=Curve, loc=<0 0 0>, rot=<0 0 0 1>}
 lowerarm_r_in_35 {type=Curve, loc=<0 0 0>, rot=<0 0 0 1>}
 lowerarm_r_in_90 {type=Curve, loc=<0 0 0>, rot=<0 0 0 1>}
 thigh_l_bck_10 {type=Curve, loc=<0 0 0>, rot=<0 0 0 1>}
 thigh_l_fwd_10 {type=Curve, loc=<0 0 0>, rot=<0 0 0 1>}
 thigh_l_in_45_out_90 {type=Curve, loc=<0 0 0>, rot=<0 0 0 1>}
 thigh_l_fwd_90 {type=Curve, loc=<0 0 0>, rot=<0 0 0 1>}
 thigh_l_bck_90 {type=Curve, loc=<0 0 0>, rot=<0 0 0 1>}
 thigh_l_bck_50 {type=Curve, loc=<0 0 0>, rot=<0 0 0 1>}
 thigh_l_out_110 {type=Curve, loc=<0 0 0>, rot=<0 0 0 1>}
 thigh_l_out_10 {type=Curve, loc=<0 0 0>, rot=<0 0 0 1>}
 thigh_l_fwd_45 {type=Curve, loc=<0 0 0>, rot=<0 0 0 1>}
 thigh_l_in_50 {type=Curve, loc=<0 0 0>, rot=<0 0 0 1>}
 thigh_l_fwd_110 {type=Curve, loc=<0 0 0>, rot=<0 0 0 1>}
 thigh_l_out_55 {type=Curve, loc=<0 0 0>, rot=<0 0 0 1>}
 thigh_l_out_85 {type=Curve, loc=<0 0 0>, rot=<0 0 0 1>}
 thigh_r_fwd_10 {type=Curve, loc=<0 0 0>, rot=<0 0 0 1>}
 thigh_r_out_55 {type=Curve, loc=<0 0 0>, rot=<0 0 0 1>}
 thigh_r_fwd_90 {type=Curve, loc=<0 0 0>, rot=<0 0 0 1>}
 thigh_r_fwd_45 {type=Curve, loc=<0 0 0>, rot=<0 0 0 1>}
 thigh_r_bck_10 {type=Curve, loc=<0 0 0>, rot=<0 0 0 1>}
 thigh_r_out_85 {type=Curve, loc=<0 0 0>, rot=<0 0 0 1>}
 thigh_r_in_50 {type=Curve, loc=<0 0 0>, rot=<0 0 0 1>}
 thigh_r_out_10 {type=Curve, loc=<0 0 0>, rot=<0 0 0 1>}
 thigh_r_out_110 {type=Curve, loc=<0 0 0>, rot=<0 0 0 1>}
 thigh_r_bck_50 {type=Curve, loc=<0 0 0>, rot=<0 0 0 1>}
 thigh_r_bck_90 {type=Curve, loc=<0 0 0>, rot=<0 0 0 1>}
 thigh_r_in_45_out_90 {type=Curve, loc=<0 0 0>, rot=<0 0 0 1>}
 thigh_r_fwd_110 {type=Curve, loc=<0 0 0>, rot=<0 0 0 1>}
 upperarm_l_back_45 {type=Curve, loc=<0 0 0>, rot=<0 0 0 1>}
 upperarm_l_out_55 {type=Curve, loc=<0 0 0>, rot=<0 0 0 1>}
 upperarm_l_out110_twist_in_70 {type=Curve, loc=<0 0 0>, rot=<0 0 0 1>}
 upperarm_l_out_85 {type=Curve, loc=<0 0 0>, rot=<0 0 0 1>}
 upperarm_l_fwd_110 {type=Curve, loc=<0 0 0>, rot=<0 0 0 1>}
 upperarm_l_in_10 {type=Curve, loc=<0 0 0>, rot=<0 0 0 1>}
 upperarm_l_out110_twist_out_70 {type=Curve, loc=<0 0 0>, rot=<0 0 0 1>}
 upperarm_l_out_twist_a {type=Curve, loc=<0 0 0>, rot=<0 0 0 1>}
 upperarm_l_fwd_15 {type=Curve, loc=<0 0 0>, rot=<0 0 0 1>}
 upperarm_l_out_15 {type=Curve, loc=<0 0 0>, rot=<0 0 0 1>}
 upperarm_l_back_10 {type=Curve, loc=<0 0 0>, rot=<0 0 0 1>}
 upperarm_l_out110 {type=Curve, loc=<0 0 0>, rot=<0 0 0 1>}
 upperarm_l_fwd_twist_in_a {type=Curve, loc=<0 0 0>, rot=<0 0 0 1>}
 upperarm_l_fwd_90 {type=Curve, loc=<0 0 0>, rot=<0 0 0 1>}
 upperarm_r_out_15 {type=Curve, loc=<0 0 0>, rot=<0 0 0 1>}
 upperarm_r_fwd_15 {type=Curve, loc=<0 0 0>, rot=<0 0 0 1>}
 upperarm_r_out_twist_a {type=Curve, loc=<0 0 0>, rot=<0 0 0 1>}
 upperarm_r_out110_twist_in_70 {type=Curve, loc=<0 0 0>, rot=<0 0 0 1>}
 upperarm_r_fwd_90 {type=Curve, loc=<0 0 0>, rot=<0 0 0 1>}
 upperarm_r_out110 {type=Curve, loc=<0 0 0>, rot=<0 0 0 1>}
 upperarm_r_back_10 {type=Curve, loc=<0 0 0>, rot=<0 0 0 1>}
 upperarm_r_out_85 {type=Curve, loc=<0 0 0>, rot=<0 0 0 1>}
 upperarm_r_in_10 {type=Curve, loc=<0 0 0>, rot=<0 0 0 1>}
 upperarm_r_back_45 {type=Curve, loc=<0 0 0>, rot=<0 0 0 1>}
 upperarm_r_out110_twist_out_70 {type=Curve, loc=<0 0 0>, rot=<0 0 0 1>}
 upperarm_r_out_55 {type=Curve, loc=<0 0 0>, rot=<0 0 0 1>}
 upperarm_r_fwd_twist_in_a {type=Curve, loc=<0 0 0>, rot=<0 0 0 1>}
 upperarm_r_fwd_110 {type=Curve, loc=<0 0 0>, rot=<0 0 0 1>}
 r_knee_corrective {type=Curve, loc=<0 0 0>, rot=<0 0 0 1>}
 l_knee_corrective {type=Curve, loc=<0 0 0>, rot=<0 0 0 1>}
 Male_Med_Head_01_Pout {type=Curve, loc=<0 0 0>, rot=<0 0 0 1>}
 Male_Med_Head_02_Pout {type=Curve, loc=<0 0 0>, rot=<0 0 0 1>}
 Male_Med_Head_03_Pout {type=Curve, loc=<0 0 0>, rot=<0 0 0 1>}
 Male_Med_Head_05_Effort {type=Curve, loc=<0 0 0>, rot=<0 0 0 1>}
 Stretch {type=Curve, loc=<0 0 0>, rot=<0 0 0 1>}
 Squash {type=Curve, loc=<0 0 0>, rot=<0 0 0 1>}
 polySurface151 {type=Curve, loc=<0 0 0>, rot=<0 0 0 1>}
 l_cheek_puff {type=Curve, loc=<0 0 0>, rot=<0 0 0 1>}
 r_cheek_puff {type=Curve, loc=<0 0 0>, rot=<0 0 0 1>}
 l_smile_fix {type=Curve, loc=<0 0 0>, rot=<0 0 0 1>}
 r_smile_fix {type=Curve, loc=<0 0 0>, rot=<0 0 0 1>}
 jaw_open_fix {type=Curve, loc=<0 0 0>, rot=<0 0 0 1>}
 r_lip_sneer_fix {type=Curve, loc=<0 0 0>, rot=<0 0 0 1>}
 l_lip_sneer_fix {type=Curve, loc=<0 0 0>, rot=<0 0 0 1>}
 L_Blink {type=Curve, loc=<0 0 0>, rot=<0 0 0 1>}
 R_Blink {type=Curve, loc=<0 0 0>, rot=<0 0 0 1>}
 L_Arm_Stretch {type=Curve, loc=<0 0 0>, rot=<0 0 0 1>}
 Hair_Hat {type=Curve, loc=<0 0 0>, rot=<0 0 0 1>}
 Hair_Helmet {type=Curve, loc=<0 0 0>, rot=<0 0 0 1>}
 Male_Med_Head_02_What {type=Curve, loc=<0 0 0>, rot=<0 0 0 1>}
 Male_Med_Head_02_Smile {type=Curve, loc=<0 0 0>, rot=<0 0 0 1>}
 Male_Med_Head_02_Scowl {type=Curve, loc=<0 0 0>, rot=<0 0 0 1>}
 Male_Med_Head_02_Hmmph {type=Curve, loc=<0 0 0>, rot=<0 0 0 1>}
 Male_Med_Head_02_Grrr {type=Curve, loc=<0 0 0>, rot=<0 0 0 1>}
 Male_Med_Head_02_Effort {type=Curve, loc=<0 0 0>, rot=<0 0 0 1>}
 Male_Med_Head_03_What {type=Curve, loc=<0 0 0>, rot=<0 0 0 1>}
 Male_Med_Head_03_Grrr {type=Curve, loc=<0 0 0>, rot=<0 0 0 1>}
 Male_Med_Head_03_Effort {type=Curve, loc=<0 0 0>, rot=<0 0 0 1>}
 Male_Med_Head_03_Hmmph {type=Curve, loc=<0 0 0>, rot=<0 0 0 1>}
 Male_Med_Head_03_Scowl {type=Curve, loc=<0 0 0>, rot=<0 0 0 1>}
 Male_Med_Head_03_Smile {type=Curve, loc=<0 0 0>, rot=<0 0 0 1>}
 Male_Med_Head_05_Teeth_Low_What {type=Curve, loc=<0 0 0>, rot=<0 0 0 1>}
 Male_Med_Head_05_Teeth_Low_Pout {type=Curve, loc=<0 0 0>, rot=<0 0 0 1>}
 Male_Med_Head_05_Teeth_Low_Hmmph {type=Curve, loc=<0 0 0>, rot=<0 0 0 1>}
 Male_Med_Head_05_Teeth_Low_Grrr {type=Curve, loc=<0 0 0>, rot=<0 0 0 1>}
 Male_Med_Head_05_Teeth_Low_Effort {type=Curve, loc=<0 0 0>, rot=<0 0 0 1>}
 Male_Med_Head_05_What {type=Curve, loc=<0 0 0>, rot=<0 0 0 1>}
 Male_Med_Head_05_Grrr {type=Curve, loc=<0 0 0>, rot=<0 0 0 1>}
 Male_Med_Head_05_Hmmph {type=Curve, loc=<0 0 0>, rot=<0 0 0 1>}
 Male_Med_Head_05_Pout {type=Curve, loc=<0 0 0>, rot=<0 0 0 1>}
 Male_Med_Head_05_Tongue_What {type=Curve, loc=<0 0 0>, rot=<0 0 0 1>}
 Male_Med_Head_05_Tongue_Pout {type=Curve, loc=<0 0 0>, rot=<0 0 0 1>}
 Male_Med_Head_05_Tongue_Hmmph {type=Curve, loc=<0 0 0>, rot=<0 0 0 1>}
 Male_Med_Head_05_Tongue_Grrr {type=Curve, loc=<0 0 0>, rot=<0 0 0 1>}
 Male_Med_Head_05_Tongue_Effort {type=Curve, loc=<0 0 0>, rot=<0 0 0 1>}
 AdditiveLayer {type=Curve, loc=<0 0 0>, rot=<0 0 0 1>}
 RightHandIK {type=Curve, loc=<0 0 0>, rot=<0 0 0 1>}
 LeftHandIK {type=Curve, loc=<0 0 0>, rot=<0 0 0 1>}
 LeftHandFK {type=Curve, loc=<0 0 0>, rot=<0 0 0 1>}
 Male_Med_Head_01_Grrr {type=Curve, loc=<0 0 0>, rot=<0 0 0 1>}
 WeaponAiming {type=Curve, loc=<0 0 0>, rot=<0 0 0 1>}
 Male_Med_Head_01_Hmmph {type=Curve, loc=<0 0 0>, rot=<0 0 0 1>}
 DragonSlash {type=Curve, loc=<0 0 0>, rot=<0 0 0 1>}
 Male_Med_Head_01_Effort {type=Curve, loc=<0 0 0>, rot=<0 0 0 1>}
 LeftHankFK {type=Curve, loc=<0 0 0>, rot=<0 0 0 1>}
 blendOrient1 {type=Curve, loc=<0 0 0>, rot=<0 0 0 1>}
 blendParent1 {type=Curve, loc=<0 0 0>, rot=<0 0 0 1>}
 bottom_R_Corner_Down {type=Curve, loc=<0 0 0>, rot=<0 0 0 1>}
 bottom_R_Corner_Up {type=Curve, loc=<0 0 0>, rot=<0 0 0 1>}
 top_L_Corner_Up {type=Curve, loc=<0 0 0>, rot=<0 0 0 1>}
 top_L_Corner_Down {type=Curve, loc=<0 0 0>, rot=<0 0 0 1>}
 top_R_Corner_Down {type=Curve, loc=<0 0 0>, rot=<0 0 0 1>}
 top_R_Corner_Up {type=Curve, loc=<0 0 0>, rot=<0 0 0 1>}
 bottom_L_Corner_Down {type=Curve, loc=<0 0 0>, rot=<0 0 0 1>}
 bottomt_L_Corner_Up {type=Curve, loc=<0 0 0>, rot=<0 0 0 1>}
 TrailCurve1 {type=Curve, loc=<0 0 0>, rot=<0 0 0 1>}
 BothHandFK {type=Curve, loc=<0 0 0>, rot=<0 0 0 1>}
 bottom_R_Corner_Down1 {type=Curve, loc=<0 0 0>, rot=<0 0 0 1>}
 blendPoint1 {type=Curve, loc=<0 0 0>, rot=<0 0 0 1>}
 WeaponAimingFreeze {type=Curve, loc=<0 0 0>, rot=<0 0 0 1>}
 l_calf_top_taper {type=Curve, loc=<0 0 0>, rot=<0 0 0 1>}
 JogBlendSpaceRotation {type=Curve, loc=<0 0 0>, rot=<0 0 0 1>}
 ZeroOutPitchWeight {type=Curve, loc=<0 0 0>, rot=<0 0 0 1>}
 MeleeTwist {type=Curve, loc=<0 0 0>, rot=<0 0 0 1>}
 MeleeAnim {type=Curve, loc=<0 0 0>, rot=<0 0 0 1>}
 l_sneer_fix {type=Curve, loc=<0 0 0>, rot=<0 0 0 1>}
 jaw_fix {type=Curve, loc=<0 0 0>, rot=<0 0 0 1>}
 r_sneer_fix {type=Curve, loc=<0 0 0>, rot=<0 0 0 1>}
 TurnRotationAmount {type=Curve, loc=<0 0 0>, rot=<0 0 0 1>}
 TurnEnd {type=Curve, loc=<0 0 0>, rot=<0 0 0 1>}
 BlendIn {type=Curve, loc=<0 0 0>, rot=<0 0 0 1>}
 BlendOut {type=Curve, loc=<0 0 0>, rot=<0 0 0 1>}
 BodyRotateX {type=Curve, loc=<0 0 0>, rot=<0 0 0 1>}
 BodyTranslateX {type=Curve, loc=<0 0 0>, rot=<0 0 0 1>}
 CrescentKickScale {type=Curve, loc=<0 0 0>, rot=<0 0 0 1>}
 mask_geo {type=Curve, loc=<0 0 0>, rot=<0 0 0 1>}
 RemoveJumpAdditive {type=Curve, loc=<0 0 0>, rot=<0 0 0 1>}
 RightLeg_IK {type=Curve, loc=<0 0 0>, rot=<0 0 0 1>}
 SnapWeapon_LHand {type=Curve, loc=<0 0 0>, rot=<0 0 0 1>}
 PivotAnim {type=Curve, loc=<0 0 0>, rot=<0 0 0 1>}
 DefaultAdditiveStateWeight {type=Curve, loc=<0 0 0>, rot=<0 0 0 1>}
 StartAnim {type=Curve, loc=<0 0 0>, rot=<0 0 0 1>}
 StopAnim {type=Curve, loc=<0 0 0>, rot=<0 0 0 1>}
 DisableSpeedWarping {type=Curve, loc=<0 0 0>, rot=<0 0 0 1>}
 DisableDBNOSlopeWarp {type=Curve, loc=<0 0 0>, rot=<0 0 0 1>}
 MaskRightArm {type=Curve, loc=<0 0 0>, rot=<0 0 0 1>}
 TurnInPlaceAnim {type=Curve, loc=<0 0 0>, rot=<0 0 0 1>}
 Head_MESH {type=Curve, loc=<0 0 0>, rot=<0 0 0 1>}
 Hair_MESH {type=Curve, loc=<0 0 0>, rot=<0 0 0 1>}
 LaunchPadTransitionOff {type=Curve, loc=<0 0 0>, rot=<0 0 0 1>}
 C_glabella_up_pose {type=Curve, loc=<0 0 0>, rot=<0 0 0 1>}
 jaw_open_pose {type=Curve, loc=<0 0 0>, rot=<0 0 0 1>}
 L_blink_pose {type=Curve, loc=<0 0 0>, rot=<0 0 0 1>}
 L_brow_down_pose {type=Curve, loc=<0 0 0>, rot=<0 0 0 1>}
 L_brow_squeeze_in_pose {type=Curve, loc=<0 0 0>, rot=<0 0 0 1>}
 L_brow_up_pose {type=Curve, loc=<0 0 0>, rot=<0 0 0 1>}
 L_frown_pose {type=Curve, loc=<0 0 0>, rot=<0 0 0 1>}
 L_lip_corner_narrow_pose {type=Curve, loc=<0 0 0>, rot=<0 0 0 1>}
 L_squint_inner_pose {type=Curve, loc=<0 0 0>, rot=<0 0 0 1>}
 L_wide_pose {type=Curve, loc=<0 0 0>, rot=<0 0 0 1>}
 look_down_pose {type=Curve, loc=<0 0 0>, rot=<0 0 0 1>}
 look_left_pose {type=Curve, loc=<0 0 0>, rot=<0 0 0 1>}
 look_right_pose {type=Curve, loc=<0 0 0>, rot=<0 0 0 1>}
 look_up_pose {type=Curve, loc=<0 0 0>, rot=<0 0 0 1>}
 phoneme_fv_pose {type=Curve, loc=<0 0 0>, rot=<0 0 0 1>}
 phoneme_oo_pose {type=Curve, loc=<0 0 0>, rot=<0 0 0 1>}
 R_blink_pose {type=Curve, loc=<0 0 0>, rot=<0 0 0 1>}
 R_brow_down_pose {type=Curve, loc=<0 0 0>, rot=<0 0 0 1>}
 R_brow_squeeze_in_pose {type=Curve, loc=<0 0 0>, rot=<0 0 0 1>}
 R_brow_up_pose {type=Curve, loc=<0 0 0>, rot=<0 0 0 1>}
 R_frown_pose {type=Curve, loc=<0 0 0>, rot=<0 0 0 1>}
 R_lip_corner_narrow_pose {type=Curve, loc=<0 0 0>, rot=<0 0 0 1>}
 R_squint_inner_pose {type=Curve, loc=<0 0 0>, rot=<0 0 0 1>}
 R_wide_pose {type=Curve, loc=<0 0 0>, rot=<0 0 0 1>}
 DisableArmsHeadAdditive {type=Curve, loc=<0 0 0>, rot=<0 0 0 1>}
 FrontendRotation {type=Curve, loc=<0 0 0>, rot=<0 0 0 1>}
 BothLegFK {type=Curve, loc=<0 0 0>, rot=<0 0 0 1>}
 AllowTurnInPlace {type=Curve, loc=<0 0 0>, rot=<0 0 0 1>}
 MontageTimeStretchCurve {type=Curve, loc=<0 0 0>, rot=<0 0 0 1>}
 ParaGlideLeanAlpha {type=Curve, loc=<0 0 0>, rot=<0 0 0 1>}
 WeaponRInChestSpace {type=Curve, loc=<0 0 0>, rot=<0 0 0 1>}
 JimBeam_AO {type=Curve, loc=<0 0 0>, rot=<0 0 0 1>}
 R_squeeze_pose {type=Curve, loc=<0 0 0>, rot=<0 0 0 1>}
 L_squeeze_pose {type=Curve, loc=<0 0 0>, rot=<0 0 0 1>}
 R_upper_lip_raiser_pose {type=Curve, loc=<0 0 0>, rot=<0 0 0 1>}
 L_cheek_up_pose {type=Curve, loc=<0 0 0>, rot=<0 0 0 1>}
 R_cheek_up_pose {type=Curve, loc=<0 0 0>, rot=<0 0 0 1>}
 DisableFaceOverride {type=Curve, loc=<0 0 0>, rot=<0 0 0 1>}
 EnableFullBodyOverride {type=Curve, loc=<0 0 0>, rot=<0 0 0 1>}
 C_glabella_down_pose {type=Curve, loc=<0 0 0>, rot=<0 0 0 1>}
 C_nose_wrinkler_pose {type=Curve, loc=<0 0 0>, rot=<0 0 0 1>}
 eye_convergence_pose {type=Curve, loc=<0 0 0>, rot=<0 0 0 1>}
 jaw_back_pose {type=Curve, loc=<0 0 0>, rot=<0 0 0 1>}
 jaw_forward_pose {type=Curve, loc=<0 0 0>, rot=<0 0 0 1>}
 jaw_left_pose {type=Curve, loc=<0 0 0>, rot=<0 0 0 1>}
 jaw_right_pose {type=Curve, loc=<0 0 0>, rot=<0 0 0 1>}
 jaw_up_pose {type=Curve, loc=<0 0 0>, rot=<0 0 0 1>}
 L_lip_corner_wide_pose {type=Curve, loc=<0 0 0>, rot=<0 0 0 1>}
 L_lower_lip_down_pose {type=Curve, loc=<0 0 0>, rot=<0 0 0 1>}
 L_lower_lip_up_pose {type=Curve, loc=<0 0 0>, rot=<0 0 0 1>}
 L_smile_pose {type=Curve, loc=<0 0 0>, rot=<0 0 0 1>}
 L_upper_lip_lower_pose {type=Curve, loc=<0 0 0>, rot=<0 0 0 1>}
 L_upper_lip_raiser_pose {type=Curve, loc=<0 0 0>, rot=<0 0 0 1>}
 L_upperLidDown_pose {type=Curve, loc=<0 0 0>, rot=<0 0 0 1>}
 mouth_down_pose {type=Curve, loc=<0 0 0>, rot=<0 0 0 1>}
 mouth_up_pose {type=Curve, loc=<0 0 0>, rot=<0 0 0 1>}
 phoneme_ch_pose {type=Curve, loc=<0 0 0>, rot=<0 0 0 1>}
 phoneme_mbp_pose {type=Curve, loc=<0 0 0>, rot=<0 0 0 1>}
 R_lip_corner_wide_pose {type=Curve, loc=<0 0 0>, rot=<0 0 0 1>}
 R_lower_lip_down_pose {type=Curve, loc=<0 0 0>, rot=<0 0 0 1>}
 R_lower_lip_up_pose {type=Curve, loc=<0 0 0>, rot=<0 0 0 1>}
 R_smile_pose {type=Curve, loc=<0 0 0>, rot=<0 0 0 1>}
 R_upper_lip_lower_pose {type=Curve, loc=<0 0 0>, rot=<0 0 0 1>}
 R_upperLidDown_pose {type=Curve, loc=<0 0 0>, rot=<0 0 0 1>}
 tongue_back_pose {type=Curve, loc=<0 0 0>, rot=<0 0 0 1>}
 tongue_down_pose {type=Curve, loc=<0 0 0>, rot=<0 0 0 1>}
 tongue_forward_pose {type=Curve, loc=<0 0 0>, rot=<0 0 0 1>}
 tongue_up_pose {type=Curve, loc=<0 0 0>, rot=<0 0 0 1>}
 LiveLink {type=Curve, loc=<0 0 0>, rot=<0 0 0 1>}
 ParaGlideDeploy {type=Curve, loc=<0 0 0>, rot=<0 0 0 1>}
 HideLeftWeapon {type=Curve, loc=<0 0 0>, rot=<0 0 0 1>}
 DisableLeftHandIKGolfCart {type=Curve, loc=<0 0 0>, rot=<0 0 0 1>}
 DisableRightHandIKGolfCart {type=Curve, loc=<0 0 0>, rot=<0 0 0 1>}
 DisableHandIKRetargeting {type=Curve, loc=<0 0 0>, rot=<0 0 0 1>}
 DisableConsumableOffsetPose {type=Curve, loc=<0 0 0>, rot=<0 0 0 1>}
 UpperBodyNoAdditive {type=Curve, loc=<0 0 0>, rot=<0 0 0 1>}
 MaskLeftArm {type=Curve, loc=<0 0 0>, rot=<0 0 0 1>}
 ForceUpperBody {type=Curve, loc=<0 0 0>, rot=<0 0 0 1>}
 RightHandFK {type=Curve, loc=<0 0 0>, rot=<0 0 0 1>}
 ParaGlideLeanTransition {type=Curve, loc=<0 0 0>, rot=<0 0 0 1>}
 InMotionEmote {type=Curve, loc=<0 0 0>, rot=<0 0 0 1>}
 ForceFullBodyInFrontEnd {type=Curve, loc=<0 0 0>, rot=<0 0 0 1>}
 WeaponLinFloorSpace {type=Curve, loc=<0 0 0>, rot=<0 0 0 1>}
 FrontEnd_Idle_Bypass {type=Curve, loc=<0 0 0>, rot=<0 0 0 1>}
 DisableSlopeWarping {type=Curve, loc=<0 0 0>, rot=<0 0 0 1>}
 Disable_Custom {type=Curve, loc=<0 0 0>, rot=<0 0 0 1>}
 EmoteRequiresWeapon {type=Curve, loc=<0 0 0>, rot=<0 0 0 1>}
 DisableIKRootAdditive {type=Curve, loc=<0 0 0>, rot=<0 0 0 1>}
 DisableFullBodySprint {type=Curve, loc=<0 0 0>, rot=<0 0 0 1>}
 DetachFeet_Vehicle {type=Curve, loc=<0 0 0>, rot=<0 0 0 1>}
 IndigoBoost {type=Curve, loc=<0 0 0>, rot=<0 0 0 1>}
 DisableUpperBodyAdditiveMeshSpace {type=Curve, loc=<0 0 0>, rot=<0 0 0 1>}
 MeleeAOWeight {type=Curve, loc=<0 0 0>, rot=<0 0 0 1>}
 HandsInRootSpace {type=Curve, loc=<0 0 0>, rot=<0 0 0 1>}
 CrouchPelvisAdjustment {type=Curve, loc=<0 0 0>, rot=<0 0 0 1>}
 EnableAccelYawCorrection {type=Curve, loc=<0 0 0>, rot=<0 0 0 1>}
 ParaGlideModifyRoot {type=Curve, loc=<0 0 0>, rot=<0 0 0 1>}
 UseScalarParameterForBlink {type=Curve, loc=<0 0 0>, rot=<0 0 0 1>}
 BlinkControl {type=Curve, loc=<0 0 0>, rot=<0 0 0 1>}
 ForceTargeting {type=Curve, loc=<0 0 0>, rot=<0 0 0 1>}
 BlinkControlProp {type=Curve, loc=<0 0 0>, rot=<0 0 0 1>}
 UseScalarParameterForBlinkProp {type=Curve, loc=<0 0 0>, rot=<0 0 0 1>}
 CrackshotCozyHeadCorrection {type=Curve, loc=<0 0 0>, rot=<0 0 0 1>}
 ForwardMotion {type=Curve, loc=<0 0 0>, rot=<0 0 0 1>}
 EmoteForcedRootYaw {type=Curve, loc=<0 0 0>, rot=<0 0 0 1>}
 MeleeTrailWidth {type=Curve, loc=<0 0 0>, rot=<0 0 0 1>}
 space_world {type=Curve, loc=<0 0 0>, rot=<0 0 0 1>}
 PeelyBlenderFace {type=Curve, loc=<0 0 0>, rot=<0 0 0 1>}
 Width1 {type=Curve, loc=<0 0 0>, rot=<0 0 0 1>}
 OverrideIKHands {type=Curve, loc=<0 0 0>, rot=<0 0 0 1>}
 ParaGlideAddAlpha {type=Curve, loc=<0 0 0>, rot=<0 0 0 1>}
 ParaGlideTurn {type=Curve, loc=<0 0 0>, rot=<0 0 0 1>}
 DisablePelvisAdjustment {type=Curve, loc=<0 0 0>, rot=<0 0 0 1>}
 UncleBrollyDeploy {type=Curve, loc=<0 0 0>, rot=<0 0 0 1>}
 ForceRootSlopeOrientation {type=Curve, loc=<0 0 0>, rot=<0 0 0 1>}
 DisableAdditiveOffsetPose {type=Curve, loc=<0 0 0>, rot=<0 0 0 1>}
 DisableIdleNoiseAdditive {type=Curve, loc=<0 0 0>, rot=<0 0 0 1>}
 UsingConsumable {type=Curve, loc=<0 0 0>, rot=<0 0 0 1>}
 DisableRightArmAdditive {type=Curve, loc=<0 0 0>, rot=<0 0 0 1>}
 NoGlider {type=Curve, loc=<0 0 0>, rot=<0 0 0 1>}
 ParaGlideModifyPelvis {type=Curve, loc=<0 0 0>, rot=<0 0 0 1>}
 ParaGlideModifyLegs {type=Curve, loc=<0 0 0>, rot=<0 0 0 1>}
 HT_L1_TransitionCurve {type=Curve, loc=<0 0 0>, rot=<0 0 0 1>}
 KeepPelvisAdjustment {type=Curve, loc=<0 0 0>, rot=<0 0 0 1>}
 PatrolDisableIdleDuringWalk {type=Curve, loc=<0 0 0>, rot=<0 0 0 1>}
 PatrolWalkUseFullbodyBlendspace {type=Curve, loc=<0 0 0>, rot=<0 0 0 1>}
 RemoveJumpUpAdditive {type=Curve, loc=<0 0 0>, rot=<0 0 0 1>}
 RemoveJumpLoopAdditive {type=Curve, loc=<0 0 0>, rot=<0 0 0 1>}
 RemoveJumpFallAdditive {type=Curve, loc=<0 0 0>, rot=<0 0 0 1>}
 JH_TransformCurve {type=Curve, loc=<0 0 0>, rot=<0 0 0 1>}
 PhysicsOff {type=Curve, loc=<0 0 0>, rot=<0 0 0 1>}
 DisableWeaponAdditivePose {type=Curve, loc=<0 0 0>, rot=<0 0 0 1>}
 Lexa_TransformationCurve {type=Curve, loc=<0 0 0>, rot=<0 0 0 1>}
 HideProp {type=Curve, loc=<0 0 0>, rot=<0 0 0 1>}
 CopyPose {type=Curve, loc=<0 0 0>, rot=<0 0 0 1>}
 MaterialCurveA {type=Curve, loc=<0 0 0>, rot=<0 0 0 1>}
 MaterialCurveB {type=Curve, loc=<0 0 0>, rot=<0 0 0 1>}
 MaterialCurveC {type=Curve, loc=<0 0 0>, rot=<0 0 0 1>}
 MaterialCurveD {type=Curve, loc=<0 0 0>, rot=<0 0 0 1>}
 MaterialCurveE {type=Curve, loc=<0 0 0>, rot=<0 0 0 1>}
 UseEmoteAnimOffset {type=Curve, loc=<0 0 0>, rot=<0 0 0 1>}
 emissive mult 2 {type=Curve, loc=<0 0 0>, rot=<0 0 0 1>}
 MaxHandle {type=Curve, loc=<0 0 0>, rot=<0 0 0 1>}
 FootPhase {type=Curve, loc=<0 0 0>, rot=<0 0 0 1>}
 Root {loc=<0 0 0>, rot=<0 0 0 1>}
 pelvis {parent=Root, loc=<0 -0.9885 86.1786>, rot=<0 0.7071 0 -0.7071>}
 Spine_01 {parent=pelvis, loc=<5.3147 0.474 0>, rot=<0 0 -0.0011 -1>}
 Spine_02 {parent=Spine_01, loc=<7.7324 -0.0759 0>, rot=<0 0 0.0121 -0.9999>}
 Spine_03 {parent=Spine_02, loc=<7.1151 -0.1704 0>, rot=<0 0 0.035 -0.9994>}
 spine_04 {parent=Spine_03, loc=<7.911 -0.1939 0>, rot=<0 0 -0.0005 -1>}
 spine_05 {parent=spine_04, loc=<10.1153 0.079 0>, rot=<0 0 -0.0055 -1>}
 clavicle_l {parent=spine_05, loc=<7.3758 -1.1305 -3.2823>, rot=<0.0179 -0.7069 -0.0179 -0.7069>}
 upperarm_l {parent=clavicle_l, loc=<10.6837 -0.4418 -3.5086>, rot=<0.0196 -0.5195 0.0356 -0.8535>}
 lowerarm_l {parent=upperarm_l, loc=<27.2431 1.4821 0.4211>, rot=<0.0042 0.0119 -0.2147 -0.9766>}
 hand_l {parent=lowerarm_l, loc=<26.9518 1.0011 0.7872>, rot=<-0.4993 -0.0461 0.0438 -0.8641>}
 index_metacarpal_l {parent=hand_l, loc=<2.9443 -0.6871 -1.4235>, rot=<0.0818 -0.0789 0.0084 -0.9935>}
 index_01_l {parent=index_metacarpal_l, loc=<7.7267 0.2373 -0.2513>, rot=<-0.0295 0.0652 0.0555 -0.9959>}
 index_02_l {parent=index_01_l, loc=<3.4049 0.1733 -0.0703>, rot=<0.0007 -0.007 0.095 -0.9955>}
 index_03_l {parent=index_02_l, loc=<2.6931 -0.0326 0.123>, rot=<-0.0004 0.007 0.0319 -0.9995>}
 middle_metacarpal_l {parent=hand_l, loc=<3.0581 -0.4829 -0.0672>, rot=<0.0393 0.0025 -0.0009 -0.9992>}
 middle_01_l {parent=middle_metacarpal_l, loc=<8.4777 0.2839 -0.3479>, rot=<-0.06 0.0075 0.1009 -0.9931>}
 middle_02_l {parent=middle_01_l, loc=<3.1302 0.0764 0.0245>, rot=<-0.0116 0.0197 0.0067 -0.9997>}
 middle_03_l {parent=middle_02_l, loc=<3.0228 -0.0927 -0.0102>, rot=<0.0091 -0.0067 0.0536 -0.9985>}
 pinky_metacarpal_l {parent=hand_l, loc=<3.1002 -0.584 2.2451>, rot=<-0.0734 0.0993 -0.0118 -0.9923>}
 pinky_01_l {parent=pinky_metacarpal_l, loc=<7.2261 -0.002 0.21>, rot=<0.0959 -0.041 0.098 -0.9897>}
 pinky_02_l {parent=pinky_01_l, loc=<2.7394 -0.0544 -0.0695>, rot=<0.0039 0.0002 0.0421 -0.9991>}
 pinky_03_l {parent=pinky_02_l, loc=<1.5966 -0.2015 0.0114>, rot=<0.0012 0.0098 0.0059 -0.9999>}
 ring_metacarpal_l {parent=hand_l, loc=<3.0732 -0.4596 1.2011>, rot=<-0.0082 0.0557 -0.0051 -0.9984>}
 ring_01_l {parent=ring_metacarpal_l, loc=<8.1739 0.1202 -0.3808>, rot=<-0.0136 0.0015 0.133 -0.991>}
 ring_02_l {parent=ring_01_l, loc=<2.9099 0.0624 0.0569>, rot=<0 0 0 -1>}
 ring_03_l {parent=ring_02_l, loc=<2.6446 -0.1326 -0.0569>, rot=<0 0 0.024 -0.9997>}
 thumb_01_l {parent=hand_l, loc=<2.7726 -1.6981 -1.7892>, rot=<0.666 -0.3553 -0.1285 -0.6432>}
 thumb_02_l {parent=thumb_01_l, loc=<4.5942 0.1039 -0.0993>, rot=<-0.0057 -0.0455 0.1247 -0.9911>}
 thumb_03_l {parent=thumb_02_l, loc=<3.8743 -0.6419 -0.0603>, rot=<0.0181 -0.0562 -0.0281 -0.9979>}
 weapon_l {parent=hand_l, loc=<0.4458 -3.5863 0.3694>, rot=<0 0 0 -1>}
 hand_attach_l {parent=hand_l, loc=<8.5 -3 0>, rot=<0.7071 0.7071 0 0>}
 lowerarm_twist_l {parent=lowerarm_l, loc=<26.946 0.9829 0.7898>, rot=<-0.4993 -0.0461 0.0438 -0.8641>}
 lowerarm_twist_01_l {parent=lowerarm_l, loc=<20.2138 0.7508 0.5904>, rot=<0 0 0 -1>}
 lowerarm_twist_02_l {parent=lowerarm_l, loc=<11.5508 0.429 0.3374>, rot=<0 0 0 -1>}
 upperarm_twist_l {parent=upperarm_l, loc=<0 0 0>, rot=<0 0 0 -1>}
 upperarm_twist_01_l {parent=upperarm_l, loc=<1 0 0>, rot=<0 0 0 -1>}
 upperarm_twist_02_l {parent=upperarm_l, loc=<13.7941 0 0>, rot=<0 0 0 1>}
 clavicle_r {parent=spine_05, loc=<7.3759 -1.1305 3.2823>, rot=<0.7069 0.0179 -0.7069 0.0179>}
 upperarm_r {parent=clavicle_r, loc=<-10.6837 0.4418 3.5085>, rot=<0.0196 -0.5195 0.0356 -0.8535>}
 lowerarm_r {parent=upperarm_r, loc=<-27.2432 -1.4821 -0.4211>, rot=<0.0042 0.0119 -0.2147 -0.9766>}
 hand_r {parent=lowerarm_r, loc=<-26.9519 -1.0011 -0.7873>, rot=<-0.4993 -0.0461 0.0438 -0.8641>}
 index_metacarpal_r {parent=hand_r, loc=<-2.9441 0.6873 1.4236>, rot=<0.0818 -0.0789 0.0084 -0.9935>}
 index_01_r {parent=index_metacarpal_r, loc=<-7.7267 -0.2373 0.2513>, rot=<-0.0295 0.0652 0.0555 -0.9959>}
 index_02_r {parent=index_01_r, loc=<-3.4051 -0.1734 0.0702>, rot=<0.0007 -0.007 0.095 -0.9955>}
 index_03_r {parent=index_02_r, loc=<-2.6931 0.0325 -0.123>, rot=<-0.0004 0.007 0.0319 -0.9995>}
 middle_metacarpal_r {parent=hand_r, loc=<-3.0581 0.483 0.0672>, rot=<0.0393 0.0025 -0.0009 -0.9992>}
 middle_01_r {parent=middle_metacarpal_r, loc=<-8.4775 -0.284 0.348>, rot=<-0.06 0.0075 0.1009 -0.9931>}
 middle_02_r {parent=middle_01_r, loc=<-3.1304 -0.0763 -0.0245>, rot=<-0.0116 0.0197 0.0067 -0.9997>}
 middle_03_r {parent=middle_02_r, loc=<-3.0228 0.0927 0.0102>, rot=<0.0091 -0.0067 0.0536 -0.9985>}
 pinky_metacarpal_r {parent=hand_r, loc=<-3.1004 0.5841 -2.2452>, rot=<-0.0734 0.0993 -0.0118 -0.9923>}
 pinky_01_r {parent=pinky_metacarpal_r, loc=<-7.2259 0.0022 -0.21>, rot=<0.0959 -0.041 0.098 -0.9897>}
 pinky_02_r {parent=pinky_01_r, loc=<-2.7394 0.0544 0.0695>, rot=<0.0039 0.0002 0.0421 -0.9991>}
 pinky_03_r {parent=pinky_02_r, loc=<-1.5964 0.2012 -0.0114>, rot=<0.0012 0.0098 0.0059 -0.9999>}
 ring_metacarpal_r {parent=hand_r, loc=<-3.0731 0.4597 -1.2011>, rot=<-0.0082 0.0557 -0.0051 -0.9984>}
 ring_01_r {parent=ring_metacarpal_r, loc=<-8.1741 -0.1202 0.3808>, rot=<-0.0136 0.0015 0.133 -0.991>}
 ring_02_r {parent=ring_01_r, loc=<-2.9096 -0.0624 -0.0569>, rot=<0 0 0 -1>}
 ring_03_r {parent=ring_02_r, loc=<-2.6447 0.1325 0.0569>, rot=<0 0 0.024 -0.9997>}
 thumb_01_r {parent=hand_r, loc=<-2.7727 1.6982 1.7892>, rot=<0.666 -0.3553 -0.1285 -0.6432>}
 thumb_02_r {parent=thumb_01_r, loc=<-4.5941 -0.104 0.0993>, rot=<-0.0057 -0.0455 0.1247 -0.9911>}
 thumb_03_r {parent=thumb_02_r, loc=<-3.8743 0.6419 0.0603>, rot=<0.0181 -0.0562 -0.0281 -0.9979>}
 weapon_r {parent=hand_r, loc=<-0.4457 3.5864 -0.3694>, rot=<0 0 0 -1>}
 hand_attach_r {parent=hand_r, loc=<-8.5 3 0>, rot=<0 0 -0.7071 -0.7071>}
 lowerarm_twist_r {parent=lowerarm_r, loc=<-26.9461 -0.983 -0.7899>, rot=<-0.4993 -0.0461 0.0438 -0.8641>}
 lowerarm_twist_01_r {parent=lowerarm_r, loc=<-20.2139 -0.7508 -0.5904>, rot=<0 0 0 -1>}
 dyn_hose01 {parent=lowerarm_twist_01_r, loc=<0 0 0>, rot=<0 0 0 1>}
 lowerarm_twist_02_r {parent=lowerarm_r, loc=<-11.5508 -0.429 -0.3374>, rot=<0 0 0 -1>}
 upperarm_twist_r {parent=upperarm_r, loc=<0 0 0>, rot=<0 0 0 -1>}
 upperarm_twist_01_r {parent=upperarm_r, loc=<-1 0 0>, rot=<0 0 0 -1>}
 upperarm_twist_02_r {parent=upperarm_r, loc=<-13.7942 0 0>, rot=<0 0 0 1>}
 neck_01 {parent=spine_05, loc=<13.9426 -0.2264 0>, rot=<0 0 -0.1691 -0.9856>}
 neck_02 {parent=neck_01, loc=<4.6914 0.3829 0>, rot=<0 0 0.0424 -0.9991>}
 thigh_l {parent=pelvis, loc=<-1.5743 -0.2938 -8.7855>, rot=<0.026 0.0004 0.0136 -0.9996>}
 calf_l {parent=thigh_l, loc=<-38.4135 -0.6432 2.1503>, rot=<0.0002 -0.0066 -0.0379 -0.9993>}
 foot_l {parent=calf_l, loc=<-35.4503 -0.3763 -0.7574>, rot=<-0.0261 0.0071 0.0242 -0.9993>}
 ball_l {parent=foot_l, loc=<-7.9777 12.8059 -0.8901>, rot=<-0.0008 -0.0089 -0.7161 -0.698>}
 calf_twist_l {parent=calf_l, loc=<-17.7251 0 0>, rot=<0 0 0 -1>}
 calf_twist_01_l {parent=calf_l, loc=<-34.4503 -0.3763 -0.7574>, rot=<0 0 0 -1>}
 calf_twist_02_l {parent=calf_l, loc=<-17.7251 -0.3763 -0.7574>, rot=<0 0 0 -1>}
 lowerleg_twist_01_l {parent=calf_l, loc=<-38.7753 0 0>, rot=<0 0 0 1>}
 lowerleg_twist_02_l {parent=calf_l, loc=<-20.0451 0 0>, rot=<0 0 0 1>}
 thigh_twist_l {parent=thigh_l, loc=<0 0 0>, rot=<0 0 0 -1>}
 thigh_twist_01_l {parent=thigh_l, loc=<-1 0 0>, rot=<0 0 0 -1>}
 upperleg_twist_01_l {parent=thigh_l, loc=<-0.3846 0 0>, rot=<0 0 0 1>}
 thigh_r {parent=pelvis, loc=<-1.5741 -0.2938 8.7856>, rot=<0.0004 -0.026 0.9996 0.0136>}
 calf_r {parent=thigh_r, loc=<38.4137 0.6432 -2.1504>, rot=<0.0002 -0.0066 -0.0379 -0.9993>}
 foot_r {parent=calf_r, loc=<35.4504 0.3763 0.7575>, rot=<-0.0261 0.0071 0.0242 -0.9993>}
 ball_r {parent=foot_r, loc=<7.9776 -12.8059 0.8901>, rot=<-0.0008 -0.0089 -0.7161 -0.698>}
 calf_twist_r {parent=calf_r, loc=<17.7252 0 0>, rot=<0 0 0 -1>}
 calf_twist_01_r {parent=calf_r, loc=<34.4504 0.3763 0.7574>, rot=<0 0 0 -1>}
 calf_twist_02_r {parent=calf_r, loc=<17.7252 0.3763 0.7574>, rot=<0 0 0 -1>}
 lowerleg_twist_01_r {parent=calf_r, loc=<38.7753 0 0>, rot=<0 0 0 1>}
 lowerleg_twist_02_r {parent=calf_r, loc=<20.0452 0 0>, rot=<0 0 0 1>}
 thigh_twist_r {parent=thigh_r, loc=<0 0 0>, rot=<0 0 0 -1>}
 thigh_twist_01_r {parent=thigh_r, loc=<1 0 0>, rot=<0 0 0 -1>}
 upperleg_twist_01_r {parent=thigh_r, loc=<0.3846 0 0>, rot=<0 0 0 1>}
 attach {parent=Root, loc=<0 0 0>, rot=<0 0 0 1>}
 ik_foot_root {parent=Root, loc=<0 0 0>, rot=<0 0 0 1>}
 ik_foot_l {parent=ik_foot_root, loc=<6.8559 -2.8834 10.7904>, rot=<0 0.7071 0 -0.7071>}
 ik_foot_r {parent=ik_foot_root, loc=<-6.8559 -2.8834 10.7902>, rot=<0.7071 0 -0.7071 0>}
 ik_hand_root {parent=Root, loc=<0 0 0>, rot=<0 0 0 1>}
 ik_hand_gun {parent=ik_hand_root, loc=<-39.562 6.118 82.7824>, rot=<0.6645 0.1448 0.5538 -0.4804>}
 ik_hand_l {parent=ik_hand_gun, loc=<27.267 57.3293 47.2276>, rot=<-0.3446 -0.7246 -0.5969 0>}
 ik_hand_r {parent=ik_hand_gun, loc=<0 0 0>, rot=<0 0 0 1>}
]
variables [
 arm_l_ik_mode {type=bool, default=False}
 arm_r_ik_mode {type=bool, default=False}
 leg_l_ik_mode {type=bool, default=False}
 leg_r_ik_mode {type=bool, default=False}
]
