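# Shared MetaHuman Control Rig for player characters, providing FK and IK animation controls over the skeleton.
# Setup functions place controls on initial bone transforms, mirroring negative-side limbs with flipped scale and roll offsets.
# Solve functions switch limbs between FK and two-bone IK, snap the inactive controls to match, toggle visibility and request auto-keys on switching.
# Backward-style functions project bone transforms onto controls for limbs, spine and head.
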ControlRigBlueprint /Game/Characters/Player/Common/SharedRigs/MetaHuman_ControlRig
graphs [
 ContainedGraph {nodes [
   n1 = RigVMFunctionEntryNode()
   n2 = /Script/RigVM.RigVMFunction_MathQuaternionToRotator(Value=@n1.Input.Rotation) {is=struct RigVMUnitNode, template="Cast::Execute(in Value,out Result)"}
   RerouteNode = RigVMDispatchNode(
    Elements=(Rotation=(X=0,Y=0,Z=0,W=1),Translation=(X=0,Y=0,Z=0),Scale3D=(X=-1,Y=1,Z=1))
    Elements.Scale3D.Y=@n1.Input.Scale3D.Y
    Elements.Scale3D.Z=@n1.Input.Scale3D.Z
    Elements.Translation=@n1.Input.Translation
    Elements.Scale3D.X=/Script/RigVM.RigVMFunction_MathFloatNegate(Value=@n1.Input.Scale3D.X) {is=struct RigVMUnitNode, template="Negate::Execute(in Value,out Result)"}.Result
    Elements.Rotation=/Script/RigVM.RigVMFunction_MathQuaternionFromRotator(
     Rotator=(Pitch=0,Yaw=0,Roll=0)
     Rotator.Roll=&RigVMFunction_MathFloatAdd(B=180, A=@n2.Result.Roll) {template="Add::Execute(in A,in B,out Result)"}.Result
     Rotator.Pitch=@n2.Result.Pitch
     Rotator.Yaw=@n2.Result.Yaw
    ) {is=struct RigVMUnitNode}.Result
   ) {template="DISPATCH_RigVMDispatch_MakeStruct(in Elements,out Struct)"}
   > RigVMFunctionReturnNode(Output=@RerouteNode.Struct)
  ]}
 ContainedGraph {nodes [
   n1 = RigVMFunctionEntryNode()
   n2 = /Script/ControlRig.RigUnit_SequenceExecution(ExecuteContext=@n1.ExecuteContext) {is=struct RigVMUnitNode}
   n3 = RigVMVariableNode() {variable=Bones}
   RigVMDispatchNode(
    Num=RigVMDispatchNode(Array=@n3.Value) {template="DISPATCH_RigVMDispatch_ArrayGetNum(in Array,out Num)"}.Num
    ExecuteContext=@n2.A
    Array=RigVMVariableNode() {variable=Transforms}.Value
   ) {template="DISPATCH_RigVMDispatch_ArraySetNum(io Array,in Num)"}
   > n4 = RigVMDispatchNode(Array=/Script/ControlRig.RigUnit_GetTransformItemArray(Space=GlobalSpace, Items=@n3.Value) {is=struct RigVMUnitNode}.Transforms) {template="DISPATCH_RigVMDispatch_ArrayIterator(in Array,out Element,out Index,out Count,out Ratio)"}
   Get Negative Transform = RigVMFunctionReferenceNode(Input=@n4.Element)
   n5 = RigVMDispatchNode(False=@n4.Element, True=@"Get Negative Transform".Output, Condition=RigVMVariableNode() {variable=Negative Side}.Value) {template="DISPATCH_RigVMDispatch_If(in Condition,in True,in False,out Result)"}
   n6 = /Script/RigVM.RigVMFunction_MathIntSub(B=1, A=@n4.Count) {is=struct RigVMUnitNode, template="Subtract::Execute(in A,in B,out Result)"}
   n7 = &RigVMFunction_MathQuaternionToRotator(Value=@n5.Result.Rotation) {template="Cast::Execute(in Value,out Result)"}
   n8 = &RigVMFunction_MathFloatAdd(B=90, A=@n7.Result.Roll) {template="Add::Execute(in A,in B,out Result)"}
   > RigVMDispatchNode(
    Element=(Rotation=(X=0,Y=0,Z=0,W=1),Translation=(X=0,Y=0,Z=0),Scale3D=(X=1,Y=1,Z=1))
    Index=@n4.Index
    Element.Scale3D=@n5.Result.Scale3D
    Element.Translation=@n5.Result.Translation
    Element.Rotation=RigVMDispatchNode(
     False=@n5.Result.Rotation
     Condition=/Script/RigVM.RigVMFunction_MathIntEquals(A=@n4.Index, B=@n6.Result) {is=struct RigVMUnitNode}.Result
     True=&RigVMFunction_MathQuaternionFromRotator(Rotator=(Pitch=0,Yaw=0,Roll=0), Rotator.Roll=@n8.Result, Rotator.Pitch=@n7.Result.Pitch, Rotator.Yaw=@n7.Result.Yaw).Result
    ) {template="DISPATCH_RigVMDispatch_If(in Condition,in True,in False,out Result)"}.Result
    Array=RigVMVariableNode() {variable=Transforms}.Value
   ) {template="DISPATCH_RigVMDispatch_ArraySetAtIndex(io Array,in Index,in Element)"}
   &RigUnit_SetTransform(
    Weight=1
    bInitial=true
    Value=RigVMDispatchNode(Index=0, Array=RigVMVariableNode() {variable=Transforms}.Value) {template="DISPATCH_RigVMDispatch_ArrayGetAtIndex(in Array,in Index,out Element)"}.Element
    ExecutePin=@n2.B
    Item=RigVMVariableNode() {variable=FK Space}.Value
   )
   > /Script/ControlRig.RigUnit_SetTransformItemArray(
    Space=GlobalSpace
    Weight=1
    bInitial=true
    bPropagateToChildren=True
    Transforms=RigVMVariableNode() {variable=Transforms}.Value
    Items=RigVMVariableNode() {variable=FK Controls}.Value
   ) {is=struct RigVMUnitNode}
   n9 = &RigUnit_GetTransform(Item=RigVMDispatchNode(Index=2, Array=RigVMVariableNode() {variable=Bones}.Value) {template="DISPATCH_RigVMDispatch_ArrayGetAtIndex(in Array,in Index,out Element)"}.Element)
   Get Negative Transform_0 = RigVMFunctionReferenceNode(Input=@n9.Transform)
   > &RigUnit_SetTransform(
    Weight=1
    bInitial=true
    Value=RigVMDispatchNode(True=@"Get Negative Transform_0".Output, False=@n9.Transform, Condition=RigVMVariableNode() {variable=Negative Side}.Value) {template="DISPATCH_RigVMDispatch_If(in Condition,in True,in False,out Result)"}.Result
    Item=RigVMVariableNode() {variable=Fingers Space}.Value
   )
   n10 = RigVMVariableNode() {variable=Transforms}
   n11 = /Script/RigVM.RigVMFunction_MathFloatDiv(
    B=2
    A=&RigVMFunction_MathVectorDistance(
     A=RigVMDispatchNode(Index=0, Array=@n10.Value) {template="DISPATCH_RigVMDispatch_ArrayGetAtIndex(in Array,in Index,out Element)"}.Element.Translation
     B=RigVMDispatchNode(Index=2, Array=@n10.Value) {template="DISPATCH_RigVMDispatch_ArrayGetAtIndex(in Array,in Index,out Element)"}.Element.Translation
    ).Result
   ) {is=struct RigVMUnitNode, template="Divide::Execute(in A,in B,out Result)"}
   &RigUnit_SetTransform(
    Value=(Rotation=(X=0,Y=0,Z=0,W=1),Translation=(X=0,Y=0,Z=0),Scale3D=(X=1,Y=1,Z=1))
    Weight=1
    bInitial=true
    Value.Scale3D=RigVMDispatchNode(False=(X=1,Y=1,Z=1), True=(X=-1,Y=1,Z=1), Condition=RigVMVariableNode() {variable=Negative Side}.Value) {template="DISPATCH_RigVMDispatch_If(in Condition,in True,in False,out Result)"}.Result
    Value.Translation=&RigVMFunction_MathTransformMul(
     A=(Rotation=(X=0,Y=0,Z=0,W=1),Translation=(X=0,Y=0,Z=0),Scale3D=(X=1,Y=1,Z=1))
     B=RigVMDispatchNode(Index=1, Array=@n10.Value) {template="DISPATCH_RigVMDispatch_ArrayGetAtIndex(in Array,in Index,out Element)"}.Element
     A.Translation.Y=&RigVMFunction_MathFloatNegate(Value=@n11.Result) {template="Negate::Execute(in Value,out Result)"}.Result
    ) {template="Multiply::Execute(in A,in B,out Result)"}.Result.Translation
    ExecutePin=@n2.C
    Item=RigVMVariableNode() {variable=Pole Vector Control}.Value
   )
   n12 = RigVMVariableNode() {variable=Transforms}
   n13 = RigVMDispatchNode(Array=@n12.Value) {template="DISPATCH_RigVMDispatch_ArrayGetNum(in Array,out Num)"}
   > &RigUnit_SetTransform(
    Weight=1
    bInitial=true
    Value=RigVMDispatchNode(Index=&RigVMFunction_MathIntSub(B=1, A=@n13.Num) {template="Subtract::Execute(in A,in B,out Result)"}.Result, Array=@n12.Value) {template="DISPATCH_RigVMDispatch_ArrayGetAtIndex(in Array,in Index,out Element)"}.Element
    Item=RigVMVariableNode() {variable=IK Control}.Value
   )
   n14 = RigVMDispatchNode(ExecuteContext=@n2.D, Array=RigVMVariableNode() {variable=IK Parents}.Value) {template="DISPATCH_RigVMDispatch_ArrayIterator(in Array,out Element,out Index,out Count,out Ratio)"}
   > /Script/ControlRig.RigUnit_AddParent(DisplayLabel=None, Parent=@n14.Element, Child=RigVMVariableNode() {variable=IK Control}.Value) {is=struct RigVMUnitNode}
   > &RigUnit_AddParent(DisplayLabel=None, Parent=@n14.Element, Child=RigVMVariableNode() {variable=Pole Vector Control}.Value)
   RigVMFunctionReturnNode(ExecuteContext=@n14.Completed)
  ]}
 ContainedGraph {nodes [
   n1 = RigVMFunctionEntryNode()
   n2 = &RigUnit_GetTransform(Item=@n1.Bone)
   Get Negative Transform_1 = RigVMFunctionReferenceNode(Input=@n2.Transform)
   > &RigUnit_SetTransform(
    Weight=1
    bInitial=true
    Item=@n1.Control
    Value=RigVMDispatchNode(True=@"Get Negative Transform_1".Output, False=@n2.Transform, Condition=RigVMVariableNode() {variable=Negative Side}.Value) {template="DISPATCH_RigVMDispatch_If(in Condition,in True,in False,out Result)"}.Result
   )
   > RigVMFunctionReturnNode()
  ]}
 ContainedGraph {nodes [
   n1 = RigVMFunctionEntryNode()
   n2 = RigVMDispatchNode(Array=@n1.Bones, ExecuteContext=@n1.Execute) {template="DISPATCH_RigVMDispatch_ArrayIterator(in Array,out Element,out Index,out Count,out Ratio)"}
   n3 = &RigUnit_GetTransform(Item=@n2.Element)
   n4 = RigVMVariableNode() {variable=Negative Side}
   n5 = &RigVMFunction_MathQuaternionToRotator(Value=@n3.Transform.Rotation) {template="Cast::Execute(in Value,out Result)"}
   n6 = &RigVMFunction_MathFloatAdd(B=180, A=@n5.Result.Roll) {template="Add::Execute(in A,in B,out Result)"}
   > &RigUnit_SetTransform(
    Item=(Type=Control,Name="None")
    Value=(Rotation=(X=0,Y=0,Z=0,W=1),Translation=(X=0,Y=0,Z=0),Scale3D=(X=1,Y=1,Z=1))
    Weight=1
    bInitial=true
    Value.Translation=@n3.Transform.Translation
    Value.Scale3D=RigVMDispatchNode(False=(X=1,Y=1,Z=1), True=(X=-1,Y=1,Z=1), Condition=@n4.Value) {template="DISPATCH_RigVMDispatch_If(in Condition,in True,in False,out Result)"}.Result
    Value.Rotation=RigVMDispatchNode(
     False=@n3.Transform.Rotation
     True=&RigVMFunction_MathQuaternionFromRotator(Rotator=(Pitch=0,Yaw=0,Roll=0), Rotator.Roll=@n6.Result, Rotator.Pitch=@n5.Result.Pitch, Rotator.Yaw=@n5.Result.Yaw).Result
     Condition=@n4.Value
    ) {template="DISPATCH_RigVMDispatch_If(in Condition,in True,in False,out Result)"}.Result
    Item.Name=/Script/RigVM.RigVMFunction_NameConcat(B=_ctrl, A=@n2.Element.Name) {is=struct RigVMUnitNode, template="Concat::Execute(in A,in B,out Result)"}.Result
   )
   RigVMFunctionReturnNode(Execute=@n2.Completed)
  ]}
 ContainedGraph {nodes [
   n1 = RigVMFunctionEntryNode()
   n2 = &RigUnit_SequenceExecution(ExecuteContext=@n1.ExecuteContext)
   n3 = RigVMVariableNode() {variable=Bones}
   n4 = RigVMVariableNode() {variable=Transforms}
   RigVMDispatchNode(
    Num=RigVMDispatchNode(Array=@n3.Value) {template="DISPATCH_RigVMDispatch_ArrayGetNum(in Array,out Num)"}.Num
    ExecuteContext=@n2.A
    Array=@n4.Value
   ) {template="DISPATCH_RigVMDispatch_ArraySetNum(io Array,in Num)"}
   > n5 = RigVMDispatchNode(Array=&RigUnit_GetTransformItemArray(Space=GlobalSpace, Items=@n3.Value).Transforms) {template="DISPATCH_RigVMDispatch_ArrayIterator(in Array,out Element,out Index,out Count,out Ratio)"}
   Get Negative Transform_1 = RigVMFunctionReferenceNode(Input=@n5.Element)
   > RigVMDispatchNode(
    Index=@n5.Index
    Element=RigVMDispatchNode(False=@n5.Element, True=@"Get Negative Transform_1".Output, Condition=RigVMVariableNode() {variable=Negative Side}.Value) {template="DISPATCH_RigVMDispatch_If(in Condition,in True,in False,out Result)"}.Result
    Array=RigVMVariableNode() {variable=Transforms}.Value
   ) {template="DISPATCH_RigVMDispatch_ArraySetAtIndex(io Array,in Index,in Element)"}
   /Script/ControlRig.RigUnit_DebugTransformArrayMutableItemSpace(
    Color=(R=1,G=1,B=1,A=1)
    DebugDrawSettings=(DepthPriority=SDPG_Foreground,Lifetime=-1)
    Mode=Axes
    ParentIndices=()
    Scale=10
    Space=(Type=Bone)
    Thickness=1
    WorldOffset=(Rotation=(X=0,Y=0,Z=0,W=1),Translation=(X=0,Y=0,Z=0),Scale3D=(X=1,Y=1,Z=1))
    bEnabled=True
    ExecutePin=@n5.Completed
    Transforms=@n4.Value
   ) {is=struct RigVMUnitNode}
   &RigUnit_SetTransform(
    Weight=1
    bInitial=true
    Value=RigVMDispatchNode(Index=0, Array=RigVMVariableNode() {variable=Transforms}.Value) {template="DISPATCH_RigVMDispatch_ArrayGetAtIndex(in Array,in Index,out Element)"}.Element
    ExecutePin=@n2.B
    Item=RigVMVariableNode() {variable=FK Space}.Value
   )
   > &RigUnit_SetTransformItemArray(
    Space=GlobalSpace
    Weight=1
    bInitial=true
    bPropagateToChildren=True
    Transforms=RigVMVariableNode() {variable=Transforms}.Value
    Items=RigVMVariableNode() {variable=FK Controls}.Value
   )
   n6 = RigVMVariableNode() {variable=Transforms}
   n7 = &RigVMFunction_MathVectorDistance(
    A=RigVMDispatchNode(Index=0, Array=@n6.Value) {template="DISPATCH_RigVMDispatch_ArrayGetAtIndex(in Array,in Index,out Element)"}.Element.Translation
    B=RigVMDispatchNode(Index=2, Array=@n6.Value) {template="DISPATCH_RigVMDispatch_ArrayGetAtIndex(in Array,in Index,out Element)"}.Element.Translation
   )
   &RigUnit_SetTransform(
    Value=(Rotation=(X=0,Y=0,Z=0,W=1),Translation=(X=0,Y=0,Z=0),Scale3D=(X=1,Y=1,Z=1))
    Weight=1
    bInitial=true
    Value.Scale3D=RigVMDispatchNode(False=(X=1,Y=1,Z=1), True=(X=-1,Y=1,Z=1), Condition=RigVMVariableNode() {variable=Negative Side}.Value) {template="DISPATCH_RigVMDispatch_If(in Condition,in True,in False,out Result)"}.Result
    Value.Translation=&RigVMFunction_MathTransformMul(
     A=(Rotation=(X=0,Y=0,Z=0,W=1),Translation=(X=0,Y=0,Z=0),Scale3D=(X=1,Y=1,Z=1))
     B=RigVMDispatchNode(Index=1, Array=@n6.Value) {template="DISPATCH_RigVMDispatch_ArrayGetAtIndex(in Array,in Index,out Element)"}.Element
     A.Translation.Y=&RigVMFunction_MathFloatDiv(B=2, A=@n7.Result) {template="Divide::Execute(in A,in B,out Result)"}.Result
    ) {template="Multiply::Execute(in A,in B,out Result)"}.Result.Translation
    ExecutePin=@n2.C
    Item=RigVMVariableNode() {variable=Pole Vector Control}.Value
   )
   n8 = RigVMVariableNode() {variable=Transforms}
   n9 = RigVMDispatchNode(Array=@n8.Value) {template="DISPATCH_RigVMDispatch_ArrayGetNum(in Array,out Num)"}
   > &RigUnit_SetTransform(
    Value=(Rotation=(X=0,Y=0,Z=0,W=1),Translation=(X=0,Y=0,Z=0),Scale3D=(X=1,Y=1,Z=1))
    Weight=1
    bInitial=true
    Value.Translation=RigVMDispatchNode(Index=&RigVMFunction_MathIntSub(B=1, A=@n9.Num) {template="Subtract::Execute(in A,in B,out Result)"}.Result, Array=@n8.Value) {template="DISPATCH_RigVMDispatch_ArrayGetAtIndex(in Array,in Index,out Element)"}.Element.Translation
    Value.Scale3D=RigVMDispatchNode(False=(X=1,Y=1,Z=1), True=(X=-1,Y=1,Z=1), Condition=RigVMVariableNode() {variable=Negative Side}.Value) {template="DISPATCH_RigVMDispatch_If(in Condition,in True,in False,out Result)"}.Result
    Item=RigVMVariableNode() {variable=IK Control}.Value
   )
   n10 = RigVMDispatchNode(ExecuteContext=@n2.D, Array=RigVMVariableNode() {variable=IK Parents}.Value) {template="DISPATCH_RigVMDispatch_ArrayIterator(in Array,out Element,out Index,out Count,out Ratio)"}
   > &RigUnit_AddParent(DisplayLabel=None, Parent=@n10.Element, Child=RigVMVariableNode() {variable=IK Control}.Value)
   > &RigUnit_AddParent(DisplayLabel=None, Parent=@n10.Element, Child=RigVMVariableNode() {variable=Pole Vector Control}.Value)
   RigVMFunctionReturnNode(ExecuteContext=@n10.Completed)
  ]}
 ContainedGraph {nodes [
   n1 = RigVMFunctionEntryNode()
   n2 = &RigUnit_SequenceExecution(ExecuteContext=@n1.ExecuteContext)
   &RigUnit_SetTransform(
    Value=(Rotation=(X=0,Y=0,Z=0,W=1),Translation=(X=0,Y=0,Z=0),Scale3D=(X=1,Y=1,Z=1))
    Weight=1
    bInitial=true
    Value.Translation=&RigUnit_GetTransform(Item=RigVMVariableNode() {variable=Bone}.Value).Transform.Translation
    Value.Scale3D=RigVMDispatchNode(False=(X=1,Y=1,Z=1), True=(X=-1,Y=1,Z=1), Condition=RigVMVariableNode() {variable=Negative Side}.Value) {template="DISPATCH_RigVMDispatch_If(in Condition,in True,in False,out Result)"}.Result
    ExecutePin=@n2.A
    Item=RigVMVariableNode() {variable=IK Control}.Value
   )
   n3 = &RigUnit_GetTransform(Item=RigVMVariableNode() {variable=Bone}.Value)
   n4 = &RigVMFunction_MathQuaternionToRotator(Value=@n3.Transform.Rotation) {template="Cast::Execute(in Value,out Result)"}
   n5 = &RigVMFunction_MathQuaternionFromRotator(
    Rotator=(Pitch=0,Yaw=0,Roll=0)
    Rotator.Pitch=&RigVMFunction_MathFloatAdd(B=90, A=@n4.Result.Pitch) {template="Add::Execute(in A,in B,out Result)"}.Result
    Rotator.Roll=@n4.Result.Roll
    Rotator.Yaw=@n4.Result.Yaw
   )
   n6 = &RigVMFunction_MathQuaternionToEuler(RotationOrder=YXZ, Value=@n5.Result)
   n7 = &RigVMFunction_MathFloatAdd(B=180, A=@n6.Result.Y) {template="Add::Execute(in A,in B,out Result)"}
   n8 = RigVMVariableNode() {variable=Negative Side}
   &RigUnit_SetTransform(
    Value=(Rotation=(X=0,Y=0,Z=0,W=1),Translation=(X=0,Y=0,Z=0),Scale3D=(X=1,Y=1,Z=1))
    Weight=1
    bInitial=true
    Value.Translation=@n3.Transform.Translation
    Value.Rotation=RigVMDispatchNode(
     True=&RigVMFunction_MathQuaternionFromEuler(Euler=(X=0,Y=0,Z=0), RotationOrder=YXZ, Euler.Y=@n7.Result, Euler.X=@n6.Result.X, Euler.Z=@n6.Result.Z).Result
     False=@n5.Result
     Condition=@n8.Value
    ) {template="DISPATCH_RigVMDispatch_If(in Condition,in True,in False,out Result)"}.Result
    Value.Scale3D=RigVMDispatchNode(False=(X=1,Y=1,Z=1), True=(X=-1,Y=1,Z=1), Condition=@n8.Value) {template="DISPATCH_RigVMDispatch_If(in Condition,in True,in False,out Result)"}.Result
    ExecutePin=@n2.B
    Item=RigVMVariableNode() {variable=FK Control}.Value
   )
   &RigUnit_SetTransform(
    Weight=1
    bInitial=true
    Value=&RigUnit_GetTransform(Item=RigVMVariableNode() {variable=Bone}.Value).Transform
    ExecutePin=@n2.C
    Item=RigVMVariableNode() {variable=Toes Space}.Value
   )
   RigVMFunctionReturnNode(ExecuteContext=@n2.D)
  ]}
 ContainedGraph {nodes [
   n1 = RigVMFunctionEntryNode()
   n2 = RigVMVariableNode() {variable=IKFK Switch Control}
   n3 = RigVMVariableNode(
    ExecuteContext=@n1.Execute
    Value=/Script/RigVM.RigVMFunction_MathFloatGreater(B=0.5, A=&RigUnit_GetTransform(Space=LocalSpace, Item=@n2.Value).Transform.Translation.X) {is=struct RigVMUnitNode, template="Greater::Execute(in A,in B,out Result)"}.Result
   ) {variable=IK Mode}
   n4 = &RigUnit_SequenceExecution(ExecuteContext=@n3.ExecuteContext)
   Set IKFK Visibility = RigVMFunctionReferenceNode(
    IK Controls=((),())
    Show Controls=true
    ExecuteContext=@n4.A
    FK Controls=RigVMVariableNode() {variable=FK Controls}.Value
    IK Controls.0=RigVMVariableNode() {variable=IK Control}.Value
    IKFK Switch=RigVMVariableNode() {variable=IK Mode}.Value
    IK Controls.1=RigVMVariableNode() {variable=Pole Vector Control}.Value
   )
   n5 = RigVMVariableNode() {variable=FK Orient Control}
   n6 = &RigUnit_GetTransform(Space=LocalSpace, Item=@n5.Value)
   n7 = RigVMVariableNode() {variable=FK Orient References}
   n8 = RigVMDispatchNode(Index=0, Array=@n7.Value) {template="DISPATCH_RigVMDispatch_ArrayGetAtIndex(in Array,in Index,out Element)"}
   n9 = RigVMVariableNode() {variable=Arm Space}
   n10 = /Script/ControlRig.RigUnit_ProjectTransformToNewParent(bChildInitial=True, bOldParentInitial=True, NewParent=@n8.Element, OldParent=@n8.Element, Child=@n9.Value) {is=struct RigVMUnitNode}
   n11 = RigVMDispatchNode(Index=1, Array=@n7.Value) {template="DISPATCH_RigVMDispatch_ArrayGetAtIndex(in Array,in Index,out Element)"}
   &RigUnit_SetTransform(
    Value=(Rotation=(X=0,Y=0,Z=0,W=1),Translation=(X=0,Y=0,Z=0),Scale3D=(X=1,Y=1,Z=1))
    Weight=1
    bInitial=False
    Value.Rotation=RigVMDispatchNode(
     Condition=&RigVMFunction_MathFloatGreater(B=0, A=@n6.Transform.Translation.X) {template="Greater::Execute(in A,in B,out Result)"}.Result
     False=@n10.Transform.Rotation
     True=&RigUnit_ProjectTransformToNewParent(bChildInitial=True, bOldParentInitial=True, NewParent=@n11.Element, OldParent=@n11.Element, Child=@n9.Value).Transform.Rotation
    ) {template="DISPATCH_RigVMDispatch_If(in Condition,in True,in False,out Result)"}.Result
    Value.Scale3D=@n10.Transform.Scale3D
    Value.Translation=@n10.Transform.Translation
    ExecutePin=@n4.B
    Item=@n9.Value
   )
   Branch_RigVMUnitNode = &RigVMFunction_ControlFlowBranch(ExecuteContext=@n4.C, Condition=RigVMVariableNode() {variable=IK Mode}.Value)
   FK Point At = RigVMFunctionReferenceNode(
    Point At Axis=(X=1,Y=0,Z=0)
    Up Vector Axis=(X=0,Y=0,Z=1)
    Up Vector Target=(X=0,Y=0,Z=1)
    ExecuteContext=@Branch_RigVMUnitNode.False
    Negative Side=RigVMVariableNode() {variable=Negative Side}.Value
    Bones=RigVMVariableNode() {variable=Bones}.Value
    FK Controls=RigVMVariableNode() {variable=FK Controls}.Value
   )
   > Match IK onto FK = RigVMFunctionReferenceNode(
    Bones=RigVMVariableNode() {variable=Bones}.Value
    Pole Vector Control=RigVMVariableNode() {variable=Pole Vector Control}.Value
    IK Control=RigVMVariableNode() {variable=IK Control}.Value
   )
   IK 2 Bones = RigVMFunctionReferenceNode(
    Primary Axis=(X=1,Y=0,Z=0)
    Secondary Axis=(X=0,Y=-1,Z=0)
    ExecuteContext=@Branch_RigVMUnitNode.True
    Bones=RigVMVariableNode() {variable=Bones}.Value
    Enable Stretch=RigVMVariableNode() {variable=Enable Stretch}.Value
    Softness=RigVMVariableNode() {variable=Softness}.Value
    Pole Vector Control=RigVMVariableNode() {variable=Pole Vector Control}.Value
    Negative Side=RigVMVariableNode() {variable=Negative Side}.Value
    IK Controls=RigVMVariableNode() {variable=IK Effector}.Value
   )
   > Project to New Parents = RigVMFunctionReferenceNode(Propagate to Children=True, Parents=RigVMVariableNode() {variable=Bones}.Value, Children=RigVMVariableNode() {variable=FK Controls}.Value)
   n12 = &RigUnit_SequenceExecution(ExecuteContext=@n4.D)
   Set IKFK Keys = RigVMFunctionReferenceNode(
    "IK Controls "=((),())
    ExecuteContext=@n12.A
    FK Controls=RigVMVariableNode() {variable=FK Controls}.Value
    IK Controls .0=RigVMVariableNode() {variable=IK Control}.Value
    IK Controls .1=RigVMVariableNode() {variable=Pole Vector Control}.Value
    Switch Control=RigVMVariableNode() {variable=IKFK Switch Control}.Value
    IK Mode=RigVMVariableNode() {variable=IK Mode}.Value
    Previous IK Mode=RigVMVariableNode() {variable=Previous IK Mode}.Value
   )
   n13 = &RigUnit_GetTransform(Item=RigVMDispatchNode(Index=2, Array=RigVMVariableNode() {variable=Bones}.Value) {template="DISPATCH_RigVMDispatch_ArrayGetAtIndex(in Array,in Index,out Element)"}.Element)
   Get Negative Transform = RigVMFunctionReferenceNode(Input=@n13.Transform)
   &RigUnit_SetTransform(
    Weight=1
    bInitial=False
    Value=RigVMDispatchNode(True=@"Get Negative Transform".Output, False=@n13.Transform, Condition=RigVMVariableNode() {variable=Negative Side}.Value) {template="DISPATCH_RigVMDispatch_If(in Condition,in True,in False,out Result)"}.Result
    ExecutePin=@n12.C
    Item=RigVMVariableNode() {variable=Finger Space}.Value
   )
   /Script/ControlRig.RigUnit_GetControlBool(Control=@n5.Value.Name) {is=struct RigVMUnitNode, template="GetControlValue::Execute(in Control,out BoolValue)"}
   RigVMFunctionReturnNode(Execute=@n12.D)
  ]}
 ContainedGraph {nodes [
   n1 = RigVMFunctionEntryNode()
   n2 = RigVMDispatchNode(Array=@n1.Bones, ExecuteContext=@n1.Execute) {template="DISPATCH_RigVMDispatch_ArrayIterator(in Array,out Element,out Index,out Count,out Ratio)"}
   n3 = &RigVMFunction_NameConcat(B=_ctrl, A=@n2.Element.Name) {template="Concat::Execute(in A,in B,out Result)"}
   > &RigUnit_SetTransform(
    Weight=1
    bInitial=False
    bPropagateToChildren=true
    Item=@n2.Element
    Value=&RigUnit_ProjectTransformToNewParent(
     NewParent=(Type=Control)
     OldParent=(Type=Control)
     bChildInitial=True
     bOldParentInitial=True
     Child=@n2.Element
     NewParent.Name=@n3.Result
     OldParent.Name=@n3.Result
    ).Transform
   )
   RigVMFunctionReturnNode(Execute=@n2.Completed)
  ]}
 ContainedGraph {nodes [
   n1 = RigVMFunctionEntryNode()
   > &RigUnit_SetTransform(
    Weight=1
    bInitial=False
    bPropagateToChildren=true
    Item=@n1.Bone
    Value=&RigUnit_ProjectTransformToNewParent(bChildInitial=True, bOldParentInitial=True, Child=@n1.Bone, NewParent=@n1.Control, OldParent=@n1.Control).Transform
   )
   > RigVMFunctionReturnNode()
  ]}
 ContainedGraph {nodes [
   n1 = RigVMFunctionEntryNode()
   n2 = RigVMVariableNode() {variable=IKFK Switch Control}
   n3 = RigVMVariableNode(
    ExecuteContext=@n1.Execute
    Value=&RigVMFunction_MathFloatGreater(B=0.5, A=&RigUnit_GetTransform(Space=LocalSpace, Item=@n2.Value).Transform.Translation.X) {template="Greater::Execute(in A,in B,out Result)"}.Result
   ) {variable=IK Mode}
   n4 = &RigUnit_SequenceExecution(ExecuteContext=@n3.ExecuteContext)
   Set IKFK Visibility = RigVMFunctionReferenceNode(
    IK Controls=((),())
    Show Controls=true
    ExecuteContext=@n4.A
    IK Controls.0=RigVMVariableNode() {variable=IK Control}.Value
    FK Controls=RigVMVariableNode() {variable=FK Controls}.Value
    IK Controls.1=RigVMVariableNode() {variable=Pole Vector Control}.Value
    IKFK Switch=RigVMVariableNode() {variable=IK Mode}.Value
   )
   Branch_RigVMUnitNode = &RigVMFunction_ControlFlowBranch(ExecuteContext=@n4.B, Condition=RigVMVariableNode() {variable=IK Mode}.Value)
   FK Point At = RigVMFunctionReferenceNode(
    Point At Axis=(X=-1,Y=0,Z=0)
    Up Vector Axis=(X=0,Y=0,Z=1)
    Up Vector Target=(X=0,Y=0,Z=1)
    ExecuteContext=@Branch_RigVMUnitNode.False
    Negative Side=RigVMVariableNode() {variable=Negative Side}.Value
    FK Controls=RigVMVariableNode() {variable=FK Controls}.Value
    Bones=RigVMVariableNode() {variable=Bones}.Value
   )
   > Match IK onto FK = RigVMFunctionReferenceNode(
    Bones=RigVMVariableNode() {variable=Bones}.Value
    Pole Vector Control=RigVMVariableNode() {variable=Pole Vector Control}.Value
    IK Control=RigVMVariableNode() {variable=IK Control}.Value
   )
   IK 2 Bones = RigVMFunctionReferenceNode(
    Primary Axis=(X=-1,Y=0,Z=0)
    Secondary Axis=(X=0,Y=1,Z=0)
    ExecuteContext=@Branch_RigVMUnitNode.True
    Bones=RigVMVariableNode() {variable=Bones}.Value
    Enable Stretch=RigVMVariableNode() {variable=Enable Stretch}.Value
    IK Controls=RigVMVariableNode() {variable=IK Effector}.Value
    Softness=RigVMVariableNode() {variable=Softness}.Value
    Pole Vector Control=RigVMVariableNode() {variable=Pole Vector Control}.Value
    Negative Side=RigVMVariableNode() {variable=Negative Side}.Value
   )
   > Project to New Parents = RigVMFunctionReferenceNode(Propagate to Children=True, Parents=RigVMVariableNode() {variable=Bones}.Value, Children=RigVMVariableNode() {variable=FK Controls}.Value)
   Set IKFK Keys = RigVMFunctionReferenceNode(
    "IK Controls "=((),())
    ExecuteContext=@n4.C
    FK Controls=RigVMVariableNode() {variable=FK Controls}.Value
    Switch Control=RigVMVariableNode() {variable=IKFK Switch Control}.Value
    IK Controls .0=RigVMVariableNode() {variable=IK Control}.Value
    IK Controls .1=RigVMVariableNode() {variable=Pole Vector Control}.Value
    Previous IK Mode=RigVMVariableNode() {variable=Previous IK Mode}.Value
    IK Mode=RigVMVariableNode() {variable=IK Mode}.Value
   )
   RigVMFunctionReturnNode(Execute=@n4.D)
  ]}
 ContainedGraph {nodes [
   Branch_4_1_1_1_RigVMUnitNode = &RigVMFunction_ControlFlowBranch(ExecuteContext=(), Condition=RigVMVariableNode() {variable=IK Mode}.Value)
   /Script/ControlRig.RigUnit_SendEvent(
    Event=RequestAutoKey
    bEnable=True
    bOnlyDuringInteraction=True
    ExecutePin=@Branch_4_1_1_1_RigVMUnitNode.False
    Item=RigVMVariableNode() {variable=IK Control}.Value
   ) {is=struct RigVMUnitNode, template="SendEvent::Execute(in Event,in Item,in OffsetInSeconds,in bEnable,in bOnlyDuringInteraction)"}
   > &RigUnit_SendEvent(Event=RequestAutoKey, bEnable=True, bOnlyDuringInteraction=True, Item=RigVMVariableNode() {variable=FK Control}.Value) {template="SendEvent::Execute(in Event,in Item,in OffsetInSeconds,in bEnable,in bOnlyDuringInteraction)"}
   > RigVMVariableNode(Value=True) {variable=IK Mode}
   Branch_4_1_1_RigVMUnitNode = &RigVMFunction_ControlFlowBranch(ExecuteContext=(), Condition=RigVMVariableNode() {variable=IK Mode}.Value)
   &RigUnit_SendEvent(
    Event=RequestAutoKey
    bEnable=True
    bOnlyDuringInteraction=True
    ExecutePin=@Branch_4_1_1_RigVMUnitNode.True
    Item=RigVMVariableNode() {variable=IK Control}.Value
   ) {template="SendEvent::Execute(in Event,in Item,in OffsetInSeconds,in bEnable,in bOnlyDuringInteraction)"}
   > &RigUnit_SendEvent(Event=RequestAutoKey, bEnable=True, bOnlyDuringInteraction=True, Item=RigVMVariableNode() {variable=FK Control}.Value) {template="SendEvent::Execute(in Event,in Item,in OffsetInSeconds,in bEnable,in bOnlyDuringInteraction)"}
   > RigVMVariableNode(Value=True) {variable=IK Mode}
   RigVMFunctionEntryNode()
   n1 = RigVMVariableNode() {variable=IKFK Switch Control}
   > n2 = RigVMVariableNode(Value=&RigVMFunction_MathFloatGreater(B=0.5, A=&RigUnit_GetTransform(Item=@n1.Value).Transform.Translation.X) {template="Greater::Execute(in A,in B,out Result)"}.Result) {variable=IK Mode}
   n3 = &RigUnit_SequenceExecution(ExecuteContext=@n2.ExecuteContext)
   n4 = RigVMVariableNode() {variable=IK Mode}
   /Script/ControlRig.RigUnit_SetControlVisibility(ExecutePin=@n3.A, bVisible=@n4.Value, Item=RigVMVariableNode() {variable=IK Control}.Value) {is=struct RigVMUnitNode, template="SetControlVisibility::Execute(in Item,in Pattern,in bVisible)"}
   > &RigUnit_SetControlVisibility(bVisible=&RigVMFunction_MathBoolNot(Value=@n4.Value).Result, Item=RigVMVariableNode() {variable=FK Control}.Value) {template="SetControlVisibility::Execute(in Item,in Pattern,in bVisible)"}
   Branch_RigVMUnitNode = &RigVMFunction_ControlFlowBranch(ExecuteContext=@n3.B, Condition=RigVMVariableNode() {variable=IK Mode}.Value)
   n5 = RigVMVariableNode() {variable=Bone}
   n6 = RigVMVariableNode() {variable=FK Control}
   &RigUnit_SetTransform(
    Weight=1
    bInitial=False
    ExecutePin=@Branch_RigVMUnitNode.False
    Value=&RigUnit_ProjectTransformToNewParent(bChildInitial=True, bOldParentInitial=True, Child=@n5.Value, NewParent=@n6.Value, OldParent=@n6.Value).Transform
    Item=@n5.Value
   )
   n7 = RigVMVariableNode() {variable=IK Control}
   n8 = RigVMVariableNode() {variable=Bone}
   > &RigUnit_SetTransform(
    Weight=1
    bInitial=False
    Value=&RigUnit_ProjectTransformToNewParent(bChildInitial=True, bOldParentInitial=True, Child=@n7.Value, NewParent=@n8.Value, OldParent=@n8.Value).Transform
    Item=@n7.Value
   )
   n9 = RigVMVariableNode() {variable=Bone}
   n10 = RigVMVariableNode() {variable=IK Control}
   &RigUnit_SetTransform(
    Weight=1
    bInitial=False
    ExecutePin=@Branch_RigVMUnitNode.True
    Value=&RigUnit_ProjectTransformToNewParent(bChildInitial=True, bOldParentInitial=True, Child=@n9.Value, NewParent=@n10.Value, OldParent=@n10.Value).Transform
    Item=@n9.Value
   )
   n11 = RigVMVariableNode() {variable=Bone}
   n12 = RigVMVariableNode() {variable=FK Control}
   > &RigUnit_SetTransform(
    Weight=1
    bInitial=False
    Value=&RigUnit_ProjectTransformToNewParent(bChildInitial=True, bOldParentInitial=True, NewParent=@n11.Value, OldParent=@n11.Value, Child=@n12.Value).Transform
    Item=@n12.Value
   )
   Branch_4_1_1_2_RigVMUnitNode = &RigVMFunction_ControlFlowBranch(
    Condition=/Script/RigVM.RigVMFunction_MathBoolNotEquals(B=RigVMVariableNode() {variable=IK Mode}.Value, A=RigVMVariableNode() {variable=Previous IK Mode}.Value) {is=struct RigVMUnitNode}.Result
    ExecuteContext=@n3.C
   )
   &RigUnit_SendEvent(
    Event=RequestAutoKey
    bEnable=True
    bOnlyDuringInteraction=True
    ExecutePin=@Branch_4_1_1_2_RigVMUnitNode.True
    Item=RigVMVariableNode() {variable=IK Control}.Value
   ) {template="SendEvent::Execute(in Event,in Item,in OffsetInSeconds,in bEnable,in bOnlyDuringInteraction)"}
   > &RigUnit_SendEvent(Event=RequestAutoKey, bEnable=True, bOnlyDuringInteraction=True, Item=RigVMVariableNode() {variable=FK Control}.Value) {template="SendEvent::Execute(in Event,in Item,in OffsetInSeconds,in bEnable,in bOnlyDuringInteraction)"}
   n13 = RigVMVariableNode() {variable=Toes Space}
   n14 = RigVMVariableNode() {variable=Bone}
   &RigUnit_SetTransform(
    Weight=1
    bInitial=False
    Value=&RigUnit_ProjectTransformToNewParent(bChildInitial=True, bOldParentInitial=True, Child=@n13.Value, NewParent=@n14.Value, OldParent=@n14.Value).Transform
    ExecutePin=@n3.D
    Item=@n13.Value
   )
   > RigVMFunctionReturnNode()
  ]}
 ContainedGraph {nodes [
   n1 = RigVMFunctionEntryNode()
   n2 = &RigUnit_SequenceExecution(ExecuteContext=@n1.ExecuteContext)
   n3 = RigVMDispatchNode(ExecuteContext=@n2.A, Array=RigVMVariableNode() {variable=FK Controls}.Value) {template="DISPATCH_RigVMDispatch_ArrayIterator(in Array,out Element,out Index,out Count,out Ratio)"}
   > &RigUnit_SetControlVisibility(
    Item=@n3.Element
    bVisible=RigVMDispatchNode(False=False, Condition=@n1.Show Controls, True=&RigVMFunction_MathBoolNot(Value=@n1.IKFK Switch).Result) {template="DISPATCH_RigVMDispatch_If(in Condition,in True,in False,out Result)"}.Result
   ) {template="SetControlVisibility::Execute(in Item,in Pattern,in bVisible)"}
   n4 = RigVMDispatchNode(ExecuteContext=@n2.B, Array=RigVMVariableNode() {variable=IK Controls}.Value) {template="DISPATCH_RigVMDispatch_ArrayIterator(in Array,out Element,out Index,out Count,out Ratio)"}
   > &RigUnit_SetControlVisibility(
    Item=@n4.Element
    bVisible=RigVMDispatchNode(False=False, True=@n1.IKFK Switch, Condition=@n1.Show Controls) {template="DISPATCH_RigVMDispatch_If(in Condition,in True,in False,out Result)"}.Result
   ) {template="SetControlVisibility::Execute(in Item,in Pattern,in bVisible)"}
   RigVMFunctionReturnNode(ExecuteContext=@n2.D)
  ]}
 ContainedGraph {nodes [
   n1 = RigVMFunctionEntryNode()
   n2 = &RigUnit_SequenceExecution(ExecuteContext=@n1.ExecuteContext)
   Project to New Parent_1 = RigVMFunctionReferenceNode(
    Propagate to Children=True
    ExecuteContext=@n2.A
    Children=RigVMVariableNode() {variable=FK Controls}.Value
    Parents=RigVMVariableNode() {variable=Bones}.Value
   )
   RigVMFunctionReturnNode(ExecuteContext=@n2.D)
   Spine Match IK onto FK = RigVMFunctionReferenceNode(
    Execute=@n2.B
    Offsets=RigVMVariableNode() {variable=Offsets}.Value
    Bones=RigVMVariableNode() {variable=Bones}.Value
    IK Controls=RigVMVariableNode() {variable=IK Controls}.Value
   )
  ]}
 ContainedGraph {nodes [
   n1 = RigVMFunctionEntryNode()
   n2 = &RigUnit_SequenceExecution(ExecuteContext=@n1.ExecuteContext)
   n3 = RigVMDispatchNode(Index=1, Array=RigVMVariableNode() {variable=Bones}.Value) {template="DISPATCH_RigVMDispatch_ArrayGetAtIndex(in Array,in Index,out Element)"}
   n4 = RigVMVariableNode() {variable=Head FK Space}
   n5 = RigVMVariableNode() {variable=Head Orient Reference}
   &RigUnit_SetTransform(
    Value=(Rotation=(X=0,Y=0,Z=0,W=1),Translation=(X=0,Y=0,Z=0),Scale3D=(X=1,Y=1,Z=1))
    Weight=1
    bInitial=False
    Value.Translation=&RigUnit_ProjectTransformToNewParent(bChildInitial=True, bOldParentInitial=True, NewParent=@n3.Element, OldParent=@n3.Element, Child=@n4.Value).Transform.Translation
    Value.Rotation=&RigUnit_ProjectTransformToNewParent(bChildInitial=True, bOldParentInitial=True, Child=@n4.Value, NewParent=@n5.Value, OldParent=@n5.Value).Transform.Rotation
    ExecutePin=@n2.A
    Item=@n4.Value
   )
   n6 = RigVMVariableNode() {variable=Head IK Space}
   n7 = RigVMVariableNode() {variable=Head Orient Reference}
   n8 = /Script/ControlRig.RigUnit_HierarchyGetParent(bDefaultParent=True, Child=@n6.Value) {is=struct RigVMUnitNode}
   > &RigUnit_SetTransform(
    Value=(Rotation=(X=0,Y=0,Z=0,W=1),Translation=(X=0,Y=0,Z=0),Scale3D=(X=1,Y=1,Z=1))
    Weight=1
    bInitial=False
    Value.Rotation=&RigUnit_ProjectTransformToNewParent(bChildInitial=True, bOldParentInitial=True, Child=@n6.Value, NewParent=@n7.Value, OldParent=@n7.Value).Transform.Rotation
    Value.Translation=&RigUnit_ProjectTransformToNewParent(bChildInitial=True, bOldParentInitial=True, NewParent=@n8.Parent, OldParent=@n8.Parent, Child=@n6.Value).Transform.Translation
    Item=@n6.Value
   )
   Project to New Parents = RigVMFunctionReferenceNode(
    Propagate to Children=True
    ExecuteContext=@n2.B
    Parents=RigVMVariableNode() {variable=Bones}.Value
    Children=RigVMVariableNode() {variable=FK Controls}.Value
   )
   n9 = RigVMVariableNode() {variable=Bones}
   n10 = RigVMDispatchNode(
    Index=&RigVMFunction_MathIntSub(B=1, A=RigVMDispatchNode(Array=@n9.Value) {template="DISPATCH_RigVMDispatch_ArrayGetNum(in Array,out Num)"}.Num) {template="Subtract::Execute(in A,in B,out Result)"}.Result
    Array=@n9.Value
   ) {template="DISPATCH_RigVMDispatch_ArrayGetAtIndex(in Array,in Index,out Element)"}
   n11 = RigVMVariableNode() {variable=IK Control}
   &RigUnit_SetTransform(
    Weight=1
    bInitial=False
    Value=&RigUnit_ProjectTransformToNewParent(bChildInitial=True, bOldParentInitial=True, NewParent=@n10.Element, OldParent=@n10.Element, Child=@n11.Value).Transform
    ExecutePin=@n2.C
    Item=@n11.Value
   )
   RigVMFunctionReturnNode(ExecuteContext=@n2.D)
  ]}
 ContainedGraph {nodes [
   n1 = RigVMFunctionEntryNode()
   n2 = &RigUnit_SetTransform(
    Weight=1
    bInitial=False
    bPropagateToChildren=true
    Item=@n1.Control
    ExecutePin=@n1.Execute
    Value=&RigUnit_ProjectTransformToNewParent(bChildInitial=True, bOldParentInitial=True, NewParent=@n1.Bone, OldParent=@n1.Bone, Child=@n1.Control).Transform
   )
   RigVMFunctionReturnNode(Execute=@n2.ExecutePin)
  ]}
 ContainedGraph {nodes [
   n1 = RigVMFunctionEntryNode()
   n2 = &RigUnit_SequenceExecution(ExecuteContext=@n1.ExecuteContext)
   n3 = RigVMVariableNode() {variable=FK Orient References}
   n4 = RigVMDispatchNode(Index=1, Array=@n3.Value) {template="DISPATCH_RigVMDispatch_ArrayGetAtIndex(in Array,in Index,out Element)"}
   n5 = RigVMVariableNode() {variable=FK Arm Space}
   n6 = RigVMDispatchNode(Index=0, Array=@n3.Value) {template="DISPATCH_RigVMDispatch_ArrayGetAtIndex(in Array,in Index,out Element)"}
   n7 = &RigUnit_ProjectTransformToNewParent(bChildInitial=True, bOldParentInitial=True, NewParent=@n6.Element, OldParent=@n6.Element, Child=@n5.Value)
   &RigUnit_SetTransform(
    Value=(Rotation=(X=0,Y=0,Z=0,W=1),Translation=(X=0,Y=0,Z=0),Scale3D=(X=1,Y=1,Z=1))
    Weight=1
    bInitial=False
    bPropagateToChildren=true
    Value.Rotation=&RigUnit_ProjectTransformToNewParent(bChildInitial=True, bOldParentInitial=True, NewParent=@n4.Element, OldParent=@n4.Element, Child=@n5.Value).Transform.Rotation
    Value.Scale3D=@n7.Transform.Scale3D
    Value.Translation=@n7.Transform.Translation
    ExecutePin=@n2.A
    Item=@n5.Value
   )
   n8 = RigVMDispatchNode(Array=@n1.FK Controls, ExecuteContext=@n2.B) {template="DISPATCH_RigVMDispatch_ArrayIterator(in Array,out Element,out Index,out Count,out Ratio)"}
   n9 = RigVMDispatchNode(Index=@n8.Index, Array=@n1.Bones) {template="DISPATCH_RigVMDispatch_ArrayGetAtIndex(in Array,in Index,out Element)"}
   > &RigUnit_SetTransform(
    Weight=1
    bInitial=False
    Item=@n8.Element
    Value=&RigUnit_ProjectTransformToNewParent(bChildInitial=True, bOldParentInitial=True, NewParent=@n9.Element, OldParent=@n9.Element, Child=@n8.Element).Transform
   )
   n10 = RigVMDispatchNode(Index=0, Array=RigVMVariableNode() {variable=Bones}.Value) {template="DISPATCH_RigVMDispatch_ArrayGetAtIndex(in Array,in Index,out Element)"}
   n11 = RigVMVariableNode() {variable=UpV Control}
   &RigUnit_SetTransform(
    Weight=1
    bInitial=False
    Value=&RigUnit_ProjectTransformToNewParent(bChildInitial=True, bOldParentInitial=True, NewParent=@n10.Element, OldParent=@n10.Element, Child=@n11.Value).Transform
    ExecutePin=@n2.C
    Item=@n11.Value
   )
   n12 = RigVMDispatchNode(Index=2, Array=RigVMVariableNode() {variable=Bones}.Value) {template="DISPATCH_RigVMDispatch_ArrayGetAtIndex(in Array,in Index,out Element)"}
   n13 = RigVMVariableNode() {variable=Ik Control}
   > &RigUnit_SetTransform(
    Weight=1
    bInitial=False
    bPropagateToChildren=true
    Value=&RigUnit_ProjectTransformToNewParent(bChildInitial=True, bOldParentInitial=True, NewParent=@n12.Element, OldParent=@n12.Element, Child=@n13.Value).Transform
    Item=@n13.Value
   )
   n14 = RigVMDispatchNode(Index=2, Array=RigVMVariableNode() {variable=Bones}.Value) {template="DISPATCH_RigVMDispatch_ArrayGetAtIndex(in Array,in Index,out Element)"}
   n15 = RigVMVariableNode() {variable=Fingers Space}
   &RigUnit_SetTransform(
    Weight=1
    bInitial=False
    Value=&RigUnit_ProjectTransformToNewParent(bChildInitial=True, bOldParentInitial=True, NewParent=@n14.Element, OldParent=@n14.Element, Child=@n15.Value).Transform
    ExecutePin=@n2.D
    Item=@n15.Value
   )
   > RigVMFunctionReturnNode()
  ]}
 ContainedGraph {nodes [
   n1 = RigVMFunctionEntryNode()
   n2 = RigVMDispatchNode(Array=@n1.Bones, ExecuteContext=@n1.Execute) {template="DISPATCH_RigVMDispatch_ArrayIterator(in Array,out Element,out Index,out Count,out Ratio)"}
   n3 = &RigVMFunction_NameConcat(B=_ctrl, A=@n2.Element.Name) {template="Concat::Execute(in A,in B,out Result)"}
   > &RigUnit_SetTransform(
    Item=(Type=Control,Name="None")
    Weight=1
    bInitial=false
    Item.Name=@n3.Result
    Value=&RigUnit_ProjectTransformToNewParent(Child=(Type=Control), bChildInitial=True, bOldParentInitial=True, NewParent=@n2.Element, OldParent=@n2.Element, Child.Name=@n3.Result).Transform
   )
   RigVMFunctionReturnNode(Execute=@n2.Completed)
  ]}
 ContainedGraph {nodes [
   n1 = RigVMFunctionEntryNode()
   n2 = &RigUnit_SequenceExecution(ExecuteContext=@n1.ExecuteContext)
   n3 = RigVMDispatchNode(Array=@n1.FK Controls, ExecuteContext=@n2.B) {template="DISPATCH_RigVMDispatch_ArrayIterator(in Array,out Element,out Index,out Count,out Ratio)"}
   n4 = RigVMDispatchNode(Index=@n3.Index, Array=@n1.Bones) {template="DISPATCH_RigVMDispatch_ArrayGetAtIndex(in Array,in Index,out Element)"}
   > &RigUnit_SetTransform(
    Weight=1
    bInitial=False
    Item=@n3.Element
    Value=&RigUnit_ProjectTransformToNewParent(bChildInitial=True, bOldParentInitial=True, NewParent=@n4.Element, OldParent=@n4.Element, Child=@n3.Element).Transform
   )
   n5 = RigVMDispatchNode(Index=0, Array=RigVMVariableNode() {variable=Bones}.Value) {template="DISPATCH_RigVMDispatch_ArrayGetAtIndex(in Array,in Index,out Element)"}
   n6 = RigVMVariableNode() {variable=UpV Control}
   &RigUnit_SetTransform(
    Weight=1
    bInitial=False
    Value=&RigUnit_ProjectTransformToNewParent(bChildInitial=True, bOldParentInitial=True, NewParent=@n5.Element, OldParent=@n5.Element, Child=@n6.Value).Transform
    ExecutePin=@n2.C
    Item=@n6.Value
   )
   n7 = RigVMDispatchNode(Index=2, Array=RigVMVariableNode() {variable=Bones}.Value) {template="DISPATCH_RigVMDispatch_ArrayGetAtIndex(in Array,in Index,out Element)"}
   n8 = RigVMVariableNode() {variable=Ik Control}
   > &RigUnit_SetTransform(
    Weight=1
    bInitial=False
    bPropagateToChildren=true
    Value=&RigUnit_ProjectTransformToNewParent(bChildInitial=True, bOldParentInitial=True, NewParent=@n7.Element, OldParent=@n7.Element, Child=@n8.Value).Transform
    Item=@n8.Value
   )
   RigVMFunctionReturnNode(ExecuteContext=@n2.D)
  ]}
 ContainedGraph {nodes [
   n1 = RigVMFunctionEntryNode()
   n2 = &RigUnit_SequenceExecution(ExecuteContext=@n1.Execute)
   n3 = RigVMVariableNode() {variable=Bone}
   n4 = RigVMVariableNode() {variable=FK Control}
   &RigUnit_SetTransform(
    Weight=1
    bInitial=False
    Value=&RigUnit_ProjectTransformToNewParent(bChildInitial=True, bOldParentInitial=True, NewParent=@n3.Value, OldParent=@n3.Value, Child=@n4.Value).Transform
    ExecutePin=@n2.A
    Item=@n4.Value
   )
   n5 = RigVMVariableNode() {variable=Bone}
   n6 = RigVMVariableNode() {variable=IK Control}
   &RigUnit_SetTransform(
    Weight=1
    bInitial=False
    Value=&RigUnit_ProjectTransformToNewParent(bChildInitial=True, bOldParentInitial=True, NewParent=@n5.Value, OldParent=@n5.Value, Child=@n6.Value).Transform
    ExecutePin=@n2.B
    Item=@n6.Value
   )
   &RigUnit_SetTransform(
    Weight=1
    bInitial=false
    Value=&RigUnit_GetTransform(Item=RigVMVariableNode() {variable=Bone}.Value).Transform
    ExecutePin=@n2.C
    Item=RigVMVariableNode() {variable=Toes Space}.Value
   )
   RigVMFunctionReturnNode(Execute=@n2.D)
  ]}
 ContainedGraph {nodes [
   n1 = RigVMFunctionEntryNode()
   n2 = &RigUnit_SequenceExecution(ExecuteContext=@n1.ExecuteContext)
   n3 = RigVMDispatchNode(ExecuteContext=@n2.A, Array=RigVMVariableNode() {variable=Bones}.Value) {template="DISPATCH_RigVMDispatch_ArrayIterator(in Array,out Element,out Index,out Count,out Ratio)"}
   Branch_1_RigVMUnitNode = &RigVMFunction_ControlFlowBranch(
    ExecuteContext=@n3.ExecuteContext
    Condition=/Script/RigVM.RigVMFunction_MathIntGreaterEqual(A=@n3.Index, B=&RigVMFunction_MathIntSub(B=1, A=@n3.Count) {template="Subtract::Execute(in A,in B,out Result)"}.Result) {is=struct RigVMUnitNode, template="GreaterEqual::Execute(in A,in B,out Result)"}.Result
   )
   n4 = RigVMVariableNode() {variable=FK Controls}
   n5 = /Script/ControlRig.RigUnit_AimBoneMath(
    DebugSettings=(Scale=10,WorldOffset=(Rotation=(X=0,Y=0,Z=0,W=1),Translation=(X=0,Y=0,Z=0),Scale3D=(X=1,Y=1,Z=1)))
    InputTransform=(Rotation=(X=0,Y=0,Z=0,W=1),Translation=(X=0,Y=0,Z=0),Scale3D=(X=1,Y=1,Z=1))
    Primary=(Weight=1,Axis=(X=1,Y=0,Z=0),Target=(X=0,Y=0,Z=0),Kind=Location,Space=(Type=Bone))
    Secondary=(Weight=1,Axis=(X=0,Y=0,Z=1),Target=(X=1,Y=0,Z=0),Kind=Direction,Space=(Type=Bone))
    Weight=1
    Secondary.Space=RigVMDispatchNode(Index=@n3.Index, Array=@n4.Value) {template="DISPATCH_RigVMDispatch_ArrayGetAtIndex(in Array,in Index,out Element)"}.Element
    Primary.Space=RigVMDispatchNode(
     Index=/Script/RigVM.RigVMFunction_MathIntAdd(B=1, A=@n3.Index) {is=struct RigVMUnitNode, template="Add::Execute(in A,in B,out Result)"}.Result
     Array=@n4.Value
    ) {template="DISPATCH_RigVMDispatch_ArrayGetAtIndex(in Array,in Index,out Element)"}.Element
    InputTransform.Translation=RigVMDispatchNode(Index=@n3.Index, Array=&RigUnit_GetTransformItemArray(Space=GlobalSpace, Items=@n4.Value).Transforms) {template="DISPATCH_RigVMDispatch_ArrayGetAtIndex(in Array,in Index,out Element)"}.Element.Translation
    Primary.Axis=RigVMVariableNode() {variable=Point At Axis}.Value
    Secondary.Axis=RigVMVariableNode() {variable=Up Vector Axis}.Value
    Secondary.Target=RigVMVariableNode() {variable=Up Vector Target}.Value
   ) {is=struct RigVMUnitNode}
   n6 = &RigVMFunction_MathQuaternionFromRotator(Rotator=(Pitch=180,Yaw=0,Roll=0))
   &RigUnit_SetTransform(
    Weight=1
    bInitial=False
    Item=@n3.Element
    ExecutePin=@Branch_1_RigVMUnitNode.False
    Value=RigVMDispatchNode(
     False=@n5.Result
     True=&RigVMFunction_MathTransformMul(A=(Rotation=(X=0,Y=0,Z=0,W=1),Translation=(X=0,Y=0,Z=0),Scale3D=(X=1,Y=1,Z=1)), B=@n5.Result, A.Rotation=@n6.Result) {template="Multiply::Execute(in A,in B,out Result)"}.Result
     Condition=RigVMVariableNode() {variable=Negative Side}.Value
    ) {template="DISPATCH_RigVMDispatch_If(in Condition,in True,in False,out Result)"}.Result
   )
   n7 = RigVMVariableNode() {variable=Bones}
   n8 = &RigVMFunction_MathIntSub(B=1, A=RigVMDispatchNode(Array=@n7.Value) {template="DISPATCH_RigVMDispatch_ArrayGetNum(in Array,out Num)"}.Num) {template="Subtract::Execute(in A,in B,out Result)"}
   n9 = RigVMDispatchNode(Index=@n8.Result, Array=@n7.Value) {template="DISPATCH_RigVMDispatch_ArrayGetAtIndex(in Array,in Index,out Element)"}
   n10 = RigVMDispatchNode(Index=@n8.Result, Array=RigVMVariableNode() {variable=FK Controls}.Value) {template="DISPATCH_RigVMDispatch_ArrayGetAtIndex(in Array,in Index,out Element)"}
   &RigUnit_SetTransform(
    Weight=1
    bInitial=False
    Item=@n9.Element
    Value=&RigUnit_ProjectTransformToNewParent(bChildInitial=True, bOldParentInitial=True, Child=@n9.Element, NewParent=@n10.Element, OldParent=@n10.Element).Transform
    ExecutePin=@n2.B
   )
   RigVMFunctionReturnNode(ExecuteContext=@n2.D)
  ]}
 ContainedGraph {nodes [
   RigVMFunctionEntryNode()
   n1 = RigVMDispatchNode(Index=0, Array=RigVMVariableNode() {variable=Bones}.Value) {template="DISPATCH_RigVMDispatch_ArrayGetAtIndex(in Array,in Index,out Element)"}
   n2 = RigVMVariableNode() {variable=Pole Vector Control}
   > &RigUnit_SetTransform(
    Weight=1
    bInitial=False
    Value=&RigUnit_ProjectTransformToNewParent(bChildInitial=True, bOldParentInitial=True, NewParent=@n1.Element, OldParent=@n1.Element, Child=@n2.Value).Transform
    Item=@n2.Value
   )
   n3 = RigVMDispatchNode(Index=2, Array=RigVMVariableNode() {variable=Bones}.Value) {template="DISPATCH_RigVMDispatch_ArrayGetAtIndex(in Array,in Index,out Element)"}
   n4 = RigVMVariableNode() {variable=IK Control}
   > &RigUnit_SetTransform(
    Weight=1
    bInitial=False
    Value=&RigUnit_ProjectTransformToNewParent(bChildInitial=True, bOldParentInitial=True, NewParent=@n3.Element, OldParent=@n3.Element, Child=@n4.Value).Transform
    Item=@n4.Value
   )
   > RigVMFunctionReturnNode()
  ]}
 ContainedGraph {nodes [
   n1 = RigVMFunctionEntryNode()
   n2 = &RigUnit_SequenceExecution(ExecuteContext=@n1.ExecuteContext)
   n3 = RigVMVariableNode() {variable=Bones}
   n4 = RigVMVariableNode() {variable=Primary Axis}
   n5 = RigVMVariableNode() {variable=Negative Side}
   n6 = RigVMVariableNode() {variable=Secondary Axis}
   n7 = RigVMVariableNode() {variable=Bones}
   n8 = RigVMVariableNode() {variable=IK Controls}
   Softness = RigVMCollapseNode(
    Bones=RigVMVariableNode() {variable=Bones}.Value
    Softness=RigVMVariableNode() {variable=Softness}.Value
    Effector=RigVMVariableNode() {variable=IK Controls}.Value
   )
   /Script/ControlRig.RigUnit_TwoBoneIKSimplePerItem(
    DebugSettings=(Scale=10,WorldOffset=(Rotation=(X=0,Y=0,Z=0,W=1),Translation=(X=0,Y=0,Z=0),Scale3D=(X=1,Y=1,Z=1)))
    PoleVector=(X=0,Y=0,Z=0)
    PoleVectorKind=Location
    SecondaryAxisWeight=1
    StretchMaximumRatio=2
    StretchStartRatio=1
    Weight=1
    bPropagateToChildren=True
    ItemB=RigVMDispatchNode(Index=1, Array=@n3.Value) {template="DISPATCH_RigVMDispatch_ArrayGetAtIndex(in Array,in Index,out Element)"}.Element
    EffectorItem=RigVMDispatchNode(Index=2, Array=@n3.Value) {template="DISPATCH_RigVMDispatch_ArrayGetAtIndex(in Array,in Index,out Element)"}.Element
    ItemA=RigVMDispatchNode(Index=0, Array=@n3.Value) {template="DISPATCH_RigVMDispatch_ArrayGetAtIndex(in Array,in Index,out Element)"}.Element
    PrimaryAxis=RigVMDispatchNode(
     True=/Script/RigVM.RigVMFunction_MathVectorNegate(Value=@n4.Value) {is=struct RigVMUnitNode, template="Negate::Execute(in Value,out Result)"}.Result
     Condition=@n5.Value
     False=@n4.Value
    ) {template="DISPATCH_RigVMDispatch_If(in Condition,in True,in False,out Result)"}.Result
    SecondaryAxis=RigVMDispatchNode(
     True=&RigVMFunction_MathVectorNegate(Value=@n6.Value) {template="Negate::Execute(in Value,out Result)"}.Result
     Condition=@n5.Value
     False=@n6.Value
    ) {template="DISPATCH_RigVMDispatch_If(in Condition,in True,in False,out Result)"}.Result
    Effector=&RigUnit_ProjectTransformToNewParent(
     bChildInitial=True
     bOldParentInitial=True
     Child=RigVMDispatchNode(Index=2, Array=@n7.Value) {template="DISPATCH_RigVMDispatch_ArrayGetAtIndex(in Array,in Index,out Element)"}.Element
     NewParent=@n8.Value
     OldParent=@n8.Value
    ).Transform
    ExecutePin=@n2.A
    ItemALength=@Softness.Length A
    ItemBLength=@Softness.Length B
    PoleVectorSpace=RigVMVariableNode() {variable=Pole Vector Control}.Value
    bEnableStretch=RigVMVariableNode() {variable=Enable Stretch}.Value
   ) {is=struct RigVMUnitNode}
   n9 = RigVMVariableNode() {variable=Bones}
   /Script/ControlRig.RigUnit_DebugLineItemSpace(
    Color=(R=0.2,G=0.2,B=0.2,A=1)
    DebugDrawSettings=(DepthPriority=SDPG_Foreground,Lifetime=-1)
    Space=(Type=Bone)
    Thickness=0.5
    WorldOffset=(Rotation=(X=0,Y=0,Z=0,W=1),Translation=(X=0,Y=0,Z=0),Scale3D=(X=1,Y=1,Z=1))
    bEnabled=True
    B=&RigUnit_GetTransform(Item=RigVMVariableNode() {variable=Pole Vector Control}.Value).Transform.Translation
    A=&RigUnit_GetTransform(Item=RigVMDispatchNode(Index=1, Array=@n9.Value) {template="DISPATCH_RigVMDispatch_ArrayGetAtIndex(in Array,in Index,out Element)"}.Element).Transform.Translation
    ExecutePin=@n2.B
   ) {is=struct RigVMUnitNode}
   RigVMFunctionReturnNode(ExecuteContext=@n2.D)
  ]}
 ContainedGraph {nodes [
   n1 = RigVMVariableNode() {variable=Softness}
   n2 = RigVMFunctionEntryNode()
   n3 = /Script/RigVM.RigVMFunction_MathFloatAbs(Value=&RigUnit_GetTransform(
     Space=LocalSpace
     bInitial=true
     Item=RigVMDispatchNode(Index=1, Array=@n2.Bones) {template="DISPATCH_RigVMDispatch_ArrayGetAtIndex(in Array,in Index,out Element)"}.Element
    ).Transform.Translation.X) {is=struct RigVMUnitNode, template="Absolute::Execute(in Value,out Result)"}
   n4 = &RigVMFunction_MathFloatAbs(Value=&RigUnit_GetTransform(
     Space=LocalSpace
     bInitial=true
     Item=RigVMDispatchNode(Index=2, Array=@n2.Bones) {template="DISPATCH_RigVMDispatch_ArrayGetAtIndex(in Array,in Index,out Element)"}.Element
    ).Transform.Translation.X) {template="Absolute::Execute(in Value,out Result)"}
   n5 = &RigVMFunction_MathFloatAdd(A=@n3.Result, B=@n4.Result) {template="Add::Execute(in A,in B,out Result)"}
   n6 = &RigVMFunction_MathFloatMul(
    B=0.02
    A=&RigVMFunction_MathFloatMul(A=@n5.Result, B=RigVMVariableNode() {variable=Softness}.Value) {template="Multiply::Execute(in A,in B,out Result)"}.Result
   ) {template="Multiply::Execute(in A,in B,out Result)"}
   n7 = /Script/RigVM.RigVMFunction_MathFloatSub(A=@n5.Result, B=@n6.Result) {is=struct RigVMUnitNode, template="Subtract::Execute(in A,in B,out Result)"}
   n8 = &RigVMFunction_MathVectorDistance(
    B=&RigUnit_GetTransform(Item=@n2.Effector).Transform.Translation
    A=&RigUnit_GetTransform(Item=RigVMDispatchNode(Index=0, Array=@n2.Bones) {template="DISPATCH_RigVMDispatch_ArrayGetAtIndex(in Array,in Index,out Element)"}.Element).Transform.Translation
   )
   Exponential Softness = RigVMCollapseNode(Softness=@n6.Result, DA=@n7.Result, Distance=@n8.Result, Stretch Distance=@n8.Result)
   n9 = RigVMDispatchNode(
    False=1
    Condition=&RigVMFunction_MathBoolAnd(
     A=&RigVMFunction_MathFloatGreater(B=0, A=@n1.Value) {template="Greater::Execute(in A,in B,out Result)"}.Result
     B=&RigVMFunction_MathFloatGreater(B=@n7.Result, A=@n8.Result) {template="Greater::Execute(in A,in B,out Result)"}.Result
    ).Result
    True=/Script/RigVM.RigVMFunction_MathFloatMax(B=1, A=@"Exponential Softness".Result) {is=struct RigVMUnitNode, template="Maximum::Execute(in A,in B,out Result)"}.Result
   ) {template="DISPATCH_RigVMDispatch_If(in Condition,in True,in False,out Result)"}
   RigVMFunctionReturnNode(
    Length A=&RigVMFunction_MathFloatMul(B=@n9.Result, A=@n3.Result) {template="Multiply::Execute(in A,in B,out Result)"}.Result
    Length B=&RigVMFunction_MathFloatMul(B=@n9.Result, A=@n4.Result) {template="Multiply::Execute(in A,in B,out Result)"}.Result
   )
  ]}
 ContainedGraph {nodes [
   n1 = RigVMFunctionEntryNode()
   n2 = &RigVMFunction_MathFloatNegate(Value=&RigVMFunction_MathFloatDiv(
     B=@n1.Softness
     A=&RigVMFunction_MathFloatSub(B=@n1.DA, A=@n1.Distance) {template="Subtract::Execute(in A,in B,out Result)"}.Result
    ) {template="Divide::Execute(in A,in B,out Result)"}.Result) {template="Negate::Execute(in Value,out Result)"}
   n3 = &RigVMFunction_MathFloatMul(
    B=@n1.Softness
    A=&RigVMFunction_MathFloatSub(
     A=1
     B=/Script/RigVM.RigVMFunction_MathFloatExponential(Value=@n2.Result) {is=struct RigVMUnitNode, template="Exponential::Execute(in Value,out Result)"}.Result
    ) {template="Subtract::Execute(in A,in B,out Result)"}.Result
   ) {template="Multiply::Execute(in A,in B,out Result)"}
   RigVMFunctionReturnNode(Result=&RigVMFunction_MathFloatDiv(
     A=@n1.Stretch Distance
     B=&RigVMFunction_MathFloatAdd(B=@n1.DA, A=@n3.Result) {template="Add::Execute(in A,in B,out Result)"}.Result
    ) {template="Divide::Execute(in A,in B,out Result)"}.Result)
  ]}
 ContainedGraph {nodes [
   n1 = RigVMFunctionEntryNode()
   > n2 = RigVMDispatchNode(Array=@n1.Children) {template="DISPATCH_RigVMDispatch_ArrayIterator(in Array,out Element,out Index,out Count,out Ratio)"}
   n3 = RigVMDispatchNode(Index=@n2.Index, Array=@n1.Parents) {template="DISPATCH_RigVMDispatch_ArrayGetAtIndex(in Array,in Index,out Element)"}
   > &RigUnit_SetTransform(
    Weight=1
    bInitial=False
    Item=@n2.Element
    bPropagateToChildren=@n1.Propagate to Children
    Value=&RigUnit_ProjectTransformToNewParent(bChildInitial=True, bOldParentInitial=True, NewParent=@n3.Element, OldParent=@n3.Element, Child=@n2.Element).Transform
   )
   RigVMFunctionReturnNode(ExecuteContext=@n2.Completed)
  ]}
 ContainedGraph {nodes [
   n1 = RigVMFunctionEntryNode()
   n2 = &RigUnit_SequenceExecution(ExecuteContext=@n1.ExecuteContext)
   Branch_1_1_RigVMUnitNode = &RigVMFunction_ControlFlowBranch(Condition=&RigVMFunction_MathBoolNotEquals(A=@n1.IK Mode, B=@n1.Previous IK Mode).Result, ExecuteContext=@n2.A)
   n3 = &RigUnit_SequenceExecution(ExecuteContext=@Branch_1_1_RigVMUnitNode.True)
   n4 = RigVMDispatchNode(ExecuteContext=@n3.A, Array=RigVMVariableNode() {variable=FK Controls}.Value) {template="DISPATCH_RigVMDispatch_ArrayIterator(in Array,out Element,out Index,out Count,out Ratio)"}
   > &RigUnit_SendEvent(Event=RequestAutoKey, bEnable=True, bOnlyDuringInteraction=True, Item=@n4.Element) {template="SendEvent::Execute(in Event,in Item,in OffsetInSeconds,in bEnable,in bOnlyDuringInteraction)"}
   n5 = RigVMDispatchNode(ExecuteContext=@n3.B, Array=RigVMVariableNode() {variable="IK Controls "}.Value) {template="DISPATCH_RigVMDispatch_ArrayIterator(in Array,out Element,out Index,out Count,out Ratio)"}
   > &RigUnit_SendEvent(Event=RequestAutoKey, bEnable=True, bOnlyDuringInteraction=True, Item=@n5.Element) {template="SendEvent::Execute(in Event,in Item,in OffsetInSeconds,in bEnable,in bOnlyDuringInteraction)"}
   &RigUnit_SendEvent(Event=RequestAutoKey, bEnable=True, bOnlyDuringInteraction=True, ExecutePin=@n3.C, Item=RigVMVariableNode() {variable=Switch Control}.Value) {template="SendEvent::Execute(in Event,in Item,in OffsetInSeconds,in bEnable,in bOnlyDuringInteraction)"}
   RigVMFunctionReturnNode(ExecuteContext=@n2.D)
  ]}
 ContainedGraph {nodes [
   n1 = RigVMFunctionEntryNode()
   n2 = &RigUnit_SequenceExecution(ExecuteContext=@n1.ExecuteContext)
   n3 = RigVMDispatchNode(ExecuteContext=@n2.A, Array=RigVMVariableNode() {variable=Bones}.Value) {template="DISPATCH_RigVMDispatch_ArrayIterator(in Array,out Element,out Index,out Count,out Ratio)"}
   n4 = &RigUnit_GetTransform(Item=RigVMDispatchNode(Index=0, Array=RigVMVariableNode() {variable=Bones}.Value) {template="DISPATCH_RigVMDispatch_ArrayGetAtIndex(in Array,in Index,out Element)"}.Element)
   n5 = &RigUnit_GetTransform(Item=@n3.Element)
   Get Negative Transform_1 = RigVMFunctionReferenceNode(
    Input=(Rotation=(X=0,Y=0,Z=0,W=1),Translation=(X=0,Y=0,Z=0),Scale3D=(X=0,Y=0,Z=0))
    Input.Scale3D=@n5.Transform.Scale3D
    Input.Translation=@n5.Transform.Translation
   )
   n6 = RigVMDispatchNode(
    False=(Rotation=(X=0,Y=0,Z=0,W=1),Translation=(X=0,Y=0,Z=0),Scale3D=(X=0,Y=0,Z=0))
    True=@"Get Negative Transform_1".Output
    False.Scale3D=@n5.Transform.Scale3D
    False.Translation=@n5.Transform.Translation
    False.Rotation=@n4.Transform.Rotation
    Condition=RigVMVariableNode() {variable=Negative Side}.Value
   ) {template="DISPATCH_RigVMDispatch_If(in Condition,in True,in False,out Result)"}
   > &RigUnit_SetTransform(
    Value=(Rotation=(X=0,Y=0,Z=0,W=1),Translation=(X=0,Y=0,Z=0),Scale3D=(X=1,Y=1,Z=1))
    Weight=1
    bInitial=true
    Item=RigVMDispatchNode(Index=@n3.Index, Array=RigVMVariableNode() {variable=BK Buffers}.Value) {template="DISPATCH_RigVMDispatch_ArrayGetAtIndex(in Array,in Index,out Element)"}.Element
    Value.Rotation=@n4.Transform.Rotation
    Value.Scale3D=@n6.Result.Scale3D
    Value.Translation=@n6.Result.Translation
   )
   > &RigUnit_SetTransform(
    Value=(Rotation=(X=0,Y=0,Z=0,W=1),Translation=(X=0,Y=0,Z=0),Scale3D=(X=1,Y=1,Z=1))
    Weight=1
    bInitial=true
    Item=RigVMDispatchNode(Index=@n3.Index, Array=RigVMVariableNode() {variable=BK Controls}.Value) {template="DISPATCH_RigVMDispatch_ArrayGetAtIndex(in Array,in Index,out Element)"}.Element
    Value.Rotation=@n4.Transform.Rotation
    Value.Scale3D=@n6.Result.Scale3D
    Value.Translation=@n6.Result.Translation
   )
   n7 = RigVMVariableNode() {variable=Bones}
   n8 = RigVMDispatchNode(Array=@n7.Value) {template="DISPATCH_RigVMDispatch_ArrayGetNum(in Array,out Num)"}
   n9 = &RigUnit_GetTransform(Item=RigVMDispatchNode(Index=&RigVMFunction_MathIntSub(B=1, A=@n8.Num) {template="Subtract::Execute(in A,in B,out Result)"}.Result, Array=@n7.Value) {template="DISPATCH_RigVMDispatch_ArrayGetAtIndex(in Array,in Index,out Element)"}.Element)
   n10 = RigVMDispatchNode(False=(X=1,Y=1,Z=1), True=(X=-1,Y=1,Z=1), Condition=RigVMVariableNode() {variable=Negative Side}.Value) {template="DISPATCH_RigVMDispatch_If(in Condition,in True,in False,out Result)"}
   &RigUnit_SetTransform(
    Value=(Rotation=(X=0,Y=0,Z=0,W=1),Translation=(X=0,Y=0,Z=0),Scale3D=(X=1,Y=1,Z=1))
    Weight=1
    bInitial=true
    Value.Translation=@n9.Transform.Translation
    Value.Rotation=@n4.Transform.Rotation
    Value.Scale3D=@n10.Result
    ExecutePin=@n2.B
    Item=RigVMVariableNode() {variable=Roll Buffer}.Value
   )
   > &RigUnit_SetTransform(
    Value=(Rotation=(X=0,Y=0,Z=0,W=1),Translation=(X=0,Y=0,Z=0),Scale3D=(X=1,Y=1,Z=1))
    Weight=1
    bInitial=true
    Value.Translation=@n9.Transform.Translation
    Value.Rotation=@n4.Transform.Rotation
    Value.Scale3D=@n10.Result
    Item=RigVMVariableNode() {variable=Roll Control}.Value
   )
   RigVMFunctionReturnNode(ExecuteContext=@n2.D)
  ]}
 ContainedGraph {nodes [
   RigVMFunctionEntryNode()
   n1 = RigVMVariableNode() {variable=IKFK Switch Control}
   > n2 = RigVMVariableNode(Value=&RigVMFunction_MathFloatGreater(B=0.5, A=&RigUnit_GetTransform(Item=@n1.Value).Transform.Translation.X) {template="Greater::Execute(in A,in B,out Result)"}.Result) {variable=IK Mode}
   n3 = &RigUnit_SequenceExecution(ExecuteContext=@n2.ExecuteContext)
   n4 = RigVMDispatchNode(ExecuteContext=@n3.A, Array=RigVMVariableNode() {variable=BK Controls}.Value) {template="DISPATCH_RigVMDispatch_ArrayIterator(in Array,out Element,out Index,out Count,out Ratio)"}
   n5 = RigVMVariableNode() {variable=IK Mode}
   > SetControlVisibility = &RigUnit_SetControlVisibility(Item=@n4.Element, bVisible=@n5.Value) {template="SetControlVisibility::Execute(in Item,in Pattern,in bVisible)"}
   &RigUnit_SetControlVisibility(ExecutePin=@n4.Completed, bVisible=@n5.Value, Item=RigVMVariableNode() {variable=Roll Control}.Value) {template="SetControlVisibility::Execute(in Item,in Pattern,in bVisible)"}
   RigUnit_GetTransform = &RigUnit_GetTransform(Space=LocalSpace, Item=RigVMVariableNode() {variable=Roll Control}.Value)
   n6 = /Script/RigVM.RigVMFunction_MathFloatClamp(
    Maximum=180
    Minimum=0
    Value=&RigVMFunction_MathQuaternionToRotator(Value=@RigUnit_GetTransform.Transform.Rotation) {template="Cast::Execute(in Value,out Result)"}.Result.Yaw
   ) {is=struct RigVMUnitNode, template="Clamp::Execute(in Value,in Minimum,in Maximum,out Result)"}
   n7 = &RigUnit_GetTransform(
    Space=LocalSpace
    bInitial=false
    Item=RigVMDispatchNode(Index=0, Array=RigVMVariableNode() {variable=BK Controls}.Value) {template="DISPATCH_RigVMDispatch_ArrayGetAtIndex(in Array,in Index,out Element)"}.Element
   )
   n8 = /Script/RigVM.RigVMFunction_MathTransformMakeAbsolute(
    Local=&RigVMFunction_MathTransformMul(
     B=(Rotation=(X=0,Y=0,Z=0,W=1),Translation=(X=0,Y=0,Z=0),Scale3D=(X=1,Y=1,Z=1))
     A=&RigVMFunction_MathTransformMakeRelative(Global=(Rotation=(X=0,Y=0,Z=0,W=1),Translation=(X=0,Y=0,Z=0),Scale3D=(X=1,Y=1,Z=1)), Parent=@n7.Transform).Local
     B.Rotation=@n7.Transform.Rotation
    ) {template="Multiply::Execute(in A,in B,out Result)"}.Result
    Parent=@n7.Transform
   ) {is=struct RigVMUnitNode, template="Make Absolute::Execute(in Local,in Parent,out Global)"}
   n9 = &RigUnit_GetTransform(
    Space=LocalSpace
    bInitial=true
    Item=RigVMDispatchNode(Index=0, Array=RigVMVariableNode() {variable=BK Buffers}.Value) {template="DISPATCH_RigVMDispatch_ArrayGetAtIndex(in Array,in Index,out Element)"}.Element
   )
   RigUnit_SetTransform = &RigUnit_SetTransform(
    Space=LocalSpace
    Weight=1
    bInitial=False
    Item=RigVMDispatchNode(Index=0, Array=RigVMVariableNode() {variable=BK Buffers}.Value) {template="DISPATCH_RigVMDispatch_ArrayGetAtIndex(in Array,in Index,out Element)"}.Element
    Value=&RigVMFunction_MathTransformMul(
     A=(Rotation=(X=0,Y=0,Z=0,W=1),Translation=(X=0,Y=0,Z=0),Scale3D=(X=1,Y=1,Z=1))
     A.Rotation=&RigVMFunction_MathQuaternionFromRotator(Rotator=(Pitch=0,Yaw=0,Roll=0), Rotator.Yaw=@n6.Result).Result
     B=&RigVMFunction_MathTransformMul(A=@n8.Global, B=@n9.Transform) {template="Multiply::Execute(in A,in B,out Result)"}.Result
    ) {template="Multiply::Execute(in A,in B,out Result)"}.Result
    ExecutePin=@n3.B
   )
   n10 = RigVMDispatchNode(
    Index=0
    Array=RigVMDispatchNode(Array=RigVMVariableNode() {variable=BK Buffers}.Value) {template="DISPATCH_RigVMDispatch_ArrayClone(in Array,out Clone)"}.Clone
    ExecuteContext=@n3.C
   ) {template="DISPATCH_RigVMDispatch_ArrayRemove(io Array,in Index)"}
   > n11 = RigVMDispatchNode(Array=@n10.Array) {template="DISPATCH_RigVMDispatch_ArrayReverse(io Array)"}
   > n12 = RigVMDispatchNode(Array=@n11.Array) {template="DISPATCH_RigVMDispatch_ArrayIterator(in Array,out Element,out Index,out Count,out Ratio)"}
   n13 = &RigVMFunction_MathFloatMul(A=RigVMVariableNode() {variable=Roll Clamp}.Value, B=RigVMVariableNode() {variable=Roll Blend}.Value) {template="Multiply::Execute(in A,in B,out Result)"}
   n14 = &RigVMFunction_MathIntEquals(B=&RigVMFunction_MathIntSub(B=1, A=@n12.Count) {template="Subtract::Execute(in A,in B,out Result)"}.Result, A=@n12.Index)
   n15 = &RigVMFunction_MathFloatMul(
    B=@n13.Result
    A=/Script/RigVM.RigVMFunction_MathIntToFloat(Value=@n12.Index) {is=struct RigVMUnitNode, template="Cast::Execute(in Value,out Result)"}.Result
   ) {template="Multiply::Execute(in A,in B,out Result)"}
   n16 = RigVMVariableNode() {variable=Roll Control}
   n17 = &RigVMFunction_MathFloatNegate(Value=&RigVMFunction_MathQuaternionToRotator(Value=&RigUnit_GetTransform(Space=LocalSpace, Item=@n16.Value).Transform.Rotation) {template="Cast::Execute(in Value,out Result)"}.Result.Yaw) {template="Negate::Execute(in Value,out Result)"}
   n18 = &RigVMFunction_MathFloatNegate(Value=&RigVMFunction_MathFloatClamp(
     Minimum=0
     Maximum=RigVMDispatchNode(True=180, False=@n13.Result, Condition=@n14.Result) {template="DISPATCH_RigVMDispatch_If(in Condition,in True,in False,out Result)"}.Result
     Value=&RigVMFunction_MathFloatSub(B=@n15.Result, A=@n17.Result) {template="Subtract::Execute(in A,in B,out Result)"}.Result
    ) {template="Clamp::Execute(in Value,in Minimum,in Maximum,out Result)"}.Result) {template="Negate::Execute(in Value,out Result)"}
   > &RigUnit_SetRotation(
    Weight=1
    bInitial=False
    Item=@n12.Element
    Value=&RigVMFunction_MathTransformMul(
     B=(Rotation=(X=0,Y=0,Z=0,W=1),Translation=(X=0,Y=0,Z=0),Scale3D=(X=1,Y=1,Z=1))
     B.Rotation=&RigVMFunction_MathQuaternionFromRotator(Rotator=(Pitch=0,Yaw=0,Roll=0), Rotator.Yaw=@n18.Result).Result
     A=&RigUnit_GetTransform(Space=LocalSpace, bInitial=true, Item=@n12.Element).Transform
    ) {template="Multiply::Execute(in A,in B,out Result)"}.Result.Rotation
   )
   n19 = &RigUnit_SequenceExecution(ExecuteContext=@n3.D)
   n20 = RigVMDispatchNode(Index=1, Array=RigVMVariableNode() {variable=BK Buffers}.Value) {template="DISPATCH_RigVMDispatch_ArrayGetAtIndex(in Array,in Index,out Element)"}
   n21 = &RigUnit_GetTransform(
    Space=LocalSpace
    bInitial=false
    Item=RigVMDispatchNode(Index=1, Array=RigVMVariableNode() {variable=BK Controls}.Value) {template="DISPATCH_RigVMDispatch_ArrayGetAtIndex(in Array,in Index,out Element)"}.Element
   )
   n22 = &RigVMFunction_MathTransformMakeAbsolute(
    Local=&RigVMFunction_MathTransformMul(
     B=(Rotation=(X=0,Y=0,Z=0,W=1),Translation=(X=0,Y=0,Z=0),Scale3D=(X=1,Y=1,Z=1))
     A=&RigVMFunction_MathTransformMakeRelative(Global=(Rotation=(X=0,Y=0,Z=0,W=1),Translation=(X=0,Y=0,Z=0),Scale3D=(X=1,Y=1,Z=1)), Parent=@n21.Transform).Local
     B.Rotation=@n21.Transform.Rotation
    ) {template="Multiply::Execute(in A,in B,out Result)"}.Result
    Parent=@n21.Transform
   ) {template="Make Absolute::Execute(in Local,in Parent,out Global)"}
   n23 = &RigUnit_GetTransform(Space=LocalSpace, bInitial=true, Item=@n20.Element)
   &RigUnit_SetTransform(
    Space=LocalSpace
    Weight=1
    bInitial=False
    Item=@n20.Element
    Value=&RigVMFunction_MathTransformMul(
     A=(Rotation=(X=0,Y=0,Z=0,W=1),Translation=(X=0,Y=0,Z=0),Scale3D=(X=1,Y=1,Z=1))
     A.Rotation=&RigUnit_GetTransform(Space=LocalSpace, Item=@n20.Element).Transform.Rotation
     B=&RigVMFunction_MathTransformMul(A=@n22.Global, B=@n23.Transform) {template="Multiply::Execute(in A,in B,out Result)"}.Result
    ) {template="Multiply::Execute(in A,in B,out Result)"}.Result
    ExecutePin=@n19.A
   )
   n24 = &RigUnit_HierarchyGetParent(
    bDefaultParent=True
    Child=RigVMDispatchNode(Index=0, Array=RigVMVariableNode() {variable=BK Buffers}.Value) {template="DISPATCH_RigVMDispatch_ArrayGetAtIndex(in Array,in Index,out Element)"}.Element
   )
   n25 = RigVMVariableNode() {variable=Roll Buffer}
   RigUnit_SetRotation = &RigUnit_SetRotation(
    Space=GlobalSpace
    Weight=1
    bInitial=False
    Value=&RigUnit_ProjectTransformToNewParent(bChildInitial=True, bOldParentInitial=True, NewParent=@n24.Parent, OldParent=@n24.Parent, Child=@n25.Value).Transform.Rotation
    ExecutePin=@n19.B
    Item=@n25.Value
   )
   RigVMFunctionReturnNode(ExecuteContext=@n19.D)
  ]}
 ContainedGraph {nodes [
   n1 = RigVMFunctionEntryNode()
   n2 = RigVMDispatchNode(Num=4, Array=RigVMVariableNode() {variable=Outputs}.Value, ExecuteContext=@n1.Execute) {template="DISPATCH_RigVMDispatch_ArraySetNum(io Array,in Num)"}
   n3 = RigVMVariableNode() {variable=Point 0}
   > n4 = RigVMDispatchNode(Index=0, Array=@n2.Array, Element=@n3.Value) {template="DISPATCH_RigVMDispatch_ArraySetAtIndex(io Array,in Index,in Element)"}
   n5 = RigVMVariableNode() {variable=Point 3}
   n6 = &RigVMFunction_MathVectorDistance(B=@n1.Point 0, A=@n1.Point 1)
   n7 = &RigVMFunction_MathFloatAdd(A=@n6.Result, B=&RigVMFunction_MathVectorDistance(A=@n1.Point 1, B=@n1.Point 2).Result) {template="Add::Execute(in A,in B,out Result)"}
   n8 = &RigVMFunction_MathFloatAdd(A=@n7.Result, B=&RigVMFunction_MathVectorDistance(A=@n1.Point 2, B=@n1.Point 3).Result) {template="Add::Execute(in A,in B,out Result)"}
   n9 = RigVMVariableNode() {variable=Auto Weights}
   n10 = RigVMDispatchNode(
    True=&RigVMFunction_MathFloatDiv(B=@n8.Result, A=@n6.Result) {template="Divide::Execute(in A,in B,out Result)"}.Result
    Condition=@n9.Value
    False=RigVMVariableNode() {variable=Weight 1}.Value
   ) {template="DISPATCH_RigVMDispatch_If(in Condition,in True,in False,out Result)"}
   b3t1 = RigVMCollapseNode(t=@n10.Result)
   n11 = RigVMVariableNode() {variable=Point 0}
   b0t1 = RigVMCollapseNode(t=@n10.Result)
   n12 = /Script/RigVM.RigVMFunction_MathVectorScale(Value=@n11.Value, Factor=@b0t1.Result) {is=struct RigVMUnitNode, template="Scale::Execute(in Value,in Factor,out Result)"}
   n13 = RigVMVariableNode() {variable=Point 1}
   n14 = RigVMDispatchNode(
    True=&RigVMFunction_MathFloatDiv(A=@n7.Result, B=@n8.Result) {template="Divide::Execute(in A,in B,out Result)"}.Result
    Condition=@n9.Value
    False=RigVMVariableNode() {variable=Weight 2}.Value
   ) {template="DISPATCH_RigVMDispatch_If(in Condition,in True,in False,out Result)"}
   b3t2 = RigVMCollapseNode(t=@n14.Result)
   b0t2 = RigVMCollapseNode(t=@n14.Result)
   n15 = &RigVMFunction_MathVectorScale(Value=@n11.Value, Factor=@b0t2.Result) {template="Scale::Execute(in Value,in Factor,out Result)"}
   n16 = RigVMVariableNode() {variable=Point 2}
   b1t1 = RigVMCollapseNode(t=@n10.Result)
   b1t2 = RigVMCollapseNode(t=@n14.Result)
   b2t1 = RigVMCollapseNode(t=@n10.Result)
   b2t2 = RigVMCollapseNode(t=@n14.Result)
   Solve = RigVMCollapseNode(
    c=/Script/RigVM.RigVMFunction_MathVectorSub(
     B=&RigVMFunction_MathVectorScale(Value=@n5.Value, Factor=@b3t1.Result) {template="Scale::Execute(in Value,in Factor,out Result)"}.Result
     A=&RigVMFunction_MathVectorSub(B=@n12.Result, A=@n13.Value) {template="Subtract::Execute(in A,in B,out Result)"}.Result
    ) {is=struct RigVMUnitNode, template="Subtract::Execute(in A,in B,out Result)"}.Result
    f=&RigVMFunction_MathVectorSub(
     B=&RigVMFunction_MathVectorScale(Value=@n5.Value, Factor=@b3t2.Result) {template="Scale::Execute(in Value,in Factor,out Result)"}.Result
     A=&RigVMFunction_MathVectorSub(B=@n15.Result, A=@n16.Value) {template="Subtract::Execute(in A,in B,out Result)"}.Result
    ) {template="Subtract::Execute(in A,in B,out Result)"}.Result
    a=@b1t1.Result
    d=@b1t2.Result
    b=@b2t1.Result
    e=@b2t2.Result
   )
   > n17 = RigVMDispatchNode(Index=1, Array=@n4.Array, Element=@Solve.i) {template="DISPATCH_RigVMDispatch_ArraySetAtIndex(io Array,in Index,in Element)"}
   > n18 = RigVMDispatchNode(Index=2, Array=@n17.Array, Element=@Solve.j) {template="DISPATCH_RigVMDispatch_ArraySetAtIndex(io Array,in Index,in Element)"}
   n19 = RigVMVariableNode() {variable=Point 3}
   > n20 = RigVMDispatchNode(Index=3, Array=@n18.Array, Element=@n19.Value) {template="DISPATCH_RigVMDispatch_ArraySetAtIndex(io Array,in Index,in Element)"}
   RigVMFunctionReturnNode(
    Control Points=@n20.Array
    Execute=@n20.ExecuteContext
    Spline=/Script/ControlRigSpline.RigUnit_ControlRigSplineFromPoints(
     Points=((X=0,Y=0,Z=0),(X=0,Y=0,Z=0),(X=0,Y=0,Z=0),(X=0,Y=0,Z=0))
     SamplesPerSegment=16
     SplineMode=BSpline
     bClosed=False
     Points.1=@Solve.i
     Points.2=@Solve.j
     Points.0=@n3.Value
     Points.3=@n19.Value
    ) {is=struct RigVMUnitNode}.Spline
   )
  ]}
 ContainedGraph {nodes [
   RigVMFunctionEntryNode()
   n1 = &RigVMFunction_MathFloatDiv(
    A=&RigVMFunction_MathFloatMul(A=RigVMVariableNode() {variable=a}.Value, B=RigVMVariableNode() {variable=e}.Value) {template="Multiply::Execute(in A,in B,out Result)"}.Result
    B=RigVMVariableNode() {variable=d}.Value
   ) {template="Divide::Execute(in A,in B,out Result)"}
   n2 = RigVMVariableNode() {variable=b}
   n3 = &RigVMFunction_MathFloatDiv(A=RigVMVariableNode() {variable=a}.Value, B=RigVMVariableNode() {variable=d}.Value) {template="Divide::Execute(in A,in B,out Result)"}
   n4 = RigVMVariableNode() {variable=f}
   n5 = &RigVMFunction_MathVectorScale(
    Factor=&RigVMFunction_MathFloatDiv(A=1, B=&RigVMFunction_MathFloatSub(B=@n1.Result, A=@n2.Value) {template="Subtract::Execute(in A,in B,out Result)"}.Result) {template="Divide::Execute(in A,in B,out Result)"}.Result
    Value=&RigVMFunction_MathVectorSub(
     B=&RigVMFunction_MathVectorScale(Factor=@n3.Result, Value=@n4.Value) {template="Scale::Execute(in Value,in Factor,out Result)"}.Result
     A=RigVMVariableNode() {variable=c}.Value
    ) {template="Subtract::Execute(in A,in B,out Result)"}.Result
   ) {template="Scale::Execute(in Value,in Factor,out Result)"}
   n6 = RigVMVariableNode() {variable=a}
   n7 = &RigVMFunction_MathVectorScale(Value=@n5.Result, Factor=RigVMVariableNode() {variable=b}.Value) {template="Scale::Execute(in Value,in Factor,out Result)"}
   n8 = RigVMVariableNode() {variable=c}
   RigVMFunctionReturnNode(
    j=@n5.Result
    i=&RigVMFunction_MathVectorScale(
     Factor=&RigVMFunction_MathFloatDiv(A=1, B=@n6.Value) {template="Divide::Execute(in A,in B,out Result)"}.Result
     Value=&RigVMFunction_MathVectorSub(B=@n7.Result, A=@n8.Value) {template="Subtract::Execute(in A,in B,out Result)"}.Result
    ) {template="Scale::Execute(in Value,in Factor,out Result)"}.Result
   )
  ]}
 ContainedGraph {nodes [
   n1 = RigVMFunctionEntryNode()
   n2 = &RigVMFunction_MathFloatSub(A=1, B=@n1.t) {template="Subtract::Execute(in A,in B,out Result)"}
   n3 = &RigVMFunction_MathFloatMul(A=@n2.Result, B=@n2.Result) {template="Multiply::Execute(in A,in B,out Result)"}
   RigVMFunctionReturnNode(Result=&RigVMFunction_MathFloatMul(B=3, A=&RigVMFunction_MathFloatMul(A=@n1.t, B=@n3.Result) {template="Multiply::Execute(in A,in B,out Result)"}.Result) {template="Multiply::Execute(in A,in B,out Result)"}.Result)
  ]}
 ContainedGraph {nodes [
   n1 = RigVMFunctionEntryNode()
   n2 = &RigVMFunction_MathFloatMul(A=@n1.t, B=@n1.t) {template="Multiply::Execute(in A,in B,out Result)"}
   n3 = &RigVMFunction_MathFloatSub(A=1, B=@n1.t) {template="Subtract::Execute(in A,in B,out Result)"}
   RigVMFunctionReturnNode(Result=&RigVMFunction_MathFloatMul(B=3, A=&RigVMFunction_MathFloatMul(B=@n2.Result, A=@n3.Result) {template="Multiply::Execute(in A,in B,out Result)"}.Result) {template="Multiply::Execute(in A,in B,out Result)"}.Result)
  ]}
 ContainedGraph {nodes [
   n1 = RigVMFunctionEntryNode()
   n2 = &RigVMFunction_MathFloatSub(A=1, B=@n1.t) {template="Subtract::Execute(in A,in B,out Result)"}
   n3 = &RigVMFunction_MathFloatMul(A=@n2.Result, B=@n2.Result) {template="Multiply::Execute(in A,in B,out Result)"}
   RigVMFunctionReturnNode(Result=&RigVMFunction_MathFloatMul(B=3, A=&RigVMFunction_MathFloatMul(A=@n1.t, B=@n3.Result) {template="Multiply::Execute(in A,in B,out Result)"}.Result) {template="Multiply::Execute(in A,in B,out Result)"}.Result)
  ]}
 ContainedGraph {nodes [
   n1 = RigVMFunctionEntryNode()
   n2 = &RigVMFunction_MathFloatMul(A=@n1.t, B=@n1.t) {template="Multiply::Execute(in A,in B,out Result)"}
   n3 = &RigVMFunction_MathFloatSub(A=1, B=@n1.t) {template="Subtract::Execute(in A,in B,out Result)"}
   RigVMFunctionReturnNode(Result=&RigVMFunction_MathFloatMul(B=3, A=&RigVMFunction_MathFloatMul(B=@n2.Result, A=@n3.Result) {template="Multiply::Execute(in A,in B,out Result)"}.Result) {template="Multiply::Execute(in A,in B,out Result)"}.Result)
  ]}
 ContainedGraph {nodes [
   n1 = RigVMFunctionEntryNode()
   RigVMFunctionReturnNode(Result=/Script/RigVM.RigVMFunction_MathFloatPow(B=3, A=&RigVMFunction_MathFloatSub(A=1, B=@n1.t) {template="Subtract::Execute(in A,in B,out Result)"}.Result) {is=struct RigVMUnitNode, template="Power::Execute(in A,in B,out Result)"}.Result)
  ]}
 ContainedGraph {nodes [RigVMFunctionReturnNode(Result=&RigVMFunction_MathFloatPow(B=3, A=RigVMFunctionEntryNode().t) {template="Power::Execute(in A,in B,out Result)"}.Result)]}
 ContainedGraph {nodes [
   n1 = RigVMFunctionEntryNode()
   RigVMFunctionReturnNode(Result=&RigVMFunction_MathFloatPow(B=3, A=&RigVMFunction_MathFloatSub(A=1, B=@n1.t) {template="Subtract::Execute(in A,in B,out Result)"}.Result) {template="Power::Execute(in A,in B,out Result)"}.Result)
  ]}
 ContainedGraph {nodes [RigVMFunctionReturnNode(Result=&RigVMFunction_MathFloatPow(B=3, A=RigVMFunctionEntryNode().t) {template="Power::Execute(in A,in B,out Result)"}.Result)]}
 ContainedGraph {nodes [
   n1 = &RigUnit_GetTransformItemArray(Space=GlobalSpace, Items=RigVMVariableNode() {variable=Bones}.Value)
   n2 = &RigUnit_SequenceExecution(ExecuteContext=@n4.ExecuteContext)
   BSpline Passing thru 4 Points_1 = RigVMFunctionReferenceNode(
    Auto Weights=False
    Weight 1=0.333
    Weight 2=0.666
    Point 0=RigVMDispatchNode(Index=0, Array=@n1.Transforms) {template="DISPATCH_RigVMDispatch_ArrayGetAtIndex(in Array,in Index,out Element)"}.Element.Translation
    Point 1=RigVMDispatchNode(Index=1, Array=@n1.Transforms) {template="DISPATCH_RigVMDispatch_ArrayGetAtIndex(in Array,in Index,out Element)"}.Element.Translation
    Point 2=RigVMDispatchNode(Index=2, Array=@n1.Transforms) {template="DISPATCH_RigVMDispatch_ArrayGetAtIndex(in Array,in Index,out Element)"}.Element.Translation
    Point 3=RigVMDispatchNode(Index=3, Array=@n1.Transforms) {template="DISPATCH_RigVMDispatch_ArrayGetAtIndex(in Array,in Index,out Element)"}.Element.Translation
    Execute=@n2.A
   )
   n3 = RigVMDispatchNode(ExecuteContext=@"BSpline Passing thru 4 Points_1".Execute, Array=RigVMVariableNode() {variable=Bones}.Value) {template="DISPATCH_RigVMDispatch_ArrayIterator(in Array,out Element,out Index,out Count,out Ratio)"}
   > &RigUnit_SetTransform(
    Value=(Rotation=(X=0,Y=0,Z=0,W=1),Translation=(X=0,Y=0,Z=0),Scale3D=(X=1,Y=1,Z=1))
    Weight=1
    bInitial=true
    bPropagateToChildren=false
    Item=RigVMDispatchNode(Index=@n3.Index, Array=RigVMVariableNode() {variable=FK Controls}.Value) {template="DISPATCH_RigVMDispatch_ArrayGetAtIndex(in Array,in Index,out Element)"}.Element
    Value.Translation=&RigUnit_GetTransform(Item=@n3.Element).Transform.Translation
   )
   > &RigUnit_SetTransform(
    Value=(Rotation=(X=0,Y=0,Z=0,W=1),Translation=(X=0,Y=0,Z=0),Scale3D=(X=1,Y=1,Z=1))
    Weight=1
    bInitial=true
    bPropagateToChildren=false
    Value.Translation=RigVMDispatchNode(Index=@n3.Index, Array=@"BSpline Passing thru 4 Points_1".Control Points) {template="DISPATCH_RigVMDispatch_ArrayGetAtIndex(in Array,in Index,out Element)"}.Element
    Item=RigVMDispatchNode(Index=@n3.Index, Array=RigVMVariableNode() {variable=IK Controls}.Value) {template="DISPATCH_RigVMDispatch_ArrayGetAtIndex(in Array,in Index,out Element)"}.Element
   )
   n4 = RigVMFunctionEntryNode()
   n5 = RigVMDispatchNode(ExecuteContext=@n8.ExecuteContext, Array=RigVMVariableNode() {variable=Bones}.Value) {template="DISPATCH_RigVMDispatch_ArrayIterator(in Array,out Element,out Index,out Count,out Ratio)"}
   Spline From Element Setup = RigVMCollapseNode(Items=RigVMVariableNode() {variable=IK Controls}.Value)
   > RigVMDispatchNode(
    Element=&RigVMFunction_MathTransformMakeRelative(
     Global=&RigUnit_GetTransform(Item=@n5.Element).Transform
     Parent=/Script/ControlRigSpline.RigUnit_TransformFromControlRigSpline(UpVector=(X=0,Y=1,Z=0), U=@n5.Ratio, Spline=@"Spline From Element Setup".Spline) {is=struct RigVMUnitNode}.Transform
    ).Local
    Array=RigVMVariableNode() {variable=IK Offsets}.Value
   ) {template="DISPATCH_RigVMDispatch_ArrayAdd(io Array,in Element,out Index)"}
   RigVMFunctionReturnNode(
    Length=/Script/ControlRigSpline.RigUnit_GetLengthControlRigSpline(Spline=@"BSpline Passing thru 4 Points_1".Spline) {is=struct RigVMUnitNode}.Length
    ExecuteContext=@n2.D
    Spline Offsets=RigVMVariableNode() {variable=IK Offsets}.Value
   )
   n6 = RigVMVariableNode() {variable=Bones}
   n7 = &RigVMFunction_MathIntSub(B=1, A=RigVMDispatchNode(Array=@n6.Value) {template="DISPATCH_RigVMDispatch_ArrayGetNum(in Array,out Num)"}.Num) {template="Subtract::Execute(in A,in B,out Result)"}
   &RigUnit_SetTransform(
    Weight=1
    bInitial=true
    bPropagateToChildren=false
    Value=&RigUnit_GetTransform(Item=RigVMDispatchNode(Index=@n7.Result, Array=@n6.Value) {template="DISPATCH_RigVMDispatch_ArrayGetAtIndex(in Array,in Index,out Element)"}.Element).Transform
    ExecutePin=@n2.C
    Item=RigVMVariableNode() {variable=Chest Space}.Value
   )
   n8 = RigVMDispatchNode(ExecuteContext=@n2.B, Array=RigVMVariableNode() {variable=IK Offsets}.Value) {template="DISPATCH_RigVMDispatch_ArrayReset(io Array)"}
  ]}
 ContainedGraph {nodes [
   n1 = &RigUnit_GetTransformItemArray(Space=GlobalSpace, Items=RigVMFunctionEntryNode().Items)
   RigVMFunctionReturnNode(Spline=&RigUnit_ControlRigSplineFromPoints(
     Points=((X=0,Y=0,Z=0),(X=0,Y=0,Z=0),(X=0,Y=0,Z=0),(X=0,Y=0,Z=0))
     SamplesPerSegment=16
     SplineMode=BSpline
     bClosed=False
     Points.0=RigVMDispatchNode(Index=0, Array=@n1.Transforms) {template="DISPATCH_RigVMDispatch_ArrayGetAtIndex(in Array,in Index,out Element)"}.Element.Translation
     Points.1=RigVMDispatchNode(Index=1, Array=@n1.Transforms) {template="DISPATCH_RigVMDispatch_ArrayGetAtIndex(in Array,in Index,out Element)"}.Element.Translation
     Points.2=RigVMDispatchNode(Index=2, Array=@n1.Transforms) {template="DISPATCH_RigVMDispatch_ArrayGetAtIndex(in Array,in Index,out Element)"}.Element.Translation
     Points.3=RigVMDispatchNode(Index=3, Array=@n1.Transforms) {template="DISPATCH_RigVMDispatch_ArrayGetAtIndex(in Array,in Index,out Element)"}.Element.Translation
    ).Spline)
  ]}
 ContainedGraph {nodes [
   RigVMFunctionEntryNode()
   n1 = RigVMVariableNode() {variable=IKFK Switch Control}
   > n2 = RigVMVariableNode(Value=&RigVMFunction_MathFloatGreater(B=0.5, A=&RigUnit_GetTransform(Space=LocalSpace, Item=@n1.Value).Transform.Translation.X) {template="Greater::Execute(in A,in B,out Result)"}.Result) {variable=IK Mode}
   n3 = &RigUnit_SequenceExecution(ExecuteContext=@n2.ExecuteContext)
   n4 = RigVMDispatchNode(
    Index=0
    Array=RigVMDispatchNode(Array=RigVMVariableNode() {variable=FK Controls}.Value) {template="DISPATCH_RigVMDispatch_ArrayClone(in Array,out Clone)"}.Clone
    ExecuteContext=@n3.A
   ) {template="DISPATCH_RigVMDispatch_ArrayRemove(io Array,in Index)"}
   > n5 = RigVMDispatchNode(
    Index=0
    Array=RigVMDispatchNode(Array=RigVMVariableNode() {variable=IK Controls}.Value) {template="DISPATCH_RigVMDispatch_ArrayClone(in Array,out Clone)"}.Clone
   ) {template="DISPATCH_RigVMDispatch_ArrayRemove(io Array,in Index)"}
   > Set IKFK Visibility = RigVMFunctionReferenceNode(Show Controls=true, FK Controls=@n4.Array, IK Controls=@n5.Array, IKFK Switch=RigVMVariableNode() {variable=IK Mode}.Value)
   n6 = RigVMDispatchNode(Index=0, Array=RigVMVariableNode() {variable=FK Controls}.Value) {template="DISPATCH_RigVMDispatch_ArrayGetAtIndex(in Array,in Index,out Element)"}
   n7 = RigVMVariableNode() {variable=Pelvis Bone}
   &RigUnit_SetTransform(
    Weight=1
    bInitial=False
    Value=&RigUnit_ProjectTransformToNewParent(bChildInitial=True, bOldParentInitial=True, NewParent=@n6.Element, OldParent=@n6.Element, Child=@n7.Value).Transform
    ExecutePin=@n3.B
    Item=@n7.Value
   )
   Branch_RigVMUnitNode = &RigVMFunction_ControlFlowBranch(ExecuteContext=@n3.C, Condition=RigVMVariableNode() {variable=IK Mode}.Value)
   FK Point At_1 = RigVMFunctionReferenceNode(
    Point At Axis=(X=1,Y=0,Z=0)
    Up Vector Axis=(X=0,Y=0,Z=-1)
    Up Vector Target=(X=1,Y=0,Z=0)
    ExecuteContext=@Branch_RigVMUnitNode.False
    Bones=RigVMVariableNode() {variable=Bones}.Value
    FK Controls=RigVMVariableNode() {variable=FK Controls}.Value
   )
   Spine Match IK onto FK = RigVMFunctionReferenceNode(
    Execute=@"FK Point At_1".ExecuteContext
    Bones=RigVMVariableNode() {variable=Bones}.Value
    Offsets=RigVMVariableNode() {variable=Spine Offsets}.Value
    IK Controls=RigVMVariableNode() {variable=IK Controls}.Value
   )
   n8 = &RigUnit_SequenceExecution(ExecuteContext=@n3.D)
   n9 = RigVMVariableNode() {variable=Bones}
   n10 = RigVMDispatchNode(
    Index=&RigVMFunction_MathIntSub(B=1, A=RigVMDispatchNode(Array=@n9.Value) {template="DISPATCH_RigVMDispatch_ArrayGetNum(in Array,out Num)"}.Num) {template="Subtract::Execute(in A,in B,out Result)"}.Result
    Array=@n9.Value
   ) {template="DISPATCH_RigVMDispatch_ArrayGetAtIndex(in Array,in Index,out Element)"}
   n11 = RigVMVariableNode() {variable=Chest Space}
   &RigUnit_SetTransform(
    Weight=1
    bInitial=False
    Value=&RigUnit_ProjectTransformToNewParent(bChildInitial=True, bOldParentInitial=True, NewParent=@n10.Element, OldParent=@n10.Element, Child=@n11.Value).Transform
    ExecutePin=@n8.A
    Item=@n11.Value
   )
   Set IKFK Keys_1 = RigVMFunctionReferenceNode(
    ExecuteContext=@n8.B
    Previous IK Mode=RigVMVariableNode() {variable=Previous IK Mode}.Value
    FK Controls=RigVMVariableNode() {variable=FK Controls}.Value
    "IK Controls "=RigVMVariableNode() {variable=IK Controls}.Value
    Switch Control=RigVMVariableNode() {variable=IKFK Switch Control}.Value
    IK Mode=RigVMVariableNode() {variable=IK Mode}.Value
   )
   RigVMFunctionReturnNode(ExecuteContext=@n8.D)
   Spine IK Solve = RigVMCollapseNode(
    Completed=@Branch_RigVMUnitNode.True
    Default Length=RigVMVariableNode() {variable=Default Length}.Value
    Enable Stretch=RigVMVariableNode() {variable="Enable Stretch "}.Value
    Bones=RigVMVariableNode() {variable=Bones}.Value
    IK Controls=RigVMVariableNode() {variable=IK Controls}.Value
    Spine Offsets=RigVMVariableNode() {variable=Spine Offsets}.Value
   )
   n12 = RigVMDispatchNode(
    Index=0
    Array=RigVMDispatchNode(Array=RigVMVariableNode() {variable=FK Controls}.Value) {template="DISPATCH_RigVMDispatch_ArrayClone(in Array,out Clone)"}.Clone
    ExecuteContext=@"Spine IK Solve".Completed
   ) {template="DISPATCH_RigVMDispatch_ArrayRemove(io Array,in Index)"}
   > n13 = RigVMDispatchNode(
    Index=0
    Array=RigVMDispatchNode(Array=RigVMVariableNode() {variable=Bones}.Value) {template="DISPATCH_RigVMDispatch_ArrayClone(in Array,out Clone)"}.Clone
   ) {template="DISPATCH_RigVMDispatch_ArrayRemove(io Array,in Index)"}
   > Project to New Parent = RigVMFunctionReferenceNode(Propagate to Children=True, Children=@n12.Array, Parents=@n13.Array)
  ]}
 ContainedGraph {nodes [
   n1 = RigVMDispatchNode(ExecuteContext=(), Array=RigVMVariableNode() {variable=Bones}.Value) {template="DISPATCH_RigVMDispatch_ArrayIterator(in Array,out Element,out Index,out Count,out Ratio)"}
   Branch_RigVMUnitNode = &RigVMFunction_ControlFlowBranch(
    ExecuteContext=@n1.ExecuteContext
    Condition=&RigVMFunction_MathIntEquals(A=@n1.Index, B=&RigVMFunction_MathIntSub(B=1, A=@n1.Count) {template="Subtract::Execute(in A,in B,out Result)"}.Result).Result
   )
   n2 = &RigUnit_AimBoneMath(
    DebugSettings=(Scale=10,WorldOffset=(Rotation=(X=0,Y=0,Z=0,W=1),Translation=(X=0,Y=0,Z=0),Scale3D=(X=1,Y=1,Z=1)))
    Primary=(Weight=1,Axis=(X=1,Y=0,Z=0),Target=(X=0,Y=0,Z=0),Kind=Location,Space=(Type=Bone))
    Secondary=(Weight=1,Axis=(X=0,Y=0,Z=1),Target=(X=0,Y=0,Z=1),Kind=Direction,Space=(Type=Bone))
    Weight=1
    Primary.Space=RigVMDispatchNode(
     Index=&RigVMFunction_MathIntAdd(B=1, A=@n1.Index) {template="Add::Execute(in A,in B,out Result)"}.Result
     Array=RigVMVariableNode() {variable=Bones}.Value
    ) {template="DISPATCH_RigVMDispatch_ArrayGetAtIndex(in Array,in Index,out Element)"}.Element
    Secondary.Space=@n1.Element
    InputTransform=&RigUnit_GetTransform(Item=@n1.Element).Transform
   )
   &RigUnit_SetTransform(Weight=1, bInitial=False, bPropagateToChildren=false, Value=@n2.Result, Item=@n1.Element, ExecutePin=@Branch_RigVMUnitNode.False)
   n3 = RigVMFunctionEntryNode()
   n4 = &RigUnit_SequenceExecution(ExecuteContext=@n3.Completed)
   n5 = RigVMDispatchNode(ExecuteContext=@n4.A, Array=RigVMVariableNode() {variable=Bones}.Value) {template="DISPATCH_RigVMDispatch_ArrayIterator(in Array,out Element,out Index,out Count,out Ratio)"}
   Is First or Last = RigVMCollapseNode(Count=@n5.Count, Index=@n5.Index)
   n6 = RigVMVariableNode() {variable=IK Controls}
   Spline From Element = RigVMCollapseNode(Items=@n6.Value)
   n7 = &RigUnit_GetLengthControlRigSpline(Spline=@"Spline From Element".Spline)
   n8 = RigVMVariableNode() {variable=Default Length}
   n9 = RigVMDispatchNode(
    True=1
    False=/Script/RigVM.RigVMFunction_MathFloatMin(B=1, A=&RigVMFunction_MathFloatDiv(B=@n7.Length, A=@n8.Value) {template="Divide::Execute(in A,in B,out Result)"}.Result) {is=struct RigVMUnitNode, template="Minimum::Execute(in A,in B,out Result)"}.Result
    Condition=RigVMVariableNode() {variable=Enable Stretch}.Value
   ) {template="DISPATCH_RigVMDispatch_If(in Condition,in True,in False,out Result)"}
   UpVector = RigVMCollapseNode(
    Start=RigVMDispatchNode(Index=0, Array=@n6.Value) {template="DISPATCH_RigVMDispatch_ArrayGetAtIndex(in Array,in Index,out Element)"}.Element
    End=RigVMDispatchNode(Index=3, Array=@n6.Value) {template="DISPATCH_RigVMDispatch_ArrayGetAtIndex(in Array,in Index,out Element)"}.Element
    T=@n5.Ratio
   )
   n10 = &RigVMFunction_MathTransformMakeAbsolute(
    Local=RigVMDispatchNode(Index=@n5.Index, Array=RigVMVariableNode() {variable=Spine Offsets}.Value) {template="DISPATCH_RigVMDispatch_ArrayGetAtIndex(in Array,in Index,out Element)"}.Element
    Parent=&RigUnit_TransformFromControlRigSpline(
     Roll=0
     U=&RigVMFunction_MathFloatMul(B=@n5.Ratio, A=@n9.Result) {template="Multiply::Execute(in A,in B,out Result)"}.Result
     Spline=@"Spline From Element".Spline
     UpVector=@UpVector.Up Vector
    ).Transform
   ) {template="Make Absolute::Execute(in Local,in Parent,out Global)"}
   n11 = RigVMDispatchNode(Index=@n5.Index, Array=RigVMVariableNode() {variable=IK Controls}.Value) {template="DISPATCH_RigVMDispatch_ArrayGetAtIndex(in Array,in Index,out Element)"}
   > &RigUnit_SetTransform(
    Weight=1
    bInitial=False
    Item=@n5.Element
    Value=RigVMDispatchNode(
     True=(Rotation=(X=0,Y=0,Z=0,W=1),Translation=(X=0,Y=0,Z=0),Scale3D=(X=0,Y=0,Z=0))
     Condition=@"Is First or Last".Result
     False=@n10.Global
     True.Scale3D=@n10.Global.Scale3D
     True.Translation=@n10.Global.Translation
     True.Rotation=&RigUnit_ProjectTransformToNewParent(bChildInitial=True, bOldParentInitial=True, NewParent=@n11.Element, OldParent=@n11.Element, Child=@n5.Element).Transform.Rotation
    ) {template="DISPATCH_RigVMDispatch_If(in Condition,in True,in False,out Result)"}.Result
   )
   RigVMFunctionReturnNode(Completed=@n4.D)
  ]}
 ContainedGraph {nodes [
   n1 = &RigUnit_GetTransformItemArray(Space=GlobalSpace, Items=RigVMFunctionEntryNode().Items)
   RigVMFunctionReturnNode(Spline=&RigUnit_ControlRigSplineFromPoints(
     Points=((X=0,Y=0,Z=0),(X=0,Y=0,Z=0),(X=0,Y=0,Z=0),(X=0,Y=0,Z=0))
     SamplesPerSegment=16
     SplineMode=BSpline
     bClosed=False
     Points.0=RigVMDispatchNode(Index=0, Array=@n1.Transforms) {template="DISPATCH_RigVMDispatch_ArrayGetAtIndex(in Array,in Index,out Element)"}.Element.Translation
     Points.1=RigVMDispatchNode(Index=1, Array=@n1.Transforms) {template="DISPATCH_RigVMDispatch_ArrayGetAtIndex(in Array,in Index,out Element)"}.Element.Translation
     Points.2=RigVMDispatchNode(Index=2, Array=@n1.Transforms) {template="DISPATCH_RigVMDispatch_ArrayGetAtIndex(in Array,in Index,out Element)"}.Element.Translation
     Points.3=RigVMDispatchNode(Index=3, Array=@n1.Transforms) {template="DISPATCH_RigVMDispatch_ArrayGetAtIndex(in Array,in Index,out Element)"}.Element.Translation
    ).Spline)
  ]}
 ContainedGraph {nodes [
   n1 = RigVMFunctionEntryNode()
   n2 = &RigUnit_GetTransform(Item=@n1.Start)
   n3 = &RigUnit_GetTransform(Item=@n1.End)
   RigVMFunctionReturnNode(Up Vector=/Script/RigVM.RigVMFunction_MathQuaternionRotateVector(
     Vector=(X=0,Y=1,Z=0)
     Transform=/Script/RigVM.RigVMFunction_MathQuaternionSlerp(T=@n1.T, A=@n2.Transform.Rotation, B=@n3.Transform.Rotation) {is=struct RigVMUnitNode, template="Interpolate::Execute(in A,in B,in T,out Result)"}.Result
    ) {is=struct RigVMUnitNode, template="Rotate Vector::Execute(in Transform,in Vector,out Result)"}.Result)
  ]}
 ContainedGraph {nodes [
   n1 = RigVMFunctionEntryNode()
   n2 = &RigVMFunction_MathIntSub(B=1, A=@n1.Count) {template="Subtract::Execute(in A,in B,out Result)"}
   RigVMFunctionReturnNode(Result=&RigVMFunction_MathBoolOr(
     A=&RigVMFunction_MathIntEquals(B=0, A=@n1.Index).Result
     B=&RigVMFunction_MathIntEquals(A=@n1.Index, B=@n2.Result).Result
    ).Result)
  ]}
 ContainedGraph {nodes [
   n1 = RigVMFunctionEntryNode()
   n2 = &RigUnit_SequenceExecution(ExecuteContext=@n1.ExecuteContext)
   n3 = RigVMVariableNode() {variable=Bones}
   n4 = RigVMDispatchNode(Array=@n3.Value) {template="DISPATCH_RigVMDispatch_ArrayGetNum(in Array,out Num)"}
   n5 = &RigUnit_GetTransform(Item=RigVMDispatchNode(Index=&RigVMFunction_MathIntSub(B=1, A=@n4.Num) {template="Subtract::Execute(in A,in B,out Result)"}.Result, Array=@n3.Value) {template="DISPATCH_RigVMDispatch_ArrayGetAtIndex(in Array,in Index,out Element)"}.Element)
   &RigUnit_SetTransform(
    Value=(Rotation=(X=0,Y=0,Z=0,W=1),Translation=(X=0,Y=0,Z=0),Scale3D=(X=1,Y=1,Z=1))
    Weight=1
    bInitial=true
    Value.Translation=@n5.Transform.Translation
    ExecutePin=@n2.A
    Item=RigVMVariableNode() {variable=Head FK Space}.Value
   )
   > &RigUnit_SetTransform(
    Value=(Rotation=(X=0,Y=0,Z=0,W=1),Translation=(X=0,Y=0,Z=0),Scale3D=(X=1,Y=1,Z=1))
    Weight=1
    bInitial=true
    Value.Translation=@n5.Transform.Translation
    Item=RigVMVariableNode() {variable=Head IK Space}.Value
   )
   n6 = RigVMDispatchNode(ExecuteContext=@n2.B, Array=RigVMVariableNode() {variable=FK Controls}.Value) {template="DISPATCH_RigVMDispatch_ArrayIterator(in Array,out Element,out Index,out Count,out Ratio)"}
   n7 = RigVMVariableNode() {variable=Bones}
   > &RigUnit_SetTransform(
    Value=(Rotation=(X=0,Y=0,Z=0,W=1),Translation=(X=0,Y=0,Z=0),Scale3D=(X=1,Y=1,Z=1))
    Weight=1
    bInitial=true
    Item=@n6.Element
    Value.Translation=&RigUnit_GetTransform(Item=RigVMDispatchNode(Index=@n6.Index, Array=@n7.Value) {template="DISPATCH_RigVMDispatch_ArrayGetAtIndex(in Array,in Index,out Element)"}.Element).Transform.Translation
   )
   n8 = RigVMVariableNode() {variable=Bones}
   n9 = &RigVMFunction_MathIntSub(B=1, A=RigVMDispatchNode(Array=@n8.Value) {template="DISPATCH_RigVMDispatch_ArrayGetNum(in Array,out Num)"}.Num) {template="Subtract::Execute(in A,in B,out Result)"}
   &RigUnit_SetTransform(
    Value=(Rotation=(X=0,Y=0,Z=0,W=1),Translation=(X=0,Y=0,Z=0),Scale3D=(X=1,Y=1,Z=1))
    Weight=1
    bInitial=true
    Value.Translation=&RigUnit_GetTransform(Item=RigVMDispatchNode(Index=@n9.Result, Array=@n8.Value) {template="DISPATCH_RigVMDispatch_ArrayGetAtIndex(in Array,in Index,out Element)"}.Element).Transform.Translation
    ExecutePin=@n2.C
    Item=RigVMVariableNode() {variable=IK Control}.Value
   )
   n10 = &RigUnit_GetTransformItemArray(Space=GlobalSpace, Items=RigVMVariableNode() {variable=Bones}.Value)
   n11 = RigVMDispatchNode(Index=0, Array=@n10.Transforms) {template="DISPATCH_RigVMDispatch_ArrayGetAtIndex(in Array,in Index,out Element)"}
   n12 = RigVMDispatchNode(Index=1, Array=@n10.Transforms) {template="DISPATCH_RigVMDispatch_ArrayGetAtIndex(in Array,in Index,out Element)"}
   n13 = RigVMDispatchNode(Index=2, Array=@n10.Transforms) {template="DISPATCH_RigVMDispatch_ArrayGetAtIndex(in Array,in Index,out Element)"}
   RigVMFunctionReturnNode(
    Length=&RigUnit_GetLengthControlRigSpline(Spline=&RigUnit_ControlRigSplineFromPoints(
      Points=((X=0,Y=0,Z=0),(X=0,Y=0,Z=0),(X=0,Y=0,Z=0),(X=0,Y=0,Z=0))
      SamplesPerSegment=16
      SplineMode=BSpline
      bClosed=False
      Points.0=@n11.Element.Translation
      Points.1=@n11.Element.Translation
      Points.2=@n12.Element.Translation
      Points.3=@n13.Element.Translation
     ).Spline).Length
    ExecuteContext=@n2.D
   )
  ]}
 ContainedGraph {nodes [
   RigVMFunctionEntryNode()
   n1 = RigVMVariableNode() {variable=IKFK Switch Control}
   > RigVMVariableNode(Value=&RigVMFunction_MathFloatGreater(B=0.5, A=&RigUnit_GetTransform(Space=LocalSpace, Item=@n1.Value).Transform.Translation.X) {template="Greater::Execute(in A,in B,out Result)"}.Result) {variable=IK Mode}
   n2 = RigVMVariableNode() {variable=Head Orient Control}
   > n3 = RigVMVariableNode(Value=&RigVMFunction_MathFloatGreater(B=0.5, A=&RigUnit_GetTransform(Space=LocalSpace, Item=@n2.Value).Transform.Translation.X) {template="Greater::Execute(in A,in B,out Result)"}.Result) {variable=Local Space}
   n4 = &RigUnit_SequenceExecution(ExecuteContext=@n3.ExecuteContext)
   Set IKFK Visibility = RigVMFunctionReferenceNode(
    IK Controls=((Type=Control,Name="head_ik_ctrl"))
    Show Controls=true
    ExecuteContext=@n4.A
    IKFK Switch=RigVMVariableNode() {variable=IK Mode}.Value
    FK Controls=RigVMVariableNode() {variable=FK Controls}.Value
    IK Controls.0=RigVMVariableNode() {variable=IK Control}.Value
   )
   n5 = RigVMDispatchNode(Index=1, Array=RigVMVariableNode() {variable=FK Controls}.Value) {template="DISPATCH_RigVMDispatch_ArrayGetAtIndex(in Array,in Index,out Element)"}
   n6 = RigVMVariableNode() {variable=Head FK Space}
   n7 = &RigUnit_ProjectTransformToNewParent(bChildInitial=True, bOldParentInitial=True, NewParent=@n5.Element, OldParent=@n5.Element, Child=@n6.Value)
   n8 = RigVMVariableNode() {variable=Head Orient Reference}
   &RigUnit_SetTransform(
    Value=(Rotation=(X=0,Y=0,Z=0,W=1),Translation=(X=0,Y=0,Z=0),Scale3D=(X=1,Y=1,Z=1))
    Weight=1
    bInitial=False
    Value.Rotation=RigVMDispatchNode(
     False=@n7.Transform.Rotation
     True=&RigUnit_ProjectTransformToNewParent(bChildInitial=True, bOldParentInitial=True, Child=@n6.Value, NewParent=@n8.Value, OldParent=@n8.Value).Transform.Rotation
     Condition=RigVMVariableNode() {variable=Local Space}.Value
    ) {template="DISPATCH_RigVMDispatch_If(in Condition,in True,in False,out Result)"}.Result
    Value.Translation=@n7.Transform.Translation
    ExecutePin=@n4.B
    Item=@n6.Value
   )
   n9 = RigVMVariableNode() {variable=Head IK Space}
   n10 = RigVMVariableNode() {variable=Head Orient Reference}
   n11 = &RigUnit_HierarchyGetParent(bDefaultParent=True, Child=@n9.Value)
   n12 = &RigUnit_ProjectTransformToNewParent(bChildInitial=True, bOldParentInitial=True, NewParent=@n11.Parent, OldParent=@n11.Parent, Child=@n9.Value)
   > &RigUnit_SetTransform(
    Value=(Rotation=(X=0,Y=0,Z=0,W=1),Translation=(X=0,Y=0,Z=0),Scale3D=(X=1,Y=1,Z=1))
    Weight=1
    bInitial=False
    Value.Rotation=RigVMDispatchNode(
     True=&RigUnit_ProjectTransformToNewParent(bChildInitial=True, bOldParentInitial=True, Child=@n9.Value, NewParent=@n10.Value, OldParent=@n10.Value).Transform.Rotation
     False=@n12.Transform.Rotation
     Condition=RigVMVariableNode() {variable=Local Space}.Value
    ) {template="DISPATCH_RigVMDispatch_If(in Condition,in True,in False,out Result)"}.Result
    Value.Translation=@n12.Transform.Translation
    Item=@n9.Value
   )
   Branch_RigVMUnitNode = &RigVMFunction_ControlFlowBranch(ExecuteContext=@n4.C, Condition=RigVMVariableNode() {variable=IK Mode}.Value)
   Project to New Parents_2 = RigVMFunctionReferenceNode(
    Propagate to Children=True
    ExecuteContext=@Branch_RigVMUnitNode.False
    Children=RigVMVariableNode() {variable=Bones}.Value
    Parents=RigVMVariableNode() {variable=FK Controls}.Value
   )
   n13 = RigVMVariableNode() {variable=Bones}
   n14 = RigVMDispatchNode(
    Index=&RigVMFunction_MathIntSub(B=1, A=RigVMDispatchNode(Array=@n13.Value) {template="DISPATCH_RigVMDispatch_ArrayGetNum(in Array,out Num)"}.Num) {template="Subtract::Execute(in A,in B,out Result)"}.Result
    Array=@n13.Value
   ) {template="DISPATCH_RigVMDispatch_ArrayGetAtIndex(in Array,in Index,out Element)"}
   n15 = RigVMVariableNode() {variable=IK Control}
   > &RigUnit_SetTransform(
    Weight=1
    bInitial=False
    Value=&RigUnit_ProjectTransformToNewParent(bChildInitial=True, bOldParentInitial=True, NewParent=@n14.Element, OldParent=@n14.Element, Child=@n15.Value).Transform
    Item=@n15.Value
   )
   Neck IK Neck = RigVMCollapseNode(
    ExecuteContext=@Branch_RigVMUnitNode.True
    Length=RigVMVariableNode() {variable=Length}.Value
    Enable Stretch=RigVMVariableNode() {variable=Enable Stretch}.Value
    IK Control=RigVMVariableNode() {variable=IK Control}.Value
    Mid Null=RigVMVariableNode() {variable=Mid IK Null}.Value
    Bones=RigVMVariableNode() {variable=Bones}.Value
    Positions=RigVMVariableNode() {variable=Positions}.Value
   )
   > Project to New Parents = RigVMFunctionReferenceNode(Propagate to Children=True, Parents=RigVMVariableNode() {variable=Bones}.Value, Children=RigVMVariableNode() {variable=FK Controls}.Value)
   Set IKFK Keys = RigVMFunctionReferenceNode(
    "IK Controls "=((Type=Control,Name="head_ik_ctrl"))
    ExecuteContext=@n4.D
    IK Mode=RigVMVariableNode() {variable=IK Mode}.Value
    Previous IK Mode=RigVMVariableNode() {variable=Previous IK Mode}.Value
    FK Controls=RigVMVariableNode() {variable=FK Controls}.Value
    Switch Control=RigVMVariableNode() {variable=IKFK Switch Control}.Value
    IK Controls .0=RigVMVariableNode() {variable=IK Control}.Value
   )
   > RigVMFunctionReturnNode()
  ]}
 ContainedGraph {nodes [
   n1 = RigVMFunctionEntryNode()
   > Get Points on Spline = RigVMCollapseNode(
    Bones=@n1.Bones
    Enable Stretch=@n1.Enable Stretch
    IK Control=@n1.IK Control
    Length=@n1.Length
    Mid Null=@n1.Mid Null
    Array=RigVMVariableNode() {variable=Positions}.Value
   )
   n2 = &RigUnit_SequenceExecution(ExecuteContext=@"Get Points on Spline".ExecuteContext)
   n3 = RigVMDispatchNode(ExecuteContext=@n2.A, Array=RigVMVariableNode() {variable=Bones}.Value) {template="DISPATCH_RigVMDispatch_ArrayIterator(in Array,out Element,out Index,out Count,out Ratio)"}
   At and Next = RigVMCollapseNode(Index=@n3.Index, Array=RigVMVariableNode() {variable=Positions}.Value)
   First Last Rotation = RigVMCollapseNode(
    First Bone=RigVMDispatchNode(Index=0, Array=RigVMVariableNode() {variable=Bones}.Value) {template="DISPATCH_RigVMDispatch_ArrayGetAtIndex(in Array,in Index,out Element)"}.Element
    T=@n3.Ratio
    IK Control=RigVMVariableNode() {variable=IK Control}.Value
   )
   n4 = &RigUnit_AimBoneMath(
    DebugSettings=(Scale=10,WorldOffset=(Rotation=(X=0,Y=0,Z=0,W=1),Translation=(X=0,Y=0,Z=0),Scale3D=(X=1,Y=1,Z=1)))
    InputTransform=(Rotation=(X=0,Y=0,Z=0,W=1),Translation=(X=0,Y=0,Z=0),Scale3D=(X=1,Y=1,Z=1))
    Primary=(Weight=1,Axis=(X=1,Y=0,Z=0),Target=(X=1,Y=0,Z=0),Kind=Location,Space=(Type=Bone))
    Secondary=(Weight=1,Axis=(X=0,Y=0,Z=1),Target=(X=1,Y=0,Z=0),Kind=Direction,Space=(Type=Bone))
    Weight=1
    InputTransform.Translation=@"At and Next".At
    Primary.Target=@"At and Next".Next
    Secondary.Target=&RigVMFunction_MathQuaternionRotateVector(Vector=(X=0,Y=0,Z=1), Transform=@"First Last Rotation".Interpolate) {template="Rotate Vector::Execute(in Transform,in Vector,out Result)"}.Result
   )
   Is Last = RigVMCollapseNode(Count=@n3.Count, Index=@n3.Index)
   > &RigUnit_SetTransform(
    Value=(Rotation=(X=0,Y=0,Z=0,W=1),Translation=(X=0,Y=0,Z=0),Scale3D=(X=1,Y=1,Z=1))
    Weight=1
    bInitial=False
    Item=@n3.Element
    Value.Translation=@"At and Next".At
    Value.Rotation=RigVMDispatchNode(False=@n4.Result.Rotation, True=@"First Last Rotation".Last, Condition=@"Is Last".Result) {template="DISPATCH_RigVMDispatch_If(in Condition,in True,in False,out Result)"}.Result
   )
   RigVMFunctionReturnNode(ExecuteContext=@n2.D)
  ]}
 ContainedGraph {nodes [
   n1 = RigVMFunctionEntryNode()
   n2 = &RigUnit_SequenceExecution(ExecuteContext=@n1.ExecuteContext)
   n3 = RigVMDispatchNode(ExecuteContext=@n2.A, Array=RigVMVariableNode() {variable=Array}.Value) {template="DISPATCH_RigVMDispatch_ArrayReset(io Array)"}
   > n4 = RigVMDispatchNode(Array=RigVMVariableNode() {variable=Bones}.Value) {template="DISPATCH_RigVMDispatch_ArrayIterator(in Array,out Element,out Index,out Count,out Ratio)"}
   n5 = &RigUnit_GetTransform(Item=RigVMDispatchNode(Index=0, Array=RigVMVariableNode() {variable=Bones}.Value) {template="DISPATCH_RigVMDispatch_ArrayGetAtIndex(in Array,in Index,out Element)"}.Element)
   RigUnit_GetTransform = &RigUnit_GetTransform(Item=RigVMVariableNode() {variable=Mid Null}.Value)
   n6 = &RigUnit_ControlRigSplineFromPoints(
    Points=((X=0,Y=0,Z=0),(X=0,Y=0,Z=0),(X=0,Y=0,Z=0),(X=0,Y=0,Z=0))
    SamplesPerSegment=16
    SplineMode=BSpline
    bClosed=False
    Points.0=@n5.Transform.Translation
    Points.1=@n5.Transform.Translation
    Points.3=&RigUnit_GetTransform(Item=RigVMVariableNode() {variable=IK Control}.Value).Transform.Translation
    Points.2=@RigUnit_GetTransform.Transform.Translation
   )
   n7 = &RigVMFunction_MathFloatDiv(B=&RigUnit_GetLengthControlRigSpline(Spline=@n6.Spline).Length, A=RigVMVariableNode() {variable=Length}.Value) {template="Divide::Execute(in A,in B,out Result)"}
   n8 = &RigVMFunction_MathFloatMul(
    A=@n4.Ratio
    B=RigVMDispatchNode(
     True=1
     False=&RigVMFunction_MathFloatMin(B=1, A=@n7.Result) {template="Minimum::Execute(in A,in B,out Result)"}.Result
     Condition=RigVMVariableNode() {variable=Enable Stretch}.Value
    ) {template="DISPATCH_RigVMDispatch_If(in Condition,in True,in False,out Result)"}.Result
   ) {template="Multiply::Execute(in A,in B,out Result)"}
   > RigVMDispatchNode(
    Array=@n3.Array
    Element=/Script/ControlRigSpline.RigUnit_PositionFromControlRigSpline(
     Spline=@n6.Spline
     U=/Script/ControlRigSpline.RigUnit_ParameterAtPercentage(Spline=@n6.Spline, Percentage=@n8.Result) {is=struct RigVMUnitNode}.U
    ) {is=struct RigVMUnitNode}.Position
   ) {template="DISPATCH_RigVMDispatch_ArrayAdd(io Array,in Element,out Index)"}
   RigVMFunctionReturnNode(ExecuteContext=@n2.D, Positions=RigVMVariableNode() {variable=Array}.Value)
  ]}
 ContainedGraph {nodes [
   n1 = RigVMFunctionEntryNode()
   RigVMFunctionReturnNode(Result=&RigVMFunction_MathIntEquals(A=@n1.Index, B=&RigVMFunction_MathIntSub(B=1, A=@n1.Count) {template="Subtract::Execute(in A,in B,out Result)"}.Result).Result)
  ]}
 ContainedGraph {nodes [
   n1 = RigVMFunctionEntryNode()
   RigVMFunctionReturnNode(
    At=RigVMDispatchNode(Array=@n1.Array, Index=@n1.Index) {template="DISPATCH_RigVMDispatch_ArrayGetAtIndex(in Array,in Index,out Element)"}.Element
    Next=RigVMDispatchNode(Array=@n1.Array, Index=&RigVMFunction_MathIntAdd(B=1, A=@n1.Index) {template="Add::Execute(in A,in B,out Result)"}.Result) {template="DISPATCH_RigVMDispatch_ArrayGetAtIndex(in Array,in Index,out Element)"}.Element
   )
  ]}
 ContainedGraph {nodes [
   n1 = RigVMFunctionEntryNode()
   n2 = /Script/RigVM.RigVMFunction_MathQuaternionMul(
    A=&RigUnit_GetTransform(Item=@n1.IK Control).Transform.Rotation
    B=&RigVMFunction_MathQuaternionFromRotator(Rotator=(Pitch=90,Yaw=0,Roll=0)).Result
   ) {is=struct RigVMUnitNode, template="Multiply::Execute(in A,in B,out Result)"}
   n3 = &RigUnit_HierarchyGetParent(bDefaultParent=True, Child=@n1.First Bone)
   n4 = &RigUnit_ProjectTransformToNewParent(bChildInitial=True, bOldParentInitial=True, Child=@n1.First Bone, NewParent=@n3.Parent, OldParent=@n3.Parent)
   RigVMFunctionReturnNode(
    Last=@n2.Result
    Interpolate=&RigVMFunction_MathQuaternionSlerp(T=@n1.T, B=@n2.Result, A=@n4.Transform.Rotation) {template="Interpolate::Execute(in A,in B,in T,out Result)"}.Result
    First=@n4.Transform.Rotation
   )
  ]}
 ContainedGraph {nodes [
   n1 = &RigUnit_GetTransformItemArray(Space=GlobalSpace, Items=RigVMVariableNode() {variable=Bones}.Value)
   n2 = &RigUnit_SequenceExecution(ExecuteContext=@n7.Execute)
   BSpline Passing thru 4 Points_1 = RigVMFunctionReferenceNode(
    Auto Weights=false
    Weight 1=0.33
    Weight 2=0.66
    Point 0=RigVMDispatchNode(Index=0, Array=@n1.Transforms) {template="DISPATCH_RigVMDispatch_ArrayGetAtIndex(in Array,in Index,out Element)"}.Element.Translation
    Point 1=RigVMDispatchNode(Index=1, Array=@n1.Transforms) {template="DISPATCH_RigVMDispatch_ArrayGetAtIndex(in Array,in Index,out Element)"}.Element.Translation
    Point 2=RigVMDispatchNode(Index=2, Array=@n1.Transforms) {template="DISPATCH_RigVMDispatch_ArrayGetAtIndex(in Array,in Index,out Element)"}.Element.Translation
    Point 3=RigVMDispatchNode(Index=3, Array=@n1.Transforms) {template="DISPATCH_RigVMDispatch_ArrayGetAtIndex(in Array,in Index,out Element)"}.Element.Translation
    Execute=@n2.A
   )
   n3 = RigVMDispatchNode(ExecuteContext=@"BSpline Passing thru 4 Points_1".Execute, Array=RigVMVariableNode() {variable=IK Controls}.Value) {template="DISPATCH_RigVMDispatch_ArrayIterator(in Array,out Element,out Index,out Count,out Ratio)"}
   Branch_1_1_RigVMUnitNode = &RigVMFunction_ControlFlowBranch(ExecuteContext=@n3.ExecuteContext, Condition=&RigVMFunction_MathIntEquals(B=0, A=@n3.Index).Result)
   n4 = RigVMDispatchNode(Index=@n3.Index, Array=RigVMVariableNode() {variable=Bones}.Value) {template="DISPATCH_RigVMDispatch_ArrayGetAtIndex(in Array,in Index,out Element)"}
   n5 = &RigUnit_ProjectTransformToNewParent(bChildInitial=True, bOldParentInitial=True, NewParent=@n4.Element, OldParent=@n4.Element, Child=@n3.Element)
   n6 = &RigVMFunction_MathTransformMakeAbsolute(
    Local=(Rotation=(X=0,Y=0,Z=0,W=1),Translation=(X=0,Y=0,Z=0),Scale3D=(X=1,Y=1,Z=1))
    Parent=(Rotation=(X=0,Y=0,Z=0,W=1),Translation=(X=0,Y=0,Z=0),Scale3D=(X=1,Y=1,Z=1))
    Parent.Translation=RigVMDispatchNode(Index=@n3.Index, Array=@"BSpline Passing thru 4 Points_1".Control Points) {template="DISPATCH_RigVMDispatch_ArrayGetAtIndex(in Array,in Index,out Element)"}.Element
    Parent.Rotation=@n5.Transform.Rotation
    Parent.Scale3D=@n5.Transform.Scale3D
   ) {template="Make Absolute::Execute(in Local,in Parent,out Global)"}
   &RigUnit_SetTransform(Weight=1, bInitial=False, bPropagateToChildren=false, Item=@n3.Element, ExecutePin=@Branch_1_1_RigVMUnitNode.False, Value=@n6.Global)
   n7 = RigVMFunctionEntryNode()
   RigVMDispatchNode(Index=@n3.Index, Array=RigVMVariableNode() {variable=Offsets}.Value) {template="DISPATCH_RigVMDispatch_ArrayGetAtIndex(in Array,in Index,out Element)"}
   RigVMFunctionReturnNode(Execute=@n2.D)
  ]}
 RigVMFunctionLibrary {nodes [
   BSpline Passing thru 4 Points = RigVMCollapseNode(
    Auto Weights=true
    Execute=()
    Point 0=(X=0,Y=0,Z=0)
    Point 1=(X=0,Y=0,Z=0)
    Point 2=(X=0,Y=0,Z=0)
    Point 3=(X=0,Y=0,Z=0)
    Weight 1=0.33
    Weight 2=0.66
   )
   Backward Arm = RigVMCollapseNode(Bones=(), ExecuteContext=(), FK Arm Space=(), FK Controls=(), FK Orient References=(), Fingers Space=(), Ik Control=(), UpV Control=())
   Backward Clavicle = RigVMCollapseNode(Bone=(), Control=(), Execute=())
   Backward Fingers = RigVMCollapseNode(Bones=(), Execute=())
   Backward Foot = RigVMCollapseNode(Bone=(), Execute=(), FK Control=(), IK Control=(), Toes Space=())
   Backward Leg = RigVMCollapseNode(Bones=(), ExecuteContext=(), FK Controls=(), Ik Control=(), UpV Control=())
   Backward Neck = RigVMCollapseNode(Bones=(), ExecuteContext=(), FK Controls=(), Head FK Space=(), Head IK Space=(), Head Orient Reference=(), IK Control=())
   Backward Spine = RigVMCollapseNode(Bones=(), ExecuteContext=(), FK Controls=(), IK Controls=(), Offsets=())
   FK Point At = RigVMCollapseNode(Bones=(), ExecuteContext=(), FK Controls=(), Point At Axis=(X=1,Y=0,Z=0), Up Vector Axis=(X=0,Y=0,Z=1), Up Vector Target=(X=0,Y=0,Z=1))
   Forward Arm = RigVMCollapseNode(
    Arm Space=()
    Bones=()
    Enable Stretch=False
    Execute=()
    FK Controls=()
    FK Orient Control=()
    FK Orient References=()
    Finger Space=()
    IK Control=()
    IK Effector=()
    IKFK Switch Control=()
    Negative Side=False
    Pole Vector Control=()
    Previous IK Mode=False
    Softness=1
   )
   Forward Clavicle = RigVMCollapseNode(Bone=(), Control=(), ExecuteContext=())
   Forward Fingers = RigVMCollapseNode(Bones=(), Execute=())
   Forward Foot = RigVMCollapseNode(Bone=(), ExecuteContext=(), FK Control=(), IK Control=(), IKFK Switch Control=(), Toes Space=())
   Forward Leg = RigVMCollapseNode(
    Bones=()
    Enable Stretch=False
    Execute=()
    FK Controls=()
    IK Control=()
    IK Effector=()
    IKFK Switch Control=()
    Negative Side=False
    Pole Vector Control=()
    Previous IK Mode=False
    Softness=1
   )
   Forward Neck = RigVMCollapseNode(
    Bones=()
    ExecuteContext=()
    FK Controls=()
    Head FK Space=()
    Head IK Space=()
    Head Orient Control=()
    Head Orient Reference=()
    IK Control=()
    IKFK Switch Control=()
    Mid IK Null=()
   )
   Forward Reverse Foot = RigVMCollapseNode(BK Buffers=(), BK Controls=(), ExecuteContext=(), IKFK Switch Control=(), Roll Buffer=(), Roll Clamp=30, Roll Control=())
   Forward Spine = RigVMCollapseNode(Bones=(), Chest Space=(), ExecuteContext=(), FK Controls=(), IK Controls=(), IKFK Switch Control=(), Pelvis Bone=(), Spline Offsets=())
   Get Negative Transform = RigVMCollapseNode(Input=(Rotation=(X=0,Y=0,Z=0,W=1),Translation=(X=0,Y=0,Z=0),Scale3D=(X=0,Y=0,Z=0)))
   IK 2 Bones = RigVMCollapseNode(
    Bones=()
    Enable Stretch=true
    ExecuteContext=()
    IK Controls=(Type=Control,Name="hand_l_ik_ctrl")
    Pole Vector Control=(Type=Control,Name="arm_l_pv_ik_ctrl")
    Primary Axis=(X=1,Y=0,Z=0)
    Secondary Axis=(X=0,Y=-1,Z=0)
    Softness=1
   )
   Match 2 Bones IK onto FK = RigVMCollapseNode(Bones=(), ExecuteContext=(), IK Control=(Type=Control,Name="hand_l_ik_ctrl"), Pole Vector Control=(Type=Control,Name="arm_l_pv_ik_ctrl"))
   Project to New Parents = RigVMCollapseNode(Children=(), ExecuteContext=(), Parents=(), Propagate to Children=True)
   Set IKFK Keys = RigVMCollapseNode(ExecuteContext=(), FK Controls=(), "IK Controls "=(), Switch Control=())
   Set IKFK Visibility = RigVMCollapseNode(ExecuteContext=(), FK Controls=(), IK Controls=(), IKFK Switch=True)
   Setup Arm = RigVMCollapseNode(Bones=(), ExecuteContext=(), FK Controls=(), FK Space=(), Fingers Space=(), IK Control=(), IK Parents=(), Pole Vector Control=())
   Setup Clavicle = RigVMCollapseNode(Bone=(), Control=(), ExecuteContext=())
   Setup Fingers = RigVMCollapseNode(Bones=(), Execute=(), Negative Side=False)
   Setup Foot = RigVMCollapseNode(Bone=(), ExecuteContext=(), FK Control=(), IK Control=(), Toes Space=())
   Setup Leg = RigVMCollapseNode(Bones=(), ExecuteContext=(), FK Controls=(), FK Space=(), IK Control=(), IK Parents=(), Pole Vector Control=())
   Setup Neck = RigVMCollapseNode(Bones=(), ExecuteContext=(), FK Controls=(), Head FK Space=(), Head IK Space=(), IK Control=())
   Setup Reverse Foot = RigVMCollapseNode(BK Buffers=(), BK Controls=(), Bones=(), ExecuteContext=(), Roll Buffer=(), Roll Control=())
   Setup Spine = RigVMCollapseNode(Bones=(), Chest Space=(), ExecuteContext=(), FK Controls=(), IK Controls=())
   Spine Match IK onto FK = RigVMCollapseNode(Bones=(), Execute=(), IK Controls=(), Offsets=())
  ]}
 RigVMModel {nodes [
   n1 = &RigUnit_SequenceExecution(ExecuteContext=@n2.ExecutePin)
   Backward Clavicle_1 = RigVMFunctionReferenceNode(Bone=(Type=Bone,Name="clavicle_l"), Control=(Type=Control,Name="clavicle_l_ctrl"), Execute=@n1.B)
   Backward Clavicle = RigVMFunctionReferenceNode(Bone=(Type=Bone,Name="clavicle_r"), Control=(Type=Control,Name="clavicle_r_ctrl"), Execute=@"Backward Clavicle_1".Execute)
   RerouteNode_12_1 = RigVMDispatchNode(Values=((Type=Control,Name="clavicle_l_ctrl"),((),Type=Control,Name="body_ctrl"))) {template="DISPATCH_RigVMDispatch_ArrayMake(in Values,out Array)"}
   RerouteNode_55_1 = RigVMDispatchNode(Values=((Type=Bone,Name="upperarm_l"),((),Type=Bone,Name="lowerarm_l"),((),Type=Bone,Name="hand_l"))) {template="DISPATCH_RigVMDispatch_ArrayMake(in Values,out Array)"}
   RerouteNode_56_1 = RigVMDispatchNode(Values=((Type=Control,Name="upperarm_l_fk_ctrl"),((),Type=Control,Name="lowerarm_l_fk_ctrl"),((),Type=Control,Name="hand_l_fk_ctrl"))) {template="DISPATCH_RigVMDispatch_ArrayMake(in Values,out Array)"}
   Backward Arm = RigVMFunctionReferenceNode(
    FK Arm Space=(Type=Null,Name="upperarm_l_fk_ctrl_space")
    Fingers Space=(Type=Null,Name="fingers_l_space")
    Ik Control=(Type=Control,Name="hand_l_ik_ctrl")
    UpV Control=(Type=Control,Name="arm_l_pv_ik_ctrl")
    ExecuteContext=@"Backward Clavicle".Execute
    FK Orient References=@RerouteNode_12_1.Array
    Bones=@RerouteNode_55_1.Array
    FK Controls=@RerouteNode_56_1.Array
   )
   RerouteNode_13_1 = RigVMDispatchNode(Values=((Type=Control,Name="clavicle_r_ctrl"),((),Type=Control,Name="body_ctrl"))) {template="DISPATCH_RigVMDispatch_ArrayMake(in Values,out Array)"}
   RerouteNode_57_1 = RigVMDispatchNode(Values=((Type=Control,Name="upperarm_r_fk_ctrl"),((),Type=Control,Name="lowerarm_r_fk_ctrl"),((),Type=Control,Name="hand_r_fk_ctrl"))) {template="DISPATCH_RigVMDispatch_ArrayMake(in Values,out Array)"}
   RerouteNode_58_1 = RigVMDispatchNode(Values=((Type=Bone,Name="upperarm_r"),((),Type=Bone,Name="lowerarm_r"),((),Type=Bone,Name="hand_r"))) {template="DISPATCH_RigVMDispatch_ArrayMake(in Values,out Array)"}
   > Backward Arm_1 = RigVMFunctionReferenceNode(
    FK Arm Space=(Type=Null,Name="upperarm_r_fk_ctrl_space")
    Fingers Space=(Type=Null,Name="fingers_r_space")
    Ik Control=(Type=Control,Name="hand_r_ik_ctrl")
    UpV Control=(Type=Control,Name="arm_r_pv_ik_ctrl")
    FK Orient References=@RerouteNode_13_1.Array
    FK Controls=@RerouteNode_57_1.Array
    Bones=@RerouteNode_58_1.Array
   )
   RerouteNode_59_2 = RigVMDispatchNode(Values="((Type=Bone,Name=\"thumb_01_l\"),((),Type=Bone,Name=\"thumb_02_l\"),((),Type=Bone,Name=\"thumb_03_l\"),((),Type=Bone,Name=\"index_metacarpal_l\"),((),Type=Bone,Name=\"index_01_l\"),((),Type=Bone,Name=\"index_02_l\"),((),Type=Bone,Name=\"index_03_l\"),((),Type=Bone,Name= ...(truncated)") {template="DISPATCH_RigVMDispatch_ArrayMake(in Values,out Array)"}
   Backward Fingers = RigVMFunctionReferenceNode(Execute=@"Backward Arm_1".ExecuteContext, Bones=@RerouteNode_59_2.Array)
   RerouteNode_15 = RigVMDispatchNode(Values=((Type=Control,Name="thigh_l_fk_ctrl"),((),Type=Control,Name="calf_l_fk_ctrl"),((),Type=Control,Name="foot_l_fk_ctrl"))) {template="DISPATCH_RigVMDispatch_ArrayMake(in Values,out Array)"}
   RerouteNode_17 = RigVMDispatchNode(Values=((Type=Bone,Name="thigh_l"),((),Type=Bone,Name="calf_l"),((),Type=Bone,Name="foot_l"))) {template="DISPATCH_RigVMDispatch_ArrayMake(in Values,out Array)"}
   Forward Leg_1 = RigVMFunctionReferenceNode(
    IK Control=(Type=Control,Name="foot_l_ik_ctrl")
    IK Effector=(Type=Null,Name="foot_ik_l")
    IKFK Switch Control=(Type=Control,Name="leg_l_fk_ik_switch")
    Negative Side=False
    Pole Vector Control=(Type=Control,Name="leg_l_pv_ik_ctrl")
    Execute=@"Forward Reverse Foot_1".ExecuteContext
    FK Controls=@RerouteNode_15.Array
    Bones=@RerouteNode_17.Array
    Enable Stretch=&RigUnit_GetControlBool(Control=leg_l_stretch_switch) {template="GetControlValue::Execute(in Control,out BoolValue)"}.BoolValue
    Softness=/Script/ControlRig.RigUnit_GetControlFloat(Control=leg_l_softness) {is=struct RigVMUnitNode, template="GetControlValue::Execute(in Control,out FloatValue,out Minimum,out Maximum)"}.FloatValue
    Previous IK Mode=RigVMVariableNode() {variable="L Leg IK Mode "}.Value
   )
   RerouteNode_14 = RigVMDispatchNode(Values=((Type=Control,Name="thigh_r_fk_ctrl"),((),Type=Control,Name="calf_r_fk_ctrl"),((),Type=Control,Name="foot_r_fk_ctrl"))) {template="DISPATCH_RigVMDispatch_ArrayMake(in Values,out Array)"}
   RerouteNode_16 = RigVMDispatchNode(Values=((Type=Bone,Name="thigh_r"),((),Type=Bone,Name="calf_r"),((),Type=Bone,Name="foot_r"))) {template="DISPATCH_RigVMDispatch_ArrayMake(in Values,out Array)"}
   Forward Leg = RigVMFunctionReferenceNode(
    IK Control=(Type=Control,Name="foot_r_ik_ctrl")
    IK Effector=(Type=Null,Name="foot_ik_r")
    IKFK Switch Control=(Type=Control,Name="leg_r_fk_ik_switch")
    Negative Side=true
    Pole Vector Control=(Type=Control,Name="leg_r_pv_ik_ctrl")
    Execute=@"Forward Leg_1".Execute
    FK Controls=@RerouteNode_14.Array
    Bones=@RerouteNode_16.Array
    Enable Stretch=&RigUnit_GetControlBool(Control=leg_r_stretch_switch) {template="GetControlValue::Execute(in Control,out BoolValue)"}.BoolValue
    Softness=&RigUnit_GetControlFloat(Control=leg_r_softness) {template="GetControlValue::Execute(in Control,out FloatValue,out Minimum,out Maximum)"}.FloatValue
    Previous IK Mode=RigVMVariableNode() {variable=R Leg IK Mode}.Value
   )
   Forward Foot = RigVMFunctionReferenceNode(
    Bone=(Type=Bone,Name="ball_l")
    FK Control=(Type=Control,Name="ball_l_fk_ctrl")
    IK Control=(Type=Control,Name="ball_l_ik_ctrl")
    IKFK Switch Control=(Type=Control,Name="leg_l_fk_ik_switch")
    Toes Space=(Type=Null,Name="toes_l_space")
    ExecuteContext=@"Forward Leg".Execute
    Previous IK Mode=RigVMVariableNode() {variable="L Leg IK Mode "}.Value
   )
   > Forward Foot_1 = RigVMFunctionReferenceNode(
    Bone=(Type=Bone,Name="ball_r")
    FK Control=(Type=Control,Name="ball_r_fk_ctrl")
    IK Control=(Type=Control,Name="ball_r_ik_ctrl")
    IKFK Switch Control=(Type=Control,Name="leg_r_fk_ik_switch")
    Toes Space=(Type=Null,Name="toes_r_space")
    Previous IK Mode=RigVMVariableNode() {variable=R Leg IK Mode}.Value
   )
   > RigVMVariableNode(Value=&RigUnit_GetControlBool(Control=leg_l_fk_ik_switch) {template="GetControlValue::Execute(in Control,out BoolValue)"}.BoolValue) {variable="L Leg IK Mode "}
   > RigVMVariableNode(Value=&RigUnit_GetControlBool(Control=leg_r_fk_ik_switch) {template="GetControlValue::Execute(in Control,out BoolValue)"}.BoolValue) {variable=R Leg IK Mode}
   n2 = /Script/ControlRig.RigUnit_InverseExecution() {is=struct RigVMUnitNode}
   n3 = &RigUnit_PrepareForExecution()
   n4 = &RigUnit_SequenceExecution(ExecuteContext=@n3.ExecutePin)
   RigUnit_SetTransform = &RigUnit_SetTransform(
    Item=(Type=Control,Name="body_ctrl")
    Value=(Rotation=(X=0,Y=0,Z=0,W=1),Translation=(X=0,Y=2.5434,Z=91.1834),Scale3D=(X=1,Y=1,Z=1))
    Weight=1
    bInitial=true
    Value.Translation=/Script/RigVM.RigVMFunction_MathTransformLerp(
     T=0.5
     A=&RigUnit_GetTransform(Item=(Type=Bone,Name="spine_01"), bInitial=true).Transform
     B=&RigUnit_GetTransform(Item=(Type=Bone,Name="spine_02"), bInitial=true).Transform
    ) {is=struct RigVMUnitNode, template="Interpolate::Execute(in A,in B,in T,out Result)"}.Result.Translation
    ExecutePin=@n4.A
   )
   RerouteNode_11_1 = RigVMDispatchNode(Values=((Type=Control,Name="hips_ctrl"),((),Type=Control,Name="hips_tan_ctrl"),((),Type=Control,Name="chest_tan_ctrl"),((),Type=Control,Name="chest_ctrl"))) {template="DISPATCH_RigVMDispatch_ArrayMake(in Values,out Array)"}
   RerouteNode_52 = RigVMDispatchNode(Values=((Type=Bone,Name="spine_01"),((),Type=Bone,Name="spine_02"),((),Type=Bone,Name="spine_03"),((),Type=Bone,Name="spine_04"))) {template="DISPATCH_RigVMDispatch_ArrayMake(in Values,out Array)"}
   RerouteNode_53 = RigVMDispatchNode(Values=((Type=Control,Name="hips_ctrl"),((),Type=Control,Name="spine_01_ctrl"),((),Type=Control,Name="spine_02_ctrl"),((),Type=Control,Name="spine_03_ctrl"))) {template="DISPATCH_RigVMDispatch_ArrayMake(in Values,out Array)"}
   > Setup Spine = RigVMFunctionReferenceNode(
    Chest Space=(Type=Null,Name="chest_space")
    IK Controls=@RerouteNode_11_1.Array
    Bones=@RerouteNode_52.Array
    FK Controls=@RerouteNode_53.Array
   )
   > RigVMVariableNode(Value=@"Setup Spine".Spline Offsets) {variable=Spine Offsets}
   > RigVMVariableNode(Value=@"Setup Spine".Length) {variable=Spine Length}
   RerouteNode_51 = RigVMDispatchNode(Values=((Type=Bone,Name="neck_01"),((),Type=Bone,Name="neck_02"),((),Type=Bone,Name="head"))) {template="DISPATCH_RigVMDispatch_ArrayMake(in Values,out Array)"}
   RerouteNode_54 = RigVMDispatchNode(Values=((Type=Control,Name="neck_01_ctrl"),((),Type=Control,Name="neck_02_ctrl"),((),Type=Control,Name="head_ctrl"))) {template="DISPATCH_RigVMDispatch_ArrayMake(in Values,out Array)"}
   > Setup Neck_1 = RigVMFunctionReferenceNode(
    Head FK Space=(Type=Null,Name="head_fk_space")
    Head IK Space=(Type=Null,Name="head_ik_space")
    IK Control=(Type=Control,Name="head_ik_ctrl")
    Bones=@RerouteNode_51.Array
    FK Controls=@RerouteNode_54.Array
   )
   > RigVMVariableNode(Value=@"Setup Neck_1".Length) {variable=Neck Length}
   Setup Clavicle = RigVMFunctionReferenceNode(Bone=(Type=Bone,Name="clavicle_l"), Control=(Type=Control,Name="clavicle_l_ctrl"), ExecuteContext=@n4.B)
   > Setup Clavicle_1 = RigVMFunctionReferenceNode(Bone=(Type=Bone,Name="clavicle_r"), Control=(Type=Control,Name="clavicle_r_ctrl"), Negative Side=true)
   RerouteNode = RigVMDispatchNode(Values=((Type=Control,Name="body_ctrl"),((),Type=Null,Name="chest_space"),((),Type=Control,Name="clavicle_l_ctrl"))) {template="DISPATCH_RigVMDispatch_ArrayMake(in Values,out Array)"}
   RerouteNode_55 = RigVMDispatchNode(Values=((Type=Bone,Name="upperarm_l"),((),Type=Bone,Name="lowerarm_l"),((),Type=Bone,Name="hand_l"))) {template="DISPATCH_RigVMDispatch_ArrayMake(in Values,out Array)"}
   RerouteNode_56 = RigVMDispatchNode(Values=((Type=Control,Name="upperarm_l_fk_ctrl"),((),Type=Control,Name="lowerarm_l_fk_ctrl"),((),Type=Control,Name="hand_l_fk_ctrl"))) {template="DISPATCH_RigVMDispatch_ArrayMake(in Values,out Array)"}
   > Setup Arm = RigVMFunctionReferenceNode(
    FK Space=(Type=Null,Name="upperarm_l_fk_ctrl_space")
    Fingers Space=(Type=Null,Name="fingers_l_space")
    IK Control=(Type=Control,Name="hand_l_ik_ctrl")
    Pole Vector Control=(Type=Control,Name="arm_l_pv_ik_ctrl")
    IK Parents=@RerouteNode.Array
    Bones=@RerouteNode_55.Array
    FK Controls=@RerouteNode_56.Array
   )
   RerouteNode_34 = RigVMDispatchNode(Values=((Type=Control,Name="body_ctrl"),((),Type=Null,Name="chest_space"),((),Type=Control,Name="clavicle_r_ctrl"))) {template="DISPATCH_RigVMDispatch_ArrayMake(in Values,out Array)"}
   RerouteNode_57 = RigVMDispatchNode(Values=((Type=Control,Name="upperarm_r_fk_ctrl"),((),Type=Control,Name="lowerarm_r_fk_ctrl"),((),Type=Control,Name="hand_r_fk_ctrl"))) {template="DISPATCH_RigVMDispatch_ArrayMake(in Values,out Array)"}
   RerouteNode_58 = RigVMDispatchNode(Values=((Type=Bone,Name="upperarm_r"),((),Type=Bone,Name="lowerarm_r"),((),Type=Bone,Name="hand_r"))) {template="DISPATCH_RigVMDispatch_ArrayMake(in Values,out Array)"}
   > Setup Arm_1 = RigVMFunctionReferenceNode(
    FK Space=(Type=Null,Name="upperarm_r_fk_ctrl_space")
    Fingers Space=(Type=Null,Name="fingers_r_space")
    IK Control=(Type=Control,Name="hand_r_ik_ctrl")
    Negative Side=true
    Pole Vector Control=(Type=Control,Name="arm_r_pv_ik_ctrl")
    IK Parents=@RerouteNode_34.Array
    FK Controls=@RerouteNode_57.Array
    Bones=@RerouteNode_58.Array
   )
   RerouteNode_59 = RigVMDispatchNode(Values="((Type=Bone,Name=\"thumb_01_l\"),((),Type=Bone,Name=\"thumb_02_l\"),((),Type=Bone,Name=\"thumb_03_l\"),((),Type=Bone,Name=\"index_metacarpal_l\"),((),Type=Bone,Name=\"index_01_l\"),((),Type=Bone,Name=\"index_02_l\"),((),Type=Bone,Name=\"index_03_l\"),((),Type=Bone,Name= ...(truncated)") {template="DISPATCH_RigVMDispatch_ArrayMake(in Values,out Array)"}
   Setup Fingers = RigVMFunctionReferenceNode(Negative Side=False, Bones=@RerouteNode_59.Array, Execute=@"Setup Arm_1".ExecuteContext)
   RerouteNode_27 = RigVMDispatchNode(Values=((Type=Bone,Name="thigh_l"),((),Type=Bone,Name="calf_l"),((),Type=Bone,Name="foot_l"))) {template="DISPATCH_RigVMDispatch_ArrayMake(in Values,out Array)"}
   RerouteNode_35 = RigVMDispatchNode(Values=((Type=Control,Name="body_ctrl"),((),Type=Control,Name="hips_ctrl"))) {template="DISPATCH_RigVMDispatch_ArrayMake(in Values,out Array)"}
   RerouteNode_61 = RigVMDispatchNode(Values=((Type=Control,Name="thigh_l_fk_ctrl"),((),Type=Control,Name="calf_l_fk_ctrl"),((),Type=Control,Name="foot_l_fk_ctrl"))) {template="DISPATCH_RigVMDispatch_ArrayMake(in Values,out Array)"}
   Setup Leg = RigVMFunctionReferenceNode(
    FK Space=(Type=Null,Name="thigh_l_fk_ctrl_space")
    IK Control=(Type=Control,Name="foot_l_ik_ctrl")
    Pole Vector Control=(Type=Control,Name="leg_l_pv_ik_ctrl")
    Bones=@RerouteNode_27.Array
    IK Parents=@RerouteNode_35.Array
    FK Controls=@RerouteNode_61.Array
    ExecuteContext=@n4.C
   )
   RerouteNode_35_1 = RigVMDispatchNode(Values=((Type=Control,Name="body_ctrl"),((),Type=Control,Name="hips_ctrl"))) {template="DISPATCH_RigVMDispatch_ArrayMake(in Values,out Array)"}
   RerouteNode_62 = RigVMDispatchNode(Values=((Type=Control,Name="thigh_r_fk_ctrl"),((),Type=Control,Name="calf_r_fk_ctrl"),((),Type=Control,Name="foot_r_fk_ctrl"))) {template="DISPATCH_RigVMDispatch_ArrayMake(in Values,out Array)"}
   RerouteNode_64 = RigVMDispatchNode(Values=((Type=Bone,Name="thigh_r"),((),Type=Bone,Name="calf_r"),((),Type=Bone,Name="foot_r"))) {template="DISPATCH_RigVMDispatch_ArrayMake(in Values,out Array)"}
   > Setup Leg_1 = RigVMFunctionReferenceNode(
    FK Space=(Type=Null,Name="thigh_r_fk_ctrl_space")
    IK Control=(Type=Control,Name="foot_r_ik_ctrl")
    Negative Side=true
    Pole Vector Control=(Type=Control,Name="leg_r_pv_ik_ctrl")
    IK Parents=@RerouteNode_35_1.Array
    FK Controls=@RerouteNode_62.Array
    Bones=@RerouteNode_64.Array
   )
   > Setup Foot = RigVMFunctionReferenceNode(
    Bone=(Type=Bone,Name="ball_l")
    FK Control=(Type=Control,Name="ball_l_fk_ctrl")
    IK Control=(Type=Control,Name="ball_l_ik_ctrl")
    Toes Space=(Type=Null,Name="toes_l_space")
   )
   > Setup Foot_1 = RigVMFunctionReferenceNode(
    Bone=(Type=Bone,Name="ball_r")
    FK Control=(Type=Control,Name="ball_r_fk_ctrl")
    IK Control=(Type=Control,Name="ball_r_ik_ctrl")
    Negative Side=true
    Toes Space=(Type=Null,Name="toes_r_space")
   )
   RerouteNode_11 = RigVMDispatchNode(Values=((Type=Bone,Name="heel_l"),((),Type=Bone,Name="tip_l"),((),Type=Bone,Name="ball_l"),((),Type=Bone,Name="foot_l"))) {template="DISPATCH_RigVMDispatch_ArrayMake(in Values,out Array)"}
   RerouteNode_18 = RigVMDispatchNode(Values=((Type=Control,Name="heel_l_ctrl"),((),Type=Control,Name="tip_l_ctrl"),((),Type=Control,Name="foot_bk1_l_ctrl"),((),Type=Null,Name="foot_ik_l"))) {template="DISPATCH_RigVMDispatch_ArrayMake(in Values,out Array)"}
   RerouteNode_23 = RigVMDispatchNode(Values=((Type=Null,Name="heel_l"),((),Type=Null,Name="tip_l"),((),Type=Null,Name="foot_bk1_l"),((),Type=Null,Name="foot_ik_l"))) {template="DISPATCH_RigVMDispatch_ArrayMake(in Values,out Array)"}
   > Setup Reverse Foot = RigVMFunctionReferenceNode(
    Roll Buffer=(Type=Null,Name="foot_roll_l_bfr")
    Roll Control=(Type=Control,Name="foot_roll_l_ctrl")
    Bones=@RerouteNode_11.Array
    BK Controls=@RerouteNode_18.Array
    BK Buffers=@RerouteNode_23.Array
   )
   RerouteNode_29 = RigVMDispatchNode(Values=((Type=Null,Name="heel_r"),((),Type=Null,Name="tip_r"),((),Type=Null,Name="foot_bk1_r"),((),Type=Null,Name="foot_ik_r"))) {template="DISPATCH_RigVMDispatch_ArrayMake(in Values,out Array)"}
   RerouteNode_30 = RigVMDispatchNode(Values=((Type=Control,Name="heel_r_ctrl"),((),Type=Control,Name="tip_r_ctrl"),((),Type=Control,Name="foot_bk1_r_ctrl"),((),Type=Null,Name="foot_ik_r"))) {template="DISPATCH_RigVMDispatch_ArrayMake(in Values,out Array)"}
   RerouteNode_31 = RigVMDispatchNode(Values=((Type=Bone,Name="heel_r"),((),Type=Bone,Name="tip_r"),((),Type=Bone,Name="ball_r"),((),Type=Bone,Name="foot_r"))) {template="DISPATCH_RigVMDispatch_ArrayMake(in Values,out Array)"}
   > Setup Reverse Foot_1 = RigVMFunctionReferenceNode(
    Negative Side=true
    Roll Buffer=(Type=Null,Name="foot_roll_r_bfr")
    Roll Control=(Type=Control,Name="foot_roll_r_ctrl")
    BK Buffers=@RerouteNode_29.Array
    BK Controls=@RerouteNode_30.Array
    Bones=@RerouteNode_31.Array
   )
   n5 = &RigUnit_BeginExecution()
   n6 = &RigUnit_SequenceExecution(ExecuteContext=@n5.ExecutePin)
   &RigUnit_SetTransform(
    Item=(Type=Bone,Name="root")
    Weight=1
    bInitial=False
    Value=&RigUnit_GetTransform(Item=(Type=Control,Name="root_ctrl")).Transform
    ExecutePin=@n6.A
   )
   RerouteNode_11_1_1 = RigVMDispatchNode(Values=((Type=Control,Name="hips_ctrl"),((),Type=Control,Name="hips_tan_ctrl"),((),Type=Control,Name="chest_tan_ctrl"),((),Type=Control,Name="chest_ctrl"))) {template="DISPATCH_RigVMDispatch_ArrayMake(in Values,out Array)"}
   RerouteNode_21 = RigVMDispatchNode(Values=((Type=Control,Name="hips_ctrl"),((),Type=Control,Name="spine_01_ctrl"),((),Type=Control,Name="spine_02_ctrl"),((),Type=Control,Name="spine_03_ctrl"))) {template="DISPATCH_RigVMDispatch_ArrayMake(in Values,out Array)"}
   RerouteNode_4 = RigVMDispatchNode(Values=((Type=Bone,Name="spine_01"),((),Type=Bone,Name="spine_02"),((),Type=Bone,Name="spine_03"),((),Type=Bone,Name="spine_04"))) {template="DISPATCH_RigVMDispatch_ArrayMake(in Values,out Array)"}
   > Forward Spine = RigVMFunctionReferenceNode(
    Chest Space=(Type=Null,Name="chest_space")
    IKFK Switch Control=(Type=Control,Name="spine_fk_ik_switch")
    Pelvis Bone=(Type=Bone,Name="pelvis")
    IK Controls=@RerouteNode_11_1_1.Array
    FK Controls=@RerouteNode_21.Array
    Bones=@RerouteNode_4.Array
    "Enable Stretch "=&RigUnit_GetControlBool(Control=spine_stretch_switch) {template="GetControlValue::Execute(in Control,out BoolValue)"}.BoolValue
    Previous IK Mode=RigVMVariableNode() {variable=Spine IK Mode}.Value
    Default Length=RigVMVariableNode() {variable=Spine Length}.Value
    Spline Offsets=RigVMVariableNode() {variable=Spine Offsets}.Value
   )
   RerouteNode_5_1 = RigVMDispatchNode(Values=((Type=Control,Name="neck_01_ctrl"),((),Type=Control,Name="neck_02_ctrl"),((),Type=Control,Name="head_ctrl"))) {template="DISPATCH_RigVMDispatch_ArrayMake(in Values,out Array)"}
   RerouteNode_9 = RigVMDispatchNode(Values=((Type=Bone,Name="neck_01"),((),Type=Bone,Name="neck_02"),((),Type=Bone,Name="head"))) {template="DISPATCH_RigVMDispatch_ArrayMake(in Values,out Array)"}
   > Forward Neck_1 = RigVMFunctionReferenceNode(
    Head FK Space=(Type=Null,Name="head_fk_space")
    Head IK Space=(Type=Null,Name="head_ik_space")
    Head Orient Control=(Type=Control,Name="head_local")
    Head Orient Reference=(Type=Control,Name="body_ctrl")
    IK Control=(Type=Control,Name="head_ik_ctrl")
    IKFK Switch Control=(Type=Control,Name="neck_fk_ik_switch")
    Mid IK Null=(Type=Null,Name="head_mid")
    FK Controls=@RerouteNode_5_1.Array
    Bones=@RerouteNode_9.Array
    Enable Stretch=&RigUnit_GetControlBool(Control=head_stretch) {template="GetControlValue::Execute(in Control,out BoolValue)"}.BoolValue
    Length=RigVMVariableNode() {variable=Neck Length}.Value
    Previous IK Mode=RigVMVariableNode() {variable=Neck IK Mode}.Value
   )
   > RigVMVariableNode(Value=&RigUnit_GetControlBool(Control=head_stretch) {template="GetControlValue::Execute(in Control,out BoolValue)"}.BoolValue) {variable=Neck IK Mode}
   Forward Clavicle_1 = RigVMFunctionReferenceNode(Bone=(Type=Bone,Name="clavicle_l"), Control=(Type=Control,Name="clavicle_l_ctrl"), ExecuteContext=@n6.B)
   > Forward Clavicle = RigVMFunctionReferenceNode(Bone=(Type=Bone,Name="clavicle_r"), Control=(Type=Control,Name="clavicle_r_ctrl"))
   RerouteNode_12 = RigVMDispatchNode(Values=((Type=Control,Name="clavicle_l_ctrl"),((),Type=Control,Name="body_ctrl"))) {template="DISPATCH_RigVMDispatch_ArrayMake(in Values,out Array)"}
   RerouteNode_65 = RigVMDispatchNode(Values=((Type=Control,Name="upperarm_l_fk_ctrl"),((),Type=Control,Name="lowerarm_l_fk_ctrl"),((),Type=Control,Name="hand_l_fk_ctrl"))) {template="DISPATCH_RigVMDispatch_ArrayMake(in Values,out Array)"}
   RerouteNode_66 = RigVMDispatchNode(Values=((Type=Bone,Name="upperarm_l"),((),Type=Bone,Name="lowerarm_l"),((),Type=Bone,Name="hand_l"))) {template="DISPATCH_RigVMDispatch_ArrayMake(in Values,out Array)"}
   Forward Arm = RigVMFunctionReferenceNode(
    Arm Space=(Type=Null,Name="upperarm_l_fk_ctrl_space")
    FK Orient Control=(Type=Control,Name="arm_l_fk_local")
    Finger Space=(Type=Null,Name="fingers_l_space")
    IK Control=(Type=Control,Name="hand_l_ik_ctrl")
    IK Effector=(Type=Control,Name="hand_l_ik_ctrl")
    IKFK Switch Control=(Type=Control,Name="arm_l_fk_ik_switch")
    Negative Side=False
    Pole Vector Control=(Type=Control,Name="arm_l_pv_ik_ctrl")
    Execute=@"Forward Clavicle".ExecuteContext
    FK Orient References=@RerouteNode_12.Array
    FK Controls=@RerouteNode_65.Array
    Bones=@RerouteNode_66.Array
    Enable Stretch=&RigUnit_GetControlBool(Control=arm_l_stretch_switch) {template="GetControlValue::Execute(in Control,out BoolValue)"}.BoolValue
    Softness=&RigUnit_GetControlFloat(Control=arm_l_softness) {template="GetControlValue::Execute(in Control,out FloatValue,out Minimum,out Maximum)"}.FloatValue
    Previous IK Mode=RigVMVariableNode() {variable=L Arm IK Mode}.Value
   )
   n7 = RigVMVariableNode(
    ExecuteContext=@"Forward Arm".Execute
    Value=&RigUnit_GetControlBool(Control=arm_l_fk_ik_switch) {template="GetControlValue::Execute(in Control,out BoolValue)"}.BoolValue
   ) {variable=L Arm IK Mode}
   RerouteNode_13 = RigVMDispatchNode(Values=((Type=Control,Name="clavicle_r_ctrl"),((),Type=Control,Name="body_ctrl"))) {template="DISPATCH_RigVMDispatch_ArrayMake(in Values,out Array)"}
   RerouteNode_67 = RigVMDispatchNode(Values=((Type=Control,Name="upperarm_r_fk_ctrl"),((),Type=Control,Name="lowerarm_r_fk_ctrl"),((),Type=Control,Name="hand_r_fk_ctrl"))) {template="DISPATCH_RigVMDispatch_ArrayMake(in Values,out Array)"}
   RerouteNode_68 = RigVMDispatchNode(Values=((Type=Bone,Name="upperarm_r"),((),Type=Bone,Name="lowerarm_r"),((),Type=Bone,Name="hand_r"))) {template="DISPATCH_RigVMDispatch_ArrayMake(in Values,out Array)"}
   Forward Arm_1 = RigVMFunctionReferenceNode(
    Arm Space=(Type=Null,Name="upperarm_r_fk_ctrl_space")
    FK Orient Control=(Type=Control,Name="arm_r_fk_local")
    Finger Space=(Type=Null,Name="fingers_r_space")
    IK Control=(Type=Control,Name="hand_r_ik_ctrl")
    IK Effector=(Type=Control,Name="hand_r_ik_ctrl")
    IKFK Switch Control=(Type=Control,Name="arm_r_fk_ik_switch")
    Negative Side=true
    Pole Vector Control=(Type=Control,Name="arm_r_pv_ik_ctrl")
    FK Orient References=@RerouteNode_13.Array
    FK Controls=@RerouteNode_67.Array
    Bones=@RerouteNode_68.Array
    Enable Stretch=&RigUnit_GetControlBool(Control=arm_r_stretch_switch) {template="GetControlValue::Execute(in Control,out BoolValue)"}.BoolValue
    Softness=&RigUnit_GetControlFloat(Control=arm_r_softness) {template="GetControlValue::Execute(in Control,out FloatValue,out Minimum,out Maximum)"}.FloatValue
    Previous IK Mode=RigVMVariableNode() {variable=R Arm IK Mode}.Value
    Execute=@n7.ExecuteContext
   )
   n8 = RigVMVariableNode(
    ExecuteContext=@"Forward Arm_1".Execute
    Value=&RigUnit_GetControlBool(Control=arm_r_fk_ik_switch) {template="GetControlValue::Execute(in Control,out BoolValue)"}.BoolValue
   ) {variable=R Arm IK Mode}
   RerouteNode_59_1 = RigVMDispatchNode(Values="((Type=Bone,Name=\"thumb_01_l\"),((),Type=Bone,Name=\"thumb_02_l\"),((),Type=Bone,Name=\"thumb_03_l\"),((),Type=Bone,Name=\"index_metacarpal_l\"),((),Type=Bone,Name=\"index_01_l\"),((),Type=Bone,Name=\"index_02_l\"),((),Type=Bone,Name=\"index_03_l\"),((),Type=Bone,Name= ...(truncated)") {template="DISPATCH_RigVMDispatch_ArrayMake(in Values,out Array)"}
   Forward Fingers = RigVMFunctionReferenceNode(Bones=@RerouteNode_59_1.Array, Execute=@n8.ExecuteContext)
   RerouteNode_7 = RigVMDispatchNode(Values=((Type=Null,Name="thigh_l_fk_ctrl_space"),((),Type=Null,Name="thigh_r_fk_ctrl_space"))) {template="DISPATCH_RigVMDispatch_ArrayMake(in Values,out Array)"}
   n9 = RigVMDispatchNode(Array=@RerouteNode_7.Array, ExecuteContext=@n6.C) {template="DISPATCH_RigVMDispatch_ArrayIterator(in Array,out Element,out Index,out Count,out Ratio)"}
   n10 = &RigUnit_ProjectTransformToNewParent(
    NewParent=(Type=Control,Name="hips_ctrl")
    OldParent=(Type=Control,Name="hips_ctrl")
    bChildInitial=True
    bOldParentInitial=True
    Child=@n9.Element
   )
   > &RigUnit_SetTransform(Weight=1, bInitial=False, bPropagateToChildren=true, Item=@n9.Element, Value=@n10.Transform)
   RerouteNode_25 = RigVMDispatchNode(Values=((Type=Null,Name="heel_l"),((),Type=Null,Name="tip_l"),((),Type=Null,Name="foot_bk1_l"))) {template="DISPATCH_RigVMDispatch_ArrayMake(in Values,out Array)"}
   RerouteNode_28 = RigVMDispatchNode(Values=((Type=Control,Name="heel_l_ctrl"),((),Type=Control,Name="tip_l_ctrl"),((),Type=Control,Name="foot_bk1_l_ctrl"))) {template="DISPATCH_RigVMDispatch_ArrayMake(in Values,out Array)"}
   RigUnit_GetControlFloat = &RigUnit_GetControlFloat(Control=roll_blend_l_ctrl) {template="GetControlValue::Execute(in Control,out FloatValue,out Minimum,out Maximum)"}
   Forward Reverse Foot = RigVMFunctionReferenceNode(
    IKFK Switch Control=(Type=Control,Name="leg_l_fk_ik_switch")
    Roll Buffer=(Type=Null,Name="foot_roll_l_bfr")
    Roll Clamp=40
    Roll Control=(Type=Control,Name="foot_roll_l_ctrl")
    ExecuteContext=@n9.Completed
    BK Buffers=@RerouteNode_25.Array
    BK Controls=@RerouteNode_28.Array
    Roll Blend=@RigUnit_GetControlFloat.FloatValue
   )
   RerouteNode_32 = RigVMDispatchNode(Values=((Type=Null,Name="heel_r"),((),Type=Null,Name="tip_r"),((),Type=Null,Name="foot_bk1_r"))) {template="DISPATCH_RigVMDispatch_ArrayMake(in Values,out Array)"}
   RerouteNode_33 = RigVMDispatchNode(Values=((Type=Control,Name="heel_r_ctrl"),((),Type=Control,Name="tip_r_ctrl"),((),Type=Control,Name="foot_bk1_r_ctrl"))) {template="DISPATCH_RigVMDispatch_ArrayMake(in Values,out Array)"}
   > Forward Reverse Foot_1 = RigVMFunctionReferenceNode(
    IKFK Switch Control=(Type=Control,Name="leg_r_fk_ik_switch")
    Roll Buffer=(Type=Null,Name="foot_roll_r_bfr")
    Roll Clamp=40
    Roll Control=(Type=Control,Name="foot_roll_r_ctrl")
    BK Buffers=@RerouteNode_32.Array
    BK Controls=@RerouteNode_33.Array
    Roll Blend=&RigUnit_GetControlFloat(Control=roll_blend_r_ctrl) {template="GetControlValue::Execute(in Control,out FloatValue,out Minimum,out Maximum)"}.FloatValue
   )
   RerouteNode_19 = RigVMDispatchNode(Values="((Type=Control,Name=\"pinky_03_r_ctrl\"),((),Type=Control,Name=\"pinky_02_r_ctrl\"),((),Type=Control,Name=\"pinky_01_r_ctrl\"),((),Type=Control,Name=\"pinky_metacarpal_r_ctrl\"),((),Type=Control,Name=\"ring_03_r_ctrl\"),((),Type=Control,Name=\"ring_02_r_ctrl\"),((),Ty ...(truncated)") {template="DISPATCH_RigVMDispatch_ArrayMake(in Values,out Array)"}
   n11 = RigVMDispatchNode(Array=@RerouteNode_19.Array, ExecuteContext=@n6.D) {template="DISPATCH_RigVMDispatch_ArrayIterator(in Array,out Element,out Index,out Count,out Ratio)"}
   RigUnit_GetControlBool = &RigUnit_GetControlBool(Control=ShowBodyControls) {template="GetControlValue::Execute(in Control,out BoolValue)"}
   > &RigUnit_SetControlVisibility(Item=@n11.Element, bVisible=@RigUnit_GetControlBool.BoolValue) {template="SetControlVisibility::Execute(in Item,in Pattern,in bVisible)"}
   RerouteNode_22 = RigVMDispatchNode(Values=((Type=Control,Name="thigh_l_fk_ctrl"),((),Type=Control,Name="calf_l_fk_ctrl"),((),Type=Control,Name="foot_l_fk_ctrl"),((),Type=Control,Name="ball_l_fk_ctrl"))) {template="DISPATCH_RigVMDispatch_ArrayMake(in Values,out Array)"}
   RerouteNode_26 = RigVMDispatchNode(Values=((Type=Control,Name="leg_l_pv_ik_ctrl"),((),Type=Control,Name="foot_l_ik_ctrl"),((),Type=Control,Name="ball_l_ik_ctrl"))) {template="DISPATCH_RigVMDispatch_ArrayMake(in Values,out Array)"}
   n12 = &RigUnit_GetControlBool(Control=ShowBodyControls) {template="GetControlValue::Execute(in Control,out BoolValue)"}
   Set IKFK Visibility = RigVMFunctionReferenceNode(
    ExecuteContext=@n11.Completed
    FK Controls=@RerouteNode_22.Array
    IK Controls=@RerouteNode_26.Array
    IKFK Switch=&RigUnit_GetControlBool(Control=leg_l_fk_ik_switch) {template="GetControlValue::Execute(in Control,out BoolValue)"}.BoolValue
    Show Controls=@n12.BoolValue
   )
   RerouteNode_38 = RigVMDispatchNode(Values=((Type=Control,Name="thigh_r_fk_ctrl"),((),Type=Control,Name="calf_r_fk_ctrl"),((),Type=Control,Name="foot_r_fk_ctrl"),((),Type=Control,Name="ball_r_fk_ctrl"))) {template="DISPATCH_RigVMDispatch_ArrayMake(in Values,out Array)"}
   RerouteNode_39 = RigVMDispatchNode(Values=((Type=Control,Name="foot_r_ik_ctrl"),((),Type=Control,Name="ball_r_ik_ctrl"),((),Type=Control,Name="leg_r_pv_ik_ctrl"))) {template="DISPATCH_RigVMDispatch_ArrayMake(in Values,out Array)"}
   > Set IKFK Visibility_1 = RigVMFunctionReferenceNode(
    FK Controls=@RerouteNode_38.Array
    IK Controls=@RerouteNode_39.Array
    IKFK Switch=&RigUnit_GetControlBool(Control=leg_r_fk_ik_switch) {template="GetControlValue::Execute(in Control,out BoolValue)"}.BoolValue
    Show Controls=@n12.BoolValue
   )
   RerouteNode_40 = RigVMDispatchNode(Values=((Type=Control,Name="upperarm_l_fk_ctrl"),((),Type=Control,Name="lowerarm_l_fk_ctrl"),((),Type=Control,Name="hand_l_fk_ctrl"))) {template="DISPATCH_RigVMDispatch_ArrayMake(in Values,out Array)"}
   RerouteNode_41 = RigVMDispatchNode(Values=((Type=Control,Name="hand_l_ik_ctrl"),((),Type=Control,Name="arm_l_pv_ik_ctrl"))) {template="DISPATCH_RigVMDispatch_ArrayMake(in Values,out Array)"}
   > Set IKFK Visibility_1_1 = RigVMFunctionReferenceNode(
    FK Controls=@RerouteNode_40.Array
    IK Controls=@RerouteNode_41.Array
    IKFK Switch=&RigUnit_GetControlBool(Control=arm_l_fk_ik_switch) {template="GetControlValue::Execute(in Control,out BoolValue)"}.BoolValue
    Show Controls=@n12.BoolValue
   )
   RerouteNode_42 = RigVMDispatchNode(Values=((Type=Control,Name="upperarm_r_fk_ctrl"),((),Type=Control,Name="lowerarm_r_fk_ctrl"),((),Type=Control,Name="hand_r_fk_ctrl"))) {template="DISPATCH_RigVMDispatch_ArrayMake(in Values,out Array)"}
   RerouteNode_43 = RigVMDispatchNode(Values=((Type=Control,Name="hand_r_ik_ctrl"),((),Type=Control,Name="arm_r_pv_ik_ctrl"))) {template="DISPATCH_RigVMDispatch_ArrayMake(in Values,out Array)"}
   > Set IKFK Visibility_1_2 = RigVMFunctionReferenceNode(
    FK Controls=@RerouteNode_42.Array
    IK Controls=@RerouteNode_43.Array
    IKFK Switch=&RigUnit_GetControlBool(Control=arm_r_fk_ik_switch) {template="GetControlValue::Execute(in Control,out BoolValue)"}.BoolValue
    Show Controls=@n12.BoolValue
   )
   RerouteNode_11_1_1_1 = RigVMDispatchNode(Values=((Type=Control,Name="hips_tan_ctrl"),((),Type=Control,Name="chest_tan_ctrl"),((),Type=Control,Name="chest_ctrl"))) {template="DISPATCH_RigVMDispatch_ArrayMake(in Values,out Array)"}
   RerouteNode_21_1 = RigVMDispatchNode(Values=((Type=Control,Name="spine_01_ctrl"),((),Type=Control,Name="spine_02_ctrl"),((),Type=Control,Name="spine_03_ctrl"))) {template="DISPATCH_RigVMDispatch_ArrayMake(in Values,out Array)"}
   > Set IKFK Visibility_1_2_1 = RigVMFunctionReferenceNode(
    IK Controls=@RerouteNode_11_1_1_1.Array
    FK Controls=@RerouteNode_21_1.Array
    IKFK Switch=&RigUnit_GetControlBool(Control=spine_fk_ik_switch) {template="GetControlValue::Execute(in Control,out BoolValue)"}.BoolValue
    Show Controls=@n12.BoolValue
   )
   RerouteNode_5 = RigVMDispatchNode(Values=((Type=Control,Name="head_ik_ctrl"))) {template="DISPATCH_RigVMDispatch_ArrayMake(in Values,out Array)"}
   RerouteNode_5_1_1 = RigVMDispatchNode(Values=((Type=Control,Name="neck_01_ctrl"),((),Type=Control,Name="neck_02_ctrl"),((),Type=Control,Name="head_ctrl"))) {template="DISPATCH_RigVMDispatch_ArrayMake(in Values,out Array)"}
   > Set IKFK Visibility_1_2_1_1 = RigVMFunctionReferenceNode(
    IK Controls=@RerouteNode_5.Array
    FK Controls=@RerouteNode_5_1_1.Array
    IKFK Switch=&RigUnit_GetControlBool(Control=neck_fk_ik_switch) {template="GetControlValue::Execute(in Control,out BoolValue)"}.BoolValue
    Show Controls=@n12.BoolValue
   )
   RerouteNode_60_2 = RigVMDispatchNode(Values="((Type=Bone,Name=\"thumb_01_r\"),((),Type=Bone,Name=\"thumb_02_r\"),((),Type=Bone,Name=\"thumb_03_r\"),((),Type=Bone,Name=\"index_metacarpal_r\"),((),Type=Bone,Name=\"index_01_r\"),((),Type=Bone,Name=\"index_02_r\"),((),Type=Bone,Name=\"index_03_r\"),((),Type=Bone,Name= ...(truncated)") {template="DISPATCH_RigVMDispatch_ArrayMake(in Values,out Array)"}
   Backward Fingers_1 = RigVMFunctionReferenceNode(Execute=@"Backward Fingers".Execute, Bones=@RerouteNode_60_2.Array)
   Backward Foot_1 = RigVMFunctionReferenceNode(
    Bone=(Type=Bone,Name="ball_l")
    FK Control=(Type=Control,Name="ball_l_fk_ctrl")
    IK Control=(Type=Control,Name="ball_l_ik_ctrl")
    Toes Space=(Type=Null,Name="toes_l_space")
    Execute=@"Backward Leg_1".ExecuteContext
   )
   Backward Foot = RigVMFunctionReferenceNode(
    Bone=(Type=Bone,Name="ball_r")
    FK Control=(Type=Control,Name="ball_r_fk_ctrl")
    IK Control=(Type=Control,Name="ball_r_ik_ctrl")
    Toes Space=(Type=Null,Name="toes_r_space")
    Execute=@"Backward Foot_1".Execute
   )
   RerouteNode_20 = RigVMDispatchNode(Values=((Type=Control,Name="neck_01_ctrl"),((),Type=Control,Name="neck_02_ctrl"),((),Type=Control,Name="head_ctrl"))) {template="DISPATCH_RigVMDispatch_ArrayMake(in Values,out Array)"}
   RerouteNode_24 = RigVMDispatchNode(Values=((Type=Bone,Name="neck_01"),((),Type=Bone,Name="neck_02"),((),Type=Bone,Name="head"))) {template="DISPATCH_RigVMDispatch_ArrayMake(in Values,out Array)"}
   Backward Neck_1 = RigVMFunctionReferenceNode(
    Head FK Space=(Type=Null,Name="head_fk_space")
    Head IK Space=(Type=Null,Name="head_ik_space")
    Head Orient Reference=(Type=Control,Name="body_ctrl")
    IK Control=(Type=Control,Name="head_ik_ctrl")
    ExecuteContext=@"Backward Spine".ExecuteContext
    FK Controls=@RerouteNode_20.Array
    Bones=@RerouteNode_24.Array
   )
   RerouteNode_60_1 = RigVMDispatchNode(Values="((Type=Bone,Name=\"thumb_01_r\"),((),Type=Bone,Name=\"thumb_02_r\"),((),Type=Bone,Name=\"thumb_03_r\"),((),Type=Bone,Name=\"index_metacarpal_r\"),((),Type=Bone,Name=\"index_01_r\"),((),Type=Bone,Name=\"index_02_r\"),((),Type=Bone,Name=\"index_03_r\"),((),Type=Bone,Name= ...(truncated)") {template="DISPATCH_RigVMDispatch_ArrayMake(in Values,out Array)"}
   Forward Fingers_1 = RigVMFunctionReferenceNode(Execute=@"Forward Fingers".Execute, Bones=@RerouteNode_60_1.Array)
   RerouteNode_60 = RigVMDispatchNode(Values="((Type=Bone,Name=\"thumb_01_r\"),((),Type=Bone,Name=\"thumb_02_r\"),((),Type=Bone,Name=\"thumb_03_r\"),((),Type=Bone,Name=\"index_metacarpal_r\"),((),Type=Bone,Name=\"index_01_r\"),((),Type=Bone,Name=\"index_02_r\"),((),Type=Bone,Name=\"index_03_r\"),((),Type=Bone,Name= ...(truncated)") {template="DISPATCH_RigVMDispatch_ArrayMake(in Values,out Array)"}
   Setup Fingers_1 = RigVMFunctionReferenceNode(Negative Side=true, Bones=@RerouteNode_60.Array, Execute=@"Setup Fingers".Execute)
   RigVMVariableNode(
    ExecuteContext=()
    Value=&RigUnit_GetControlBool(Control=spine_fk_ik_switch) {template="GetControlValue::Execute(in Control,out BoolValue)"}.BoolValue
   ) {variable=Spine IK Mode}
   RerouteNode_3 = RigVMDispatchNode(Values=((Type=Null,Name="thigh_l_fk_ctrl_space"),((),Type=Null,Name="thigh_r_fk_ctrl_space"))) {template="DISPATCH_RigVMDispatch_ArrayMake(in Values,out Array)"}
   n13 = RigVMDispatchNode(Array=@RerouteNode_3.Array, ExecuteContext=@n1.C) {template="DISPATCH_RigVMDispatch_ArrayIterator(in Array,out Element,out Index,out Count,out Ratio)"}
   n14 = &RigUnit_ProjectTransformToNewParent(
    NewParent=(Type=Control,Name="hips_ctrl")
    OldParent=(Type=Control,Name="hips_ctrl")
    bChildInitial=True
    bOldParentInitial=True
    Child=@n13.Element
   )
   > &RigUnit_SetTransform(Weight=1, bInitial=False, bPropagateToChildren=true, Item=@n13.Element, Value=@n14.Transform)
   RerouteNode_1 = RigVMDispatchNode(Values=((Type=Control,Name="thigh_l_fk_ctrl"),((),Type=Control,Name="calf_l_fk_ctrl"),((),Type=Control,Name="foot_l_fk_ctrl"))) {template="DISPATCH_RigVMDispatch_ArrayMake(in Values,out Array)"}
   RerouteNode_10 = RigVMDispatchNode(Values=((Type=Bone,Name="thigh_l"),((),Type=Bone,Name="calf_l"),((),Type=Bone,Name="foot_l"))) {template="DISPATCH_RigVMDispatch_ArrayMake(in Values,out Array)"}
   > Backward Leg = RigVMFunctionReferenceNode(
    Ik Control=(Type=Control,Name="foot_l_ik_ctrl")
    UpV Control=(Type=Control,Name="leg_l_pv_ik_ctrl")
    FK Controls=@RerouteNode_1.Array
    Bones=@RerouteNode_10.Array
   )
   RerouteNode_6 = RigVMDispatchNode(Values=((Type=Control,Name="thigh_r_fk_ctrl"),((),Type=Control,Name="calf_r_fk_ctrl"),((),Type=Control,Name="foot_r_fk_ctrl"))) {template="DISPATCH_RigVMDispatch_ArrayMake(in Values,out Array)"}
   RerouteNode_8 = RigVMDispatchNode(Values=((Type=Bone,Name="thigh_r"),((),Type=Bone,Name="calf_r"),((),Type=Bone,Name="foot_r"))) {template="DISPATCH_RigVMDispatch_ArrayMake(in Values,out Array)"}
   > Backward Leg_1 = RigVMFunctionReferenceNode(
    Ik Control=(Type=Control,Name="foot_r_ik_ctrl")
    UpV Control=(Type=Control,Name="leg_r_pv_ik_ctrl")
    FK Controls=@RerouteNode_6.Array
    Bones=@RerouteNode_8.Array
   )
   RerouteNode_11_1_2 = RigVMDispatchNode(Values=((Type=Control,Name="hips_ctrl"),((),Type=Control,Name="hips_tan_ctrl"),((),Type=Control,Name="chest_tan_ctrl"),((),Type=Control,Name="chest_ctrl"))) {template="DISPATCH_RigVMDispatch_ArrayMake(in Values,out Array)"}
   RerouteNode_52_1 = RigVMDispatchNode(Values=((Type=Bone,Name="spine_01"),((),Type=Bone,Name="spine_02"),((),Type=Bone,Name="spine_03"),((),Type=Bone,Name="spine_04"))) {template="DISPATCH_RigVMDispatch_ArrayMake(in Values,out Array)"}
   RerouteNode_53_1 = RigVMDispatchNode(Values=((Type=Control,Name="hips_ctrl"),((),Type=Control,Name="spine_01_ctrl"),((),Type=Control,Name="spine_02_ctrl"),((),Type=Control,Name="spine_03_ctrl"))) {template="DISPATCH_RigVMDispatch_ArrayMake(in Values,out Array)"}
   Backward Spine = RigVMFunctionReferenceNode(
    IK Controls=@RerouteNode_11_1_2.Array
    Bones=@RerouteNode_52_1.Array
    FK Controls=@RerouteNode_53_1.Array
    ExecuteContext=@n17.ExecutePin
    Offsets=RigVMVariableNode() {variable=Spine Offsets}.Value
   )
   n15 = &RigUnit_GetTransform(Item=(Type=Bone,Name="root"), bInitial=false)
   n16 = &RigUnit_ProjectTransformToNewParent(
    Child=(Type=Control,Name="body_ctrl")
    NewParent=(Type=Bone,Name="spine_01")
    OldParent=(Type=Bone,Name="spine_01")
    bChildInitial=True
    bOldParentInitial=True
   )
   &RigUnit_SetTransform(Item=(Type=Control,Name="root_ctrl"), Weight=1, bInitial=False, Value=@n15.Transform, ExecutePin=@n1.A)
   > &RigUnit_SetTransform(Item=(Type=Control,Name="body_offset_ctrl"), Weight=1, bInitial=False, bPropagateToChildren=true, Value=@n15.Transform)
   > n17 = &RigUnit_SetTransform(Item=(Type=Control,Name="body_ctrl"), Weight=1, bInitial=False, Value=@n16.Transform)
  ]}
]
hierarchy [
 root {loc=<0 0 0>, rot=<0 0 0 1>}
 pelvis {parent=root, loc=<0 -0.9885 86.1786>, rot=<0 0.7071 0 -0.7071>}
 spine_01 {parent=pelvis, loc=<5.3147 0.474 0>, rot=<0 0 -0.0011 -1>}
 spine_02 {parent=spine_01, loc=<7.7324 -0.0759 0>, rot=<0 0 0.0121 -0.9999>}
 spine_03 {parent=spine_02, loc=<7.1151 -0.1704 0>, rot=<0 0 0.035 -0.9994>}
 spine_04 {parent=spine_03, loc=<7.911 -0.1939 0>, rot=<0 0 -0.0005 -1>}
 spine_05 {parent=spine_04, loc=<10.1153 0.079 0>, rot=<0 0 -0.0055 -1>}
 neck_01 {parent=spine_05, loc=<13.9426 -0.2264 0>, rot=<0 0 -0.1691 -0.9856>}
 neck_02 {parent=neck_01, loc=<4.6914 0.3829 0>, rot=<0 0 0.0424 -0.9991>}
 head {parent=neck_02, loc=<6.5905 0.484 0>, rot=<0 0 0.0987 -0.9951>}
 Jaw {parent=head, loc=<2.7572 -0.5862 0>, rot=<0.0082 -0.7071 -0.0082 -0.7071>}
 l_brow_inner {parent=head, loc=<5.3622 8.2744 -1.3658>, rot=<0.0082 -0.7071 -0.0082 -0.7071>}
 l_brow_mid_a {parent=head, loc=<5.6645 7.8866 -2.6785>, rot=<0.0082 -0.7071 -0.0082 -0.7071>}
 l_brow_mid_b {parent=head, loc=<5.4943 6.9734 -4.0039>, rot=<0.0082 -0.7071 -0.0082 -0.7071>}
 l_brow_outer {parent=head, loc=<4.9434 5.8827 -4.7441>, rot=<0.0082 -0.7071 -0.0082 -0.7071>}
 brow_mid {parent=head, loc=<5.4506 8.5738 0.1846>, rot=<0.0082 -0.7071 -0.0082 -0.7071>}
 l_nostril {parent=head, loc=<0.538 8.7003 -0.8582>, rot=<0.0082 -0.7071 -0.0082 -0.7071>}
 l_nose {parent=head, loc=<0.6504 8.3098 -1.6992>, rot=<0.0082 -0.7071 -0.0082 -0.7071>}
 l_naso_upper {parent=head, loc=<2.1876 7.7056 -1.4304>, rot=<0.0082 -0.7071 -0.0082 -0.7071>}
 l_naso_mid {parent=head, loc=<0.8323 6.9414 -2.6324>, rot=<0.0082 -0.7071 -0.0082 -0.7071>}
 nosebridge {parent=head, loc=<4.146 8.7255 -0.0051>, rot=<0.0082 -0.7071 -0.0082 -0.7071>}
 Nose {parent=head, loc=<0.8985 9.4805 0>, rot=<0.0082 -0.7071 -0.0082 -0.7071>}
 l_ear {parent=head, loc=<1.9022 0.054 -5.772>, rot=<0.0082 -0.7071 -0.0082 -0.7071>}
 l_squint_mid_b {parent=head, loc=<2.4771 6.6715 -3.4906>, rot=<0.0082 -0.7071 -0.0082 -0.7071>}
 l_squint_outer {parent=head, loc=<3.3388 5.7576 -4.5892>, rot=<0.0082 -0.7071 -0.0082 -0.7071>}
 l_squint_mid_a {parent=head, loc=<2.9144 7.1169 -1.812>, rot=<0.0082 -0.7071 -0.0082 -0.7071>}
 l_forehead_outer {parent=head, loc=<7.8016 6.7955 -3.3107>, rot=<0.0082 -0.7071 -0.0082 -0.7071>}
 l_forehead_inner {parent=head, loc=<7.6873 7.4321 -0.7914>, rot=<0.0082 -0.7071 -0.0082 -0.7071>}
 l_naso_lower {parent=head, loc=<-0.7022 6.0103 -3.4425>, rot=<0.0082 -0.7071 -0.0082 -0.7071>}
 l_lip_corner_outer {parent=head, loc=<-2.139 5.3057 -3.9779>, rot=<0.0082 -0.7071 -0.0082 -0.7071>}
 l_cheek {parent=head, loc=<-0.9641 3.6064 -4.8529>, rot=<0.0082 -0.7071 -0.0082 -0.7071>}
 l_cheekbone_lower {parent=head, loc=<1.375 5.5983 -4.627>, rot=<0.0082 -0.7071 -0.0082 -0.7071>}
 l_nose_wrinkle {parent=head, loc=<1.8761 8.8925 -1.2608>, rot=<0.0082 -0.7071 -0.0082 -0.7071>}
 chin_mid {parent=head, loc=<-5.8719 8.76 0>, rot=<0.0082 -0.7071 -0.0082 -0.7071>}
 chin_lower {parent=head, loc=<-6.8547 7.5583 0>, rot=<0.0082 -0.7071 -0.0082 -0.7071>}
 chin_upper {parent=head, loc=<-4.527 8.4555 0>, rot=<0.0082 -0.7071 -0.0082 -0.7071>}
 lip_lower_mid {parent=head, loc=<-2.2205 8.5933 0>, rot=<0.0082 -0.7071 -0.0082 -0.7071>}
 l_lip_lower_inner {parent=head, loc=<-2.3301 7.9013 -1.2771>, rot=<0.0082 -0.7071 -0.0082 -0.7071>}
 l_lip_lower_outer {parent=head, loc=<-2.204 6.817 -2.3122>, rot=<0.0082 -0.7071 -0.0082 -0.7071>}
 l_lip_corner {parent=head, loc=<-2.203 5.9961 -2.9991>, rot=<0.0082 -0.7071 -0.0082 -0.7071>}
 l_lip_upper_outer {parent=head, loc=<-1.5528 6.7704 -2.3016>, rot=<0.0082 -0.7071 -0.0082 -0.7071>}
 l_lip_upper_inner {parent=head, loc=<-1.4785 8.036 -1.3095>, rot=<0.0082 -0.7071 -0.0082 -0.7071>}
 lip_upper_mid {parent=head, loc=<-1.4946 8.5437 0>, rot=<0.0082 -0.7071 -0.0082 -0.7071>}
 l_philtrum {parent=head, loc=<-0.2636 8.3223 -0.9648>, rot=<0.0082 -0.7071 -0.0082 -0.7071>}
 l_lip_upper_above {parent=head, loc=<-0.4586 7.2489 -2.1413>, rot=<0.0082 -0.7071 -0.0082 -0.7071>}
 l_naso_frown {parent=head, loc=<-4.4164 6.4833 -2.7996>, rot=<0.0082 -0.7071 -0.0082 -0.7071>}
 l_jaw {parent=head, loc=<-3.2994 1.835 -4.8708>, rot=<0.0082 -0.7071 -0.0082 -0.7071>}
 l_chin {parent=head, loc=<-5.0311 7.9204 -1.6297>, rot=<0.0082 -0.7071 -0.0082 -0.7071>}
 chin_under {parent=head, loc=<-6.8237 4.1979 0>, rot=<0.0082 -0.7071 -0.0082 -0.7071>}
 lip_lower_below_mid {parent=head, loc=<-3.4628 8.3288 0>, rot=<0.0082 -0.7071 -0.0082 -0.7071>}
 l_lip_lower_below {parent=head, loc=<-3.5217 7.7849 -1.8249>, rot=<0.0082 -0.7071 -0.0082 -0.7071>}
 l_squint_inner {parent=head, loc=<3.8269 7.6165 -1.1266>, rot=<0.0082 -0.7071 -0.0082 -0.7071>}
 l_cheekbone_upper {parent=head, loc=<2.7228 3.3178 -5.1801>, rot=<0.0082 -0.7071 -0.0082 -0.7071>}
 l_temple {parent=head, loc=<5.6469 3.4203 -5.2502>, rot=<0.0082 -0.7071 -0.0082 -0.7071>}
 r_temple {parent=head, loc=<5.6473 3.4203 5.2502>, rot=<0.0082 -0.7071 -0.0082 -0.7071>}
 r_cheekbone_upper {parent=head, loc=<2.7228 3.3177 5.1801>, rot=<0.0082 -0.7071 -0.0082 -0.7071>}
 r_squint_inner {parent=head, loc=<3.827 7.6165 1.1266>, rot=<0.0082 -0.7071 -0.0082 -0.7071>}
 r_lip_lower_below {parent=head, loc=<-3.5218 7.7849 1.8249>, rot=<0.0082 -0.7071 -0.0082 -0.7071>}
 r_chin {parent=head, loc=<-5.0314 7.9204 1.6297>, rot=<0.0082 -0.7071 -0.0082 -0.7071>}
 r_jaw {parent=head, loc=<-3.2996 1.835 4.8708>, rot=<0.0082 -0.7071 -0.0082 -0.7071>}
 r_naso_frown {parent=head, loc=<-4.4161 6.4833 2.7996>, rot=<0.0082 -0.7071 -0.0082 -0.7071>}
 r_lip_upper_above {parent=head, loc=<-0.4587 7.2489 2.1413>, rot=<0.0082 -0.7071 -0.0082 -0.7071>}
 r_philtrum {parent=head, loc=<-0.2634 8.3223 0.9648>, rot=<0.0082 -0.7071 -0.0082 -0.7071>}
 r_lip_upper_inner {parent=head, loc=<-1.4781 8.036 1.3095>, rot=<0.0082 -0.7071 -0.0082 -0.7071>}
 r_lip_upper_outer {parent=head, loc=<-1.5529 6.7704 2.3016>, rot=<0.0082 -0.7071 -0.0082 -0.7071>}
 r_lip_corner {parent=head, loc=<-2.2032 5.9961 2.9991>, rot=<0.0082 -0.7071 -0.0082 -0.7071>}
 r_lip_lower_outer {parent=head, loc=<-2.2042 6.817 2.3122>, rot=<0.0082 -0.7071 -0.0082 -0.7071>}
 r_lip_lower_inner {parent=head, loc=<-2.3302 7.9013 1.2771>, rot=<0.0082 -0.7071 -0.0082 -0.7071>}
 r_nose_wrinkle {parent=head, loc=<1.8761 8.8925 1.2608>, rot=<0.0082 -0.7071 -0.0082 -0.7071>}
 r_cheekbone_lower {parent=head, loc=<1.3749 5.5983 4.627>, rot=<0.0082 -0.7071 -0.0082 -0.7071>}
 r_cheek {parent=head, loc=<-0.9638 3.6064 4.8529>, rot=<0.0082 -0.7071 -0.0082 -0.7071>}
 r_lip_corner_outer {parent=head, loc=<-2.1393 5.3057 3.9779>, rot=<0.0082 -0.7071 -0.0082 -0.7071>}
 r_naso_lower {parent=head, loc=<-0.7022 6.0103 3.4425>, rot=<0.0082 -0.7071 -0.0082 -0.7071>}
 r_forehead_inner {parent=head, loc=<7.6873 7.4321 0.7914>, rot=<0.0082 -0.7071 -0.0082 -0.7071>}
 r_forehead_outer {parent=head, loc=<7.802 6.7955 3.3107>, rot=<0.0082 -0.7071 -0.0082 -0.7071>}
 r_squint_mid_a {parent=head, loc=<2.9143 7.1169 1.812>, rot=<0.0082 -0.7071 -0.0082 -0.7071>}
 r_squint_outer {parent=head, loc=<3.3387 5.7576 4.5892>, rot=<0.0082 -0.7071 -0.0082 -0.7071>}
 r_squint_mid_b {parent=head, loc=<2.4774 6.6716 3.4906>, rot=<0.0082 -0.7071 -0.0082 -0.7071>}
 r_ear {parent=head, loc=<1.9019 0.054 5.772>, rot=<0.0082 -0.7071 -0.0082 -0.7071>}
 r_naso_mid {parent=head, loc=<0.8328 6.9414 2.6324>, rot=<0.0082 -0.7071 -0.0082 -0.7071>}
 r_naso_upper {parent=head, loc=<2.1875 7.7056 1.4304>, rot=<0.0082 -0.7071 -0.0082 -0.7071>}
 r_nose {parent=head, loc=<0.6502 8.3098 1.6992>, rot=<0.0082 -0.7071 -0.0082 -0.7071>}
 r_nostril {parent=head, loc=<0.5381 8.7003 0.8582>, rot=<0.0082 -0.7071 -0.0082 -0.7071>}
 r_brow_outer {parent=head, loc=<4.9433 5.8827 4.7441>, rot=<0.0082 -0.7071 -0.0082 -0.7071>}
 r_brow_mid_b {parent=head, loc=<5.4943 6.9735 4.0039>, rot=<0.0082 -0.7071 -0.0082 -0.7071>}
 r_brow_mid_a {parent=head, loc=<5.6643 7.8866 2.6785>, rot=<0.0082 -0.7071 -0.0082 -0.7071>}
 r_brow_inner {parent=head, loc=<5.3622 8.2744 1.3658>, rot=<0.0082 -0.7071 -0.0082 -0.7071>}
 l_lid_upper_mid_pivot {parent=head, loc=<4.5041 5.4698 -2.9459>, rot=<0.0082 -0.7071 -0.0082 -0.7071>}
 l_lid_upper_mid {parent=l_lid_upper_mid_pivot, loc=<0 1.7137 0.3395>, rot=<0 0 0 1>}
 l_lid_upper_inner_pivot {parent=head, loc=<4.5039 5.4698 -2.6261>, rot=<0.0082 -0.7071 -0.0082 -0.7071>}
 l_lid_upper_inner {parent=l_lid_upper_inner_pivot, loc=<-0.7994 1.6829 0.0594>, rot=<0 0 0 1>}
 l_lid_upper_outer_pivot {parent=head, loc=<4.5039 5.4698 -3.3097>, rot=<0.0082 -0.7071 -0.0082 -0.7071>}
 l_lid_upper_outer {parent=l_lid_upper_outer_pivot, loc=<0.6783 1.3652 0.0152>, rot=<0 0 0 1>}
 l_lid_lower_inner_pivot {parent=head, loc=<4.2002 5.4628 -2.6261>, rot=<0.0082 -0.7071 -0.0082 -0.7071>}
 l_lid_lower_inner {parent=l_lid_lower_inner_pivot, loc=<-0.7212 1.6233 -0.2995>, rot=<0 0 0 1>}
 l_lid_lower_outer_pivot {parent=head, loc=<4.2002 5.4628 -3.3097>, rot=<0.0082 -0.7071 -0.0082 -0.7071>}
 l_lid_lower_outer {parent=l_lid_lower_outer_pivot, loc=<0.5549 1.2153 -0.3146>, rot=<0 0 0 1>}
 l_lid_lower_mid_pivot {parent=head, loc=<4.2004 5.4628 -2.9459>, rot=<0.0082 -0.7071 -0.0082 -0.7071>}
 l_lid_lower_mid {parent=l_lid_lower_mid_pivot, loc=<0.0133 1.6234 -0.4365>, rot=<0 0 0 1>}
 r_lid_upper_mid_pivot {parent=head, loc=<4.5037 5.4698 2.9459>, rot=<0.7071 0.0082 -0.7071 0.0082>}
 r_lid_upper_mid {parent=r_lid_upper_mid_pivot, loc=<0 -1.7137 -0.34>, rot=<0 0 0 1>}
 r_lid_upper_inner_pivot {parent=head, loc=<4.5037 5.4698 2.6261>, rot=<0.7071 0.0082 -0.7071 0.0082>}
 r_lid_upper_inner {parent=r_lid_upper_inner_pivot, loc=<0.7994 -1.6829 -0.06>, rot=<0 0 0 1>}
 r_lid_upper_outer_pivot {parent=head, loc=<4.5037 5.4698 3.3097>, rot=<0.7071 0.0082 -0.7071 0.0082>}
 r_lid_upper_outer {parent=r_lid_upper_outer_pivot, loc=<-0.6783 -1.3652 -0.015>, rot=<0 0 0 1>}
 r_lid_lower_inner_pivot {parent=head, loc=<4.1998 5.4628 2.6261>, rot=<0.7071 0.0082 -0.7071 0.0082>}
 r_lid_lower_inner {parent=r_lid_lower_inner_pivot, loc=<0.7212 -1.6233 0.299>, rot=<0 0 0 1>}
 r_lid_lower_outer_pivot {parent=head, loc=<4.1998 5.4628 3.3097>, rot=<0.7071 0.0082 -0.7071 0.0082>}
 r_lid_lower_outer {parent=r_lid_lower_outer_pivot, loc=<-0.5549 -1.2153 0.314>, rot=<0 0 0 1>}
 r_lid_lower_mid_pivot {parent=head, loc=<4.2008 5.4628 2.9459>, rot=<0.7071 0.0082 -0.7071 0.0082>}
 r_lid_lower_mid {parent=r_lid_lower_mid_pivot, loc=<-0.0133 -1.6234 0.437>, rot=<0 0 0 1>}
 l_eyeball {parent=head, loc=<4.5071 5.2985 -2.6011>, rot=<0.0082 -0.7071 -0.0082 -0.7071>}
 l_eyeball_end {parent=l_eyeball, loc=<0.2677 2.5612 0>, rot=<0 0 0 1>}
 r_eyeball {parent=head, loc=<4.5066 5.2985 2.6011>, rot=<0.0082 -0.7071 -0.0082 -0.7071>}
 r_eyeball_end {parent=r_eyeball, loc=<-0.2677 2.5612 0>, rot=<0 0 0 1>}
 lower_teeth {parent=head, loc=<-4.1646 7.8348 0>, rot=<0.0082 -0.7071 -0.0082 -0.7071>}
 upper_teeth {parent=head, loc=<-2.3203 7.8752 0>, rot=<0.0082 -0.7071 -0.0082 -0.7071>}
 VB head_fx {parent=head, loc=<0 0 0>, rot=<0 0 0 1>}
 clavicle_l {parent=spine_05, loc=<7.3758 -1.1305 -3.2823>, rot=<0.0179 -0.7069 -0.0179 -0.7069>}
 upperarm_l {parent=clavicle_l, loc=<10.6837 -0.4418 -3.5086>, rot=<0.0196 -0.5195 0.0356 -0.8535>}
 lowerarm_l {parent=upperarm_l, loc=<27.2431 1.4821 0.4211>, rot=<0.0042 0.0119 -0.2147 -0.9766>}
 hand_l {parent=lowerarm_l, loc=<26.9518 1.0011 0.7872>, rot=<-0.4993 -0.0461 0.0438 -0.8641>}
 index_metacarpal_l {parent=hand_l, loc=<2.9443 -0.6871 -1.4235>, rot=<0.0818 -0.0789 0.0084 -0.9935>}
 index_01_l {parent=index_metacarpal_l, loc=<7.7267 0.2373 -0.2513>, rot=<-0.0295 0.0652 0.0555 -0.9959>}
 index_02_l {parent=index_01_l, loc=<3.4049 0.1733 -0.0703>, rot=<0.0007 -0.007 0.095 -0.9955>}
 index_03_l {parent=index_02_l, loc=<2.6931 -0.0326 0.123>, rot=<-0.0004 0.007 0.0319 -0.9995>}
 middle_metacarpal_l {parent=hand_l, loc=<3.0581 -0.4829 -0.0672>, rot=<0.0393 0.0025 -0.0009 -0.9992>}
 middle_01_l {parent=middle_metacarpal_l, loc=<8.4777 0.2839 -0.3479>, rot=<-0.06 0.0075 0.1009 -0.9931>}
 middle_02_l {parent=middle_01_l, loc=<3.1302 0.0764 0.0245>, rot=<-0.0116 0.0197 0.0067 -0.9997>}
 middle_03_l {parent=middle_02_l, loc=<3.0228 -0.0927 -0.0102>, rot=<0.0091 -0.0067 0.0536 -0.9985>}
 pinky_metacarpal_l {parent=hand_l, loc=<3.1002 -0.584 2.2451>, rot=<-0.0734 0.0993 -0.0118 -0.9923>}
 pinky_01_l {parent=pinky_metacarpal_l, loc=<7.2261 -0.002 0.21>, rot=<0.0959 -0.041 0.098 -0.9897>}
 pinky_02_l {parent=pinky_01_l, loc=<2.7394 -0.0544 -0.0695>, rot=<0.0039 0.0002 0.0421 -0.9991>}
 pinky_03_l {parent=pinky_02_l, loc=<1.5966 -0.2015 0.0114>, rot=<0.0012 0.0098 0.0059 -0.9999>}
 ring_metacarpal_l {parent=hand_l, loc=<3.0732 -0.4596 1.2011>, rot=<-0.0082 0.0557 -0.0051 -0.9984>}
 ring_01_l {parent=ring_metacarpal_l, loc=<8.1739 0.1202 -0.3808>, rot=<-0.0136 0.0015 0.133 -0.991>}
 ring_02_l {parent=ring_01_l, loc=<2.9099 0.0624 0.0569>, rot=<0 0 0 -1>}
 ring_03_l {parent=ring_02_l, loc=<2.6446 -0.1326 -0.0569>, rot=<0 0 0.024 -0.9997>}
 thumb_01_l {parent=hand_l, loc=<2.7726 -1.6981 -1.7892>, rot=<0.666 -0.3553 -0.1285 -0.6432>}
 thumb_02_l {parent=thumb_01_l, loc=<4.5942 0.1039 -0.0993>, rot=<-0.0057 -0.0455 0.1247 -0.9911>}
 thumb_03_l {parent=thumb_02_l, loc=<3.8743 -0.6419 -0.0603>, rot=<0.0181 -0.0562 -0.0281 -0.9979>}
 weapon_l {parent=hand_l, loc=<0.4458 -3.5863 0.3694>, rot=<0 0 0 -1>}
 hand_attach_l {parent=hand_l, loc=<8.5 -3 0>, rot=<0.7071 0.7071 0 0>}
 lowerarm_twist_02_l {parent=lowerarm_l, loc=<11.5508 0.429 0.3374>, rot=<0 0 0 -1>}
 lowerarm_twist_01_l {parent=lowerarm_l, loc=<20.2138 0.7508 0.5904>, rot=<0 0 0 -1>}
 lowerarm_twist_l {parent=lowerarm_l, loc=<26.946 0.9829 0.7898>, rot=<-0.4993 -0.0461 0.0438 -0.8641>}
 upperarm_twist_01_l {parent=upperarm_l, loc=<1 0 0>, rot=<0 0 0 -1>}
 upperarm_twist_02_l {parent=upperarm_l, loc=<13.7941 0 0>, rot=<0 0 0 1>}
 upperarm_twist_l {parent=upperarm_l, loc=<0 0 0>, rot=<0 0 0 -1>}
 clavicle_r {parent=spine_05, loc=<7.3759 -1.1305 3.2823>, rot=<0.7069 0.0179 -0.7069 0.0179>}
 upperarm_r {parent=clavicle_r, loc=<-10.6837 0.4418 3.5085>, rot=<0.0196 -0.5195 0.0356 -0.8535>}
 lowerarm_r {parent=upperarm_r, loc=<-27.2432 -1.4821 -0.4211>, rot=<0.0042 0.0119 -0.2147 -0.9766>}
 hand_r {parent=lowerarm_r, loc=<-26.9519 -1.0011 -0.7873>, rot=<-0.4993 -0.0461 0.0438 -0.8641>}
 index_metacarpal_r {parent=hand_r, loc=<-2.9441 0.6873 1.4236>, rot=<0.0818 -0.0789 0.0084 -0.9935>}
 index_01_r {parent=index_metacarpal_r, loc=<-7.7267 -0.2373 0.2513>, rot=<-0.0295 0.0652 0.0555 -0.9959>}
 index_02_r {parent=index_01_r, loc=<-3.4051 -0.1734 0.0702>, rot=<0.0007 -0.007 0.095 -0.9955>}
 index_03_r {parent=index_02_r, loc=<-2.6931 0.0325 -0.123>, rot=<-0.0004 0.007 0.0319 -0.9995>}
 middle_metacarpal_r {parent=hand_r, loc=<-3.0581 0.483 0.0672>, rot=<0.0393 0.0025 -0.0009 -0.9992>}
 middle_01_r {parent=middle_metacarpal_r, loc=<-8.4775 -0.284 0.348>, rot=<-0.06 0.0075 0.1009 -0.9931>}
 middle_02_r {parent=middle_01_r, loc=<-3.1304 -0.0763 -0.0245>, rot=<-0.0116 0.0197 0.0067 -0.9997>}
 middle_03_r {parent=middle_02_r, loc=<-3.0228 0.0927 0.0102>, rot=<0.0091 -0.0067 0.0536 -0.9985>}
 pinky_metacarpal_r {parent=hand_r, loc=<-3.1004 0.5841 -2.2452>, rot=<-0.0734 0.0993 -0.0118 -0.9923>}
 pinky_01_r {parent=pinky_metacarpal_r, loc=<-7.2259 0.0022 -0.21>, rot=<0.0959 -0.041 0.098 -0.9897>}
 pinky_02_r {parent=pinky_01_r, loc=<-2.7394 0.0544 0.0695>, rot=<0.0039 0.0002 0.0421 -0.9991>}
 pinky_03_r {parent=pinky_02_r, loc=<-1.5964 0.2012 -0.0114>, rot=<0.0012 0.0098 0.0059 -0.9999>}
 ring_metacarpal_r {parent=hand_r, loc=<-3.0731 0.4597 -1.2011>, rot=<-0.0082 0.0557 -0.0051 -0.9984>}
 ring_01_r {parent=ring_metacarpal_r, loc=<-8.1741 -0.1202 0.3808>, rot=<-0.0136 0.0015 0.133 -0.991>}
 ring_02_r {parent=ring_01_r, loc=<-2.9096 -0.0624 -0.0569>, rot=<0 0 0 -1>}
 ring_03_r {parent=ring_02_r, loc=<-2.6447 0.1325 0.0569>, rot=<0 0 0.024 -0.9997>}
 thumb_01_r {parent=hand_r, loc=<-2.7727 1.6982 1.7892>, rot=<0.666 -0.3553 -0.1285 -0.6432>}
 thumb_02_r {parent=thumb_01_r, loc=<-4.5941 -0.104 0.0993>, rot=<-0.0057 -0.0455 0.1247 -0.9911>}
 thumb_03_r {parent=thumb_02_r, loc=<-3.8743 0.6419 0.0603>, rot=<0.0181 -0.0562 -0.0281 -0.9979>}
 weapon_r {parent=hand_r, loc=<-0.4457 3.5864 -0.3694>, rot=<0 0 0 -1>}
 hand_attach_r {parent=hand_r, loc=<-8.5 3 0>, rot=<0 0 -0.7071 -0.7071>}
 lowerarm_twist_02_r {parent=lowerarm_r, loc=<-11.5508 -0.429 -0.3374>, rot=<0 0 0 -1>}
 lowerarm_twist_01_r {parent=lowerarm_r, loc=<-20.2139 -0.7508 -0.5904>, rot=<0 0 0 -1>}
 dyn_hose01 {parent=lowerarm_twist_01_r, loc=<0 0 0>, rot=<0 0 0 1>}
 lowerarm_twist_r {parent=lowerarm_r, loc=<-26.9461 -0.983 -0.7899>, rot=<-0.4993 -0.0461 0.0438 -0.8641>}
 upperarm_twist_01_r {parent=upperarm_r, loc=<-1 0 0>, rot=<0 0 0 -1>}
 mech_arm_root {parent=upperarm_twist_01_r, loc=<-5.3638 0.8311 -15.0265>, rot=<0.7836 -0.0189 0.6208 -0.015>}
 mech_arm_upperarm {parent=mech_arm_root, loc=<-12.4844 -1.6674 -1.2368>, rot=<-0.6111 0.0041 -0.7909 0.0316>}
 mech_arm_lowerArm {parent=mech_arm_upperarm, loc=<-29.0827 -1.2816 0.0087>, rot=<0.0129 0.0027 0.1144 0.9933>}
 mech_arm_claw_root {parent=mech_arm_lowerArm, loc=<-20.8105 4.7837 -5.722>, rot=<0.1152 -0.6864 0.7002 -0.1587>}
 mech_arm_claw_spin {parent=mech_arm_claw_root, loc=<0.0096 2.6231 0.1286>, rot=<-0.0245 0.0134 -0.0209 0.9994>}
 mech_arm_claw_base {parent=mech_arm_claw_spin, loc=<0.2755 3.2046 -0.1306>, rot=<0.5319 -0.4815 -0.4893 0.4958>}
 mech_arm_clawA {parent=mech_arm_claw_base, loc=<0.1056 -9.7661 -0.0952>, rot=<-0.9999 -0.016 0 0>}
 mech_arm_clawB {parent=mech_arm_claw_base, loc=<-0.4809 9.7861 0>, rot=<0 0 0.014 0.9999>}
 mech_arm_pistonBottom {parent=mech_arm_lowerArm, loc=<14.5396 13.0442 0.0895>, rot=<0.6536 -0.2514 0.6743 0.2344>}
 mech_arm_pistonTop {parent=mech_arm_upperarm, loc=<-0.852 6.7365 0.2039>, rot=<-0.1369 -0.6921 -0.1597 0.6904>}
 mech_arm_upArmTwist {parent=mech_arm_upperarm, loc=<0 0 0>, rot=<0 0 0 1>}
 upperarm_twist_02_r {parent=upperarm_r, loc=<-13.7942 0 0>, rot=<0 0 0 1>}
 upperarm_twist_r {parent=upperarm_r, loc=<0 0 0>, rot=<0 0 0 -1>}
 Backpack_Root {parent=spine_05, loc=<1.335 -16.6059 0>, rot=<0.023 -0.7067 -0.023 -0.7067>}
 backpack_base {parent=Backpack_Root, loc=<0 -2.321 -12.4453>, rot=<0.0115 0 0 -0.9999>}
 backpack_base_jiggle {parent=backpack_base, loc=<0 0 0>, rot=<0 0 0 1>}
 bedroll_l {parent=backpack_base_jiggle, loc=<8.1598 4.2534 -8.1691>, rot=<0 0 0 1>}
 bedroll_l_jiggle {parent=bedroll_l, loc=<5.6597 0.0315 -4.0128>, rot=<0 0 0 1>}
 bedroll_r {parent=backpack_base_jiggle, loc=<-8.7467 4.251 -7.8692>, rot=<1 0 0 0>}
 bedroll_r_jiggle {parent=bedroll_r, loc=<-3.8493 -0.0303 3.8573>, rot=<0 0 0 1>}
 backpack_bullets_01 {parent=Backpack_Root, loc=<13.2715 -7.379 9.0503>, rot=<-0.3216 -0.2232 -0.6805 -0.6194>}
 backpack_bullets_02 {parent=backpack_bullets_01, loc=<0 -5.6641 0>, rot=<-0.1384 -0.2776 -0.1909 -0.9313>}
 backpack_bullets_03 {parent=backpack_bullets_02, loc=<0 -3.862 0>, rot=<-0.0015 0.016 -0.0913 -0.9957>}
 backpak_handle {parent=Backpack_Root, loc=<0 4.6451 13.4752>, rot=<0 0 0 1>}
 backpack_handle_jiggle {parent=backpak_handle, loc=<0 0.086 5.6563>, rot=<0 0 0 1>}
 strap_l {parent=Backpack_Root, loc=<9.7494 21.621 3.4509>, rot=<0 0 0 1>}
 strap_r {parent=Backpack_Root, loc=<-9.7494 21.6209 3.4511>, rot=<1 0 0 0>}
 VB spine_05_weapon_r {parent=spine_05, loc=<-44.0344 3.3383 37.3375>, rot=<0.8703 -0.2026 -0.0959 -0.4386>}
 VB VB spine_05_weapon_r_ik_hand_gun {parent=VB spine_05_weapon_r, loc=<0.4457 -3.5864 0.3695>, rot=<0 0 0 -1>}
 VB VB spine_05_weapon_r_ik_hand_l {parent=VB spine_05_weapon_r, loc=<27.7128 53.7427 47.597>, rot=<0.3446 0.7245 0.5969 0>}
 VB spine_05_upperarm_r {parent=spine_05, loc=<3.8941 -1.7489 13.9659>, rot=<-0.9696 -0.0635 0.2364 0.0053>}
 VB VB spine_05_upperarm_r_lowerarm_r {parent=VB spine_05_upperarm_r, loc=<-27.2431 -1.4821 -0.4212>, rot=<0.0042 0.0119 -0.2147 -0.9766>}
 VB VB VB spine_05_upperarm_r_lowerarm_r_hand_r {parent=VB VB spine_05_upperarm_r_lowerarm_r, loc=<-26.9518 -1.0011 -0.7874>, rot=<-0.4993 -0.0461 0.0438 -0.8641>}
 cloth_back_01 {parent=spine_02, loc=<0.3988 -11.6844 0.046>, rot=<0.0282 -0.7065 -0.0282 -0.7065>}
 cloth_back_02 {parent=cloth_back_01, loc=<0 -0.0893 -10.5643>, rot=<0 0 0 1>}
 cloth_back_03 {parent=cloth_back_02, loc=<0 -1.1168 -10.8365>, rot=<0 0 0 1>}
 thigh_r {parent=pelvis, loc=<-1.5741 -0.2938 8.7856>, rot=<0.0004 -0.026 0.9996 0.0136>}
 calf_r {parent=thigh_r, loc=<38.4137 0.6432 -2.1504>, rot=<0.0002 -0.0066 -0.0379 -0.9993>}
 foot_r {parent=calf_r, loc=<35.4504 0.3763 0.7575>, rot=<-0.0261 0.0071 0.0242 -0.9993>}
 ball_r {parent=foot_r, loc=<7.9776 -12.8059 0.8901>, rot=<-0.0008 -0.0089 -0.7161 -0.698>}
 calf_twist_02_r {parent=calf_r, loc=<17.7252 0.3763 0.7574>, rot=<0 0 0 -1>}
 calf_twist_01_r {parent=calf_r, loc=<34.4504 0.3763 0.7574>, rot=<0 0 0 -1>}
 calf_twist_r {parent=calf_r, loc=<17.7252 0 0>, rot=<0 0 0 -1>}
 lowerleg_twist_01_r {parent=calf_r, loc=<38.7753 0 0>, rot=<0 0 0 1>}
 lowerleg_twist_02_r {parent=calf_r, loc=<20.0452 0 0>, rot=<0 0 0 1>}
 thigh_twist_01_r {parent=thigh_r, loc=<1 0 0>, rot=<0 0 0 -1>}
 thigh_twist_01_rRigidBody1 {parent=thigh_twist_01_r, loc=<0 0 0>, rot=<0 0 0 1>}
 thigh_twist_01_rRigidBodyShape1Transform {parent=thigh_twist_01_rRigidBody1, loc=<2.1093 -3.9335 -2.7962>, rot=<-0.046 0.0124 -0.0081 -0.9988>}
 thigh_twist_01_rConstraint1 {parent=thigh_twist_01_rRigidBody1, loc=<0 0 0>, rot=<0 0 0 1>}
 pack {parent=thigh_twist_01_r, loc=<0.099 -3.599 5.3503>, rot=<-0.7068 -0.0171 0.7069 0.0197>}
 pack_jiggle {parent=pack, loc=<0 0 0>, rot=<0 0 0 1>}
 wrench {parent=thigh_twist_01_r, loc=<-3.3868 -9.6206 -0.0457>, rot=<-0.5571 0.0105 0.8276 -0.0671>}
 thigh_twist_r {parent=thigh_r, loc=<0 0 0>, rot=<0 0 0 -1>}
 thigh_rRigidBody1 {parent=thigh_r, loc=<0 0 0>, rot=<0 0 0 1>}
 thigh_rRigidBodyShape1Transform {parent=thigh_rRigidBody1, loc=<2.9317 -9.4629 -1.4637>, rot=<0.4095 0.0102 -0.4674 -0.7834>}
 thigh_rConstraint1 {parent=thigh_rRigidBody1, loc=<0 0 0>, rot=<0 0.7665 0.0256 -0.6418>}
 upperleg_twist_01_r {parent=thigh_r, loc=<0.3846 0 0>, rot=<0 0 0 1>}
 thigh_l {parent=pelvis, loc=<-1.5743 -0.2938 -8.7855>, rot=<0.026 0.0004 0.0136 -0.9996>}
 calf_l {parent=thigh_l, loc=<-38.4135 -0.6432 2.1503>, rot=<0.0002 -0.0066 -0.0379 -0.9993>}
 foot_l {parent=calf_l, loc=<-35.4503 -0.3763 -0.7574>, rot=<-0.0261 0.0071 0.0242 -0.9993>}
 ball_l {parent=foot_l, loc=<-7.9777 12.8059 -0.8901>, rot=<-0.0008 -0.0089 -0.7161 -0.698>}
 calf_twist_02_l {parent=calf_l, loc=<-17.7251 -0.3763 -0.7574>, rot=<0 0 0 -1>}
 calf_twist_01_l {parent=calf_l, loc=<-34.4503 -0.3763 -0.7574>, rot=<0 0 0 -1>}
 calf_twist_l {parent=calf_l, loc=<-17.7251 0 0>, rot=<0 0 0 -1>}
 lowerleg_twist_01_l {parent=calf_l, loc=<-38.7753 0 0>, rot=<0 0 0 1>}
 lowerleg_twist_02_l {parent=calf_l, loc=<-20.0451 0 0>, rot=<0 0 0 1>}
 thigh_twist_01_l {parent=thigh_l, loc=<-1 0 0>, rot=<0 0 0 -1>}
 thigh_twist_01_lRigidBody1 {parent=thigh_twist_01_l, loc=<0 0 0>, rot=<0 0 0 1>}
 thigh_twist_01_lRigidBodyShape1Transform {parent=thigh_twist_01_lRigidBody1, loc=<0.1148 -1.7829 0.9561>, rot=<0 0 0.0574 -0.9983>}
 thigh_twist_01_lConstraint1 {parent=thigh_twist_01_lRigidBody1, loc=<0 0 0>, rot=<0 -1 0 0>}
 thigh_twist_l {parent=thigh_l, loc=<0 0 0>, rot=<0 0 0 -1>}
 thigh_lRigidBody1 {parent=thigh_l, loc=<0 0 0>, rot=<0 0 0 1>}
 thigh_lRigidBodyShape1Transform {parent=thigh_lRigidBody1, loc=<-5.3539 3.327 2.7638>, rot=<0.1698 -0.0002 -0.6067 -0.7766>}
 thigh_lConstraint1 {parent=thigh_lRigidBody1, loc=<0 0 0>, rot=<0 -0.7665 0.0256 -0.6418>}
 upperleg_twist_01_l {parent=thigh_l, loc=<-0.3846 0 0>, rot=<0 0 0 1>}
 belt_back {parent=pelvis, loc=<-3.1326 -10.407 0>, rot=<-0.0013 -0.7071 0.0013 -0.7071>}
 belt_front {parent=pelvis, loc=<-2.9332 9.8423 0>, rot=<-0.0013 -0.7071 0.0013 -0.7071>}
 belt_l {parent=pelvis, loc=<4.9138 0 -11.2468>, rot=<-0.0013 -0.7071 0.0013 -0.7071>}
 pelvisRigidBody1 {parent=pelvis, loc=<0 0 0>, rot=<0 0 0 1>}
 pelvisRigidBodyShape1Transform {parent=pelvisRigidBody1, loc=<0 0 0>, rot=<0 0 0 1>, scale=<1.124 1.124 1.124>}
 belt01_r {parent=pelvis, loc=<0.4733 12.5671 3.0499>, rot=<-0.0127 -0.707 0.0127 -0.707>}
 belt02_r {parent=belt01_r, loc=<-0.217 1.0304 -5.9212>, rot=<0 0 0 1>}
 belt03_r {parent=belt02_r, loc=<-0.61 0.9064 -7.6285>, rot=<0 0 0 1>}
 belt04_r {parent=belt03_r, loc=<-0.1659 0.3748 -6.6814>, rot=<0 0 0 1>}
 belt01_l {parent=pelvis, loc=<1.4574 12.8641 -2.6501>, rot=<-0.0127 -0.707 0.0127 -0.707>}
 belt02_l {parent=belt01_l, loc=<0.9484 0.7969 -6.7385>, rot=<0 0 0 1>}
 belt03_l {parent=belt02_l, loc=<0.5025 0.863 -6.2875>, rot=<0 0 0 1>}
 attach {parent=root, loc=<0 0 0>, rot=<0 0 0 1>}
 VB attach_ik_hand_gun {parent=attach, loc=<-39.562 6.118 82.7824>, rot=<0.6645 0.1448 0.5538 -0.4804>}
 VB VB attach_ik_hand_gun_ik_hand_l {parent=VB attach_ik_hand_gun, loc=<27.2671 57.3292 47.2276>, rot=<-0.3446 -0.7246 -0.5969 0>}
 VB VB attach_ik_hand_gun_ik_hand_r {parent=VB attach_ik_hand_gun, loc=<0 0 0>, rot=<0 0 0 -1>}
 ik_foot_root {parent=root, loc=<0 0 0>, rot=<0 0 0 1>}
 ik_foot_l {parent=ik_foot_root, loc=<6.8559 -2.8834 10.7904>, rot=<0 0.7071 0 -0.7071>}
 ik_foot_r {parent=ik_foot_root, loc=<-6.8559 -2.8834 10.7902>, rot=<0.7071 0 -0.7071 0>}
 VB ik_foot_root_weapon_l {parent=ik_foot_root, loc=<37.3376 4.1936 80.6503>, rot=<0.4804 0.5538 -0.1448 0.6645>}
 ik_hand_root {parent=root, loc=<0 0 0>, rot=<0 0 0 1>}
 ik_hand_gun {parent=ik_hand_root, loc=<-39.562 6.118 82.7824>, rot=<0.6645 0.1448 0.5538 -0.4804>}
 ik_hand_l {parent=ik_hand_gun, loc=<27.267 57.3293 47.2276>, rot=<-0.3446 -0.7246 -0.5969 0>}
 ik_hand_r {parent=ik_hand_gun, loc=<0 0 0>, rot=<0 0 0 1>}
 nxRagdoll1 {parent=root, loc=<0 -5.2501 0>, rot=<0 0 0 1>}
 VB root_prop {parent=root, loc=<0 0 0>, rot=<0 0 0 1>}
 VB Root_hand_l {parent=root, loc=<39.562 6.1179 82.7825>, rot=<0.4804 0.5538 -0.1448 0.6645>}
 VB Root_hand_r {parent=root, loc=<-39.562 6.118 82.7825>, rot=<-0.6645 -0.1448 -0.5538 0.4804>}
 global_ctrl {controlType=EULER_TRANSFORM, type=Control, loc=<0 0 0>, rot=<0 0 0 1>}
 root_ctrl {controlType=EULER_TRANSFORM, parent=global_ctrl, type=Control, loc=<0 0 0>, rot=<0 0 0 1>}
 body_offset_ctrl {controlType=EULER_TRANSFORM, parent=global_ctrl, type=Control, loc=<0 0 0>, rot=<0 0 0 1>}
 leg_l_controls_space {parent=body_offset_ctrl, type=Null, loc=<0 0 0>, rot=<0 0 0 1>}
 thigh_l_fk_ctrl_space {parent=leg_l_controls_space, type=Null, loc=<7.7501 -0.81 90.5518>, rot=<-0.0019 -0.7464 0.0016 0.6655>}
 thigh_l_fk_ctrl {controlType=EULER_TRANSFORM, parent=thigh_l_fk_ctrl_space, type=Control, loc=<0 0 0>, rot=<0 0 0 1>}
 calf_l_fk_ctrl {controlType=EULER_TRANSFORM, parent=thigh_l_fk_ctrl, type=Control, loc=<0 0 0>, rot=<0 0 0 1>}
 foot_l_fk_ctrl {controlType=EULER_TRANSFORM, parent=calf_l_fk_ctrl, type=Control, loc=<0 0 0>, rot=<0 0 0 1>}
 ball_l_fk_ctrl {controlType=EULER_TRANSFORM, parent=foot_l_fk_ctrl, type=Control, loc=<0 0 0>, rot=<0 0 0 1>}
 toes_l_space {parent=leg_l_controls_space, type=Null, loc=<15.7991 9.3506 3.1406>, rot=<0.5229 0.4735 -0.4801 -0.5214>}
 foot_l_ik_ctrl {controlType=EULER_TRANSFORM, parent=leg_l_controls_space, type=Control, loc=<0 0 0>, rot=<0 0 0 1>}
 leg_l_stretch_switch {controlType=BOOL, parent=foot_l_ik_ctrl, type=Control, loc=<0 0 0>, rot=<0 0 0 1>}
 heel_l {parent=foot_l_ik_ctrl, type=Null, loc=<-2.7001 1.0746 -8.8107>, rot=<-0.4699 0.5284 -0.543 0.453>}
 tip_l {parent=heel_l, type=Null, loc=<-20.4653 -0.6281 0>, rot=<0 0 0 1>}
 ball_l_ik_ctrl {controlType=EULER_TRANSFORM, parent=tip_l, type=Control, loc=<0 0 0>, rot=<0 0 0 1>}
 foot_bk1_l {parent=tip_l, type=Null, loc=<7.5178 -2.29 1.3226>, rot=<0 0 0 1>}
 foot_bk1_l_ctrl {controlType=EULER_TRANSFORM, parent=foot_bk1_l, type=Control, loc=<0 0 0>, rot=<0 0 0 1>}
 foot_ik_l {parent=foot_bk1_l_ctrl, type=Null, loc=<13.8882 -5.8672 1.5071>, rot=<0 0 0 1>}
 foot_roll_l_bfr {parent=foot_bk1_l_ctrl, type=Null, loc=<13.8882 -5.8672 1.5071>, rot=<0 0 0 1>}
 foot_roll_l_ctrl {controlType=EULER_TRANSFORM, parent=foot_roll_l_bfr, type=Control, loc=<0 0 0>, rot=<0 0 0 1>}
 roll_blend_l_ctrl {controlType=FLOAT, parent=foot_roll_l_ctrl, type=Control, loc=<1 0 0>, rot=<0 0 0 1>}
 tip_l_ctrl {controlType=EULER_TRANSFORM, parent=heel_l, type=Control, loc=<0 0 0>, rot=<0 0 0 1>}
 heel_l_ctrl {controlType=EULER_TRANSFORM, parent=foot_l_ik_ctrl, type=Control, loc=<0 0 0>, rot=<0 0 0 1>}
 leg_l_softness {controlType=FLOAT, parent=foot_l_ik_ctrl, type=Control, loc=<0 0 0>, rot=<0 0 0 1>}
 leg_l_pv_ik_ctrl {controlType=EULER_TRANSFORM, parent=leg_l_controls_space, type=Control, loc=<0 0 0>, rot=<0 0 0 1>}
 leg_r_controls_space {parent=body_offset_ctrl, type=Null, loc=<0 0 0>, rot=<0 0 0 1>, scale=<-1 1 1>}
 thigh_r_fk_ctrl_space {parent=leg_r_controls_space, type=Null, loc=<7.7501 -0.81 90.552>, rot=<-0.0019 -0.7464 0.0016 0.6655>}
 thigh_r_fk_ctrl {controlType=EULER_TRANSFORM, parent=thigh_r_fk_ctrl_space, type=Control, loc=<0 0 0>, rot=<0 0 0 1>}
 calf_r_fk_ctrl {controlType=EULER_TRANSFORM, parent=thigh_r_fk_ctrl, type=Control, loc=<0 0 0>, rot=<0 0 0 1>}
 foot_r_fk_ctrl {controlType=EULER_TRANSFORM, parent=calf_r_fk_ctrl, type=Control, loc=<0 0 0>, rot=<0 0 0 1>}
 ball_r_fk_ctrl {controlType=EULER_TRANSFORM, parent=foot_r_fk_ctrl, type=Control, loc=<0 0 0>, rot=<0 0 0 1>}
 toes_r_space {parent=leg_r_controls_space, type=Null, loc=<15.7991 9.3506 3.1405>, rot=<0.5214 0.4801 0.4735 0.5229>, scale=<-1 1 1>}
 foot_r_ik_ctrl {controlType=EULER_TRANSFORM, parent=leg_r_controls_space, type=Control, loc=<0 0 0>, rot=<0 0 0 1>}
 leg_r_stretch_switch {controlType=BOOL, parent=foot_r_ik_ctrl, type=Control, loc=<0 0 0>, rot=<0 0 0 1>}
 heel_r {parent=foot_r_ik_ctrl, type=Null, loc=<-2.7001 1.0747 -8.8106>, rot=<0.4699 -0.5284 0.543 -0.453>}
 tip_r {parent=heel_r, type=Null, loc=<-20.4653 -0.6281 0>, rot=<0 0 0 1>}
 ball_r_ik_ctrl {controlType=EULER_TRANSFORM, parent=tip_r, type=Control, loc=<0 0 0>, rot=<0 0 0 1>}
 foot_bk1_r {parent=tip_r, type=Null, loc=<7.5178 -2.2898 1.3226>, rot=<0 0 0 1>}
 foot_bk1_r_ctrl {controlType=EULER_TRANSFORM, parent=foot_bk1_r, type=Control, loc=<0 0 0>, rot=<0 0 0 1>}
 foot_ik_r {parent=foot_bk1_r_ctrl, type=Null, loc=<13.8882 -5.8672 1.5071>, rot=<0 0 0 1>}
 foot_roll_r_bfr {parent=foot_bk1_r_ctrl, type=Null, loc=<13.8882 -5.8672 1.5071>, rot=<0 0 0 1>}
 foot_roll_r_ctrl {controlType=EULER_TRANSFORM, parent=foot_roll_r_bfr, type=Control, loc=<0 0 0>, rot=<0 0 0 1>}
 roll_blend_r_ctrl {controlType=FLOAT, parent=foot_roll_r_ctrl, type=Control, loc=<1 0 0>, rot=<0 0 0 1>}
 tip_r_ctrl {controlType=EULER_TRANSFORM, parent=heel_r, type=Control, loc=<0 0 0>, rot=<0 0 0 1>}
 heel_r_ctrl {controlType=EULER_TRANSFORM, parent=foot_r_ik_ctrl, type=Control, loc=<0 0 0>, rot=<0 0 0 1>}
 leg_r_softness {controlType=FLOAT, parent=foot_r_ik_ctrl, type=Control, loc=<0 0 0>, rot=<0 0 0 1>}
 leg_r_pv_ik_ctrl {controlType=EULER_TRANSFORM, parent=leg_r_controls_space, type=Control, loc=<0 0 0>, rot=<0 0 0 1>}
 arm_l_controls_space {parent=body_offset_ctrl, type=Null, loc=<0 0 0>, rot=<0 0 0 1>}
 upperarm_l_fk_ctrl_space {parent=arm_l_controls_space, type=Null, loc=<17.7788 -4.8444 134.4501>, rot=<-0.016 0.5646 -0.014 0.8251>}
 upperarm_l_fk_ctrl {controlType=EULER_TRANSFORM, parent=upperarm_l_fk_ctrl_space, type=Control, loc=<0 0 0>, rot=<0 0 0 1>}
 lowerarm_l_fk_ctrl {controlType=EULER_TRANSFORM, parent=upperarm_l_fk_ctrl, type=Control, loc=<0 0 0>, rot=<0 0 0 1>}
 hand_l_fk_ctrl {controlType=EULER_TRANSFORM, parent=lowerarm_l_fk_ctrl, type=Control, loc=<0 0 0>, rot=<0 0 0 1>}
 arm_l_fk_local {controlType=BOOL, parent=upperarm_l_fk_ctrl, type=Control, loc=<1 0 0>, rot=<0 0 0 1>}
 fingers_l_space {parent=arm_l_controls_space, type=Null, loc=<33.8029 3.8282 90.5694>, rot=<0.5621 0.5887 -0.3223 0.4834>}
 thumb_01_l_ctrl {controlType=EULER_TRANSFORM, parent=fingers_l_space, type=Control, loc=<0 0 0>, rot=<0 0 0 1>}
 thumb_02_l_ctrl {controlType=EULER_TRANSFORM, parent=thumb_01_l_ctrl, type=Control, loc=<0 0 0>, rot=<0 0 0 1>}
 thumb_03_l_ctrl {controlType=EULER_TRANSFORM, parent=thumb_02_l_ctrl, type=Control, loc=<0 0 0>, rot=<0 0 0 1>}
 index_metacarpal_l_ctrl {controlType=EULER_TRANSFORM, parent=fingers_l_space, type=Control, loc=<0 0 0>, rot=<0 0 0 1>}
 index_01_l_ctrl {controlType=EULER_TRANSFORM, parent=index_metacarpal_l_ctrl, type=Control, loc=<0 0 0>, rot=<0 0 0 1>}
 index_02_l_ctrl {controlType=EULER_TRANSFORM, parent=index_01_l_ctrl, type=Control, loc=<0 0 0>, rot=<0 0 0 1>}
 index_03_l_ctrl {controlType=EULER_TRANSFORM, parent=index_02_l_ctrl, type=Control, loc=<0 0 0>, rot=<0 0 0 1>}
 middle_metacarpal_l_ctrl {controlType=EULER_TRANSFORM, parent=fingers_l_space, type=Control, loc=<0 0 0>, rot=<0 0 0 1>}
 middle_01_l_ctrl {controlType=EULER_TRANSFORM, parent=middle_metacarpal_l_ctrl, type=Control, loc=<0 0 0>, rot=<0 0 0 1>}
 middle_02_l_ctrl {controlType=EULER_TRANSFORM, parent=middle_01_l_ctrl, type=Control, loc=<0 0 0>, rot=<0 0 0 1>}
 middle_03_l_ctrl {controlType=EULER_TRANSFORM, parent=middle_02_l_ctrl, type=Control, loc=<0 0 0>, rot=<0 0 0 1>}
 ring_metacarpal_l_ctrl {controlType=EULER_TRANSFORM, parent=fingers_l_space, type=Control, loc=<0 0 0>, rot=<0 0 0 1>}
 ring_01_l_ctrl {controlType=EULER_TRANSFORM, parent=ring_metacarpal_l_ctrl, type=Control, loc=<0 0 0>, rot=<0 0 0 1>}
 ring_02_l_ctrl {controlType=EULER_TRANSFORM, parent=ring_01_l_ctrl, type=Control, loc=<0 0 0>, rot=<0 0 0 1>}
 ring_03_l_ctrl {controlType=EULER_TRANSFORM, parent=ring_02_l_ctrl, type=Control, loc=<0 0 0>, rot=<0 0 0 1>}
 pinky_metacarpal_l_ctrl {controlType=EULER_TRANSFORM, parent=fingers_l_space, type=Control, loc=<0 0 0>, rot=<0 0 0 1>}
 pinky_01_l_ctrl {controlType=EULER_TRANSFORM, parent=pinky_metacarpal_l_ctrl, type=Control, loc=<0 0 0>, rot=<0 0 0 1>}
 pinky_02_l_ctrl {controlType=EULER_TRANSFORM, parent=pinky_01_l_ctrl, type=Control, loc=<0 0 0>, rot=<0 0 0 1>}
 pinky_03_l_ctrl {controlType=EULER_TRANSFORM, parent=pinky_02_l_ctrl, type=Control, loc=<0 0 0>, rot=<0 0 0 1>}
 hand_l_ik_ctrl {controlType=EULER_TRANSFORM, parent=arm_l_controls_space, type=Control, loc=<0 0 0>, rot=<0 0 0 1>}
 arm_l_stretch_switch {controlType=BOOL, parent=hand_l_ik_ctrl, type=Control, loc=<0 0 0>, rot=<0 0 0 1>}
 arm_l_softness {controlType=FLOAT, parent=hand_l_ik_ctrl, type=Control, loc=<0 0 0>, rot=<0 0 0 1>}
 arm_l_pv_ik_ctrl {controlType=EULER_TRANSFORM, parent=arm_l_controls_space, type=Control, loc=<0 0 0>, rot=<0 0 0 1>}
 arm_r_controls_space {parent=body_offset_ctrl, type=Null, loc=<0 0 0>, rot=<0 0 0 1>, scale=<-1 1 1>}
 upperarm_r_fk_ctrl_space {parent=arm_r_controls_space, type=Null, loc=<17.7788 -4.8444 134.4502>, rot=<-0.016 0.5646 -0.014 0.8251>}
 upperarm_r_fk_ctrl {controlType=EULER_TRANSFORM, parent=upperarm_r_fk_ctrl_space, type=Control, loc=<0 0 0>, rot=<0 0 0 1>}
 lowerarm_r_fk_ctrl {controlType=EULER_TRANSFORM, parent=upperarm_r_fk_ctrl, type=Control, loc=<0 0 0>, rot=<0 0 0 1>}
 hand_r_fk_ctrl {controlType=EULER_TRANSFORM, parent=lowerarm_r_fk_ctrl, type=Control, loc=<0 0 0>, rot=<0 0 0 1>}
 arm_r_fk_local {controlType=BOOL, parent=upperarm_r_fk_ctrl, type=Control, loc=<1 0 0>, rot=<0 0 0 1>}
 fingers_r_space {parent=arm_r_controls_space, type=Null, loc=<33.8029 3.8283 90.5694>, rot=<0.5621 0.5887 -0.3223 0.4834>}
 thumb_01_r_ctrl {controlType=EULER_TRANSFORM, parent=fingers_r_space, type=Control, loc=<0 0 0>, rot=<0 0 0 1>}
 thumb_02_r_ctrl {controlType=EULER_TRANSFORM, parent=thumb_01_r_ctrl, type=Control, loc=<0 0 0>, rot=<0 0 0 1>}
 thumb_03_r_ctrl {controlType=EULER_TRANSFORM, parent=thumb_02_r_ctrl, type=Control, loc=<0 0 0>, rot=<0 0 0 1>}
 index_metacarpal_r_ctrl {controlType=EULER_TRANSFORM, parent=fingers_r_space, type=Control, loc=<0 0 0>, rot=<0 0 0 1>}
 index_01_r_ctrl {controlType=EULER_TRANSFORM, parent=index_metacarpal_r_ctrl, type=Control, loc=<0 0 0>, rot=<0 0 0 1>}
 index_02_r_ctrl {controlType=EULER_TRANSFORM, parent=index_01_r_ctrl, type=Control, loc=<0 0 0>, rot=<0 0 0 1>}
 index_03_r_ctrl {controlType=EULER_TRANSFORM, parent=index_02_r_ctrl, type=Control, loc=<0 0 0>, rot=<0 0 0 1>}
 middle_metacarpal_r_ctrl {controlType=EULER_TRANSFORM, parent=fingers_r_space, type=Control, loc=<0 0 0>, rot=<0 0 0 1>}
 middle_01_r_ctrl {controlType=EULER_TRANSFORM, parent=middle_metacarpal_r_ctrl, type=Control, loc=<0 0 0>, rot=<0 0 0 1>}
 middle_02_r_ctrl {controlType=EULER_TRANSFORM, parent=middle_01_r_ctrl, type=Control, loc=<0 0 0>, rot=<0 0 0 1>}
 middle_03_r_ctrl {controlType=EULER_TRANSFORM, parent=middle_02_r_ctrl, type=Control, loc=<0 0 0>, rot=<0 0 0 1>}
 ring_metacarpal_r_ctrl {controlType=EULER_TRANSFORM, parent=fingers_r_space, type=Control, loc=<0 0 0>, rot=<0 0 0 1>}
 ring_01_r_ctrl {controlType=EULER_TRANSFORM, parent=ring_metacarpal_r_ctrl, type=Control, loc=<0 0 0>, rot=<0 0 0 1>}
 ring_02_r_ctrl {controlType=EULER_TRANSFORM, parent=ring_01_r_ctrl, type=Control, loc=<0 0 0>, rot=<0 0 0 1>}
 ring_03_r_ctrl {controlType=EULER_TRANSFORM, parent=ring_02_r_ctrl, type=Control, loc=<0 0 0>, rot=<0 0 0 1>}
 pinky_metacarpal_r_ctrl {controlType=EULER_TRANSFORM, parent=fingers_r_space, type=Control, loc=<0 0 0>, rot=<0 0 0 1>}
 pinky_01_r_ctrl {controlType=EULER_TRANSFORM, parent=pinky_metacarpal_r_ctrl, type=Control, loc=<0 0 0>, rot=<0 0 0 1>}
 pinky_02_r_ctrl {controlType=EULER_TRANSFORM, parent=pinky_01_r_ctrl, type=Control, loc=<0 0 0>, rot=<0 0 0 1>}
 pinky_03_r_ctrl {controlType=EULER_TRANSFORM, parent=pinky_02_r_ctrl, type=Control, loc=<0 0 0>, rot=<0 0 0 1>}
 hand_r_ik_ctrl {controlType=EULER_TRANSFORM, parent=arm_r_controls_space, type=Control, loc=<0 0 0>, rot=<0 0 0 1>}
 arm_r_stretch_switch {controlType=BOOL, parent=hand_r_ik_ctrl, type=Control, loc=<0 0 0>, rot=<0 0 0 1>}
 arm_r_softness {controlType=FLOAT, parent=hand_r_ik_ctrl, type=Control, loc=<0 0 0>, rot=<0 0 0 1>}
 arm_r_pv_ik_ctrl {controlType=EULER_TRANSFORM, parent=arm_r_controls_space, type=Control, loc=<0 0 0>, rot=<0 0 0 1>}
 body_ctrl {controlType=EULER_TRANSFORM, parent=body_offset_ctrl, type=Control, loc=<0 0 0>, rot=<0 0 0 1>}
 hips_ctrl {controlType=EULER_TRANSFORM, parent=body_ctrl, type=Control, loc=<0 0 0>, rot=<0 0 0 1>}
 hips_tan_ctrl {controlType=EULER_TRANSFORM, parent=hips_ctrl, type=Control, loc=<0 0 0>, rot=<0 0 0 1>}
 spine_fk_ik_switch {controlType=BOOL, parent=hips_ctrl, type=Control, loc=<0 0 0>, rot=<0 0 0 1>}
 spine_01_ctrl {controlType=EULER_TRANSFORM, parent=body_ctrl, type=Control, loc=<0 0 0>, rot=<0 0 0 1>}
 spine_02_ctrl {controlType=EULER_TRANSFORM, parent=spine_01_ctrl, type=Control, loc=<0 0 0>, rot=<0 0 0 1>}
 spine_03_ctrl {controlType=EULER_TRANSFORM, parent=spine_02_ctrl, type=Control, loc=<0 0 0>, rot=<0 0 0 1>}
 chest_space {parent=spine_03_ctrl, type=Null, loc=<0 0 0>, rot=<-0.0294 0.7065 0.0294 -0.7065>}
 head_fk_space {parent=chest_space, type=Null, loc=<30.6612 3.6106 0>, rot=<0.0294 -0.7065 -0.0294 -0.7065>}
 head_ctrl {controlType=EULER_TRANSFORM, parent=head_fk_space, type=Control, loc=<0 0 0>, rot=<0 0 0 1>}
 head_local {controlType=BOOL, parent=head_ctrl, type=Control, loc=<1 0 0>, rot=<0 0 0 1>}
 clavicle_l_ctrl {controlType=EULER_TRANSFORM, parent=chest_space, type=Control, loc=<0 0 0>, rot=<0 0 0 1>}
 clavicle_r_ctrl {controlType=EULER_TRANSFORM, parent=chest_space, type=Control, loc=<0 0 0>, rot=<0 0 0 1>}
 neck_01_ctrl {controlType=EULER_TRANSFORM, parent=chest_space, type=Control, loc=<0 0 0>, rot=<0 0 0 1>}
 neck_02_ctrl {controlType=EULER_TRANSFORM, parent=neck_01_ctrl, type=Control, loc=<0 0 0>, rot=<0 0 0 1>}
 head_ik_space {parent=chest_space, type=Null, loc=<30.6612 3.6106 0>, rot=<0.0294 -0.7065 -0.0294 -0.7065>}
 head_ik_ctrl {controlType=EULER_TRANSFORM, parent=head_ik_space, type=Control, loc=<0 0 0>, rot=<0 0 0 1>}
 head_mid {parent=head_ik_ctrl, type=Null, loc=<-0.0001 -1.0459 -5.1749>, rot=<0 0 0 1>}
 head_stretch {controlType=BOOL, parent=head_ik_ctrl, type=Control, loc=<0 0 0>, rot=<0 0 0 1>}
 chest_ctrl {controlType=EULER_TRANSFORM, parent=body_ctrl, type=Control, loc=<0 0 0>, rot=<0 0 0 1>}
 chest_tan_ctrl {controlType=EULER_TRANSFORM, parent=chest_ctrl, type=Control, loc=<0 0 0>, rot=<0 0 0 1>}
 spine_stretch_switch {controlType=BOOL, parent=chest_ctrl, type=Control, loc=<0 0 0>, rot=<0 0 0 1>}
 arm_l_fk_ik_switch {controlType=BOOL, parent=global_ctrl, type=Control, loc=<0 0 0>, rot=<0 0 0 1>}
 leg_l_fk_ik_switch {controlType=BOOL, parent=global_ctrl, type=Control, loc=<1 0 0>, rot=<0 0 0 1>}
 arm_r_fk_ik_switch {controlType=BOOL, parent=global_ctrl, type=Control, loc=<0 0 0>, rot=<0 0 0 1>}
 leg_r_fk_ik_switch {controlType=BOOL, parent=global_ctrl, type=Control, loc=<1 0 0>, rot=<0 0 0 1>}
 ShowBodyControls {controlType=BOOL, parent=global_ctrl, type=Control, loc=<1 0 0>, rot=<0 0 0 1>}
 neck_fk_ik_switch {controlType=BOOL, parent=global_ctrl, type=Control, loc=<0 0 0>, rot=<0 0 0 1>}
 heel_l {loc=<12.5608 -3.3429 0.6325>, rot=<-0.4699 0.5284 -0.543 0.453>}
 tip_l {parent=heel_l, loc=<-20.4653 -0.6281 0>, rot=<0 0 0 1>}
 heel_r {loc=<-12.5608 -3.3429 0.6325>, rot=<-0.4699 -0.5284 0.543 0.453>}
 tip_r {parent=heel_r, loc=<20.4653 -0.6281 0>, rot=<0 0 0 1>}
]
variables [
 L Arm IK Mode {type=bool, default=False}
 "L Leg IK Mode " {type=bool, default=True}
 Neck IK Mode {type=bool, default=False}
 Neck Length {type=double, default=0}
 R Arm IK Mode {type=bool, default=False}
 R Leg IK Mode {type=bool, default=True}
 Spine IK Mode {type=bool, default=False}
 Spine Length {type=double, default=0}
 Spine Offsets {type="TArray<FTransform>"}
]
